
using { /Fortnite.com/Devices }
using { /Verse.org/Simulation }
using { /UnrealEngine.com/Temporary/Diagnostics }
using { /Verse.org/Assets }

Blink := module:
    (Blink:)NPCs<public> :=module:
        (NPCs:)Stylized<public> :=module:
            (Stylized:)Forest_Animals<public> :=module:
                (Forest_Animals:)Boar<public> :=module:
                    (Boar:)Animations<public> :=module:

Content := module:
    (Content:)Textures<public>:=module:
    (Content:)UI<public> := module:
        (UI:)Buttons<public> := module:

Dwarf1:= module:
    (Dwarf1:)NewFolder1<public>:=module:

Props := module:
    (Props:)LavaMite<public> := module:
        (LavaMite:)ReImport1<public> := module:
    (Props:)DarkImp<public> := module: 
       (DarkImp:) Arm18<public> := module:
    (Props:)Murloc<public> := module:
    (Props:)Boar2<public> := module:
        (Boar2:)Baor1<public> := module:
    (Props:)TameBoarBoss<public> := module:
        (TameBoarBoss:)BossBoar<public> := module:
    (Props:)buffalo<public> := module:
        (buffalo:)Import6<public> := module:
    (Props:)KoboldWarrior<public> := module:
        (KoboldWarrior:)NewFolder2<public> := module:
    (Props:)Hyena<public> := module:
        (Hyena:)NewFolder10<public> := module:
    (Props:)DesertGolem<public> := module:
        (DesertGolem:)NewFolder<public> := module:
    (Props:)Scorpion<public> := module:
        (Scorpion:)NewFolder<public> := module:
    (Props:)Bear1<public> := module:
        (Bear1:)NewFolder2<public> := module:
    (Props:)GiantBeast<public> := module:
    (Props:)TimberWolf<public> := module:
        (TimberWolf:)Timber10<public> := module:
    (Props:)SnowLeapard<public> := module:
        (SnowLeapard:)NewFolder8<public> := module:
    (Props:)Ram<public> := module:
        (Ram:)NewFolder9<public> := module:
    (Props:)LavaDragon<public> := module:
        (LavaDragon:)NewFolder3<public> := module:
    (Props:)terravore<public> := module:
        (terravore:)NewFolder7<public> := module:
    (Props:)Yeti<public> := module:
    (Props:)LavaJaw1<public> := module:
        (LavaJaw1:)NewFolder5<public> := module:
    (Props:)desertwyvern<public> := module:
        (desertwyvern:)NewFolder4<public> := module:
        (desertwyvern:)Shrinkingnew1<public> := module:
        (desertwyvern:)theOne2<public> := module:
        
    (Props:)ArcticColosus<public> := module:
        (ArcticColosus:)NewFolder<public> := module:
    (Props:)SerenitySentinel<public> := module:
        (SerenitySentinel:)NewFolder2<public> := module:
    (Props:)HellHound<public> := module:
        (HellHound:)NewFolder<public> := module:
    (Props:)MagmaBoar<public> := module:
        (MagmaBoar:)Resized1<public> := module:


    (Props:)Spider<public> := module:
        (Spider:)NewFolder1<public> := module:
        
        
        
    
NPCData<public> := module:
    NPCTypes<public>:= module:
        #Enum to define home type
        npc_types<public>:= enum{None, LavaMite, DarkImp, Murloc, Boar, BossBoar, Buffalo, KoboldWarrior, KoboldArcher, RealBoar, Hyena, Dwarf, DesertBoss, Scorpion, PolarBear, GiantBeast,
        TimberWolf, SnowLeopard, Ram, LavaDragon, Terravore, Yeti, LavaJaw, DesertWyvern, SerenitySentinel, ArcticColosus, HellHound, MagmaBoar}

        ToString<public>(npc_type: npc_types): string = 
            case(npc_type):
                npc_types.None => return "None"
                npc_types.LavaMite => return "LavaMite"
                npc_types.DarkImp => return "DarkImp"
                npc_types.Murloc => return "Murloc"
                npc_types.Boar => return "Boar"
                npc_types.BossBoar => return "BossBoar"
                npc_types.Buffalo => return "Buffalo"
                npc_types.KoboldWarrior => return "KoboldWarrior"
                npc_types.KoboldArcher => return "KoboldArcher"
                npc_types.RealBoar => return "RealBaor"
                npc_types.Hyena => return "Hyena"
                npc_types.Dwarf => return "Dwarf1"
                npc_types.DesertBoss => return "DesertBoss"
                npc_types.Scorpion => return "Scorpion"
                npc_types.PolarBear => return "PolarBear"
                npc_types.GiantBeast => return "GiantBeast"
                npc_types.TimberWolf => return "TimberWolf"
                npc_types.SnowLeopard => return "SnowLeopard"
                npc_types.Ram => return "Ram"
                npc_types.LavaDragon => return "LavaDragon"
                npc_types.Terravore => return "Terravore"
                npc_types.Yeti => return "Yeti"
                npc_types.LavaJaw => return "LavaJaw"
                npc_types.DesertWyvern => return "DesertWyvern"
                npc_types.SerenitySentinel => return "SerenitySentinel"
                npc_types.ArcticColosus => return "ArcticColosus"
                npc_types.HellHound => return "HellHound"
                npc_types.MagmaBoar => return "MagmaBoar"


    NPCStates<public> := module:
        #Enum to define the States avaiable fo NPCs
        npc_states<public> := enum{Idle, GoToLocation, Attack, RangedAttack, Charge, Tornado, Death}

        # Could probably put this at top level or in another utils/extension file
        ToString<public>(State: npc_states):string=
            case(State):
                npc_states.Idle => return "Idle"
                npc_states.GoToLocation => return "GoToLocation"
                npc_states.Attack => return "Attack"
                npc_states.RangedAttack => return "RangedAttack"
                npc_states.Charge => return "Charge"
                npc_states.Tornado => return "Tornado"
                npc_states.Death => return "Death"

        GetAllStates<public>():[]npc_states = 
            return array{
                npc_states.Idle,
                npc_states.GoToLocation,
                npc_states.Attack,
                npc_states.RangedAttack,
                npc_states.Charge,
                npc_states.Tornado,
                npc_states.Death}

        # Mapping State -> Animation Sequence for each npc_types
        lava_might_npc_animations<public> : [npc_states]animation_sequence = map{
            npc_states.Idle => Props.LavaMite.ReImport1.idle
            npc_states.GoToLocation => Props.LavaMite.ReImport1.walk
            npc_states.Attack => Props.LavaMite.ReImport1.attack
            npc_states.Death => Props.LavaMite.ReImport1.death
        }
        dark_imp_npc_animations<public> : [npc_states]animation_sequence = map{
            npc_states.Idle => Props.DarkImp.Arm18.idle
            npc_states.GoToLocation => Props.DarkImp.Arm18.walk
            npc_states.Attack => Props.DarkImp.Arm18.attack
            npc_states.Death => Props.DarkImp.Arm18.death
        }
        murloc_npc_animations<public> : [npc_states]animation_sequence = map{
            npc_states.Idle => Props.Murloc.idle
            npc_states.GoToLocation => Props.Murloc.walk
            npc_states.RangedAttack => Props.Murloc.throw
            npc_states.Death => Props.Murloc.land_death
        }
        boar_npc_animations<public> : [npc_states]animation_sequence = map{
            npc_states.Idle => Props.Boar2.Baor1.idle
            npc_states.GoToLocation => Props.Boar2.Baor1.walk
            npc_states.Attack => Props.Boar2.Baor1.rear
            npc_states.Death => Props.Boar2.Baor1.death
        }
        boss_boar_npc_animations<public> : [npc_states]animation_sequence = map{
            npc_states.Idle => Props.TameBoarBoss.BossBoar.idle
            npc_states.GoToLocation => Props.TameBoarBoss.BossBoar.walk
            npc_states.Attack => Props.TameBoarBoss.BossBoar.rear
            npc_states.Charge => Props.TameBoarBoss.BossBoar.charge
            npc_states.Death => Props.TameBoarBoss.BossBoar.death
        }
        buffalo_npc_animations<public> : [npc_states]animation_sequence = map{
            npc_states.Idle => Props.buffalo.Import6.idle
            npc_states.GoToLocation => Props.buffalo.Import6.walk
            npc_states.Attack => Props.buffalo.Import6.rear
            npc_states.Death => Props.buffalo.Import6.death
        }
        kobold_warrior_npc_animations<public> : [npc_states]animation_sequence = map{
            npc_states.Idle => Props.KoboldWarrior.NewFolder2.idle
            npc_states.GoToLocation => Props.KoboldWarrior.NewFolder2.walk
            npc_states.Attack => Props.KoboldWarrior.NewFolder2.attack
            npc_states.Death => Props.KoboldWarrior.NewFolder2.death
        }
        kobold_archer_npc_animations<public> : [npc_states]animation_sequence = map{
            npc_states.Idle => KoboldArcherNew.idle
            npc_states.GoToLocation => KoboldArcherNew.walk
            npc_states.RangedAttack => KoboldArcherNew.attack
            npc_states.Attack => KoboldArcherNew.attack
            npc_states.Death => KoboldArcherNew.idle
        }
        hyena_npc_animations<public> : [npc_states]animation_sequence = map{
            npc_states.Idle => Props.Hyena.NewFolder10.idle
            npc_states.GoToLocation => Props.Hyena.NewFolder10.walk
            npc_states.RangedAttack => Props.Hyena.NewFolder10.bite
            npc_states.Attack => Props.Hyena.NewFolder10.bite
            npc_states.Death => Props.Hyena.NewFolder10.death
        }
        desert_boss_npc_animations<public> : [npc_states]animation_sequence = map{
            npc_states.Idle => Props.DesertGolem.NewFolder.idle
            npc_states.GoToLocation => Props.DesertGolem.NewFolder.walk
            npc_states.RangedAttack => Props.DesertGolem.NewFolder.skill
            npc_states.Attack => Props.DesertGolem.NewFolder.attack
            npc_states.Death => Props.DesertGolem.NewFolder.death
        }
        scorpion_npc_animations<public> : [npc_states]animation_sequence = map{
            npc_states.Idle => Props.Scorpion.NewFolder.idle
            npc_states.GoToLocation => Props.Scorpion.NewFolder.walk
            npc_states.RangedAttack => Props.Scorpion.NewFolder.attack_poison
            npc_states.Attack => Props.Scorpion.NewFolder.attack_poison
            npc_states.Death => Props.Scorpion.NewFolder.death
        }
        polar_bear_npc_animations<public> : [npc_states]animation_sequence = map{
            npc_states.Idle => Props.Bear1.NewFolder2.idle
            npc_states.GoToLocation => Props.Bear1.NewFolder2.walk
            npc_states.RangedAttack => Props.Bear1.NewFolder2.attack
            npc_states.Attack => Props.Bear1.NewFolder2.attack
            npc_states.Death => Props.Bear1.NewFolder2.death
        }
        giant_beast_npc_animations<public> : [npc_states]animation_sequence = map{
            npc_states.Idle => Props.GiantBeast.idle
            npc_states.GoToLocation => Props.GiantBeast.walk
            npc_states.RangedAttack => Props.GiantBeast.attack
            npc_states.Attack => Props.GiantBeast.attack
            npc_states.Death => Props.GiantBeast.death
        }
        timber_wolf_npc_animations<public> : [npc_states]animation_sequence = map{
            npc_states.Idle => Props.TimberWolf.Timber10.idle
            npc_states.GoToLocation => Props.TimberWolf.Timber10.walk
            npc_states.RangedAttack => Props.TimberWolf.Timber10.attack_2
            npc_states.Attack => Props.TimberWolf.Timber10.attack_2
            npc_states.Death => Props.TimberWolf.Timber10.death
        }
        snow_leopard_npc_animations<public> : [npc_states]animation_sequence = map{
            npc_states.Idle => Props.SnowLeapard.NewFolder8.idle
            npc_states.GoToLocation => Props.SnowLeapard.NewFolder8.walk
            npc_states.RangedAttack => Props.SnowLeapard.NewFolder8.attack_static
            npc_states.Attack => Props.SnowLeapard.NewFolder8.attack_static
            npc_states.Death => Props.SnowLeapard.NewFolder8.death
        }
        ram_npc_animations<public> : [npc_states]animation_sequence = map{
            npc_states.Idle => Props.Spider.NewFolder1.idle
            npc_states.GoToLocation => Props.Spider.NewFolder1.walk
            npc_states.RangedAttack => Props.Spider.NewFolder1.attack
            npc_states.Attack => Props.Spider.NewFolder1.attack
            npc_states.Death => Props.Spider.NewFolder1.death
        }
        lava_dragon_npc_animations<public> : [npc_states]animation_sequence = map{
            npc_states.Idle => Props.LavaDragon.NewFolder3.waiting
            npc_states.GoToLocation => Props.LavaDragon.NewFolder3.walkingmove
            npc_states.RangedAttack => Props.LavaDragon.NewFolder3.meleehit
            npc_states.Attack => Props.LavaDragon.NewFolder3.meleehit
            npc_states.Death => Props.LavaDragon.NewFolder3.dying
        }
        terravore_npc_animations<public> : [npc_states]animation_sequence = map{
            npc_states.Idle => Props.terravore.NewFolder7.idle
            npc_states.GoToLocation => Props.terravore.NewFolder7.run
            npc_states.RangedAttack => Props.terravore.NewFolder7.attack_left
            npc_states.Attack => Props.terravore.NewFolder7.attack_left
            npc_states.Death => Props.terravore.NewFolder7.death
        }
        yeti_npc_animations<public> : [npc_states]animation_sequence = map{
            npc_states.Idle => Props.Yeti.idle
            npc_states.GoToLocation => Props.Yeti.walk
            npc_states.RangedAttack => Props.Yeti.roar
            npc_states.Attack => Props.Yeti.roar
        }
        lava_jaw_npc_animations<public> : [npc_states]animation_sequence = map{
            npc_states.Idle => Props.LavaJaw1.NewFolder5.waiting
            npc_states.GoToLocation => Props.LavaJaw1.NewFolder5.walking
            npc_states.RangedAttack => Props.LavaJaw1.NewFolder5.fireball
            npc_states.Attack => Props.LavaJaw1.NewFolder5.fireball
            npc_states.Death => Props.LavaJaw1.NewFolder5.dying
        }
        desert_wyvern_npc_animations<public> : [npc_states]animation_sequence = map{
            npc_states.Idle => Props.desertwyvern.theOne2.idle
            npc_states.GoToLocation => Props.desertwyvern.theOne2.walk
            npc_states.RangedAttack => Props.desertwyvern.theOne2.throw
            npc_states.Attack => Props.desertwyvern.theOne2.attack
            npc_states.Death => Props.desertwyvern.theOne2.death
            npc_states.Charge => Props.desertwyvern.theOne2.quicksand
        }
        serenity_sentinel_npc_animations<public> : [npc_states]animation_sequence = map{
            npc_states.Idle => Props.SerenitySentinel.NewFolder2.idle
            npc_states.GoToLocation => Props.SerenitySentinel.NewFolder2.walking
            npc_states.RangedAttack => Props.SerenitySentinel.NewFolder2.punch
            npc_states.Attack => Props.SerenitySentinel.NewFolder2.punch
            npc_states.Tornado => Props.SerenitySentinel.NewFolder2.shock
            npc_states.Death => Props.SerenitySentinel.NewFolder2.death
        }
        arctic_colosus_npc_animations<public> : [npc_states]animation_sequence = map{
            npc_states.Idle => Props.ArcticColosus.idle
            npc_states.GoToLocation => Props.ArcticColosus.walk
            npc_states.RangedAttack => Props.ArcticColosus.skill_ice_age
            npc_states.Attack => Props.ArcticColosus.skill_ice_age
            npc_states.Death => Props.ArcticColosus.death
        }
        hell_hound_npc_animations<public> : [npc_states]animation_sequence = map{
            npc_states.Idle => Props.HellHound.NewFolder.idle
            npc_states.GoToLocation => Props.HellHound.NewFolder.walk
            npc_states.RangedAttack => Props.HellHound.NewFolder.attack_left
            npc_states.Attack => Props.HellHound.NewFolder.attack_left
            npc_states.Death => Props.HellHound.NewFolder.death
        }
        magma_boar_npc_animations<public> : [npc_states]animation_sequence = map{
            npc_states.Idle => Props.MagmaBoar.Resized1.idle
            npc_states.GoToLocation => Props.MagmaBoar.Resized1.walk
            npc_states.RangedAttack => Props.MagmaBoar.Resized1.attack
            npc_states.Attack => Props.MagmaBoar.Resized1.attack
            npc_states.Tornado => Props.MagmaBoar.Resized1.rodeo
            npc_states.Death => Props.MagmaBoar.Resized1.death
        }
        
        

