
using { /Fortnite.com/Devices }
using { /Verse.org/Simulation }
using { /UnrealEngine.com/Temporary/Diagnostics }


areas_device := class(creative_device):

    @editable
    AreaBarriers : []barrier_device = array{}

    @editable
    BossBarriers : []barrier_device = array{}

    AreaNames : []string = array{"Peaceful Ridge", "Pleasant Valley", "Enchanted Forest","Scorched Plains","Radiant Dunes","Dusty Wastes","Frostwind Pass","Frostbite Peaks","Chillfrost Grove","Scorchpath","Magma Crossing","Cinder Basin"}

    @editable
    HMD : hud_message_device = hud_message_device{}

    storedstring<localizes>(S:string) : message = "{S}"


    UnlockArea(Agent:agent,i:int):void=
        if:
            MyData := Data[player[Agent]]
            AreaBarrier := AreaBarriers[i]
        then:
            AreaBarrier.AddToIgnoreList(Agent)
            if:
                var NewStat : []logic = MyData.AreasUnlocked
                NewStat[i] = false
                set NewStat[i] = true
                set Data[player[Agent]] = player_data{player_data_update<constructor>(MyData),AreasUnlocked := NewStat}
                AreaName := AreaNames[i]
            then:
                HMD.SetText(storedstring("{AreaName}"))
                HMD.Show(Agent)

    UlockBossArea(Agent:agent,i:int):void=
        if:
            MyData := Data[player[Agent]]
            AreaBarrier := BossBarriers[i]
        then:
            AreaBarrier.AddToIgnoreList(Agent)

    LockAreas(Agent:agent):void=
        for(i:=1..AreaBarriers.Length-1):
            if(AreaBarrier := AreaBarriers[i]):
                AreaBarrier.RemoveFromIgnoreList(Agent)


        




    
