
using { /Fortnite.com/Devices }
using { /Verse.org/Simulation }
using { /UnrealEngine.com/Temporary/Diagnostics }
using { /UnrealEngine.com/Temporary/UI }
using { /Fortnite.com/UI }
using { /UnrealEngine.com/Temporary/SpatialMath }
using { /Verse.org/Colors }
using { /Verse.org/Assets }


Combat_Level_Hud := class():
  #@editable
    #VfxLevelSwitch : switch_device = switch_device{}
    #@editable
    #Audio : audio_player_device = audio_player_device{}


    var CurrentLevel : int = 1

    var iTotalXp : float = 32.0
    var iCurrentXp: float = 0.0
    var iBossName: string = "Xp"
    
    BossBarColorBase<private> : texture_block = texture_block{
        DefaultImage := hud.XpBarBackground2
        #DefaultColor := MakeColorFromHex("000000")
        DefaultDesiredSize := vector2{X:= 320.0, Y := 30.0}} 

    var BossBarColorTop<private> : texture_block = texture_block{DefaultImage := hud.XpBar7}  
    var Overlay<private> : overlay = overlay{}
    var Overlay2<private> : overlay = overlay{}
    var Overlay3<private> : overlay = overlay{}
        
    var XpBarTitleTextBlock<private> : text_block = text_block{DefaultTextColor := MakeColorFromHex("FFC000"), DefaultShadowColor:={MakeColorFromHex("5b5b5b")}, DefaultShadowOffset:=option{vector2{X:=2.0,Y:=2.0}}}
    var LevelTitleTextBlock<private> : text_block = text_block{DefaultTextColor := MakeColorFromHex("FFC000"), DefaultShadowColor:={MakeColorFromHex("5b5b5b")}, DefaultShadowOffset:=option{vector2{X:=2.0,Y:=2.0}}}
    var LevelTitleTextBlock2<private> : text_block = text_block{DefaultTextColor := MakeColorFromHex("5b5b5b"), DefaultShadowColor:={MakeColorFromHex("5b5b5b")}, DefaultShadowOffset:=option{vector2{X:=2.0,Y:=2.0}}}
    var XpBarTitleTextBlock2<private> : text_block = text_block{DefaultTextColor := MakeColorFromHex("5b5b5b"), DefaultShadowColor:={MakeColorFromHex("5b5b5b")}, DefaultShadowOffset:=option{vector2{X:=2.0,Y:=2.0}}}

    HPTitleText<private><localizes>(HPText : string) : message = "{HPText}"

    Init(Agent:agent):void=

        # Loades from Saved Data
        if (Player := player[Agent]):
            if(PlayerLevel := Data[Player].Level):
                if(XpCurrent := Data[Player].Xp):
                    #set iBossName = BossName
                    set iTotalXp = CalculateExpReq(PlayerLevel)
                    set iCurrentXp = XpCurrent
                    set CurrentLevel = PlayerLevel

        #HealthPercentage := iCurrentXp / iTotalXp
        #BarSize := BossBarColorBase.GetDesiredSize()
        #NewHealthBarWidth := BarSize.X * HealthPercentage
        #if (Current := Int[iCurrentXp], Total := Int[iTotalXp]):

            # 2 Blocks are used for the background
        #    XpBarTitleTextBlock.SetText(HPTitleText("{Current}/{Total}"))
        #    LevelTitleTextBlock.SetText(HPTitleText("Level: {CurrentLevel}"))
        #    XpBarTitleTextBlock2.SetText(HPTitleText("{Current}/{Total}"))
        #    LevelTitleTextBlock2.SetText(HPTitleText("Level: {CurrentLevel}"))
        
        #UpdateHealthBar(NewHealthBarWidth)


        #spawn:
        #    DelayedFix(Agent)

    AddXp(XpAmount : float, Agent:?agent):logic=

        var LeveledUp : logic = false

        if(ValidAgent := Agent?):
            if (Player := player[ValidAgent]):
                if(PlayerLevel := Data[Player].Level):
                    if(var XpCurrent : float = Data[Player].Xp):

                        if:
                            OldData := Data[Player]
                            set Data[Player] = player_data{player_data_update<constructor>(OldData),Xp := OldData.Xp + XpAmount}
                            
                        # used to tell if player leveled
                        set XpCurrent += XpAmount
                        if (XpCurrent < 0.0):
                            set XpCurrent = 0.0
                        if:
                            XpCurrent >= iTotalXp
                            OldData := Data[Player]
                            set Data[Player] = player_data{player_data_update<constructor>(OldData),Level := OldData.Level + 1,Xp := 0.0}
                            ExtraXp :=  XpCurrent - iTotalXp
                        then:
                            #Print("Leveled up")
                            set LeveledUp = true
                            set iCurrentXp = 0.0
                            set CurrentLevel = CurrentLevel + 1
                            set iTotalXp = CalculateExpReq(PlayerLevel+1)

                            

                            AddXp(ExtraXp,Agent)
                
                        HealthPercentage := XpCurrent / iTotalXp
                        BarSize := BossBarColorBase.GetDesiredSize()
                        NewHealthBarWidth := BarSize.X * HealthPercentage
                        if (Current := Int[XpCurrent], Total := Int[iTotalXp]):
                            #XpBarTitleTextBlock.SetText(HPTitleText("{Current}/{Total}"))
                            #LevelTitleTextBlock.SetText(HPTitleText("Lv: {CurrentLevel}"))
                            #XpBarTitleTextBlock2.SetText(HPTitleText("{Current}/{Total}"))
                            #LevelTitleTextBlock2.SetText(HPTitleText("Lv: {CurrentLevel}"))
                        
                        #UpdateHealthBar(NewHealthBarWidth)

                        

        if(LeveledUp = true):
            return true
        else:
            return false

    ShowUIForPlayer(Player : player):void=

        if (PlayerUI := GetPlayerUI[Player]):
            PlayerUI.AddWidget(CreateUI())

        if(Agent:=agent[Player]):
            spawn:
                DelayedFix(Agent)

        #spawn:
        #    RemoveUIForPlayerDelayed(Player)
        

    RemoveUIForPlayer(Player:player)<suspends>:void=
        
        Overlay.RemoveWidget(XpBarTitleTextBlock)
        Overlay.RemoveWidget(LevelTitleTextBlock)
        Overlay.RemoveWidget(BossBarColorTop)
        Overlay.RemoveWidget(BossBarColorBase)

        if (PlayerUI := GetPlayerUI[Player]):
            PlayerUI.RemoveWidget(Overlay)

    UpdateHealthBar(NewWidth : float):void=

        Overlay3.RemoveWidget(XpBarTitleTextBlock2)
        Overlay3.RemoveWidget(LevelTitleTextBlock2)
        Overlay2.RemoveWidget(XpBarTitleTextBlock)
        Overlay2.RemoveWidget(LevelTitleTextBlock)
        Overlay.RemoveWidget(BossBarColorTop)

        set BossBarColorTop = CreateTopHealthBar(NewWidth)

        Overlay.AddWidget(CreateOverlaySlot(BossBarColorTop, horizontal_alignment.Left, margin{Top := 0.0, Left := 1.0, Right := 0.0, Bottom := 0.0}))
        Overlay2.AddWidget(CreateOverlaySlot(XpBarTitleTextBlock,horizontal_alignment.Right, margin{Top := 0.0, Left := 0.0, Right := 0.0, Bottom := 0.0}))
        Overlay2.AddWidget(CreateOverlaySlot(LevelTitleTextBlock,horizontal_alignment.Left , margin{Top := 0.0, Left := 0.0, Right := 2.0, Bottom := 0.0}))
        Overlay3.AddWidget(CreateOverlaySlot(XpBarTitleTextBlock2,horizontal_alignment.Right, margin{Top := 2.0, Left := 0.0, Right := 0.0, Bottom := 0.0}))
        Overlay3.AddWidget(CreateOverlaySlot(LevelTitleTextBlock2,horizontal_alignment.Left , margin{Top := 2.0, Left := 2.0, Right := 0.0, Bottom := 0.0}))
        
    CreateOverlaySlot<private>(TheWidget : widget, HAlignment : horizontal_alignment, padding : margin):overlay_slot=  
        overlay_slot:
            Widget := TheWidget 
            HorizontalAlignment := HAlignment
            Padding := padding

    CreateTopHealthBar<private>(Width: float):texture_block=

        ColorBlock := texture_block{
            DefaultImage := hud.XpBar5
            #DefaultColor := MakeColorFromHex("#7EC8E3")
            DefaultDesiredSize := vector2{X:= Width, Y := 30.0}} 
        
        set BossBarColorTop = ColorBlock
        

    CreateOverlay<private>() : overlay=
        TheOverlay := overlay:
            Slots := array:
                overlay_slot:
                    Widget := BossBarColorBase
                overlay_slot:
                    Widget := BossBarColorTop
                    Padding := margin{Top := 0.0, Left := 0.0, Right := 0.0, Bottom := 0.0}
                #overlay_slot:
                #    Widget := XpBarTitleTextBlock
                #overlay_slot:
                #    Widget := LevelTitleTextBlock
        set Overlay = TheOverlay

    CreateOverlay2<private>() : overlay = 

        OverlayWidget : overlay = CreateOverlay3()

        TheOverlay := overlay:
            Slots := array:
                overlay_slot:
                    Widget := XpBarTitleTextBlock
                    HorizontalAlignment := horizontal_alignment.Right
                    
                overlay_slot:
                    Widget := LevelTitleTextBlock
                    HorizontalAlignment := horizontal_alignment.Left
                overlay_slot:
                    Widget := OverlayWidget
                    HorizontalAlignment := horizontal_alignment.Fill

        set Overlay2 = TheOverlay

    CreateOverlay3<private>() : overlay = 
        TheOverlay := overlay:
            Slots := array:
                overlay_slot:
                    Widget := XpBarTitleTextBlock2
                    HorizontalAlignment := horizontal_alignment.Right
                    
                overlay_slot:
                    Widget := LevelTitleTextBlock2
                    HorizontalAlignment := horizontal_alignment.Left
        set Overlay3 = TheOverlay

    CreateUI<private>() : canvas=
        TheCanvas : canvas = canvas:
            Slots := array:
                canvas_slot:
                    Anchors := anchors{Minimum := vector2{X := 0.99, Y := 0.4}, Maximum := vector2{X := 0.99, Y := 0.4}}
                    Offsets := margin{Top := 0.0, Left := 0.0, Right := 0.0, Bottom := 0.0}
                    Alignment := vector2{X := 1.0, Y := 0.5}
                    SizeToContent := true
                    Widget := stack_box:
                        Orientation := orientation.Vertical
                        Slots := array:
                            stack_box_slot:
                                Widget := CreateOverlay2()
                                HorizontalAlignment := horizontal_alignment.Fill
                            stack_box_slot:
                                Widget := CreateOverlay()
                                HorizontalAlignment := horizontal_alignment.Fill
                                                                
        return TheCanvas    


    CalculateExpReq(Level:int):float = 

        Levelf := Level *1.0
        var ExpReq : float = ((Levelf+70.0) * Pow(2.0,(Levelf/3.0)))

        if:
            Level5Mod := Floor(Level/5)
            XpMultiple := 8.0 # Exponent Base
            Level5Mod > 0
        then:
            set ExpReq =  ExpReq * Pow(XpMultiple, (Level5Mod * 1.0))

        return ExpReq
 
        #var total : float = 0.0
        #for(i:=1..Level):
        #    #o:= i*1.0 + 600.0
        #    Power := Pow(1.9,(i*1.0) / 4.0)*1.0
        #    if (FloorThing := Floor[(i*1.0)+ 88.0 * Power * 1.0]):
        #        set total = total + (FloorThing*1.0)
        #if(ReturnThing := Floor[total/4.0]*1.0):
        #    return ReturnThing
        #else:
        #    0.0

    DelayedFix(Agent:agent)<suspends>:void = 
        Sleep(0.01)
        #Print("adding xp")
        AddXp(0.0, option{Agent})

combat_level_device := class(creative_device):

    var XpBarPerPlayer : [agent]?Combat_Level_Hud = map{}

    
    @editable
    HudMessage : hud_message_device = hud_message_device{}

    @editable
    Spawners : []player_spawner_device = array{}
    @editable
    TestButton : button_device = button_device{}

    @editable
    vfx : trigger_device = trigger_device{}
    @editable 
    audio : audio_player_device = audio_player_device{}
    @editable
    Score : score_manager_device = score_manager_device{}
    @editable
    LevelUpAccolade : accolades_device = accolades_device{}

    @editable
    CosmeticUnlock : cosmetic_unlock_device = cosmetic_unlock_device{}
    @editable
    AreaDevice : areas_device = areas_device{}

    @editable
    LevelUpMissionTrigger : trigger_device = trigger_device{}

    var CanvasPerPlayer : [agent]?canvas = map{}
    var HitPerPlayer : [agent]?hud_rising_animation = map{}

    storedstringpopup<localizes>(INT: int) : message = "+{INT}XP"

    var xpcounter : int = 0
    var CurrentLevel : int = 1
    var delayenabled : logic = false

    XpDropIcon : texture = Icons.DefaultIcon
    GoldDropIcon : texture = wooden_pixel_gui.textures.Icons2.Size8.t_CoinGold_8

    IncreasePerLevelUp : float = 0.14

    OnBegin<override>()<suspends>:void=

        # Need to add an init function / delay was the old fix
        #Sleep(1.0)

            
        Players:=GetPlayspace().GetPlayers()
        for(Player:Players):
            if(Agent:=agent[Player]):
                Pressed(Agent)

        for(i:=0..Spawners.Length-1):
            if(Spawner:= Spawners[i]):
                Spawner.SpawnedEvent.Subscribe(Pressed)
 
    Pressed(Agent:agent):void =

        if(Player := player[Agent]):
            if(Testing := XpBarPerPlayer[Agent]):
            else:
                XpBar : Combat_Level_Hud = Combat_Level_Hud{}
                XpBar.Init(Agent)
                if(set XpBarPerPlayer[Agent] = option{XpBar}){}


                #if(XpBar2 := XpBarPerPlayer[Agent]?): was used for old xp bar
                    #XpBar2.Init("XP", Agent)
                    #XpBar2.ShowUIForPlayer(Player)

    AddXpMaster(Agent:agent, XpAdding:float):void =
        if(Player:= player[Agent]):
            if(XpBar := XpBarPerPlayer[Agent]?):
                #XpBar.ShowUIForPlayer(Player)
                var LeveledUp : logic = false
                set LeveledUp = XpBar.AddXp(XpAdding, option{Agent})
                if(LeveledUp = true):

                    Print("Leveled Up")
                    LevelUpMissionTrigger.Trigger(Agent)
                    #GrantGoldOnLevelUp(Agent)
                    #CosmeticOnLevelUp(Agent)
                    AreasOnLevelUp(Agent)
                    #BossAreasOnLevelUp(Agent)
    
                    HudMessage.Show(Agent)
                    LevelUpAudio(Agent)

                if(Rounded : int = Round[XpAdding]):
                    #ElimHandle(Agent,Rounded) # show xp gained

    LevelUpAudio(Agent:agent):void= # used by prestige device
        audio.Play(Agent)
        vfx.Trigger(Agent)
        LevelUpAccolade.Award(Agent)


    CosmeticLevels : []int = array{10,15,20,25,30,35,40}
    UnlockLevelsCosmeticID : []int = array{12,1,13,2,14,15,3}

    CosmeticOnLevelUp(Agent:agent):void=
        for(i:=0..CosmeticLevels.Length-1):
            if(MyData := Data[player[Agent]]):
                if(LevelReq:=CosmeticLevels[i]):
                    Print("UnlockLevels{LevelReq}")
                    Print("level{MyData.Level}")
                    #if(MyData.Level >= LevelReq):
                        #if(CosmeticId := UnlockLevelsCosmeticID[i]):
                            #CosmeticUnlock.UnlockCosmetic(Agent,CosmeticId)

    AreaUnlockLevels : []int = array{1,5,10,15,20,25,30,35,40,45,50,55}
    BossAreaUnlockLevels : []int = array{7,17,27}

    AreasOnLevelUp(Agent:agent):void=
        for(i:=0..AreaUnlockLevels.Length-1):
            if(MyData := Data[player[Agent]]):
                if(LevelReq:=AreaUnlockLevels[i]):
                    if(MyData.Level >= LevelReq):
                        Print("Unlocking Area : {i}")
                        AreaDevice.UnlockArea(Agent,i)


    BossAreasOnLevelUp(Agent:agent):void=
        for(i:=0..BossAreaUnlockLevels.Length-1):
            if(MyData := Data[player[Agent]]):
                if(LevelReq:=BossAreaUnlockLevels[i]):
                    if(MyData.Level >= LevelReq):
                        AreaDevice.UlockBossArea(Agent,i)

    GrantGoldOnLevelUp(Agent:agent):void=
        if(XpBar := XpBarPerPlayer[Agent]?):

            var GoldMult : float = 0.0
            var PlusGold : float = 0.0

            if(PlayerData := Data[Agent]):
                if(PlayerData.GoldPerElim > 0):
                    #set GoldMult = IncreasePerLevelUp * PlayerData.GoldPerLevel

                set PlusGold =  (XpBar.iTotalXp / 6.0) + ((XpBar.iTotalXp / 6.0) * PlayerData.GoldPerElim) # use to be GoldMult
                        
            if(Player:= player[Agent]):

                if:
                    OldData := Data[Player]
                    set Data[Player] = player_data{player_data_update<constructor>(OldData),Gold := OldData.Gold + PlusGold, PrestigeEarnings:= OldData.PrestigeEarnings + PlusGold}
                

    ResetTimerOnXpText(xpwhensent : int)<suspends>:void = 
        Sleep(0.5)
        if(xpcounter = xpwhensent):
            set xpcounter = 0
            
    XpDelay()<suspends>:void = 
        Sleep(0.5)
        set delayenabled = false

    MakeBasicCanvas(XpGranted:float):tuple(canvas,hud_rising_animation)= # xp drop

        MyHit:=MakeProgressBar(1.0,vector2{X:=150.0,Y:=400.0},MakeColorFromSRGB(1.0,1.0,1.0),XpGranted, GoldDropIcon)
        MyWidget1:=MyHit.GetWidget1()
        MyWidget2:=MyHit.GetWidget2()
        MyWidget3:=MyHit.GetWidget3()
        MyWidget4:=MyHit.GetWidget4()

        Canvas:=canvas:
            Slots:=array:
                canvas_slot:
                    Anchors := anchors{Minimum := vector2{X := 0.08, Y := 0.45}, Maximum := vector2{X := 0.08, Y := 0.45}}
                    Offsets := margin{Top := 0.0, Left := 0.0, Right := 0.0, Bottom := 0.0}
                    Alignment := vector2{X := 0.0, Y := 0.0}
                    Widget := stack_box:
                        Orientation := orientation.Horizontal
                        Slots := array:
                            stack_box_slot:
                                Widget := MyWidget1
                            stack_box_slot:
                                Widget := MyWidget2
                            stack_box_slot:
                                Widget := MyWidget3
                            stack_box_slot:
                                Widget := MyWidget4

        return(Canvas,MyHit)

    ElimHandle(Agent:agent, XpGranted:float):void=
        Vals:=MakeBasicCanvas(XpGranted)
        Canvas:=Vals(0)
        Hit:=Vals(1)

        if:
            PlayerUI:=GetPlayerUI[player[Agent]]
        then:
            PlayerUI.AddWidget(Canvas,player_ui_slot{InputMode:=ui_input_mode.None})
            if:
                set CanvasPerPlayer[Agent] = option{Canvas}
                set HitPerPlayer[Agent] = option{Hit}
            then:
                spawn:
                    MoveAnimController(Hit,Canvas,PlayerUI)

    MoveAnimController(Hit:hud_rising_animation,Canvas:canvas, PlayerUI:player_ui)<suspends>:void=
        loop:
            Sleep(0.01)

            set Hit.Progress = Hit.Progress -0.01
            Hit.SetPosition(Hit.Progress)

            if:
                Hit.Progress <= 1.0
                Hit.Progress >= 0.0
                myConstrainedValue: type{_X:float where 0.000000 <= _X, _X <= 1.000000} = type{_X:float where 0.000000 <= _X, _X <= 1.000000}[Hit.Progress]
                
            then:
                Hit.TextBlock.SetTextOpacity(myConstrainedValue)
                Hit.TextBlock.SetShadowOpacity(myConstrainedValue)
                Hit.TextBlock2.SetTextOpacity(myConstrainedValue)
                Hit.TextBlock2.SetShadowOpacity(myConstrainedValue)
            
            if(Hit.Progress <= 0.1):
                RemoveUIForPlayer(Hit,Canvas,PlayerUI)
            if(Hit.Progress <= 0.0):
                break

    RemoveUIForPlayer(Hit:hud_rising_animation, Canvas:canvas, PlayerUI:player_ui)<suspends>:void=
        #Overlay := Hit.GetWidget()
        Canvas.RemoveWidget(Hit.MovingColorBase)
        Canvas.RemoveWidget(Hit.TextureIconBlock)
        Canvas.RemoveWidget(Hit.TextBlock)
        PlayerUI.RemoveWidget(Canvas)



                
        

