using { /Fortnite.com/Devices }
using { /Verse.org/Simulation }
using { /UnrealEngine.com/Temporary/Diagnostics }
using { /UnrealEngine.com/Temporary/UI }
using { /UnrealEngine.com/Temporary/SpatialMath }
using { /Fortnite.com/UI }
using { /Verse.org/Colors }
using { /Fortnite.com/Playspaces }
using { /Fortnite.com/Characters }
using { /Verse.org/Assets }


cosmetic_block := class():

    cosmeticdevice : cosmetic_ui_device
    S2T<localizes>(String : string) : message = "{String}"
    storedstring<localizes>(S:string) : message = "{S}"

    OuterOverlay<private> : overlay = overlay{}

    ColorBlockBG : texture_block = texture_block:
        DefaultImage:=wooden_pixel_gui.textures.Panels.WhiteSquare75
        DefaultTint := MakeColorFromHex("4D3720FF")
        #DefaultColor := MakeColorFromHex("0d256c")
        #DefaultColor := color{R:=13.0,G:=37.0,B:=108.0}
        DefaultDesiredSize := vector2:
            X := 230.0
            Y := 230.0
    CosmeticBlockName : text_block = text_block:
        DefaultTextColor := NamedColors.White

    CosmeticIcon : texture_block = texture_block:
        DefaultImage:=wooden_pixel_gui.textures.Icons2.Size32.t_Helmet02_32
        DefaultDesiredSize := vector2:
            X := 100.0
            Y := 100.0

    CosmeticIconShadow : texture_block = texture_block:
        DefaultImage:=wooden_pixel_gui.textures.Icons2.Size32.t_Helmet02_32
        DefaultTint:=MakeColorFromHex("110803FF")
        DefaultDesiredSize := vector2:
            X := 100.0
            Y := 100.0

    CosmeticButton : button_quiet := button_quiet{}
    
    var AName : string = "Jetpack"
    var ButtonText : string = "Wear"
    var CosmeticId : int = 0
    var MaybeIcon : ?texture = false

    SetName(V:string):void=set AName=V
    SetButtonText(V:string):void=set ButtonText=V
    SetCosmeticId(V:int):void=set CosmeticId=V
    SetIcon(V:texture):void=set MaybeIcon = option{V}

    Update():void=

        CosmeticBlockName.SetText(S2T(AName))
        CosmeticButton.SetText(S2T(ButtonText))
        if (Icon := MaybeIcon?):
            CosmeticIcon.SetImage(Icon)
            CosmeticIconShadow.SetImage(Icon)

    Init<public>():overlay=

        CosmeticButton.OnClick().SubscribeMessage(cosmeticdevice.ShowInfoBox,CosmeticId)
        if (Icon := MaybeIcon?):
            CosmeticIcon.SetImage(Icon)
            CosmeticIconShadow.SetImage(Icon)

        OuterOverlay.AddWidget(
            overlay_slot:
                Widget := ColorBlockBG
                HorizontalAlignment := horizontal_alignment.Fill
                VerticalAlignment := vertical_alignment.Fill
        )
        OuterOverlay.AddWidget(
            overlay_slot:
                HorizontalAlignment := horizontal_alignment.Center
                VerticalAlignment := vertical_alignment.Center
                Widget:=stack_box:
                    Orientation:=orientation.Vertical
                    Slots:=array:     
                        stack_box_slot:
                            HorizontalAlignment := horizontal_alignment.Center
                            VerticalAlignment := vertical_alignment.Center
                            Widget:=CosmeticBlockName       
                        stack_box_slot:
                            HorizontalAlignment := horizontal_alignment.Fill
                            VerticalAlignment := vertical_alignment.Fill
                            Padding:=margin{Left:=0.0,Top:=0.0,Right:=0.0,Bottom:=25.0}
                            Widget:=overlay:
                                Slots:=array:
                                    overlay_slot:
                                        HorizontalAlignment := horizontal_alignment.Center
                                        VerticalAlignment := vertical_alignment.Center
                                        Padding:=margin{Left:=0.0,Top:=6.0,Right:=0.0,Bottom:=0.0}
                                        Widget:=CosmeticIconShadow
                                    overlay_slot:
                                        HorizontalAlignment := horizontal_alignment.Center
                                        VerticalAlignment := vertical_alignment.Center
                                        Widget:=CosmeticIcon
                        stack_box_slot:
                            HorizontalAlignment := horizontal_alignment.Fill
                            VerticalAlignment := vertical_alignment.Fill
                            Widget:=CosmeticButton
        )

        return OuterOverlay


cosmetic_data:=class():

    var name : string
    var icon : texture
    var description: string
    var hint : string
    var buttontext : string
    var level : int

cosmetic_ui_device := class(creative_device):

    @editable
    UpgradeAccolade : accolades_device = accolades_device{}

    @editable
    OpenShop : button_device = button_device{}
    @editable
    PurchaseAudio : audio_player_device = audio_player_device{}

    # Just to pull % multtipliers
    @editable
    NPCManager : npc_manager_device = npc_manager_device{}
    @editable
    ButtonClickSound : audio_player_device = audio_player_device{}
    @editable
    OpenAudio : audio_player_device = audio_player_device{}
    @editable
    CloseAudio : audio_player_device = audio_player_device{}
    @editable
    TabAudio : audio_player_device = audio_player_device{}

    @editable
    NegativeAudio : audio_player_device = audio_player_device{}
    @editable
    UpgradeAudio : audio_player_device = audio_player_device{}
    @editable
    WearCosmeticAudio : audio_player_device = audio_player_device{}

    @editable
    OreConditionals : []conditional_button_device = array{}

    @editable
    HatMissionTrigger : trigger_device = trigger_device{}

    @editable
    VFXPickups : []trigger_device = array{}

    var AgentStackBox : [agent]stack_box = map{} # parent
    var AgentOverlay : [agent]overlay = map{}

    var AgentInfoOverlay : [agent]overlay = map{} # info box parent
    var AgentInfoOverlay2 : [agent]overlay = map{} # info box parent Upgrade Button

    var AgentBlocks : [agent][]cosmetic_block = map{}

    var MyUI : [agent]?canvas = map{}

    var TextBlockPerAgent : [agent]?[]text_block = map{}
    var AgentOreTextBlocks : [agent]?[]text_block = map{}

    storedstring<localizes>(S:string) : message = "{S}"
    JustIntString<localizes>(I:int) : message = "{I}"
    LevelString<localizes>(I:int) : message = "Lv{I}"
    DamageString<localizes>(I:int) : message = "{I}%"
    SecondString<localizes>(I:int) : message = "{I} Sec"
    HpString<localizes>(I:int) : message = "{I} hp"

    @editable
    UpgradeMissionTrigger : trigger_device = trigger_device{}


    OnBegin<override>()<suspends>:void=
        OpenShop.InteractedWithEvent.Subscribe(OpenShopHandle)

    EquipCosmeticHandle(Wdata:widget_message,id:int):void=
        if(MyTrigger := VFXPickups[id]):
            if(Player:=Wdata.Player,SelectedButton:=Wdata.Source,Agent:=agent[Player]):
                if(MyData :=  Data[Player]):
                    ButtonClickSound.Play(Agent)
                    if(MyData.CosmeticsUnlocked[id] = true):
                        MyTrigger.Trigger(Agent)
                        RemoveHUD(Wdata)
                        WearCosmeticAudio.Play(Agent)
                        if:
                            id<12
                        then:
                            if:
                                OldData := Data[Player]
                                set Data[Player] = player_data{player_data_update<constructor>(OldData),BackCosmeticEquiped := id}
                        else:
                            if:
                                OldData := Data[Player]
                                set Data[Player] = player_data{player_data_update<constructor>(OldData),HatCosmeticEquiped := id-12}
                            then:
                                HatMissionTrigger.Trigger(Agent)

                        NPCManager.CosmeticUnlockDevice.UpdateCosmeticStats(Agent)

                    if(MyDamage:=NPCManager.CosmeticUnlockDevice.AgentCosmeticStatsPerLevel[Agent][id].StatMap["Crit Chance"]):
                        Print("CritChance := {MyDamage}")
                        Print("Equiped Crit Chance := {MyData.CosmeticCritChance}")

                
    ShowInfoBox(Wdata:widget_message,id:int):void=
        if(Player:=Wdata.Player, PlayerUI:=GetPlayerUI[Player],SelectedButton:=Wdata.Source,Agent:=agent[Player]):
            ButtonClickSound.Play(Agent)
            CreateCosmeticInfoBox(Agent,id)
    

    TabOneHandle(Wdata:widget_message):void=
        if(Player:=Wdata.Player, PlayerUI:=GetPlayerUI[Player],SelectedButton:=Wdata.Source,Agent:=agent[Player]):
            if(MyParentStackBox := AgentStackBox[Agent]):
                if(MyOverlay := AgentOverlay[Agent]):
                    
                    TabAudio.Play(Agent)

                    MyOverlay.RemoveWidget(MyParentStackBox)
                    MyParentStackBox.SetVisibility(widget_visibility.Visible)
                    MyParentStackBox.SetVisibility(widget_visibility.Hidden)

                    CreateTabsAll(Agent,MyOverlay,1)

    TabTwoHandle(Wdata:widget_message):void=
        if(Player:=Wdata.Player, PlayerUI:=GetPlayerUI[Player],SelectedButton:=Wdata.Source,Agent:=agent[Player]):
            if(MyParentStackBox := AgentStackBox[Agent]):
                if(MyOverlay := AgentOverlay[Agent]):
                    
                    TabAudio.Play(Agent)

                    MyOverlay.RemoveWidget(MyParentStackBox)
                    MyParentStackBox.SetVisibility(widget_visibility.Visible)
                    MyParentStackBox.SetVisibility(widget_visibility.Hidden)

                    CreateTabsAll(Agent,MyOverlay,2)

        
        
    MakeEntireBackground (Agent:agent):canvas=

        GemTexture : texture = wooden_pixel_gui.textures.Icons2.Size32.t_Gem05_32
        GoldTexture : texture = wooden_pixel_gui.textures.Icons2.Size8.t_CoinGold_8
        CopperTexture : texture = hud.Copper_icon512
        SilverTexture : texture = hud.Silver_icon
        MalachiteTexture : texture = hud.Malachite_icon512
        ObsidianTexture : texture = hud.Obsidian_icon512
        BrightCoreTexture : texture = hud.Brightcore_icon512
        SpectroliteTexture : texture = hud.Spectrolite_icon512
        GoldCoinTexture : texture = wooden_pixel_gui.textures.Icons2.Size8.t_CoinGold_8

        TitleBarTexture := texture_block :
            DefaultDesiredSize:=vector2{X:=16.0,Y:=90.0}
            DefaultImage:=wooden_pixel_gui.textures.Panels.t_TitleBar

        InventoryGoldText : text_block = text_block:
            DefaultTextColor := MakeColorFromHex("FFFFFF")
            DefaultText:=storedstring("")

        if(set TextBlockPerAgent[Agent] = option{array{InventoryGoldText}}){} 
        TempTextBlock := text_block:
        if(set AgentOreTextBlocks[Agent] = option{array{TempTextBlock,TempTextBlock,TempTextBlock,TempTextBlock,TempTextBlock,TempTextBlock,TempTextBlock,TempTextBlock}}){}

        CopperSlot : stack_box_slot = CreateOreSlot(CopperTexture,0,Agent)
        SilverSlot : stack_box_slot = CreateOreSlot(SilverTexture,1,Agent)
        MalachiteSlot : stack_box_slot = CreateOreSlot(MalachiteTexture,2,Agent)
        ObsidianSlot : stack_box_slot = CreateOreSlot(ObsidianTexture,3,Agent)
        BrightCoreSlot : stack_box_slot = CreateOreSlot(BrightCoreTexture,4,Agent)
        SpectroliteSlot : stack_box_slot = CreateOreSlot(SpectroliteTexture,5,Agent)


        MainOreStackBox := stack_box : 
            Orientation := orientation.Horizontal

        MainOreStackBox.AddWidget(CopperSlot)
        MainOreStackBox.AddWidget(SilverSlot)
        MainOreStackBox.AddWidget(MalachiteSlot)
        MainOreStackBox.AddWidget(ObsidianSlot)
        MainOreStackBox.AddWidget(BrightCoreSlot)
        MainOreStackBox.AddWidget(SpectroliteSlot)

        UpdateOre(Agent)
            

        UpdateGoldText(Agent)

        Tab1Button := button_quiet :
            DefaultText:=storedstring("")
        Tab1Button.OnClick().Subscribe(TabOneHandle)

        Tab2Button := button_quiet :
            DefaultText:=storedstring("")
        Tab2Button.OnClick().Subscribe(TabTwoHandle)

        CloseWindowButton := button_quiet :
            DefaultText:=storedstring("Close")
        CloseWindowButton.OnClick().Subscribe(RemoveHUD)

        OverlayGrid : overlay = overlay:

        CreateTabsAll(Agent,OverlayGrid,1)
        
        if (set AgentOverlay[Agent] = OverlayGrid){}

        Overlay1 := overlay:
            Slots:=array:
                overlay_slot:
                    HorizontalAlignment:=horizontal_alignment.Fill
                    VerticalAlignment:=vertical_alignment.Fill
                    Padding:=margin{Left:=-5.0,Bottom:=-35.0,Top:=-5.0,Right:=-5.0}
                    Widget:=texture_block:
                        DefaultImage:=wooden_pixel_gui.textures.Panels.WhiteSquare75
                        DefaultTint:=MakeColorFromHex("543C22FF")
                overlay_slot:
                    HorizontalAlignment:=horizontal_alignment.Fill
                    VerticalAlignment:=vertical_alignment.Fill
                    Padding:=margin{Left:=0.0,Bottom:=-30.0,Top:=0.0,Right:=0.0}
                    Widget:=texture_block:
                        DefaultImage:=wooden_pixel_gui.textures.Panels.WhiteSquare75
                        DefaultTint:=MakeColorFromHex("6C4B2AE6")
                overlay_slot:
                    HorizontalAlignment:=horizontal_alignment.Fill
                    VerticalAlignment:=vertical_alignment.Fill
                    Padding:=margin{Left:=0.0,Bottom:=0.0,Top:=0.0,Right:=0.0}
                    Widget := stack_box:
                        Orientation:=orientation.Vertical
                        Slots:=array:
                            stack_box_slot:
                                HorizontalAlignment:=horizontal_alignment.Fill
                                VerticalAlignment:=vertical_alignment.Fill
                                Widget := overlay:
                                    Slots:=array:
                                        overlay_slot:
                                            HorizontalAlignment:=horizontal_alignment.Fill
                                            VerticalAlignment:=vertical_alignment.Fill
                                            Padding:=margin{Left:=0.0,Top:=0.0,Right:=0.0,Bottom:=7.0}
                                            Widget:= TitleBarTexture
                                        overlay_slot:
                                            HorizontalAlignment:=horizontal_alignment.Center
                                            VerticalAlignment:=vertical_alignment.Center
                                            Widget := text_block :
                                                DefaultText:=storedstring("Armor")
                                                DefaultTextColor := MakeColorFromHex("643D19FF")
                            stack_box_slot:
                                HorizontalAlignment:=horizontal_alignment.Fill
                                VerticalAlignment:=vertical_alignment.Fill
                                Padding:=margin{Left:=0.0,Top:=0.0,Right:=24.0,Bottom:=0.0}
                                Widget := stack_box:
                                    Orientation:=orientation.Horizontal
                                    Slots:=array:
                                        stack_box_slot:
                                            HorizontalAlignment:=horizontal_alignment.Fill
                                            VerticalAlignment:=vertical_alignment.Fill
                                            Padding:=margin{Left:=15.0,Top:=0.0,Right:=15.0,Bottom:=0.0}
                                            Widget := stack_box:
                                                Orientation:=orientation.Vertical
                                                Slots:=array:
                                                    stack_box_slot:
                                                        HorizontalAlignment:=horizontal_alignment.Fill
                                                        VerticalAlignment:=vertical_alignment.Fill
                                                        Padding:=margin{Left:=0.0,Top:=30.0,Right:=0.0,Bottom:=0.0}
                                                        Widget := overlay:
                                                            Slots:=array:
                                                                overlay_slot:
                                                                    HorizontalAlignment:=horizontal_alignment.Fill
                                                                    VerticalAlignment:=vertical_alignment.Fill
                                                                    Widget:=texture_block :
                                                                        DefaultDesiredSize:=vector2{X:=100.0,Y:=100.0}
                                                                        DefaultImage:=RavenIcons.MapMarkers.tile008
                                                                overlay_slot:
                                                                    HorizontalAlignment:=horizontal_alignment.Fill
                                                                    VerticalAlignment:=vertical_alignment.Fill
                                                                    Widget:=Tab2Button
                                                    stack_box_slot:
                                                        HorizontalAlignment:=horizontal_alignment.Fill
                                                        VerticalAlignment:=vertical_alignment.Fill
                                                        Padding:=margin{Left:=0.0,Top:=30.0,Right:=0.0,Bottom:=0.0}
                                                        Widget := overlay:
                                                            Slots:=array:
                                                                overlay_slot:
                                                                    HorizontalAlignment:=horizontal_alignment.Fill
                                                                    VerticalAlignment:=vertical_alignment.Fill
                                                                    Widget:=texture_block :
                                                                        DefaultDesiredSize:=vector2{X:=100.0,Y:=100.0}
                                                                        DefaultImage:=RavenIcons.MapMarkers.S4545
                                                                overlay_slot:
                                                                    HorizontalAlignment:=horizontal_alignment.Fill
                                                                    VerticalAlignment:=vertical_alignment.Fill
                                                                    Widget:=Tab1Button
                                        stack_box_slot:
                                            HorizontalAlignment:=horizontal_alignment.Fill
                                            VerticalAlignment:=vertical_alignment.Fill
                                            Widget := OverlayGrid
                            stack_box_slot:
                                HorizontalAlignment:=horizontal_alignment.Center
                                VerticalAlignment:=vertical_alignment.Center
                                Padding:=margin{Left:=0.0,Top:=15.0,Right:=0.0,Bottom:=0.0}
                                Widget := CloseWindowButton
                overlay_slot:
                    HorizontalAlignment:=horizontal_alignment.Left
                    VerticalAlignment:=vertical_alignment.Bottom
                    Padding:=margin{Left:=20.0,Top:=0.0,Right:=0.0,Bottom:=20.0}
                    Widget:=overlay:
                        Slots:=array:
                            overlay_slot:
                                HorizontalAlignment:=horizontal_alignment.Left
                                VerticalAlignment:=vertical_alignment.Top
                                Widget:=stack_box:
                                    Orientation:=orientation.Horizontal
                                    Slots:=array:
                                        stack_box_slot:
                                            HorizontalAlignment:=horizontal_alignment.Fill
                                            VerticalAlignment:=vertical_alignment.Fill
                                            Widget:=texture_block:
                                                DefaultImage:=GoldCoinTexture
                                                DefaultDesiredSize:=vector2{X:=41.0,Y:=32.0}
                                        stack_box_slot:
                                            HorizontalAlignment:=horizontal_alignment.Left
                                            VerticalAlignment:=vertical_alignment.Center
                                            Padding:=margin{Left:=10.0,Top:=0.0,Right:=0.0,Bottom:=0.0}
                                                Widget:=InventoryGoldText
                overlay_slot:
                    HorizontalAlignment:=horizontal_alignment.Left
                    VerticalAlignment:=vertical_alignment.Bottom
                    Padding:=margin{Left:=20.0,Top:=0.0,Right:=0.0,Bottom:=20.0}
                    Widget:=overlay:
                        Slots:=array:
                            overlay_slot:
                                HorizontalAlignment:=horizontal_alignment.Left
                                VerticalAlignment:=vertical_alignment.Top
                                Widget:=stack_box:
                                    Orientation:=orientation.Horizontal
                                    Slots:=array:
                                        stack_box_slot:
                                            HorizontalAlignment:=horizontal_alignment.Fill
                                            VerticalAlignment:=vertical_alignment.Fill
                                            Widget:=texture_block:
                                                DefaultImage:=GoldCoinTexture
                                                DefaultDesiredSize:=vector2{X:=41.0,Y:=32.0}
                                        stack_box_slot:
                                            HorizontalAlignment:=horizontal_alignment.Left
                                            VerticalAlignment:=vertical_alignment.Center
                                            Padding:=margin{Left:=10.0,Top:=0.0,Right:=0.0,Bottom:=0.0}
                                                Widget:=InventoryGoldText
                overlay_slot:
                    HorizontalAlignment:=horizontal_alignment.Left
                    VerticalAlignment:=vertical_alignment.Bottom
                    Padding:=margin{Left:=20.0,Top:=0.0,Right:=0.0,Bottom:=-30.0}
                    Widget:=overlay:
                        Slots:=array:
                            overlay_slot:
                                HorizontalAlignment:=horizontal_alignment.Left
                                VerticalAlignment:=vertical_alignment.Top
                                Widget:=MainOreStackBox
  
        MyCanvas:=canvas:
            Slots:=array:
                canvas_slot:
                    Anchors:=anchors{Minimum:=vector2{X:=0.05,Y:=0.07}, Maximum:=vector2{X:=0.65,Y:=0.93}}
                    Widget:= Overlay1
                    
        return MyCanvas

    UpdateOre(Agent:agent):void= 

        GemTexture : texture = wooden_pixel_gui.textures.Icons2.Size32.t_Gem05_32
        GoldTexture : texture = wooden_pixel_gui.textures.Icons2.Size8.t_CoinGold_8
        CopperTexture : texture = hud.Copper_icon512
        SilverTexture : texture = hud.Silver_icon
        MalachiteTexture : texture = hud.Malachite_icon512
        ObsidianTexture : texture = hud.Obsidian_icon512
        BrightCoreTexture : texture = hud.Brightcore_icon512
        SpectroliteTexture : texture = hud.Spectrolite_icon512


            if:
                Player:=player[Agent]
                PlayerData := Data[Player]
            then:

                if(Cond := OreConditionals[0]): # Copper
                    OreCount := Cond.GetItemCount(Agent,0)
                    if(OreCount > 0):
                        if(TextBlock := AgentOreTextBlocks[Agent]?[0]):
                            TextBlock.SetText(storedstring("{formatter.Abbreviate((OreCount * 1.0),false)}"))


                if(Cond := OreConditionals[1]): # Silver
                    OreCount := Cond.GetItemCount(Agent,0)
                    if(OreCount > 0):
                        if(TextBlock := AgentOreTextBlocks[Agent]?[1]):
                            TextBlock.SetText(storedstring("{formatter.Abbreviate((OreCount * 1.0),false)}"))


                if(Cond := OreConditionals[2]): # Malachite
                    OreCount := Cond.GetItemCount(Agent,0)
                    if(OreCount > 0):
                        if(TextBlock := AgentOreTextBlocks[Agent]?[2]):
                            TextBlock.SetText(storedstring("{formatter.Abbreviate((OreCount * 1.0),false)}"))


                if(Cond := OreConditionals[3]): # Obsidian
                    OreCount := Cond.GetItemCount(Agent,0)
                    if(OreCount > 0):
                        if(TextBlock := AgentOreTextBlocks[Agent]?[3]):
                            TextBlock.SetText(storedstring("{formatter.Abbreviate((OreCount * 1.0),false)}"))


                if(Cond := OreConditionals[4]): # BrightCore
                    OreCount := Cond.GetItemCount(Agent,0)
                    if(OreCount > 0):
                        if(TextBlock := AgentOreTextBlocks[Agent]?[4]):
                            TextBlock.SetText(storedstring("{formatter.Abbreviate((OreCount * 1.0),false)}"))


                if(Cond := OreConditionals[5]): # BrightCore
                    OreCount := Cond.GetItemCount(Agent,0)
                    if(OreCount > 0):
                        if(TextBlock := AgentOreTextBlocks[Agent]?[5]):
                            TextBlock.SetText(storedstring("{formatter.Abbreviate((OreCount * 1.0),false)}"))


    CreateOreSlot(Texture : texture, TextBlockId : int, Agent:agent):stack_box_slot= 

        IconTextureBlock: texture_block = texture_block{DefaultImage:= Texture, DefaultDesiredSize:=vector2{X:=36.0,Y:=36.0}}
        IconTextureShadowBlock: texture_block = texture_block{DefaultImage:= Texture,DefaultTint:=MakeColorFromHex("4C3116FF"), DefaultDesiredSize:=vector2{X:=36.0,Y:=36.0}}
        SlotText: text_block = text_block:
            DefaultTextColor := MakeColorFromHex("FFFFFF")
            DefaultText:=storedstring("0")
            DefaultShadowOffset:=option{vector2{X:=1.0,Y:=3.0}}
            DefaultShadowColor:=MakeColorFromHex("4C3116FF")
        SlotText.SetShadowOpacity(X:=0.27)

        Ov1 := overlay_slot: # need to create these so I can save overlay and remove it later
            HorizontalAlignment:=horizontal_alignment.Left
            VerticalAlignment:=vertical_alignment.Center
            Padding:=margin{Left:=2.0,Top:=2.0,Right:=0.0,Bottom:=0.0}
            Widget:=IconTextureShadowBlock
        Ov2 := overlay_slot:
                HorizontalAlignment:=horizontal_alignment.Left
                VerticalAlignment:=vertical_alignment.Center
                Padding:=margin{Left:=2.0,Top:=0.0,Right:=0.0,Bottom:=0.0}
                Widget:=IconTextureBlock
        Ov3 := overlay_slot:
            HorizontalAlignment:=horizontal_alignment.Right
            VerticalAlignment:=vertical_alignment.Center
            Padding:=margin{Left:=40.0,Top:=0.0,Right:=5.0,Bottom:=0.0}
            Widget:=SlotText

        MyOverlay := overlay:

        MyOverlay.AddWidget(Ov1)
        MyOverlay.AddWidget(Ov2)
        MyOverlay.AddWidget(Ov3)
            
        MyStackBoxSlot := stack_box_slot:
            Widget:=MyOverlay
                    
        # set array index = textblock to update
        if(var OldTextBlock : []text_block  = AgentOreTextBlocks[Agent]?):
            if(set OldTextBlock[TextBlockId] = SlotText):
                if(set AgentOreTextBlocks[Agent] = option{OldTextBlock}){}

        return MyStackBoxSlot

    UpdateInvGold(Agent:agent)<suspends>:void=
        loop:
            Sleep(0.5)
            if(PlayerData := Data[player[Agent]]):
                if(TextBlocks  := TextBlockPerAgent[Agent]?):
                    if(GoldTextBlock := TextBlocks[0]):
                        GoldTextBlock.SetText(storedstring("{formatter.Abbreviate((PlayerData.Gold * 1.0),false)}"))
                        if(TextBlockPerAgent[Agent] = false):
                            break
                else:
                    break

    UpdateGoldText(Agent:agent):void=
        if(PlayerData := Data[player[Agent]]):
            if(TextBlocks  := TextBlockPerAgent[Agent]?):
                if(GoldTextBlock := TextBlocks[0]):
                    GoldTextBlock.SetText(storedstring("{formatter.Abbreviate((PlayerData.Gold * 1.0),false)}"))

    CreateCosmeticInfoBox(Agent:agent,id:int):void=

        Print("cosmetic id:{id}")

        GlowIcon:=wooden_pixel_gui.textures.Icons2.Size32.GlowM3

        # Close old info overlay if it exists
        if(MyCanvas := MyUI[Agent]?):
            if(OldOverlay := AgentInfoOverlay[Agent]):
                MyCanvas.RemoveWidget(OldOverlay)
                OldOverlay.SetVisibility(widget_visibility.Visible)
                OldOverlay.SetVisibility(widget_visibility.Hidden)
            if(OldOverlay2 := AgentInfoOverlay2[Agent]):
                MyCanvas.RemoveWidget(OldOverlay2)
                OldOverlay2.SetVisibility(widget_visibility.Visible)
                OldOverlay2.SetVisibility(widget_visibility.Hidden)

            
        var CData : cosmetic_data = cosmetic_data :
            name:="Test"
            description:="Test"
            hint:="Test"
            icon:=wooden_pixel_gui.textures.Icons2.Size16.t_LockGrey_16
            buttontext:="Info"
            level:=1


        var CosmeticDatas : []cosmetic_data = array{}
        if(id <= 11):
            set CosmeticDatas = CreateTabOne(Agent)
            if(MyCosmetic := CosmeticDatas[id]):
                set CData = MyCosmetic
        else:
            set CosmeticDatas = CreateTabTwo(Agent)
            if(MyCosmetic := CosmeticDatas[id - 12]):
                set CData = MyCosmetic

        Name := text_block :
            DefaultText:=storedstring(CData.name)
            DefaultTextColor := MakeColorFromHex("543C22FF")
        Description := text_block :
            DefaultText:=storedstring(CData.description)
            DefaultTextColor := MakeColorFromHex("FFFFFF")
        ImageTexture := texture_block :
            DefaultImage:=CData.icon
            DefaultDesiredSize:=vector2{X:=150.0,Y:=150.0}

        GlowWidget := texture_block:
            DefaultImage:=GlowIcon
            DefaultDesiredSize:=vector2{X:=200.0,Y:=200.0}


        EquipButton := button_quiet :
            DefaultText:=storedstring("Wear")

        if(MyData :=  Data[player[Agent]]):

            if(CosLv := MyData.CosmeticsLevels[id]):
                GlowWidget.SetTint(CalculateColor(CosLv))

            if(MyData.CosmeticsUnlocked[id] = true):
                EquipButton.SetText(storedstring("Wear"))
            else:
                EquipButton.SetText(storedstring("Locked"))

            
        EquipButton.OnClick().SubscribeMessage(EquipCosmeticHandle,id)

        
        StatsStackBox := stack_box:
            Orientation:=orientation.Vertical

        PossibleHintTxt := text_block:
            DefaultText:=storedstring(CData.hint)
            DefaultTextColor := MakeColorFromHex("FFFFFF")

        HintStackBox := stack_box_slot:
            Widget:=PossibleHintTxt

        if(MyData :=  Data[player[Agent]]):
            if(MyData.CosmeticsUnlocked[id] = false):
                StatsStackBox.AddWidget(HintStackBox)
            else:
                AllStats:= NPCManager.CosmeticUnlockDevice.LoadAllCosmeticStats() # Creating Stats Desription
                if:
                    MyStats := AllStats[id]
                    MyLevelStatMap:= NPCManager.CosmeticUnlockDevice.AgentCosmeticStatsPerLevel[Agent][id].StatMap
                then:
                    for(StatName->StatValue:MyStats.StatMap):
                        
                        if:
                            StatValue > 0.0
                            LevelValue := MyLevelStatMap[StatName]
                        then:
                            IndividualStackBox:= stack_box:
                                Orientation:=orientation.Horizontal
                            ValueText:=text_block :
                                DefaultText:=storedstring("{"{formatter.Abbreviate(StatValue * 1.0,false)}"}%")
                                DefaultTextColor := MakeColorFromHex("FFFFFF")

                            var LevelValueText : text_block  = text_block{DefaultTextColor := MakeColorFromHex("FFD700"), DefaultText:=storedstring("{}") }

                            if(LevelValue > 0.0):      
                                LevelValueText.SetText(storedstring("{"+{formatter.Abbreviate(LevelValue * 1.0,true)}"}%"))

                            NameText:=text_block :
                                DefaultText:=storedstring(" {StatName}")
                                DefaultTextColor := MakeColorFromHex("FFFFFF")

                            slot1 := stack_box_slot:
                                Widget:=ValueText
                            slot2 := stack_box_slot:
                                Widget:=NameText
                            slot3 := stack_box_slot:
                                Widget:=LevelValueText

                            
                            if(CosLv := MyData.CosmeticsLevels[id]):
                                LevelValueText.SetTextColor(CalculateColor(CosLv))

                            IndividualStackBox.AddWidget(slot1)
                            IndividualStackBox.AddWidget(slot3)
                            IndividualStackBox.AddWidget(slot2)

                            mainslot := stack_box_slot:
                                Widget:=IndividualStackBox

                            StatsStackBox.AddWidget(mainslot)

        Overlay1 :=  overlay:
            Slots:=array:
                overlay_slot:
                    HorizontalAlignment:=horizontal_alignment.Fill
                    VerticalAlignment:=vertical_alignment.Fill
                    Padding:=margin{Left:=-5.0,Bottom:=-5.0,Top:=-5.0,Right:=-5.0}
                    Widget:=texture_block:
                        DefaultImage:=wooden_pixel_gui.textures.Panels.WhiteSquare75
                        DefaultTint:=MakeColorFromHex("543C22FF")
                overlay_slot:
                    HorizontalAlignment:=horizontal_alignment.Fill
                    VerticalAlignment:=vertical_alignment.Fill
                    Widget:=texture_block:
                        DefaultImage:=wooden_pixel_gui.textures.Panels.WhiteSquare75
                        DefaultTint:=MakeColorFromHex("6C4B2AE6")
                overlay_slot:
                    HorizontalAlignment:=horizontal_alignment.Fill
                    VerticalAlignment:=vertical_alignment.Fill
                    Padding:=margin{Left:=0.0,Bottom:=7.0,Top:=0.0,Right:=0.0}
                    Widget := stack_box:
                        Orientation:=orientation.Vertical
                        Slots:=array:
                            stack_box_slot:
                                HorizontalAlignment:=horizontal_alignment.Fill
                                VerticalAlignment:=vertical_alignment.Fill
                                Widget := overlay:
                                    Slots:=array:
                                        overlay_slot:
                                            HorizontalAlignment:=horizontal_alignment.Fill
                                            VerticalAlignment:=vertical_alignment.Fill
                                            Padding:=margin{Left:=0.0,Top:=0.0,Right:=0.0,Bottom:=7.0}
                                            Widget:=texture_block :
                                                DefaultDesiredSize:=vector2{X:=16.0,Y:=90.0}
                                                DefaultImage:=wooden_pixel_gui.textures.Panels.t_TitleBar
                                        overlay_slot:
                                            HorizontalAlignment:=horizontal_alignment.Center
                                            VerticalAlignment:=vertical_alignment.Center
                                            Widget := Name
                            stack_box_slot:
                                HorizontalAlignment:=horizontal_alignment.Fill
                                VerticalAlignment:=vertical_alignment.Fill
                                Widget := overlay:
                                    Slots:=array:
                                        overlay_slot: # background box
                                            Widget:=texture_block :
                                                DefaultDesiredSize:=vector2{X:=200.0,Y:=200.0}
                                                DefaultImage:=wooden_pixel_gui.textures.Panels.WhiteSquare75
                                                DefaultTint:=MakeColorFromHex("4D3720FF")
                                        overlay_slot:
                                            HorizontalAlignment:=horizontal_alignment.Center
                                            VerticalAlignment:=vertical_alignment.Center
                                            Padding:=margin{Left:=0.0,Top:=0.0,Right:=0.0,Bottom:=0.0}
                                            Widget:=GlowWidget
                                        overlay_slot:
                                            HorizontalAlignment:=horizontal_alignment.Center
                                            VerticalAlignment:=vertical_alignment.Center
                                            Padding:=margin{Left:=0.0,Top:=0.0,Right:=0.0,Bottom:=0.0}
                                            Widget:=ImageTexture

                            stack_box_slot:
                                HorizontalAlignment:=horizontal_alignment.Left
                                VerticalAlignment:=vertical_alignment.Center
                                Padding:=margin{Left:=10.0,Top:=10.0,Right:=10.0,Bottom:=10.0}
                                Widget:=text_block :
                                    DefaultText:=storedstring("Stats:")
                                    DefaultTextColor := MakeColorFromHex("FFFFFF")
                            stack_box_slot:
                                HorizontalAlignment:=horizontal_alignment.Fill
                                VerticalAlignment:=vertical_alignment.Fill
                                Widget := overlay:
                                    Slots:=array:
                                        overlay_slot: # background box
                                            Widget:=texture_block :
                                                DefaultDesiredSize:=vector2{X:=440.0,Y:=141.0}
                                                DefaultImage:=wooden_pixel_gui.textures.Panels.WhiteSquare75
                                                DefaultTint:=MakeColorFromHex("4D3720FF")
                                        overlay_slot:
                                            HorizontalAlignment:=horizontal_alignment.Center
                                            VerticalAlignment:=vertical_alignment.Center
                                            Widget:=StatsStackBox
                            stack_box_slot:
                                Padding:=margin{Left:=0.0,Top:=25.0,Right:=0.0,Bottom:=0.0}
                                HorizontalAlignment:=horizontal_alignment.Center
                                VerticalAlignment:=vertical_alignment.Center
                                Widget:=EquipButton

        CanvasSlot:= canvas_slot:
            Anchors:=anchors{Minimum:=vector2{X:=0.7,Y:=0.1}, Maximum:=vector2{X:=0.95,Y:=0.65}}
            Widget := Overlay1

        ################################################# Upgrade Button Overlay

        NPCManager.CosmeticUnlockDevice.CalculateLevelDamage(Agent)

        LevelText := text_block:
            DefaultTextColor:=MakeColorFromHex("FFFFFF")

        var OrePriceStackBox :stack_box = stack_box :
            Orientation:=orientation.Horizontal

        var GoldPriceStackBox : stack_box = stack_box :
            Orientation:=orientation.Horizontal

        if(MyData :=  Data[player[Agent]]):
            if(MyData.CosmeticsUnlocked[id] = false):
                LevelText.SetText(storedstring("Lv. ?"))
            else:
                if(CosmeticLevel := MyData.CosmeticsLevels[id]):
                    LevelText.SetText(storedstring("Lv. {CosmeticLevel}"))
                    set OrePriceStackBox = CalculateOreUpgradePrice(CosmeticLevel)
                    #set GoldPriceStackBox = CreateGoldStackBox(CosmeticLevel)

        UpgradeButton := button_quiet :
            DefaultText:=storedstring("upgrade") # needs to be linked to upgrading cosmetic levels

        UpgradeButton.OnClick().SubscribeMessage(UpgradeCosmetic,id)

        Overlay2 := overlay:
            Slots:=array:
                overlay_slot:
                    HorizontalAlignment:=horizontal_alignment.Fill
                    VerticalAlignment:=vertical_alignment.Fill
                    Padding:=margin{Left:=-5.0,Bottom:=-5.0,Top:=-5.0,Right:=-5.0}
                    Widget:=texture_block:
                        DefaultImage:=wooden_pixel_gui.textures.Panels.WhiteSquare75
                        DefaultTint:=MakeColorFromHex("543C22FF")
                overlay_slot:
                    HorizontalAlignment:=horizontal_alignment.Fill
                    VerticalAlignment:=vertical_alignment.Fill
                    Widget:=texture_block:
                        DefaultImage:=wooden_pixel_gui.textures.Panels.WhiteSquare75
                        DefaultTint:=MakeColorFromHex("6C4B2AE6")
                overlay_slot: # background box
                    Padding:=margin{Left:=0.0,Bottom:=40.0,Top:=0.0,Right:=0.0}
                    Widget:=texture_block :
                        DefaultDesiredSize:=vector2{X:=275.0,Y:=145.0}
                        DefaultImage:=wooden_pixel_gui.textures.Panels.WhiteSquare75
                        DefaultTint:=MakeColorFromHex("4D3720FF")
                overlay_slot:
                    HorizontalAlignment:=horizontal_alignment.Fill
                    VerticalAlignment:=vertical_alignment.Fill
                    Padding:=margin{Left:=31.0,Bottom:=31.0,Top:=31.0,Right:=31.0}
                    Widget := stack_box:
                        Orientation:=orientation.Vertical
                        Slots:=array:
                            stack_box_slot:
                                HorizontalAlignment:=horizontal_alignment.Center
                                VerticalAlignment:=vertical_alignment.Center
                                Widget:=LevelText
                            stack_box_slot:
                                Widget := stack_box:
                                    Orientation:=orientation.Vertical
                                    Slots:=array:
                                        stack_box_slot:
                                            HorizontalAlignment:=horizontal_alignment.Center
                                            VerticalAlignment:=vertical_alignment.Center
                                            Widget:=LevelText
                                        stack_box_slot:
                                            Widget:=OrePriceStackBox
                                        stack_box_slot:
                                            Widget:=GoldPriceStackBox
                                        stack_box_slot:
                                            HorizontalAlignment:=horizontal_alignment.Center
                                            VerticalAlignment:=vertical_alignment.Center
                                            Padding:=margin{Left:=0.0,Bottom:=0.0,Top:=23.0,Right:=0.0}
                                            Widget:=UpgradeButton

        CanvasSlot2:= canvas_slot:
            Anchors:=anchors{Minimum:=vector2{X:=0.7,Y:=0.7}, Maximum:=vector2{X:=0.95,Y:=0.95}}
            Widget := Overlay2


        if(MyCanvas := MyUI[Agent]?):
            if(set AgentInfoOverlay[Agent] = Overlay1){}
            if(set AgentInfoOverlay2[Agent] = Overlay2){}
            MyCanvas.AddWidget(CanvasSlot)
            MyCanvas.AddWidget(CanvasSlot2)
            Overlay1.SetVisibility(widget_visibility.Hidden)
            Overlay1.SetVisibility(widget_visibility.Visible)
            Overlay2.SetVisibility(widget_visibility.Hidden)
            Overlay2.SetVisibility(widget_visibility.Visible)

        RefreshBox(Agent)

    # Used to fix bug where you have to open it twice to see the info box
    RefreshBox(Agent:agent):void=
        if(MyCanvas := MyUI[Agent]?):
            if(OldOverlay := AgentInfoOverlay[Agent]):
                OldOverlay.SetVisibility(widget_visibility.Hidden)
                OldOverlay.SetVisibility(widget_visibility.Visible)
            if(OldOverlay2 := AgentInfoOverlay2[Agent]):
                OldOverlay2.SetVisibility(widget_visibility.Hidden)
                OldOverlay2.SetVisibility(widget_visibility.Visible)
                


    UpgradeCosmetic(Wdata:widget_message,id:int):void=

        if(Player:=Wdata.Player,Agent:=agent[Player]):
            if(MyData :=  Data[Player]):
                if(CosmeticLevel := MyData.CosmeticsLevels[id]):

                    var OreType : int = 0
                    var OreCount : int = 0

                    var OreConditional : conditional_button_device = conditional_button_device{}

                    if(CosmeticLevel < 20):
                        set OreType = 1
                        set OreCount = CosmeticLevel
                    else:
                        if(CosmeticLevel < 40):
                            set OreType = 2
                            set OreCount = CosmeticLevel-19
                        else:
                            if(CosmeticLevel < 60):
                                set OreType = 3
                                set OreCount = CosmeticLevel-39

                            else:
                                if(CosmeticLevel < 80):
                                    set OreType = 4
                                    set OreCount = CosmeticLevel-59
                                else:
                                    if(CosmeticLevel < 100):
                                        set OreType = 5
                                        set OreCount = CosmeticLevel-79
                                    else:
                                        set OreType = 6
                                        set OreCount = CosmeticLevel-99



                    if(Cond := OreConditionals[OreType-1]):
                        set OreConditional = Cond

                    MyOreCount := OreConditional.GetItemCount(Agent,0)

                    Print("OreType : {OreType}, OreCount : {MyOreCount}")
                    if(MyOreCount >= OreCount):
                        for(i:=0..OreCount-1):
                            OreConditional.Activate(Agent)
                        if:
                            var NewStat : []int =  MyData.CosmeticsLevels
                            set NewStat[id] += 1
                            set Data[Player] = player_data{player_data_update<constructor>(MyData), CosmeticsLevels:= NewStat}
                        then:
                            UpgradeAccolade.Award(Agent)
                            UpgradeAudio.Play(Agent)
                            UpdateOre(Agent)
                            Print("NewStat[{id}] = {CosmeticLevel+1}")
                            UpgradeMissionTrigger.Trigger(Agent)

                        
                        NPCManager.CosmeticUnlockDevice.UpdateCosmeticStats(Agent)
                        CreateCosmeticInfoBox(Agent,id)
                        
                    else:
                        NegativeAudio.Play(Agent)
                        Print("Not enough Ore")
                    #else:
                    #    NegativeAudio.Play(Agent)
                    #    Print("Not enough Gold")


    CalculateOreUpgradePrice(cosmeticlevel:int):stack_box=

        PriceStackBox := stack_box:
            Orientation:=orientation.Horizontal

        CopperOreTexture : texture = hud.Copper_icon512
        SilveOreTexture : texture = hud.Silver_icon
        MalachiteTexture : texture = hud.Malachite_icon512
        ObsidianTexture : texture = hud.Obsidian_icon512
        SpectroliteTexture : texture = hud.Spectrolite_icon512
        BrightcoreTexture : texture = hud.Brightcore_icon512

        OreWidget:= texture_block:
            DefaultImage := CopperOreTexture
            DefaultDesiredSize:=vector2{X:=41.0,Y:=32.0}

        PriceText := text_block:
            DefaultText:=storedstring("1")
            DefaultTextColor:=MakeColorFromHex("FFFFFF")

        OreSlot := stack_box_slot:
            Widget:=OreWidget
        PriceSlot := stack_box_slot:
            Widget:=PriceText

        if(cosmeticlevel < 20):
            OreWidget.SetImage(CopperOreTexture)
            PriceText.SetText(storedstring("{cosmeticlevel}"))
        else:
            if(cosmeticlevel < 40):
                OreWidget.SetImage(SilveOreTexture)
                PriceText.SetText(storedstring("{cosmeticlevel-19}"))
            else:
                if(cosmeticlevel < 60):
                    OreWidget.SetImage(MalachiteTexture)
                    PriceText.SetText(storedstring("{cosmeticlevel-39}"))
                else:
                    if(cosmeticlevel < 80):
                        OreWidget.SetImage(ObsidianTexture)
                        PriceText.SetText(storedstring("{cosmeticlevel-59}"))
                    else:
                        if(cosmeticlevel < 100):
                            OreWidget.SetImage(BrightcoreTexture)
                            PriceText.SetText(storedstring("{cosmeticlevel-79}"))
                        else:
                            OreWidget.SetImage(SpectroliteTexture)
                            PriceText.SetText(storedstring("{cosmeticlevel-99}"))


        PriceStackBox.AddWidget(OreSlot)
        PriceStackBox.AddWidget(PriceSlot)

        return PriceStackBox

    CreateGoldStackBox(cosmeticlevel:int):stack_box=

        PriceStackBox := stack_box:
            Orientation:=orientation.Horizontal

        GoldWidget:= texture_block:
            DefaultImage := wooden_pixel_gui.textures.Icons2.Size8.t_CoinGold_8
            DefaultDesiredSize:=vector2{X:=32.0,Y:=32.0}

        PriceText := text_block:
            DefaultText:=storedstring("{formatter.Abbreviate(CalculateGoldUpgradePrice(cosmeticlevel)*1.0,false)}")
            DefaultTextColor:=MakeColorFromHex("FFFFFF")

        GoldStackBox := stack_box_slot:
            Widget:=GoldWidget
        PriceTextStackBox := stack_box_slot:
            Widget:=PriceText
        
        PriceStackBox.AddWidget(GoldStackBox)
        PriceStackBox.AddWidget(PriceTextStackBox)

        return PriceStackBox


    CalculateGoldUpgradePrice(CosmeticLevel:int):int=

        var Price : float = 0.0
        var InitialPrice : float = 145.0
        set Price += InitialPrice
        for(i:=0..CosmeticLevel-1):
            set Price = Price * 1.30

        if(RoundedPrice := Ceil[Price]):
            return RoundedPrice

        CantRound := 0
        return CantRound



    CreateTabsAll(Agent:agent, OverlayMain:overlay, tabnumber:int):void=

        ParentStackBox := stack_box :
            Orientation:=orientation.Vertical
    
        CosmeticStackBox1 := stack_box :
            Orientation:=orientation.Horizontal
        CosmeticStackBox2 := stack_box :
            Orientation:=orientation.Horizontal
        CosmeticStackBox3 := stack_box :
            Orientation:=orientation.Horizontal
    
        OverlaySlot := overlay_slot :
            HorizontalAlignment:=horizontal_alignment.Fill
            VerticalAlignment:=vertical_alignment.Fill
            Padding:=margin{Left:=0.0,Top:=10.0,Right:=0.0,Bottom:=0.0}
                Widget := ParentStackBox
    
        ParentSlot1 := stack_box_slot:
            HorizontalAlignment:=horizontal_alignment.Fill
            VerticalAlignment:=vertical_alignment.Fill
            Padding:=margin{Left:=0.0,Top:=10.0,Right:=0.0,Bottom:=0.0}
                Widget := CosmeticStackBox1
    
        ParentSlot2 := stack_box_slot:
            HorizontalAlignment:=horizontal_alignment.Fill
            VerticalAlignment:=vertical_alignment.Fill
            Padding:=margin{Left:=0.0,Top:=10.0,Right:=0.0,Bottom:=0.0}
                Widget := CosmeticStackBox2
    
        ParentSlot3 := stack_box_slot:
            HorizontalAlignment:=horizontal_alignment.Fill
            VerticalAlignment:=vertical_alignment.Fill
            Padding:=margin{Left:=0.0,Top:=10.0,Right:=0.0,Bottom:=0.0}
                Widget := CosmeticStackBox3
    
        OverlayMain.AddWidget(OverlaySlot)
        ParentStackBox.AddWidget(ParentSlot1)
        ParentStackBox.AddWidget(ParentSlot2)
        ParentStackBox.AddWidget(ParentSlot3)
    
        if (set AgentStackBox[Agent] = ParentStackBox){}
        
        LockIcon:=wooden_pixel_gui.textures.Icons2.Size16.t_LockGrey_16
        
    
        var CosmeticDatas : []cosmetic_data = array{}
        if(tabnumber = 1):
            set CosmeticDatas = CreateTabOne(Agent)
        if(tabnumber = 2):
            set CosmeticDatas = CreateTabTwo(Agent)
    
        for(i:=0..CosmeticDatas.Length-1):
            if(CData := CosmeticDatas[i]):
    
                CBlock : cosmetic_block = cosmetic_block{cosmeticdevice:=Self}
                CBlock.SetName(CData.name)
                CBlock.SetIcon(CData.icon)
                CBlock.SetButtonText(CData.buttontext)
                if(tabnumber = 1):
                    CBlock.SetCosmeticId(i)
                else:
                    CBlock.SetCosmeticId(i+12)
                CBlock.Update()
    
                NewSlot := stack_box_slot :
                    HorizontalAlignment:=horizontal_alignment.Fill
                    VerticalAlignment:=vertical_alignment.Fill
                    Padding:=margin{Left:=0.0,Top:=0.0,Right:=10.0,Bottom:=0.0}
                    Widget := CBlock.Init()
    
                if(i<4):
                    CosmeticStackBox1.AddWidget(NewSlot)
                else:
                    if(i<8):
                         CosmeticStackBox2.AddWidget(NewSlot)
                    else:
                        CosmeticStackBox3.AddWidget(NewSlot)
        return 


    CreateTabOne(Agent:agent):[]cosmetic_data=

        var CosmeticDatas : []cosmetic_data = array{}
        if:
            OldData:=Data[player[Agent]]
            CosmeticLevelsArr := OldData.CosmeticsLevels
            CosLv1 := CosmeticLevelsArr[0]
            CosLv2 := CosmeticLevelsArr[1]
            CosLv3 := CosmeticLevelsArr[2]
            CosLv4 := CosmeticLevelsArr[3]
            CosLv5 := CosmeticLevelsArr[4]
            CosLv6 := CosmeticLevelsArr[5]
            CosLv7 := CosmeticLevelsArr[6]
            CosLv8 := CosmeticLevelsArr[7]
            CosLv9 := CosmeticLevelsArr[8]
            CosLv10 := CosmeticLevelsArr[9]
            CosLv11 := CosmeticLevelsArr[10]
            CosLv12 := CosmeticLevelsArr[11]
        then:
            CData1 := cosmetic_data:
                name:="Loot Chest"
                description:="+ Gold Per Elim"
                hint:="Hint: Hidden in the water"
                icon:=Icons.CosmeticIcons.Chest
                buttontext:="Info"
                level:=CosLv1
            CData2 := cosmetic_data:
                name:="Hiking Bag"
                description:=" + Armor"
                hint:="Hint: Complete Quests"
                icon:=Icons.CosmeticIcons.Backpack2
                buttontext:="Info"
                level:=CosLv2
            CData3 := cosmetic_data:
                name:="Guitar"
                description:="+ Lightning Chance\n+ Damage "
                hint:="Hint: Climbing area\n in the tundra"
                icon:=Icons.CosmeticIcons.Guitar2
                buttontext:="Info"
                level:=CosLv3
            CData4 := cosmetic_data:
                name:="Jetpack"
                description:="+ Damage"
                hint:="Hint: On top of a  \n volcano "
                icon:=Icons.CosmeticIcons.JetPack
                buttontext:="Info"
                level:=CosLv4
            CData5 := cosmetic_data:
                name:="Teddy"
                description:="++ Armor"
                hint:="Hint: High up in the \n the desert"
                icon:=Icons.CosmeticIcons.Teddy
                buttontext:="Info"
                level:=CosLv4
            CData6 := cosmetic_data:
                name:="Steel Sword"
                description:="++ Damage"
                hint:="Hint: Hidden in the lava"
                icon:=Icons.CosmeticIcons.Sword
                buttontext:="Info"
                level:=CosLv5
            CData7 := cosmetic_data:
                name:="Skateboard"
                description:="+ Crit Chance"
                hint:="Hint: Hidden in the snow"
                icon:=Icons.CosmeticIcons.Skateboard
                buttontext:="Info"
                level:=CosLv6
            CData8 := cosmetic_data:
                name:="Surfboard"
                description:="+ Xp"
                hint:="Hint: Hidden in the sand"
                icon:=Icons.CosmeticIcons.Surfboard
                buttontext:="Info"
                level:=CosLv7
            CData9 := cosmetic_data:
                name:="Dual wield"
                description:="++ Damage \n + Crit Chance"
                hint:="Hint: Complete Quests"
                icon:=Icons.CosmeticIcons.Swords
                buttontext:="Info"
                level:=CosLv8
            CData10 := cosmetic_data:
                name:="Shell"
                description:="++ Armor\n + Shield"
                hint:="Hint: Treasure map found\n in the volcano"
                icon:=Icons.CosmeticIcons.Shell
                buttontext:="Info"
                level:=CosLv9
            CData11 := cosmetic_data:
                name:="Angel Wings"
                description:="15% attack"
                hint:="Hint: Fly to the top of \n the tundra tower"
                icon:=Icons.CosmeticIcons.AngelWings
                buttontext:="Info"
                level:=CosLv10
            CData12 := cosmetic_data:
                name:="Red Wings"
                description:="+ Fire Chance\n + Crit Chanec\n+ Damage"
                hint:="Hint: Complete Quests "
                icon:=Icons.CosmeticIcons.RedWings
                buttontext:="Info"
                level:=CosLv11

                
            set CosmeticDatas = array{CData1,CData2,CData3,CData4,CData5,CData6,CData7,CData8,CData9,CData10,CData11,CData12}

            for(i:=0..CosmeticDatas.Length-1):
                if(CData := CosmeticDatas[i]):
                    if:
                        PData := Data[player[Agent]]
                        PData.CosmeticsUnlocked[i] = true

                    then:
                    else:
                        set CData.buttontext = "Locked"
                        set CData.name = ""
                        set CData.description = CData.hint
                        set CData.icon = wooden_pixel_gui.textures.Icons2.Size16.t_LockGrey_16


        return CosmeticDatas

    CreateTabTwo(Agent:agent):[]cosmetic_data=


        var CosmeticDatas : []cosmetic_data = array{}

        if:
            OldData:=Data[player[Agent]]
            CosmeticLevelsArr := OldData.CosmeticsLevels
            CosLv1 := CosmeticLevelsArr[12]
            CosLv2 := CosmeticLevelsArr[13]
            CosLv3 := CosmeticLevelsArr[14]
            CosLv4 := CosmeticLevelsArr[15]
            CosLv5 := CosmeticLevelsArr[16]
            CosLv6 := CosmeticLevelsArr[17]
            CosLv7 := CosmeticLevelsArr[18]
            CosLv8 := CosmeticLevelsArr[19]
            CosLv9 := CosmeticLevelsArr[20]
            CosLv10 := CosmeticLevelsArr[21]
            CosLv11 := CosmeticLevelsArr[22]
            CosLv12 := CosmeticLevelsArr[23]
        then:
        
            CData1 := cosmetic_data:
                name:="Green Cap"
                description:="+ Damage"
                hint:="Hint: Complete the tutorials "
                icon:=Icons.CosmeticIcons.Greenhat
                buttontext:="Info"
                level:=CosLv1
            CData2 := cosmetic_data:
                name:="Box Hat"
                description:="+ Crit Damage"
                hint:="Hint: Climbing area in \n the desert "
                icon:=Icons.CosmeticIcons.Cardboard
                buttontext:="Info"
                level:=CosLv2
            CData3 := cosmetic_data:
                name:="Cone"
                description:="+ Xp"
                hint:="Hint: On top of the \n Grasslands "
                icon:=Icons.CosmeticIcons.Cone
                buttontext:="Info"
                level:=CosLv3
            CData4 := cosmetic_data:
                name:="Halo"
                description:="+ Armor"
                hint:="Hint: ???? "
                icon:=Icons.CosmeticIcons.Halo
                buttontext:="Info"
                level:=CosLv4
            CData5 := cosmetic_data:
                name:="Farmer Hat"
                description:="+ Gold"
                hint:="Hint: Treasure map found\n in the grasslands"
                icon:=Icons.CosmeticIcons.Farmer
                buttontext:="Info"
                level:=CosLv5
            CData6 := cosmetic_data:
                name:="Stone Head"
                description:="++ Armor"
                hint:="Hint: ????"
                icon:=Icons.CosmeticIcons.StoneHead
                buttontext:="Info"
                level:=CosLv6
            CData7 := cosmetic_data:
                name:="Jester"
                description:="++ Crit Damage"
                hint:="Hint: The desert holds many\n secrets."
                icon:=Icons.CosmeticIcons.Jester
                buttontext:="Info"
                level:=CosLv7
            CData8 := cosmetic_data:
                name:="Crown"
                description:="++ Damage"
                hint:="Hint: Treasure map found\n in the tundra"
                icon:=Icons.CosmeticIcons.Crown
                buttontext:="Info"
                level:=CosLv8
            CData9 := cosmetic_data:
                name:="Black Mask"
                description:="+++ Damage"
                hint:="Hint: ????"
                icon:=Icons.CosmeticIcons.Mask
                buttontext:="Info"
                level:=CosLv9
            CData10 := cosmetic_data:
                name:="Horns"
                description:="+ Burn Chance\n + Crit Chance"
                hint:="Hint: ????"
                icon:=Icons.CosmeticIcons.Horns
                buttontext:="Info"
                level:=CosLv10
            CData11 := cosmetic_data:
                name:="Knight"
                description:="++ Armor \n++ Damage"
                hint:="Hint: ????"
                icon:=Icons.CosmeticIcons.KnightHelm
                buttontext:="Info"
                level:=CosLv11
            CData12 := cosmetic_data:
                name:="Dark Helm"
                description:="++ Damage\n ++ Fire Damage"
                hint:="Hint: Complete Quests "
                icon:=Icons.CosmeticIcons.BossHelm
                buttontext:="Info"
                level:=CosLv12

            set CosmeticDatas = array{CData1,CData2,CData3,CData4,CData5,CData6,CData7,CData8,CData9,CData10,CData11,CData12}

            for(i:=0..CosmeticDatas.Length-1):
                if(CData := CosmeticDatas[i]):
                    if:
                        PData := Data[player[Agent]]
                        PData.CosmeticsUnlocked[i+12] = true

                    then:
                    else:
                        set CData.buttontext = "Locked"
                        set CData.name = ""
                        set CData.description = CData.hint
                        set CData.icon = wooden_pixel_gui.textures.Icons2.Size16.t_LockGrey_16

        return CosmeticDatas


    RemoveHUD(Wdata:widget_message):void =
        
        Print("Removing")
        if(Player:=Wdata.Player, PlayerUI:=GetPlayerUI[Player],SelectedButton:=Wdata.Source,Agent:=agent[Player]):
            if(MyActualUI := MyUI[Agent]?):


                CloseAudio.Play(Agent)

                PlayerUI.RemoveWidget(MyActualUI)
                MyActualUI.SetVisibility(widget_visibility.Visible)
                MyActualUI.SetVisibility(widget_visibility.Hidden)

                if(set MyUI[Agent] = false){}
                ShowFortHud(Agent)

    OpenShopHandle(Agent:agent):void=

        OpenAudio.Play(Agent)

        RemoveFortHud(Agent)
        MyCanvas: canvas = MakeEntireBackground(Agent)

        if(PlayerUI:=GetPlayerUI[player[Agent]]):
            PlayerUI.AddWidget(MyCanvas,player_ui_slot{InputMode:=ui_input_mode.All}) # want it to appear on top
            MyCanvas.SetVisibility(widget_visibility.Hidden)
            MyCanvas.SetVisibility(widget_visibility.Visible)
        if(set MyUI[Agent]= option{MyCanvas}){}

        UpdateUiText(Agent)

    UpdateUiText(Agent:agent):void=
        block:

    CompleteBoughtAnims(Agent:agent):void= 
        PurchaseAudio.Play(Agent)
        UpdateUiText(Agent)
        ButtonClickSound.Play(Agent)



    CalculateColor(Level:int):color=

        var sat : float = 0.0
        var hue : float =  0.0
        var value : float = 1.0

        WhiteEnd := 10

        if(Level <= WhiteEnd):
            set sat = ((Level * 1.0) - 1.0) / ((WhiteEnd*1.0) - 1.0)
            set hue = 120.0/360.0
        else:
            set sat = 1.0
            set hue = ((Level * 1.0) - 1.0)  * (360.0 / 99.0)

        return MakeColorFromHSV(hue,sat,value)

