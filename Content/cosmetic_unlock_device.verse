
using { /Fortnite.com/Devices }
using { /Verse.org/Simulation }
using { /UnrealEngine.com/Temporary/Diagnostics }
using { /Verse.org/Random }

CosmeticStats := class():
    var StatMap : [string]float


cosmetic_unlock_device := class(creative_device):

    @editable
    CosmeticUnlockAccolade : accolades_device = accolades_device{}

    @editable
    CosmeticUnlockHMD : hud_message_device = hud_message_device{}

    @editable
    OpenMapAudio : audio_player_device = audio_player_device{}
    @editable
    CloseMapAudio : audio_player_device = audio_player_device{}

    @editable
    Grass1Button : button_device = button_device{}

    @editable
    Grass2Mutator : mutator_zone_device = mutator_zone_device{}
    @editable
    Grass2Mutator1 : mutator_zone_device = mutator_zone_device{}
    @editable
    Grass2Mutator2 : mutator_zone_device = mutator_zone_device{}

    @editable
    DesertMutator : mutator_zone_device = mutator_zone_device{}
    @editable
    DesertMutator1 : mutator_zone_device = mutator_zone_device{}
    @editable
    DesertMutator2 : mutator_zone_device = mutator_zone_device{}

    @editable
    SnowMutator : mutator_zone_device = mutator_zone_device{}
    @editable
    SnowMutator1 : mutator_zone_device = mutator_zone_device{}
    @editable
    SnowMutator2 : mutator_zone_device = mutator_zone_device{}

    @editable
    LavaMutator : mutator_zone_device = mutator_zone_device{}
    @editable
    LavaMutator1 : mutator_zone_device = mutator_zone_device{}
    @editable
    LavaMutator2 : mutator_zone_device = mutator_zone_device{}
    

    @editable
    MapButton : button_device = button_device{}
    @editable
    MapPopups : []popup_dialog_device = array{}

    @editable
    MapButton1 : button_device = button_device{}
    @editable
    MapPopups1 : []popup_dialog_device = array{}

    @editable
    MapButton2 : button_device = button_device{}
    @editable
    MapPopups2 : []popup_dialog_device = array{}

    @editable
    MapButton3 : button_device = button_device{}
    @editable
    MapPopups3 : []popup_dialog_device = array{}
    



    @editable
    ConeButton : button_device = button_device{}
    @editable
    SwordButton : button_device = button_device{}
    @editable
    CardBoardButton : button_device = button_device{}
    @editable
    JetPackButton : button_device = button_device{}
    @editable
    GuitarButton : button_device = button_device{}
    @editable
    HaloButton : button_device = button_device{} 


    @editable
    LootChestButton : button_device = button_device{}
    @editable
    SurfboardButton : button_device = button_device{}
    @editable
    SkateboardButton : button_device = button_device{}
    @editable
    WhiteWingsButton : button_device = button_device{}

    @editable
    JesterButton : button_device = button_device{}

    @editable
    BossBoar : npc_spawner_device = npc_spawner_device{}
    @editable
    BossBoarDropOdds : int = 5
    @editable
    DesertBoss : npc_spawner_device = npc_spawner_device{}
    @editable
    DesertBossOdds : int = 4
    @editable
    SnowBoss : npc_spawner_device = npc_spawner_device{}
    @editable
    SnowBossOdds : int = 3
    @editable
    LavaBoss : npc_spawner_device = npc_spawner_device{}
    @editable
    LavaBossOdds : int = 2

    @editable
    BossTrigger1 : trigger_device = trigger_device{}
    @editable
    BossTrigger2 : trigger_device = trigger_device{}
    @editable
    BossTrigger3 : trigger_device = trigger_device{}
    @editable
    BossTrigger4 : trigger_device = trigger_device{}

    @editable
    SecretAudio : audio_player_device = audio_player_device{}

    @editable
    HatRewardTrigger : trigger_device = trigger_device{}

    @editable
    Spawners : []player_spawner_device = array{}

    var AgentCosmeticStats : [agent]CosmeticStats = map{} # Holds total cosmetic damage
    var AgentCosmeticStatsPerLevel : [agent][]CosmeticStats = map{} # holds all cosmetic stats changed by levels

    @editable
    QuestCompletedTierUpTrigger : trigger_device = trigger_device{}

    @editable
    VFXPickups : []trigger_device = array{}

    QuestsCompletedRewards : []int = array{1,8,11,23,23}
    CosmeticNames : []string = array{"Loot Chest", "Hiking Bag","Guitar","Jetpack","Teddy","Steel Sword","Skateboard","Surfboard","Dual wield","Shell","Angel Wings","Red Wings","Green Cap","Box Hat","Cone","Halo","Farmer Hat","Stone Head","Jester","Crown","Black Mask","Horns","Knight","Dark Helm" }

    storedstring<localizes>(S:string) : message = "{S}"

    @editable
    GrassUnlockMissionTrigger : trigger_device = trigger_device{}

    OnBegin<override>()<suspends>:void=

        for(Map:MapPopups):
            Map.DismissedEvent.Subscribe(PlayCloseAudio)
        for(Map:MapPopups1):
            Map.DismissedEvent.Subscribe(PlayCloseAudio)

        MapButton.InteractedWithEvent.Subscribe(OpenMap1)
        MapButton1.InteractedWithEvent.Subscribe(OpenMap2)
        MapButton2.InteractedWithEvent.Subscribe(OpenMap3)
        MapButton3.InteractedWithEvent.Subscribe(OpenMap4)

        Grass2Mutator.AgentBeginsEmotingEvent.SubscribeAgent(CheckCosmeticGrass2,0)
        Grass2Mutator1.AgentBeginsEmotingEvent.SubscribeAgent(CheckCosmeticGrass2,1)
        Grass2Mutator2.AgentBeginsEmotingEvent.SubscribeAgent(CheckCosmeticGrass2,2)  # Farmer Hat h

        DesertMutator.AgentBeginsEmotingEvent.SubscribeAgent(CheckCosmeticDesertTreasure,0) # Teddy Bear bb
        DesertMutator1.AgentBeginsEmotingEvent.SubscribeAgent(CheckCosmeticDesertTreasure,1)
        DesertMutator2.AgentBeginsEmotingEvent.SubscribeAgent(CheckCosmeticDesertTreasure,2)

        SnowMutator.AgentBeginsEmotingEvent.SubscribeAgent(CheckCosmeticSnowTreasure,0) # Crown h
        SnowMutator1.AgentBeginsEmotingEvent.SubscribeAgent(CheckCosmeticSnowTreasure,1)
        SnowMutator2.AgentBeginsEmotingEvent.SubscribeAgent(CheckCosmeticSnowTreasure,2)

        LavaMutator.AgentBeginsEmotingEvent.SubscribeAgent(CheckCosmeticLavaTreasure,0)
        LavaMutator1.AgentBeginsEmotingEvent.SubscribeAgent(CheckCosmeticLavaTreasure,1)
        LavaMutator2.AgentBeginsEmotingEvent.SubscribeAgent(CheckCosmeticLavaTreasure,2) # Turtle Shell bb

        #########################################################################################

        
        # Invis Jumps Chests
        ConeButton.InteractedWithEvent.SubscribeAgent(UnlockCosmetic,14) # Cone  # Grass h
        CardBoardButton.InteractedWithEvent.SubscribeAgent(UnlockCosmetic,13) # CardBoard # Desert h 
        GuitarButton.InteractedWithEvent.SubscribeAgent(UnlockCosmetic,2) # Guitar # Snow bb 
        JetPackButton.InteractedWithEvent.SubscribeAgent(UnlockCosmetic,3) # JetPack # Lava bb
        
        # Breaks
        LootChestButton.InteractedWithEvent.SubscribeAgent(UnlockCosmetic,6)   # Chest # Grass bb # Switched with skateboard both give money 
        SurfboardButton.InteractedWithEvent.SubscribeAgent(UnlockCosmetic,7) # SurfBoard # Desert bb
        SkateboardButton.InteractedWithEvent.SubscribeAgent(UnlockCosmetic,0) # SkateBoard # Snow bb # Switched with loot chest
        SwordButton.InteractedWithEvent.SubscribeAgent(UnlockCosmetic,5) # Sword # Lava bb


        # Outside
        JesterButton.InteractedWithEvent.SubscribeAgent(UnlockCosmetic,18) # Jester Hat # Desert h
        WhiteWingsButton.InteractedWithEvent.SubscribeAgent(UnlockCosmetic,10) # white wings # Snow bb
        HaloButton.InteractedWithEvent.SubscribeAgent(UnlockCosmetic,15) # Far Outside map super hidden h

        # Boss Drops
        #BossBoar.EliminatedEvent.SubscribeAiDevice(RollOdds, array{17,BossBoarDropOdds}) # moyai hat
        #DesertBoss.EliminatedEvent.SubscribeAiDevice(RollOdds, array{20,DesertBossOdds}) # Mask Hat
        #SnowBoss.EliminatedEvent.SubscribeAiDevice(RollOdds, array{22,SnowBossOdds}) # Knight Hat
        #LavaBoss.EliminatedEvent.SubscribeAiDevice(RollOdds, array{21,LavaBossOdds}) # Horn Hat

        BossTrigger1.TriggeredEvent.SubscribeMaybeAgent(RollOddsTrigger,0)
        BossTrigger2.TriggeredEvent.SubscribeMaybeAgent(RollOddsTrigger,1)
        BossTrigger3.TriggeredEvent.SubscribeMaybeAgent(RollOddsTrigger,2)
        BossTrigger4.TriggeredEvent.SubscribeMaybeAgent(RollOddsTrigger,3)

        #array{12,1,8,11,23}    12 Green Hat , 1 BackPack, 8 Swords bb, 11 RedWings bb, 23 Helmboss h

        QuestCompletedTierUpTrigger.TriggeredEvent.Subscribe(AcheivementRewardCos)

        HatRewardTrigger.TriggeredEvent.Subscribe(UnlockHat)
        
        

        # This is to set stats at the start of the game
        Players:=GetPlayspace().GetPlayers()
        for(Player:Players):
            if(Agent:=agent[Player]):
                Init(Agent)
        for(i:=0..Spawners.Length-1):
            if(Spawner:= Spawners[i]):
                Spawner.SpawnedEvent.Subscribe(Init)

    UnlockHat(Agent:?agent):void=
        if(MyAgent:=Agent?):
            UnlockCosmetic(MyAgent,12)

    AcheivementRewardCos(Agent:?agent):void=
        if:
            MyAgent:= Agent?
            Player := player[MyAgent]
            MyData := Data[Player]
            AchievementTier := MyData.QuestAchievementTier
            MyReward := QuestsCompletedRewards[AchievementTier]
        then:
            UnlockCosmetic(MyAgent,MyReward)
            if(set Data[Player] = player_data{player_data_update<constructor>(MyData),QuestAchievementTier := MyData.QuestAchievementTier + 1}){}


    RollOddsTrigger(Agent:?agent,id:int):void=

        var CosmeticId : int = 17
        var CosmeticOdds : int = 0

        if(id = 0):
            set CosmeticOdds = BossBoarDropOdds
            set CosmeticId = 17
        if(id = 1):
            set CosmeticOdds = DesertBossOdds
            set CosmeticId = 20
        if(id = 2):
            set CosmeticOdds = SnowBossOdds
            set CosmeticId = 22
        if(id = 3):
            set CosmeticOdds = LavaBossOdds
            set CosmeticId = 21

        if(MyAgent:= Agent?):
            RandomOdds := GetRandomInt(0,100)
            Print("Random Cosmetic Roll = {RandomOdds}")
            Print("Random Cosmetic Odds =  {CosmeticOdds}")
            if(RandomOdds <= CosmeticOdds):
                UnlockCosmetic(MyAgent,CosmeticId)


    RollOdds(Interaction:device_ai_interaction_result,Arr:[]int):void=
        RandomOdds := GetRandomInt(0,100)
        if (MyOdds := Arr[1]):
            Print("Random Cosmetic Roll = {RandomOdds}")
            Print("Random Cosmetic Odds =  {MyOdds}")
            if(RandomOdds <= MyOdds):
                if(Agent := Interaction.Source?):
                    if(MyReward := Arr[0]):
                        UnlockCosmetic(Agent,MyReward)

    
    InitRandomEmoteLocations(Agent:agent):void= # should init

        if:
            Player := player[Agent]
            MyData := Data[Player]
            var MyRandArr : []int = MyData.RandomMapLocations
        then:

            Random1 := GetRandomInt(0,2)
            Print("Random1 := {Random1}")
            Random2 := GetRandomInt(0,2)
            Print("Random2 := {Random2}")
            Random3 := GetRandomInt(0,2)
            Print("Random3 := {Random3}")
            Random4 := GetRandomInt(0,2)
            Print("Random4 := {Random4}")

            if(set MyRandArr[0] = Random1){}
            if(set MyRandArr[1] = Random2){}
            if(set MyRandArr[2] = Random3){}
            if(set MyRandArr[3] = Random4){}

            if(set Data[Player] = player_data{player_data_update<constructor>(MyData),RandomMapLocations := MyRandArr}){}
                    
    CheckCosmeticGrass2(Agent:agent,id:int):void= # Farmer hat
        if:
            Player := player[Agent]
            MyData := Data[Player]
            MyData.RandomMapLocations[0] = id
        then:
            UnlockCosmetic(Agent,16)
            GrassUnlockMissionTrigger.Trigger(Agent)

    CheckCosmeticDesertTreasure(Agent:agent,id:int):void= # teddy
        if:
            Player := player[Agent]
            MyData := Data[Player]
            MyData.RandomMapLocations[1] = id
        then:
            UnlockCosmetic(Agent,4)

    CheckCosmeticSnowTreasure(Agent:agent,id:int):void= # Crown
        if:
            Player := player[Agent]
            MyData := Data[Player]
            MyData.RandomMapLocations[2] = id
        then:
            UnlockCosmetic(Agent,19)

    CheckCosmeticLavaTreasure(Agent:agent,id:int):void= # Turtle Shell
        if:
            Player := player[Agent]
            MyData := Data[Player]
            MyData.RandomMapLocations[3] = id
        then:
            UnlockCosmetic(Agent,9)

    PlayCloseAudio(Agent:agent):void=
        CloseMapAudio.Play(Agent)

    OpenMap1(Agent:agent):void=
        if:
            Player := player[Agent]
            MyData := Data[Player]
            MyRand := MyData.RandomMapLocations[0]
            MyPopup := MapPopups[MyRand]
        then:
            MyPopup.Show(Agent)
            OpenMapAudio.Play(Agent)

    OpenMap2(Agent:agent):void=
        if:
            Player := player[Agent]
            MyData := Data[Player]
            MyRand := MyData.RandomMapLocations[1]
            MyPopup := MapPopups1[MyRand]
        then:
            MyPopup.Show(Agent)
            OpenMapAudio.Play(Agent)

    OpenMap3(Agent:agent):void=
        if:
            Player := player[Agent]
            MyData := Data[Player]
            MyRand := MyData.RandomMapLocations[2]
            MyPopup := MapPopups2[MyRand]
        then:
            MyPopup.Show(Agent)
            OpenMapAudio.Play(Agent)
    
    OpenMap4(Agent:agent):void=
        if:
            Player := player[Agent]
            MyData := Data[Player]
            MyRand := MyData.RandomMapLocations[3]
            MyPopup := MapPopups3[MyRand]
        then:
            MyPopup.Show(Agent)
            OpenMapAudio.Play(Agent)
    
    UnlockCosmetic(Agent:agent,id:int):void=

        if:
            Player:=player[Agent]
            MyData := Data[Player]
            var NewStat : []logic =  MyData.CosmeticsUnlocked
            NewStat[id] = false
            set NewStat[id] = true
            set Data[Player] = player_data{player_data_update<constructor>(MyData),CosmeticsUnlocked := NewStat}
        then:
            SecretAudio.Play(Agent)
            CosmeticUnlockAccolade.Award(Agent)
            if (Name:=CosmeticNames[id]):
                CosmeticUnlockHMD.SetText(storedstring(Name))
                CosmeticUnlockHMD.Show(Agent)

            


    Init(Agent:agent):void= # used to create stats class for eachplayer

        if:
            MyData := Data[player[Agent]]
            EquipedBackPack := MyData.BackCosmeticEquiped
            EquipedHead := MyData.HatCosmeticEquiped
        then:

            if(EquipedBackPack = 13):
            else:
                EquipeCosmetic(Agent, EquipedBackPack)
            if(EquipedHead = 13):
            else:
                EquipeCosmetic(Agent,EquipedHead+12)

            if(Testing := AgentCosmeticStats[Agent]):
            else:
                UpdateCosmeticStats(Agent)

    EquipeCosmetic(Agent:agent,id:int) : void =  
        if(MyTrigger := VFXPickups[id]):
            if(MyData :=  Data[player[Agent]]):
                if(MyData.CosmeticsUnlocked[id] = true):
                    MyTrigger.Trigger(Agent)
                    if(id <= 11):
                        if:
                            set Data[player[Agent]] = player_data{player_data_update<constructor>(MyData),BackCosmeticEquiped := id}
                    else:
                        if:
                            set Data[player[Agent]] = player_data{player_data_update<constructor>(MyData),HatCosmeticEquiped := id-12}
        UpdateCosmeticStats(Agent)

    LoadAllCosmeticStats():[]CosmeticStats=

        AllStats : []CosmeticStats = array:

            ChestStats := CosmeticStats :

                StatMap := map{"Damage"=>0.0,
                    "Crit Chance" => 0.0,
                    "Crit Damage" => 0.0,
                    "Burn Damage" => 0.0,
                    "Burn Chance" => 0.0,
                    "Armor" => 0.0,
                    "Lightning Chance" => 5.0,
                    "Lightning Damage" => 5.0,
                    "GoldPerElim" => 50.0,
                    "Xp" => 0.0}

            BackpackStats := CosmeticStats :

                StatMap := map{"Damage"=>0.0,
                    "Crit Chance" => 0.0,
                    "Crit Damage" => 0.0,
                    "Burn Damage" => 0.0,
                    "Burn Chance" => 0.0,
                    "Armor" => 25.0,
                    "Lightning Chance" => 0.0,
                    "Lightning Damage" => 0.0,
                    "GoldPerElim" => 0.0,
                    "Xp" => 0.0}

            GuitarStats := CosmeticStats :

                StatMap := map{"Damage"=>0.0,
                    "Crit Chance" => 0.0,
                    "Crit Damage" => 0.0,
                    "Burn Damage" => 0.0,
                    "Burn Chance" => 0.0,
                    "Armor" => 0.0,
                    "Lightning Chance" => 20.0,
                    "Lightning Damage" => 20.0,
                    "GoldPerElim" => 0.0,
                    "Xp" => 0.0}

            JetpackStats := CosmeticStats :

                StatMap := map{"Damage"=>0.0,
                    "Crit Chance" => 0.0,
                    "Crit Damage" => 0.0,
                    "Burn Damage" => 0.0,
                    "Burn Chance" => 0.0,
                    "Armor" => 10.0,
                    "Lightning Chance" => 50.0,
                    "Lightning Damage" => 50.0,
                    "GoldPerElim" => 0.0,
                    "Xp" => 0.0}

            TeddyStats := CosmeticStats :

                StatMap := map{"Damage"=>0.0,
                    "Crit Chance" => 0.0,
                    "Crit Damage" => 0.0,
                    "Burn Damage" => 20.0,
                    "Burn Chance" => 20.0,
                    "Armor" => 0.0,
                    "Lightning Chance" => 0.0,
                    "Lightning Damage" => 0.0,
                    "GoldPerElim" => 0.0,
                    "Xp" => 0.0}

            SwordStats := CosmeticStats :

                StatMap := map{"Damage"=> 100.0,
                    "Crit Chance" => 0.0,
                    "Crit Damage" => 0.0,
                    "Burn Damage" => 5.0,
                    "Burn Chance" => 5.0,
                    "Armor" => 0.0,
                    "Lightning Chance" => 0.0,
                    "Lightning Damage" => 0.0,
                    "GoldPerElim" => 0.0,
                    "Xp" => 0.0}

            SkateboardStats := CosmeticStats :

                StatMap := map{"Damage"=> 0.0,
                    "Crit Chance" => 0.0,
                    "Crit Damage" => 0.0,
                    "Burn Damage" => 0.0,
                    "Burn Chance" => 0.0,
                    "Armor" => 0.0,
                    "Lightning Chance" => 0.0,
                    "Lightning Damage" => 0.0,
                    "GoldPerElim" => 20.0,
                    "Xp" => 0.0}

            SurfboardStats := CosmeticStats :

                StatMap := map{"Damage"=> 0.0,
                    "Crit Chance" => 20.0,
                    "Crit Damage" => 20.0,
                    "Burn Damage" => 0.0,
                    "Burn Chance" => 0.0,
                    "Armor" => 0.0,
                    "Lightning Chance" => 0.0,
                    "Lightning Damage" => 0.0,
                    "GoldPerElim" => 0.0,
                    "Xp" => 0.0}

            Swords2Stats := CosmeticStats :

                StatMap := map{"Damage"=> 50.0,
                    "Crit Chance" => 0.0,
                    "Crit Damage" => 0.0,
                    "Burn Damage" => 0.0,
                    "Burn Chance" => 0.0,
                    "Armor" => 0.0,
                    "Lightning Chance" => 0.0,
                    "Lightning Damage" => 0.0,
                    "GoldPerElim" => 0.0,
                    "Xp" => 0.0}

            ShellStats := CosmeticStats :

                StatMap := map{"Damage"=> 15.0,
                    "Crit Chance" => 0.0,
                    "Crit Damage" => 0.0,
                    "Burn Damage" => 0.0,
                    "Burn Chance" => 0.0,
                    "Armor" => 60.0,
                    "Lightning Chance" => 0.0,
                    "Lightning Damage" => 0.0,
                    "GoldPerElim" => 0.0,
                    "Xp" => 0.0}

            WhiteWings := CosmeticStats :

                StatMap := map{"Damage"=> 0.0,
                    "Crit Chance" => 40.0,
                    "Crit Damage" => 40.0,
                    "Burn Damage" => 0.0,
                    "Burn Chance" => 0.0,
                    "Armor" => 40.0,
                    "Lightning Chance" => 0.0,
                    "Lightning Damage" => 0.0,
                    "GoldPerElim" => 0.0,
                    "Xp" => 0.0}

            RedWings := CosmeticStats :

                StatMap := map{"Damage"=> 0.0,
                    "Crit Chance" => 40.0,
                    "Crit Damage" => 40.0,
                    "Burn Damage" => 0.0,
                    "Burn Chance" => 0.0,
                    "Armor" => 0.0,
                    "Lightning Chance" => 0.0,
                    "Lightning Damage" => 0.0,
                    "GoldPerElim" => 20.0,
                    "Xp" => 0.0}

            HatStats := CosmeticStats :

                StatMap := map{"Damage"=> 0.0,
                    "Crit Chance" => 0.0,
                    "Crit Damage" => 0.0,
                    "Burn Damage" => 0.0,
                    "Burn Chance" => 0.0,
                    "Armor" => 25.0,
                    "Lightning Chance" => 0.0,
                    "Lightning Damage" => 0.0,
                    "GoldPerElim" => 0.0,
                    "Xp" => 0.0}

            CardBoardStats := CosmeticStats :

                StatMap := map{"Damage"=> 0.0,
                    "Crit Chance" => 0.0,
                    "Crit Damage" => 0.0,
                    "Burn Damage" => 20.0,
                    "Burn Chance" => 20.0,
                    "Armor" => 0.0,
                    "Lightning Chance" => 0.0,
                    "Lightning Damage" => 0.0,
                    "GoldPerElim" => 0.0,
                    "Xp" => 0.0}

            ConeStats := CosmeticStats :

                StatMap := map{"Damage"=> 0.0,
                    "Crit Chance" => 20.0,
                    "Crit Damage" => 20.0,
                    "Burn Damage" => 0.0,
                    "Burn Chance" => 0.0,
                    "Armor" => 0.0,
                    "Lightning Chance" => 0.0,
                    "Lightning Damage" => 0.0,
                    "GoldPerElim" => 0.0,
                    "Xp" => 0.0}

            HaloStats := CosmeticStats :

                StatMap := map{"Damage"=> 0.0,
                    "Crit Chance" => 0.0,
                    "Crit Damage" => 0.0,
                    "Burn Damage" => 0.0,
                    "Burn Chance" => 0.0,
                    "Armor" => 80.0,
                    "Lightning Chance" => 0.0,
                    "Lightning Damage" => 0.0,
                    "GoldPerElim" => 4.0,
                    "Xp" => 0.0}

            FarmerStats := CosmeticStats :

                StatMap := map{"Damage"=> 0.0,
                    "Crit Chance" => 0.0,
                    "Crit Damage" => 0.0,
                    "Burn Damage" => 0.0,
                    "Burn Chance" => 0.0,
                    "Armor" => 0.0,
                    "Lightning Chance" => 0.0,
                    "Lightning Damage" => 0.0,
                    "GoldPerElim" => 20.0,
                    "Xp" => 0.0}

            StoneHeadStats := CosmeticStats :

                StatMap := map{"Damage"=> 25.0,
                    "Crit Chance" => 0.0,
                    "Crit Damage" => 0.0,
                    "Burn Damage" => 0.0,
                    "Burn Chance" => 0.0,
                    "Armor" => 0.0,
                    "Lightning Chance" => 0.0,
                    "Lightning Damage" => 0.0,
                    "GoldPerElim" => 0.0,
                    "Xp" => 0.0}
            
            Jester := CosmeticStats :

                StatMap := map{"Damage"=> 0.0,
                    "Crit Chance" => 0.0,
                    "Crit Damage" => 0.0,
                    "Burn Damage" => 0.0,
                    "Burn Chance" => 0.0,
                    "Armor" => 0.0,
                    "Lightning Chance" => 20.0,
                    "Lightning Damage" => 20.0,
                    "GoldPerElim" => 0.0,
                    "Xp" => 0.0}

            Crown := CosmeticStats :

                StatMap := map{"Damage"=> 0.0,
                    "Crit Chance" => 0.0,
                    "Crit Damage" => 0.0,
                    "Burn Damage" => 0.0,
                    "Burn Chance" => 0.0,
                    "Armor" => 0.0,
                    "Lightning Chance" => 5.0,
                    "Lightning Damage" => 5.0,
                    "GoldPerElim" => 50.0,
                    "Xp" => 0.0}

            Mask := CosmeticStats :

                StatMap := map{"Damage"=> 5.0,
                    "Crit Chance" => 0.0,
                    "Crit Damage" => 0.0,
                    "Burn Damage" => 40.0,
                    "Burn Chance" => 40.0,
                    "Armor" => 0.0,
                    "Lightning Chance" => 0.0,
                    "Lightning Damage" => 0.0,
                    "GoldPerElim" => 0.0,
                    "Xp" => 0.0}

            Horn := CosmeticStats :

                StatMap := map{"Damage"=> 0.0,
                    "Crit Chance" => 40.0,
                    "Crit Damage" => 40.0,
                    "Burn Damage" => 5.0,
                    "Burn Chance" => 5.0,
                    "Armor" => 0.0,
                    "Lightning Chance" => 0.0,
                    "Lightning Damage" => 0.0,
                    "GoldPerElim" => 0.0,
                    "Xp" => 0.0}

            Knight := CosmeticStats :

                StatMap := map{"Damage"=> 5.0,
                    "Crit Chance" => 0.0,
                    "Crit Damage" => 0.0,
                    "Burn Damage" => 0.0,
                    "Burn Chance" => 0.0,
                    "Armor" => 40.0,
                    "Lightning Chance" => 40.0,
                    "Lightning Damage" => 40.0,
                    "GoldPerElim" => 0.0,
                    "Xp" => 0.0}

            BossHelm := CosmeticStats :

                StatMap := map{"Damage"=> 150.0,
                    "Crit Chance" => 40.0,
                    "Crit Damage" => 150.0,
                    "Burn Damage" => 0.0,
                    "Burn Chance" => 0.0,
                    "Armor" => 0.0,
                    "Lightning Chance" => 0.0,
                    "Lightning Damage" => 0.0,
                    "GoldPerElim" => 0.0,
                    "Xp" => 0.0}
                
  
        return AllStats

            
    UpdateCosmeticStats(Agent:agent):void=

        LoadedStats : []CosmeticStats = LoadAllCosmeticStats()

        CalculateLevelDamage(Agent)

        var MyStats : CosmeticStats = CosmeticStats :
            StatMap := map{"Damage"=> 0.0,
                "Crit Chance" => 0.0,
                "Crit Damage" => 0.0,
                "Burn Damage" => 0.0,
                "Burn Chance" => 0.0,
                "Armor" => 0.0,
                "Lightning Chance" => 0.0,
                "Lightning Damage" => 0.0,
                "GoldPerElim" => 0.0,
                "Xp" => 0.0}

        if:
            MyData := Data[player[Agent]]
        then:

            if(MyData.BackCosmeticEquiped < 13):
                
                if(LoadedStat := LoadedStats[MyData.BackCosmeticEquiped]):
                    set MyStats = LoadedStat
                if(MyLevelStats := AgentCosmeticStatsPerLevel[Agent], CurrentEquipStats := MyLevelStats[MyData.BackCosmeticEquiped]):
                    if(set MyStats.StatMap["Damage"] += CurrentEquipStats.StatMap["Damage"]){}
                    if(set MyStats.StatMap["Crit Chance"] += CurrentEquipStats.StatMap["Crit Chance"]){}
                    if(set MyStats.StatMap["Crit Damage"] += CurrentEquipStats.StatMap["Crit Damage"]){}
                    if(set MyStats.StatMap["Burn Damage"] += CurrentEquipStats.StatMap["Burn Damage"]){}
                    if(set MyStats.StatMap["Burn Chance"] += CurrentEquipStats.StatMap["Burn Chance"]){}
                    if(set MyStats.StatMap["Armor"] += CurrentEquipStats.StatMap["Armor"]){}
                    if(set MyStats.StatMap["Lightning Chance"] += CurrentEquipStats.StatMap["Lightning Chance"]){}
                    if(set MyStats.StatMap["Lightning Damage"] += CurrentEquipStats.StatMap["Lightning Damage"]){}
                    if(set MyStats.StatMap["GoldPerElim"] += CurrentEquipStats.StatMap["GoldPerElim"]){}
                    if(set MyStats.StatMap["Xp"] += CurrentEquipStats.StatMap["Xp"]){}


            if(MyData.HatCosmeticEquiped < 13):
                if(LoadedStat := LoadedStats[MyData.HatCosmeticEquiped+12]):
                    if(set MyStats.StatMap["Damage"] += LoadedStat.StatMap["Damage"]){}
                    if(set MyStats.StatMap["Crit Chance"] += LoadedStat.StatMap["Crit Chance"]){}
                    if(set MyStats.StatMap["Crit Damage"] += LoadedStat.StatMap["Crit Damage"]){}
                    if(set MyStats.StatMap["Burn Damage"] += LoadedStat.StatMap["Burn Damage"]){}
                    if(set MyStats.StatMap["Burn Chance"] += LoadedStat.StatMap["Burn Chance"]){}
                    if(set MyStats.StatMap["Armor"] += LoadedStat.StatMap["Armor"]){}
                    if(set MyStats.StatMap["Lightning Chance"] += LoadedStat.StatMap["Lightning Chance"]){}
                    if(set MyStats.StatMap["Lightning Damage"] += LoadedStat.StatMap["Lightning Damage"]){}
                    if(set MyStats.StatMap["GoldPerElim"] += LoadedStat.StatMap["GoldPerElim"]){}
                    if(set MyStats.StatMap["Xp"] += LoadedStat.StatMap["Xp"]){}

                if(MyLevelStats := AgentCosmeticStatsPerLevel[Agent], CurrentEquipStats := MyLevelStats[MyData.HatCosmeticEquiped+12]):
                    if(set MyStats.StatMap["Damage"] += CurrentEquipStats.StatMap["Damage"]){}
                    if(set MyStats.StatMap["Crit Chance"] += CurrentEquipStats.StatMap["Crit Chance"]){}
                    if(set MyStats.StatMap["Crit Damage"] += CurrentEquipStats.StatMap["Crit Damage"]){}
                    if(set MyStats.StatMap["Burn Damage"] += CurrentEquipStats.StatMap["Burn Damage"]){}
                    if(set MyStats.StatMap["Burn Chance"] += CurrentEquipStats.StatMap["Burn Chance"]){}
                    if(set MyStats.StatMap["Armor"] += CurrentEquipStats.StatMap["Armor"]){}
                    if(set MyStats.StatMap["Lightning Chance"] += CurrentEquipStats.StatMap["Lightning Chance"]){}
                    if(set MyStats.StatMap["Lightning Damage"] += CurrentEquipStats.StatMap["Lightning Damage"]){}
                    if(set MyStats.StatMap["GoldPerElim"] += CurrentEquipStats.StatMap["GoldPerElim"]){}
                    if(set MyStats.StatMap["Xp"] += CurrentEquipStats.StatMap["Xp"]){}

        if(set AgentCosmeticStats[Agent] = MyStats){Print("Agent Cosmetic Set")}

        for(StatName->StatValue:MyStats.StatMap):
            Print("{StatName} = {StatValue}")


        # New Way Of Setting Cosmetic Stats For Players
        if:
            TestingStats := AgentCosmeticStats[Agent]
            Damage := TestingStats.StatMap["Damage"]
            CritChance := TestingStats.StatMap["Crit Chance"]
            CritDamage := TestingStats.StatMap["Crit Chance"]
            BurnDamage := TestingStats.StatMap["Burn Damage"]
            BurnChance := TestingStats.StatMap["Burn Chance"] # just use burn damage to save stuff 
            Armor:= TestingStats.StatMap["Armor"]
            LightningChance:= TestingStats.StatMap["Lightning Chance"]
            LightningDamage:= TestingStats.StatMap["Lightning Damage"]
            GoldPerElim:= TestingStats.StatMap["GoldPerElim"]

            MyData := Data[player[Agent]]
        then:
            if:
                set Data[player[Agent]] = player_data{player_data_update<constructor>(MyData),CosmeticDamage := Damage, CosmeticCritChance := CritChance, CosmeticCritDamage := CritDamage, CosmeticBurnDamage := BurnDamage, CosmeticArmor := Armor, CosmeticLightningChance := LightningChance, CosmeticLightningDamage := LightningDamage, CosmeticGoldPerElim := GoldPerElim }


        
            

    CalculateLevelDamage(Agent:agent):void=

        ChangedStats : []CosmeticStats = LoadAllCosmeticStats()

        if:
            MyData := Data[player[Agent]]
        then:
            for(i:=0..MyData.CosmeticsLevels.Length-1):
                if(CosLev := MyData.CosmeticsLevels[i]):
                    #Print("i cosmetic level : {CosLev}")
                    if(MyStats := ChangedStats[i]):
                        for(StatName->StatValue:MyStats.StatMap):
                            if(StatValue>0.0):
                                if(StatName = "Damage"):
                                    if(set MyStats.StatMap[StatName] = (CosLev-1) * 5.00){}
                                if(StatName = "Crit Chance"):
                                    if(set MyStats.StatMap[StatName] = (CosLev-1) * 2.50){}
                                if(StatName = "Crit Damage"):
                                    if(set MyStats.StatMap[StatName] = (CosLev-1) * 2.50){}
                                if(StatName = "Burn Damage"):
                                    if(set MyStats.StatMap[StatName] = (CosLev-1) * 4.00){}
                                if(StatName = "Burn Chance"):
                                    if(set MyStats.StatMap[StatName] = (CosLev-1) * 4.00){}
                                if(StatName = "Armor"):
                                    if(set MyStats.StatMap[StatName] = (CosLev-1) * 10.00){}
                                if(StatName = "Lightning Chance"):
                                    if(set MyStats.StatMap[StatName] = (CosLev-1) * 2.50){}
                                if(StatName = "Lightning Damage"):
                                    if(set MyStats.StatMap[StatName] = (CosLev-1) * 2.50){}
                                if(StatName = "GoldPerElim"):
                                    if(set MyStats.StatMap[StatName] = (CosLev-1) * 5.00){}



        if(set AgentCosmeticStatsPerLevel[Agent] = ChangedStats){}




                        

            


                        
                        

                        

                        
                   






            
        


