
using { /Fortnite.com/Devices }
using { /Verse.org/Simulation }
using { /UnrealEngine.com/Temporary/Diagnostics }
using { /UnrealEngine.com/Temporary/UI }
using { /UnrealEngine.com/Temporary/SpatialMath }
using { /Fortnite.com/UI }
using { /Verse.org/Colors }
using { /Fortnite.com/Playspaces }
using { /Fortnite.com/Characters }
using { /Verse.org/Assets }


currency_ui_device := class(creative_device):


    var MyUI : [agent]?canvas = map{}

    var AgentTextBlocks : [agent]?[]text_block = map{}
    var MainStackBox : [agent]?stack_box = map{}
    var ShowingBlockLogic : [agent]?[]logic = map{}

    var MyUI2 : [agent]?canvas = map{}

    var AgentTextBlocks2 : [agent]?[]text_block = map{}
    var MainStackBox2 : [agent]?stack_box = map{}
    var ShowingBlockLogic2 : [agent]?[]logic = map{}

    var OreStackBoxSlots : [agent]?[]stack_box_slot = map{}


    RefreshRate : float = 0.1

    @editable
    OreConditionals : []conditional_button_device = array{}

    @editable
    PlayerSpawners : []player_spawner_device = array{}

    storedstring<localizes>(S:string) : message = "{S}"

    OnBegin<override>()<suspends>:void=

        Sleep(0.5)

        AllPlayers := GetPlayspace().GetPlayers()
        for(Player:AllPlayers):
            if(Agent:=agent[Player]):
                ShowUi(Agent)

        GetPlayspace().PlayerAddedEvent().Subscribe(ShowUi)

        #for(Spawner:PlayerSpawners):
        #    Spawner.SpawnedEvent.Subscribe(ShowUi)


        GetPlayspace().PlayerRemovedEvent().Subscribe(PlayerRemoved)

    PlayerRemoved(Agent:agent):void=

        if(set MyUI[Agent] = false){}

        if(set AgentTextBlocks[Agent] = false){}
        if(set MainStackBox[Agent] = false){}
        if(set ShowingBlockLogic[Agent] = false){}

        if(set MyUI2[Agent] = false){}

        if(set AgentTextBlocks2[Agent] = false){}
        if(set MainStackBox2[Agent] = false){}
        if(set ShowingBlockLogic2[Agent] = false){}

        if(set OreStackBoxSlots[Agent] = false){}


    MakeCanvas(Agent:agent):canvas=

        TempTextBlock := text_block:
        if(set AgentTextBlocks[Agent] = option{array{TempTextBlock,TempTextBlock,TempTextBlock,TempTextBlock,TempTextBlock,TempTextBlock,TempTextBlock,TempTextBlock}}){}

        GemTexture : texture = wooden_pixel_gui.textures.Icons2.Size8.t_GemRed_8
        GoldTexture : texture = wooden_pixel_gui.textures.Icons2.Size8.t_CoinGold_8
        CopperTexture : texture = hud.Copper_icon512
        SilverTexture : texture = hud.Silver_icon
        MalachiteTexture : texture = hud.Malachite_icon512
        ObsidianTexture : texture = hud.Obsidian_icon512
        BrightCoreTexture : texture = hud.Brightcore_icon512
        SpectroliteTexture : texture = hud.Spectrolite_icon512

        GoldSlot : stack_box_slot = CreateOreSlot(GoldTexture,0,Agent)

        MainGoldStackBox := stack_box : 
            Orientation := orientation.Horizontal

        MainGoldStackBox.AddWidget(GoldSlot)

        if(set MainStackBox[Agent] = option{MainGoldStackBox}){}
        if(set ShowingBlockLogic[Agent] = option{array{false,false,false,false,false,false,false,false,false}}){}

        MyCanvas := canvas :
            Slots:=array:
                canvas_slot:
                    Anchors:=anchors{Minimum:=vector2{X:=0.05,Y:=0.7}, Maximum:=vector2{X:=0.05,Y:=0.7}}
                    Widget:=overlay:
                        Slots:=array:
                            overlay_slot:
                                HorizontalAlignment:=horizontal_alignment.Fill
                                VerticalAlignment:=vertical_alignment.Fill
                                Widget:=overlay:
                                    Slots:=array:
                                        overlay_slot:
                                            HorizontalAlignment:=horizontal_alignment.Fill
                                            VerticalAlignment:=vertical_alignment.Fill
                                            Padding:=margin{Left:=0.0,Top:=-5.0,Right:=0.0,Bottom:=0.0}
                                            Widget:=texture_block:
                                                DefaultImage:=wooden_pixel_gui.textures.Panels.WhiteSquare100
                                                DefaultTint:=MakeColorFromHex("D69E63FF")
                                        overlay_slot:
                                            HorizontalAlignment:=horizontal_alignment.Fill
                                            VerticalAlignment:=vertical_alignment.Fill
                                            Padding:=margin{Left:=0.0,Top:=0.0,Right:=-5.0,Bottom:=0.0}
                                            Widget:=texture_block:
                                                DefaultImage:=wooden_pixel_gui.textures.Panels.WhiteSquare100
                                                DefaultTint:=MakeColorFromHex("D69E63FF")
                                        overlay_slot:
                                            HorizontalAlignment:=horizontal_alignment.Fill
                                            VerticalAlignment:=vertical_alignment.Fill
                                            Padding:=margin{Left:=0.0,Top:=0.0,Right:=0.0,Bottom:=-5.0}
                                            Widget:=texture_block:
                                                DefaultImage:=wooden_pixel_gui.textures.Panels.WhiteSquare100
                                                DefaultTint:=MakeColorFromHex("93612FFF")
                                        overlay_slot:
                                            HorizontalAlignment:=horizontal_alignment.Fill
                                            VerticalAlignment:=vertical_alignment.Fill
                                            Padding:=margin{Left:=-5.0,Top:=0.0,Right:=0.0,Bottom:=0.0}
                                            Widget:=texture_block:
                                                DefaultImage:=wooden_pixel_gui.textures.Panels.WhiteSquare100
                                                DefaultTint:=MakeColorFromHex("93612FFF")
                                        overlay_slot:
                                            HorizontalAlignment:=horizontal_alignment.Fill
                                            VerticalAlignment:=vertical_alignment.Fill
                                            Padding:=margin{Left:=0.0,Top:=0.0,Right:=0.0,Bottom:=0.0}
                                            Widget:=texture_block:
                                                DefaultImage:=wooden_pixel_gui.textures.Panels.WhiteSquare100
                                                DefaultTint:=MakeColorFromHex("B87A3BFF")
                            overlay_slot:
                                HorizontalAlignment:=horizontal_alignment.Fill
                                VerticalAlignment:=vertical_alignment.Fill
                                Padding:=margin{Left:=10.0,Top:=0.0,Right:=0.0,Bottom:=0.0}
                                Widget:=MainGoldStackBox
                
        return MyCanvas

    MakeOreCanvas(Agent:agent):canvas=

        TempTextBlock := text_block:
        if(set AgentTextBlocks2[Agent] = option{array{TempTextBlock,TempTextBlock,TempTextBlock,TempTextBlock,TempTextBlock,TempTextBlock,TempTextBlock,TempTextBlock}}){}

        GemTexture : texture = wooden_pixel_gui.textures.Icons2.Size8.t_GemRed_8
        GoldTexture : texture = wooden_pixel_gui.textures.Icons2.Size8.t_CoinGold_8
        
        

        MainGoldStackBox := stack_box : 
            Orientation := orientation.Horizontal

        if(set MainStackBox2[Agent] = option{MainGoldStackBox}){}

        MyCanvas := canvas :
            Slots:=array:
                canvas_slot:
                    Anchors:=anchors{Minimum:=vector2{X:=0.05,Y:=0.78}, Maximum:=vector2{X:=0.05,Y:=0.78}}
                    Widget:=overlay:
                        Slots:=array:
                            overlay_slot:
                                HorizontalAlignment:=horizontal_alignment.Fill
                                VerticalAlignment:=vertical_alignment.Fill
                                Widget:=overlay:
                                    Slots:=array:
                                        overlay_slot:
                                            HorizontalAlignment:=horizontal_alignment.Fill
                                            VerticalAlignment:=vertical_alignment.Fill
                                            Padding:=margin{Left:=0.0,Top:=-5.0,Right:=0.0,Bottom:=0.0}
                                            Widget:=texture_block:
                                                DefaultImage:=wooden_pixel_gui.textures.Panels.WhiteSquare100
                                                DefaultTint:=MakeColorFromHex("D69E63FF")
                                        overlay_slot:
                                            HorizontalAlignment:=horizontal_alignment.Fill
                                            VerticalAlignment:=vertical_alignment.Fill
                                            Padding:=margin{Left:=0.0,Top:=0.0,Right:=-5.0,Bottom:=0.0}
                                            Widget:=texture_block:
                                                DefaultImage:=wooden_pixel_gui.textures.Panels.WhiteSquare100
                                                DefaultTint:=MakeColorFromHex("D69E63FF")
                                        overlay_slot:
                                            HorizontalAlignment:=horizontal_alignment.Fill
                                            VerticalAlignment:=vertical_alignment.Fill
                                            Padding:=margin{Left:=0.0,Top:=0.0,Right:=0.0,Bottom:=-5.0}
                                            Widget:=texture_block:
                                                DefaultImage:=wooden_pixel_gui.textures.Panels.WhiteSquare100
                                                DefaultTint:=MakeColorFromHex("93612FFF")
                                        overlay_slot:
                                            HorizontalAlignment:=horizontal_alignment.Fill
                                            VerticalAlignment:=vertical_alignment.Fill
                                            Padding:=margin{Left:=-5.0,Top:=0.0,Right:=0.0,Bottom:=0.0}
                                            Widget:=texture_block:
                                                DefaultImage:=wooden_pixel_gui.textures.Panels.WhiteSquare100
                                                DefaultTint:=MakeColorFromHex("93612FFF")
                                        overlay_slot:
                                            HorizontalAlignment:=horizontal_alignment.Fill
                                            VerticalAlignment:=vertical_alignment.Fill
                                            Padding:=margin{Left:=0.0,Top:=0.0,Right:=0.0,Bottom:=0.0}
                                            Widget:=texture_block:
                                                DefaultImage:=wooden_pixel_gui.textures.Panels.WhiteSquare100
                                                DefaultTint:=MakeColorFromHex("B87A3BFF")
                            overlay_slot:
                                HorizontalAlignment:=horizontal_alignment.Fill
                                VerticalAlignment:=vertical_alignment.Fill
                                Padding:=margin{Left:=10.0,Top:=0.0,Right:=0.0,Bottom:=0.0}
                                Widget:=MainGoldStackBox
                
        return MyCanvas
            

    CreateOreSlot(Texture : texture, TextBlockId : int, Agent:agent):stack_box_slot= 

        IconTextureBlock: texture_block = texture_block{DefaultImage:= Texture, DefaultDesiredSize:=vector2{X:=36.0,Y:=36.0}}
        IconTextureShadowBlock: texture_block = texture_block{DefaultImage:= Texture,DefaultTint:=MakeColorFromHex("4C3116FF"), DefaultDesiredSize:=vector2{X:=36.0,Y:=36.0}}
        SlotText: text_block = text_block:
            DefaultTextColor := MakeColorFromHex("FFFFFF")
            DefaultText:=storedstring("0")
            DefaultShadowOffset:=option{vector2{X:=1.0,Y:=3.0}}
            DefaultShadowColor:=MakeColorFromHex("4C3116FF")
        SlotText.SetShadowOpacity(X:=0.27)

        Ov1 := overlay_slot: # need to create these so I can save overlay and remove it later
            HorizontalAlignment:=horizontal_alignment.Left
            VerticalAlignment:=vertical_alignment.Center
            Padding:=margin{Left:=2.0,Top:=2.0,Right:=0.0,Bottom:=0.0}
            Widget:=IconTextureShadowBlock
        Ov2 := overlay_slot:
                HorizontalAlignment:=horizontal_alignment.Left
                VerticalAlignment:=vertical_alignment.Center
                Padding:=margin{Left:=2.0,Top:=0.0,Right:=0.0,Bottom:=0.0}
                Widget:=IconTextureBlock
        Ov3 := overlay_slot:
            HorizontalAlignment:=horizontal_alignment.Right
            VerticalAlignment:=vertical_alignment.Center
            Padding:=margin{Left:=40.0,Top:=0.0,Right:=5.0,Bottom:=0.0}
            Widget:=SlotText

        MyOverlay := overlay:

        MyOverlay.AddWidget(Ov1)
        MyOverlay.AddWidget(Ov2)
        MyOverlay.AddWidget(Ov3)
            
        MyStackBoxSlot := stack_box_slot:
            Widget:=MyOverlay
                    
        # set array index = textblock to update
        if(var OldTextBlock : []text_block  = AgentTextBlocks[Agent]?):
            if(set OldTextBlock[TextBlockId] = SlotText):
                if(set AgentTextBlocks[Agent] = option{OldTextBlock}){}

        return MyStackBoxSlot

    
    ShowUi(Agent:agent):void=

        Print("Show Currency ui")

        TempStackBoxSlot := stack_box_slot :
            Widget:=overlay:

        if(set ShowingBlockLogic2[Agent] = option{array{false,false,false,false,false,false,false,false,false}}){}
        if(set OreStackBoxSlots[Agent] = option{array{TempStackBoxSlot,TempStackBoxSlot,TempStackBoxSlot,TempStackBoxSlot,TempStackBoxSlot,TempStackBoxSlot,TempStackBoxSlot,TempStackBoxSlot}}){}



        MyCanvas: canvas = MakeCanvas(Agent)

        if(PlayerUI:=GetPlayerUI[player[Agent]]):
            PlayerUI.AddWidget(MyCanvas,player_ui_slot{InputMode:=ui_input_mode.None}) 
            
        if(set MyUI[Agent]= option{MyCanvas}){}
    
        spawn:
            UpdateText(Agent)
        spawn:
            UpdateOreLoop(Agent)


    UpdateText(Agent:agent)<suspends>:void=

        GemTexture : texture = wooden_pixel_gui.textures.Icons2.Size8.t_GemRed_8
        GoldTexture : texture = wooden_pixel_gui.textures.Icons2.Size8.t_CoinGold_8
        SilverTexture : texture = hud.Silver_icon
        MalachiteTexture : texture = hud.Malachite_icon512
        ObsidianTexture : texture = hud.Obsidian_icon512
        BrightCoreTexture : texture = hud.Brightcore_icon512
        SpectroliteTexture : texture = hud.Spectrolite_icon512

        loop:
            Sleep(RefreshRate)

            if(ShowingBlockLogic2[Agent] = false):
                break

            if:
                FortChar := Agent.GetFortCharacter[]
                FortChar.IsActive[]
                Player:=player[Agent]
                PlayerData := Data[Player]
            then:
                
                if(PlayerData.Gems > 0.0):
                    if(StackB := MainStackBox[Agent]?):
                        if(var ShowingB : []logic = ShowingBlockLogic[Agent]?):
                            if(ShowingB[1] = false):
                                GemsSlot : stack_box_slot = CreateOreSlot(GemTexture,1,Agent)
                                StackB.AddWidget(GemsSlot)
                                if:
                                    set ShowingB[1] = true
                                    set ShowingBlockLogic[Agent] = option{ShowingB}

                if(GoldText := AgentTextBlocks[Agent]?[0]):
                    GoldText.SetText(storedstring("{formatter.Abbreviate((PlayerData.Gold * 1.0),false)}"))
                if(GemsText := AgentTextBlocks[Agent]?[1]):
                    GemsText.SetText(storedstring(" {formatter.Abbreviate((PlayerData.Gems * 1.0),false)}"))


    UpdateOreLoop(Agent:agent)<suspends>:void= 

        GemTexture : texture = wooden_pixel_gui.textures.Icons2.Size32.t_Gem05_32
        GoldTexture : texture = wooden_pixel_gui.textures.Icons2.Size8.t_CoinGold_8
        CopperTexture : texture = hud.Copper_icon512
        SilverTexture : texture = hud.Silver_icon
        MalachiteTexture : texture = hud.Malachite_icon512
        ObsidianTexture : texture = hud.Obsidian_icon512
        BrightCoreTexture : texture = hud.Brightcore_icon512
        SpectroliteTexture : texture = hud.Spectrolite_icon512

        loop:
            Sleep(0.5)

            if(ShowingBlockLogic2[Agent] = false):
                break

            if:
                FortChar := Agent.GetFortCharacter[]
                FortChar.IsActive[]
                Player:=player[Agent]
                PlayerData := Data[Player]
            then:
                if(var ShowingB : []logic = ShowingBlockLogic2[Agent]?):
                    if(ShowingB[6] = false):
                        for(Cond:OreConditionals):
                            OreCount := Cond.GetItemCount(Agent,0)
                            if(OreCount > 0):
                                if(PlayerUI:=GetPlayerUI[player[Agent]]):

                                    MyOreCanvas : canvas = MakeOreCanvas(Agent)

                                    PlayerUI.AddWidget(MyOreCanvas,player_ui_slot{InputMode:=ui_input_mode.None,ZOrder:=2147483641}) 

                                    if(set MyUI2[Agent] = option{MyOreCanvas}){}
                                    if:
                                        set ShowingB[6] = true
                                        set ShowingBlockLogic2[Agent] = option{ShowingB}


                if(var ShowingB : []logic = ShowingBlockLogic2[Agent]?):
                    if(ShowingB[0] = false):
                        if(Cond := OreConditionals[0]):
                            CopperCount := Cond.GetItemCount(Agent,0)
                            if(CopperCount > 0):
                                if(MyStackBox:=MainStackBox2[Agent]?):
                                    NewSlot := CreateOreSlot(CopperTexture,2,Agent)
                                    MyStackBox.AddWidget(NewSlot)
                                    if:
                                        set ShowingB[0] = true
                                        set ShowingBlockLogic2[Agent] = option{ShowingB}
                                        var MyArr : []stack_box_slot = OreStackBoxSlots[Agent]?
                                        set MyArr[0] = NewSlot
                                        set OreStackBoxSlots[Agent] = option{MyArr}
            
                if(var ShowingB : []logic = ShowingBlockLogic2[Agent]?):
                    if(ShowingB[1] = false):
                        if(Cond := OreConditionals[1]):
                            CopperCount := Cond.GetItemCount(Agent,0)
                            if(CopperCount > 0):
                                if(MyStackBox:=MainStackBox2[Agent]?):
                                    NewSlot := CreateOreSlot(SilverTexture,3,Agent)
                                    MyStackBox.AddWidget(NewSlot)
                                    if:
                                        set ShowingB[1] = true
                                        set ShowingBlockLogic2[Agent] = option{ShowingB}
                                        var MyArr : []stack_box_slot = OreStackBoxSlots[Agent]?
                                        set MyArr[1] = NewSlot
                                        set OreStackBoxSlots[Agent] = option{MyArr}

                if(var ShowingB : []logic = ShowingBlockLogic2[Agent]?):
                    if(ShowingB[2] = false):
                        if(Cond := OreConditionals[2]):
                            CopperCount := Cond.GetItemCount(Agent,0)
                            if(CopperCount > 0):
                                if(MyStackBox:=MainStackBox2[Agent]?):
                                    NewSlot := CreateOreSlot(MalachiteTexture,4,Agent)
                                    MyStackBox.AddWidget(NewSlot)
                                    if:
                                        set ShowingB[2] = true
                                        set ShowingBlockLogic2[Agent] = option{ShowingB}
                                        var MyArr : []stack_box_slot = OreStackBoxSlots[Agent]?
                                        set MyArr[2] = NewSlot
                                        set OreStackBoxSlots[Agent] = option{MyArr}

                if(var ShowingB : []logic = ShowingBlockLogic2[Agent]?):
                    if(ShowingB[3] = false):
                        if(Cond := OreConditionals[3]):
                            CopperCount := Cond.GetItemCount(Agent,0)
                            if(CopperCount > 0):
                                if(MyStackBox:=MainStackBox2[Agent]?):
                                    NewSlot := CreateOreSlot(ObsidianTexture,5,Agent)
                                    MyStackBox.AddWidget(NewSlot)
                                    if:
                                        set ShowingB[3] = true
                                        set ShowingBlockLogic2[Agent] = option{ShowingB}
                                        var MyArr : []stack_box_slot = OreStackBoxSlots[Agent]?
                                        set MyArr[3] = NewSlot
                                        set OreStackBoxSlots[Agent] = option{MyArr}
                                    
                if(var ShowingB : []logic = ShowingBlockLogic2[Agent]?):
                    if(ShowingB[4] = false):
                        if(Cond := OreConditionals[4]):
                            CopperCount := Cond.GetItemCount(Agent,0)
                            if(CopperCount > 0):
                                if(MyStackBox:=MainStackBox2[Agent]?):
                                    NewSlot := CreateOreSlot(BrightCoreTexture,6,Agent)
                                    MyStackBox.AddWidget(NewSlot)
                                    if:
                                        set ShowingB[4] = true
                                        set ShowingBlockLogic2[Agent] = option{ShowingB}
                                        var MyArr : []stack_box_slot = OreStackBoxSlots[Agent]?
                                        set MyArr[4] = NewSlot
                                        set OreStackBoxSlots[Agent] = option{MyArr}

                if(var ShowingB : []logic = ShowingBlockLogic2[Agent]?):
                    if(ShowingB[5] = false):
                        if(Cond := OreConditionals[5]):
                            CopperCount := Cond.GetItemCount(Agent,0)
                            if(CopperCount > 0):
                                if(MyStackBox:=MainStackBox2[Agent]?):
                                    NewSlot := CreateOreSlot(SpectroliteTexture,7,Agent)
                                    MyStackBox.AddWidget(NewSlot)
                                    if:
                                        set ShowingB[5] = true
                                        set ShowingBlockLogic2[Agent] = option{ShowingB}
                                        var MyArr : []stack_box_slot = OreStackBoxSlots[Agent]?
                                        set MyArr[5] = NewSlot
                                        set OreStackBoxSlots[Agent] = option{MyArr}


                if(Cond := OreConditionals[0]): # Copper
                    OreCount := Cond.GetItemCount(Agent,0)
                    if(OreCount > 0):
                        if(TextBlock := AgentTextBlocks[Agent]?[2]):
                            TextBlock.SetText(storedstring("{formatter.Abbreviate((OreCount * 1.0),false)}"))
                    else:
                        if:
                            var ShowingB : []logic = ShowingBlockLogic2[Agent]?
                            MyStackBox:=MainStackBox2[Agent]?
                            set ShowingB[0] = false
                            set ShowingBlockLogic2[Agent] = option{ShowingB}
                            MyArr : []stack_box_slot = OreStackBoxSlots[Agent]?
                            MySlot := MyArr[0]
                        then:
                            MyStackBox.RemoveWidget(MySlot.Widget)

                if(Cond := OreConditionals[1]): # Silver
                    OreCount := Cond.GetItemCount(Agent,0)
                    if(OreCount > 0):
                        if(TextBlock := AgentTextBlocks[Agent]?[3]):
                            TextBlock.SetText(storedstring("{formatter.Abbreviate((OreCount * 1.0),false)}"))
                    else:
                        if:
                            var ShowingB : []logic = ShowingBlockLogic2[Agent]?
                            MyStackBox:=MainStackBox2[Agent]?
                            set ShowingB[1] = false
                            set ShowingBlockLogic2[Agent] = option{ShowingB}
                            MyArr : []stack_box_slot = OreStackBoxSlots[Agent]?
                            MySlot := MyArr[1]
                        then:
                            MyStackBox.RemoveWidget(MySlot.Widget)

                if(Cond := OreConditionals[2]): # Malachite
                    OreCount := Cond.GetItemCount(Agent,0)
                    if(OreCount > 0):
                        if(TextBlock := AgentTextBlocks[Agent]?[4]):
                            TextBlock.SetText(storedstring("{formatter.Abbreviate((OreCount * 1.0),false)}"))
                    else:
                        if:
                            var ShowingB : []logic = ShowingBlockLogic2[Agent]?
                            MyStackBox:=MainStackBox2[Agent]?
                            set ShowingB[2] = false
                            set ShowingBlockLogic2[Agent] = option{ShowingB}
                            MyArr : []stack_box_slot = OreStackBoxSlots[Agent]?
                            MySlot := MyArr[2]
                        then:
                            MyStackBox.RemoveWidget(MySlot.Widget)

                if(Cond := OreConditionals[3]): # Obsidian
                    OreCount := Cond.GetItemCount(Agent,0)
                    if(OreCount > 0):
                        if(TextBlock := AgentTextBlocks[Agent]?[5]):
                            TextBlock.SetText(storedstring("{formatter.Abbreviate((OreCount * 1.0),false)}"))
                    else:
                        if:
                            var ShowingB : []logic = ShowingBlockLogic2[Agent]?
                            MyStackBox:=MainStackBox2[Agent]?
                            set ShowingB[3] = false
                            set ShowingBlockLogic2[Agent] = option{ShowingB}
                            MyArr : []stack_box_slot = OreStackBoxSlots[Agent]?
                            MySlot := MyArr[3]
                        then:
                            MyStackBox.RemoveWidget(MySlot.Widget)

                if(Cond := OreConditionals[4]): # BrightCore
                    OreCount := Cond.GetItemCount(Agent,0)
                    if(OreCount > 0):
                        if(TextBlock := AgentTextBlocks[Agent]?[6]):
                            TextBlock.SetText(storedstring("{formatter.Abbreviate((OreCount * 1.0),false)}"))
                    else:
                        if:
                            var ShowingB : []logic = ShowingBlockLogic2[Agent]?
                            MyStackBox:=MainStackBox2[Agent]?
                            set ShowingB[4] = false
                            set ShowingBlockLogic2[Agent] = option{ShowingB}
                            MyArr : []stack_box_slot = OreStackBoxSlots[Agent]?
                            MySlot := MyArr[4]
                        then:
                            MyStackBox.RemoveWidget(MySlot.Widget)


                if(Cond := OreConditionals[5]): # BrightCore
                    OreCount := Cond.GetItemCount(Agent,0)
                    if(OreCount > 0):
                        if(TextBlock := AgentTextBlocks[Agent]?[7]):
                            TextBlock.SetText(storedstring("{formatter.Abbreviate((OreCount * 1.0),false)}"))
                    else:
                        if:
                            var ShowingB : []logic = ShowingBlockLogic2[Agent]?
                            MyStackBox:=MainStackBox2[Agent]?
                            set ShowingB[5] = false
                            set ShowingBlockLogic2[Agent] = option{ShowingB}
                            MyArr : []stack_box_slot = OreStackBoxSlots[Agent]?
                            MySlot := MyArr[5]
                        then:
                            MyStackBox.RemoveWidget(MySlot.Widget)
            else:
            


    HideCurrencyUI(Agent:agent):void=
        if(MyCanvas := MyUI[Agent]?):
            MyCanvas.SetVisibility(widget_visibility.Visible)
            MyCanvas.SetVisibility(widget_visibility.Hidden)
        if(MyCanvas := MyUI2[Agent]?):
            MyCanvas.SetVisibility(widget_visibility.Visible)
            MyCanvas.SetVisibility(widget_visibility.Hidden)
           

    ShowCurrencyUI(Agent:agent):void=
        if(MyCanvas := MyUI[Agent]?):
            MyCanvas.SetVisibility(widget_visibility.Hidden)
            MyCanvas.SetVisibility(widget_visibility.Visible)
        if(MyCanvas := MyUI2[Agent]?):
            MyCanvas.SetVisibility(widget_visibility.Hidden)
            MyCanvas.SetVisibility(widget_visibility.Visible)












