
using { /Fortnite.com/Devices }
using { /Verse.org/Simulation }
using { /UnrealEngine.com/Temporary/Diagnostics }


custom_mission2_device := class(creative_device):

    @editable
    CompAccolade : accolades_device = accolades_device{}

    @editable
    IncrementMissionTriggersList : []trigger_device = array{}


    @editable
    CompleteMissionSound : audio_player_device = audio_player_device{}

    @editable
    MapIndicators : []map_indicator_device = array{}

    @editable
    TrackerDevices : []tracker_device = array{}
    @editable
    TrackerDevices1 : []tracker_device = array{}
    @editable
    TrackerDevices2 : []tracker_device = array{}
    @editable
    TrackerDevices3 : []tracker_device = array{}

    @editable
    TrackerTrigger : []trigger_device = array{}
    @editable
    TrackerTrigger1 : []trigger_device = array{}
    @editable
    TrackerTrigger2 : []trigger_device = array{}
    @editable
    TrackerTrigger3 : []trigger_device = array{}


    @editable
    IdDevice : npc_hp_cinematic_hud_device = npc_hp_cinematic_hud_device{}

    @editable
    Testbutton : button_device = button_device{}

   
    GoldGranted : []int = array{25,50,50,75,250,350,200,300,400,1000,5000,20000,10000,15000,100000,0,0,0,0,0,0,0,0,0,0,0,0,0,0,0,0,0,0,0,0,0,0,0}
    
    DamageGranted : []int = array{0,0,0,0,0,0,0,0,0,0,0,0,0,0,0,10,10,10,10,12,15,15,20,20,20,25,25,25,25,25}

    OnBegin<override>()<suspends>:void=

        

        for(MyTracker : TrackerDevices):
            MyTracker.CompleteEvent.Subscribe(TrackerCompleted)

        for(MyTracker : TrackerDevices1):
            MyTracker.CompleteEvent.Subscribe(TrackerCompleted)

        for(MyTracker : TrackerDevices2):
            MyTracker.CompleteEvent.Subscribe(TrackerCompleted)

        for(MyTracker : TrackerDevices3):
            MyTracker.CompleteEvent.Subscribe(TrackerCompleted)
        
        for(i:= 0..IncrementMissionTriggersList.Length-1):
            if(CurrentTrig := IncrementMissionTriggersList[i]):
                CurrentTrig.TriggeredEvent.SubscribeMaybeAgent(IncrementMissionTriggers,i)

        AllPlayers := GetPlayspace().GetPlayers()
        for(Player:AllPlayers):
            LoadMission(Player)

        Self.GetPlayspace().PlayerAddedEvent().Subscribe(LoadMission)
        Testbutton.InteractedWithEvent.Subscribe(LoadingTest)

        #ChestConditionalButton.ActivatedEvent.SubscribeAgent(IncrementMission,0) # cahange this to triggers
        #BoarTrigger.TriggeredEvent.SubscribeMaybeAgent(IncrementMissionTriggers,1)
        #DamageMissionTrigger.TriggeredEvent.SubscribeMaybeAgent(IncrementMissionTriggers,2)
        #TeleportMissionTrigger.TriggeredEvent.SubscribeMaybeAgent(IncrementMissionTriggers,3)
        #TaskMissionTrigger.TriggeredEvent.SubscribeMaybeAgent(IncrementMissionTriggers,7)
        #OpenBurnShopButton.InteractedWithEvent.SubscribeAgent(IncrementMission,4)
        #OpenOreShopButton.InteractedWithEvent.SubscribeAgent(IncrementMission,5)
        #OpenGearButton.InteractedWithEvent.SubscribeAgent(IncrementMission,6)
        #TaskMissionTrigger.TriggeredEvent.SubscribeMaybeAgent(IncrementMissionTriggers,7)



    LoadingTest(Agent:agent):void=
        if(Player:=player[Agent]):
            LoadMission(Player)

    LoadMission(Player:player):void=

        var MyTracker : tracker_device = tracker_device{}
        var MyTrigger : trigger_device = trigger_device{}

        if:
            Agent:=agent[Player]
            OldData := Data[Player]
            var MissionsCompleted : int = OldData.MissionsCompletedPersis
            SavedCurrent : int = OldData.MissionsCurrentCount
            ID := IdDevice.IdPerAgent[Agent]?
        then:
            Print("Loading Mission : {MissionsCompleted}")

            if:
                ID = 0
                NewTrigger := TrackerTrigger[MissionsCompleted]
                NewTracker := TrackerDevices[MissionsCompleted]
            then:
                set MyTrigger = NewTrigger
                set MyTracker = NewTracker
            if:
                ID = 1
                NewTrigger := TrackerTrigger1[MissionsCompleted]
                NewTracker := TrackerDevices1[MissionsCompleted]
            then:
                set MyTrigger = NewTrigger
                set MyTracker = NewTracker
            if:
                ID = 2
                NewTrigger := TrackerTrigger2[MissionsCompleted]
                NewTracker := TrackerDevices2[MissionsCompleted]
            then:
                set MyTrigger = NewTrigger
                set MyTracker = NewTracker

            if:
                ID = 3
                NewTrigger := TrackerTrigger3[MissionsCompleted]
                NewTracker := TrackerDevices3[MissionsCompleted]
            then:
                set MyTrigger = NewTrigger
                set MyTracker = NewTracker

            MyTracker.SetValue(SavedCurrent)
            # Trigger Asigns
            MyTrigger.Trigger(Agent) 
            

            for(i:=0..MapIndicators.Length-1):
                if(MapInd:=MapIndicators[i]):
                    MapInd.DeactivateObjectivePulse(Agent)
                    MapInd.DeactivateObjectivePulse(Agent)
            if(MapIndicator := MapIndicators[MissionsCompleted]):
                MapIndicator.ActivateObjectivePulse(Agent)
                MapIndicator.ActivateObjectivePulse(Agent)



    IncrementMission(Agent:agent,MissionNumber:int):void=
        if:
            Player:=player[Agent]
            OldData := Data[Player]
            TutorialNumber := OldData.MissionsCompletedPersis
            ID := IdDevice.IdPerAgent[Agent]?
        then:
            Print("Incrementing Mission")
            Print("TutorialNumber : {OldData.MissionsCompletedPersis}")
            if(TutorialNumber = MissionNumber):
                if:
                    ID = 0
                    MyTracker := TrackerDevices[TutorialNumber]
                then:
                    MyTracker.Increment(Agent)
                if:
                    ID = 1
                    MyTracker := TrackerDevices1[TutorialNumber]
                then:
                    MyTracker.Increment(Agent)
                if:
                    ID = 2
                    MyTracker := TrackerDevices2[TutorialNumber]
                then:
                    MyTracker.Increment(Agent)
                if:
                    ID = 3
                    MyTracker := TrackerDevices3[TutorialNumber]
                then:
                    MyTracker.Increment(Agent)

                Print("Incrementing Mission2 Testing shit")


    TrackerCompleted(Agent:agent):void=
        if:
            Player:=player[Agent]
            OldData := Data[Player]
            CurrentMission := OldData.MissionsCompletedPersis
            GrantAmount := GoldGranted[CurrentMission]
            DamageAmount := DamageGranted[CurrentMission]
            set Data[Player] = player_data{player_data_update<constructor>(OldData),MissionsCompletedPersis := CurrentMission + 1,Gold := OldData.Gold + (GrantAmount*1.0),AchievementsDamageBonus := OldData.AchievementsDamageBonus + (DamageAmount*1.0) }
            ID := IdDevice.IdPerAgent[Agent]?
        then:

            if:
                ID = 0
                MyTracker := TrackerDevices[OldData.MissionsCompletedPersis]
            then:
                MyTracker.Remove(Agent)
            if:
                ID = 1
                MyTracker := TrackerDevices1[OldData.MissionsCompletedPersis]
            then:
                MyTracker.Remove(Agent)
            if:
                ID = 2
                MyTracker := TrackerDevices2[OldData.MissionsCompletedPersis]
            then:
                MyTracker.Remove(Agent)
            if:
                ID = 3
                MyTracker := TrackerDevices3[OldData.MissionsCompletedPersis]
            then:
                MyTracker.Remove(Agent)

                
            CompleteMissionSound.Play(Agent)
            

            Print("TrackersCompleted:{OldData.MissionsCompletedPersis}")

            if(OldData.MissionsCompletedPersis >= TrackerDevices.Length-1):
                Print("Completed Quests")
                for(i:=0..MapIndicators.Length-1):
                    if(MapInd := MapIndicators[i]):
                        MapInd.DeactivateObjectivePulse(Agent)

            ############################################################

        LoadingTest(Agent)

            
                


                
    IncrementMissionAi(Interaction:device_ai_interaction_result, MissionNumber:int):void=
        if:
            Agent:=Interaction.Source?
            Player:=player[Agent]
            OldData := Data[Player]
            TutorialNumber := OldData.MissionsCompletedPersis
            ID := IdDevice.IdPerAgent[Agent]?
        then:
            Print("Incrementing Mission")
            if(TutorialNumber = MissionNumber):
                if:
                    ID = 0
                    MyTracker := TrackerDevices[TutorialNumber]
                then:
                    MyTracker.Increment(Agent)
                if:
                    ID = 1
                    MyTracker := TrackerDevices1[TutorialNumber]
                then:
                    MyTracker.Increment(Agent)
                if:
                    ID = 2
                    MyTracker := TrackerDevices2[TutorialNumber]
                then:
                    MyTracker.Increment(Agent)
                if:
                    ID = 3
                    MyTracker := TrackerDevices3[TutorialNumber]
                then:
                    MyTracker.Increment(Agent)
                Print("Incrementing Mission2 Testing shit")


    IncrementMissionTriggers(Agent:?agent,MissionNumber:int):void=
        if:
            MyAgent:= Agent?
            Player:=player[MyAgent]
            OldData := Data[Player]
            TutorialNumber := OldData.MissionsCompletedPersis
            ID := IdDevice.IdPerAgent[MyAgent]?
        then:
            Print("Incrementing Mission  : {MissionNumber}")
            Print("TutorialNumber : {OldData.MissionsCompletedPersis}, MissionNumber : {MissionNumber}")
            if(TutorialNumber = MissionNumber):
                if:
                    ID = 0
                    MyTracker := TrackerDevices[TutorialNumber]
                then:
                    MyTracker.Increment(MyAgent)
                if:
                    ID = 1
                    MyTracker := TrackerDevices1[TutorialNumber]
                then:
                    MyTracker.Increment(MyAgent)
                if:
                    ID = 2
                    MyTracker := TrackerDevices2[TutorialNumber]
                then:
                    MyTracker.Increment(MyAgent)
                if:
                    ID = 3
                    MyTracker := TrackerDevices3[TutorialNumber]
                then:
                    MyTracker.Increment(MyAgent)
                Print("Incrementing Mission2 Testing shit")
      
            
    UpdateTracker(Agent:agent,MissionNumber:int):void= 
        if:
            Player:=player[Agent]
            OldData := Data[Player]
            Current := OldData.MissionsCurrentCount
            MyTracker := TrackerDevices[MissionNumber]
        then:
          
            for(i:=0..MapIndicators.Length-1):
                if(MapInd:=MapIndicators[i]):
                    MapInd.DeactivateObjectivePulse(Agent)
                    MapInd.DeactivateObjectivePulse(Agent)

            if(MapIndicator := MapIndicators[MissionNumber]):
                MapIndicator.ActivateObjectivePulse(Agent)
                MapIndicator.ActivateObjectivePulse(Agent)
                MyTracker.Assign(Agent)
            
                #MapIndicator.ActivateObjectivePulse(Agent)
        else:    
            for(i:=0..MapIndicators.Length-1):
                if(MapInd := MapIndicators[i]):
                    MapInd.DeactivateObjectivePulse(Agent)
            



                    












