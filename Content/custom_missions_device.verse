
using { /Fortnite.com/Devices }
using { /Verse.org/Simulation }
using { /UnrealEngine.com/Temporary/Diagnostics }

# Old file from farmlife4
custom_missions_device := class(creative_device):

    @editable
    CompAccolade : accolades_device = accolades_device{}


    # Mission Array is used to hard code titles and targets for each mission
    var MissionsArray : []custom_tracker_tutorial = array{}

    @editable
    CustomTracker : custom_tutorial_device = custom_tutorial_device{}

    # Mission Incrementables

    @editable
    ChestConditionalButton : conditional_button_device = conditional_button_device{}
    @editable
    BoarTrigger : trigger_device = trigger_device{}
    @editable
    DamageMissionTrigger : trigger_device = trigger_device{}
    @editable
    TeleportMissionTrigger : trigger_device = trigger_device{}
    @editable
    TaskMissionTrigger : trigger_device = trigger_device{}
    @editable
    HatMissionTrigger : trigger_device = trigger_device{}


    @editable
    OpenBurnShopButton : button_device = button_device{}
    @editable
    OpenOreShopButton : button_device = button_device{}
    @editable
    OpenGearButton : button_device = button_device{}

    @editable
    CompleteMissionSound : audio_player_device = audio_player_device{}

    @editable
    MapIndicators : []map_indicator_device = array{}
    @editable
    HatRewardTrigger : trigger_device = trigger_device{}

    @editable
    FinishMissionsCine : cinematic_sequence_device = cinematic_sequence_device{}

    @editable
    TrackerDevices : []tracker_device = array{}
    @editable
    TrackerDevices1 : []tracker_device = array{}
    @editable
    TrackerDevices2 : []tracker_device = array{}
    @editable
    TrackerDevices3 : []tracker_device = array{}

    @editable
    TrackerTrigger : []trigger_device = array{}
    @editable
    TrackerTrigger1 : []trigger_device = array{}
    @editable
    TrackerTrigger2 : []trigger_device = array{}
    @editable
    TrackerTrigger3 : []trigger_device = array{}


    @editable
    IdDevice : npc_hp_cinematic_hud_device = npc_hp_cinematic_hud_device{}

    OnBegin<override>()<suspends>:void=

        for(MyTracker : TrackerDevices):
            MyTracker.CompleteEvent.Subscribe(TrackerCompleted)

        for(MyTracker : TrackerDevices1):
            MyTracker.CompleteEvent.Subscribe(TrackerCompleted)

        for(MyTracker : TrackerDevices2):
            MyTracker.CompleteEvent.Subscribe(TrackerCompleted)

        for(MyTracker : TrackerDevices3):
            MyTracker.CompleteEvent.Subscribe(TrackerCompleted)


        InitMissions() # sets mission array to be loaded later

        AllPlayers := GetPlayspace().GetPlayers()
        for(Player:AllPlayers):
            LoadMission(Player)

        Self.GetPlayspace().PlayerAddedEvent().Subscribe(LoadMission)

        ChestConditionalButton.ActivatedEvent.SubscribeAgent(IncrementMission,0) # cahange this to triggers
        BoarTrigger.TriggeredEvent.SubscribeMaybeAgent(IncrementMissionTriggers,1)
        DamageMissionTrigger.TriggeredEvent.SubscribeMaybeAgent(IncrementMissionTriggers,2)
        TeleportMissionTrigger.TriggeredEvent.SubscribeMaybeAgent(IncrementMissionTriggers,3)
        TaskMissionTrigger.TriggeredEvent.SubscribeMaybeAgent(IncrementMissionTriggers,7)

        OpenBurnShopButton.InteractedWithEvent.SubscribeAgent(IncrementMission,4)
        OpenOreShopButton.InteractedWithEvent.SubscribeAgent(IncrementMission,5)
        OpenGearButton.InteractedWithEvent.SubscribeAgent(IncrementMission,6)

        TaskMissionTrigger.TriggeredEvent.SubscribeMaybeAgent(IncrementMissionTriggers,7)



    LoadingTest(Agent:agent):void=
        if(Player:=player[Agent]):
            LoadMission(Player)

    LoadMission(Player:player):void=
        if:
            Agent:=agent[Player]
            OldData := Data[Player]
            var TutorialsCompleted : int = OldData.TutorialsCompleted
            MyMission := MissionsArray[TutorialsCompleted]
            SavedCurrent : int = OldData.TutorialsCurrentCount
            
            ID := IdDevice.IdPerAgent[Agent]?
        then:

            if:
                ID = 0
                MyTrigger := TrackerTrigger[TutorialsCompleted]
            then:
            
                MyTrigger.Trigger(Agent)
            if:
                ID = 1
                MyTrigger := TrackerTrigger1[TutorialsCompleted]
            then:

                MyTrigger.Trigger(Agent)
            if:
                ID = 2
                MyTrigger := TrackerTrigger2[TutorialsCompleted]
            then:

                MyTrigger.Trigger(Agent)

            if:
                ID = 3
                MyTrigger := TrackerTrigger3[TutorialsCompleted]
            then:
                MyTrigger.Trigger(Agent)



            #CustomTracker.InitCustomUI(Agent,MyMission.title,MyMission.descriptionline1,MyMission.descriptionline2, MyMission.target, SavedCurrent, MyMission.icon)
            #set MyMission.current = SavedCurrent
            #if(SavedCurrent >= MyMission.target):
            #    IncrementMission(Agent,TutorialsCompleted)

            for(i:=0..MapIndicators.Length-1):
                if(MapInd:=MapIndicators[i]):
                    MapInd.DeactivateObjectivePulse(Agent)
                    MapInd.DeactivateObjectivePulse(Agent)
            if(MapIndicator := MapIndicators[TutorialsCompleted]):
                MapIndicator.ActivateObjectivePulse(Agent)
                MapIndicator.ActivateObjectivePulse(Agent)

        
    InitMissions():void=

        OpenChest : custom_tracker_tutorial = custom_tracker_tutorial :
            current := 0
            target := 1
            icon := MissionsTrackers.CommonChest1
            title := "Tutorials 0/8"
            descriptionline1 := " Open common weapon "
            descriptionline2 := " chest"
            
        EliminateBoar : custom_tracker_tutorial = custom_tracker_tutorial :
            current := 0
            target := 1
            icon := MissionsTrackers.BoarMission
            title := "Tutorials 1/8"
            descriptionline1 := "Eliminate a wild boar"
            descriptionline2 := ""

        UpgradeDamage: custom_tracker_tutorial = custom_tracker_tutorial :
            current := 0
            target := 1
            icon := RavenIcons.SkillsSpellsandScores.A64x64.S79
            title := "Tutorials 2/8"
            descriptionline1 := "Use remote to upgrade  "
            descriptionline2 := "damage talent"

        TeleportUsingAlter : custom_tracker_tutorial = custom_tracker_tutorial :
            current := 0
            target := 1
            icon := MissionsTrackers.AlterImage
            title := "Tutorials 3/8"
            descriptionline1 := "Teleport using your remote"
            descriptionline2 := "secondary or teleporter"

        OpenBurnShop : custom_tracker_tutorial = custom_tracker_tutorial :
            current := 0
            target := 1
            icon := RavenIcons.MapMarkers.tile000
            title := "Tutorials 4/8"
            descriptionline1 := "Visit the burn skill shop"
            descriptionline2 := ""

        OreDropChance : custom_tracker_tutorial = custom_tracker_tutorial :
            current := 0
            target := 1
            icon := RavenIcons.MapMarkers.S30
            title := "Tutorials 5/8"
            descriptionline1 := "Visit the ore shop"
            descriptionline2 := ""

        OpenTailor : custom_tracker_tutorial = custom_tracker_tutorial :
            current := 0
            target := 1
            icon := wooden_pixel_gui.textures.Icons2.Size32.t_Helmet02_32
            title := "Tutorials 6/8"
            descriptionline1 := "Visit the armor tailor"
            descriptionline2 := ""

        StartATask : custom_tracker_tutorial = custom_tracker_tutorial :
            current := 0
            target := 1
            icon := RavenIcons.MapMarkers.S794
            title := "Tutorials 8/8"
            descriptionline1 := "Go to quest board and"
            descriptionline2 := "buy a new quest"



        set MissionsArray = array{OpenChest, EliminateBoar, UpgradeDamage, TeleportUsingAlter, OpenBurnShop , OreDropChance, OpenTailor, StartATask} # index = subscribe numer


    IncrementMission(Agent:agent,MissionNumber:int):void=
        if:
            Player:=player[Agent]
            OldData := Data[Player]
            TutorialNumber := OldData.TutorialsCompleted
            ID := IdDevice.IdPerAgent[Agent]?
        then:
            Print("Incrementing Mission")
            if(TutorialNumber = MissionNumber):
                if:
                    ID = 0
                    MyTracker := TrackerDevices[TutorialNumber]
                then:
                    MyTracker.Increment(Agent)
                if:
                    ID = 1
                    MyTracker := TrackerDevices1[TutorialNumber]
                then:
                    MyTracker.Increment(Agent)
                if:
                    ID = 2
                    MyTracker := TrackerDevices2[TutorialNumber]
                then:
                    MyTracker.Increment(Agent)
                if:
                    ID = 3
                    MyTracker := TrackerDevices3[TutorialNumber]
                then:
                    MyTracker.Increment(Agent)
                Print("Incrementing Mission2 Testing shit")

                #SaveCurrent(Agent,OldData.TutorialsCurrentCount+1)
                #UpdateTracker(Agent,MissionNumber)
                #CheckIfMissionComplete(Agent,MissionNumber)

    TrackerCompleted(Agent:agent):void=
        if:
            Player:=player[Agent]
            OldData := Data[Player]
            set Data[Player] = player_data{player_data_update<constructor>(OldData),TutorialsCompleted := OldData.TutorialsCompleted + 1}
            ID := IdDevice.IdPerAgent[Agent]?
        then:

            if:
                ID = 0
                MyTracker := TrackerDevices[OldData.TutorialsCompleted]
            then:
                MyTracker.Remove(Agent)
            if:
                ID = 1
                MyTracker := TrackerDevices1[OldData.TutorialsCompleted]
            then:
                MyTracker.Remove(Agent)
            if:
                ID = 2
                MyTracker := TrackerDevices2[OldData.TutorialsCompleted]
            then:
                MyTracker.Remove(Agent)
            if:
                ID = 3
                MyTracker := TrackerDevices3[OldData.TutorialsCompleted]
            then:
                MyTracker.Remove(Agent)

                
            CompleteMissionSound.Play(Agent)
            LoadingTest(Agent)

            Print("TutorialsCompleted:{OldData.TutorialsCompleted}")

            if(OldData.TutorialsCompleted >= TrackerDevices.Length-1):
                Print("Completed Tuts ")
                HatRewardTrigger.Trigger(Agent)
                CompAccolade.Award(Agent)

                FinishMissionsCine.Play(Agent)
                

                for(i:=0..MapIndicators.Length-1):
                    if(MapInd := MapIndicators[i]):
                        MapInd.DeactivateObjectivePulse(Agent)
                


                
    IncrementMissionAi(Interaction:device_ai_interaction_result, MissionNumber:int):void=
        if:
            Agent:=Interaction.Source?
            Player:=player[Agent]
            OldData := Data[Player]
            TutorialNumber := OldData.TutorialsCompleted
            ID := IdDevice.IdPerAgent[Agent]?
        then:
            Print("Incrementing Mission")
            if(TutorialNumber = MissionNumber):
                if:
                    ID = 0
                    MyTracker := TrackerDevices[TutorialNumber]
                then:
                    MyTracker.Increment(Agent)
                if:
                    ID = 1
                    MyTracker := TrackerDevices1[TutorialNumber]
                then:
                    MyTracker.Increment(Agent)
                if:
                    ID = 2
                    MyTracker := TrackerDevices2[TutorialNumber]
                then:
                    MyTracker.Increment(Agent)
                if:
                    ID = 3
                    MyTracker := TrackerDevices3[TutorialNumber]
                then:
                    MyTracker.Increment(Agent)
                Print("Incrementing Mission2 Testing shit")


    IncrementMissionTriggers(Agent:?agent,MissionNumber:int):void=
        if:
            MyAgent:= Agent?
            Player:=player[MyAgent]
            OldData := Data[Player]
            TutorialNumber := OldData.TutorialsCompleted
            ID := IdDevice.IdPerAgent[MyAgent]?
        then:
            Print("Incrementing Mission")
            if(TutorialNumber = MissionNumber):
                if:
                    ID = 0
                    MyTracker := TrackerDevices[TutorialNumber]
                then:
                    MyTracker.Increment(MyAgent)
                if:
                    ID = 1
                    MyTracker := TrackerDevices1[TutorialNumber]
                then:
                    MyTracker.Increment(MyAgent)
                if:
                    ID = 2
                    MyTracker := TrackerDevices2[TutorialNumber]
                then:
                    MyTracker.Increment(MyAgent)
                if:
                    ID = 3
                    MyTracker := TrackerDevices3[TutorialNumber]
                then:
                    MyTracker.Increment(MyAgent)
                Print("Incrementing Mission2 Testing shit")
      

    SaveCurrent(Agent:agent,Current:int):void=
        if:
            Player:=player[Agent]
            OldData := Data[Player]
            set Data[Player] = player_data{player_data_update<constructor>(OldData),TutorialsCurrentCount := Current}
        then:
            Print("Saving current : {Current}")

    SaveMission(Agent:agent,MissionNumber:int):void=
        if:
            Player:=player[Agent]
            OldData := Data[Player]
            set Data[Player] = player_data{player_data_update<constructor>(OldData),TutorialsCompleted := MissionNumber}
        then:
            Print("Mission set to MissionNumber = {MissionNumber}")

            if(MissionsArray.Length < MissionNumber): # Finished Missions
                for(i:=0..MapIndicators.Length-1):
                    if(MapInd := MapIndicators[i]):
                        MapInd.DeactivateObjectivePulse(Agent)
    
                HatRewardTrigger.Trigger(Agent)
                CompAccolade.Award(Agent)
                

                CustomTracker.RemoveHudTest(Agent)
                

    UpdateTracker(Agent:agent,MissionNumber:int):void= 
        if:
            MyMission := MissionsArray[MissionNumber]
            Player:=player[Agent]
            OldData := Data[Player]
            Current := OldData.TutorialsCurrentCount
            MyTracker := TrackerDevices[MissionNumber]
        then:
          
            for(i:=0..MapIndicators.Length-1):
                if(MapInd:=MapIndicators[i]):
                    MapInd.DeactivateObjectivePulse(Agent)
                    MapInd.DeactivateObjectivePulse(Agent)

            if(MapIndicator := MapIndicators[MissionNumber]):
                MapIndicator.ActivateObjectivePulse(Agent)
                MapIndicator.ActivateObjectivePulse(Agent)
                MyTracker.Assign(Agent)
            
                #MapIndicator.ActivateObjectivePulse(Agent)
        else:    
            for(i:=0..MapIndicators.Length-1):
                if(MapInd := MapIndicators[i]):
                    MapInd.DeactivateObjectivePulse(Agent)

            HatRewardTrigger.Trigger(Agent)
            CompAccolade.Award(Agent)
            

    CheckIfMissionComplete(Agent:agent,MissionNumber:int):void=
        if:
            MyMission := MissionsArray[MissionNumber]
            Player:=player[Agent]
            OldData := Data[Player]
        then:
            if(OldData.TutorialsCurrentCount >= MyMission.target):

                # Award Accolade
                CompleteMissionSound.Play(Agent)
                SaveMission(Agent,MissionNumber+1)
                    
                #Assign next mission
                SaveCurrent(Agent,0)
                UpdateTracker(Agent,MissionNumber+1)


                    












