
using { /Fortnite.com/Devices }
using { /Verse.org/Simulation }
using { /UnrealEngine.com/Temporary/Diagnostics }
using { /UnrealEngine.com/Temporary/UI }
using { /Fortnite.com/UI }
using { /UnrealEngine.com/Temporary/SpatialMath }
using { /Verse.org/Colors }
using { /Verse.org/Assets }

custom_tracker := class() :

    var current : int = 0
    var target : int = 0
    var goldreward : int = 0
    var xpreward : int = 0
    var icon : texture = wooden_pixel_gui.textures.Icons2.Size8.t_CoinGold_8
    var title : string = ""
    var descriptionline1 : string = ""
    var descriptionline2 : string = ""

custom_tracker_widgets := class() :

    var Canvas : canvas = canvas{}
    var TitleWidget : text_block = text_block{}
    var DescriptionWidget : text_block = text_block{}
    var DescriptionWidget2 : text_block = text_block{}
    var IconWidget : texture_block = texture_block{DefaultImage := Icons.Coin64}
    var TargetWidget : text_block = text_block{}
    var RewardWidget : text_block = text_block{}

custom_tracker_device := class(creative_device):

    @editable
    TestButton : button_device = button_device {}
    @editable
    RemoveTestButton: button_device = button_device {}
    @editable
    UpdateTestButton: button_device = button_device {}
    var CanvasPerPlayer : [agent]?canvas = map{}
    S2M<localizes>(String: string) : message = "{String}"
    TargetMessage<localizes>(CurrentInt : int, TargetInt : int) : message = "{CurrentInt}/{TargetInt}"
    RewardMessage<localizes>(RewardInt : int) : message = "{RewardInt}"

    var UIPerPlayer : [player]?canvas = map{} # new
    var CustomTrackerPerPlayer : [player]?custom_tracker_widgets = map{}

    OnBegin<override>()<suspends>:void=

        TestButton.InteractedWithEvent.Subscribe(HandleButton)
        RemoveTestButton.InteractedWithEvent.Subscribe(RemoveHudTest)
        
    HandleButton(Agent:agent):void=
        InitCustomUI(Agent,"TitleTest",10,0,100)


    RemoveHudTest(Agent:agent):void=
        for(Player->Canvas:CanvasPerPlayer):
            if:
                PlayerUI:=GetPlayerUI[player[Agent]]
            then:
                if(ValidCanvas := Canvas?):
                    PlayerUI.RemoveWidget(ValidCanvas)


    UpdateTrackers(Agent:agent,TitleString :string, TargetAmount : int , CurrentAmount : int, Reward : int):void=
        if(Player := player[Agent]):
            if(MyCustomTracker:=CustomTrackerPerPlayer[Player]?):
                MyCustomTracker.TitleWidget.SetText(S2M(TitleString))
                MyCustomTracker.TargetWidget.SetText(TargetMessage(CurrentAmount,TargetAmount))

    CreateCanvas( TitleString :string, TargetAmount : int , CurrentAmount : int, Reward : int): custom_tracker_widgets =

        MyCustomTracker : custom_tracker_widgets = custom_tracker_widgets{}

        IconTexture : texture = wooden_pixel_gui.textures.Icons2.Size32.t_Scroll_32
        
        set MyCustomTracker.TitleWidget = text_block:
            DefaultTextColor:=MakeColorFromHex("FFA500")
            DefaultText:=S2M(TitleString)
        set MyCustomTracker.TargetWidget = text_block:
            DefaultTextColor:=MakeColorFromHex("FFFFFF")
            DefaultText:=TargetMessage(CurrentAmount, TargetAmount)
        set MyCustomTracker.IconWidget = texture_block:
            DefaultDesiredSize:=vector2{X:=32.0,Y:=32.0}
            DefaultImage := IconTexture

        Indent2 :=texture_block:
            DefaultDesiredSize:=vector2{X:=55.0,Y:=55.0}
            DefaultImage := wooden_pixel_gui.textures.Panels.WhiteSquare75
        Indent2.SetVisibility(widget_visibility.Hidden)

        Overlay1:=overlay:
            Slots:=array:
                overlay_slot:
                    HorizontalAlignment := horizontal_alignment.Fill
                    VerticalAlignment := vertical_alignment.Fill
                    Widget:=texture_block:
                        DefaultDesiredSize:=vector2{X:=320.0,Y:=42.0}
                        DefaultImage := wooden_pixel_gui.textures.Panels.WhiteSquare100
                        DefaultTint := MakeColorFromHex("6C4B2AE6")
                overlay_slot:
                    HorizontalAlignment := horizontal_alignment.Center
                    VerticalAlignment := vertical_alignment.Center
                    Widget:=MyCustomTracker.TitleWidget
                overlay_slot:
                    HorizontalAlignment:=horizontal_alignment.Left
                    VerticalAlignment:=vertical_alignment.Top
                    Padding:=margin{Left:=0.0,Top:=0.0,Right:=0.0,Bottom:=0.0}
                    Widget:=stack_box:
                        Orientation := orientation.Horizontal
                        Slots:=array:
                            stack_box_slot:
                                HorizontalAlignment := horizontal_alignment.Center
                                VerticalAlignment := vertical_alignment.Center
                                Padding:=margin{Left:=8.0,Top:=0.0,Right:=0.0,Bottom:=0.0}
                                Widget:=MyCustomTracker.IconWidget
                            stack_box_slot:
                                HorizontalAlignment := horizontal_alignment.Center
                                VerticalAlignment := vertical_alignment.Center
                                Padding:=margin{Left:=8.0,Top:=10.0,Right:=10.0,Bottom:=10.0}
                                Widget:=MyCustomTracker.TitleWidget
                            stack_box_slot:
                                HorizontalAlignment := horizontal_alignment.Center
                                VerticalAlignment := vertical_alignment.Center
                                Padding:=margin{Left:=0.0,Top:=0.0,Right:=8.0,Bottom:=0.0}
                                Widget:=MyCustomTracker.TargetWidget


        var MaxAnchors : vector2 = vector2{X := 0.99, Y := 0.35}
        var MinAnchors : vector2 = vector2{X := 0.99, Y := 0.35}

        Canvas:=canvas:
            Slots:=array:
                canvas_slot:
                    Anchors := anchors{Minimum := MinAnchors, Maximum := MaxAnchors}
                    Alignment := vector2{X := 1.0, Y := 0.0}
                    SizeToContent := true
                    Widget := Overlay1

        set MyCustomTracker.Canvas = Canvas
        return MyCustomTracker


    InitCustomUI(Agent:agent,TitleString :string, TargetAmount : int , CurrentAmount : int, Reward : int):void=
        if(Player:=player[Agent]):
            if (PlayerUI := GetPlayerUI[Player]):
                if (MyCustomTrackerWidget := CustomTrackerPerPlayer[Player]?):
                    PlayerUI.RemoveWidget(MyCustomTrackerWidget.Canvas)
                    if {set CustomTrackerPerPlayer[Player] = false}
                else:
                    MyCustomTrackerWidget := CreateCanvas(TitleString , TargetAmount , CurrentAmount , Reward)
                    PlayerUI.AddWidget(MyCustomTrackerWidget.Canvas, player_ui_slot{InputMode := ui_input_mode.None})
                    if (set CustomTrackerPerPlayer[Player] = option{MyCustomTrackerWidget}) {}

                if (MyCustomTrackerWidget := CustomTrackerPerPlayer[Player]?):
                else:
                    MyCustomTrackerWidget := CreateCanvas(TitleString , TargetAmount , CurrentAmount , Reward )
                    PlayerUI.AddWidget(MyCustomTrackerWidget.Canvas, player_ui_slot{InputMode := ui_input_mode.None})
                    if (set CustomTrackerPerPlayer[Player] = option{MyCustomTrackerWidget}) {}

    RemoveTracker(Agent:agent):void=
        if(Player:=player[Agent]):
            if (PlayerUI := GetPlayerUI[Player]):
                if (MyCustomTrackerWidget := CustomTrackerPerPlayer[Player]?):
                    PlayerUI.RemoveWidget(MyCustomTrackerWidget.Canvas)
                    if {set CustomTrackerPerPlayer[Player] = false}





                    