
using { /Fortnite.com/Devices }
using { /Verse.org/Simulation }
using { /UnrealEngine.com/Temporary/Diagnostics }
using { /UnrealEngine.com/Temporary/UI }
using { /Fortnite.com/UI }
using { /UnrealEngine.com/Temporary/SpatialMath }
using { /Verse.org/Colors }
using { /Verse.org/Assets }

custom_tracker_tutorial := class() :

    var current : int = 0
    var target : int = 0
    var icon : texture = wooden_pixel_gui.textures.Icons2.Size8.t_CoinGold_8
    var title : string = ""
    var descriptionline1 : string = ""
    var descriptionline2 : string = ""

custom_tutorial_device := class(creative_device):

    #@editable
    #TestButton : button_device = button_device {}
    #@editable
    #RemoveTestButton: button_device = button_device {}
    #@editable
    #UpdateTestButton: button_device = button_device {}

    var CanvasPerPlayer : [agent]?canvas = map{}

    S2M<localizes>(String: string) : message = "{String}"
    TargetMessage<localizes>(CurrentInt : int, TargetInt : int) : message = "{CurrentInt}/{TargetInt}"
    RewardMessage<localizes>(RewardInt : int) : message = "{RewardInt}"



    var CustomTrackerPerPlayer : [player]?custom_tracker_widgets = map{}


    #OnBegin<override>()<suspends>:void=

        #TestButton.InteractedWithEvent.Subscribe(HandleButton)
        #RemoveTestButton.InteractedWithEvent.Subscribe(RemoveHudTest)
        #UpdateTestButton.InteractedWithEvent.Subscribe(UpdateHandle)
        

    #HandleButton(Agent:agent):void=
        #InitCustomUI(Agent,"Title","Descritption","Descritption", 10, 0, Icons.Coin64)


    #UpdateHandle(Agent:agent):void=
        #block:
        #UpdateTracker("Testing Update","Line1","Line2",0,10,15, Icons.Coin64)


    RemoveHudTest(Agent:agent):void=
        if:
            PlayerUI:=GetPlayerUI[player[Agent]]
            MyCanvas := CanvasPerPlayer[Agent]?
        then:
            PlayerUI.RemoveWidget(MyCanvas)
            if(set CanvasPerPlayer[Agent] = false){}


    UpdateTrackers(TitleString :string, DescString1 : string, DescString2 : string, TargetAmount : int , CurrentAmount : int, MissionIcon : texture):void=

        for(Player->MaybeCustomTracker:CustomTrackerPerPlayer):
            if(MyCustomTracker:=MaybeCustomTracker?):
                MyCustomTracker.TitleWidget.SetText(S2M(TitleString))
                MyCustomTracker.DescriptionWidget.SetText(S2M(DescString1 ))
                MyCustomTracker.DescriptionWidget2.SetText(S2M(DescString2))
                MyCustomTracker.IconWidget.SetImage(MissionIcon)
                #MyCustomTracker.TargetWidget.SetText(TargetMessage(CurrentAmount,TargetAmount))

    CreateCanvas( TitleString :string, DescString1 : string, DescString2 : string, TargetAmount : int , CurrentAmount : int, MissionIcon : texture): custom_tracker_widgets =

        MyCustomTracker : custom_tracker_widgets = custom_tracker_widgets{}
        
        set MyCustomTracker.TitleWidget = text_block:
            DefaultTextColor:=MakeColorFromHex("FFA500")
            DefaultText:=S2M(TitleString)
        set MyCustomTracker.DescriptionWidget = text_block:
            DefaultTextColor:=MakeColorFromHex("B1B1B1")
            DefaultText:=S2M(DescString1)
        set MyCustomTracker.DescriptionWidget2 = text_block:
            DefaultTextColor:=MakeColorFromHex("B1B1B1")
            DefaultText:=S2M(DescString2)
        set MyCustomTracker.TargetWidget = text_block:
            DefaultTextColor:=MakeColorFromHex("FFFFFF")
            DefaultText:=TargetMessage(CurrentAmount, TargetAmount)
        set MyCustomTracker.IconWidget = texture_block:
            DefaultDesiredSize:=vector2{X:=70.0,Y:=70.0}
            DefaultImage := MissionIcon

        Overlay1:=overlay:
            Slots:=array:
                overlay_slot:
                    Widget:=texture_block:
                        DefaultDesiredSize:=vector2{X:=500.0,Y:=42.0}
                        DefaultImage := MissionsTrackers.UiBackgroundOpacityLess
                overlay_slot:
                    Widget:=texture_block:
                        DefaultDesiredSize:=vector2{X:=500.0,Y:=42.0}
                        DefaultImage := MissionsTrackers.UiBackgroundOpacityLess
                overlay_slot:
                    HorizontalAlignment := horizontal_alignment.Center
                    VerticalAlignment := vertical_alignment.Center
                    Widget:=MyCustomTracker.TitleWidget


        #Indent is used for description
        Indent :=texture_block:
            DefaultDesiredSize:=vector2{X:=74.0,Y:=30.0}
            DefaultImage := MissionsTrackers.UiBackgroundOpacityLess
        Indent.SetVisibility(widget_visibility.Hidden)

        #Indent is used for description line2
        IndentLine2 :=texture_block:
            DefaultDesiredSize:=vector2{X:=74.0,Y:=50.0}
            DefaultImage := MissionsTrackers.UiBackgroundOpacityLess
        IndentLine2.SetVisibility(widget_visibility.Hidden)

        #Used for gold and target text
        Indent2 :=texture_block:
            DefaultDesiredSize:=vector2{X:=20.0,Y:=30.0}
            DefaultImage := MissionsTrackers.UiBackgroundOpacityLess
        Indent2.SetVisibility(widget_visibility.Hidden)
   

        Overlay2:=overlay:
            Slots:=array:
                overlay_slot: # Background1
                    Widget:=texture_block:
                        DefaultDesiredSize:=vector2{X:=500.0,Y:=73.0}
                        DefaultImage := MissionsTrackers.UiBackgroundOpacityLess
                overlay_slot:  # Background2
                    HorizontalAlignment := horizontal_alignment.Left
                    VerticalAlignment := vertical_alignment.Top
                    Widget:=texture_block:
                        DefaultDesiredSize:=vector2{X:=70.0,Y:=70.0}
                        DefaultImage := MissionsTrackers.UiBackgroundOpacityLess
                overlay_slot:  # Icon
                    HorizontalAlignment := horizontal_alignment.Left
                    VerticalAlignment := vertical_alignment.Center
                    Widget:=MyCustomTracker.IconWidget
                overlay_slot: # Description
                    HorizontalAlignment := horizontal_alignment.Left
                    VerticalAlignment := vertical_alignment.Top
                    Widget:=stack_box:
                        Orientation := orientation.Horizontal
                        Slots:=array:
                            stack_box_slot: # Description Indentation
                                Widget:=Indent
                            stack_box_slot: # Description
                                Widget:= MyCustomTracker.DescriptionWidget
                overlay_slot: # Description Line 2
                    HorizontalAlignment := horizontal_alignment.Left
                    VerticalAlignment := vertical_alignment.Bottom
                    Widget:=stack_box:
                        Orientation := orientation.Horizontal
                        Slots:=array:
                            stack_box_slot: # Description Indentation
                                Widget:=IndentLine2
                            stack_box_slot: # Description
                                Widget:=MyCustomTracker.DescriptionWidget2
                #overlay_slot: # Target Text
                #    HorizontalAlignment := horizontal_alignment.Right
                #    VerticalAlignment := vertical_alignment.Top
                #    Widget:=stack_box:
                #        Orientation := orientation.Horizontal
                #        Slots:=array:
                #            stack_box_slot: # Target Text
                #                Widget:=MyCustomTracker.TargetWidget
                #            stack_box_slot: # Indentation
                #                Widget:=Indent2

        var MaxAnchors : vector2 = vector2{X := 0.01, Y := 0.35}
        var MinAnchors : vector2 = vector2{X := 0.01, Y := 0.35}

        Canvas:=canvas:
            Slots:=array:
                canvas_slot:
                    Anchors := anchors{Minimum := MinAnchors, Maximum := MaxAnchors}
                    Alignment := vector2{X := 0.0, Y := 0.0}
                    SizeToContent := true
                    Widget := stack_box:
                        Orientation := orientation.Vertical
                        Slots := array:
                            stack_box_slot:
                                Widget := Overlay1
                            stack_box_slot:
                                Widget := Overlay2

        set MyCustomTracker.Canvas = Canvas
        return MyCustomTracker


    InitCustomUI(Agent:agent,TitleString :string, DescString1 : string, DescString2 : string, TargetAmount : int , CurrentAmount : int, MissionIcon : texture):void=
        if(Player := player[Agent]):
            if (PlayerUI := GetPlayerUI[Player]):
                if (MyCustomTrackerWidget := CustomTrackerPerPlayer[Player]?):
                    #PlayerUI.RemoveWidget(MyCustomTrackerWidget.Canvas)
                    #if {set CustomTrackerPerPlayer[Player] = false}
                else:
                    MyCustomTrackerWidget := CreateCanvas(TitleString , DescString1 , DescString2 , TargetAmount , CurrentAmount  , MissionIcon )
                    PlayerUI.AddWidget(MyCustomTrackerWidget.Canvas, player_ui_slot{InputMode := ui_input_mode.None})
                    if (set CustomTrackerPerPlayer[Player] = option{MyCustomTrackerWidget}) {}
                    if(set CanvasPerPlayer[Agent] = option{MyCustomTrackerWidget.Canvas}){}



        

