
using { /Fortnite.com/Devices }
using { /Verse.org/Simulation }
using { /UnrealEngine.com/Temporary/Diagnostics }


dash_device := class(creative_device):


    @editable
    HudDevice : npc_hp_cinematic_hud_device = npc_hp_cinematic_hud_device{} # To get player id

    @editable
    TriggerForPlayers : []trigger_device = array{}

    @editable
    MovementDevice : movement_modulator_device = movement_modulator_device{}

    @editable
    InputManagerDevice : input_trigger_device = input_trigger_device{}
    
    @editable
    PickUpEffectTrigger : trigger_device = trigger_device{}

    @editable
    ClearEffectTrigger : trigger_device = trigger_device{}


    OnBegin<override>()<suspends>:void=

        InputManagerDevice.PressedEvent.Subscribe(InputPressed)
        for(i:=0..TriggerForPlayers.Length-1):
            if(CurrentTrigger := TriggerForPlayers[i]):
                CurrentTrigger.TriggeredEvent.Subscribe(ShowEffects)


    InputPressed(Agent:agent):void= 

        if:
            Player := player[Agent]
            OldData := Data[Player]
            OldData.Dash > 0
            PlayerId := HudDevice.IdPerAgent[Agent]?
            MyTrigger := TriggerForPlayers[PlayerId]
        then:
            
            MyTrigger.Trigger(Agent)

    ShowEffects(MaybeAgent:?agent):void=
        if(RealAgent:=MaybeAgent?):
            ClearEffectTrigger.Trigger(RealAgent)
                PickUpEffectTrigger.Trigger(RealAgent)
        



