
using { /Fortnite.com/Devices }
using { /Verse.org/Simulation }
using { /UnrealEngine.com/Temporary/Diagnostics }
using { /Verse.org/Random }

formater_device := class(creative_device):

    @editable
    TestButton : button_device = button_device{}


    OnBegin<override>()<suspends>:void=

        TestButton.InteractedWithEvent.Subscribe(HandleTestButton)

    HandleTestButton(Agent:agent):void=
        Random := GetRandomFloat(0.0,100000.0)
        Print("Random Number Test : {Random}")
        Print("Random Number Abv test true :{formatter.Abbreviate(Random * 1.0, true)}")
        Print("Random Number Abv test false :{formatter.Abbreviate(Random * 1.0, false)}")


formatter<public> := module:
    Abbreviations : []char = array { ' ', 'K', 'M', 'B', 'T','a','b','c','d','e','f','g','h','i','j','k','l','m','n','o','p','q','r','s','t','u','v','w','x','y','z'}

    GetText<localizes><public>(Message : string)<computes>:message = "{Message}"

    #https://forums.unrealengine.com/t/how-to-round-float-to-x-decimal-places/821189/4
    GetRoundedString(Input : float, ShowDecimals : logic): string =

        var NumDigitsAfterDecimal : int = 0

        if:
            if(ShowDecimals = true):
                BeforeBeforeDecimal := Floor[Input]
                BeforeBeforeString := ToString(BeforeBeforeDecimal)
                if(BeforeBeforeString.Length >=4):
                    set NumDigitsAfterDecimal = 0
                else:
                    if(BeforeBeforeString.Length =3):
                        set NumDigitsAfterDecimal = 0
                    else:
                        if(BeforeBeforeString.Length =2):
                            set NumDigitsAfterDecimal = 1
                        else:
                            if(BeforeBeforeString.Length =1 ):
                                set NumDigitsAfterDecimal = 2
            else:
                set NumDigitsAfterDecimal = 0

            Multiplier := Pow(10.0, NumDigitsAfterDecimal * 1.0)
            RoundedValue :=  float[Round[Input*Multiplier]*1.0] / Multiplier
            BeforeDecimal := Floor[RoundedValue]
            AfterDecimal := Abs(Round[(RoundedValue - BeforeDecimal*1.0)*Multiplier])
            var AfterDecimalString : string = ToString(AfterDecimal)

            #pad with leading zeroes
            for (It := 0..(NumDigitsAfterDecimal-AfterDecimalString.Length-1)):
                set AfterDecimalString = array{'0'} + AfterDecimalString

        then:
            if(NumDigitsAfterDecimal <= 0):
                "{BeforeDecimal}"
            else:
                "{BeforeDecimal}.{AfterDecimalString}" 
        else:
            ToString(Input)

    Abbreviate<public>(Number : float, Decimals:logic):string= # decimals represents numbers that you want it to show deciamls before reaching 1000
        for (Index : int = 0..Abbreviations.Length, Abbreviation := Abbreviations[Index]):
            if:
                Number < Pow(10.0, ((Index+1)*3.0))
                AbbreviatedNum := Number/((Pow(10.0, (Index)*3.0)))
            then:
                var RoundedAbbreviation : string = ""

                if:
                    Decimals = true
                then:
                    set RoundedAbbreviation = GetRoundedString(AbbreviatedNum * 1.0, true)
                else:
                    if(Number >= 1000.0):
                        set RoundedAbbreviation = GetRoundedString(AbbreviatedNum * 1.0, true)
                    else:
                        set RoundedAbbreviation = GetRoundedString(AbbreviatedNum * 1.0, false)


                if(Index = 0):
                    return "{RoundedAbbreviation}"
                else:
                    return "{RoundedAbbreviation}{Abbreviation}"

        return ""

    CommaSeparator<public>(Number : int):string=
        NumberAsString : string = "{Number}"
        StringLength : int = NumberAsString.Length

        var FormattedString : string = ""

        for (Index : int = 0..StringLength-1, Character := NumberAsString[Index]):
            RemainingLength := StringLength-(Index+1)
            set FormattedString += "{Character}"

            if (RemainingLength > 0 and Mod[RemainingLength, 3] = 0):
                set FormattedString += ","             

        return FormattedString