
using { /Fortnite.com/Devices }
using { /Verse.org/Simulation }
using { /UnrealEngine.com/Temporary/Diagnostics }
using { /UnrealEngine.com/Temporary/UI }
using { /UnrealEngine.com/Temporary/SpatialMath }
using { /Verse.org/Colors }
using { /Verse.org/Random }
using { /Verse.org/Assets }

hit_ui_device := class(creative_device):

    var CanvasPerPlayer : [agent]?canvas = map{}
    var HitPerPlayer : [agent]?hud_rising_animation = map{}
    
    MakeBasicCanvas(IconTexture:texture,Damage:float,TextColor:color):tuple(canvas,hud_rising_animation)=

        MyHit:=MakeProgressBar(1.0,vector2{X:=400.0,Y:=400.0},TextColor,Damage, IconTexture)
        MyWidget1:=MyHit.GetWidget1()
        MyWidget2:=MyHit.GetWidget2()
        MyWidget4:=MyHit.GetWidget4()
        # Used to add the plus sign
        #MyWidget3:=MyHit.GetWidget3()

        # Getting Random Locaiton For Anchors
        
        AnchorsVariation : float = 50.0
        AnchorsLeftRightVariation : float = 100.0
        # Previously had 0.53
        var MaxAnchors : vector2 = vector2{X := 0.5, Y := 0.26}
        var MinAnchors : vector2 = vector2{X := 0.5, Y := 0.26}
        AnchorRandom1 := GetRandomFloat(-AnchorsVariation,AnchorsVariation)
        AnchorRandom2 := GetRandomFloat(-AnchorsLeftRightVariation,AnchorsLeftRightVariation)
        AnchorRandom3 := GetRandomFloat(-AnchorsLeftRightVariation,AnchorsLeftRightVariation)
        AnchorRandom4 := GetRandomFloat(-AnchorsVariation,AnchorsVariation)


        Canvas:=canvas:
            Slots:=array:
                canvas_slot:
                    Anchors := anchors{Minimum := MinAnchors, Maximum := MaxAnchors}
                    Offsets := margin{Top := AnchorRandom1, Left := AnchorRandom2, Right := AnchorRandom3, Bottom := AnchorRandom4}
                    Alignment := vector2{X := 0.0, Y := 0.0}
                    Widget := stack_box:
                        Orientation := orientation.Horizontal
                        Slots := array:
                            stack_box_slot:
                                Widget := MyWidget1
                            stack_box_slot:
                                Widget := MyWidget2
                            # PlusSignText 
                            #stack_box_slot:
                            #    Widget := MyWidget3
                            stack_box_slot:
                                Widget := MyWidget4

        return(Canvas,MyHit)
                        
                        
    ShowHitUi(Agent:agent,IconTexture:texture,TextColor:color,Damage:float):void=
        
        Vals:=MakeBasicCanvas(IconTexture,Damage,TextColor)
        Canvas:=Vals(0)
        Hit:=Vals(1)

        if:
            PlayerUI:=GetPlayerUI[player[Agent]]
        then:
            PlayerUI.AddWidget(Canvas,player_ui_slot{InputMode:=ui_input_mode.None})
            if:
                set CanvasPerPlayer[Agent] = option{Canvas}
                set HitPerPlayer[Agent] = option{Hit}
            then:
                spawn:
                    MoveAnimController(Hit,Canvas,PlayerUI)

    MoveAnimController(Hit:hud_rising_animation,Canvas:canvas, PlayerUI:player_ui)<suspends>:void=

        # small delay so it waits
        #Sleep(0.2)

        loop:
            Sleep(0.01)

            set Hit.Progress = Hit.Progress -0.025
            Hit.SetPosition(Hit.Progress)

            if:
                Hit.Progress <= 1.0
                Hit.Progress >= 0.0
                myConstrainedValue: type{_X:float where 0.000000 <= _X, _X <= 1.000000} = type{_X:float where 0.000000 <= _X, _X <= 1.000000}[Hit.Progress]
                
            then:
                Hit.TextBlock.SetTextOpacity(myConstrainedValue)
                Hit.TextBlock.SetShadowOpacity(myConstrainedValue)
                Hit.TextBlock2.SetTextOpacity(myConstrainedValue)
                Hit.TextBlock2.SetShadowOpacity(myConstrainedValue)
            
            if(Hit.Progress <= 0.1):
                RemoveUIForPlayer(Hit,Canvas,PlayerUI)
            if(Hit.Progress <= 0.0):
                break


            
    RemoveUIForPlayer(Hit:hud_rising_animation, Canvas:canvas, PlayerUI:player_ui)<suspends>:void=

        #Overlay := Hit.GetWidget()
        Canvas.RemoveWidget(Hit.MovingColorBase)
        Canvas.RemoveWidget(Hit.TextureIconBlock)
        Canvas.RemoveWidget(Hit.TextBlock)
        PlayerUI.RemoveWidget(Canvas)


        
