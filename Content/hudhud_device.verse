
using { /Fortnite.com/Devices }
using { /Verse.org/Simulation }
using { /UnrealEngine.com/Temporary/Diagnostics }
using { /UnrealEngine.com/Temporary/UI }

hudhud_device := class(creative_device):

    XpBar : Combat_Level_Hud = Combat_Level_Hud{}

    @editable
    Spawners : []player_spawner_device = array{}

    @editable
    StartButton : button_device = button_device{}
    @editable
    TestButton : button_device = button_device{}

    @editable
    vfx : vfx_creator_device = vfx_creator_device{}
    @editable 
    audio : audio_player_device = audio_player_device{}
    @editable
    Score : score_manager_device = score_manager_device{}

    @editable
    Chickens : []wildlife_spawner_device = array{}
    @editable
    XpPerChicken : float = 4.0
    @editable
    Boars : []wildlife_spawner_device = array{}
    @editable
    XpPerBoar : float = 4.0
    @editable
    Wolves : []wildlife_spawner_device = array{}
    @editable
    XpPerWolf : float = 4.0

    
    @editable 
    hudpopup : hud_message_device = hud_message_device{}

    storedstringpopup<localizes>(INT: int) : message = "+{INT}XP"

    var xpcounter : int = 0
    var CurrentLevel : int = 1

    var delayenabled : logic = false

    OnBegin<override>()<suspends>:void=

        StartButton.InteractedWithEvent.Subscribe(Pressed)
        TestButton.InteractedWithEvent.Subscribe(Pressed2)

        for(i:=0..Chickens.Length):
            if(Chicken := Chickens[i]):
                Chicken.EliminatedEvent.Subscribe(ChickenDeath)
        for(i:=0..Boars.Length):
            if(Boar := Boars[i]):
                Boar.EliminatedEvent.Subscribe(BoarDeath)
        for(i:=0..Wolves.Length):
            if(Wolf := Wolves[i]):
                Wolf.EliminatedEvent.Subscribe(WolvesDeath)
        
    Pressed(Agent:agent):void =
        if(Player := player[Agent]):
            XpBar.Init(Agent)
            XpBar.ShowUIForPlayer(Player)

    Pressed2(Agent:agent):void =
        if(Player:= player[Agent]):
            XpBar.ShowUIForPlayer(Player)
            XpBar.AddXp(10.0,option{Agent})

    RaiseXp(Amount : float, Agent:agent,ShowHud:logic):void =
        var LeveledUp : logic = false

        if(Player:= player[Agent]):
            XpBar.ShowUIForPlayer(Player)
            set LeveledUp =  XpBar.AddXp(Amount,option{Agent})
            if(LeveledUp = true):
                set CurrentLevel = CurrentLevel + 1
                Print("LeveledUp")
                vfx.SpawnAt(Agent)
                vfx.Begin(Agent)
                audio.Play(Agent)

            if(ShowHud = true):
                if(IntAmount := Floor[Amount]):
                    set xpcounter = xpcounter + IntAmount
                    hudpopup.SetText(storedstringpopup(xpcounter))
                    hudpopup.Show(Agent)
                    spawn:
                        ResetTimerOnXpText(xpcounter)

    ResetTimerOnXpText(xpwhensent : int)<suspends>:void = 
        Sleep(0.5)
        if(xpcounter = xpwhensent):
            set xpcounter = 0

    ChickenDeath(Interaction:device_ai_interaction_result):void = 
        if(Agent:= Interaction.Source?):
            if(delayenabled=false):
                set delayenabled = true
                RaiseXp(XpPerChicken,Agent,true)
                spawn:
                    XpDelay()
            else:
    BoarDeath(Interaction:device_ai_interaction_result):void = 
        if(Agent:= Interaction.Source?):
            RaiseXp(XpPerBoar,Agent,true)
    WolvesDeath(Interaction:device_ai_interaction_result):void = 
        if(Agent:= Interaction.Source?):
            RaiseXp(XpPerWolf,Agent,true)

    XpDelay()<suspends>:void = 
        Sleep(0.5)
        set delayenabled = false






