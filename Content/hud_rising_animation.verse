
using { /Fortnite.com/Devices }
using { /Verse.org/Simulation }
using { /UnrealEngine.com/Temporary/Diagnostics }
using { /UnrealEngine.com/Temporary/UI }
using { /Verse.org/Colors }
using { /UnrealEngine.com/Temporary/SpatialMath }
using { /Fortnite.com/UI }
using { /Verse.org/Assets }

MakeProgressBar(InProgress:float,InSize:vector2,InTextColor:color,InUnits:float,InIconTexture:texture):hud_rising_animation=
    return hud_rising_animation{Progress:=InProgress,Size:=InSize,TextColor:=InTextColor, Units:=InUnits, IconTexture := InIconTexture}

hud_rising_animation := class():

    # Defaults
    var IconTexture : texture = Icons.DefaultIcon
    var Units : float = 0.0
    var Progress : float = 0.0
    Size : vector2
    var TextColor : color = color{}

    var MovingColorBase : color_block = color_block{}
    var TextureIconBlock : texture_block = texture_block{
        DefaultImage := Icons.DefaultIcon}
    var TextBlock : text_block = text_block{}
    var TextBlock2 : text_block = text_block{}

    S2M<private><localizes>(value : string) : message = "{value}"
    Plus<private><localizes>() : message = "+"

    #Return a widget to be used in your UI
    #As a widget, you can only use one progress_bar once.
    GetWidget1():overlay=

        set MovingColorBase = color_block{DefaultColor:=TextColor,DefaultDesiredSize:=Size}
        
        SetPosition(Progress)

        return overlay:
            Slots:=array:
                overlay_slot:
                    Widget:=MovingColorBase
                    HorizontalAlignment:=horizontal_alignment.Fill
                    VerticalAlignment:=vertical_alignment.Fill


    GetWidget2():overlay=

        set TextureIconBlock = texture_block{DefaultImage := IconTexture,DefaultDesiredSize:=vector2{X:=32.0,Y:=32.0}}

        return overlay:
            Slots:=array:
                overlay_slot:
                    Widget:=TextureIconBlock
                    HorizontalAlignment:=horizontal_alignment.Center
                    VerticalAlignment:=vertical_alignment.Center

    # OutDated USed to add a plus sign
    GetWidget3():overlay=

        set TextBlock = text_block{DefaultTextColor := TextColor, DefaultText:=Plus(), DefaultShadowColor:={MakeColorFromHex("5b5b5b")}, DefaultShadowOffset:=option{vector2{X:=2.0,Y:=2.0}}}

        return overlay:
            Slots:=array:
                overlay_slot:
                    Widget:=TextBlock
                    HorizontalAlignment:=horizontal_alignment.Center
                    VerticalAlignment:=vertical_alignment.Center

    GetWidget4():overlay=

        set TextBlock2 = text_block{DefaultTextColor := TextColor, DefaultText:=S2M(formatter.Abbreviate((Units * 1.0),false)), DefaultShadowColor:={MakeColorFromHex("5b5b5b")}, DefaultShadowOffset:=option{vector2{X:=2.0,Y:=2.0}}}

        return overlay:
            Slots:=array:
                overlay_slot:
                    Widget:=TextBlock2
                    HorizontalAlignment:=horizontal_alignment.Center
                    VerticalAlignment:=vertical_alignment.Center

    #Updates the progress displayed
    SetPosition(InProgress:float):void=
        set Progress = InProgress
        #If the orientation is vertical
        MovingColorBase.SetDesiredSize(vector2{Y:=Size.Y*Progress})

    #Sets the color of the front of the progress bar
    SetTextColor(MyColor:color):void=
        set TextColor = MyColor
        TextBlock.SetTextColor(TextColor)
