
using { /Fortnite.com/Devices }
using { /Verse.org/Simulation }
using { /UnrealEngine.com/Temporary/Diagnostics }
using { /Fortnite.com/Characters }
using { /Verse.org/Random }
using { /UnrealEngine.com/Temporary/SpatialMath }

ice_area_device := class(creative_device):
    
    @editable
    IceAreaProp : creative_prop_asset = DefaultCreativePropAsset
    @editable
    BossNpcSpawner : npc_manager_device = npc_manager_device{}
    @editable
    HitAudio : audio_player_device = audio_player_device{}
    @editable
    FreezeAudio : audio_player_device = audio_player_device{}


    @editable
    var IceAreaNumber : int = 5
    @editable
    SpawnRadius : float = 1500.0
    @editable
    IceLifeTime : float = 15.0
    @editable
    DamageZoneSize : float = 256.0
    @editable
    Damage : float = 10.0

    @editable
    FreezeTime : float = 3.0

    @editable
    IceBlockProp : creative_prop_asset = DefaultCreativePropAsset

    var PlayersFrozen : [agent]?logic = map{}


    OnBegin<override>()<suspends>:void=
        StartIceSpawning()
           

    StartIceSpawning()<suspends>:void=
        loop:

            Sleep(IceLifeTime)
            
            if:
                BossCharacter := BossNpcSpawner.NPCAgent?.GetFortCharacter[]
                BossCharacter.IsActive[]
            then:
                for(i:=0..IceAreaNumber-1):

                    var SpawnLoc :vector3 = BossCharacter.GetTransform().Translation
                    set SpawnLoc.X = SpawnLoc.X + GetRandomFloat(-SpawnRadius,SpawnRadius)
                    set SpawnLoc.Y = SpawnLoc.Y + GetRandomFloat(-SpawnRadius,SpawnRadius)
                    set SpawnLoc.Z = 101.0

                    RandomRotation : float = GetRandomFloat(0.0,180.0)
                    NewRotation := BossCharacter.GetTransform().Rotation.ApplyYaw(RandomRotation)
            
                    SpawnedPropResult := SpawnProp(IceAreaProp,SpawnLoc,NewRotation)

                    if(SpawnedProp := SpawnedPropResult(0)?):
                        spawn:
                            FreezePlayers(SpawnedProp)
                        spawn:
                            DeleteEffect(IceLifeTime,SpawnedProp)

                    Sleep(0.5)

    


    FreezePlayers(IceArea : creative_prop)<suspends>:void=
        loop:

            Sleep(0.1)

            if(IceArea.IsValid[]):

                MyLocation := IceArea.GetTransform().Translation
                Players := HitAudio.GetPlayspace().GetPlayers()

                for(i:=0..Players.Length-1):
                    if:
                        Player := Players[i]
                        Char := Player.GetFortCharacter[]
                        PlayerAgent := Char.GetAgent[]
                        Char.IsActive[]
                        
                    then:
                        PlayerLocation : vector3 = Char.GetTransform().Translation
                        var Dis2Player : float = Distance(MyLocation,PlayerLocation)

                        if(Dis2Player <= DamageZoneSize):
                            if(not PlayersFrozen[PlayerAgent]? = true):
                                if(set PlayersFrozen[PlayerAgent] = option{true}){}
                                
                                spawn:
                                    FreezeAndUnfreeze(PlayerAgent)
                            

            else:
                break

    DeleteEffect(SleepTime : float, Prop : creative_prop)<suspends>:void=
        Sleep(SleepTime)
        Prop.Dispose()

    FreezeAndUnfreeze(Agent:agent)<suspends>:void=

        Sleep(0.5)

        
        if(Char := Agent.GetFortCharacter[]):

            HitAudio.Play(Agent)
            Char.Damage(Damage)

            StasisArg : stasis_args = stasis_args{}
            Char.PutInStasis(StasisArg)

            var CharTrans : vector3 = Char.GetTransform().Translation
            set CharTrans.Z = CharTrans.Z - 50.0

            SpawnedPropResult := SpawnProp(IceBlockProp,CharTrans,Char.GetTransform().Rotation)
                if(SpawnedProp := SpawnedPropResult(0)?):
                    Sleep(FreezeTime )
                    DeleteEffect(0.1,SpawnedProp)
                    Char.ReleaseFromStasis()
                    Sleep(2.0)
                    if(set PlayersFrozen[Agent] = false){}

            
        


