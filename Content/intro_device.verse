
using { /Fortnite.com/Devices }
using { /Verse.org/Simulation }
using { /UnrealEngine.com/Temporary/Diagnostics }

intro_device := class(creative_device):

    @editable
    PopUp : popup_dialog_device = popup_dialog_device{}
    @editable
    PlayerSpawners : []player_spawner_device = array{}

    @editable
    RemoteGranter : item_granter_device = item_granter_device{}
    @editable
    OreGranter : item_granter_device = item_granter_device{}

    @editable
    StartingCine : cinematic_sequence_device = cinematic_sequence_device{}

    @editable
    CineLength : float = 22.0



    @editable
    Area1Volumes : []volume_device = array{}
    @editable
    Area2Volumes : []volume_device = array{}
    @editable
    Area3Volumes : []volume_device = array{}
    @editable
    Area4Volumes : []volume_device = array{}

    @editable
    TownVol : volume_device = volume_device{}

    @editable
    RadioDevices : []radio_device = array{}

    @editable
    BossVolumes : []volume_device = array{}
    @editable
    BossRadio : radio_device = radio_device{}

    @editable
    MusicButton : button_device = button_device{}

    @editable
    MusicButton2 : button_device = button_device{}

    var AgentsShownPatchnotes : [agent]logic = map{}

    @editable
    MusicEnabled : logic = false

    @editable
    QuestDevice : quest_device = quest_device{}

    @editable
    CustomMission : custom_missions_device = custom_missions_device{}


    OnBegin<override>()<suspends>:void=

        MusicButton.InteractedWithEvent.Subscribe(MusicHandle)
        MusicButton2.InteractedWithEvent.Subscribe(MusicHandle)
        TownVol.AgentEntersEvent.Subscribe(EntersTownVolume)

        for(i:=0..Area1Volumes.Length-1):
            if(MusicVol := Area1Volumes[i]):
                MusicVol.AgentEntersEvent.SubscribeAgent(EntersMusicVolume,1)

        for(i:=0..Area2Volumes.Length-1):
            if(MusicVol := Area2Volumes[i]):
                MusicVol.AgentEntersEvent.SubscribeAgent(EntersMusicVolume,2)

        for(i:=0..Area3Volumes.Length-1):
            if(MusicVol := Area3Volumes[i]):
                MusicVol.AgentEntersEvent.SubscribeAgent(EntersMusicVolume,3)

        for(i:=0..Area4Volumes.Length-1):
            if(MusicVol := Area4Volumes[i]):
                MusicVol.AgentEntersEvent.SubscribeAgent(EntersMusicVolume,4)

        for(i:=0..BossVolumes.Length-1):
            if(MusicVol := BossVolumes[i]):
                MusicVol.AgentEntersEvent.Subscribe(BossMusicVolume)

        PopUp.RespondingButtonEvent.Subscribe(PlayStartingCine)

        Players:=GetPlayspace().GetPlayers()
        for(Player:Players):
            if(Agent:=agent[Player]):
                ShowNotes(Agent)
                
        for(Spawner:PlayerSpawners):
            Spawner.SpawnedEvent.Subscribe(ShowNotes)
            

    MusicHandle(Agent:agent):void=

        if:
            Player:=player[Agent]
            OldData := Data[Player]
            MusicLogic := OldData.MusicEnabled
        then:
            if(MusicLogic = false):
                if(set Data[Player] = player_data{player_data_update<constructor>(OldData),MusicEnabled := true}){}
                EntersTownVolume(Agent)
            else:
                if(set Data[Player] = player_data{player_data_update<constructor>(OldData),MusicEnabled := false}){}
                for(RadDev1 : RadioDevices):
                    RadDev1.Unregister(Agent)



    UnsubscribeForAll():void=
        for(RadDev : RadioDevices):
            RadDev.UnregisterAll()

    BossMusicVolume(Agent:agent):void=

        if(MusicEnabled = true):
            if:
                Player:=player[Agent]
                OldData := Data[Player]
                var AlreadyWatched : logic = OldData.IntroWatched 
                AlreadyWatched = true
                MusicLogic := OldData.MusicEnabled
                MusicLogic = true
            then:

                for(RadDev1 : RadioDevices):
                    RadDev1.Unregister(Agent)

                BossRadio.Register(Agent)



    EntersMusicVolume(Agent:agent, id:int):void=

        if(MusicEnabled = true):

            if(RadDev : radio_device = RadioDevices[id]):
                if:
                    Player:=player[Agent]
                    OldData := Data[Player]
                    var AlreadyWatched : logic = OldData.IntroWatched 
                    AlreadyWatched = true
                    MusicLogic := OldData.MusicEnabled
                    MusicLogic = true
                then:
                    for(RadDev1 : RadioDevices):
                        RadDev1.Unregister(Agent)

                    RadDev.Register(Agent)

    EntersTownVolume(Agent:agent):void=

        if(MusicEnabled = true):

            if:
                Player:=player[Agent]
                OldData := Data[Player]
                var AlreadyWatched : logic = OldData.IntroWatched 
                AlreadyWatched = true
                MusicLogic := OldData.MusicEnabled
                MusicLogic = true
            then:

                for(RadDev1 : RadioDevices):
                    RadDev1.Unregister(Agent)

                if(RadDev : radio_device = RadioDevices[0]):
                    RadDev.Register(Agent)
            
            
            
    ShowNotes(Agent:agent):void=
        if:
            ShownLogic := AgentsShownPatchnotes[Agent]
            ShownLogic = true
        else:
            if(set AgentsShownPatchnotes[Agent] = true){}
            PopUp.Show(Agent)


    PlayStartingCine(Agent:agent,ButtonPressed:int):void=

        spawn:
            QuestDevice.LoadQuest(Agent)
        CustomMission.LoadingTest(Agent)

        if:
            Player:=player[Agent]
            OldData := Data[Player]
            var AlreadyWatched : logic = OldData.IntroWatched
            AlreadyWatched = false
        then:

            RemoteGranter.GrantItem(Agent)
            OreGranter.GrantItem(Agent)
    
            spawn:
                PlayCine(Agent)

    PlayCine(Agent:agent)<suspends>:void=

        StartingCine.Play(Agent)
        Sleep(CineLength)
        if:
            Player:=player[Agent]
            Player.IsActive[]
            OldData := Data[Player]
            set Data[Player] = player_data{player_data_update<constructor>(OldData),IntroWatched := true}
        then:
            EntersTownVolume(Agent)



    
                


        
