
using { /Fortnite.com/Devices }
using { /Verse.org/Simulation }
using { /UnrealEngine.com/Temporary/Diagnostics }
using { /UnrealEngine.com/Temporary/UI }
using { /Fortnite.com/UI }
using { /UnrealEngine.com/Temporary/SpatialMath }
using { /Verse.org/Colors }

Level_Hud := class():
  #@editable
    #VfxLevelSwitch : switch_device = switch_device{}
    #@editable
    #Audio : audio_player_device = audio_player_device{}

    var CurrentLevel : int = 1

    var iTotalXp<private> : float = 32.0
    var iCurrentXp<private> : float = 0.0
    var iBossName<private> : string = "Xp"
    
    BossBarColorBase<private> : color_block = color_block{
        DefaultColor := MakeColorFromHex("000000")
        DefaultDesiredSize := vector2{X:= 600.0, Y := 40.0}} 

    var BossBarColorTop<private> : color_block = color_block{}  
    var Overlay<private> : overlay = overlay{}
        
    XpBarTitleTextBlock<private> : text_block = text_block{DefaultTextColor := MakeColorFromHex("FFFFFF")}
    LevelTitleTextBlock<private> : text_block = text_block{DefaultTextColor := MakeColorFromHex("FFFFFF")}

    HPTitleText<private><localizes>(HPText : string) : message = "{HPText}"


    Init(BossName : string, BossTotalHP: float, Agent:agent):void=

        set iBossName = BossName
        set iTotalXp = BossTotalHP
        set iCurrentXp = 0.0

        XpBarTitleTextBlock.SetText(HPTitleText("{iBossName}: {iCurrentXp}/{iTotalXp}"))
        LevelTitleTextBlock.SetText(HPTitleText("Level: {CurrentLevel}"))

        HealthPercentage := iCurrentXp / iTotalXp
        BarSize := BossBarColorBase.GetDesiredSize()
        NewHealthBarWidth := BarSize.X * HealthPercentage
        if (Current := Int[iCurrentXp], Total := Int[iTotalXp]):
            XpBarTitleTextBlock.SetText(HPTitleText("{iBossName}: {Current}/{Total}"))
            LevelTitleTextBlock.SetText(HPTitleText("Level: {CurrentLevel}"))
        
        UpdateHealthBar(NewHealthBarWidth)

        spawn:
            DelayedFix()

    AddXp(XpAmount : float):logic=

        var LeveledUp : logic = false
        set iCurrentXp += XpAmount
        if (iCurrentXp < 0.0):
            set iCurrentXp = 0.0
        if(iCurrentXp >= iTotalXp):
            set LeveledUp = true
            set iCurrentXp = 0.0
            set CurrentLevel = CurrentLevel + 1
            set iTotalXp = CalculateExpReq(CurrentLevel)
 
        HealthPercentage := iCurrentXp / iTotalXp
        BarSize := BossBarColorBase.GetDesiredSize()
        NewHealthBarWidth := BarSize.X * HealthPercentage
        if (Current := Int[iCurrentXp], Total := Int[iTotalXp]):
            XpBarTitleTextBlock.SetText(HPTitleText("{iBossName}: {Current}/{Total}"))
            LevelTitleTextBlock.SetText(HPTitleText("Fishing Lv: {CurrentLevel}"))
        
        UpdateHealthBar(NewHealthBarWidth)

        if(LeveledUp = true):
            return true
        else:
            return false

    ShowUIForPlayer(Player : player):void=

        if (PlayerUI := GetPlayerUI[Player]):
            PlayerUI.AddWidget(CreateUI())

        spawn:
            DelayedFix()

        #spawn:
        #    RemoveUIForPlayerDelayed(Player)
        

    RemoveUIForPlayer(Player:player)<suspends>:void=
        
        Overlay.RemoveWidget(XpBarTitleTextBlock)
        Overlay.RemoveWidget(LevelTitleTextBlock)
        Overlay.RemoveWidget(BossBarColorTop)
        Overlay.RemoveWidget(BossBarColorBase)

        if (PlayerUI := GetPlayerUI[Player]):
            PlayerUI.RemoveWidget(Overlay)

    UpdateHealthBar(NewWidth : float):void=

        Overlay.RemoveWidget(XpBarTitleTextBlock)
        Overlay.RemoveWidget(LevelTitleTextBlock)
        Overlay.RemoveWidget(BossBarColorTop)

        set BossBarColorTop = CreateTopHealthBar(NewWidth)

        Overlay.AddWidget(CreateOverlaySlot(BossBarColorTop, horizontal_alignment.Left))
        Overlay.AddWidget(CreateOverlaySlot(XpBarTitleTextBlock,horizontal_alignment.Center))
        Overlay.AddWidget(CreateOverlaySlot(LevelTitleTextBlock,horizontal_alignment.Left))
        
    CreateOverlaySlot<private>(TheWidget : widget, HAlignment : horizontal_alignment):overlay_slot=  
        overlay_slot:
            Widget := TheWidget 
            HorizontalAlignment := HAlignment

    CreateTopHealthBar<private>(Width: float):color_block=
        ColorBlock := color_block{
            DefaultColor := MakeColorFromHex("#7EC8E3")
            DefaultDesiredSize := vector2{X:= Width, Y := 40.0}} 
        
        set BossBarColorTop = ColorBlock
        

    CreateOverlay<private>() : overlay=
        TheOverlay := overlay:
            Slots := array:
                overlay_slot:
                    Widget := BossBarColorBase
                overlay_slot:
                    Widget := BossBarColorTop
                overlay_slot:
                    Widget := XpBarTitleTextBlock
                overlay_slot:
                    Widget := LevelTitleTextBlock
        set Overlay = TheOverlay


    CreateUI<private>() : canvas=
        TheCanvas : canvas = canvas:
            Slots := array:
                canvas_slot:
                    Anchors := anchors{Minimum := vector2{X := 0.5, Y := 0.9}, Maximum := vector2{X := 0.5, Y := 0.9}}
                    Offsets := margin{Top := 0.0, Left := 0.0, Right := 0.0, Bottom := 0.0}
                    Alignment := vector2{X := 0.0, Y := 0.0}
                    SizeToContent := true
                    Widget := stack_box:
                        Orientation := orientation.Horizontal
                        Slots := array:
                            stack_box_slot:
                                Widget := CreateOverlay()
                            
                                                                
        return TheCanvas    


    CalculateExpReq(Level:int):float = 
        var total : float = 0.0
        for(i:=1..Level):
            #o:= i*1.0 + 600.0
            Power := Pow(1.9,(i*1.0) / 4.0)*1.0
            if (FloorThing := Floor[(i*1.0)+56.0 * Power * 1.0]):
                set total = total + (FloorThing*1.0)
        if(ReturnThing := Floor[total/4.0]*1.0):
            return ReturnThing
        else:
            0.0

    DelayedFix()<suspends>:void = 
        Sleep(0.01)
        Print("adding xp")
        AddXp(0.0)




        