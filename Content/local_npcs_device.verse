
using { /Fortnite.com/Devices }
using { /Verse.org/Simulation }
using { /UnrealEngine.com/Temporary/Diagnostics }
using { /UnrealEngine.com/Temporary/SpatialMath }
using { /Fortnite.com/Characters }

local_npcs_device := class(creative_device):

    @editable
    LocalNpcDevices : []npc_spawner_device = array{}
    # have to be aligned
    @editable
    LocalNpcTriggers : []trigger_device = array{}

    var NpcLocationMap : [agent]?vector3 = map{}
    var NpcTriggerMap : [agent]?int = map{}


    OnBegin<override>()<suspends>:void=

        for(i:=0..LocalNpcDevices.Length-1):
            if(NpcDevice := LocalNpcDevices[i]):
                NpcDevice.SpawnedEvent.SubscribeAgent(SpawnedNpc,i)
                NpcDevice.EliminatedEvent.Subscribe(ElimHandle)

        UpdateLocationVectors()


    SpawnedNpc(Agent:agent,i:int):void=

        if(Fort := Agent.GetFortCharacter[]):
            if(NpcExists := NpcLocationMap[Agent]):
            else:
                if(set NpcLocationMap[Agent] = option{Fort.GetTransform().Translation}){}
                if(set NpcTriggerMap[Agent] = option{i}){}



    ElimHandle(Interaction:device_ai_interaction_result):void=
        
        if(Agent := Interaction.Target?):
            if(NpcExists := NpcLocationMap[Agent]):

                var TempPlayerMap : [agent]?vector3 = map{}
                for(Key -> Value : NpcLocationMap,Key <> Agent):
                    set TempPlayerMap = ConcatenateMaps(TempPlayerMap, map{Key =>Value})
                set NpcLocationMap = TempPlayerMap

                var TempPlayerMap2 : [agent]?int = map{}
                for(Key -> Value : NpcTriggerMap,Key <> Agent):
                    set TempPlayerMap2 = ConcatenateMaps(TempPlayerMap2, map{Key =>Value})
                set NpcTriggerMap = TempPlayerMap2


    UpdateLocationVectors()<suspends>:void=
        loop:
            Sleep(0.1)
            for(Key->Value:NpcLocationMap):
                if(Fort:=Key.GetFortCharacter[]):
                    if(set NpcLocationMap[Key] = option{Fort.GetTransform().Translation}):


    

