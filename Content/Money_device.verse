
using { /Fortnite.com/Devices }
using { /Verse.org/Simulation }
using { /UnrealEngine.com/Temporary/Diagnostics }
using { /Fortnite.com/UI }
using { /UnrealEngine.com/Temporary/UI }
using { /Verse.org/Colors/NamedColors }
using { /UnrealEngine.com/Temporary/SpatialMath }

Money_device := class(creative_device):


    @editable 
    RefreshRate : float = 0.1
    @editable 
    PlayerSpawners : []player_spawner_device = array{}
    @editable
    BackgroundDevice : hud_message_device = hud_message_device{}
    @editable
    GemBackgroundDevice : hud_message_device = hud_message_device{}

    var MoneyTextBlockPerAgent : [agent]?text_block = map{}

    storedstring<localizes>(S:string) : message = "{S}"

    # this is required by a lot of files but no longer necessary in the game. Delete later

    OnBegin<override>()<suspends>:void=
        
        for(PS:PlayerSpawners):
            PS.SpawnedEvent.Subscribe(PlayerSpawned)


    PlayerSpawned(Agent:agent):void=

        if(MyTextblock:=MoneyTextBlockPerAgent[Agent]?){}
        else:
            BackgroundDevice.Show(Agent)
            if(MyData := Data[player[Agent]]):
                if(MyData.Gems > 0.0):
                    GemBackgroundDevice.Show(Agent)
            spawn{RefreshUI(Agent)}

    RefreshUI(Agent:agent)<suspends>:void=
        loop:
            Sleep(RefreshRate)

            if:
                Player:=player[Agent]
                PlayerData := Data[Player]
            then:
                BackgroundDevice.SetText(storedstring(formatter.Abbreviate((PlayerData.Gold * 1.0),false)))
                BackgroundDevice.Show(Agent)
                if(PlayerData.Gems > 0.0):
                    GemBackgroundDevice.SetText(storedstring(formatter.Abbreviate((PlayerData.Gems * 1.0),false)))
                    GemBackgroundDevice.Show(Agent)


