
using { /Fortnite.com/Devices }
using { /Verse.org/Simulation }
using { /UnrealEngine.com/Temporary/Diagnostics }
using { /Fortnite.com/Characters }

movementspeed_device := class(creative_device):

    @editable
    ClassSelectors : []class_and_team_selector_device = array{}

    SetClass(Agent:agent):void=
        if:
            Player:=player[Agent]
            PlayerData:=Data[Player]
        then:
            if(MySelector := ClassSelectors[PlayerData.MovementSpeed-1]):
                MySelector.ChangeClass(Agent)