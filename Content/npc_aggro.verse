using { /Fortnite.com/AI }
using { /Fortnite.com/Characters }
using { /Verse.org/Colors }
using { /Verse.org/Random }
using { /Verse.org/Simulation }
using { /UnrealEngine.com/Temporary/Diagnostics }
using { /UnrealEngine.com/Temporary/SpatialMath }
using { /Fortnite.com/Game }
using { /Fortnite.com/Animation/PlayAnimation }
using { /Fortnite.com/Devices }
using { /Verse.org/Assets }

ShowTemperaryNiagara<public>(Spawn_Location : vector3, Spawn_Rotation : rotation, Prop_Asset : creative_prop_asset, Life_Time : float )<suspends>:void=
    SpawnedPropResult := SpawnProp(Prop_Asset, Spawn_Location ,Spawn_Rotation)
    if (SpawnedProp := SpawnedPropResult(0)?):
        Sleep(Life_Time)
        SpawnedProp.Dispose()

npc_aggro<public> := class(npc_behavior):

# The speed to play the animation at.
    #PlayRate : float = 1.0
    # How long to blend the previous animation into
     # the current one.
    BlendInTime : float = 0.1
    # How long to blend the current animation into
    # the next one.
    BlendOutTime : float = 0.1
    # The position in seconds to start playing the
    # animation from.
    StartPositionSeconds : float = 0.0
    # How long to wait before restarting the animation.

    @editable
    Stationary : logic = false

    SleepDuration : float = 2.0
    MoveToWaitDuration:float = 0.5
    MoveToWaitIntervals: int = 10

    @editable
    DistanceFromSpawnPtToMove:float = 750.0
    DistanceToMoveWhenFleeing : float = 500.0

    var CharacterAggro : ?fort_character = false
    var SavedHitRotation : rotation = rotation{}

    var GotHit : logic = false
    var Ended : logic = false
    var Aggroed : logic = false


    @editable
    AggroVolume : volume_device = volume_device{}
    #start of the attack
    @editable
    AttackAudio : audio_player_device = audio_player_device{}
    @editable
    RangeAttackAudio : audio_player_device = audio_player_device{}
    # Player getting hit
    @editable
    HitAudio : audio_player_device = audio_player_device{}

    var NPCSpawnPoint : vector3 = vector3{}
    
    @editable
    TornadoCastAudio : audio_player_device = audio_player_device{}
    @editable
    RainCastAudio : audio_player_device = audio_player_device{}
    @editable
    ChargeAudio : audio_player_device = audio_player_device{}
    

    # Attack Attributes
    # Distance from the attack point / attck point swivles around the creature
    @editable
    AttackSize : float = 50.0

    # the distance from the creature that the point hovers at 
    @editable
    AttackRange : float = 90.0
    #How far in front of the mob is the attack location
    @editable
    AttackRangeRanged : float = 750.0
    #How far in front of the mob is the attack location
    #The amount of time after the attack animation start where it checks player pos
    @editable
    AttackDamageDelay : float = 0.3
    @editable
    RangedAttackDamageDelay : float = 0.3
    @editable
    AttackAnimLength : float = 1.0
    @editable
    RangedAttackAnimLength : float = 1.0
    @editable
    AttackDamage : float = 13.0
    @editable
    RangedNPC : logic = false
    @editable
    AttackImpact : creative_prop_asset = DefaultCreativePropAsset
    @editable
    ImpactNSTime : float = 0.5

    # Animation Attributes
    var npc_animation : [NPCData.NPCStates.npc_states]animation_sequence = map{}
    @editable
    var npctype : NPCData.NPCTypes.npc_types = NPCData.NPCTypes.npc_types.LavaMite
    
    @editable
    ProjectileAsset : creative_prop_asset = DefaultCreativePropAsset
    @editable
    ProjectileEndNiagara :  creative_prop_asset = DefaultCreativePropAsset
    @editable
    EndNiagLogic : logic = false
    @editable
    RangeDamage : float = 15.0
    @editable
    Rangespeed : float = 0.015
    @editable
    Rangesampletime : float = 0.25
    @editable
    RangeHitBox : float = 30.0
    @editable
    RangeDistancePastPlayer : float = 750.0
    @editable
    RangeProjectileLifetime : float = 1.0

    @editable
    ProjectileLifetime : float = 1.0

    @editable
    ChargeAbility : logic = false
    ChargeAbilityOdds : float = 2.5

    @editable
    TornadoAbility : logic = false
    TornadoAbilityOdds : float = 3.0

    @editable
    TornadoNiag : creative_prop_asset = DefaultCreativePropAsset
    @editable
    TornadoStartProps : []creative_prop = array{}
    @editable
    TornadoEndProps : []creative_prop = array{}
    @editable
    TornadoParentProp : creative_prop = creative_prop{}

    var TornadoArray : []creative_prop = array{}

    var TornadoTravelTime : float = 5.0
    TornadoDamage : float = 75.0
    TornadoDamageZoneSize : float = 128.0
    var TornadoDamageLogic : logic = false

    @editable
    SpinningTornadoAbility : logic = false
    SpinningTornadoAbilityOdds : float = 3.0
    @editable
    SpinningTornadoNiag : creative_prop_asset = DefaultCreativePropAsset

    SpinningTornadoDamage : float = 33.0
    SpinningTornadoDamageZoneSize : float = 128.0
    var SpinningTornadoDamageLogic : logic = false
    SpinningTornadospeed : float = 0.015
    SpinningTornadosampletime : float = 0.15
    SpinningTornadoCircleSize : float = 1500.0


    @editable
    RainAbility : logic = false
    RainAbilityOdds : float = 2.5

    @editable
    RainingAoeNiag : creative_prop_asset = DefaultCreativePropAsset
    @editable
    RainingProjectile : creative_prop_asset = DefaultCreativePropAsset
    @editable
    RainingExplosion : creative_prop_asset = DefaultCreativePropAsset
    @editable
    RainingTargetProp : creative_prop = creative_prop{}
    @editable
    RainingDeletePropLocation : creative_prop = creative_prop{}
    
    RainingAoeNiagSize : float = 256.0
    RainingAoeNiagLifeTime : float = 6.0
    RainingRainProjectileSpawnDelay : float = 3.5
    RainingAoeRange  : float =  2000.0
    RainingAoeAttackNumber : int = 8
    RainingAoeSpawnHeight : float = 2000.0
    RainingRainFallSpeed : float = 1.0
    RainingDamageLifetime : float = 8.0
    RainingDamage : float = 50.0

    # used to tell when to stop damage loop
    var StopCharge : logic = false
    @editable
    ShowChargeTrigger : trigger_device = trigger_device{}
    @editable
    HideChargeTrigger : trigger_device = trigger_device{}

    @editable
    RangedAndMelee : logic = false
    @editable
    RangeAndMeleeOdds : float = 300.0

    @editable
    Slam : logic = false
    @editable
    SlamOdds : float = 200.0
    @editable
    SlamAttackAnimLength : float = 2.0
    @editable
    HeightOffsetForSlam : float = -20.0
    @editable
    SlamAudio : audio_player_device = audio_player_device{}


    @editable
    SlamRocksNiag : creative_prop_asset = DefaultCreativePropAsset
    @editable
    SlamAoeNiag : creative_prop_asset = DefaultCreativePropAsset
    @editable
    RockBreakAudio : audio_player_device = audio_player_device{}
    SlamDamage : float = 50.0
    SlamAoeDamageSize : float = 750.0
    SlamAOETime : float = 2.25
    

    @editable
    IsPosionous : logic = false
    @editable
    PosionDamage : float = 5.0

    @editable
    PhaseTrigger : trigger_device =  trigger_device{}
    var SecondPhase : logic = false

    var AgentsInArea : [agent]?logic = map{}

    @editable
    ResetHpTrigger : trigger_device = trigger_device{}

    @editable
    DeaggroArea : volume_device = volume_device{}


    # This function runs when the NPC is spawned in the world and ready to follow a behavior.
    OnBegin<override>()<suspends>:void =

        AggroVolume.AgentEntersEvent.Subscribe(HandleAgentEnters)

        DeaggroArea.AgentEntersEvent.Subscribe(EntersArea)
        DeaggroArea.AgentExitsEvent.Subscribe(ExitsArea)

        PhaseTrigger.TriggeredEvent.Subscribe(EnableSecondPhase)


        if:
            # Get the Agent (the NPC).
            Agent := GetAgent[]
            # Gets the Fortnite Character interface, which gets you access to its gameplay data 
            # including its AI module for navigation and focus.
            Character := Agent.GetFortCharacter[]
            # Get the Navigatable Interface, this allows you to tell it to move.
            Navigatable := Character.GetNavigatable[]
            # Get the Focus Interface, this allows you to tell it to look at something or somewhere.
            Focus := Character.GetFocusInterface[]
        then:

            # Used to stop idle state
            set Aggroed = false
            set Ended = false

            # Save the position of your character which we'll use to generate more points to move to from.
            set NPCSpawnPoint = Character.GetTransform().Translation
            Character.DamagedEvent().Subscribe(OnDamaged)

            if(ChargeAbility = true):
                Navigatable.SetMovementSpeedMultiplier(0.25)
        
            IdleBehavior( Focus,Character )

    EntersArea(EnteredAgent:agent):void=
        if(set AgentsInArea[EnteredAgent] = option{true}){}
        #Print("Enters")

    ExitsArea(Agent:agent):void= 
        if(set AgentsInArea[Agent] = false){}
        #Print("Exit")

    EnableSecondPhase(Agent:?agent):void=
        set SecondPhase = true

    GetAnimationSequences(npc_type: NPCData.NPCTypes.npc_types, state:NPCData.NPCStates.npc_states):animation_sequence = 

        #Defualt Animation Sequence
        var MyAnim : animation_sequence = Props.LavaMite.ReImport1.idle

        if(npc_type = NPCData.NPCTypes.npc_types.LavaMite):
            if(MyAnimVar := NPCData.NPCStates.lava_might_npc_animations[state]):
                set MyAnim = MyAnimVar
        if(npc_type = NPCData.NPCTypes.npc_types.DarkImp):
            if(MyAnimVar := NPCData.NPCStates.dark_imp_npc_animations[state]):
                set MyAnim = MyAnimVar
        if(npc_type = NPCData.NPCTypes.npc_types.Murloc):
            if(MyAnimVar := NPCData.NPCStates.murloc_npc_animations[state]):
                set MyAnim = MyAnimVar
        if(npc_type = NPCData.NPCTypes.npc_types.Boar):
            if(MyAnimVar := NPCData.NPCStates.boar_npc_animations[state]):
                set MyAnim = MyAnimVar
        if(npc_type = NPCData.NPCTypes.npc_types.BossBoar):
            if(MyAnimVar := NPCData.NPCStates.boss_boar_npc_animations[state]):
                set MyAnim = MyAnimVar
        if(npc_type = NPCData.NPCTypes.npc_types.Buffalo):
            if(MyAnimVar := NPCData.NPCStates.buffalo_npc_animations[state]):
                set MyAnim = MyAnimVar
        if(npc_type = NPCData.NPCTypes.npc_types.KoboldWarrior):
            if(MyAnimVar := NPCData.NPCStates.kobold_warrior_npc_animations[state]):
                set MyAnim = MyAnimVar
        if(npc_type = NPCData.NPCTypes.npc_types.KoboldArcher):
            if(MyAnimVar := NPCData.NPCStates.kobold_archer_npc_animations[state]):
                set MyAnim = MyAnimVar
        if(npc_type = NPCData.NPCTypes.npc_types.Hyena):
            if(MyAnimVar := NPCData.NPCStates.hyena_npc_animations[state]):
                set MyAnim = MyAnimVar
        if(npc_type = NPCData.NPCTypes.npc_types.DesertBoss):
            if(MyAnimVar := NPCData.NPCStates.desert_boss_npc_animations[state]):
                set MyAnim = MyAnimVar
        if(npc_type = NPCData.NPCTypes.npc_types.Scorpion):
            if(MyAnimVar := NPCData.NPCStates.scorpion_npc_animations[state]):
                set MyAnim = MyAnimVar
        if(npc_type = NPCData.NPCTypes.npc_types.PolarBear):
            if(MyAnimVar := NPCData.NPCStates.polar_bear_npc_animations[state]):
                set MyAnim = MyAnimVar
        if(npc_type = NPCData.NPCTypes.npc_types.GiantBeast):
            if(MyAnimVar := NPCData.NPCStates.giant_beast_npc_animations[state]):
                set MyAnim = MyAnimVar
        if(npc_type = NPCData.NPCTypes.npc_types.TimberWolf):
            if(MyAnimVar := NPCData.NPCStates.timber_wolf_npc_animations[state]):
                set MyAnim = MyAnimVar
        if(npc_type = NPCData.NPCTypes.npc_types.Ram):
            if(MyAnimVar := NPCData.NPCStates.ram_npc_animations[state]):
                set MyAnim = MyAnimVar
        if(npc_type = NPCData.NPCTypes.npc_types.SnowLeopard):
            if(MyAnimVar := NPCData.NPCStates.snow_leopard_npc_animations[state]):
                set MyAnim = MyAnimVar
        if(npc_type = NPCData.NPCTypes.npc_types.LavaDragon):
            if(MyAnimVar := NPCData.NPCStates.lava_dragon_npc_animations[state]):
                set MyAnim = MyAnimVar
        if(npc_type = NPCData.NPCTypes.npc_types.Terravore):
            if(MyAnimVar := NPCData.NPCStates.terravore_npc_animations[state]):
                set MyAnim = MyAnimVar
        if(npc_type = NPCData.NPCTypes.npc_types.Yeti):
            if(MyAnimVar := NPCData.NPCStates.yeti_npc_animations[state]):
                set MyAnim = MyAnimVar
        if(npc_type = NPCData.NPCTypes.npc_types.LavaJaw):
            if(MyAnimVar := NPCData.NPCStates.lava_jaw_npc_animations[state]):
                set MyAnim = MyAnimVar
        if(npc_type = NPCData.NPCTypes.npc_types.DesertWyvern):
            if(MyAnimVar := NPCData.NPCStates.desert_wyvern_npc_animations[state]):
                set MyAnim = MyAnimVar
        if(npc_type = NPCData.NPCTypes.npc_types.SerenitySentinel):
            if(MyAnimVar := NPCData.NPCStates.serenity_sentinel_npc_animations[state]):
                set MyAnim = MyAnimVar
        if(npc_type = NPCData.NPCTypes.npc_types.ArcticColosus):
            if(MyAnimVar := NPCData.NPCStates.arctic_colosus_npc_animations[state]):
                set MyAnim = MyAnimVar
        if(npc_type = NPCData.NPCTypes.npc_types.HellHound):
            if(MyAnimVar := NPCData.NPCStates.hell_hound_npc_animations[state]):
                set MyAnim = MyAnimVar
        if(npc_type = NPCData.NPCTypes.npc_types.MagmaBoar):
            if(MyAnimVar := NPCData.NPCStates.magma_boar_npc_animations[state]):
                set MyAnim = MyAnimVar
        
        return MyAnim

    HandleAggroTrigger(Agent:?agent):void=
        if:
            RealAgent: agent = Agent?
            Player := player[RealAgent]
            Player.IsActive[]
            Found := HitAudio.GetPlayspace().GetPlayers().Find[Player]
        then:
            HandleAgentEnters(RealAgent)

    HandleAgentEnters(EnteredAgent:agent):void=

        if:
            InstigatingCharacter : fort_character = EnteredAgent.GetFortCharacter[]
            Players := HitAudio.GetPlayspace().GetPlayers()
        then:
           
            for(i:=0..Players.Length-1):
                if(EnteredAgent = Players[i]):
                    if:
                        Ended = false
                        InstigatingCharacter.IsActive[]
                    then:
                        Print("Aggroing Agent Entered")
                        if(set AgentsInArea[EnteredAgent] = option{true}){}
                        Aggro(InstigatingCharacter)


    OnDamaged(DamageResult:damage_result):void=

        DamageSounce := DamageResult.Instigator
        if:
            Stationary = false
            InstigatingCharacter := fort_character[DamageSounce?]
            InstigatingCharacter.IsActive[]
            IsAgent := InstigatingCharacter.GetAgent[]
            Players := HitAudio.GetPlayspace().GetPlayers()
        then:
            

            for(i:=0..Players.Length-1):
                if(IsAgent = Players[i]):
                    if:
                        Ended = false
                        InstigatingCharacter.IsActive[]
                    then:
                        if(set AgentsInArea[IsAgent] = option{true}){} # this may cause issues if gun has a delayted hit
                        Aggro(InstigatingCharacter)


            
    Aggro(AggroedCharacter:fort_character):void =
            
        if:
            Agent := GetAgent[]
            Character := Agent.GetFortCharacter[]
            Navigatable := Character.GetNavigatable[]
            InstigatingCharacter := AggroedCharacter
            PlayAnimController := Character.GetPlayAnimationController[]
            Character.IsActive[]
            Aggroed = false
        then:

            if(ChargeAbility = true):
                Navigatable.SetMovementSpeedMultiplier(0.25)
            
            # Cow plays animation when it gets hit
            #if(Character.IsActive[]):
            #    AnimationResult := PlayAnimController.Play(sheepintry.Cow_GetHit, ?PlayRate := PlayRate, ?BlendInTime := BlendInTime, ?BlendOutTime := BlendInTime, ?StartPositionSeconds := StartPositionSeconds)
                    
            if (Ended = true):
                set CharacterAggro = false
                    #AnimationResult.Stop()
            else:
                if(Character.IsActive[]):
                    set Aggroed = true
                    set CharacterAggro = option{InstigatingCharacter}
                    set GotHit = true
                    if(RangedNPC = false):
                        Print("Spawned Melee Attack")
                        spawn:
                            MeleeAttackPlayer( Character, true)
                    else:
                        spawn:
                            RangeAttackPlayer( Character, true)


    IdleBehavior(Focus : focus_interface, Character : fort_character)<suspends>:void=


        if(Ended = false and Stationary = false):

            loop:
                if:
                    Character.IsActive[]
                    Navigatable := Character.GetNavigatable[]
                then:

                    if(MyAgent:=Character.GetAgent[]):
                        ResetHpTrigger.Trigger(MyAgent)
                        set SecondPhase = false
                    #CurrentPoint := Character.GetTransform().Translation
                    # Create a random offset from the spawn point to walk toward.
                    GoToPoint := NPCSpawnPoint + vector3{X := GetRandomFloat(-DistanceFromSpawnPtToMove,DistanceFromSpawnPtToMove),
                                                        Y := GetRandomFloat(-DistanceFromSpawnPtToMove,DistanceFromSpawnPtToMove),
                                                        Z := 0.0 }
                    
                    # Create a navigation target from these two positions that the navigation interface can use.
                    NavTargetStart := MakeNavigationTarget(GoToPoint)
                
                    #Focus.MaintainFocus(GoToPoint)
                    # Tell the navigation interface to walk to the location and store a navigation result for error checking.
                    
                
                    NavResultGoTo := race:

                        Navigatable.NavigateTo(NavTargetStart, ?MovementType:=movement_types.Walking)

                        block:
                            Focus.MaintainFocus(GoToPoint)
                        

                    if (NavResultGoTo = navigation_result.Unreachable):
                        Sleep(1.0)
                        if(Character.IsActive[]):
                            Print("Unreachable target")
                        else:
                            break
                    else:

                        var WaitInterval : int = GetRandomInt(-9,9)
                        set WaitInterval = WaitInterval + MoveToWaitIntervals

                        for(i:=0..WaitInterval-1):
                            Navigatable.Wait(?Duration := MoveToWaitDuration)
                            if (GotHit = true or Ended = true):
                                break
                    
                        if (GotHit = true or Ended = true):
                            set GotHit = false
                            break


    ChargeDamage()<suspends>:void=

        loop:
            Sleep(0.25)

            if (StopCharge = true):
                break
            else:
                Players := HitAudio.GetPlayspace().GetPlayers()
                for(i:=0..Players.Length-1):
                    if:
                        Player := Players[i]
                        Char := Player.GetFortCharacter[]
                        PlayerAgent := Char.GetAgent[]
                        MySelf := GetAgent[]
                        MyCharacter := MySelf.GetFortCharacter[]
                        MyCharacter.IsActive[]
                        Char.IsActive[]
                    then:
                        PlayerLocation : vector3 = Char.GetTransform().Translation
                        MyLocation : vector3 = MyCharacter.GetTransform().Translation
                        
                        var Dis2Player : float = Distance(MyLocation,PlayerLocation)
                        if(Dis2Player <= 250.0):
                            HitAudio.Play(PlayerAgent)
                            DamageAfterArmor := AttackDamage * DamageCalculation(Char)
                            Char.Damage(DamageAfterArmor)


    ChargePlayer( NPCCharacter : fort_character )<suspends>:void=
        
                Print("Rolled The Charge Ability")

                if:
                    NPCCharacter.IsActive[]
                    PlayerCharacter := CharacterAggro?
                    PlayerCharacter.IsActive[]
                    NpcAgent := NPCCharacter.GetAgent[]
                    PlayAnimController := NPCCharacter.GetPlayAnimationController[]
                    Focus := NPCCharacter.GetFocusInterface[]
                    Navigatable := NPCCharacter.GetNavigatable[]
                then:
                    set StopCharge = false
                    Navigatable.StopNavigation()
                    ChargeAudio.Play()
                    
                    # this makes it location off so its easier to dodge
                    Sleep(1.0)
                    var PlayerLocation : vector3 = PlayerCharacter.GetTransform().Translation
                    CurrentLocation : vector3 = NPCCharacter.GetTransform().Translation
                    Sleep(0.5)
                    
                    spawn:
                        ChargeDamage()

                    ShowChargeTrigger.Trigger(NpcAgent)
                    Navigatable.SetMovementSpeedMultiplier(1.5)

                    

                    DifVector := PlayerLocation - CurrentLocation
                    NormDifVector := NormalizeVector(DifVector)

                    AttackPoint := PlayerLocation + vector3{
                        X := (NormDifVector.X * 750.0),
                        Y := (NormDifVector.Y * 750.0),
                        Z := 0.0 }

                    PlayerNav := MakeNavigationTarget(AttackPoint)

                    spawn:
                        Focus.MaintainFocus(AttackPoint)

                    Navigatable.NavigateTo(PlayerNav, ?MovementType:=movement_types.Running)

                    Sleep(0.2)
                    #Sleep(3.0)
                    Navigatable.SetMovementSpeedMultiplier(0.25)
                    HideChargeTrigger.Trigger(NpcAgent)
                    set StopCharge = true
                    Sleep(0.5)

                    MeleeAttackPlayer(NPCCharacter,true)



    MeleeAttackPlayer(  NPCCharacter : fort_character ,WaitForHitToPlay : logic)<suspends>:void=

        if(Ended = false and WaitForHitToPlay = true and NPCCharacter.IsActive[]):
            Sleep(0.6)

        loop:

            if(RainAbility = true):
                var Odds : float = 100.0 
                if(SecondPhase = true):
                    set Odds = 75.0
                RandomFloat := GetRandomFloat(0.0,Odds) 
                if(RandomFloat <= RainAbilityOdds):
                    spawn:
                        RainAttack( NPCCharacter,true)
                    break

            if(ChargeAbility = true):
                var Odds : float = 100.0 
                if(SecondPhase = true):
                    set Odds = 75.0
                RandomFloat := GetRandomFloat(0.0,Odds) 
                if(RandomFloat <= ChargeAbilityOdds):
                    spawn:
                        ChargePlayer(NPCCharacter)
                    break

            if(SpinningTornadoAbility = true):
                var Odds : float = 100.0 
                if(SecondPhase = true):
                    set Odds = 75.0
                RandomFloat := GetRandomFloat(0.0,Odds) 
                if(RandomFloat <= SpinningTornadoAbilityOdds):
                    spawn:
                        SpinningTornadoAttack( NPCCharacter,true)
                    break
            
            if(TornadoAbility = true):
                var Odds : float = 100.0 
                if(SecondPhase = true):
                    set Odds = 75.0
                RandomFloat := GetRandomFloat(0.0,Odds) 
                if(RandomFloat <= TornadoAbilityOdds):
                    spawn:
                        TornadoAttack( NPCCharacter,true)
                    break
            
            if(RangedAndMelee = true):
                var Odds : float = RangeAndMeleeOdds
                if(SecondPhase = true):
                    set Odds = RangeAndMeleeOdds - 50.0
                RandomFloat := GetRandomFloat(0.0,Odds)
                if(RandomFloat <= 1.0):
                    spawn:
                        RangeAttackPlayer( NPCCharacter, true)
                    break

            if(Slam = true):
                var Odds : float = SlamOdds
                if(SecondPhase = true):
                    set Odds = SlamOdds - 50.0
                RandomFloat := GetRandomFloat(0.0,Odds) # 100.0 is charge ability for boss boar 300 seems okay
                if(RandomFloat <= 1.0):
                    if(SecondPhase = true):
                        spawn:
                            SlamAttack( NPCCharacter, true,3)
                    else:
                        spawn:
                            SlamAttack( NPCCharacter, true,2)

                    break


            if:
                NPCCharacter.IsActive[]
                Navigatable := NPCCharacter.GetNavigatable[]
                Ended = false
                PlayerCharacter := CharacterAggro?
                PlayerCharacter.IsActive[]
                PlayAnimController := NPCCharacter.GetPlayAnimationController[]
                Focus := NPCCharacter.GetFocusInterface[]
                Agent:=PlayerCharacter.GetAgent[]
                AgentsInArea[Agent]? = true
            then:

                # Navigaste Toward Player #

                var PlayerLocation : vector3 = PlayerCharacter.GetTransform().Translation
                CurrentLocation : vector3 = NPCCharacter.GetTransform().Translation
                # Caluculating Distance From Creatures Attack Point To The Aggroed Player
                deltaX := CurrentLocation.X - PlayerLocation.X 
                deltaY := CurrentLocation.Y - PlayerLocation.Y 
                var Rad : float = ArcTan(deltaY, deltaX)
    
                AttackPoint := CurrentLocation - vector3{
                    X := ((AttackRange ) * Cos(Rad)),
                    Y := ((AttackRange ) * Sin(Rad)),
                    Z := 0.0 }

                var Dis2Player : float = Distance(AttackPoint,PlayerLocation) # Switched this from attackpint and player location to current locationa dn aplayer location

                # Gets rotation of npccharacter to see if player is behind it; if it is then it will move instead of attack
                # Apply 90 to correct from the characters given rotaiton
                #myNewRotation := NPCCharacter.GetTransform().Rotation.ApplyYaw(180.0)
                myRotation := NPCCharacter.GetTransform().Rotation.GetYawPitchRollDegrees()
                
                if(var myYaw : float = myRotation[0]):
                    #Print("myYaw {myYaw}")
                    #convert yawdegrees into radians
                    set myYaw = myYaw * (PiFloat / 180.0)

                    # Calculate the forward vector components
                    # Assuming the forward vector starts pointing north (along the Z axis), 
                    forward_vector_x := Cos(myYaw)
                    forward_vector_y := Sin(myYaw)

                    # Converting it back into 3d space vector
                    forwardvector3d := vector3 {
                        X:=forward_vector_x,
                        Y:=forward_vector_y,
                        Z:=0.0
                    }
                    #normalizing forward vector
                    magnitude2 : float = Sqrt(Pow(forwardvector3d.X, 2.0) + Pow(forwardvector3d.Y,2.0) + Pow(forwardvector3d.Z,2.0))
                    NorForwardVector := vector3 {
                        X:= forwardvector3d.X / magnitude2,
                        Y:= forwardvector3d.Y / magnitude2,
                        Z:= forwardvector3d.Z / magnitude2
                    }
                    # Difference vector point from currentlocation to playerlocaiton
                    DifferenceVector : vector3 = PlayerLocation - CurrentLocation
                    # normalizing difference vector
                    magnitude1 : float = Sqrt(Pow(DifferenceVector.X, 2.0) + Pow(DifferenceVector.Y,2.0) + Pow(DifferenceVector.Z,2.0))
                    NormDifVector := vector3 {
                        X:= DifferenceVector.X / magnitude1,
                        Y:= DifferenceVector.Y / magnitude1,
                        Z:= DifferenceVector.Z / magnitude1
                    }
                    # Dot product from difference vector and facing vector will tell us if it is facing it
                    dot_product := DotProduct(NormDifVector,NorForwardVector)
                    if:
                        Dis2Player<=AttackRange
                        dot_product >= 0.0
                    then:
                        if(NPCCharacter.IsActive[]):
                            AnimationResult := PlayAnimController.Play(GetAnimationSequences(npctype,NPCData.NPCStates.npc_states.Attack), ?BlendInTime := BlendInTime, ?BlendOutTime := BlendInTime, ?StartPositionSeconds := StartPositionSeconds)
                            AttackAudio.Play()
                            spawn:
                                DelayedHitTest(PlayerCharacter, AttackPoint)
                            Sleep(AttackAnimLength)
                    else:

                        #if player is behind then change the walk loc a little
                        Sleep(0.2) # moved this 
                        
                        # Walks toward players location
                        var WalkToLoc : vector3 = PlayerLocation + vector3{
                            X := ((AttackRange - 150.0) * Cos(Rad)),
                            Y := ((AttackRange - 150.0) * Sin(Rad)),
                            Z := 0.0 }

                        if(dot_product < 0.0):
                            RandFloat := GetRandomFloat(-50.0,50.0)
                            set WalkToLoc = PlayerLocation + vector3{
                            X := (((AttackRange - 150.0) * Cos(Rad)) + RandFloat ),
                            Y := (((AttackRange - 150.0) * Sin(Rad)) + RandFloat ),
                            Z := 0.0 }

                        NavTargetStart := MakeNavigationTarget(WalkToLoc)
 
                        Navigatable.StopNavigation()
                        if(Ended = false):
                            spawn:
                                Focus.MaintainFocus(WalkToLoc)
                            spawn:
                                Navigatable.NavigateTo(NavTargetStart, ?MovementType:=movement_types.Running)
                             # The NPC Moves to the point then .2 seconds later checks if it is in the range which is causing it to not attack changed to .1 moved higher up was here
                        
            else:
                #Returns back to idle after it kills someone
                if:
                    PlayerCharacter := CharacterAggro?
                    PlayerCharacter.IsActive[]
                    Ended = false
                    Agent:=PlayerCharacter.GetAgent[]
                    AgentsInArea[Agent]? = true
                then:
                else:
                    set Aggroed = false
                    set GotHit = false
                    if(Ended = false):
                        Sleep(1.5)
                        if:
                            NPCCharacter.IsActive[]
                            Focus := NPCCharacter.GetFocusInterface[]
                        then:
                            spawn:
                                IdleBehavior(Focus,NPCCharacter)

                break

    SlamAttack(NPCCharacter : fort_character ,WaitForHitToPlay : logic, SlamNumber : int)<suspends>:void=

        if:
            NPCCharacter.IsActive[]
            Ended = false
            PlayerCharacter := CharacterAggro?
            PlayerCharacter.IsActive[]
            PlayerAgent := PlayerCharacter.GetAgent[]
            PlayAnimController := NPCCharacter.GetPlayAnimationController[]
            Focus := NPCCharacter.GetFocusInterface[]
        then:

            SlamAudio.Play()
            AnimationResult := PlayAnimController.Play(GetAnimationSequences(npctype,NPCData.NPCStates.npc_states.Charge), ?BlendInTime := BlendInTime, ?BlendOutTime := BlendInTime, ?StartPositionSeconds := StartPositionSeconds)
            spawn:
                SlamAtAgent(PlayerAgent, HeightOffsetForSlam, SlamNumber)
            Sleep(SlamAttackAnimLength)
            MeleeAttackPlayer( NPCCharacter, true)

    SlamAtAgent(Agent:agent, HeightOffset:float, Slams : int)<suspends>:void=
        if:
            Fort := Agent.GetFortCharacter[]
        then:
            for(i:=0..Slams-1):
                var RandomY : float = 500.0
                RandomYNeg : int = GetRandomInt(0,1)
                var RandomX : float = 500.0
                RandomXNeg : int = GetRandomInt(0,1)
                if(RandomYNeg = 1):
                    set RandomY = -1 * RandomY
                if(RandomXNeg = 1):
                    set RandomX = -1 * RandomX

                var SpawnLocation : vector3 = Fort.GetTransform().Translation 
                set SpawnLocation.Y = SpawnLocation.Y + (RandomY * i)
                set SpawnLocation.X = SpawnLocation.X + (RandomX * i)
                var SpawnRotation : rotation = Fort.GetTransform().Rotation

                set SpawnLocation.Z = 110.0

                SpawnTransform : transform = transform{Translation:=SpawnLocation,Rotation:=SpawnRotation,Scale:=vector3{X:=2.0,Y:=2.0,Z:=2.0}}

                SpawnedPropResult := SpawnProp(SlamAoeNiag,SpawnTransform)
                if(SpawnedProp := SpawnedPropResult(0)?):
                    spawn:
                        DeleteEffect(SlamAOETime,SpawnedProp)
                spawn:
                    AfterAOE(SpawnLocation, SpawnRotation)
                Sleep(0.5)

    AfterAOE(SpawnLocation : vector3,SpawnRotation : rotation )<suspends>:void=
        
        Sleep(SlamAOETime-0.6)
        if(RockBreakAudio.TeleportTo[SpawnLocation, SpawnRotation]){RockBreakAudio.Play()}
        SpawnedPropResult := SpawnProp(SlamRocksNiag,SpawnLocation,SpawnRotation)
        if(SpawnedProp := SpawnedPropResult(0)?):
            Sleep(0.2)
            spawn:
                DeleteEffect(5.0,SpawnedProp)
            Sleep(0.2)
            SlamDamagePlayers(SpawnLocation)

    SlamDamagePlayers(SpawnLocation : vector3):void=

        Players := HitAudio.GetPlayspace().GetPlayers()
        for(i:=0..Players.Length-1):
            if:
                Player := Players[i]
                Char := Player.GetFortCharacter[]
                PlayerAgent := Char.GetAgent[]
                Char.IsActive[]
            then:
                PlayerLocation : vector3 = Char.GetTransform().Translation
                
                var Dis2Player : float = Distance(SpawnLocation,PlayerLocation)
                if(Dis2Player <= SlamAoeDamageSize):
                    HitAudio.Play(PlayerAgent)
                    Char.Damage(SlamDamage)
        
    DeleteEffect(SleepTime : float, Prop : creative_prop)<suspends>:void=

        Sleep(SleepTime)
        Prop.Dispose()

    TornadoAttack( NPCCharacter : fort_character ,WaitForHitToPlay : logic)<suspends>:void=

        if(Ended = false and WaitForHitToPlay = true and NPCCharacter.IsActive[]):
            Sleep(0.6)

        if:
            NPCCharacter.IsActive[]
            Ended = false
            PlayerCharacter := CharacterAggro?
            PlayerCharacter.IsActive[]
            PlayerAgent := PlayerCharacter.GetAgent[]
            PlayAnimController := NPCCharacter.GetPlayAnimationController[]
            Focus := NPCCharacter.GetFocusInterface[]
        then:
            AnimationResult := PlayAnimController.Play(GetAnimationSequences(npctype,NPCData.NPCStates.npc_states.Tornado),  ?BlendInTime := BlendInTime, ?BlendOutTime := BlendInTime, ?StartPositionSeconds := StartPositionSeconds)
            TornadoCastAudio.Play()
            SummonTornadoDelay := 0.75
            Sleep(SummonTornadoDelay)
            SummonTornado(NPCCharacter)
            Sleep(0.5)
            MeleeAttackPlayer(NPCCharacter,true)

    SummonTornado(NPCCharacter:fort_character):void=
        # only one set of tornadoes can be made per device at a time
        set TornadoArray = array{}

        if(NPCCharacter.IsActive[]):

            var CurrentRotation : rotation = TornadoParentProp.GetTransform().Rotation
            var NpcLocation : vector3 = NPCCharacter.GetTransform().Translation
            set NpcLocation.Z = 105.0
            if(TornadoParentProp.TeleportTo[NpcLocation,CurrentRotation]){}

            # Rotate Targets that are children of parent prop
            CurrentLocation : vector3 = TornadoParentProp.GetTransform().Translation
            RandomRotation : float = GetRandomFloat(0.0,180.0)
            NewRotation := CurrentRotation.ApplyYaw(RandomRotation)
            if(TornadoParentProp.TeleportTo[CurrentLocation,NewRotation]){}

            for(i:=0..TornadoStartProps.Length-1):
                
                var StartProp : creative_prop = creative_prop{}
                var EndProp : creative_prop = creative_prop{}

                RandomStart : int = GetRandomInt(0,1)

                if(RandomStart = 0):
                    if(V := TornadoEndProps[i],set StartProp = V){}
                    if(V := TornadoStartProps[i],set EndProp = V){}
                else:
                    if(V := TornadoStartProps[i],set StartProp = V){}
                    if(V := TornadoEndProps[i],set EndProp = V){}

                StartLoc := StartProp.GetTransform().Translation
                StartRot := StartProp.GetTransform().Rotation
                EndLoc := EndProp.GetTransform().Translation
                EndRot := EndProp.GetTransform().Rotation

                SpawnedPropResult := SpawnProp(TornadoNiag,StartLoc,StartRot)
                if(SpawnedProp := SpawnedPropResult(0)?):
                    spawn:
                        SpawnedProp.MoveTo(EndLoc,EndRot,TornadoTravelTime)
                    spawn:
                        DeleteEffect(TornadoTravelTime,SpawnedProp)

                    set TornadoArray = TornadoArray + array{SpawnedProp}
                else:
                    

            set TornadoDamageLogic = true
            spawn:
                TornadoDamageLoop()

    TornadoDamageLoop()<suspends>:void=

        loop:

            if (TornadoDamageLogic = true):

                # Damage Freq
                Sleep(0.1)
            
                for(P:=0..TornadoArray.Length-1):
                    if:
                        TornadoProp := TornadoArray[P]
                        TornadoProp.IsValid[]
                    then:
                        MyLocation := TornadoProp.GetTransform().Translation

                        Players := HitAudio.GetPlayspace().GetPlayers()
                        for(i:=0..Players.Length-1):
                            if:
                                Player := Players[i]
                                Char := Player.GetFortCharacter[]
                                PlayerAgent := Char.GetAgent[]
                                Char.IsActive[]
                                
                            then:
                                PlayerLocation : vector3 = Char.GetTransform().Translation
                                
                                var Dis2Player : float = Distance(MyLocation,PlayerLocation)
                                if(Dis2Player <= TornadoDamageZoneSize):
                                    HitAudio.Play(PlayerAgent)
                                    Char.Damage(TornadoDamage)

            else:
                break


    SpinningTornadoAttack( NPCCharacter : fort_character ,WaitForHitToPlay : logic)<suspends>:void=

        if(Ended = false and WaitForHitToPlay = true and NPCCharacter.IsActive[]):
            Sleep(0.6)

        if:
            NPCCharacter.IsActive[]
            Ended = false
            PlayerCharacter := CharacterAggro?
            PlayerCharacter.IsActive[]
            PlayerAgent := PlayerCharacter.GetAgent[]
            PlayAnimController := NPCCharacter.GetPlayAnimationController[]
            Focus := NPCCharacter.GetFocusInterface[]
        then:
            AnimationResult := PlayAnimController.Play(GetAnimationSequences(npctype,NPCData.NPCStates.npc_states.Tornado),  ?BlendInTime := BlendInTime, ?BlendOutTime := BlendInTime, ?StartPositionSeconds := StartPositionSeconds)
            TornadoCastAudio.Play()
            SummonTornadoDelay := 0.75
            Sleep(SummonTornadoDelay)
            spawn:
                MoveAlongCurve(NPCCharacter,3)
            Sleep(0.5)
            MeleeAttackPlayer(NPCCharacter,true)

    MoveAlongCurve(NPCCharacter:fort_character, TornadoAmount : int )<suspends> : void =

        for(i:=0..TornadoAmount-1):
            spawn:
                SpinATornado(NPCCharacter)
            Sleep(2.5)

    SpinATornado(NPCCharacter:fort_character)<suspends> : void =

        var TargetPosition : vector3 = NPCCharacter.GetTransform().Translation
        set TargetPosition.Z = 108.0
        CurrentRotation := NPCCharacter.GetTransform().Rotation

        var SpawnLoc : vector3 = TargetPosition
        set SpawnLoc.Z = 0.0

        SpawnedPropResult := SpawnProp(SpinningTornadoNiag, vector3{X:=0.0,Y:=0.0,Z:=0.0} ,CurrentRotation)
            
        if (TornadoProp := SpawnedPropResult(0)?):

            spawn:
                SpinningDamageLoop(TornadoProp)

            spawn:
                DeleteEffect(15.0,TornadoProp)

            var t : float = 0.05  # interpolation factor from 0 to 1
            var radius : float = 1500.0  # radius of the circle
            var center : vector3 = TargetPosition  # center of the curve

            var angle : float = 0.0 
            var angularSpeed : float = 2.0 * 3.1415 / 7.5 
                
            loop:

                if(TornadoProp.IsValid[]):
                    set angle += angularSpeed * 0.125

                        if (angle > 2.0 * 3.1415): 
                            set angle -= 2.0 * 3.1415

                        #Sleep(0.01)  # Delay for smooth movement
                        #set t += 0.05  # Increment t to move along the curve
                        #if (t > 1.0): 
                        #    set t = 0.0  # Loop t back to 0 for continuous motion

                        # Calculate new position using circular parametric equations
                        var newX : float = center.X + radius * Cos(angle)  # Cosine for X
                        var newY : float = center.Y + radius * Sin(angle)  # Sine for Y
                        var newZ : float = center.Z  # No change in Z for a 2D circular arc

                        var newPosition : vector3 = vector3{X := newX, Y := newY, Z := newZ}
                        
                        MoveResult := TornadoProp.MoveTo(newPosition, TornadoProp.GetTransform().Rotation, 0.125)
                        if(MoveResult = move_to_result.DestinationReached){}
                            
                    else:
                        break

    SpinningDamageLoop(TornadoProp:creative_prop)<suspends>:void=

        loop:
            # Damage Freq
            Sleep(0.1)
            if:
                TornadoProp.IsValid[]
            then:
                MyLocation := TornadoProp.GetTransform().Translation

                Players := HitAudio.GetPlayspace().GetPlayers()
                for(i:=0..Players.Length-1):
                    if:
                        Player := Players[i]
                        Char := Player.GetFortCharacter[]
                        PlayerAgent := Char.GetAgent[]
                        Char.IsActive[]
                                
                    then:
                        PlayerLocation : vector3 = Char.GetTransform().Translation
                                
                        var Dis2Player : float = Distance(MyLocation,PlayerLocation)
                        if(Dis2Player <= SpinningTornadoDamageZoneSize):
                            HitAudio.Play(PlayerAgent)
                            Char.Damage(SpinningTornadoDamage)

            else:
                break



    RainAttack(NPCCharacter : fort_character ,WaitForHitToPlay : logic)<suspends>:void=

        if(Ended = false and WaitForHitToPlay = true and NPCCharacter.IsActive[]):
            Sleep(0.6)

        if:
            NPCCharacter.IsActive[]
            Ended = false
            PlayerCharacter := CharacterAggro?
            PlayerCharacter.IsActive[]
            PlayerAgent := PlayerCharacter.GetAgent[]
            PlayAnimController := NPCCharacter.GetPlayAnimationController[]
            Focus := NPCCharacter.GetFocusInterface[]
        then:
            AnimationResult := PlayAnimController.Play(GetAnimationSequences(npctype,NPCData.NPCStates.npc_states.Tornado),  ?BlendInTime := BlendInTime, ?BlendOutTime := BlendInTime, ?StartPositionSeconds := StartPositionSeconds)
            RainCastAudio.Play()
            SummonTornadoDelay := 0.5
            Sleep(SummonTornadoDelay)
            spawn:
                SummonRain(NPCCharacter)
            Sleep(0.3)
            MeleeAttackPlayer(NPCCharacter,true)

    SummonRain(NPCCharacter:fort_character)<suspends>:void=

        var CurrentRotation : rotation = RainingTargetProp.GetTransform().Rotation
        var NpcLocation : vector3 = NPCCharacter.GetTransform().Translation
        set NpcLocation.Z = 105.0
        if(RainingTargetProp.TeleportTo[NpcLocation,CurrentRotation]){}

        for(i:=0..RainingAoeAttackNumber-1):

            var TargetLocation : vector3 = RainingTargetProp.GetTransform().Translation
            var TargetRotation : rotation = RainingTargetProp.GetTransform().Rotation

            RandomX := GetRandomFloat(-RainingAoeRange,RainingAoeRange)
            RandomY := GetRandomFloat(-RainingAoeRange,RainingAoeRange)

            set TargetLocation.X = RandomX + TargetLocation.X
            set TargetLocation.Y = RandomY + TargetLocation.Y

            SpawnedPropResult := SpawnProp(RainingAoeNiag,TargetLocation,TargetRotation)
            if(SpawnedProp := SpawnedPropResult(0)?):
                spawn:
                    DeleteEffect(RainingAoeNiagLifeTime,SpawnedProp)
                set TargetLocation.Z = TargetLocation.Z + RainingAoeSpawnHeight
                spawn:
                    RainProjectileMove(SpawnedProp,TargetLocation,TargetRotation)
            
            RandomTime := GetRandomFloat(0.0,0.75)        
            Sleep(RandomTime)

    RainProjectileMove(RainProjectile:creative_prop, TargetLocation:vector3, TargetRotation:rotation)<suspends>:void=

        Sleep(RainingRainProjectileSpawnDelay)

        SpawnedPropResult2 := SpawnProp(RainingProjectile,TargetLocation,TargetRotation)
            if(SpawnedProp := SpawnedPropResult2(0)?):
                
                var Rotation : rotation = SpawnedProp.GetTransform().Rotation
                var RainLocation : vector3 = SpawnedProp.GetTransform().Translation
                set RainLocation.Z = TargetLocation.Z - RainingAoeSpawnHeight
        
                SpawnedProp.MoveTo(RainLocation,Rotation,RainingRainFallSpeed)
                RainDamagePlayers(SpawnedProp)

    RainDamagePlayers(RainProjectile : creative_prop):void=

        MyLocation : vector3 = RainProjectile.GetTransform().Translation
        MyRotation : rotation = RainProjectile.GetTransform().Rotation
        
        SpawnedPropResult := SpawnProp(RainingExplosion,MyLocation,MyRotation)
        if(SpawnedProp := SpawnedPropResult(0)?):
            spawn:
                DeleteEffect(2.0,SpawnedProp)

        RainProjectile.Dispose()
 
        Players := HitAudio.GetPlayspace().GetPlayers()
        for(i:=0..Players.Length-1):
            if:
                Player := Players[i]
                Char := Player.GetFortCharacter[]
                PlayerAgent := Char.GetAgent[]
                Char.IsActive[]
            then:
                PlayerLocation : vector3 = Char.GetTransform().Translation
                
                var Dis2Player : float = Distance(MyLocation,PlayerLocation)
                if(Dis2Player <= RainingAoeNiagSize+100.0):
                    HitAudio.Play(PlayerAgent)
                    Char.Damage(RainingDamage)


    RangeAttackPlayer( NPCCharacter : fort_character ,WaitForHitToPlay : logic)<suspends>:void=

 
        if(Ended = false and WaitForHitToPlay = true and NPCCharacter.IsActive[]):
            Sleep(0.6)


        if:
            NPCCharacter.IsActive[]
            Navigatable := NPCCharacter.GetNavigatable[]
            Ended = false
            PlayerCharacter := CharacterAggro?
            PlayerCharacter.IsActive[]
            PlayerAgent := PlayerCharacter.GetAgent[]
            PlayAnimController := NPCCharacter.GetPlayAnimationController[]
            Focus := NPCCharacter.GetFocusInterface[]
            Agent:=PlayerCharacter.GetAgent[]
            AgentsInArea[Agent]? = true
        then:

            var PlayerLocation : vector3 = PlayerCharacter.GetTransform().Translation
            CurrentLocation : vector3 = NPCCharacter.GetTransform().Translation

            # Caluculating Distance From Creatures Attack Point To The Aggroed Player
            deltaX := CurrentLocation.X - PlayerLocation.X 
            deltaY := CurrentLocation.Y - PlayerLocation.Y 
            var Rad : float = ArcTan(deltaY, deltaX)
    
            AttackPoint := CurrentLocation - vector3{
                X := (AttackRangeRanged * Cos(Rad)),
                Y := (AttackRangeRanged * Sin(Rad)),
                Z := 0.0 }

            var Dis2Player : float = Distance(AttackPoint,PlayerLocation)

            # Gets rotation of npccharacter to see if player is behind it; if it is then it will move instead of attack
            # Apply 90 to correct from the characters given rotaiton
            #myNewRotation := NPCCharacter.GetTransform().Rotation.ApplyYaw(180.0)
            myRotation := NPCCharacter.GetTransform().Rotation.GetYawPitchRollDegrees()
                
            if(var myYaw : float = myRotation[0]):
                #Print("myYaw {myYaw}")
                #convert yawdegrees into radians
                set myYaw = myYaw * (PiFloat / 180.0)

                # Calculate the forward vector components
                # Assuming the forward vector starts pointing north (along the Z axis), 
                forward_vector_x := Cos(myYaw)
                forward_vector_y := Sin(myYaw)

                # Converting it back into 3d space vector
                forwardvector3d := vector3 {
                    X:=forward_vector_x,
                    Y:=forward_vector_y,
                    Z:=0.0
                }

                #normalizing forward vector
                magnitude2 : float = Sqrt(Pow(forwardvector3d.X, 2.0) + Pow(forwardvector3d.Y,2.0) + Pow(forwardvector3d.Z,2.0))
                NorForwardVector := vector3 {
                    X:= forwardvector3d.X / magnitude2,
                    Y:= forwardvector3d.Y / magnitude2,
                    Z:= forwardvector3d.Z / magnitude2
                }

                # Difference vector point from currentlocation to playerlocaiton
                DifferenceVector : vector3 = PlayerLocation - CurrentLocation

                # normalizing difference vector
                magnitude1 : float = Sqrt(Pow(DifferenceVector.X, 2.0) + Pow(DifferenceVector.Y,2.0) + Pow(DifferenceVector.Z,2.0))
                NormDifVector := vector3 {
                    X:= DifferenceVector.X / magnitude1,
                    Y:= DifferenceVector.Y / magnitude1,
                    Z:= DifferenceVector.Z / magnitude1
                }


                # Dot product from difference vector and facing vector will tell us if it is facing it
                dot_product := DotProduct(NormDifVector,NorForwardVector)

                # if the player is in range and in front of npc then attack the player
                if:
                    Dis2Player<=AttackRangeRanged
                    dot_product >= 0.0
                then:

                    if(NPCCharacter.IsActive[]):
                        #Play attack animation
                        spawn:
                            Focus.MaintainFocus(PlayerAgent)
                        Sleep(0.2)
                        AnimationResult := PlayAnimController.Play(GetAnimationSequences(npctype,NPCData.NPCStates.npc_states.RangedAttack),  ?BlendInTime := BlendInTime, ?BlendOutTime := BlendInTime, ?StartPositionSeconds := StartPositionSeconds)
                        RangeAttackAudio.Play()
                        Sleep(RangedAttackDamageDelay)
                        ThrowProjecticle(NPCCharacter,PlayerCharacter)
                        Sleep(RangedAttackAnimLength)
                else:
                         # Walks slightly away from player location in the direction of the npc
                    var WalkToLoc : vector3 = PlayerLocation + vector3{
                        X := AttackRangeRanged  * Cos(Rad),
                        Y := AttackRangeRanged * Sin(Rad),
                        Z := 0.0 }

                    #if aggroed player is behind but in range walk to 1/4 of attack range and turn a random angle
                    if(dot_product < 0.0):
                        RandFloat := GetRandomFloat(-50.0,50.0)
                        set WalkToLoc = PlayerLocation + vector3{
                        X := (((AttackRangeRanged/8.0) * Cos(Rad)) + RandFloat ),
                        Y := (((AttackRangeRanged/8.0) * Sin(Rad)) + RandFloat ),
                        Z := 0.0 }
                        spawn:
                            Focus.MaintainFocus(PlayerLocation)

                    NavTargetStart := MakeNavigationTarget(WalkToLoc)

                    Navigatable.StopNavigation()
                    if(Ended = false):
                        spawn:
                            Navigatable.NavigateTo(NavTargetStart, ?MovementType:=movement_types.Running)

                        Sleep(0.0)
                    #if (NavResultGoTo = navigation_result.Unreachable):
                    #    Print("unreached")
                    #else:
                    #    Navigatable.Wait(?Duration := 0.01)
                    # Caluculating Distance From Creatures Attack Point To The Aggroed Player
                    #if(AttackDebugLocationProp.TeleportTo[AttackPoint, AttackDebugLocationProp.GetTransform().Rotation]){}

            MeleeAttackPlayer(NPCCharacter,true)
            
        else:

            #Returns back to idle after it kills someone
            #Print("Breaking attack loop")
            if:
                PlayerCharacter := CharacterAggro?
                PlayerCharacter.IsActive[]
                Ended = false
                Agent:=PlayerCharacter.GetAgent[]
                AgentsInArea[Agent]? = true
            then:
                #Print("Player Chartacter is still active but breaking loop")
            else:
                set Aggroed = false
                set GotHit = false
                if(Ended = false):
                    #Print("Spawning Idle after Aggroed Character is lost")
                    if:
                        NPCCharacter.IsActive[]
                        Focus := NPCCharacter.GetFocusInterface[]
                    then:
                        spawn:
                            IdleBehavior(Focus,NPCCharacter)

    ThrowProjecticle<public>(NPC : fort_character ,TargetChar : fort_character)<suspends>:void=
        
        StartPosition:= NPC.GetTransform().Translation
        StartRotation := NPC.GetTransform().Rotation
        TargetPosition := TargetChar.GetTransform().Translation

        DifVector := TargetPosition - StartPosition
        NormDifVector := NormalizeVector(DifVector)

        # Used to find point past the player
        AttackPoint := TargetPosition + vector3{
            X := (NormDifVector.X * RangeDistancePastPlayer),
            Y := (NormDifVector.Y * RangeDistancePastPlayer),
            Z := 0.0 }

        SpawnedPropResult := SpawnProp(ProjectileAsset, StartPosition ,StartRotation)
        if (Projectile := SpawnedPropResult(0)?):
            spawn:
                DamageLoop(Projectile)

            spawn:
                Projectile.MoveTo(AttackPoint, StartRotation, RangeProjectileLifetime)


            Sleep(RangeProjectileLifetime)    

            if(EndNiagLogic = true):
                spawn:
                    SpawnEffect(ProjectileEndNiagara, 0.5, Projectile.GetTransform().Translation, Projectile.GetTransform().Rotation, vector3{X:=1.0,Y:=1.0,Z:=1.0})
                Sleep(0.05)

            Projectile.Dispose()   
            
    SpawnEffect(Prop:creative_prop_asset, LifeTime:float, Location : vector3 , Rotatoe : rotation, Scale:vector3)<suspends>:void=

        Tran:=transform{
            Translation := Location,
            Rotation := Rotatoe,
            Scale := Scale
        }

        SpawnedPropResult := SpawnProp(Prop,Tran)
        if(SpawnedProp := SpawnedPropResult(0)?):
            spawn:
                DeleteEffect(LifeTime,SpawnedProp)


    Normalize(Vector : vector3): vector3 =
        magnitude : float = Sqrt(Pow(Vector.X, 2.0) + Pow(Vector.Y,2.0) + Pow(Vector.Z,2.0))
        NorForwardVector := vector3 {
            X:= Vector.X / magnitude,
            Y:= Vector.Y / magnitude,
            Z:= Vector.Z / magnitude
        }
        return NorForwardVector

    DamageCheck(Players:[]player, MyLocation : vector3):void=
    
        for(i:=0..Players.Length-1):
            if:
                Player := Players[i]
                Char := Player.GetFortCharacter[]
                PlayerAgent := Char.GetAgent[]
                Char.IsActive[]
            then:
                PlayerLocation : vector3 = Char.GetTransform().Translation
                
                var Dis2Player : float = Distance(MyLocation,PlayerLocation)
                if(Dis2Player <= RangeHitBox):
                    HitAudio.Play(PlayerAgent)
                    Char.Damage(RangeDamage)

    DamageLoop(ProjectileProp:creative_prop)<suspends>:void=

        loop:
            
            
            if:
                ProjectileProp.IsValid[]
            then:

                MyLocation := ProjectileProp.GetTransform().Translation
                Players := HitAudio.GetPlayspace().GetPlayers()

                for(i:=0..Players.Length-1):
                    if:
                        Player := Players[i]
                        Char := Player.GetFortCharacter[]
                        PlayerAgent := Char.GetAgent[]
                        Char.IsActive[] 
                    then:
                        PlayerLocation : vector3 = Char.GetTransform().Translation    
                        var Dis2Player : float = Distance(MyLocation,PlayerLocation)

                        if(Dis2Player <= RangeHitBox):
                            HitAudio.Play(PlayerAgent)
                            Char.Damage(RangeDamage)

            else:
                break
            Sleep(0.01)
            # Damage Freq



    DelayedHitTest(PlayerCharacter:fort_character,AttackPoint:vector3)<suspends>:void=

        # Attack Hit Test
        Sleep(AttackDamageDelay)
        spawn:
            ShowTemperaryNiagara(AttackPoint, PlayerCharacter.GetTransform().Rotation, AttackImpact, ImpactNSTime)

        # Sleep Hit Delayed

        PlayerLocation := PlayerCharacter.GetTransform().Translation
        Dis2Player := Distance(AttackPoint,PlayerLocation)

        if(Dis2Player <= AttackSize):
            #Print("DamagePlayer")
            if:
                PlayerAgent := PlayerCharacter.GetAgent[]
                MySelf := GetAgent[]
                MyCharacter := MySelf.GetFortCharacter[]
                MyCharacter.IsActive[]

            then:

                HitAudio.Play(PlayerAgent)
                DamageAfterArmor := AttackDamage * DamageCalculation(PlayerCharacter)
                PlayerCharacter.Damage(DamageAfterArmor)

                if(IsPosionous = true):
                    PosionHit(PlayerCharacter)



    # This function runs when the NPC is despawned or eliminated from the world.
    OnEnd<override>():void =

        if:
            Agent := GetAgent[]
            Character := Agent.GetFortCharacter[]
            PlayAnimController := Character.GetPlayAnimationController[]
            Navigatable := Character.GetNavigatable[]
        then:
            Navigatable.StopNavigation()
            AnimationResult := PlayAnimController.Play(GetAnimationSequences(npctype,NPCData.NPCStates.npc_states.Death),  ?BlendInTime := BlendInTime, ?BlendOutTime := BlendInTime, ?StartPositionSeconds := StartPositionSeconds)
            
        set Ended = true
        set Aggroed = false
        set CharacterAggro = false
        set SecondPhase = false

        #Print("Ended")
        #Print("Goodbye, NPC!")

    PosionHit(PlayerCharacter:fort_character)<suspends>:void=
        for(i:=0..4):
            Sleep(2.0)
            if:
                PlayerAgent := PlayerCharacter.GetAgent[]
                PlayerCharacter.IsActive[]
            then:
                HitAudio.Play(PlayerAgent)
                DamageAfterArmor := PosionDamage * DamageCalculation(PlayerCharacter)
                PlayerCharacter.Damage(DamageAfterArmor)
                spawn:
                    ShowTemperaryNiagara(PlayerCharacter.GetTransform().Translation, PlayerCharacter.GetTransform().Rotation, AttackImpact, ImpactNSTime)

