
using { /Fortnite.com/Devices }
using { /Verse.org/Simulation }
using { /UnrealEngine.com/Temporary/Diagnostics }
using { /Fortnite.com/Characters }



# Not yet finished each player mush have a cinematic asigned to it and reset when players leave
# Tried to implement something but I need to test with multiple people
npc_hp_cinematic_hud_device := class(creative_device):

    @editable
    Spawners : []player_spawner_device = array{}

    @editable
    npcdevice : npc_spawner_device = npc_spawner_device{}

    @editable
    cines : []cinematic_sequence_device = array{}
    @editable
    HPBarHuds : []hud_message_device = array{}
    @editable
    hudmessages2 : []hud_message_device = array{}
    @editable
    TrueDamageHuds : []hud_message_device = array{}
    @editable
    FireStackHuds : []hud_message_device = array{}
    # Use Triggers because cant hide doest work for hud message devices
    @editable
    HideFiresTriggers : []trigger_device = array{}
    @editable
    MarkHuds : []hud_message_device = array{}
    # Use Triggers because cant hide doest work for hud message devices
    @editable
    MarkHudTriggers : []trigger_device = array{}

    @editable
    LongRangeHuds : []hud_message_device = array{}
    @editable
    LongRangeHudsTriggers : []trigger_device = array{}
    @editable
    CloseRangeHuds : []hud_message_device = array{}
    @editable
    CloseRangeHudsTriggers : []trigger_device = array{}

    @editable
    XpBarCine : []cinematic_sequence_device = array{} # set by combat_level_device
    @editable
    XpText : []hud_message_device = array{}
    @editable
    LevelText : []hud_message_device = array{}

    var AgentList : []agent = array{}
    var DamagePerAgent : [agent]float = map{}
    var IdPerAgent : [agent]?int = map{}

    @editable
    CMBDevice : combat_level_device = combat_level_device{}

    S2M<localizes>(String1:string, String2:string) : message = "{String1} /  {String2}"
    OneString2M<localizes>(Int2:string):message = "{Int2}"
    I2M<localizes>(Int2:int):message = "{Int2}"
    LevelMessage<localizes>(Int2:int):message = "Lv {Int2}"


    var IdPerAgent2 : [agent]int = map{} # Used to reset counters for ui

    OnBegin<override>()<suspends>:void=

        #spawn:
        #    RefreshBarLoop()

        Players:=GetPlayspace().GetPlayers()
        for(Player:Players):
            OnPlayerAdded(Player)

        GetPlayspace().PlayerAddedEvent().Subscribe(OnPlayerAdded)
        GetPlayspace().PlayerRemovedEvent().Subscribe(OnPlayerRemoved)

        for(Obj : Spawners):
            Obj.SpawnedEvent.Subscribe(PlayerSpawned)

    DelayedXpBarRefresh(Agent:agent)<suspends>:void=
        Sleep(1.5)
        RefreshXpBar(Agent)


    #RefreshBarLoop()<suspends>:void=
    #    loop:
    #        Sleep(0.25)
    #        Players := GetPlayspace().GetPlayers()
    #        for(i:=0..Players.Length-1):
    #            Sleep(0.15)
    #            if(Player := Players[i]):
    #                if(AgentPlaySpace := agent[Player]):
    #                    RefreshXpBar(AgentPlaySpace)

    PlayerSpawned(Agent:agent):void=
        if(Player:= player[Agent]):
            OnPlayerAdded(Player)

    OnPlayerAdded(InPlayer : player) : void =

        Print("PlayerAdded")
        if(Agent:=InPlayer.GetFortCharacter[].GetAgent[]):

            var AvailableID : int = 0
            var StrikeList : []int = array{0,0,0,0} # only suitable for 6 players
            for(MyAgent->Id:IdPerAgent):
                if(RealId := Id?):
                    if(set StrikeList[RealId] = 1){} # loop through idperagent and strike which ids are being used

            for(i:=0..StrikeList.Length-1): # loop through strike list and if its a zero set it to available
                if(StrikeList[i] = 0):
                    set AvailableID = i
                    

            if(ID:=IdPerAgent[Agent]?):
                Print("IdPerAgent:{ID}")
                RefreshXpBar(Agent)
                RefreshXpBar(Agent)
            else:
                if(set IdPerAgent[Agent] = option{AvailableID}):
                    RefreshXpBar(Agent)
                    RefreshXpBar(Agent)
                    if(ID:=IdPerAgent[Agent]?):
                        Print("IdPerAgent2:{ID}")



            spawn:
                DelayedXpBarRefresh(Agent)
            

    OnPlayerRemoved(InPlayer : player): void=
        if(Agent:=InPlayer.GetFortCharacter[].GetAgent[]):
            var NewMap : [agent]?int = map{}
            for(Key->Value:IdPerAgent, Key = Agent):
                set NewMap = ConcatenateMaps(NewMap,map{Key=>Value})
                set IdPerAgent = NewMap


    ShowHud(Agent:agent, HealthTotal : string, HealthCurrent : string, NPCName : string):void=

        Players := GetPlayspace().GetPlayers()

        for(i:=0..Players.Length-1):
            if(Player := Players[i]):
                if(AgentPlaySpace := agent[Player]):
                    if(AgentPlaySpace = Agent):

                        if:
                            Hud1 := HPBarHuds[i]
                            Hud2 := hudmessages2[i]
                        then:
                            Hud1.Show(Agent)
                            Hud1.SetText(OneString2M(NPCName))
                            Hud2.SetText(S2M(HealthCurrent,HealthTotal))
                            Hud2.Show(Agent)

    UpdateAgentUI(Agent:agent,frame:int):void=

        Players := GetPlayspace().GetPlayers()

        for(i:=0..Players.Length-1):
            if(Player := Players[i]):
                if(AgentPlaySpace := agent[Player]):
                    if(AgentPlaySpace = Agent):

                        if(cine := cines[i]):
                            cine.SetPlaybackFrame(frame)
                            cine.Play()
                            cine.Pause()

    AddHud(NPCName : string, Agent:agent, TrueDamage:float):void=

        Players := GetPlayspace().GetPlayers()
        if(TrueDamage > 1.0):
            for(i:=0..Players.Length-1):
                if(Player := Players[i]):
                    if(AgentPlaySpace := agent[Player]):
                        if(AgentPlaySpace = Agent):



                            if:
                                Hud := TrueDamageHuds[i]
                                HPHud := HPBarHuds[i]
                            then:
                                if(CurrentNumber := DamagePerAgent[Agent]):
                                    if(set DamagePerAgent[Agent] = CurrentNumber + TrueDamage){}
                                    Hud.SetText(OneString2M(formatter.Abbreviate((CurrentNumber + TrueDamage) * 1.0,false)))
                                    Hud.Show(Agent)


                                    if(CurrentId :=  IdPerAgent2[Agent]):
                                        if(set IdPerAgent2[Agent] = CurrentId + 1){}
                                    else:
                                        if(set IdPerAgent2[Agent] = 1){}

                                    if(CurrentId :=  IdPerAgent2[Agent]):
                                        spawn:
                                            ResetTrueDamage(Agent,CurrentId)
                                else:
                                    if(set DamagePerAgent[Agent] = TrueDamage){}
                                    Hud.SetText(OneString2M(formatter.Abbreviate(TrueDamage*1.0,false)))
                                    Hud.Show(Agent)

                                    if(CurrentId :=  IdPerAgent2[Agent]):
                                        if(set IdPerAgent2[Agent] = CurrentId + 1){}
                                    else:
                                        if(set IdPerAgent2[Agent] = 1){}
                                    
                                    if(CurrentId :=  IdPerAgent2[Agent]):
                                        spawn:
                                            ResetTrueDamage(Agent,CurrentId)


    HideHudMessage(Agent:agent,huddevices:[]hud_message_device,triggers:[]trigger_device):void=
        Players := GetPlayspace().GetPlayers()
        for(i:=0..Players.Length-1):
            if(Player := Players[i]):
                if(AgentPlaySpace := agent[Player]):
                    if(AgentPlaySpace = Agent):
                        if:
                            huddevice := huddevices[i]
                            Trigg := triggers[i]
                        then:
                            huddevice.Hide()
                            Trigg.Trigger(Agent)

    ShowHudMessage(Agent:agent,huddevices:[]hud_message_device):void=
        Players := GetPlayspace().GetPlayers()
        for(i:=0..Players.Length-1):
            if(Player := Players[i]):
                if(AgentPlaySpace := agent[Player]):
                    if(AgentPlaySpace = Agent):
                        if:
                            huddevice := huddevices[i]
                        then:
                            huddevice.Show(Agent)

    UpdateBurnStacks(Agent:agent,BurnStacks:int):void=
        Players := GetPlayspace().GetPlayers()
        for(i:=0..Players.Length-1):
            if(Player := Players[i]):
                if(AgentPlaySpace := agent[Player]):
                    if(AgentPlaySpace = Agent):
                        if:
                            BurnHudDev := FireStackHuds[i]
                        then:
                            BurnHudDev.SetText(I2M(BurnStacks))
                            BurnHudDev.Show(Agent)
    

    ResetTrueDamage(Agent:agent,Id:int)<suspends>:void=
        Sleep(1.0)
        if(CurrentId :=  IdPerAgent2[Agent]):
            if(CurrentId = Id):
                if(set IdPerAgent2[Agent] = 0){}
                if(set DamagePerAgent[Agent] = 0.0){}


    RefreshXpBar(Agent:agent):void=

        CMBDevice.AreasOnLevelUp(Agent)

        if(XpBar :=  CMBDevice.XpBarPerPlayer[Agent]?):

            if:
                Player := player[Agent]
                OldData := Data[Player]
                CurrentXp := OldData.Xp
            then:
                TotalXp := XpBar.iTotalXp
                CurrentLevel := XpBar.CurrentLevel

                #Print("TotalXp : {TotalXp}")
                #Print("CurrentXp : {CurrentXp}")
                #Print("CurrentLevel : {CurrentLevel}")
                if:
                    i:= IdPerAgent[Agent]?
                    var HealthPercent : int = Round[CurrentXp/TotalXp*100.0]
                then:
                    
                    if(cine := XpBarCine[i]): 
                        
                        cine.SetPlaybackFrame(HealthPercent)
                        cine.Play()
                        cine.Pause()

                        cine.SetPlaybackFrame(HealthPercent)
                        cine.Play()
                        cine.Pause()


                    if(HUDXPText := XpText[i]):
                        Print("I:{i}")
                        HUDXPText.SetText(S2M(formatter.Abbreviate(CurrentXp,false),formatter.Abbreviate(TotalXp,false)))
                        HUDXPText.Show(Agent)

                    if(MyLevelText := LevelText[i]):
                        MyLevelText.SetText(LevelMessage(CurrentLevel))
                        MyLevelText.Show(Agent)
        

    GoldDrop(Agent:agent,I:float):void=
        CMBDevice.ElimHandle(Agent,I)

    AddXp(Agent:agent,XpAdded:float):void=
        CMBDevice.AddXpMaster(Agent, XpAdded)
        RefreshXpBar(Agent)


        
