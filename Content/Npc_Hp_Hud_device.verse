
using { /Fortnite.com/Devices }
using { /Verse.org/Simulation }
using { /UnrealEngine.com/Temporary/Diagnostics }
using { /UnrealEngine.com/Temporary/UI }
using { /Fortnite.com/UI }
using { /UnrealEngine.com/Temporary/SpatialMath }
using { /Verse.org/Colors }
using { /Verse.org/Random }


Npc_Hp_Hud := class():

    var TotalHp<private> : float = 32.0
    var CurrentHp<private> : float = 0.0
    var iBossName<private> : string = "Boar"
    
    BossBarColorBase<private> : texture_block = texture_block{
        DefaultImage := hud.XpBarBackground2
        #DefaultColor := MakeColorFromHex("000000")
        DefaultDesiredSize := vector2{X:= 320.0, Y := 30.0}} 

    var BossBarColorTop<private> : texture_block = texture_block{DefaultImage := hud.HpBar}  
    var Overlay<private> : overlay = overlay{}
    var Overlay2<private> : overlay = overlay{}
        
    XpBarTitleTextBlock<private> : text_block = text_block{DefaultTextColor := MakeColorFromHex("FFFFFF"), DefaultShadowColor:={MakeColorFromHex("5b5b5b")}, DefaultShadowOffset:=option{vector2{X:=2.0,Y:=2.0}}}
    #LevelTitleTextBlock<private> : text_block = text_block{DefaultTextColor := MakeColorFromHex("FFFFFF"), DefaultShadowColor:={MakeColorFromHex("5b5b5b")}, DefaultShadowOffset:=option{vector2{X:=2.0,Y:=2.0}}}

    HPTitleText<private><localizes>(HPText : string) : message = "{HPText}"


    Init(Name : string, BossTotalHP: float):void=

        set iBossName = Name
        set TotalHp = BossTotalHP
        set CurrentHp = TotalHp

        XpBarTitleTextBlock.SetText(HPTitleText("Name"))
        #LevelTitleTextBlock.SetText(HPTitleText("Level: nothing"))

        HealthPercentage := CurrentHp / TotalHp
        BarSize := BossBarColorBase.GetDesiredSize()
        NewHealthBarWidth := BarSize.X * HealthPercentage
        #if (Current := Int[CurrentHp], Total := Int[TotalHp]):
        #    XpBarTitleTextBlock.SetText(HPTitleText("Name"))
            #LevelTitleTextBlock.SetText(HPTitleText("Level: {CurrentLevel}"))
        
        UpdateHealthBar(NewHealthBarWidth)

        spawn:
            DelayedFix()

    # Use this to show damage taken
    UpdateHealth(iCurrentHealth : float):void=
        HealthPercentage := iCurrentHealth / TotalHp
        BarSize := BossBarColorBase.GetDesiredSize()
        NewHealthBarWidth := BarSize.X * HealthPercentage
        #if (Current := Int[CurrentHp], Total := Int[TotalHp]):
        #    XpBarTitleTextBlock.SetText(HPTitleText("{iBossName}"))
            #LevelTitleTextBlock.SetText(HPTitleText("Lv: {CurrentLevel}"))
        
        UpdateHealthBar(NewHealthBarWidth)

    ShowUIForPlayer(Player : player):void=

        if (PlayerUI := GetPlayerUI[Player]):
            PlayerUI.AddWidget(CreateUI())
        spawn:
            DelayedFix()

        #spawn:
        #    RemoveUIForPlayerDelayed(Player)
        

    RemoveUIForPlayer(Player:player)<suspends>:void=
        
        Overlay2.RemoveWidget(XpBarTitleTextBlock)
        #Overlay.RemoveWidget(LevelTitleTextBlock)
        Overlay.RemoveWidget(BossBarColorTop)
        Overlay.RemoveWidget(BossBarColorBase)

        if (PlayerUI := GetPlayerUI[Player]):
            PlayerUI.RemoveWidget(Overlay)

    UpdateHealthBar(NewWidth : float):void=

        #Overlay2.RemoveWidget(XpBarTitleTextBlock)
        #Overlay2.RemoveWidget(LevelTitleTextBlock)
        Overlay.RemoveWidget(BossBarColorTop)

        set BossBarColorTop = CreateTopHealthBar(NewWidth)

        Overlay.AddWidget(CreateOverlaySlot(BossBarColorTop, horizontal_alignment.Left, margin{Top := 0.0, Left := 2.0, Right := 0.0, Bottom := 0.0}))
        #Overlay2.AddWidget(CreateOverlaySlot(XpBarTitleTextBlock,horizontal_alignment.Center, margin{Top := 0.0, Left := 0.0, Right := 0.0, Bottom := 0.0}))
        #Overlay2.AddWidget(CreateOverlaySlot(LevelTitleTextBlock,horizontal_alignment.Left , margin{Top := 0.0, Left := 0.0, Right := 0.0, Bottom := 0.0}))
        
    CreateOverlaySlot<private>(TheWidget : widget, HAlignment : horizontal_alignment, padding : margin):overlay_slot=  
        overlay_slot:
            Widget := TheWidget 
            HorizontalAlignment := HAlignment

    CreateTopHealthBar<private>(Width: float):texture_block=

        ColorBlock := texture_block{
            DefaultImage := hud.HpBar
            #DefaultColor := MakeColorFromHex("#7EC8E3")
            DefaultDesiredSize := vector2{X:= Width, Y := 30.0}} 
        
        set BossBarColorTop = ColorBlock
        

    CreateOverlay<private>() : overlay=
        TheOverlay := overlay:
            Slots := array:
                overlay_slot:
                    Widget := BossBarColorBase
                overlay_slot:
                    Widget := BossBarColorTop
                    Padding := margin{Top := 0.0, Left := 2.0, Right := 0.0, Bottom := 0.0}
                #overlay_slot:
                #    Widget := XpBarTitleTextBlock
                #overlay_slot:
                #    Widget := LevelTitleTextBlock
        set Overlay = TheOverlay

    CreateOverlay2<private>() : overlay = 
        TheOverlay := overlay:
            Slots := array:
                overlay_slot:
                    Widget := XpBarTitleTextBlock
                    HorizontalAlignment := horizontal_alignment.Center
                    
                #overlay_slot:
                #    Widget := LevelTitleTextBlock
                #    HorizontalAlignment := horizontal_alignment.Left
        set Overlay2 = TheOverlay

    CreateUI<private>() : canvas=
        TheCanvas : canvas = canvas:
            Slots := array:
                canvas_slot:
                    Anchors := anchors{Minimum := vector2{X := 0.5, Y := 0.02}, Maximum := vector2{X := 0.5, Y := 0.02}}
                    Offsets := margin{Top := 0.0, Left := 0.0, Right := 0.0, Bottom := 0.0}
                    Alignment := vector2{X := 0.5, Y := 0.0}
                    SizeToContent := true
                    Widget := stack_box:
                        Orientation := orientation.Vertical
                        Slots := array:
                            stack_box_slot:
                                Widget := CreateOverlay2()
                                HorizontalAlignment := horizontal_alignment.Fill
                            stack_box_slot:
                                Widget := CreateOverlay()
                                HorizontalAlignment := horizontal_alignment.Fill
                            
                            
                            
                            
                                                                
        return TheCanvas    

    DelayedFix()<suspends>:void = 
        Sleep(0.01)
        #Print("adding xp")
        #UpdateHealth(0.0)

Npc_Hp_Hud_device := class(creative_device):

    HpBar : Npc_Hp_Hud = Npc_Hp_Hud{}

    var CanvasPerPlayer : [agent]?canvas = map{}
    var HitPerPlayer : [agent]?hud_rising_animation = map{}

    storedstringpopup<localizes>(INT: int) : message = "+{INT}XP"

    var xpcounter : float = 0.0
    var CurrentLevel : int = 1

    var delayenabled : logic = false

    @editable
    testButton : button_device = button_device{}

    OnBegin<override>()<suspends>:void=

        testButton.InteractedWithEvent.Subscribe(Pressed2)

    Pressed(Agent:agent):void =
        if(Player := player[Agent]):
            HpBar.Init("XP", 35.0)
            HpBar.ShowUIForPlayer(Player)

    Pressed2(Agent:agent):void =
        if(Player:= player[Agent]):
            HpBar.ShowUIForPlayer(Player)
            HpBar.UpdateHealth(10.0)
            ElimHandle(Agent,-10.0)

    RaiseXp(Amount : float, Agent:agent,ShowHud:logic):void =

        if(Player:= player[Agent]):
            HpBar.ShowUIForPlayer(Player)
            set CurrentLevel = CurrentLevel + 1
                Print("LeveledUp")

            if(ShowHud = true):
                set xpcounter = xpcounter + Amount
                spawn:
                    ResetTimerOnXpText(xpcounter)

    ResetTimerOnXpText(xpwhensent : float)<suspends>:void = 
        Sleep(0.5)
        if(xpcounter = xpwhensent):
            set xpcounter = 0.0


    XpDelay()<suspends>:void = 
        Sleep(0.5)
        set delayenabled = false

    MakeBasicCanvas(XpGranted:float):tuple(canvas,hud_rising_animation)=

        MyHit:=MakeProgressBar(1.0,vector2{X:=150.0,Y:=400.0},MakeColorFromSRGB(1.0,1.0,1.0),XpGranted, Icons.DefaultIcon)
        MyWidget1:=MyHit.GetWidget1()
        MyWidget2:=MyHit.GetWidget2()
        MyWidget3:=MyHit.GetWidget3()
        MyWidget4:=MyHit.GetWidget4()

        Canvas:=canvas:
            Slots:=array:
                canvas_slot:
                    Anchors := anchors{Minimum := vector2{X := 0.9, Y := 0.4}, Maximum := vector2{X := 0.9, Y := 0.4}}
                    Offsets := margin{Top := 0.0, Left := 0.0, Right := 0.0, Bottom := 0.0}
                    Alignment := vector2{X := 0.0, Y := 0.0}
                    Widget := stack_box:
                        Orientation := orientation.Horizontal
                        Slots := array:
                            stack_box_slot:
                                Widget := MyWidget1
                            stack_box_slot:
                                Widget := MyWidget2
                            stack_box_slot:
                                Widget := MyWidget3
                            stack_box_slot:
                                Widget := MyWidget4

        return(Canvas,MyHit)

    ElimHandle(Agent:agent, XpGranted:float):void=

        Print("Assign Ui")
        Vals:=MakeBasicCanvas(XpGranted)
        Canvas:=Vals(0)
        Hit:=Vals(1)

        if:
            PlayerUI:=GetPlayerUI[player[Agent]]
        then:
            PlayerUI.AddWidget(Canvas,player_ui_slot{InputMode:=ui_input_mode.None})
            if:
                set CanvasPerPlayer[Agent] = option{Canvas}
                set HitPerPlayer[Agent] = option{Hit}
            then:
                spawn:
                    MoveAnimController(Hit,Canvas,PlayerUI)

    MoveAnimController(Hit:hud_rising_animation,Canvas:canvas, PlayerUI:player_ui)<suspends>:void=
        loop:
            Sleep(0.01)

            set Hit.Progress = Hit.Progress -0.01
            Hit.SetPosition(Hit.Progress)

            if:
                Hit.Progress <= 1.0
                Hit.Progress >= 0.0
                myConstrainedValue: type{_X:float where 0.000000 <= _X, _X <= 1.000000} = type{_X:float where 0.000000 <= _X, _X <= 1.000000}[Hit.Progress]
                
            then:
                Hit.TextBlock.SetTextOpacity(myConstrainedValue)
                Hit.TextBlock.SetShadowOpacity(myConstrainedValue)
                Hit.TextBlock2.SetTextOpacity(myConstrainedValue)
                Hit.TextBlock2.SetShadowOpacity(myConstrainedValue)
            
            if(Hit.Progress <= 0.1):
                RemoveUIForPlayer(Hit,Canvas,PlayerUI)
            if(Hit.Progress <= 0.0):
                break

    RemoveUIForPlayer(Hit:hud_rising_animation, Canvas:canvas, PlayerUI:player_ui)<suspends>:void=
        #Overlay := Hit.GetWidget()
        Canvas.RemoveWidget(Hit.MovingColorBase)
        Canvas.RemoveWidget(Hit.TextureIconBlock)
        Canvas.RemoveWidget(Hit.TextBlock)
        PlayerUI.RemoveWidget(Canvas)






