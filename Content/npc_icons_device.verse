
using { /Fortnite.com/Devices }
using { /Verse.org/Simulation }
using { /UnrealEngine.com/Temporary/Diagnostics }


# Used to show charge 
npc_icons_device := class(creative_device):

    # Has to be in the same numbered as the npc devices
    @editable
    ChargeCines : []cinematic_sequence_device = array{}

    @editable
    ShowTrigger : trigger_device = trigger_device{}
    @editable
    Hidetrigger : trigger_device = trigger_device{}

    OnBegin<override>()<suspends>:void=


        ShowTrigger.TriggeredEvent.Subscribe(On)
        Hidetrigger.TriggeredEvent.Subscribe(Off)


    On(Agent:?agent):void=
        ShowCharge(1)

    Off(Agent:?agent):void=
        HideCharge(1)


    ShowCharge(id : int):void=
        if(ChargeCine := ChargeCines[id-1]):
            
            ChargeCine.SetPlaybackFrame(90)
            ChargeCine.Play()
            ChargeCine.Pause()
    HideCharge(id : int):void=
        if(ChargeCine := ChargeCines[id-1]):
            
            ChargeCine.SetPlaybackFrame(0)
            ChargeCine.Play()
            ChargeCine.Pause()
