
using { /Fortnite.com/Devices }
using { /Verse.org/Simulation }
using { /UnrealEngine.com/Temporary/Diagnostics }
using { /Fortnite.com/Characters }
using { /Fortnite.com/Game }
using { /UnrealEngine.com/Temporary/SpatialMath }
using { /Verse.org/Assets }
using { /Verse.org/Random }
using { /Verse.org/Colors }

Niagra := module:
    (Niagra:)DropsVFX<public> := module:

RavenIcons:= module:
    (RavenIcons:)SkillsSpellsandScores<public> := module:
        (SkillsSpellsandScores:)A64x64<public> := module:
    (RavenIcons:)MapMarkers<public> := module:
    (RavenIcons:)AchievementIcons<public> := module:
npc_manager_device := class(creative_device):

    @editable
    ElimAccolade : accolades_device = accolades_device{}

    @editable
    FirstElimAccolade : accolades_device = accolades_device{}

    @editable
    NPCName : string = "Boar"

    @editable
    NpcIconsId : int = 1

    @editable
    NPCDevice : npc_spawner_device = npc_spawner_device{}
    @editable
    LocalNpcsDevice : local_npcs_device = local_npcs_device{}

    @editable
    DamagedAudio : audio_player_device = audio_player_device{}
    @editable
    FireDamageAudio : audio_player_device = audio_player_device{}
    @editable
    CritDamageAudio : audio_player_device = audio_player_device{}
    @editable
    LightningAudio : audio_player_device = audio_player_device{}
    @editable
    EliminatedAudio : audio_player_device = audio_player_device{}
    
    @editable
    ParentMoveProp : creative_prop = creative_prop{}


    # This Spawns When NPC Is Eliminated
    @editable
    NS_Sad_Prop : creative_prop_asset = DefaultCreativePropAsset
    @editable
    CritProp : creative_prop_asset = DefaultCreativePropAsset
    @editable
    SadPropsDroppedOnDeath : int = 1

    @editable
    DeleteLocationProp : creative_prop = creative_prop{}

    var SpawnedCount : int = 1

    # Converted To use with gold
    @editable
    DeathAnimationLength : float = 4.0
    # Changes where the prop is spawned
    @editable
    DeathAnimZCorrection : float = -8.0

    @editable
    NiagaraLogic : logic = false

    #var HpBar : Npc_Hp_Hud = Npc_Hp_Hud{}
    var PlayerShownId : [agent]?int = map{}


    @editable
    CineHudDevice : npc_hp_cinematic_hud_device = npc_hp_cinematic_hud_device{}
    @editable
    IconsDevice : npc_icons_device = npc_icons_device{}
    @editable
    HitHudDevice : hit_ui_device = hit_ui_device{}


    @editable
    BallLightningProp : creative_prop_asset = DefaultCreativePropAsset
    @editable
    ChainLightningProp : creative_prop_asset = DefaultCreativePropAsset
    @editable
    CLOffset : float = -80.0
    @editable
    FireNiag : creative_prop_asset = DefaultCreativePropAsset

    @editable
    ArmorLevel : int = 5
    var CurrentArmor : int = 0

    @editable
    var ElimXP : float = 4.0
    @editable
    var ElimGold : float = 5.2
    # ReDoing Health
    @editable
    TotalHealth : float = 10000000.0
    var CurrentHealth : float = 0.0
    
    var NPCAgent : ?agent = false


    DamagePerPlayerLevel : float = 0.10

    TrueDamagePercentPerLevel : float = 0.125 # Changed 1.125 to 0.125
    CritChancePerLevel : float = 0.015

    GoldPerElimInitial : float = 0.125 # changed from 1.125 to 0.125
    XpBoostInitial : float = 0.25

    # Elemental Upgrades Some are Percentages
    FireOddsPerUpgrade : float = 0.015
    # Fire Damage is Flat
    FireDamagePerUpgrade : float = 21.0

    SuperCritOddsPerUpgrade : float = 0.01
    #SuperCritDamagePerUpgrade : float = 2.0

    #ChainLightningDamagePerUpgrade : float = 0.45
    ChainLightningOdds : float = 0.015
    # 2000.0 max
    ChainLightningRange : float = 2000.0
    # array has agents aleady target by the chain
    var ChainLightningList : [fort_character]int = map{}

    FocusDamagePerLevel : float = 0.25 # will only reach 100% of damage if monster is close af
    FocusDamageRange : float = 1000.0
    AccuracyDamagePerLevel : float = 0.1 # you will always make 100% of damage can make more if mob is super far away
    AccuracyDamageRange : float = 3000.0



    var MarkHitCount: int = 0
    HitsReqMark : int = 20
    MarkDamageMultiple : float = 0.33

    var BurnStacks : [agent]int = map{}
    var BurnId : int = 0

    CritTextColor : color = MakeColorFromHex("FFA500") # Orange
    SuperCritTextColor : color = MakeColorFromHex("A020F0") # Purple
    FireTextColor : color = MakeColorFromHex("FF0000") # Red
    TDTextColor : color = MakeColorFromHex("#FFFFFF") # White
    CLTextColor : color = MakeColorFromHex("#0000FF") # White
    FocusTextColor : color = MakeColorFromHex("#FFFFFF") # White

    FireIcon : texture = RavenIcons.MapMarkers.tile000
    CritIcon : texture = RavenIcons.SkillsSpellsandScores.A64x64.S16
    TrueDamageIcon : texture = RavenIcons.SkillsSpellsandScores.A64x64.S79
    LightningIcon : texture = RavenIcons.MapMarkers.SS90
    FocusIcon : texture = RavenIcons.MapMarkers.S234
    AccuracyIcon : texture = RavenIcons.MapMarkers.SS147

    var AgentDamageMap : [agent]float = map{}

    var NPCSAlive : int = 0 # used for healig boss

    @editable
    OreGranters : []item_granter_device = array{}
    @editable
    OreOdds : []int = array{1000,-1,-1,-1,-1,-1} # out of 10,000
    @editable
    OreQuantities : []int = array{1,0,0,0,0,0}
    @editable
    OreNiag : []creative_prop_asset = array{}

    @editable
    ResetHpTrigger : trigger_device = trigger_device{}
    @editable
    SecondPhaseTrigger : trigger_device = trigger_device{}


    @editable
    WeaponRarityDevice : weapon_rarity_device = weapon_rarity_device{}
    @editable
    AchievementTrigger : trigger_device = trigger_device{}
    @editable
    CosmeticUnlockDevice : cosmetic_unlock_device = cosmetic_unlock_device{}

    @editable
    IsBoss : logic = false
    @editable
    BossDefeatedHud : hud_message_device = hud_message_device{}

    MaxOreDropRate : int = 5

    var DmgNumberSpawnProp : creative_prop_asset = DefaultCreativePropAsset
    var CritNumberSpawnProp : creative_prop_asset = DefaultCreativePropAsset
    var FireNumberSpawnProp : creative_prop_asset = DefaultCreativePropAsset
    var LightningNumberSpawnProp : creative_prop_asset = DefaultCreativePropAsset
    var AccuracyNumberSpawnProp : creative_prop_asset = DefaultCreativePropAsset
    var FocusNumberSpawnProp : creative_prop_asset = DefaultCreativePropAsset
    var SuperCritNumberSpawnProp : creative_prop_asset = DefaultCreativePropAsset

    @editable
    DamageNumberProps : []creative_prop_asset = array{}

    @editable
    DamageHeightOffset : float = 0.0
    @editable
    DamageSideOffset : float = 0.0

    @editable
    ForceSpawnButton : button_device = button_device{}

    OnBegin<override>()<suspends>:void=

        if(MyProp := DamageNumberProps[0]):
            set DmgNumberSpawnProp = MyProp
        if(MyProp := DamageNumberProps[1]):
            set CritNumberSpawnProp = MyProp
        if(MyProp := DamageNumberProps[2]):
            set FireNumberSpawnProp = MyProp
        if(MyProp := DamageNumberProps[3]):
            set LightningNumberSpawnProp = MyProp
        if(MyProp := DamageNumberProps[4]):
            set AccuracyNumberSpawnProp = MyProp
        if(MyProp := DamageNumberProps[5]):
            set FocusNumberSpawnProp = MyProp
        if(MyProp := DamageNumberProps[6]):
            set SuperCritNumberSpawnProp = MyProp

        NPCDevice.SpawnedEvent.Subscribe(SpawnedHandle)
        NPCDevice.EliminatedEvent.Subscribe(ElimHandle)
        ResetHpTrigger.TriggeredEvent.Subscribe(ResetHp)

        spawn:
            FireDamageTicks()

        spawn:
            ForceSpawn()

    ResetHp(MaybeAgent:?agent):void=

        #set CurrentHealth = TotalHealth

        if(CurrentHealth < TotalHealth):
            HealHealth := CurrentHealth + (0.25 * TotalHealth)
            if(HealHealth < TotalHealth):
                set CurrentHealth = HealHealth
            else:
                set CurrentHealth = TotalHealth

        set BurnStacks = map{}


    ButtonForceSpawn(Agent:agent) : void=
        spawn:
            ForceSpawn()



    # Used to fix npc spawners not working after 29.3
    ForceSpawn()<suspends>:void=

        loop:
            Sleep(3.0)
            if(NPCSAlive < 1):
                NPCDevice.Spawn()
            else:
                break

    RollOreSpawn(MyCharacter:fort_character):void=

        Players := 	GetPlayspace().GetPlayers()
        for(Player:Players):
            if:
                Player.IsActive[]
                Agent:=agent[Player]
                Key := Agent
                Value := AgentDamageMap[Agent]
            then:
                PercentOfXp := (Value*1.0)/TotalHealth
                if(PercentOfXp > 0.1):
                    if(PlayerData := Data[Player]):
                        if(PlayerData.OreDropLevel>0):

                            MYLoc :vector3 = MyCharacter.GetTransform().Translation
                            var Location : vector3 = MyCharacter.GetTransform().Translation
                            Rotation := MyCharacter.GetTransform().Rotation
                            set Location.Z = Location.Z + DeathAnimZCorrection

                            for(i:=0..OreOdds.Length-1):
                                if:
                                    Odds := OreOdds[i]
                                    Odds > 0
                                    Granter := OreGranters[i]
                                    Quant := OreQuantities[i]
                                    Prop := OreNiag[i]
                                    RealOdds := Round[((Odds*1.0)/(MaxOreDropRate*1.0)) * ((PlayerData.OreDropLevel + PlayerData.GemsOreDropLevel) *1.0)]
                                then:
                                    for(i2:=0..Quant-1):
                                        RandFloat : int = GetRandomInt(0,10000)
                                        Print("RandomFloat{RandFloat} Odds {RealOdds}")
                                        if(RandFloat <= RealOdds):
                                            Granter.GrantItem(Key)
                                            spawn:
                                                SpawnEffect(Prop, DeathAnimationLength, Location, Rotation, vector3{X:=1.0,Y:=1.0,Z:=1.0})

   
    SpawnedHandle(Agent:agent):void=

        set NPCSAlive = NPCSAlive + 1
        set NPCAgent = option{Agent}
        set CurrentHealth = TotalHealth
        set BurnId = 0
        set BurnStacks = map{}
        set CurrentArmor = ArmorLevel 
        set MarkHitCount = 0
        

        #HpBarNew : Npc_Hp_Hud = Npc_Hp_Hud{}
        #set HpBar = HpBarNew
        #HpBar.Init("Boar", TotalHealth)

        if(MyChar := Agent.GetFortCharacter[]):
            var MyTrans : vector3 = MyChar.GetTransform().Translation
             set MyTrans.Y = MyTrans.Y + 600.0
            if(ParentMoveProp.TeleportTo[MyTrans, MyChar.GetTransform().Rotation]){}
            spawn:
                DevicesFollow(Agent)
            MyChar.DamagedEvent().Subscribe(DamageHandle)
            

    DevicesFollow(Agent:agent)<suspends>:void=
        #Delay to reset the position of the argo range device
        Sleep(0.5)
        loop:
            Sleep(1.0)
            if(MyChar := Agent.GetFortCharacter[]):
                if(MyChar.IsActive[]):
                    if(ParentMoveProp.IsValid[]):
                        if(ParentMoveProp.TeleportTo[MyChar.GetTransform().Translation, MyChar.GetTransform().Rotation]){}
                else:
                    break

    DamageHandle(DamageResult : damage_result):void=
        # Shows player Ui / doesnt need to happen all the time
        DamageSounce := DamageResult.Instigator
        MyTarget := DamageResult.Target
        var DamageDone : float = DamageResult.Amount
        DamagedAudio.Play()

        # Custom Health
        if:
            MyCharacter := fort_character[MyTarget]
            MyCharacter.IsActive[]
        then:
            set CurrentHealth = CurrentHealth - (DamageDone)
            
        if:
            MyCharacter := fort_character[MyTarget]
            MyAgent:= MyCharacter.GetAgent[]
            InstigatingPlayer := DamageSounce?
            Fort := fort_character[InstigatingPlayer]
            InstigatingAgent := Fort.GetAgent[]
            SourceAgent := Fort.GetAgent[]
            Player := player[SourceAgent]
            not PlayerShownId[SourceAgent]? = 0
            MyCharacter.IsActive[]
            MyData := Data[Player]
        then:
            if(CurrentHealth <= 0.0):
                RollOreSpawn(MyCharacter)
                LocationFunctionsOnDeath(MyCharacter)
                MyCharacter.Damage(1000000.0)
                
            
            # Check Mark Be For Everything
            MarkCal := MarkCalculation(SourceAgent)
            if(MarkCal = true):
                set DamageDone = DamageDone + (DamageDone * MarkDamageMultiple * MyData.MarkDamage)
            
            AddToDamgeMap(SourceAgent, DamageDone)

            TD := TrueDamageCalculation(SourceAgent,DamageDone)
            set CurrentHealth = CurrentHealth - (TD)
            AddToDamgeMap(SourceAgent, TD)
        
            FireCalculation(MyAgent,SourceAgent,(DamageDone+TD))
            CritCalculation(MyCharacter, SourceAgent, (DamageDone+TD))
            SuperCritCalculation(MyCharacter, SourceAgent, (DamageDone+TD))
            set ChainLightningList = map{}
            LightningCalculation(MyCharacter,SourceAgent,(DamageDone+TD))
            FocusCalculation(MyCharacter,SourceAgent,(DamageDone+TD))
            AccuracyDamageCalculation(MyCharacter,SourceAgent,(DamageDone+TD))

            UpdateHealthCine(SourceAgent)

            if(CurrentHealth <= TotalHealth / 2.0): # might be a problem constantly activating a trigger
                SecondPhaseTrigger.Trigger(MyAgent)

        if:
            MyCharacter := fort_character[MyTarget]
            MyCharacter.IsActive[]
        then:
            if(CurrentHealth <= 0.0):

                RollOreSpawn(MyCharacter)
                LocationFunctionsOnDeath(MyCharacter)
                MyCharacter.Damage(1000000.0)
            else:
                MyCharacter.SetHealth(10000.0)

    CustomDamage(TargetChar:fort_character,Damage:float):void=
        set CurrentHealth = CurrentHealth - ( Damage )
        if(MyAgent:=TargetChar.GetAgent[]):
            UpdateHealthCine(MyAgent)
        CheckHeath(TargetChar)

    CheckHeath(TargetChar:fort_character):void=
        if:
            TargetChar.IsActive[]
        then:
            if(CurrentHealth <= 0.0):
                RollOreSpawn(TargetChar)
                LocationFunctionsOnDeath(TargetChar)
                TargetChar.Damage(1000000.0)


    LocationFunctionsOnDeath(MyCharacter:fort_character):void=

        MYLoc :vector3 = MyCharacter.GetTransform().Translation
        var Location : vector3 = MyCharacter.GetTransform().Translation
        Rotation := MyCharacter.GetTransform().Rotation
        set Location.Z = Location.Z + DeathAnimZCorrection
        SadPropsDroppedOnDeath
        
        for(i:=0..SadPropsDroppedOnDeath-1):
            spawn:
                SpawnEffect(NS_Sad_Prop, DeathAnimationLength, Location, Rotation, vector3{X:=1.0,Y:=1.0,Z:=1.0})


    MarkCalculation(SourceAgent:agent):logic=
        if (Player := player[SourceAgent]):
            if(MarkLevel := Data[Player].MarkDamage):
                if (MarkLevel > 0):
                    set MarkHitCount = MarkHitCount + 1
                    if(MarkHitCount >= HitsReqMark):
                        
                        CineHudDevice.ShowHudMessage(SourceAgent,CineHudDevice.MarkHuds)
                        return true
                    else:
                        CineHudDevice.HideHudMessage(SourceAgent,CineHudDevice.MarkHuds, CineHudDevice.MarkHudTriggers)
                else:
                    CineHudDevice.HideHudMessage(SourceAgent,CineHudDevice.MarkHuds, CineHudDevice.MarkHudTriggers)
        return false

    LightningCalculation(Target:fort_character,SourceAgent:agent,DamageDone:float):void=
         # Loading Players FireChance and FireDamage From Persistance
         if (Player := player[SourceAgent]):
            if(ChainLightningChance := Data[Player].ChainLightningChance):
                if(CosmeticStat := CosmeticUnlockDevice.AgentCosmeticStats[SourceAgent]):
                    if (CosmeticLightningChance : float = Data[Player].CosmeticLightningChance / 100.0):
                        if (CosmeticLightningDamage : float = Data[Player].CosmeticLightningDamage / 100.0 ):
                            if(ChainLightningDamage := Data[Player].ChainLightningDamage * Data[Player].ChainLightningDamagePerUpgrade):
                                if(GetRandomFloat(0.0,100.0) < (((ChainLightningChance ) * ChainLightningOdds) * 100.0 ) + CosmeticLightningChance) :
                                    CLDamage := (ChainLightningDamage * DamageDone) + (CosmeticLightningDamage * ChainLightningDamage * DamageDone)
                                    SpawnDamageNumberProp(SourceAgent, CLDamage,LightningNumberSpawnProp)
                                    CustomDamage(Target,CLDamage)
                                    #HitHudDevice.ShowHitUi(SourceAgent,LightningIcon,CLTextColor,CLDamage)
                                    
                                        #if(A:=Target.GetAgent[]):
                                        #    DamageNumberDevice.ShowDamageNumberAtAgent(A, CLDamage)
                                    LightningAudio.Play(SourceAgent)

                                    var Location :vector3= Target.GetTransform().Translation 
                                    set Location.Z = Location.Z + CLOffset
                                    Rotation := Target.GetTransform().Rotation
                                    spawn:
                                        SpawnEffect(BallLightningProp, 3.0, Location, Rotation,vector3{X:=1.0,Y:=1.0,Z:=1.0})
                                    if(set ChainLightningList[Target] = 1){}
                                    LightningChain(Target,SourceAgent,CLDamage)

                                    UpdateHealthCine(SourceAgent)
                                    AddToDamgeMap(SourceAgent, CLDamage)


    InsertDistanceArr(Arr:[]float, Dis:float, index:int):[]float=

        var Arr1 : []float = array{}
        var Arr2 : []float = array{}

        for(i:=0..index-1):
            if(In := Arr[i]):
                set Arr1 = Arr1 + array{In}

        for(i:=index..Arr.Length-1):
            if(In := Arr[i]):
                set Arr2 = Arr2 + array{In}

        return Arr1 + array{Dis} + Arr2

    InsertCharArr(Arr:[]fort_character, Dis:fort_character, index:int):[]fort_character=

        var Arr1 : []fort_character = array{}
        var Arr2 : []fort_character = array{}

        for(i:=0..index-1):
            if(In := Arr[i]):
                set Arr1 = Arr1 + array{In}

        for(i:=index..Arr.Length-1):
            if(In := Arr[i]):
                set Arr2 = Arr2 + array{In}

        return Arr1 + array{Dis} + Arr2
          
    LightningChain(Target:fort_character,SourceAgent:agent,CLDamage:float):void=

        var InRange : []fort_character = array{}
        var InRangeDistances : []float = array{}
        var StartingNodeChar : fort_character = Target

        if (Player := player[SourceAgent]):
            if(Jumps := Data[Player].ChainLightningJumps):

                for(Key->Value:LocalNpcsDevice.NpcLocationMap):
                    if(Char := Key.GetFortCharacter[]):
                        Dis := Distance(StartingNodeChar.GetTransform().Translation,Char.GetTransform().Translation)

                        if(Dis > 0.0):

                            # Debugging tool
                            var FiledIn : int = 0

                            if(InRangeDistances.Length > 0):
                                var FoundLarger : logic = false
                                for(i3:=0..InRangeDistances.Length-1):
                                    if:
                                        TestedDis := InRangeDistances[i3]
                                        TestedDis > Dis 
                                    then:
                                        if(FoundLarger = false):
                                            if(i3 = 0):
                                                set FiledIn = 0
                                                set InRangeDistances = array{Dis} + InRangeDistances 
                                                set InRange = array{Char} + InRange 
                                                set FoundLarger = true
      
                                            else:
                                                set FiledIn = i3
                                                if(NewDisArr := InRangeDistances.Insert[i3,array{Dis}]){set InRangeDistances = NewDisArr}
                                                if(NewRangeArr := InRange.Insert[i3,array{Char}]){set InRange = NewRangeArr}
                                                #set InRangeDistances = InsertDistanceArr(InRangeDistances, Dis, i3+1)
                                                #set InRange = InsertCharArr(InRange, Char, i3+1)
                                                set FoundLarger = true
                                                

                                if(FoundLarger = false):
                                    set FiledIn = InRangeDistances.Length
                                    set InRangeDistances = InRangeDistances + array{Dis} 
                                    set InRange =  InRange + array{Char} 
                            else:
                                set FiledIn = 0
                                set InRangeDistances = array{Dis}
                                set InRange = array{Char}


                for(i22:= 0..InRangeDistances.Length-1):
                    if(Distances := InRangeDistances[i22]):

                                
                        #if(Dis <= ChainLightningRange):

                        #    Print("Dis : {Dis}")
                        #    if:
                        #        AlreadyInRange := InRange[InRange.Length-1]
                        #        DisAlreadyInRange := Distance(StartingNodeChar.GetTransform().Translation,AlreadyInRange.GetTransform().Translation)
                        #        Dis < DisAlreadyInRange
                        #    then:
                        #        set InRange = InRange + array{Char}
                        #    else:
                        #        set InRange = array{Char} + InRange

                                
                                

                for(i:=0..Jumps):
                    if(InRange.Length > 0):
                        if(ChainLightningList[InRange[i]]):
                        else:
                            if(Char := InRange[i]):
                                Dis := Distance(StartingNodeChar.GetTransform().Translation,Char.GetTransform().Translation)

                                if(Dis <= ChainLightningRange):

                                    if(set ChainLightningList[Char] = 1){}

                                    var Location :vector3= Char.GetTransform().Translation 
                                    set Location.Z = Location.Z + CLOffset
                                    Rotation := Char.GetTransform().Rotation
                                    spawn:
                                        SpawnEffect(BallLightningProp, 3.0, Location, Rotation, vector3{X:=1.0,Y:=1.0,Z:=1.0})
                                    #if(CLInt : int = Round[CLDamage]){HitHudDevice.ShowHitUi(SourceAgent,LightningIcon,CLTextColor,CLInt)}
                                    CustomDamage(Char,CLDamage)

                                    # Calculating Location
                                    KeyTransform := Char.GetTransform().Translation
                                    var TargetTransform : vector3 = StartingNodeChar.GetTransform().Translation

                                    Midpoint : vector3 = vector3{X:=(TargetTransform.X + KeyTransform.X)/2.0, Y:=(TargetTransform.Y + KeyTransform.Y)/2.0, Z:=(TargetTransform.Z + KeyTransform.Z)/2.0 } 

                                    # Calculating Yaw Rotation
                                    X1 : float =  KeyTransform.X - TargetTransform.X
                                    Y1 : float =  KeyTransform.Y - TargetTransform.Y
                                    var AngleDegrees : float = (RadiansToDegrees(ArcTan(Y1,X1)) )

                                    # Calculating Pitch Still Testing
                                    X2 : float =  KeyTransform.X - TargetTransform.X
                                    H : float = Sqrt(Pow(X2,2.0) + Pow(Y1,2.0))
                                    Z2 : float =  KeyTransform.Z - TargetTransform.Z
                                    AngleDegrees2 : float = (RadiansToDegrees(ArcTan(Z2,H)) + 90.0)

                                    #if(Z2 > 0.0):
                                    #    Print("z2 is negative start is lower")
                                    #    set AngleDegrees = AngleDegrees + 180.0

                                    MyRotaiton := MakeRotationFromYawPitchRollDegrees(AngleDegrees, AngleDegrees2, 0.0)

                                    
                                    YScale : float = Dis/100.0

                                    spawn:
                                        SpawnEffect(ChainLightningProp,0.4, Midpoint, MyRotaiton, vector3{X:=1.0,Y:=YScale,Z:=1.0})

                                    if(I:=LocalNpcsDevice.NpcTriggerMap[Char.GetAgent[]]?):
                                        if(myTrigger := LocalNpcsDevice.LocalNpcTriggers[I]):
                                            myTrigger.Trigger(SourceAgent)
                                            #Print("Triggering Aggro")

                                    set StartingNodeChar = Char


        for(Key->Value:LocalNpcsDevice.NpcLocationMap):
            if(Key <> Target):
                if(Char := Key.GetFortCharacter[]):
                    Dis := Distance(Target.GetTransform().Translation,Char.GetTransform().Translation)
                    if(Dis <= ChainLightningRange):
                    


    TrueDamageCalculation(SourceAgent:agent,DamageDone:float):float=

        var TDamage : float = 0.0
        var AfterArmor : float = 0.0
        var WeaponRarityDmg : float = 0.0
        var AchievementDamage : float = 0.0
        var GemDamage : float = 0.0

        var AdjustedTrueDamageSaved : float = 0.0 # Used to deal damage when player has 0 level but more than 0 gems


        if:
            Player := player[SourceAgent]
            PlayerData := Data[Player]
            var TrueDamageSaved : int = PlayerData.TrueDamage
            #TrueDamageSaved > 0

            CosmeticStat := CosmeticUnlockDevice.AgentCosmeticStats[SourceAgent]
            CosmeticDamage : float = Data[Player].CosmeticDamage / 100.0

        then:

            if(TrueDamageSaved = 0):
                if(PlayerData.Gems > 0.0):
                    set AdjustedTrueDamageSaved = 0.01
            else:
                set AdjustedTrueDamageSaved = TrueDamageSaved * 1.0

            # Exponential damage -> Linear Changed 1.125 -> 0.125
            #set TDamage = (( Pow(TrueDamagePercentPerLevel + (PlayerData.DamagePL * PlayerData.DamagePLLevel),(TrueDamageSaved)*1.0)) - 1.0) * DamageDone 



            set TDamage = ( TrueDamagePercentPerLevel + (PlayerData.DamagePL * PlayerData.DamagePLLevel) ) * AdjustedTrueDamageSaved * DamageDone

            
            set GemDamage = ((PlayerData.PrestigeBonusPerGem*1.0)/100.0) * (PlayerData.Gems) * TDamage
            set TDamage = GemDamage + TDamage

            set AchievementDamage = (PlayerData.AchievementsDamageBonus/100.0) * TDamage

            TypeDamage := WeaponTypeMultiple(SourceAgent)
            set TDamage = TDamage + (TDamage * CosmeticDamage) + (TDamage * TypeDamage) + (TDamage * DamagePerPlayerLevel * PlayerData.Level) # adding cosmetic dmaamge
            
            #Adds Weapon Rarity Damage Based on TDamge Percent
            if(WeaponRaritySaved := WeaponRarityDevice.AgentMap[SourceAgent]):
                set WeaponRarityDmg = TDamage * WeaponRaritySaved

            set AfterArmor = DamageCalculationCreatures(TDamage + WeaponRarityDmg + AchievementDamage )
            #CineHudDevice.AddHud(NPCName,SourceAgent,AfterArmor)
            SpawnDamageNumberProp(SourceAgent, AfterArmor,DmgNumberSpawnProp)

        return AfterArmor

    WeaponTypeMultiple(Agent:agent):float =

        var Multiple : float = 0.0
        WeaponTypeDamagePerLevel := 6.0 # 6.0 Percent
        WeaponTypeExponentialGrowth := 1.2

        if:
            Player := player[Agent]
            PlayerData := Data[Player]
        then:
            
            WeaponType := WeaponRarityDevice.GetWeaponType(Agent)
            if(WeaponType = 0): 
                set Multiple = Pow(PlayerData.PistolLevel * 1.0, WeaponTypeExponentialGrowth) * WeaponTypeDamagePerLevel
            if(WeaponType = 1):
                set Multiple = Pow(PlayerData.SmgLevel * 1.0, WeaponTypeExponentialGrowth) * WeaponTypeDamagePerLevel
            if(WeaponType = 2):
                set Multiple = Pow(PlayerData.RifleLevel * 1.0, WeaponTypeExponentialGrowth) * WeaponTypeDamagePerLevel
            if(WeaponType = 3): 
                set Multiple = Pow(PlayerData.ShotgunLevel * 1.0, WeaponTypeExponentialGrowth) * WeaponTypeDamagePerLevel
            if(WeaponType = 4):
                set Multiple = Pow(PlayerData.SniperLevel * 1.0, WeaponTypeExponentialGrowth) * WeaponTypeDamagePerLevel

        return Multiple / 100.0

    FireCalculation(TargetAgent:agent,SourceAgent:agent,DamageDone:float):void=
            if (Player := player[SourceAgent]):
                if(CosmeticStat := CosmeticUnlockDevice.AgentCosmeticStats[SourceAgent]):
                    if:
                        CosmeticFireChance : float = Data[Player].CosmeticBurnDamage / 100.0 # Currently dont have burn chance so they must stay the smae
                    then:
                        if(FireChanceLevel := Data[Player].FireChance):
                            if(FireDamageLevel := Data[Player].FireDamage ):
                                if(FireTime:= Data[Player].FireTime):
                                    FireOdds : float =  (FireOddsPerUpgrade * FireChanceLevel) + (FireOddsPerUpgrade * FireChanceLevel * CosmeticFireChance)
                                    if:
                                        GetRandomFloat(0.0,100.0) < (FireOdds * 100.0)
                                    then:
                                        if (CurrentStacks := BurnStacks[SourceAgent] ):
                                            if(set BurnStacks[SourceAgent] = CurrentStacks + 1):
                                                set BurnId = BurnId + 1
                                                spawn:
                                                    ResetBurnStacks(BurnId)
                                        else:
                                            if(set BurnStacks[SourceAgent] = 1):
                                                set BurnId = BurnId + 1
                                                spawn:
                                                    ResetBurnStacks(BurnId)
                                    else:



    FocusCalculation(Target:fort_character,SourceAgent:agent,DamageDone:float):void=

        if (Player := player[SourceAgent]):
            if(FocusLevel := Data[Player].FocusDamage):
                if(FocusLevel > 0):
                    if(Fort := Player.GetFortCharacter[]):
                        Dis := Distance(Target.GetTransform().Translation,Fort.GetTransform().Translation)
                        if(Dis <= FocusDamageRange):

                            RangeMultiple := (FocusDamageRange - Dis) / FocusDamageRange + 0.10

                            AfterArmor := DamageCalculationCreatures((DamageDone * (FocusDamagePerLevel * FocusLevel) * RangeMultiple))
                            SpawnDamageNumberProp(SourceAgent, AfterArmor,FocusNumberSpawnProp)
                            CustomDamage(Target, AfterArmor )

                            CineHudDevice.ShowHudMessage(SourceAgent,CineHudDevice.CloseRangeHuds)
                                # just for trouble shooting focus damage
                            #HitHudDevice.ShowHitUi(SourceAgent,FocusIcon,FocusTextColor,AfterArmor)
                           
                            
                            AddToDamgeMap(SourceAgent, AfterArmor)
                        else:
                            CineHudDevice.HideHudMessage(SourceAgent,CineHudDevice.CloseRangeHuds,CineHudDevice.CloseRangeHudsTriggers)

    AccuracyDamageCalculation(Target:fort_character,SourceAgent:agent,DamageDone:float):void=
        if (Player := player[SourceAgent]):
            if(AccuracyLevel := Data[Player].AccuracyDamage):
                if(AccuracyLevel > 0):
                    if(Fort := Player.GetFortCharacter[]):
                        Dis := Distance(Target.GetTransform().Translation,Fort.GetTransform().Translation)
                        if(Dis >= AccuracyDamageRange):

                            RangeMultiple := (Dis / AccuracyDamageRange) + 0.10

                            CineHudDevice.ShowHudMessage(SourceAgent,CineHudDevice.LongRangeHuds)
                            AfterArmor := DamageCalculationCreatures((DamageDone * (AccuracyDamagePerLevel * AccuracyLevel) * RangeMultiple)) # the Icon and TextColor Need to be changed
                            SpawnDamageNumberProp(SourceAgent, AfterArmor,AccuracyNumberSpawnProp)
                            CustomDamage( Target , AfterArmor )
                            
                            # just for trouble shooting accuracy damage
                            #HitHudDevice.ShowHitUi(SourceAgent,AccuracyIcon,FocusTextColor,AfterArmor)

                            AddToDamgeMap(SourceAgent, AfterArmor)
                        else:
                            CineHudDevice.HideHudMessage(SourceAgent,CineHudDevice.LongRangeHuds,CineHudDevice.LongRangeHudsTriggers)
                    
    CritCalculation(Target:fort_character,SourceAgent:agent, DamageDone:float):void=
        # Loading Players FireChance and FireDamage From Persistance
        if:
            Player := player[SourceAgent]
            MyData := Data[Player]
        then:
            if(CritChance := Data[Player].CritLevel):
                if(CosmeticStat : CosmeticStats = CosmeticUnlockDevice.AgentCosmeticStats[SourceAgent]):
                    if:
                        CosmeticCritDamage : float = Data[Player].CosmeticCritDamage / 100.0
                        CosmeticCritChance : float = Data[Player].CosmeticCritChance / 100.0
                        TestCosmetic := CosmeticStat.StatMap["Crit Damage"]
                    then:
                        
                        CritDamageMultiple := (MyData.CritLevel  * (CritChancePerLevel + (MyData.CritPL * MyData.CritPLLevel))) + CosmeticCritDamage
                        CosmeticCritChanceOdds := (MyData.CritLevel  * (CritChancePerLevel + (MyData.CritPL * MyData.CritPLLevel))) * CosmeticCritChance
                        CritOdds := (MyData.CritLevel  * (CritChancePerLevel + (MyData.CritPL * MyData.CritPLLevel))) + CosmeticCritChanceOdds

                        if(GetRandomFloat(0.0,100.0) < CritOdds * 100.0):

                            var Location :vector3= Target.GetTransform().Translation 
                            set Location.Z = Location.Z + CLOffset
                            Rotation := Target.GetTransform().Rotation
                            spawn:
                                SpawnEffect(CritProp, 0.5, Location, Rotation,vector3{X:=1.0,Y:=1.0,Z:=1.0})

                            AfterArmor := DamageCalculationCreatures(DamageDone + (DamageDone * (CritDamageMultiple )))
                            SpawnDamageNumberProp(SourceAgent, AfterArmor,CritNumberSpawnProp)
                            CustomDamage( Target ,AfterArmor )
                            CritDamageAudio.Play(SourceAgent)
                            #HitHudDevice.ShowHitUi(SourceAgent,CritIcon,CritTextColor,AfterArmor)
                            

                            AddToDamgeMap(SourceAgent, AfterArmor)

                            


    SuperCritCalculation(Target:fort_character,SourceAgent:agent, DamageDone:float):void=
        # Loading Players FireChance and FireDamage From Persistance
        if (Player := player[SourceAgent]):
            if(SuperCritChance := Data[Player].SuperCritChance):
                if(CritDamageMultiple := (Data[Player].SuperCritDamage * Data[Player].SuperCritDamagePerUpgrade)):
                    if(GetRandomFloat(0.0,100.0) < (SuperCritChance * SuperCritOddsPerUpgrade) * 100.0):
                        AfterArmor := DamageCalculationCreatures(DamageDone + (DamageDone*CritDamageMultiple))
                        SpawnDamageNumberProp(SourceAgent, AfterArmor,SuperCritNumberSpawnProp)
                        CustomDamage( Target , AfterArmor )
                        CritDamageAudio.Play(SourceAgent)
                        #HitHudDevice.ShowHitUi(SourceAgent,CritIcon,SuperCritTextColor,AfterArmor)
                        AddToDamgeMap(SourceAgent, AfterArmor)
                    else:



    UpdateHudDelay(Agent:agent,Rounded:int)<suspends>:void=
        Sleep(0.25)
        CineHudDevice.UpdateAgentUI(Agent,Rounded)
            
    ElimHandle(Interaction:device_ai_interaction_result):void=

        set NPCSAlive = NPCSAlive - 1

        Players := 	GetPlayspace().GetPlayers()
        for(Player:Players):
            if:
                Player.IsActive[]
                Agent:=agent[Player]
                Key := Agent
                Value := AgentDamageMap[Agent]
            then:

                if(IsBoss = true):
                    BossDefeatedHud.Show(Key)

                WeaponRarityDevice.CheckWeaponType(Key)
    
            
                EliminatedAudio.Play(Key)
                ElimAccolade.Award(Key)
                FirstElimAccolade.Award(Key)
                AchievementTrigger.Trigger(Key)
                #Print("Damage Saved In Map {Value}")

                var PercentOfXp : float = (Value*1.0)/TotalHealth


                # Increments Achievement if did over 33% in health damage
                if(PercentOfXp > 0.33):
                    if(PercentOfXp > 2.0): # caps the xp amount to x2 if player does 2x the health in one hit
                        set PercentOfXp = 2.0

                    
                    
                    var  BoostAmount: float = 0.0
                    if(PlayerData := Data[Key]):
                        if(PlayerData.ExpBoost > 0):
                            set BoostAmount = ElimXP *  PlayerData.ExpBoost * XpBoostInitial
                    

                    var XpGranted : float = (ElimXP   * PercentOfXp) + (BoostAmount)
                    CineHudDevice.AddXp(Key,XpGranted)
                    Print("XpGranted {XpGranted}")


                Print("ElimXp {ElimXP}")
                Print("PercentOfXp {PercentOfXp}")

                var PercentAdding : float = 0.0
                if(PlayerData := Data[Key]):
                    if(PlayerData.GoldPerElim > 0):
                        if(CosmeticGoldPerElim : float = Data[Key].CosmeticGoldPerElim / 100.0):
                            Print("Cosmetic Gold Per Elim {CosmeticGoldPerElim}")
                            # Changing exponential to Linear
                            #set PercentAdding = (Pow(GoldPerElimInitial + (PlayerData.GoldPL * PlayerData.GoldPLLevel), (PlayerData.GoldPerElim )*1.0) - 1.0)
                            set PercentAdding = (GoldPerElimInitial + (PlayerData.GoldPL * PlayerData.GoldPLLevel)) * PlayerData.GoldPerElim
                            GemGold : float = (PlayerData.PrestigeBonusPerGem/100.0) * (PlayerData.Gems) * (PercentAdding)
                            set PercentAdding = GemGold + PercentAdding
                            set PercentAdding = PercentAdding + (PercentAdding * CosmeticGoldPerElim)

                Print("PercentAdding to Elim Gold {PercentAdding} + 1.0")
                
                RoundingGoldElim : float = PercentAdding + 1.0

                GoldMult : float = (ElimGold * PercentOfXp) * RoundingGoldElim # Percent of xp = percent of damage done


                if:
                    OldData := Data[Player]
                    set Data[Player] = player_data{player_data_update<constructor>(OldData),Gold := OldData.Gold + GoldMult, PrestigeEarnings:= OldData.PrestigeEarnings + GoldMult, TotalEarnings := OldData.TotalEarnings + GoldMult}

                then:
                    CineHudDevice.GoldDrop(Key,GoldMult)
           

        # Reset Damage Map After Using it
        set AgentDamageMap  = map{}
        #Print("Npc Death")
    
        if(Agent := Interaction.Target?):
            if(ParentMoveProp.TeleportTo[DeleteLocationProp.GetTransform().Translation, DeleteLocationProp.GetTransform().Rotation]){}


    FireDamageTicks()<suspends>:void=
        loop:
            Sleep(1.0)
            var IsOnFire : logic = false

            Players := 	GetPlayspace().GetPlayers()
            for(Player:Players):
                if:
                    Player.IsActive[]
                    Agent:=agent[Player]
                    Key := Agent
                    Value := BurnStacks[Agent]
                    Value > 0
                    dAgent:=NPCAgent?
                    MyChar := dAgent.GetFortCharacter[]
                    MyChar.IsActive[]
                    MyData := Data[Player]
                    CosmeticStat := CosmeticUnlockDevice.AgentCosmeticStats[Key]
                    CosmeticFireDamage : float = CosmeticStat.StatMap["Burn Damage"] / 100.0
                    FireLevel := Data[Player].FireDamage
                    FireDamage := (( FireLevel * FireDamagePerUpgrade ) * Value)  + (FireLevel * FireDamagePerUpgrade * CosmeticFireDamage * Value * MyData.Level) + (FireLevel * FireDamagePerUpgrade * MyData.Level * Value) 
                then:
                    set IsOnFire = true
                    CustomDamage(MyChar , FireDamage)
                    FireDamageAudio.Play(Key)

                    #HitHudDevice.ShowHitUi(Key,FireIcon,FireTextColor,FireDamage)
                    SpawnDamageNumberProp(dAgent, FireDamage,FireNumberSpawnProp)
                    #DamageNumberDevice.ShowDamageNumberAtAgent(dAgent, FireDamage)

                    UpdateHealthCine(Key)
                    AddToDamgeMap(Key, FireDamage)

            if(IsOnFire = true):
                if:
                    dAgent:=NPCAgent?
                    MyChar := dAgent.GetFortCharacter[]
                then:
                    var Location :vector3= MyChar.GetTransform().Translation 
                    set Location.Z = Location.Z + CLOffset
                    Rotation := MyChar.GetTransform().Rotation
                    spawn:
                        SpawnEffect(FireNiag, 1.0, Location, Rotation, vector3{X:=1.0,Y:=1.0,Z:=1.0})
                    

    ResetBurnStacks(Id:int)<suspends>:void=
        Sleep(4.0)
        if(Id = BurnId):
            for(Key->Value:BurnStacks):
                CineHudDevice.HideHudMessage(Key,CineHudDevice.FireStackHuds, CineHudDevice.HideFiresTriggers)
                if(set BurnStacks[Key] = 0){}

    UpdateHealthCine(DamageSource:agent):void=

        var TotalStacks : int = 0
        for(Key->Value:BurnStacks):
            set TotalStacks = TotalStacks + Value
        if(TotalStacks > 0):
            CineHudDevice.UpdateBurnStacks(DamageSource, TotalStacks)
        else:
            CineHudDevice.HideHudMessage(DamageSource,CineHudDevice.FireStackHuds, CineHudDevice.HideFiresTriggers)
        

        if:
            var HealthPercent : int = Round[CurrentHealth/TotalHealth*100.0]
        then:
            if(HealthPercent < 0):
                set HealthPercent = 0
            var HealthToShow : float = CurrentHealth
            if(HealthToShow < 0.0):
                set HealthToShow = 0.0
            CineHudDevice.ShowHud(DamageSource, formatter.Abbreviate(TotalHealth,false), formatter.Abbreviate(HealthToShow,false),NPCName)
            CineHudDevice.UpdateAgentUI(DamageSource,HealthPercent)
    
    AddToDamgeMap(SourceAgent : agent, Damage : float):void=
        # initialize key in map
        if(AgentDamageMap[SourceAgent]):
        else:
            if(set AgentDamageMap[SourceAgent] = 0.0){}
            
        if(DamageCount := AgentDamageMap[SourceAgent]):
            if(set AgentDamageMap[SourceAgent]  = DamageCount + Damage){}


    SpawnEffect(Prop:creative_prop_asset, LifeTime:float, Location : vector3 , Rotatoe : rotation, Scale:vector3)<suspends>:void=

        Tran:=transform{
            Translation := Location,
            Rotation := Rotatoe,
            Scale := Scale
        }

        SpawnedPropResult := SpawnProp(Prop,Tran)
        if(SpawnedProp := SpawnedPropResult(0)?):
            spawn:
                DeleteEffect(LifeTime,SpawnedProp)

    DeleteEffect(SleepTime : float, Prop : creative_prop)<suspends>:void=
        Sleep(SleepTime)
        Prop.Dispose()

    DamageCalculationCreatures(Dmg:float):float= 
        var DamageAfterArmor: float = 0.0
        set DamageAfterArmor = (100.0 / (100.0 + (CurrentArmor* 1.0))) * Dmg
        return DamageAfterArmor

    SpawnDamageNumberProp(Agent:agent, DamageAmount:float, Prop:creative_prop_asset):void=

        if:
            MyNpcAgent:=NPCAgent?
            MyChar := MyNpcAgent.GetFortCharacter[]
            MyChar.IsActive[]
            DamageAmount >= 1.0
        then:

            RandomLocation := vector3 :
                X:=GetRandomFloat(-60.0,60.0)
                Y:=GetRandomFloat(-60.0,60.0)
                Z:=GetRandomFloat(-35.0,35.0)

            #Sleep(0.01)
            XOffset : float = 12.5
            OffToCharactersRight : float = 20.0
            HowFarInfront : float = 20.0
            HeightOffset : float = 80.0

            if (Character := MyNpcAgent.GetFortCharacter[]):
                LocationPreRandom := Character.GetTransform().Translation
                Rotation := Character.GetTransform().Rotation

                if(MyRotation := 180.0 +Character.GetViewRotation().GetYawPitchRollDegrees()[0]):
                    var Location: vector3 = vector3:
                        X:=LocationPreRandom.X + RandomLocation.X,
                        Y:=LocationPreRandom.Y + RandomLocation.Y,
                        Z:=LocationPreRandom.Z + RandomLocation.Z + HeightOffset + DamageHeightOffset
                    if(MyRotation >= 135.0 ):
                        if(MyRotation < 225.0):
                            set Location.Y = Location.Y + ( XOffset) + OffToCharactersRight
                            set Location.X = Location.X + HowFarInfront
                    if(MyRotation >= 315.0 ):
                        set Location.Y = Location.Y - ( XOffset) - OffToCharactersRight
                        set Location.X = Location.X - HowFarInfront
                    if(MyRotation < 45.0):
                        set Location.Y = Location.Y - ( XOffset) - OffToCharactersRight
                        set Location.X = Location.X - HowFarInfront
                    if(MyRotation >= 225.0 ):
                        if( MyRotation < 315.0):
                            set Location.X = Location.X - ( XOffset) - OffToCharactersRight
                            set Location.Y = Location.Y + HowFarInfront
                    if(MyRotation >= 45.0 ):
                        if(MyRotation < 135.0):
                            set Location.X = Location.X + ( XOffset) + OffToCharactersRight
                            set Location.Y = Location.Y - HowFarInfront

                    SpawnedPropResult := SpawnProp(Prop,Location,Rotation)
                    if (SpawnedProp := SpawnedPropResult(0)?):
                        Transform := SpawnedProp.GetTransform()
                        var NumberScale : vector3 = MakePriceScale(DamageAmount, ?X := 2.0)
                        set NumberScale.X = 2.0
                        
                        if:
                            SpawnedProp.TeleportTo[transform{Translation := Transform.Translation, Scale := NumberScale, Rotation := Transform.Rotation}]
                        spawn:
                            AnimateDamageNumber(SpawnedProp)



    AnimateDamageNumber(Prop:creative_prop)<suspends>:void=
        
        Sleep(0.5)

        CurrentRotation := Prop.GetTransform().Rotation
    
        var NewTranslation : vector3 = Prop.GetTransform().Translation
        set NewTranslation.Z = NewTranslation.Z + 50.0
        
        var NewScale : vector3 = Prop.GetTransform().Scale
        set NewScale.X = 0.0

        AnimationLength := 0.5
        Prop.MoveTo(transform{Translation:=NewTranslation, Scale:=NewScale, Rotation:=CurrentRotation},AnimationLength)
        Sleep(AnimationLength)

        spawn:
            DeletingNiagaraLonger(Prop,0.01)

    DeletingNiagaraLonger(Prop:creative_prop, DeleteTime : float)<suspends>:void=
        
        Sleep(DeleteTime)
        Prop.Dispose()


    