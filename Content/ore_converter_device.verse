
using { /Fortnite.com/Devices }
using { /Verse.org/Simulation }
using { /UnrealEngine.com/Temporary/Diagnostics }
using { /UnrealEngine.com/Temporary/UI }
using { /UnrealEngine.com/Temporary/SpatialMath }
using { /Fortnite.com/UI }
using { /Verse.org/Colors }
using { /Fortnite.com/Playspaces }
using { /Fortnite.com/Characters }

ore_converter_device := class(creative_device):

    @editable
    OpenShop : button_device = button_device{}
    @editable
    ShopKeeperCine : cinematic_sequence_device = cinematic_sequence_device{}
    @editable
    PurchaseCine : cinematic_sequence_device = cinematic_sequence_device{}
    @editable
    PurchaseAudio : audio_player_device = audio_player_device{}
    @editable
    NegativeAudio : audio_player_device = audio_player_device{}

    @editable
    OpenAudio : audio_player_device = audio_player_device{}
    @editable
    CloseAudio : audio_player_device = audio_player_device{}

    # Just to pull % multtipliers
    @editable
    NPCManager : npc_manager_device = npc_manager_device{}

    @editable
    CopperConditional : conditional_button_device = conditional_button_device{}
    @editable
    SilverConditional : conditional_button_device = conditional_button_device{}
    @editable
    MalachiteConditional : conditional_button_device = conditional_button_device{}
    @editable
    ObsidianConditional : conditional_button_device = conditional_button_device{}
    @editable
    BrightcoreConditional : conditional_button_device = conditional_button_device{}

    @editable
    SilverGranter : item_granter_device = item_granter_device{}
    @editable
    MalachiteGranter : item_granter_device = item_granter_device{}
    @editable
    ObsidianGranter : item_granter_device = item_granter_device{}
    @editable
    BrightcoreGranter : item_granter_device = item_granter_device{}
    @editable
    SpectroliteGranter : item_granter_device = item_granter_device{}

    var MyUI : [agent]?canvas = map{}
    var InventoryTextBlockPerAgent : [agent]?[]text_block = map{}
    var ConversionRateTextBlockPerAgent : [agent]?[]text_block = map{}

    storedstring<localizes>(S:string) : message = "{S}"
    JustIntString<localizes>(I:int) : message = "{I}"
    LevelString<localizes>(I:int) : message = "Lv {I}"
    DamageString<localizes>(I:int) : message = " {I}%"
    SecondString<localizes>(I:int) : message = "{I} Sec"
    HpString<localizes>(I:int) : message = "{I} hp"

    @editable
    Slot1Cost  : int = 7
    @editable
    Slot2Cost  : int = 10
    @editable
    Slot3Cost  : int = 13
    @editable
    Slot4Cost  : int = 15
    @editable
    Slot5Cost  : int = 20


    OnBegin<override>()<suspends>:void=

        OpenShop.InteractedWithEvent.Subscribe(ShowUi)
     

    MakeOreConverterBackground (Agent:agent) :canvas= 

        Slot1InventoryText: text_block = text_block{DefaultTextColor := MakeColorFromHex("FFFFFF"), DefaultText:=storedstring("0")}
        Slot1ConversionText: text_block = text_block{DefaultTextColor := MakeColorFromHex("FFFFFF"), DefaultText:=storedstring("{Slot1Cost}")}
        Slot2InventoryText: text_block = text_block{DefaultTextColor := MakeColorFromHex("FFFFFF"), DefaultText:=storedstring("0")}
        Slot2ConversionText: text_block = text_block{DefaultTextColor := MakeColorFromHex("FFFFFF"), DefaultText:=storedstring("{Slot2Cost}")}
        Slot3InventoryText: text_block = text_block{DefaultTextColor := MakeColorFromHex("FFFFFF"), DefaultText:=storedstring("0")}
        Slot3ConversionText: text_block = text_block{DefaultTextColor := MakeColorFromHex("FFFFFF"), DefaultText:=storedstring("{Slot3Cost}")}
        Slot4InventoryText: text_block = text_block{DefaultTextColor := MakeColorFromHex("FFFFFF"), DefaultText:=storedstring("0")}
        Slot4ConversionText: text_block = text_block{DefaultTextColor := MakeColorFromHex("FFFFFF"), DefaultText:=storedstring("{Slot4Cost}")}
        Slot5InventoryText: text_block = text_block{DefaultTextColor := MakeColorFromHex("FFFFFF"), DefaultText:=storedstring("0")}
        Slot5ConversionText: text_block = text_block{DefaultTextColor := MakeColorFromHex("FFFFFF"), DefaultText:=storedstring("{Slot5Cost}")}

        AllSlotOneText: text_block = text_block{DefaultTextColor := MakeColorFromHex("FFFFFF"), DefaultText:=storedstring("1")}
        AllSlotOneText2: text_block = text_block{DefaultTextColor := MakeColorFromHex("FFFFFF"), DefaultText:=storedstring("1")}
        AllSlotOneText3: text_block = text_block{DefaultTextColor := MakeColorFromHex("FFFFFF"), DefaultText:=storedstring("1")}
        AllSlotOneText4: text_block = text_block{DefaultTextColor := MakeColorFromHex("FFFFFF"), DefaultText:=storedstring("1")}
        AllSlotOneText5: text_block = text_block{DefaultTextColor := MakeColorFromHex("FFFFFF"), DefaultText:=storedstring("1")}


        CopperTexture: texture_block = texture_block{DefaultImage:= hud.Copper_icon512, DefaultDesiredSize:=vector2{X:=100.0,Y:=100.0}}
        CopperTextureShadow: texture_block = texture_block{DefaultImage:= hud.Copper_icon512,DefaultTint:=MakeColorFromHex("080402FF"), DefaultDesiredSize:=vector2{X:=100.0,Y:=100.0}}
        CopperTexture2: texture_block = texture_block{DefaultImage:= hud.Copper_icon512, DefaultDesiredSize:=vector2{X:=100.0,Y:=100.0}}
        CopperTextureShadow2: texture_block = texture_block{DefaultImage:= hud.Copper_icon512,DefaultTint:=MakeColorFromHex("080402FF"), DefaultDesiredSize:=vector2{X:=100.0,Y:=100.0}}
        SilverTexture: texture_block = texture_block{DefaultImage:= hud.Silver_icon, DefaultDesiredSize:=vector2{X:=100.0,Y:=100.0}}
        SilverTextureShadow: texture_block = texture_block{DefaultImage:= hud.Silver_icon,DefaultTint:=MakeColorFromHex("080402FF"), DefaultDesiredSize:=vector2{X:=100.0,Y:=100.0}}
        SilverTexture2: texture_block = texture_block{DefaultImage:= hud.Silver_icon, DefaultDesiredSize:=vector2{X:=100.0,Y:=100.0}}
        SilverTextureShadow2: texture_block = texture_block{DefaultImage:= hud.Silver_icon,DefaultTint:=MakeColorFromHex("080402FF"), DefaultDesiredSize:=vector2{X:=100.0,Y:=100.0}}
        SilverTexture3: texture_block = texture_block{DefaultImage:= hud.Silver_icon, DefaultDesiredSize:=vector2{X:=100.0,Y:=100.0}}
        SilverTextureShadow3: texture_block = texture_block{DefaultImage:= hud.Silver_icon,DefaultTint:=MakeColorFromHex("080402FF"), DefaultDesiredSize:=vector2{X:=100.0,Y:=100.0}}
        MalachiteTexture: texture_block = texture_block{DefaultImage:= hud.Malachite_icon512, DefaultDesiredSize:=vector2{X:=100.0,Y:=100.0}}
        MalachiteTextureShadow: texture_block = texture_block{DefaultImage:= hud.Malachite_icon512,DefaultTint:=MakeColorFromHex("080402FF"), DefaultDesiredSize:=vector2{X:=100.0,Y:=100.0}}
        Malachite2Texture: texture_block = texture_block{DefaultImage:= hud.Malachite_icon512, DefaultDesiredSize:=vector2{X:=100.0,Y:=100.0}}
        Malachite2TextureShadow: texture_block = texture_block{DefaultImage:= hud.Malachite_icon512,DefaultTint:=MakeColorFromHex("080402FF"), DefaultDesiredSize:=vector2{X:=100.0,Y:=100.0}}
        Malachite3Texture: texture_block = texture_block{DefaultImage:= hud.Malachite_icon512, DefaultDesiredSize:=vector2{X:=100.0,Y:=100.0}}
        Malachite3TextureShadow: texture_block = texture_block{DefaultImage:= hud.Malachite_icon512,DefaultTint:=MakeColorFromHex("080402FF"), DefaultDesiredSize:=vector2{X:=100.0,Y:=100.0}}
        ObsidianTexture: texture_block = texture_block{DefaultImage:= hud.Obsidian_icon512, DefaultDesiredSize:=vector2{X:=100.0,Y:=100.0}}
        ObsidianTextureShadow: texture_block = texture_block{DefaultImage:= hud.Obsidian_icon512,DefaultTint:=MakeColorFromHex("080402FF"), DefaultDesiredSize:=vector2{X:=100.0,Y:=100.0}}
        Obsidian2Texture: texture_block = texture_block{DefaultImage:= hud.Obsidian_icon512, DefaultDesiredSize:=vector2{X:=100.0,Y:=100.0}}
        Obsidian2TextureShadow: texture_block = texture_block{DefaultImage:= hud.Obsidian_icon512,DefaultTint:=MakeColorFromHex("080402FF"), DefaultDesiredSize:=vector2{X:=100.0,Y:=100.0}}
        Obsidian3Texture: texture_block = texture_block{DefaultImage:= hud.Obsidian_icon512, DefaultDesiredSize:=vector2{X:=100.0,Y:=100.0}}
        Obsidian3TextureShadow: texture_block = texture_block{DefaultImage:= hud.Obsidian_icon512,DefaultTint:=MakeColorFromHex("080402FF"), DefaultDesiredSize:=vector2{X:=100.0,Y:=100.0}}
        BrightcoreTexture: texture_block = texture_block{DefaultImage:= hud.Brightcore_icon512, DefaultDesiredSize:=vector2{X:=100.0,Y:=100.0}}
        BrightcoreTextureShadow: texture_block = texture_block{DefaultImage:= hud.Brightcore_icon512,DefaultTint:=MakeColorFromHex("080402FF"), DefaultDesiredSize:=vector2{X:=100.0,Y:=100.0}}
        BrightcoreTexture2: texture_block = texture_block{DefaultImage:= hud.Brightcore_icon512, DefaultDesiredSize:=vector2{X:=100.0,Y:=100.0}}
        BrightcoreTextureShadow2: texture_block = texture_block{DefaultImage:= hud.Brightcore_icon512,DefaultTint:=MakeColorFromHex("080402FF"), DefaultDesiredSize:=vector2{X:=100.0,Y:=100.0}}
        BrightcoreTexture3: texture_block = texture_block{DefaultImage:= hud.Brightcore_icon512, DefaultDesiredSize:=vector2{X:=100.0,Y:=100.0}}
        BrightcoreTextureShadow3: texture_block = texture_block{DefaultImage:= hud.Brightcore_icon512,DefaultTint:=MakeColorFromHex("080402FF"), DefaultDesiredSize:=vector2{X:=100.0,Y:=100.0}}
        SpectroliteTexture: texture_block = texture_block{DefaultImage:= hud.Spectrolite_icon512, DefaultDesiredSize:=vector2{X:=100.0,Y:=100.0}}
        SpectroliteTextureShadow: texture_block = texture_block{DefaultImage:= hud.Spectrolite_icon512,DefaultTint:=MakeColorFromHex("080402FF"), DefaultDesiredSize:=vector2{X:=100.0,Y:=100.0}}
        SpectroliteTexture2: texture_block = texture_block{DefaultImage:= hud.Spectrolite_icon512, DefaultDesiredSize:=vector2{X:=100.0,Y:=100.0}}
        SpectroliteTextureShadow2: texture_block = texture_block{DefaultImage:= hud.Spectrolite_icon512,DefaultTint:=MakeColorFromHex("080402FF"), DefaultDesiredSize:=vector2{X:=100.0,Y:=100.0}}
        SpectroliteTexture3: texture_block = texture_block{DefaultImage:= hud.Spectrolite_icon512, DefaultDesiredSize:=vector2{X:=100.0,Y:=100.0}}
        SpectroliteTextureShadow3: texture_block = texture_block{DefaultImage:= hud.Spectrolite_icon512,DefaultTint:=MakeColorFromHex("080402FF"), DefaultDesiredSize:=vector2{X:=100.0,Y:=100.0}}
        ArrowTexture: texture_block = texture_block{DefaultImage:= wooden_pixel_gui.textures.Icons2.Size32.t_ArrowRight_32, DefaultDesiredSize:=vector2{X:=100.0,Y:=100.0}}
        ArrowTexture2: texture_block = texture_block{DefaultImage:= wooden_pixel_gui.textures.Icons2.Size32.t_ArrowRight_32, DefaultDesiredSize:=vector2{X:=100.0,Y:=100.0}}
        ArrowTexture3: texture_block = texture_block{DefaultImage:= wooden_pixel_gui.textures.Icons2.Size32.t_ArrowRight_32, DefaultDesiredSize:=vector2{X:=100.0,Y:=100.0}}
        ArrowTexture4: texture_block = texture_block{DefaultImage:= wooden_pixel_gui.textures.Icons2.Size32.t_ArrowRight_32, DefaultDesiredSize:=vector2{X:=100.0,Y:=100.0}}
        ArrowTexture5: texture_block = texture_block{DefaultImage:= wooden_pixel_gui.textures.Icons2.Size32.t_ArrowRight_32, DefaultDesiredSize:=vector2{X:=100.0,Y:=100.0}}

        
        Slot1Button : button_quiet = button_quiet{DefaultText:=storedstring("Convert")}
        Slot1Button.OnClick().Subscribe(Convert1)
        Slot2Button : button_quiet = button_quiet{DefaultText:=storedstring("Convert")}
        Slot2Button.OnClick().Subscribe(Convert2) 
        Slot3Button : button_quiet = button_quiet{DefaultText:=storedstring("Convert")}
        Slot3Button.OnClick().Subscribe(Convert3) 
        Slot4Button : button_quiet = button_quiet{DefaultText:=storedstring("Convert")}
        Slot4Button.OnClick().Subscribe(Convert4) 
        Slot5Button : button_quiet = button_quiet{DefaultText:=storedstring("Convert")}
        Slot5Button.OnClick().Subscribe(Convert5)

        CloseButton : button_quiet = button_quiet{DefaultText:=storedstring("Close")}
        CloseButton.OnClick().Subscribe(RemoveHUD) 

        if(set InventoryTextBlockPerAgent[Agent] = option{array{Slot1InventoryText,Slot2InventoryText,Slot3InventoryText,Slot4InventoryText,Slot5InventoryText}}){} 
        if(set ConversionRateTextBlockPerAgent[Agent] = option{array{Slot1ConversionText,Slot2ConversionText,Slot3ConversionText,Slot4ConversionText,Slot5ConversionText}}){} 


        MyCanvas:=canvas:
            Slots:=array:
                canvas_slot:
                    Anchors:=anchors{Minimum:=vector2{X:=0.2,Y:=0.05}, Maximum:=vector2{X:=0.8,Y:=0.95}}
                    Widget:=overlay:
                        Slots:=array:
                            overlay_slot:
                                HorizontalAlignment:=horizontal_alignment.Fill
                                VerticalAlignment:=vertical_alignment.Fill
                                Padding:=margin{Left:=40.0,Top:=40.0,Right:=40.0,Bottom:=40.0}
                                Widget:=texture_block:
                                    DefaultImage:=wooden_pixel_gui.textures.Panels.BrownSquare
                                    DefaultDesiredSize:=vector2{X:=50.0,Y:=110.0}
                            overlay_slot:
                                HorizontalAlignment:=horizontal_alignment.Fill
                                VerticalAlignment:=vertical_alignment.Fill
                                Widget:=stack_box:
                                    Orientation:=orientation.Vertical
                                    Slots:=array:
                                        stack_box_slot:
                                            HorizontalAlignment:=horizontal_alignment.Fill
                                            VerticalAlignment:=vertical_alignment.Fill
                                            Padding:=margin{Left:=40.0,Top:=40.0,Right:=40.0,Bottom:=0.0}
                                            Widget:=overlay:
                                                Slots:=array:
                                                    overlay_slot:
                                                        HorizontalAlignment:=horizontal_alignment.Fill
                                                        VerticalAlignment:=vertical_alignment.Fill
                                                        Widget:=texture_block:
                                                            DefaultImage:=wooden_pixel_gui.textures.Panels.t_TitleBar
                                                            DefaultDesiredSize:=vector2{X:=15.0,Y:=60.0}
                                                    overlay_slot:
                                                        HorizontalAlignment:=horizontal_alignment.Center
                                                        VerticalAlignment:=vertical_alignment.Center
                                                        Widget:=text_block:
                                                            DefaultText:=storedstring("Ore Converter")
                                                            DefaultTextColor := MakeColorFromHex("401805FF")
                                        stack_box_slot:
                                            HorizontalAlignment:=horizontal_alignment.Fill
                                            VerticalAlignment:=vertical_alignment.Fill
                                            Padding:=margin{Left:=50.0,Top:=0.0,Right:=50.0,Bottom:=50.0}
                                            Widget:=overlay:
                                                Slots:=array:
                                                    overlay_slot:
                                                        HorizontalAlignment:=horizontal_alignment.Fill
                                                        VerticalAlignment:=vertical_alignment.Fill
                                                        Widget:=texture_block:
                                                            DefaultImage:=wooden_pixel_gui.textures.Panels.BrownSquare
                                                            DefaultDesiredSize:=vector2{X:=32.0,Y:=32.0}
                                                    overlay_slot:
                                                        HorizontalAlignment:=horizontal_alignment.Fill
                                                        VerticalAlignment:=vertical_alignment.Fill
                                                        Widget:=stack_box:
                                                            Orientation:=orientation.Vertical
                                                            Slots:=array:
                                                                stack_box_slot: # Main List
                                                                    HorizontalAlignment:=horizontal_alignment.Fill
                                                                    VerticalAlignment:=vertical_alignment.Fill
                                                                    Widget:=overlay:
                                                                        Slots:=array:
                                                                            overlay_slot:
                                                                                HorizontalAlignment:=horizontal_alignment.Left
                                                                                VerticalAlignment:=vertical_alignment.Center
                                                                                Padding:=margin{Left:=30.0,Top:=0.0,Right:=0.0,Bottom:=0.0}
                                                                                Widget:=overlay:
                                                                                    Slots:=array:
                                                                                        overlay_slot:
                                                                                            Padding:=margin{Left:=3.0,Top:=3.0,Right:=0.0,Bottom:=0.0}
                                                                                            Widget:=text_block:
                                                                                                DefaultText:=storedstring("Inventory")
                                                                                                DefaultTextColor := MakeColorFromHex("142706FF")
                                                                                        overlay_slot:
                                                                                            Widget:=text_block:
                                                                                                DefaultText:=storedstring("Inventory")
                                                                                                DefaultTextColor := MakeColorFromHex("529F17FF")
                                                                            overlay_slot:
                                                                                HorizontalAlignment:=horizontal_alignment.Center
                                                                                VerticalAlignment:=vertical_alignment.Center
                                                                                Widget:=overlay:
                                                                                    Slots:=array:
                                                                                        overlay_slot:
                                                                                            Padding:=margin{Left:=3.0,Top:=3.0,Right:=0.0,Bottom:=0.0}
                                                                                            Widget:=text_block:
                                                                                                DefaultText:=storedstring("Conversion")
                                                                                                DefaultTextColor := MakeColorFromHex("142706FF")
                                                                                        overlay_slot:
                                                                                            Widget:=text_block:
                                                                                                DefaultText:=storedstring("Conversion")
                                                                                                DefaultTextColor := MakeColorFromHex("529F17FF")
                                                                stack_box_slot: # Conversion1
                                                                    HorizontalAlignment:=horizontal_alignment.Fill
                                                                    VerticalAlignment:=vertical_alignment.Fill
                                                                    Padding:=margin{Left:=0.0,Top:=10.0,Right:=0.0,Bottom:=0.0}
                                                                    Widget:=overlay:
                                                                        Slots:=array:
                                                                            overlay_slot:
                                                                                HorizontalAlignment:=horizontal_alignment.Fill
                                                                                VerticalAlignment:=vertical_alignment.Fill
                                                                                Widget:=texture_block:
                                                                                    DefaultImage:=wooden_pixel_gui.textures.Panels.BrownSquare
                                                                                    DefaultTint := MakeColorFromHex("A9A9A9FF")
                                                                            overlay_slot:
                                                                                HorizontalAlignment:=horizontal_alignment.Right
                                                                                VerticalAlignment:=vertical_alignment.Center
                                                                                Padding:=margin{Left:=0.0,Top:=0.0,Right:=20.0,Bottom:=0.0}
                                                                                Widget:=Slot1Button
                                                                            overlay_slot:
                                                                                HorizontalAlignment:=horizontal_alignment.Center
                                                                                VerticalAlignment:=vertical_alignment.Center
                                                                                Widget:=stack_box:
                                                                                    Orientation:=orientation.Horizontal
                                                                                    Slots:=array:
                                                                                        stack_box_slot:
                                                                                            Widget:=overlay:
                                                                                                Slots:=array:
                                                                                                    overlay_slot:
                                                                                                        HorizontalAlignment:=horizontal_alignment.Fill
                                                                                                        VerticalAlignment:=vertical_alignment.Fill
                                                                                                        Widget:=CopperTextureShadow
                                                                                                    overlay_slot:
                                                                                                        HorizontalAlignment:=horizontal_alignment.Fill
                                                                                                        VerticalAlignment:=vertical_alignment.Fill
                                                                                                        Widget:=CopperTexture
                                                                                        stack_box_slot:
                                                                                            HorizontalAlignment:=horizontal_alignment.Center
                                                                                            VerticalAlignment:=vertical_alignment.Center
                                                                                            Widget:=Slot1ConversionText
                                                                                        stack_box_slot:
                                                                                            Widget:=overlay:
                                                                                                Slots:=array:
                                                                                                    overlay_slot:
                                                                                                        HorizontalAlignment:=horizontal_alignment.Fill
                                                                                                        VerticalAlignment:=vertical_alignment.Fill
                                                                                                        Padding:=margin{Left:=10.0,Top:=0.0,Right:=10.0,Bottom:=0.0}
                                                                                                        Widget:=ArrowTexture
                                                                                        stack_box_slot:
                                                                                            Widget:=overlay:
                                                                                                Slots:=array:
                                                                                                    overlay_slot:
                                                                                                        HorizontalAlignment:=horizontal_alignment.Fill
                                                                                                        VerticalAlignment:=vertical_alignment.Fill
                                                                                                        Widget:=SilverTextureShadow
                                                                                                    overlay_slot:
                                                                                                        HorizontalAlignment:=horizontal_alignment.Fill
                                                                                                        VerticalAlignment:=vertical_alignment.Fill
                                                                                                        Widget:=SilverTexture
                                                                                        stack_box_slot:
                                                                                            HorizontalAlignment:=horizontal_alignment.Center
                                                                                            VerticalAlignment:=vertical_alignment.Center
                                                                                            Widget:=AllSlotOneText
                                                                            overlay_slot:
                                                                                HorizontalAlignment:=horizontal_alignment.Left
                                                                                VerticalAlignment:=vertical_alignment.Center
                                                                                Widget:=stack_box:
                                                                                    Orientation:=orientation.Horizontal
                                                                                    Slots:=array:
                                                                                        stack_box_slot:
                                                                                            Widget:=overlay:
                                                                                                Slots:=array:
                                                                                                    overlay_slot:
                                                                                                        HorizontalAlignment:=horizontal_alignment.Fill
                                                                                                        VerticalAlignment:=vertical_alignment.Fill
                                                                                                        Widget:=CopperTextureShadow2
                                                                                                    overlay_slot:
                                                                                                        HorizontalAlignment:=horizontal_alignment.Fill
                                                                                                        VerticalAlignment:=vertical_alignment.Fill
                                                                                                        Widget:=CopperTexture2
                                                                                        stack_box_slot:
                                                                                            HorizontalAlignment:=horizontal_alignment.Fill
                                                                                            VerticalAlignment:=vertical_alignment.Center
                                                                                            Widget:=Slot1InventoryText
                                                                stack_box_slot: # Conversion2
                                                                    HorizontalAlignment:=horizontal_alignment.Fill
                                                                    VerticalAlignment:=vertical_alignment.Fill
                                                                    Padding:=margin{Left:=0.0,Top:=10.0,Right:=0.0,Bottom:=0.0}
                                                                    Widget:=overlay:
                                                                        Slots:=array:
                                                                            overlay_slot:
                                                                                HorizontalAlignment:=horizontal_alignment.Fill
                                                                                VerticalAlignment:=vertical_alignment.Fill
                                                                                Widget:=texture_block:
                                                                                    DefaultImage:=wooden_pixel_gui.textures.Panels.BrownSquare
                                                                                    DefaultTint := MakeColorFromHex("A9A9A9FF")
                                                                            overlay_slot:
                                                                                HorizontalAlignment:=horizontal_alignment.Right
                                                                                VerticalAlignment:=vertical_alignment.Center
                                                                                Padding:=margin{Left:=0.0,Top:=0.0,Right:=20.0,Bottom:=0.0}
                                                                                Widget:=Slot2Button
                                                                            overlay_slot:
                                                                                HorizontalAlignment:=horizontal_alignment.Center
                                                                                VerticalAlignment:=vertical_alignment.Center
                                                                                Widget:=stack_box:
                                                                                    Orientation:=orientation.Horizontal
                                                                                    Slots:=array:
                                                                                        stack_box_slot:
                                                                                            Widget:=overlay:
                                                                                                Slots:=array:
                                                                                                    overlay_slot:
                                                                                                        HorizontalAlignment:=horizontal_alignment.Fill
                                                                                                        VerticalAlignment:=vertical_alignment.Fill
                                                                                                        Widget:=SilverTextureShadow3
                                                                                                    overlay_slot:
                                                                                                        HorizontalAlignment:=horizontal_alignment.Fill
                                                                                                        VerticalAlignment:=vertical_alignment.Fill
                                                                                                        Widget:=SilverTexture3
                                                                                        stack_box_slot:
                                                                                            HorizontalAlignment:=horizontal_alignment.Center
                                                                                            VerticalAlignment:=vertical_alignment.Center
                                                                                            Widget:=Slot2ConversionText
                                                                                        stack_box_slot:
                                                                                            Widget:=overlay:
                                                                                                Slots:=array:
                                                                                                    overlay_slot:
                                                                                                        HorizontalAlignment:=horizontal_alignment.Fill
                                                                                                        VerticalAlignment:=vertical_alignment.Fill
                                                                                                        Padding:=margin{Left:=10.0,Top:=0.0,Right:=10.0,Bottom:=0.0}
                                                                                                        Widget:=ArrowTexture2
                                                                                        stack_box_slot:
                                                                                            Widget:=overlay:
                                                                                                Slots:=array:
                                                                                                    overlay_slot:
                                                                                                        HorizontalAlignment:=horizontal_alignment.Fill
                                                                                                        VerticalAlignment:=vertical_alignment.Fill
                                                                                                        Widget:=MalachiteTextureShadow
                                                                                                    overlay_slot:
                                                                                                        HorizontalAlignment:=horizontal_alignment.Fill
                                                                                                        VerticalAlignment:=vertical_alignment.Fill
                                                                                                        Widget:=MalachiteTexture
                                                                                        stack_box_slot:
                                                                                            HorizontalAlignment:=horizontal_alignment.Center
                                                                                            VerticalAlignment:=vertical_alignment.Center
                                                                                            Widget:=AllSlotOneText2
                                                                            overlay_slot:
                                                                                HorizontalAlignment:=horizontal_alignment.Left
                                                                                VerticalAlignment:=vertical_alignment.Center
                                                                                Widget:=stack_box:
                                                                                    Orientation:=orientation.Horizontal
                                                                                    Slots:=array:
                                                                                        stack_box_slot:
                                                                                            Widget:=overlay:
                                                                                                Slots:=array:
                                                                                                    overlay_slot:
                                                                                                        HorizontalAlignment:=horizontal_alignment.Fill
                                                                                                        VerticalAlignment:=vertical_alignment.Fill
                                                                                                        Widget:=SilverTextureShadow2
                                                                                                    overlay_slot:
                                                                                                        HorizontalAlignment:=horizontal_alignment.Fill
                                                                                                        VerticalAlignment:=vertical_alignment.Fill
                                                                                                        Widget:=SilverTexture2
                                                                                        stack_box_slot:
                                                                                            HorizontalAlignment:=horizontal_alignment.Fill
                                                                                            VerticalAlignment:=vertical_alignment.Center
                                                                                            Widget:=Slot2InventoryText                            
                                                                stack_box_slot: # Conversion3
                                                                    HorizontalAlignment:=horizontal_alignment.Fill
                                                                    VerticalAlignment:=vertical_alignment.Fill
                                                                    Padding:=margin{Left:=0.0,Top:=10.0,Right:=0.0,Bottom:=0.0}
                                                                    Widget:=overlay:
                                                                        Slots:=array:
                                                                            overlay_slot:
                                                                                HorizontalAlignment:=horizontal_alignment.Fill
                                                                                VerticalAlignment:=vertical_alignment.Fill
                                                                                Widget:=texture_block:
                                                                                    DefaultImage:=wooden_pixel_gui.textures.Panels.BrownSquare
                                                                                    DefaultTint := MakeColorFromHex("A9A9A9FF")
                                                                            overlay_slot:
                                                                                HorizontalAlignment:=horizontal_alignment.Right
                                                                                VerticalAlignment:=vertical_alignment.Center
                                                                                Padding:=margin{Left:=0.0,Top:=0.0,Right:=20.0,Bottom:=0.0}
                                                                                Widget:=Slot3Button
                                                                            overlay_slot:
                                                                                HorizontalAlignment:=horizontal_alignment.Center
                                                                                VerticalAlignment:=vertical_alignment.Center
                                                                                Widget:=stack_box:
                                                                                    Orientation:=orientation.Horizontal
                                                                                    Slots:=array:
                                                                                        stack_box_slot:
                                                                                            Widget:=overlay:
                                                                                                Slots:=array:
                                                                                                    overlay_slot:
                                                                                                        HorizontalAlignment:=horizontal_alignment.Fill
                                                                                                        VerticalAlignment:=vertical_alignment.Fill
                                                                                                        Widget:=Malachite3TextureShadow
                                                                                                    overlay_slot:
                                                                                                        HorizontalAlignment:=horizontal_alignment.Fill
                                                                                                        VerticalAlignment:=vertical_alignment.Fill
                                                                                                        Widget:=Malachite3Texture
                                                                                        stack_box_slot:
                                                                                            HorizontalAlignment:=horizontal_alignment.Center
                                                                                            VerticalAlignment:=vertical_alignment.Center
                                                                                            Widget:=Slot3ConversionText
                                                                                        stack_box_slot:
                                                                                            Widget:=overlay:
                                                                                                Slots:=array:
                                                                                                    overlay_slot:
                                                                                                        HorizontalAlignment:=horizontal_alignment.Fill
                                                                                                        VerticalAlignment:=vertical_alignment.Fill
                                                                                                        Padding:=margin{Left:=10.0,Top:=0.0,Right:=10.0,Bottom:=0.0}
                                                                                                        Widget:=ArrowTexture3
                                                                                        stack_box_slot:
                                                                                            Widget:=overlay:
                                                                                                Slots:=array:
                                                                                                    overlay_slot:
                                                                                                        HorizontalAlignment:=horizontal_alignment.Fill
                                                                                                        VerticalAlignment:=vertical_alignment.Fill
                                                                                                        Widget:=ObsidianTextureShadow
                                                                                                    overlay_slot:
                                                                                                        HorizontalAlignment:=horizontal_alignment.Fill
                                                                                                        VerticalAlignment:=vertical_alignment.Fill
                                                                                                        Widget:=ObsidianTexture
                                                                                        stack_box_slot:
                                                                                            HorizontalAlignment:=horizontal_alignment.Center
                                                                                            VerticalAlignment:=vertical_alignment.Center
                                                                                            Widget:=AllSlotOneText3
                                                                            overlay_slot:
                                                                                HorizontalAlignment:=horizontal_alignment.Left
                                                                                VerticalAlignment:=vertical_alignment.Center
                                                                                Widget:=stack_box:
                                                                                    Orientation:=orientation.Horizontal
                                                                                    Slots:=array:
                                                                                        stack_box_slot:
                                                                                            Widget:=overlay:
                                                                                                Slots:=array:
                                                                                                    overlay_slot:
                                                                                                        HorizontalAlignment:=horizontal_alignment.Fill
                                                                                                        VerticalAlignment:=vertical_alignment.Fill
                                                                                                        Widget:=Malachite2TextureShadow
                                                                                                    overlay_slot:
                                                                                                        HorizontalAlignment:=horizontal_alignment.Fill
                                                                                                        VerticalAlignment:=vertical_alignment.Fill
                                                                                                        Widget:=Malachite2Texture
                                                                                        stack_box_slot:
                                                                                            HorizontalAlignment:=horizontal_alignment.Fill
                                                                                            VerticalAlignment:=vertical_alignment.Center
                                                                                            Widget:=Slot3InventoryText        
                                                                stack_box_slot: # Conversion4
                                                                    HorizontalAlignment:=horizontal_alignment.Fill
                                                                    VerticalAlignment:=vertical_alignment.Fill
                                                                    Padding:=margin{Left:=0.0,Top:=10.0,Right:=0.0,Bottom:=0.0}
                                                                    Widget:=overlay:
                                                                        Slots:=array:
                                                                            overlay_slot:
                                                                                HorizontalAlignment:=horizontal_alignment.Fill
                                                                                VerticalAlignment:=vertical_alignment.Fill
                                                                                Widget:=texture_block:
                                                                                    DefaultImage:=wooden_pixel_gui.textures.Panels.BrownSquare
                                                                                    DefaultTint := MakeColorFromHex("A9A9A9FF")
                                                                            overlay_slot:
                                                                                HorizontalAlignment:=horizontal_alignment.Right
                                                                                VerticalAlignment:=vertical_alignment.Center
                                                                                Padding:=margin{Left:=0.0,Top:=0.0,Right:=20.0,Bottom:=0.0}
                                                                                Widget:=Slot4Button
                                                                            overlay_slot:
                                                                                HorizontalAlignment:=horizontal_alignment.Center
                                                                                VerticalAlignment:=vertical_alignment.Center
                                                                                Widget:=stack_box:
                                                                                    Orientation:=orientation.Horizontal
                                                                                    Slots:=array:
                                                                                        stack_box_slot:
                                                                                            Widget:=overlay:
                                                                                                Slots:=array:
                                                                                                    overlay_slot:
                                                                                                        HorizontalAlignment:=horizontal_alignment.Fill
                                                                                                        VerticalAlignment:=vertical_alignment.Fill
                                                                                                        Widget:=Obsidian2TextureShadow
                                                                                                    overlay_slot:
                                                                                                        HorizontalAlignment:=horizontal_alignment.Fill
                                                                                                        VerticalAlignment:=vertical_alignment.Fill
                                                                                                        Widget:=Obsidian2Texture
                                                                                        stack_box_slot:
                                                                                            HorizontalAlignment:=horizontal_alignment.Center
                                                                                            VerticalAlignment:=vertical_alignment.Center
                                                                                            Widget:=Slot4ConversionText
                                                                                        stack_box_slot:
                                                                                            Widget:=overlay:
                                                                                                Slots:=array:
                                                                                                    overlay_slot:
                                                                                                        HorizontalAlignment:=horizontal_alignment.Fill
                                                                                                        VerticalAlignment:=vertical_alignment.Fill
                                                                                                        Padding:=margin{Left:=10.0,Top:=0.0,Right:=10.0,Bottom:=0.0}
                                                                                                        Widget:=ArrowTexture4
                                                                                        stack_box_slot:
                                                                                            Widget:=overlay:
                                                                                                Slots:=array:
                                                                                                    overlay_slot:
                                                                                                        HorizontalAlignment:=horizontal_alignment.Fill
                                                                                                        VerticalAlignment:=vertical_alignment.Fill
                                                                                                        Widget:=BrightcoreTextureShadow
                                                                                                    overlay_slot:
                                                                                                        HorizontalAlignment:=horizontal_alignment.Fill
                                                                                                        VerticalAlignment:=vertical_alignment.Fill
                                                                                                        Widget:=BrightcoreTexture
                                                                                        stack_box_slot:
                                                                                            HorizontalAlignment:=horizontal_alignment.Center
                                                                                            VerticalAlignment:=vertical_alignment.Center
                                                                                            Widget:=AllSlotOneText4
                                                                            overlay_slot:
                                                                                HorizontalAlignment:=horizontal_alignment.Left
                                                                                VerticalAlignment:=vertical_alignment.Center
                                                                                Widget:=stack_box:
                                                                                    Orientation:=orientation.Horizontal
                                                                                    Slots:=array:
                                                                                        stack_box_slot:
                                                                                            Widget:=overlay:
                                                                                                Slots:=array:
                                                                                                    overlay_slot:
                                                                                                        HorizontalAlignment:=horizontal_alignment.Fill
                                                                                                        VerticalAlignment:=vertical_alignment.Fill
                                                                                                        Widget:=Obsidian3TextureShadow
                                                                                                    overlay_slot:
                                                                                                        HorizontalAlignment:=horizontal_alignment.Fill
                                                                                                        VerticalAlignment:=vertical_alignment.Fill
                                                                                                        Widget:=Obsidian3Texture
                                                                                        stack_box_slot:
                                                                                            HorizontalAlignment:=horizontal_alignment.Fill
                                                                                            VerticalAlignment:=vertical_alignment.Center
                                                                                            Widget:=Slot4InventoryText      
                                                                stack_box_slot: # Conversion 5
                                                                    HorizontalAlignment:=horizontal_alignment.Fill
                                                                    VerticalAlignment:=vertical_alignment.Fill
                                                                    Padding:=margin{Left:=0.0,Top:=10.0,Right:=0.0,Bottom:=0.0}
                                                                    Widget:=overlay:
                                                                        Slots:=array:
                                                                            overlay_slot:
                                                                                HorizontalAlignment:=horizontal_alignment.Fill
                                                                                VerticalAlignment:=vertical_alignment.Fill
                                                                                Widget:=texture_block:
                                                                                    DefaultImage:=wooden_pixel_gui.textures.Panels.BrownSquare
                                                                                    DefaultTint := MakeColorFromHex("A9A9A9FF")
                                                                            overlay_slot:
                                                                                HorizontalAlignment:=horizontal_alignment.Right
                                                                                VerticalAlignment:=vertical_alignment.Center
                                                                                Padding:=margin{Left:=0.0,Top:=0.0,Right:=20.0,Bottom:=0.0}
                                                                                Widget:=Slot5Button
                                                                            overlay_slot:
                                                                                HorizontalAlignment:=horizontal_alignment.Center
                                                                                VerticalAlignment:=vertical_alignment.Center
                                                                                Widget:=stack_box:
                                                                                    Orientation:=orientation.Horizontal
                                                                                    Slots:=array:
                                                                                        stack_box_slot:
                                                                                            Widget:=overlay:
                                                                                                Slots:=array:
                                                                                                    overlay_slot:
                                                                                                        HorizontalAlignment:=horizontal_alignment.Fill
                                                                                                        VerticalAlignment:=vertical_alignment.Fill
                                                                                                        Widget:=BrightcoreTextureShadow2
                                                                                                    overlay_slot:
                                                                                                        HorizontalAlignment:=horizontal_alignment.Fill
                                                                                                        VerticalAlignment:=vertical_alignment.Fill
                                                                                                        Widget:=BrightcoreTexture2
                                                                                        stack_box_slot:
                                                                                            HorizontalAlignment:=horizontal_alignment.Center
                                                                                            VerticalAlignment:=vertical_alignment.Center
                                                                                            Widget:=Slot5ConversionText
                                                                                        stack_box_slot:
                                                                                            Widget:=overlay:
                                                                                                Slots:=array:
                                                                                                    overlay_slot:
                                                                                                        HorizontalAlignment:=horizontal_alignment.Fill
                                                                                                        VerticalAlignment:=vertical_alignment.Fill
                                                                                                        Padding:=margin{Left:=10.0,Top:=0.0,Right:=10.0,Bottom:=0.0}
                                                                                                        Widget:=ArrowTexture5
                                                                                        stack_box_slot:
                                                                                            Widget:=overlay:
                                                                                                Slots:=array:
                                                                                                    overlay_slot:
                                                                                                        HorizontalAlignment:=horizontal_alignment.Fill
                                                                                                        VerticalAlignment:=vertical_alignment.Fill
                                                                                                        Widget:=SpectroliteTextureShadow
                                                                                                    overlay_slot:
                                                                                                        HorizontalAlignment:=horizontal_alignment.Fill
                                                                                                        VerticalAlignment:=vertical_alignment.Fill
                                                                                                        Widget:=SpectroliteTexture
                                                                                        stack_box_slot:
                                                                                            HorizontalAlignment:=horizontal_alignment.Center
                                                                                            VerticalAlignment:=vertical_alignment.Center
                                                                                            Widget:=AllSlotOneText5
                                                                            overlay_slot:
                                                                                HorizontalAlignment:=horizontal_alignment.Left
                                                                                VerticalAlignment:=vertical_alignment.Center
                                                                                Widget:=stack_box:
                                                                                    Orientation:=orientation.Horizontal
                                                                                    Slots:=array:
                                                                                        stack_box_slot:
                                                                                            Widget:=overlay:
                                                                                                Slots:=array:
                                                                                                    overlay_slot:
                                                                                                        HorizontalAlignment:=horizontal_alignment.Fill
                                                                                                        VerticalAlignment:=vertical_alignment.Fill
                                                                                                        Widget:=BrightcoreTextureShadow3
                                                                                                    overlay_slot:
                                                                                                        HorizontalAlignment:=horizontal_alignment.Fill
                                                                                                        VerticalAlignment:=vertical_alignment.Fill
                                                                                                        Widget:=BrightcoreTexture3
                                                                                        stack_box_slot:
                                                                                            HorizontalAlignment:=horizontal_alignment.Fill
                                                                                            VerticalAlignment:=vertical_alignment.Center
                                                                                            Widget:=Slot5InventoryText                                        
                                                                stack_box_slot:
                                                                    HorizontalAlignment:=horizontal_alignment.Right
                                                                    VerticalAlignment:=vertical_alignment.Bottom
                                                                    Padding:=margin{Left:=0.0,Top:=20.0,Right:=0.0,Bottom:=0.0}
                                                                    Widget:=CloseButton
        return MyCanvas

    CreateCenteredOverlay(TheWidget:widget):overlay=
        overlay:
            Slots:=array:
                overlay_slot:
                    Widget:= TheWidget
                    HorizontalAlignment:=horizontal_alignment.Center
                    VerticalAlignment:=vertical_alignment.Center
        

    ShowUi(Agent:agent):void=

        RemoveFortHud(Agent)
        ShopKeeperCine.Play(Agent)
        OpenAudio.Play(Agent)

        MyCanvas: canvas = MakeOreConverterBackground(Agent)
        
        if(PlayerUI:=GetPlayerUI[player[Agent]]):

            #Removing Old Widget to stop duplicates
            if(AddedCanvas := MyUI[Agent]?):
                PlayerUI.RemoveWidget(AddedCanvas)
                AddedCanvas.SetVisibility(widget_visibility.Visible)
                AddedCanvas.SetVisibility(widget_visibility.Hidden)

            PlayerUI.AddWidget(MyCanvas,player_ui_slot{InputMode:=ui_input_mode.All}) # want it to appear on top
            UpdateUiText(Agent)
            MyCanvas.SetVisibility(widget_visibility.Hidden)
            MyCanvas.SetVisibility(widget_visibility.Visible)
        if(set MyUI[Agent]= option{MyCanvas}){}

    RemoveHUD(Wdata:widget_message):void =

        Print("Removing")
        if(Player:=Wdata.Player, PlayerUI:=GetPlayerUI[Player],SelectedButton:=Wdata.Source,Agent:=agent[Player]):
            if(MyActualUI := MyUI[Agent]?):

                CloseAudio.Play(Agent)

                PlayerUI.RemoveWidget(MyActualUI)
                MyActualUI.SetVisibility(widget_visibility.Visible)
                MyActualUI.SetVisibility(widget_visibility.Hidden)

                if(set MyUI[Agent] = false){}
                if(set InventoryTextBlockPerAgent[Agent] = false){}
                if(set ConversionRateTextBlockPerAgent[Agent] = false){}
                
                spawn:
                    CloseCineDelay(Agent)
                

    CloseCineDelay(Agent:agent)<suspends>:void=

        ShowFortHud(Agent)
        ShopKeeperCine.Stop(Agent)

    UpdateUiText(Agent:agent):void=

        Print("We In here")
        if:
            InventoryText1: text_block = InventoryTextBlockPerAgent[Agent]?[0]
            InventoryText2: text_block = InventoryTextBlockPerAgent[Agent]?[1]
            InventoryText3: text_block = InventoryTextBlockPerAgent[Agent]?[2]
            InventoryText4: text_block = InventoryTextBlockPerAgent[Agent]?[3]
            InventoryText5: text_block = InventoryTextBlockPerAgent[Agent]?[4]

            Player := player[Agent]
            PlayerData := Data[Agent]
            
        then:
            CopperCount := CopperConditional.GetItemCount(Agent,0)
            Print("CopperCount := {CopperCount}")
            InventoryText1.SetText(JustIntString(CopperCount))

            SilverCount := SilverConditional.GetItemCount(Agent,0)
            InventoryText2.SetText(JustIntString(SilverCount))
            Print("SilverCount := {SilverCount}")

            MalachiteCount := MalachiteConditional.GetItemCount(Agent,0)
            InventoryText3.SetText(JustIntString(MalachiteCount))
            Print("MalachiteCount := {MalachiteCount}")

            ObsidianCount := ObsidianConditional.GetItemCount(Agent,0)
            InventoryText4.SetText(JustIntString(ObsidianCount))
            Print("MalachiteCount := {ObsidianCount}")

            BrightcoreCount := BrightcoreConditional.GetItemCount(Agent,0)
            InventoryText5.SetText(JustIntString(BrightcoreCount))
            Print("MalachiteCount := {BrightcoreCount}")
            

    Convert1(Wdata:widget_message):void=
        if(Player:=Wdata.Player, PlayerUI:=GetPlayerUI[Player],SelectedButton:=Wdata.Source,Agent:=agent[Player]):
            if(CopperConditional.HasAllItems[Agent]):
                CopperConditional.Activate(Agent)
                SilverGranter.GrantItem(Agent)
                CompleteConversion(Agent)
                spawn{DelayedUpdate(Agent)}
            else:
                NegativeAudio.Play(Agent)

    Convert2(Wdata:widget_message):void=
        if(Player:=Wdata.Player, PlayerUI:=GetPlayerUI[Player],SelectedButton:=Wdata.Source,Agent:=agent[Player]):
            if(SilverConditional.HasAllItems[Agent]):
                SilverConditional.Activate(Agent)
                MalachiteGranter.GrantItem(Agent)
                CompleteConversion(Agent)
                spawn{DelayedUpdate(Agent)}
            else:
                NegativeAudio.Play(Agent)
                

    Convert3(Wdata:widget_message):void=
        if(Player:=Wdata.Player, PlayerUI:=GetPlayerUI[Player],SelectedButton:=Wdata.Source,Agent:=agent[Player]):
            if(MalachiteConditional.HasAllItems[Agent]):
                MalachiteConditional.Activate(Agent)
                ObsidianGranter.GrantItem(Agent)
                CompleteConversion(Agent)
                spawn{DelayedUpdate(Agent)}
            else:
                NegativeAudio.Play(Agent)

    Convert4(Wdata:widget_message):void=
        if(Player:=Wdata.Player, PlayerUI:=GetPlayerUI[Player],SelectedButton:=Wdata.Source,Agent:=agent[Player]):
            if(ObsidianConditional.HasAllItems[Agent]):
                ObsidianConditional.Activate(Agent)
                BrightcoreGranter.GrantItem(Agent)
                CompleteConversion(Agent)
                spawn{DelayedUpdate(Agent)}
            else:
                NegativeAudio.Play(Agent)

    Convert5(Wdata:widget_message):void=
        if(Player:=Wdata.Player, PlayerUI:=GetPlayerUI[Player],SelectedButton:=Wdata.Source,Agent:=agent[Player]):
            if(BrightcoreConditional.HasAllItems[Agent]):
                BrightcoreConditional.Activate(Agent)
                SpectroliteGranter.GrantItem(Agent)
                CompleteConversion(Agent)
                spawn{DelayedUpdate(Agent)}
            else:
                NegativeAudio.Play(Agent)


    CompleteConversion(Agent:agent):void= 
        spawn{DelayedUpdate(Agent)}
        PurchaseAudio.Play(Agent)
        UpdateUiText(Agent)
        PurchaseCine.Play(Agent)
        


    DelayedUpdate(Agent:agent)<suspends>:void=
        Sleep(0.5)
        UpdateUiText(Agent)





