
using { /Fortnite.com/Devices }
using { /Verse.org/Simulation }
using { /Fortnite.com/Characters }

passive_heal_device := class(creative_device):

    # Used to get updated playersmap
    @editable
    PersistanceDevice : persistence_device = persistence_device{}

    var LogicMap : [agent]logic = map{}

    @editable
    VfxOnTrigger : trigger_device = trigger_device {}
    @editable
    VfxOffTrigger : trigger_device = trigger_device {}
    @editable
    HealAudio : audio_player_device = audio_player_device{}

    @editable
    HealingZone : volume_device = volume_device{}
    var PlayersInZone : [agent]logic = map{}

    OnBegin<override>()<suspends>:void=

        HealingZone.AgentEntersEvent.Subscribe(AgentEntersZone)
        HealingZone.AgentExitsEvent.Subscribe(AgentExitsZone)

        # Need to start them at the start befor subscribing
        Sleep(3.0)

        loop:
            Sleep(1.0)

            for(Key -> Value : PersistanceDevice.PlayerMap):
                if(Fort:=Key.GetFortCharacter[]):
                    if (Player := player[Key]):
                        if(CrouchHeal := Data[Player].CrouchHeal):
                            if(CrouchHeal > 0):
                                if(Fort.IsCrouching[]):
                                    if:
                                        PlayerLogic := LogicMap[Key]?
                                        PlayerLogic = false
                                    then:
                                    else:
                                        if(set LogicMap[Key] = true){}
                                        VfxOnTrigger.Trigger(Key)
                                        HealAudio.Play(Key)
                                        HealingLoop(Key)
                                else:
                                    if(LogicMap[Key] = true):
                                        VfxOffTrigger.Trigger(Key)
                                        if (set LogicMap[Key] = false){}

                if(PlayersInZone[Key] = true):
                    if(Fort:=Key.GetFortCharacter[]):
                        if(Fort.IsActive[]):
                            Fort.Heal(1.0)

                

    HealingLoop(Agent:agent)<suspends>:void=
        loop:
            Sleep(1.0)
            if(Fort:=Agent.GetFortCharacter[]):
                if(Fort.IsCrouching[]):
                    if (Player := player[Agent]):
                        if(CrouchHeal := Data[Player].CrouchHeal):
                            Fort.Heal(CrouchHeal*0.5)
                else:
                    break

    AgentEntersZone(Agent:agent):void=
        if(set PlayersInZone[Agent] = true){}
        
    AgentExitsZone(Agent:agent):void=
        if(set PlayersInZone[Agent] = false){}


            





                








