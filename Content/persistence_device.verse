
using { /Fortnite.com/Devices }
using { /Verse.org/Simulation }
using { /UnrealEngine.com/Temporary/Diagnostics }
using { /Verse.org/Simulation/Tags }

spawners := class(tag){}

player_data := class<final><persistable>:

    MusicEnabled : logic = true

    Level : int = 1
    Xp : float = 0.0
    Gold : float = 0.0

    DamagePL : float = 0.0125
    ArmorPL : float = 0.1
    CritPL : float = 0.0015
    GoldPL : float = 0.0125

    DamagePLLevel : int = 0
    ArmorPLLevel : int = 0
    CritPLLevel : int = 0
    GoldPLLevel : int = 0

    Prestige : int = 0
    PrestigeBonusPerGem : float = 1.0
    PrestigeEarnings : float = 0.0
    TotalEarnings : float = 0.0
    Gems : float  = 0.0

    TotalGemsAwarded : float = 0.0

    # WeaponTypeLevels

    PistolLevel : int = 1
    RifleLevel : int = 1
    SmgLevel : int =  1
    ShotgunLevel : int = 1
    SniperLevel : int = 1

    # Talents

    TrueDamage : int = 0
    CritLevel : int = 0

    GoldPerElim : int = 0
    ExpBoost : int = 0
    ExpPerElim : int = 0 # Old Peristence so it doesnt error

    PlayerArmor : int = 0
    PlayerArmorCosmetic : int = 0
    ShieldLevel : int = 0
    MovementSpeed : int = 0

    # Sills

    FireChance : int = 0
    FireDamage : int = 1
    FireTime : int = 4 # not upgradeable anymore
    CrouchHeal : int = 0

    ChainLightningDamage : int = 1
    ChainLightningChance : int = 0
    ChainLightningJumps : int = 2 # Not gonna be upgradeable
    ChainLightningDamagePerUpgrade : float = 0.45
    ChainDamageGemLevel : int = 0

    FocusDamage : int = 0
    AccuracyDamage : int = 0

    MarkDamage : int = 0
    SuperCritDamage : int = 0
    SuperCritChance: int = 0
    SuperCritDamagePerUpgrade : float = 2.0
    SuperCritGemLevel : int = 0

    Dash : int = 0

    ScrapLevel:int = 1
    OreDropLevel : int = 0
    GemsOreDropLevel : int = 0

    AreasUnlocked : []logic = array{false,false,false,false,false,false,false,false,false,false,false,false,false,false}

    # Achievements
#############################################################################

    AchievementsDamageBonus : float = 0.0

    TotalBoarEliminations : int = 0
    TotalBuffaloEliminations : int =0
    TotalLizardEliminations : int =0
    TotalGiantEliminations : int = 0
    TotalScorpionEliminations : int = 0

    TotalGolemEliminations : int = 0
    TotalLeopardEliminations : int = 0
    TotalRamEliminations : int = 0
    TotalYetiEliminations : int = 0
    TotalHoundEliminations : int = 0
    TotalLavaJawEliminations : int = 0
    TotalTerravoreEliminations : int = 0

    TotalIronTuskEliminations : int = 0
    TotalEmberClawEliminations : int = 0
    TotalIceheartEliminations : int = 0
    TotalBlazebornEliminations : int = 0

    TotalQuestCompleted : int = 0
    TotalAchievementsCompleted : int = 0

    TotalPistolEliminations : int = 0
    TotalSmgEliminations : int = 0
    TotalAssaultEliminations : int = 0
    TotalShotgunEliminations : int = 0
    TotalSniperEliminations : int = 0

    
    
#############################################################################
    CosmeticsUnlocked : []logic = array{false,false,false,false,false,false,false,false,false,false,false,false,false,false,false,false,false,false,false,false,false,false,false,false}
    CosmeticsLevels : []int = array{1,1,1,1,1,1,1,1,1,1,1,1,1,1,1,1,1,1,1,1,1,1,1,1}
    # First 12 are back second 12 are head
    BackCosmeticEquiped : int = 13 # 13 means they have nothing equiped
    HatCosmeticEquiped : int = 13

    CosmeticDamage : float = 0.0
    CosmeticCritChance : float = 0.0
    CosmeticCritDamage : float = 0.0
    CosmeticBurnDamage : float = 0.0
    CosmeticArmor : float = 0.0
    CosmeticLightningChance : float = 0.0
    CosmeticLightningDamage : float = 0.0
    CosmeticGoldPerElim : float = 0.0

    ####################################

    QuestAchievementTier : int = 0

    # Random Quest
    CurrentMissionSaved : int = 0
    MissionCurrent : int = 0

    QuestValue : int = 0
    QuestNumber : int = 100 # means the player doesnt have a quest assigned
    QuestTarget : int = 0
    QuestRewardMultiple : int = 0
    QuestAreasSelected : []logic = array{true,true,true,true}
    
    KeepInventory : int = 0

    QuestRewardLevel : int = 0
    TowerLevel : int = 0

    TutorialsCurrentCount : int = 0
    TutorialsCompleted:int=0

    MissionsCompletedPersis: int =0
    MissionsCurrentCount: int =0

    IntroWatched : logic = false

    RandomMapLocations : []int = array{0,0,0,0}

    




player_data_update<constructor>(OldStat:player_data)<transacts> := player_data:

    MusicEnabled := OldStat.MusicEnabled

    Level:=OldStat.Level
    Xp:=OldStat.Xp
    Gold:=OldStat.Gold

    DamagePL := OldStat.DamagePL
    ArmorPL := OldStat.ArmorPL
    CritPL := OldStat.CritPL
    GoldPL := OldStat.GoldPL

    DamagePLLevel:=OldStat.DamagePLLevel
    ArmorPLLevel := OldStat.ArmorPLLevel
    CritPLLevel := OldStat.CritPLLevel
    GoldPLLevel := OldStat.GoldPLLevel


    Prestige:=OldStat.Prestige
    PrestigeBonusPerGem := OldStat.PrestigeBonusPerGem
    PrestigeEarnings :=OldStat.PrestigeEarnings
    TotalEarnings := OldStat.TotalEarnings
    Gems := OldStat.Gems

    TotalGemsAwarded := OldStat.TotalGemsAwarded

    PistolLevel := OldStat.PistolLevel
    RifleLevel := OldStat.RifleLevel
    SmgLevel := OldStat.SmgLevel
    ShotgunLevel := OldStat.ShotgunLevel
    SniperLevel := OldStat.SniperLevel

    # Talents

    TrueDamage:=OldStat.TrueDamage

    CritLevel := OldStat.CritLevel


    GoldPerElim := OldStat.GoldPerElim
    ExpBoost := OldStat.ExpBoost

    PlayerArmor:= OldStat.PlayerArmor
    PlayerArmorCosmetic := OldStat.PlayerArmorCosmetic
    ShieldLevel := OldStat.ShieldLevel
    MovementSpeed := OldStat.MovementSpeed

    # Skills
    
    FireChance:=OldStat.FireChance
    FireDamage:=OldStat.FireDamage
    FireTime:=OldStat.FireTime
    CrouchHeal:=OldStat.CrouchHeal

    FocusDamage:=OldStat.FocusDamage
    AccuracyDamage:=OldStat.AccuracyDamage

    ChainLightningDamage:=OldStat.ChainLightningDamage
    ChainLightningChance:=OldStat.ChainLightningChance
    ChainLightningJumps:=OldStat.ChainLightningJumps 

    ChainDamageGemLevel := OldStat.ChainDamageGemLevel

    MarkDamage:=OldStat.MarkDamage
    SuperCritDamage:=OldStat.SuperCritDamage
    SuperCritChance:=OldStat.SuperCritChance

    SuperCritGemLevel:=OldStat.SuperCritGemLevel

    Dash := OldStat.Dash

    # Achievements

    AchievementsDamageBonus:=OldStat.AchievementsDamageBonus

    TotalBoarEliminations:=OldStat.TotalBoarEliminations
    TotalBuffaloEliminations:=OldStat.TotalBuffaloEliminations
    TotalLizardEliminations := OldStat.TotalLizardEliminations
    TotalGiantEliminations := OldStat.TotalGiantEliminations
    TotalScorpionEliminations := OldStat.TotalScorpionEliminations
    TotalGolemEliminations := OldStat.TotalGolemEliminations
    TotalLeopardEliminations := OldStat.TotalLeopardEliminations
    TotalRamEliminations := OldStat.TotalRamEliminations
    TotalYetiEliminations := OldStat.TotalYetiEliminations
    TotalHoundEliminations :=OldStat.TotalHoundEliminations
    TotalLavaJawEliminations := OldStat.TotalLavaJawEliminations
    TotalTerravoreEliminations := OldStat.TotalTerravoreEliminations

    TotalIronTuskEliminations := OldStat.TotalIronTuskEliminations
    TotalEmberClawEliminations := OldStat.TotalEmberClawEliminations
    TotalIceheartEliminations := OldStat.TotalIceheartEliminations
    TotalBlazebornEliminations := OldStat.TotalBlazebornEliminations

    TotalQuestCompleted:=OldStat.TotalQuestCompleted
    TotalAchievementsCompleted:=OldStat.TotalAchievementsCompleted

    TotalPistolEliminations:=OldStat.TotalPistolEliminations
    TotalSmgEliminations:=OldStat.TotalSmgEliminations
    TotalAssaultEliminations:=OldStat.TotalAssaultEliminations
    TotalShotgunEliminations:=OldStat.TotalShotgunEliminations
    TotalSniperEliminations:=OldStat.TotalSniperEliminations

    QuestAchievementTier := OldStat.QuestAchievementTier
    
    # Cosmetics
    CosmeticsUnlocked:=OldStat.CosmeticsUnlocked
    CosmeticsLevels:=OldStat.CosmeticsLevels
    BackCosmeticEquiped:=OldStat.BackCosmeticEquiped
    HatCosmeticEquiped:=OldStat.HatCosmeticEquiped

    CosmeticDamage :=OldStat.CosmeticDamage
    CosmeticCritChance :=OldStat.CosmeticCritChance
    CosmeticCritDamage :=OldStat.CosmeticCritDamage
    CosmeticBurnDamage :=OldStat.CosmeticBurnDamage
    CosmeticArmor :=OldStat.CosmeticArmor
    CosmeticLightningChance :=OldStat.CosmeticLightningChance
    CosmeticLightningDamage :=OldStat.CosmeticLightningDamage
    CosmeticGoldPerElim :=OldStat.CosmeticGoldPerElim

    OreDropLevel := OldStat.OreDropLevel
    ScrapLevel := OldStat.ScrapLevel

    GemsOreDropLevel := OldStat.GemsOreDropLevel
    KeepInventory := OldStat.KeepInventory
    QuestRewardLevel := OldStat.QuestRewardLevel

    AreasUnlocked := OldStat.AreasUnlocked

    # Missions
    CurrentMissionSaved :=OldStat.CurrentMissionSaved
    MissionCurrent := OldStat.MissionCurrent

    # Quest Devies
    QuestValue := OldStat.QuestValue
    QuestNumber := OldStat.QuestNumber
    QuestTarget := OldStat.QuestTarget
    QuestRewardMultiple:=OldStat.QuestRewardMultiple

    QuestAreasSelected:=OldStat.QuestAreasSelected

    TowerLevel := OldStat.TowerLevel

    TutorialsCompleted:=OldStat.TutorialsCompleted
    TutorialsCurrentCount:=OldStat.TutorialsCurrentCount

    MissionsCompletedPersis := OldStat.MissionsCompletedPersis
    MissionsCurrentCount := OldStat.MissionsCurrentCount

    IntroWatched:=OldStat.IntroWatched
    RandomMapLocations:=OldStat.RandomMapLocations



     
    

var Data : weak_map(player,player_data) = map{}

persistence_device := class(creative_device):

    @editable
    CosmeticUnlocks : cosmetic_unlock_device = cosmetic_unlock_device{}

    var PlayerMap : [agent]spawners = map{}

    StatsMessage<localizes>(Score:int): message = "Level : {Score}"

    OnBegin<override>()<suspends>:void=

        #MyTrigger.TriggeredEvent.Subscribe(AddScore)
        AllPlayers := GetPlayspace().GetPlayers()
        for(Player:AllPlayers):
            OnPlayerSpawned(Player)
        Sleep(0.8)
        InitSpawners()
        GetPlayspace().PlayerRemovedEvent().Subscribe(OnPlayerRemoved)

    InitSpawners():void=
        Spawners := FindCreativeObjectsWithTag(spawners{})
        for(Obj : Spawners):
            if(spawner := player_spawner_device[Obj]):
                spawner.SpawnedEvent.Subscribe(OnPlayerSpawned)

    OnPlayerSpawned(Agent:agent):void=
        if(Player := player[Agent]):
            if(PlayerExists := PlayerMap[Agent]):
                Print("Player Data Exists")
            else:
                if(set PlayerMap[Player] = spawners{}):
                    if(not Data[Player]):
                        if(set Data[Player] = player_data{}):
                            Print("New Player Detected")
                            CosmeticUnlocks.InitRandomEmoteLocations(Agent)

    OnPlayerRemoved(PlayerLeft:agent):void=
        if(Agent:=agent[PlayerLeft]):
            if(ActualPlayer := PlayerMap[Agent]):
                var TempPlayerMap : [agent]spawners = map{}
                for(Key -> Value : PlayerMap,Key <> Agent):
                    set TempPlayerMap = ConcatenateMaps(TempPlayerMap, map{Key =>Value})
                set PlayerMap = TempPlayerMap
