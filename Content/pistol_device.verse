
using { /Fortnite.com/Devices }
using { /Verse.org/Simulation }
using { /UnrealEngine.com/Temporary/Diagnostics }


pistol_device := class(creative_device):

    @editable
    Button : button_device = button_device{}

    var AllWeaponConditionals : []conditional_button_device = array{}

    @editable
    HealConditional : conditional_button_device = conditional_button_device{}

    @editable
    Granter : item_granter_device = item_granter_device{}

    @editable
    MustHaveNoGunsMessage : hud_message_device = hud_message_device{}


    OnBegin<override>()<suspends>:void=

        TaggedConditionals := FindCreativeObjectsWithTag(InventoryCheckConditionalTag{})
        for(TaggedConditional : TaggedConditionals):
            if(Cond := conditional_button_device[TaggedConditional]):
                set AllWeaponConditionals = AllWeaponConditionals + array{Cond}

        Button.InteractedWithEvent.Subscribe(GrantPistol)

    GrantPistol(Agent:agent):void=

        LogicCheck := CheckInventoryEmpty(Agent)

        if(LogicCheck = true):
            Granter.GrantItem(Agent)
        else:
            MustHaveNoGunsMessage.Show(Agent)



    CheckInventoryEmpty(Agent:agent):logic=

        var ItemCount : int = 0
        #for(Cond : AllWeaponConditionals):

        for(i:=0..AllWeaponConditionals.Length-1):
            if(Cond:=AllWeaponConditionals[i]):
                for(i2:=0..2):
                    CondCount := Cond.GetItemCount(Agent,i2)
                    if (CondCount > 0 and CondCount < 6):
                        Print("Item Count : {CondCount}, i2 = {i2}, Conditional Number = {i}")
                        set ItemCount += CondCount

        for(i:=0..2):
            CondCount := HealConditional.GetItemCount(Agent,i)
            if(i = 0):
                if(Stacks :int= Ceil(CondCount / 15)):
                    set ItemCount += Stacks
            if(i = 1):
                if(Stacks :int= Ceil(CondCount / 6)):
                    set ItemCount += Stacks

        Print("AllWeaponConditionals : {AllWeaponConditionals.Length}")
        Print("Item Count : {ItemCount}")
        if(ItemCount <= 1):
            return true
        else:
            return false
