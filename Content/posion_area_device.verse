
using { /Fortnite.com/Devices }
using { /Verse.org/Simulation }
using { /UnrealEngine.com/Temporary/Diagnostics }
using { /Fortnite.com/Characters }


posion_area_device := class(creative_device):

    
    @editable
    Volumes : []volume_device = array{}
    @editable
    PoisonNiag : creative_prop_asset = DefaultCreativePropAsset
    @editable
    HitAudio : audio_player_device = audio_player_device{}
    @editable
    PosionDamage : float = 5.0
    
    OnBegin<override>()<suspends>:void=
        for(Vol : Volumes):
            Vol.AgentEntersEvent.Subscribe(StartSuspends)

    StartSuspends(Agent:agent):void=
        spawn:
            PosionHit(Agent)
        
    PosionHit(Agent:agent)<suspends>:void=
        
        for(i:=0..4):
            if:
                PlayerCharacter := Agent.GetFortCharacter[]
                PlayerCharacter.IsActive[]
            then:
                HitAudio.Play(Agent)
                DamageAfterArmor := PosionDamage * DamageCalculation(PlayerCharacter)
                PlayerCharacter.Damage(DamageAfterArmor)
                spawn:
                    ShowTemperaryNiagara(PlayerCharacter.GetTransform().Translation, PlayerCharacter.GetTransform().Rotation, PoisonNiag, 0.5)
            Sleep(2.0)