
using { /Fortnite.com/Devices }
using { /Verse.org/Simulation }
using { /UnrealEngine.com/Temporary/Diagnostics }


prestige_device := class(creative_device):

    @editable
    button : button_device = button_device{}
    @editable
    resetbutton : button_device = button_device{}

    @editable
    Huds : npc_hp_cinematic_hud_device = npc_hp_cinematic_hud_device{}
    @editable
    ItemGranter : item_granter_device = item_granter_device{}

    @editable
    PrestigePopup : hud_message_device = hud_message_device{}
    @editable
    AreasDevice : areas_device = areas_device{}
    @editable
    WeaponRarityDevice : weapon_rarity_device = weapon_rarity_device{}

    @editable
    QuestDevice : quest_device = quest_device{}

    @editable
    MissionTrigger : trigger_device = trigger_device{}

    storedstring<localizes>(S:string) : message = "{S}"

    OnBegin<override>()<suspends>:void=
        button.InteractedWithEvent.Subscribe(Test)
        resetbutton.InteractedWithEvent.Subscribe(ResetPrestigeProgress)

    Test(Agent:agent):void=
        CaculateAwardedGems(Agent)

    CaculateAwardedGems(Agent:agent):float= 
        var Gems : int =0
        var GemsFloat : float = 0.0
        if(MyData :=  Data[player[Agent]]):

            x := (MyData.TotalEarnings * 1.0) # was prestige earnings but changed it to lifetime earnings
            #set GemsFloat = Pow((Pow(10.0,-1.0)*x),0.27) - (1.56 * Pow(10.0,1.45)) # 1 million gold = 0 eggs Changed from .25 to .30 = 1 million = 1 gem
            set GemsFloat = 30 * Sqrt(x/(Pow(10.0,4.0)))
            Print("PrestigeEarnings : {x}")
            Print("GemsFloat : {GemsFloat}")
            #if(RoundedGems := Floor[GemsFloat]):
            #    set Gems = RoundedGems

            set GemsFloat = GemsFloat - MyData.TotalGemsAwarded
            if(GemsFloat < 0.0):
                set GemsFloat = 0.0

        return GemsFloat

    GrantGems(Agent:agent):void=
        CurentGemsCount := CaculateAwardedGems(Agent)
        if( CurentGemsCount >= 1.0):
            if(MyData :=  Data[player[Agent]]):
                if(set Data[player[Agent]]= player_data{player_data_update<constructor>(MyData),Gems := MyData.Gems + CurentGemsCount, TotalGemsAwarded := MyData.TotalGemsAwarded + CurentGemsCount}){}
                PrestigePopup.SetText(storedstring("+{formatter.Abbreviate((CurentGemsCount * MyData.PrestigeBonusPerGem),false)}% Damage and Gold"))
                PrestigePopup.Show(Agent)
            ResetPrestigeProgress(Agent)


    ResetPrestigeProgress(Agent:agent):void=
        if:
            Player := player[Agent]
            MyData := Data[player[Agent]]
            set Data[Player] = player_data{player_data_update<constructor>(MyData),

            Level := 1,
            Xp:=0.0,
            Gold:=0.0,

            PrestigeEarnings:=0.0,

            TrueDamage:=0,

            CritLevel:=0,
    
            GoldPerElim:=0,

            PlayerArmor:=0,
            PlayerArmorCosmetic:=0,
            ShieldLevel:=0,
            MovementSpeed:=0,

            FireChance := 0,
            FireDamage:= 1,
            CrouchHeal:=0,

            ChainLightningDamage:=1,
            ChainLightningChance:=0,

            FocusDamage:=0,
            AccuracyDamage:=0,

            MarkDamage:=0,
            SuperCritDamage:=0,
            SuperCritChance:=0,

            ScrapLevel:=1,
            OreDropLevel:=0,

            Dash:=0,

            AreasUnlocked := array{false,false,false,false,false,false,false,false,false,false,false}
            }
        then:

            if( MyData.KeepInventory < 1):
                ItemGranter.GrantItem(Agent)
                WeaponRarityDevice.PlayerHoldingNothing(Agent)

        if(XpBar :=  Huds.CMBDevice.XpBarPerPlayer[Agent]?):
            set XpBar.CurrentLevel = 1
            set XpBar.iTotalXp = 89.0
            set XpBar.iCurrentXp = 0.0

        Huds.CMBDevice.AreasOnLevelUp(Agent)
        Huds.RefreshXpBar(Agent)
        AreasDevice.LockAreas(Agent)
        QuestDevice.UnAssignQuest(Agent)
        MissionTrigger.Trigger(Agent)
        
        
        


            





