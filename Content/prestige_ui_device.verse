
using { /Fortnite.com/Devices }
using { /Verse.org/Simulation }
using { /UnrealEngine.com/Temporary/Diagnostics }
using { /UnrealEngine.com/Temporary/UI }
using { /UnrealEngine.com/Temporary/SpatialMath }
using { /Fortnite.com/UI }
using { /Verse.org/Colors }
using { /Fortnite.com/Playspaces }
using { /Fortnite.com/Characters }
using { /Verse.org/Assets }

prestige_ui_device := class(creative_device):

    @editable
    OpenShop : button_device = button_device{}
    @editable
    PurchaseAudio : audio_player_device = audio_player_device{}

    @editable
    PrestigeDevice : prestige_device = prestige_device{}

    @editable
    NotEnoughEarnigns : hud_message_device = hud_message_device{}

    var MyUI : [agent]?canvas = map{}
    var TextBlockPerAgent : [agent]?[]text_block = map{}
    var TextBlockPerAgentShadows : [agent]?[]text_block = map{}
    var ButtonLoudPerAgent : [agent]?[]button_loud = map{} # no longer being used

    storedstring<localizes>(S:string) : message = "{S}"
    JustIntString<localizes>(I:int) : message = "{I}"
    LevelString<localizes>(I:int) : message = "Lv {I}"
    DamageString<localizes>(I:int) : message = "{I}%"
    SecondString<localizes>(I:int) : message = "{I} Sec"
    HpString<localizes>(I:int) : message = "{I} hp"

    var StoredButtonTestingToclose : button_loud = button_loud{}

    OnBegin<override>()<suspends>:void=

        OpenShop.InteractedWithEvent.Subscribe(ShowUi)
    
    MakeEntireBackground (Agent:agent):canvas=

        GemTexture : texture = wooden_pixel_gui.textures.Icons2.Size8.t_GemRed_8

        CheckTexture : texture = wooden_pixel_gui.textures.Icons2.Size8.t_CheckGreen_8
        XTexture : texture = wooden_pixel_gui.textures.Icons2.Size8.t_XYellow_8


        TitleText: text_block = text_block{DefaultTextColor := MakeColorFromHex("529F17FF"), DefaultText:=storedstring("Prestige")}
        TitleTextShadow: text_block = text_block{DefaultTextColor := MakeColorFromHex("142706FF"), DefaultText:=storedstring("Prestige")}
        Image: texture_block = texture_block{DefaultImage:= wooden_pixel_gui.textures.Icons2.Size32.t_Chest01Open_32, DefaultDesiredSize:=vector2{X:=180.0,Y:=180.0}}
        GemImage: texture_block = texture_block{DefaultImage:= wooden_pixel_gui.textures.Icons2.Size32.t_Chest01ContentsGems_32, DefaultDesiredSize:=vector2{X:=180.0,Y:=180.0}}
        Vignette: texture_block = texture_block{DefaultImage:= Content.Textures.T_SquaresVignette, DefaultDesiredSize:=vector2{X:=180.0,Y:=180.0}}
        ImageShadow: texture_block = texture_block{DefaultImage:= wooden_pixel_gui.textures.Icons2.Size32.t_Chest01Open_32,DefaultTint:=MakeColorFromHex("080402FF"), DefaultDesiredSize:=vector2{X:=100.0,Y:=100.0}}
        ResetText: text_block = text_block{DefaultTextColor := MakeColorFromHex("4A331BFF"), DefaultText:=storedstring("Keep:                          Reset:")}
        
        GemsButtonText: text_block = text_block{DefaultTextColor := MakeColorFromHex("9ACF55"), DefaultText:=storedstring("0")} # Needs to be formated
        PrestigeButton : button_quiet = button_quiet{DefaultText:=storedstring("")}
        PrestigeButton.OnClick().Subscribe(PrestigeHandle) 

        PlusSignText : text_block = text_block{DefaultTextColor := MakeColorFromHex("9ACF55"), DefaultText:=storedstring("+")}
        DamagePerGemText : text_block = text_block{DefaultTextColor := MakeColorFromHex("9ACF55"), DefaultText:=storedstring("+1%")}
        TotalGemDamage : text_block = text_block{DefaultTextColor := MakeColorFromHex("9ACF55"), DefaultText:=storedstring("+0%")}
        TotalText : text_block = text_block{DefaultTextColor := MakeColorFromHex("FFFFFF"), DefaultText:=storedstring(" Damage & Gold")}
        ToDamageText : text_block = text_block{DefaultTextColor := MakeColorFromHex("FFFFFF"), DefaultText:=storedstring(" Per Gem")}

        

        if:
            MyData := Data[player[Agent]]
            GemBonus := MyData.PrestigeBonusPerGem
        then:
            GemCount : float = PrestigeDevice.CaculateAwardedGems(Agent)
            if(GemCount <= 0.0):
            else:
                GemsButtonText.SetText(storedstring("{formatter.Abbreviate(GemCount * 1.0,false)}"))
            DamagePerGemText.SetText(storedstring("+{formatter.Abbreviate(GemBonus * 1.0,true)} %"))

            TotalGemDamage.SetText(storedstring("+{formatter.Abbreviate(GemCount*GemBonus * 1.0,false)}%"))

        CloseButton : button_quiet = button_quiet{DefaultText:=storedstring("Close")}
        CloseButton.OnClick().Subscribe(RemoveHUD) 

        MyCanvas:=canvas:
            Slots:=array:
                canvas_slot:
                                Anchors:=anchors{Minimum:=vector2{X:=0.1,Y:=0.00}, Maximum:=vector2{X:=0.9,Y:=0.90}}
                                Widget:=overlay:
                                    Slots:=array:
                                        overlay_slot:
                                            HorizontalAlignment:=horizontal_alignment.Right
                                            VerticalAlignment:=vertical_alignment.Center
                                            Widget:=overlay:
                                                Slots:=array:
                                                    overlay_slot:
                                                        HorizontalAlignment:=horizontal_alignment.Fill
                                                        VerticalAlignment:=vertical_alignment.Fill
                                                        Padding:=margin{Left:=-10.0,Top:=-10.0,Right:=-10.0,Bottom:=-10.0}
                                                        Widget:=texture_block:
                                                            DefaultImage:=wooden_pixel_gui.textures.Panels.WhiteSquare75
                                                            DefaultTint:= MakeColorFromHex("49321CFF")
                                                    overlay_slot:
                                                        HorizontalAlignment:=horizontal_alignment.Fill
                                                        VerticalAlignment:=vertical_alignment.Fill
                                                        Widget:=texture_block:
                                                            DefaultImage:=wooden_pixel_gui.textures.Panels.WhiteSquare100
                                                            DefaultTint:= MakeColorFromHex("5F4427FF")
                                                    overlay_slot:
                                                        Widget:=stack_box:
                                                            Orientation:=orientation.Vertical
                                                            Slots:=array:
                                                                stack_box_slot:
                                                                    Padding:=margin{Left:=0.0,Top:=5.0,Right:=0.0,Bottom:=0.0}
                                                                    HorizontalAlignment:=horizontal_alignment.Left
                                                                    VerticalAlignment:=vertical_alignment.Center
                                                                    Widget:=text_block:
                                                                        DefaultTextColor := MakeColorFromHex("FFFFFF")
                                                                        DefaultText:=storedstring("Tips:\n\n- Gems granted are based\noff of total prestige earnings")
                                                                stack_box_slot:
                                                                    Padding:=margin{Left:=0.0,Top:=5.0,Right:=0.0,Bottom:=0.0}
                                                                    HorizontalAlignment:=horizontal_alignment.Left
                                                                    VerticalAlignment:=vertical_alignment.Center
                                                                    Widget:=text_block:
                                                                        DefaultTextColor := MakeColorFromHex("FFFFFF")
                                                                        DefaultText:=storedstring("\n- Use all your gold before\nprestiging")
                                                                stack_box_slot:
                                                                    Padding:=margin{Left:=0.0,Top:=5.0,Right:=0.0,Bottom:=0.0}
                                                                    HorizontalAlignment:=horizontal_alignment.Left
                                                                    VerticalAlignment:=vertical_alignment.Center
                                                                    Widget:=text_block:
                                                                        DefaultTextColor := MakeColorFromHex("FFFFFF")
                                                                        DefaultText:=storedstring("\n- Purchase upgrade to keep\n your weapons")
                                                    overlay_slot:
                                                        HorizontalAlignment:=horizontal_alignment.Center
                                                        VerticalAlignment:=vertical_alignment.Bottom
                                                        Padding:=margin{Left:=0.0,Top:=0.0,Right:=0.0,Bottom:=-200.0}
                                                        Widget:=CloseButton
                                                                                      





                                        overlay_slot: # Main stackbox with panels in it
                                                        HorizontalAlignment:=horizontal_alignment.Center
                                                        VerticalAlignment:=vertical_alignment.Center
                                                        #Padding:=margin{Left:=50.0,Top:=0.0,Right:=50.0,Bottom:=0.0}
                                                        Widget:=overlay:
                                                            Slots:=array:
                                                                overlay_slot:
                                                                    HorizontalAlignment:=horizontal_alignment.Fill
                                                                    VerticalAlignment:=vertical_alignment.Fill
                                                                    Padding:=margin{Left:=-10.0,Top:=-10.0,Right:=-10.0,Bottom:=-10.0}
                                                                    Widget:=texture_block:
                                                                        DefaultImage:=wooden_pixel_gui.textures.Panels.WhiteSquare75
                                                                        DefaultTint:= MakeColorFromHex("49321CFF")
                                                                overlay_slot:
                                                                    HorizontalAlignment:=horizontal_alignment.Fill
                                                                    VerticalAlignment:=vertical_alignment.Fill
                                                                    Widget:=texture_block:
                                                                        DefaultImage:=wooden_pixel_gui.textures.Panels.WhiteSquare100
                                                                        DefaultTint:= MakeColorFromHex("5F4427FF")
                                                                overlay_slot:
                                                                    HorizontalAlignment:=horizontal_alignment.Center
                                                                    VerticalAlignment:=vertical_alignment.Top
                                                                    Padding:=margin{Left:=0.0,Top:=10.0,Right:=0.0,Bottom:=0.0}
                                                                    Widget:=texture_block:
                                                                        DefaultImage:=wooden_pixel_gui.textures.Panels.WhiteSquare100
                                                                        DefaultTint:= MakeColorFromHex("3A2713FF")
                                                                        DefaultDesiredSize:=vector2{X:=300.0,Y:=300.0}
                                                                overlay_slot:
                                                                    HorizontalAlignment:=horizontal_alignment.Center
                                                                    VerticalAlignment:=vertical_alignment.Top
                                                                    Padding:=margin{Left:=0.0,Top:=20.0,Right:=0.0,Bottom:=0.0}
                                                                    Widget:=texture_block:
                                                                        DefaultImage:=wooden_pixel_gui.textures.Panels.WhiteSquare100
                                                                        DefaultTint:= MakeColorFromHex("4A331BFF")
                                                                        DefaultDesiredSize:=vector2{X:=280.0,Y:=280.0}
                                                                overlay_slot:
                                                                    HorizontalAlignment:=horizontal_alignment.Center
                                                                    VerticalAlignment:=vertical_alignment.Center
                                                                    Padding:=margin{Left:=0.0,Top:=15.0,Right:=0.0,Bottom:=0.0}
                                                                    Widget:=stack_box:
                                                                        Orientation:=orientation.Vertical
                                                                        Slots:=array:
                                                                            stack_box_slot:
                                                                                HorizontalAlignment:=horizontal_alignment.Fill
                                                                                VerticalAlignment:=vertical_alignment.Fill
                                                                                Widget:=overlay:
                                                                                    Slots:=array:
                                                                                        overlay_slot:
                                                                                            HorizontalAlignment:=horizontal_alignment.Center
                                                                                            VerticalAlignment:=vertical_alignment.Center
                                                                                            Padding:=margin{Left:=3.0,Top:=3.0,Right:=0.0,Bottom:=0.0}
                                                                                            Widget:=TitleTextShadow
                                                                                        overlay_slot:
                                                                                            HorizontalAlignment:=horizontal_alignment.Center
                                                                                            VerticalAlignment:=vertical_alignment.Center
                                                                                            Widget:=TitleText
                                                                                        
                                                                            stack_box_slot:
                                                                                HorizontalAlignment:=horizontal_alignment.Fill
                                                                                VerticalAlignment:=vertical_alignment.Fill
                                                                                Padding:=margin{Left:=0.0,Top:=20.0,Right:=0.0,Bottom:=0.0}
                                                                                Widget:=overlay:
                                                                                    Slots:=array:
                                                                                        overlay_slot:
                                                                                            HorizontalAlignment:=horizontal_alignment.Center
                                                                                            VerticalAlignment:=vertical_alignment.Top
                                                                                            Widget:=texture_block:
                                                                                                DefaultImage:=wooden_pixel_gui.textures.Panels.WhiteSquare100
                                                                                                DefaultTint:= MakeColorFromHex("3A2713FF")
                                                                                                DefaultDesiredSize:=vector2{X:=180.0,Y:=180.0}
                                                                                        overlay_slot:
                                                                                            HorizontalAlignment:=horizontal_alignment.Center
                                                                                            VerticalAlignment:=vertical_alignment.Center
                                                                                            Padding:=margin{Left:=0.0,Top:=6.0,Right:=0.0,Bottom:=0.0}
                                                                                            Widget:=ImageShadow
                                                                                        overlay_slot:
                                                                                            HorizontalAlignment:=horizontal_alignment.Center
                                                                                            VerticalAlignment:=vertical_alignment.Center
                                                                                            Widget:=Image
                                                                                        overlay_slot:
                                                                                            HorizontalAlignment:=horizontal_alignment.Center
                                                                                            VerticalAlignment:=vertical_alignment.Center
                                                                                            Widget:=GemImage

                                                                                        overlay_slot:
                                                                                            HorizontalAlignment:=horizontal_alignment.Center
                                                                                            VerticalAlignment:=vertical_alignment.Center
                                                                                            Widget:=Vignette
                                                                            stack_box_slot:
                                                                                HorizontalAlignment:=horizontal_alignment.Fill
                                                                                VerticalAlignment:=vertical_alignment.Fill
                                                                                Padding:=margin{Left:=0.0,Top:=80.0,Right:=0.0,Bottom:=0.0}
                                                                                Widget:=overlay:
                                                                                    Slots:=array:
                                                                                        overlay_slot:
                                                                                            HorizontalAlignment:=horizontal_alignment.Center
                                                                                            VerticalAlignment:=vertical_alignment.Fill
                                                                                            Widget:=texture_block:
                                                                                                DefaultImage:=wooden_pixel_gui.textures.Panels.WhiteSquare100
                                                                                                DefaultTint:= MakeColorFromHex("E2AE78FF")
                                                                                                DefaultDesiredSize:=vector2{X:=570.0,Y:=32.0}
                                                                                        overlay_slot:
                                                                                            HorizontalAlignment:=horizontal_alignment.Center
                                                                                            VerticalAlignment:=vertical_alignment.Center
                                                                                            Widget:=ResetText

                                                                            stack_box_slot:
                                                                                HorizontalAlignment:=horizontal_alignment.Fill
                                                                                VerticalAlignment:=vertical_alignment.Fill
                                                                                Padding:=margin{Left:=0.0,Top:=20.0,Right:=0.0,Bottom:=0.0}
                                                                                Widget:=stack_box:
                                                                                    Orientation:=orientation.Horizontal
                                                                                    Slots:=array:
                                                                                        stack_box_slot:
                                                                                            HorizontalAlignment:=horizontal_alignment.Fill
                                                                                            VerticalAlignment:=vertical_alignment.Fill
                                                                                            Widget:=stack_box:
                                                                                                Orientation:=orientation.Vertical
                                                                                                Slots:=array:
                                                                                                    stack_box_slot:
                                                                                                        HorizontalAlignment:=horizontal_alignment.Fill
                                                                                                        VerticalAlignment:=vertical_alignment.Fill
                                                                                                        Padding:=margin{Left:=0.0,Top:=5.0,Right:=0.0,Bottom:=5.0}
                                                                                                        Widget:=stack_box:
                                                                                                            Orientation:=orientation.Horizontal
                                                                                                            Slots:=array:
                                                                                                                stack_box_slot:
                                                                                                                    HorizontalAlignment:=horizontal_alignment.Fill
                                                                                                                    VerticalAlignment:=vertical_alignment.Fill
                                                                                                                    Padding:=margin{Left:=10.0,Top:=0.0,Right:=0.0,Bottom:=0.0}
                                                                                                                    Widget:=overlay:
                                                                                                                        Slots:=array:
                                                                                                                            overlay_slot:
                                                                                                                                HorizontalAlignment:=horizontal_alignment.Fill
                                                                                                                                VerticalAlignment:=vertical_alignment.Fill
                                                                                                                                Widget:=texture_block:
                                                                                                                                    DefaultImage:=wooden_pixel_gui.textures.Panels.WhiteSquare100
                                                                                                                                    DefaultTint:= MakeColorFromHex("49321CFF")
                                                                                                                                    DefaultDesiredSize:=vector2{X:=50.0,Y:=50.0}
                                                                                                                            overlay_slot:
                                                                                                                                HorizontalAlignment:=horizontal_alignment.Fill
                                                                                                                                VerticalAlignment:=vertical_alignment.Fill
                                                                                                                                Widget:=texture_block:
                                                                                                                                    DefaultImage:=CheckTexture
                                                                                                                stack_box_slot:
                                                                                                                    HorizontalAlignment:=horizontal_alignment.Fill
                                                                                                                    VerticalAlignment:=vertical_alignment.Center
                                                                                                                    Widget:=text_block:
                                                                                                                        DefaultTextColor := MakeColorFromHex("FFFFFF")
                                                                                                                        DefaultText:=storedstring("Outfits")
                                                                                                    stack_box_slot:
                                                                                                        HorizontalAlignment:=horizontal_alignment.Fill
                                                                                                        VerticalAlignment:=vertical_alignment.Fill
                                                                                                        Padding:=margin{Left:=0.0,Top:=5.0,Right:=0.0,Bottom:=5.0}
                                                                                                        Widget:=stack_box:
                                                                                                            Orientation:=orientation.Horizontal
                                                                                                            Slots:=array:
                                                                                                                stack_box_slot:
                                                                                                                    HorizontalAlignment:=horizontal_alignment.Fill
                                                                                                                    VerticalAlignment:=vertical_alignment.Fill
                                                                                                                    Padding:=margin{Left:=10.0,Top:=0.0,Right:=0.0,Bottom:=0.0}
                                                                                                                    Widget:=overlay:
                                                                                                                        Slots:=array:
                                                                                                                            overlay_slot:
                                                                                                                                HorizontalAlignment:=horizontal_alignment.Fill
                                                                                                                                VerticalAlignment:=vertical_alignment.Fill
                                                                                                                                Widget:=texture_block:
                                                                                                                                    DefaultImage:=wooden_pixel_gui.textures.Panels.WhiteSquare100
                                                                                                                                    DefaultTint:= MakeColorFromHex("49321CFF")
                                                                                                                                    DefaultDesiredSize:=vector2{X:=50.0,Y:=50.0}
                                                                                                                            overlay_slot:
                                                                                                                                HorizontalAlignment:=horizontal_alignment.Fill
                                                                                                                                VerticalAlignment:=vertical_alignment.Fill
                                                                                                                                Widget:=texture_block:
                                                                                                                                    DefaultImage:=CheckTexture
                                                                                                                stack_box_slot:
                                                                                                                    HorizontalAlignment:=horizontal_alignment.Fill
                                                                                                                    VerticalAlignment:=vertical_alignment.Center
                                                                                                                    Widget:=text_block:
                                                                                                                        DefaultTextColor := MakeColorFromHex("FFFFFF")
                                                                                                                        DefaultText:=storedstring("Achievements")
                                                                                                    stack_box_slot:
                                                                                                        HorizontalAlignment:=horizontal_alignment.Fill
                                                                                                        VerticalAlignment:=vertical_alignment.Fill
                                                                                                        Padding:=margin{Left:=0.0,Top:=5.0,Right:=0.0,Bottom:=5.0}
                                                                                                        Widget:=stack_box:
                                                                                                            Orientation:=orientation.Horizontal
                                                                                                            Slots:=array:
                                                                                                                stack_box_slot:
                                                                                                                    HorizontalAlignment:=horizontal_alignment.Fill
                                                                                                                    VerticalAlignment:=vertical_alignment.Fill
                                                                                                                    Padding:=margin{Left:=10.0,Top:=0.0,Right:=0.0,Bottom:=0.0}
                                                                                                                    Widget:=overlay:
                                                                                                                        Slots:=array:
                                                                                                                            overlay_slot:
                                                                                                                                HorizontalAlignment:=horizontal_alignment.Fill
                                                                                                                                VerticalAlignment:=vertical_alignment.Fill
                                                                                                                                Widget:=texture_block:
                                                                                                                                    DefaultImage:=wooden_pixel_gui.textures.Panels.WhiteSquare100
                                                                                                                                    DefaultTint:= MakeColorFromHex("49321CFF")
                                                                                                                                    DefaultDesiredSize:=vector2{X:=50.0,Y:=50.0}
                                                                                                                            overlay_slot:
                                                                                                                                HorizontalAlignment:=horizontal_alignment.Fill
                                                                                                                                VerticalAlignment:=vertical_alignment.Fill
                                                                                                                                Widget:=texture_block:
                                                                                                                                    DefaultImage:=CheckTexture
                                                                                                                stack_box_slot:
                                                                                                                    HorizontalAlignment:=horizontal_alignment.Fill
                                                                                                                    VerticalAlignment:=vertical_alignment.Center
                                                                                                                    Widget:=text_block:
                                                                                                                        DefaultTextColor := MakeColorFromHex("FFFFFF")
                                                                                                                        DefaultText:=storedstring("Gems")
                                                                                        stack_box_slot:
                                                                                            HorizontalAlignment:=horizontal_alignment.Fill
                                                                                            VerticalAlignment:=vertical_alignment.Fill
                                                                                            Widget:=stack_box:
                                                                                                Orientation:=orientation.Vertical
                                                                                                Slots:=array:
                                                                                                    stack_box_slot:
                                                                                                        HorizontalAlignment:=horizontal_alignment.Fill
                                                                                                        VerticalAlignment:=vertical_alignment.Fill
                                                                                                        Padding:=margin{Left:=0.0,Top:=5.0,Right:=0.0,Bottom:=5.0}
                                                                                                        Widget:=stack_box:
                                                                                                            Orientation:=orientation.Horizontal
                                                                                                            Slots:=array:
                                                                                                                stack_box_slot:
                                                                                                                    HorizontalAlignment:=horizontal_alignment.Fill
                                                                                                                    VerticalAlignment:=vertical_alignment.Fill
                                                                                                                    Padding:=margin{Left:=10.0,Top:=0.0,Right:=0.0,Bottom:=0.0}
                                                                                                                    Widget:=overlay:
                                                                                                                        Slots:=array:
                                                                                                                            overlay_slot:
                                                                                                                                HorizontalAlignment:=horizontal_alignment.Fill
                                                                                                                                VerticalAlignment:=vertical_alignment.Fill
                                                                                                                                Widget:=texture_block:
                                                                                                                                    DefaultImage:=wooden_pixel_gui.textures.Panels.WhiteSquare100
                                                                                                                                    DefaultTint:= MakeColorFromHex("49321CFF")
                                                                                                                                    DefaultDesiredSize:=vector2{X:=50.0,Y:=50.0}
                                                                                                                            overlay_slot:
                                                                                                                                HorizontalAlignment:=horizontal_alignment.Fill
                                                                                                                                VerticalAlignment:=vertical_alignment.Fill
                                                                                                                                Widget:=texture_block:
                                                                                                                                    DefaultImage:=XTexture
                                                                                                                stack_box_slot:
                                                                                                                    HorizontalAlignment:=horizontal_alignment.Fill
                                                                                                                    VerticalAlignment:=vertical_alignment.Center
                                                                                                                    Widget:=text_block:
                                                                                                                        DefaultTextColor := MakeColorFromHex("FFFFFF")
                                                                                                                        DefaultText:=storedstring("Traits")
                                                                                                    stack_box_slot:
                                                                                                        HorizontalAlignment:=horizontal_alignment.Fill
                                                                                                        VerticalAlignment:=vertical_alignment.Fill
                                                                                                        Padding:=margin{Left:=0.0,Top:=5.0,Right:=0.0,Bottom:=5.0}
                                                                                                        Widget:=stack_box:
                                                                                                            Orientation:=orientation.Horizontal
                                                                                                            Slots:=array:
                                                                                                                stack_box_slot:
                                                                                                                    HorizontalAlignment:=horizontal_alignment.Fill
                                                                                                                    VerticalAlignment:=vertical_alignment.Fill
                                                                                                                    Padding:=margin{Left:=10.0,Top:=0.0,Right:=0.0,Bottom:=0.0}
                                                                                                                    Widget:=overlay:
                                                                                                                        Slots:=array:
                                                                                                                            overlay_slot:
                                                                                                                                HorizontalAlignment:=horizontal_alignment.Fill
                                                                                                                                VerticalAlignment:=vertical_alignment.Fill
                                                                                                                                Widget:=texture_block:
                                                                                                                                    DefaultImage:=wooden_pixel_gui.textures.Panels.WhiteSquare100
                                                                                                                                    DefaultTint:= MakeColorFromHex("49321CFF")
                                                                                                                                    DefaultDesiredSize:=vector2{X:=50.0,Y:=50.0}
                                                                                                                            overlay_slot:
                                                                                                                                HorizontalAlignment:=horizontal_alignment.Fill
                                                                                                                                VerticalAlignment:=vertical_alignment.Fill
                                                                                                                                Widget:=texture_block:
                                                                                                                                    DefaultImage:=XTexture
                                                                                                                stack_box_slot:
                                                                                                                    HorizontalAlignment:=horizontal_alignment.Fill
                                                                                                                    VerticalAlignment:=vertical_alignment.Center
                                                                                                                    Widget:=text_block:
                                                                                                                        DefaultTextColor := MakeColorFromHex("FFFFFF")
                                                                                                                        DefaultText:=storedstring("Skills")
                                                                                                    stack_box_slot:
                                                                                                        HorizontalAlignment:=horizontal_alignment.Fill
                                                                                                        VerticalAlignment:=vertical_alignment.Fill
                                                                                                        Padding:=margin{Left:=0.0,Top:=5.0,Right:=0.0,Bottom:=5.0}
                                                                                                        Widget:=stack_box:
                                                                                                            Orientation:=orientation.Horizontal
                                                                                                            Slots:=array:
                                                                                                                stack_box_slot:
                                                                                                                    HorizontalAlignment:=horizontal_alignment.Fill
                                                                                                                    VerticalAlignment:=vertical_alignment.Fill
                                                                                                                    Padding:=margin{Left:=10.0,Top:=0.0,Right:=0.0,Bottom:=0.0}
                                                                                                                    Widget:=overlay:
                                                                                                                        Slots:=array:
                                                                                                                            overlay_slot:
                                                                                                                                HorizontalAlignment:=horizontal_alignment.Fill
                                                                                                                                VerticalAlignment:=vertical_alignment.Fill
                                                                                                                                Widget:=texture_block:
                                                                                                                                    DefaultImage:=wooden_pixel_gui.textures.Panels.WhiteSquare100
                                                                                                                                    DefaultTint:= MakeColorFromHex("49321CFF")
                                                                                                                                    DefaultDesiredSize:=vector2{X:=50.0,Y:=50.0}
                                                                                                                            overlay_slot:
                                                                                                                                HorizontalAlignment:=horizontal_alignment.Fill
                                                                                                                                VerticalAlignment:=vertical_alignment.Fill
                                                                                                                                Widget:=texture_block:
                                                                                                                                    DefaultImage:=XTexture
                                                                                                                stack_box_slot:
                                                                                                                    HorizontalAlignment:=horizontal_alignment.Fill
                                                                                                                    VerticalAlignment:=vertical_alignment.Center
                                                                                                                    Widget:=text_block:
                                                                                                                        DefaultTextColor := MakeColorFromHex("FFFFFF")
                                                                                                                        DefaultText:=storedstring("Gold")
                                                                                                    stack_box_slot:
                                                                                                        HorizontalAlignment:=horizontal_alignment.Fill
                                                                                                        VerticalAlignment:=vertical_alignment.Fill
                                                                                                        Padding:=margin{Left:=0.0,Top:=5.0,Right:=0.0,Bottom:=5.0}
                                                                                                        Widget:=stack_box:
                                                                                                            Orientation:=orientation.Horizontal
                                                                                                            Slots:=array:
                                                                                                                stack_box_slot:
                                                                                                                    HorizontalAlignment:=horizontal_alignment.Fill
                                                                                                                    VerticalAlignment:=vertical_alignment.Fill
                                                                                                                    Padding:=margin{Left:=10.0,Top:=0.0,Right:=0.0,Bottom:=0.0}
                                                                                                                    Widget:=overlay:
                                                                                                                        Slots:=array:
                                                                                                                            overlay_slot:
                                                                                                                                HorizontalAlignment:=horizontal_alignment.Fill
                                                                                                                                VerticalAlignment:=vertical_alignment.Fill
                                                                                                                                Widget:=texture_block:
                                                                                                                                    DefaultImage:=wooden_pixel_gui.textures.Panels.WhiteSquare100
                                                                                                                                    DefaultTint:= MakeColorFromHex("49321CFF")
                                                                                                                                    DefaultDesiredSize:=vector2{X:=50.0,Y:=50.0}
                                                                                                                            overlay_slot:
                                                                                                                                HorizontalAlignment:=horizontal_alignment.Fill
                                                                                                                                VerticalAlignment:=vertical_alignment.Fill
                                                                                                                                Widget:=texture_block:
                                                                                                                                    DefaultImage:=XTexture
                                                                                                                stack_box_slot:
                                                                                                                    HorizontalAlignment:=horizontal_alignment.Fill
                                                                                                                    VerticalAlignment:=vertical_alignment.Center
                                                                                                                    Widget:=text_block:
                                                                                                                        DefaultTextColor := MakeColorFromHex("FFFFFF")
                                                                                                                        DefaultText:=storedstring("Player Level  ")
                                                                                                    stack_box_slot:
                                                                                                        HorizontalAlignment:=horizontal_alignment.Fill
                                                                                                        VerticalAlignment:=vertical_alignment.Fill
                                                                                                        Padding:=margin{Left:=0.0,Top:=5.0,Right:=0.0,Bottom:=5.0}
                                                                                                        Widget:=stack_box:
                                                                                                            Orientation:=orientation.Horizontal
                                                                                                            Slots:=array:
                                                                                                                stack_box_slot:
                                                                                                                    HorizontalAlignment:=horizontal_alignment.Fill
                                                                                                                    VerticalAlignment:=vertical_alignment.Fill
                                                                                                                    Padding:=margin{Left:=10.0,Top:=0.0,Right:=0.0,Bottom:=0.0}
                                                                                                                    Widget:=overlay:
                                                                                                                        Slots:=array:
                                                                                                                            overlay_slot:
                                                                                                                                HorizontalAlignment:=horizontal_alignment.Fill
                                                                                                                                VerticalAlignment:=vertical_alignment.Fill
                                                                                                                                Widget:=texture_block:
                                                                                                                                    DefaultImage:=wooden_pixel_gui.textures.Panels.WhiteSquare100
                                                                                                                                    DefaultTint:= MakeColorFromHex("49321CFF")
                                                                                                                                    DefaultDesiredSize:=vector2{X:=50.0,Y:=50.0}
                                                                                                                            overlay_slot:
                                                                                                                                HorizontalAlignment:=horizontal_alignment.Fill
                                                                                                                                VerticalAlignment:=vertical_alignment.Fill
                                                                                                                                Widget:=texture_block:
                                                                                                                                    DefaultImage:=XTexture
                                                                                                                stack_box_slot:
                                                                                                                    HorizontalAlignment:=horizontal_alignment.Fill
                                                                                                                    VerticalAlignment:=vertical_alignment.Center
                                                                                                                    Widget:=text_block:
                                                                                                                        DefaultTextColor := MakeColorFromHex("FFFFFF")
                                                                                                                        DefaultText:=storedstring("Weapons* ")
                                                                            stack_box_slot:
                                                                                HorizontalAlignment:=horizontal_alignment.Center
                                                                                VerticalAlignment:=vertical_alignment.Fill
                                                                                    Widget:=texture_block:
                                                                                        DefaultImage:=wooden_pixel_gui.textures.Panels.WhiteSquare100
                                                                                        DefaultTint:= MakeColorFromHex("402F1EFF")
                                                                                        DefaultDesiredSize:=vector2{X:=570.0,Y:=10.0}

                                                                            stack_box_slot:
                                                                                HorizontalAlignment:=horizontal_alignment.Center
                                                                                VerticalAlignment:=vertical_alignment.Center
                                                                                    Widget:=stack_box:
                                                                                        Orientation:=orientation.Horizontal
                                                                                        Slots:=array:
                                                                                            stack_box_slot:
                                                                                                HorizontalAlignment:=horizontal_alignment.Center
                                                                                                VerticalAlignment:=vertical_alignment.Center
                                                                                                Widget:=texture_block:
                                                                                                    DefaultImage:=GemTexture
                                                                                            stack_box_slot:
                                                                                                Widget:=DamagePerGemText
                                                                                            stack_box_slot:
                                                                                                HorizontalAlignment:=horizontal_alignment.Center
                                                                                                VerticalAlignment:=vertical_alignment.Center
                                                                                                Widget:=ToDamageText

                                                                            stack_box_slot:
                                                                                HorizontalAlignment:=horizontal_alignment.Center
                                                                                VerticalAlignment:=vertical_alignment.Center
                                                                                    Widget:=stack_box:
                                                                                        Orientation:=orientation.Horizontal
                                                                                        Slots:=array:
                                                                                            stack_box_slot:
                                                                                                HorizontalAlignment:=horizontal_alignment.Center
                                                                                                VerticalAlignment:=vertical_alignment.Center
                                                                                                Widget:=texture_block:
                                                                                                    DefaultImage:=GemTexture
                                                                                            stack_box_slot:
                                                                                                Widget:=TotalGemDamage
                                                                                            stack_box_slot:
                                                                                                HorizontalAlignment:=horizontal_alignment.Center
                                                                                                VerticalAlignment:=vertical_alignment.Center
                                                                                                Widget:=TotalText

                                                                                
                                                                            stack_box_slot:
                                                                                HorizontalAlignment:=horizontal_alignment.Fill
                                                                                VerticalAlignment:=vertical_alignment.Fill
                                                                                Padding:=margin{Left:=0.0,Top:=20.0,Right:=0.0,Bottom:=0.0}
                                                                                Widget:=overlay:
                                                                                    Slots:=array:
                                                                                        overlay_slot:
                                                                                            HorizontalAlignment:=horizontal_alignment.Fill
                                                                                            VerticalAlignment:=vertical_alignment.Fill
                                                                                            Widget:=texture_block:
                                                                                                DefaultImage:=wooden_pixel_gui.textures.Panels.BrownSquare
                                                                                                DefaultTint:= MakeColorFromHex("0000004D")
                                                                                        overlay_slot:
                                                                                            HorizontalAlignment:=horizontal_alignment.Center
                                                                                            VerticalAlignment:=vertical_alignment.Center
                                                                                            Widget:=stack_box:
                                                                                                Orientation:=orientation.Horizontal
                                                                                                Slots:=array:
                                                                                                    stack_box_slot:
                                                                                                        Widget:=PlusSignText
                                                                                                    stack_box_slot:
                                                                                                        HorizontalAlignment:=horizontal_alignment.Center
                                                                                                        VerticalAlignment:=vertical_alignment.Center
                                                                                                        Widget:=texture_block:
                                                                                                            DefaultImage:=GemTexture
                                                                                                    stack_box_slot:
                                                                                                        Widget:=GemsButtonText
                                                                                        overlay_slot:
                                                                                            HorizontalAlignment:=horizontal_alignment.Fill
                                                                                            VerticalAlignment:=vertical_alignment.Fill
                                                                                            Widget:=overlay:
                                                                                                Slots:=array:
                                                                                                    overlay_slot:
                                                                                                        HorizontalAlignment:=horizontal_alignment.Fill
                                                                                                        VerticalAlignment:=vertical_alignment.Fill
                                                                                                        Widget:=PrestigeButton
                                                                                                        
                                                         

        return MyCanvas

    ShowUi(Agent:agent):void=

        RemoveFortHud(Agent)

        MyCanvas: canvas = MakeEntireBackground(Agent)
        
        if(PlayerUI:=GetPlayerUI[player[Agent]]):
            PlayerUI.AddWidget(MyCanvas,player_ui_slot{InputMode:=ui_input_mode.All}) # want it to appear on top
            MyCanvas.SetVisibility(widget_visibility.Hidden)
            MyCanvas.SetVisibility(widget_visibility.Visible)
        if(set MyUI[Agent]= option{MyCanvas}){}

    RemoveHUD(Wdata:widget_message):void =

        Print("Removing")
        if(Player:=Wdata.Player, PlayerUI:=GetPlayerUI[Player],SelectedButton:=Wdata.Source,Agent:=agent[Player]):
            if(MyActualUI := MyUI[Agent]?):

                PlayerUI.RemoveWidget(MyActualUI)
                MyActualUI.SetVisibility(widget_visibility.Visible)
                MyActualUI.SetVisibility(widget_visibility.Hidden)

                if(set MyUI[Agent] = false){}
                if(set TextBlockPerAgent[Agent] = false){}

                ShowFortHud(Agent)


    PrestigeHandle(Wdata:widget_message):void=
        if(Player:=Wdata.Player, PlayerUI:=GetPlayerUI[Player],SelectedButton:=Wdata.Source,Agent:=agent[Player]):
            if (MyData := Data[Player]):
                GemCount : float = PrestigeDevice.CaculateAwardedGems(Agent)
                if(GemCount >= 1.0):
                    RemoveHUD(Wdata)
                    PrestigeDevice.GrantGems(Agent)
                    PrestigeDevice.Huds.CMBDevice.LevelUpAudio(Agent)
                    PurchaseAudio.Play(Agent)
                else:
                    Print("Not enough gems")
                    NotEnoughEarnigns.Show(Agent)

                    
                    