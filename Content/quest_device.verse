
using { /Fortnite.com/Devices }
using { /Verse.org/Simulation }
using { /UnrealEngine.com/Temporary/Diagnostics }
using { /Verse.org/Random }
using { /Verse.org/Colors }
using { /Verse.org/Assets }

quest_data := class(): 

    var Value : int 
    var Target : int 
    var RandomTargets : []int
    var Name : string 
    var GoldReward : int 
    var XpReward : int 

quest_device := class(creative_device):

    @editable
    QuestAccolade : accolades_device = accolades_device{}

    @editable
    Spawners : []player_spawner_device = array{}

    @editable
    CustomTrackerDevice : custom_tracker_device = custom_tracker_device{}
    @editable
    CineHudDevice : npc_hp_cinematic_hud_device = npc_hp_cinematic_hud_device{}
    @editable
    CompleteQuestHMD : hud_message_device = hud_message_device{}

    @editable
    BoarTrigger : trigger_device = trigger_device{}
    @editable
    BuffaloTrigger : trigger_device = trigger_device{}
    @editable
    LizardTrigger : trigger_device = trigger_device{}
    @editable
    DesertGiantTrigger : trigger_device = trigger_device{}
    @editable
    ScorpionTrigger : trigger_device = trigger_device{}

    @editable
    DesertGolemTrigger : trigger_device = trigger_device{}

    @editable
    LeopardTrigger : trigger_device = trigger_device{}
    @editable
    RamTrigger : trigger_device = trigger_device{}
    @editable
    YetiTrigger : trigger_device = trigger_device{}
    @editable
    HoundTrigger : trigger_device = trigger_device{}
    @editable
    LavaJawTrigger : trigger_device = trigger_device{}
    @editable
    TerravoreTrigger : trigger_device = trigger_device{}

    @editable
    OreGranters : []item_granter_device = array{}

    @editable
    QuestMissionTrigger : trigger_device = trigger_device{}
    @editable
    StartMissionTrigger : trigger_device = trigger_device{}


    @editable
    CMDevice : combat_level_device = combat_level_device{}

    @editable
    TrackerDevices : []tracker_device = array{}
    @editable
    TrackerDevices2 : []tracker_device = array{}
    @editable
    TrackerDevices3 : []tracker_device = array{}
    @editable
    TrackerDevices4 : []tracker_device = array{}

    @editable
    TriggerDevices : []trigger_device = array{}
    @editable
    TriggerDevices1 : []trigger_device = array{}
    @editable
    TriggerDevices2 : []trigger_device = array{}
    @editable
    TriggerDevices3 : []trigger_device = array{}

    @editable
    ButtonTestAssign : button_device = button_device{}



    storedstring<localizes>(String:string):message = "{String}"
   
    OnBegin<override>()<suspends>:void=

        BoarTrigger.TriggeredEvent.SubscribeMaybeAgent(IncrementValueTriggers,0)
        BuffaloTrigger.TriggeredEvent.SubscribeMaybeAgent(IncrementValueTriggers,1)
        LizardTrigger.TriggeredEvent.SubscribeMaybeAgent(IncrementValueTriggers,2)
        DesertGiantTrigger.TriggeredEvent.SubscribeMaybeAgent(IncrementValueTriggers,3)
        ScorpionTrigger.TriggeredEvent.SubscribeMaybeAgent(IncrementValueTriggers,4)

        DesertGolemTrigger.TriggeredEvent.SubscribeMaybeAgent(IncrementValueTriggers,5)
        LeopardTrigger.TriggeredEvent.SubscribeMaybeAgent(IncrementValueTriggers,6)
        RamTrigger.TriggeredEvent.SubscribeMaybeAgent(IncrementValueTriggers,7)
        YetiTrigger.TriggeredEvent.SubscribeMaybeAgent(IncrementValueTriggers,8)
        HoundTrigger.TriggeredEvent.SubscribeMaybeAgent(IncrementValueTriggers,9)
        LavaJawTrigger.TriggeredEvent.SubscribeMaybeAgent(IncrementValueTriggers,10)
        TerravoreTrigger.TriggeredEvent.SubscribeMaybeAgent(IncrementValueTriggers,11)

        ButtonTestAssign.InteractedWithEvent.Subscribe(AssignQuestTest)

        Players:=GetPlayspace().GetPlayers()
        for(Player:Players):
            if(Agent:=agent[Player]):
                LoadQuest(Agent)

        for(Spawner:Spawners):
            Spawner.SpawnedEvent.Subscribe(LoadQuestDelay)

    AssignQuestTest(Agent:agent):void=
        if:
            MyTrigger2 := TriggerDevices2[3]
        then:
            MyTrigger2.Trigger(Agent)
        
        

    CalculateExpReward(Agent:agent):int= 
        var Reward : int = 0
        if(MyData := Data[player[Agent]]):
            PlayerLevel := MyData.Level
            QuestMultiple := MyData.QuestRewardMultiple
            if(MyXpBar := CMDevice.XpBarPerPlayer[Agent]?):
                PriceFloat : float = ((MyXpBar.CalculateExpReq(PlayerLevel) / 5.0) * QuestMultiple ) / (((PlayerLevel * 1.0) / 30.0 ) + 1.0) + ((MyXpBar.CalculateExpReq(PlayerLevel) / 5.0) * QuestMultiple )/ (((PlayerLevel * 1.0) / 30.0 ) + 1.0) * 0.15 * MyData.QuestRewardLevel
                Print("Exp Reward : {PriceFloat}")
                if(RoundedPrice := Round[PriceFloat]){set Reward = RoundedPrice}
                
        return Reward

    GetOreRewardType(Agent:agent):int = 

        var RewardType : int = 0

        if(MyData := Data[player[Agent]]):
            QuestNum := MyData.QuestNumber
            Print("QuestNum := {QuestNum}")
            Area := (QuestNum*1.0) / 3.0
            if(RoundedArea := Floor[Area]){set RewardType = RoundedArea}

        Print("RewardType = {RewardType}")
                    
        return RewardType
    
    GetOreRewardTypeText(Agent:agent, RewardType:int):string=

        var RewardString : string = "Copper Ore"

        if (RewardType = 1):
            set RewardString = "Silver Ore"
        if (RewardType = 2):
            set RewardString = "Malchite Ore"
        if (RewardType = 3):
            set RewardString = "Obsidian Ore"
        if (RewardType = 4):
            set RewardString = "Brightcore Ore"
        if (RewardType = 5):
            set RewardString = "Spectrolite Ore"

        return RewardString

        
    GetOreRewardNumber(Agent:agent):int = 

        var Reward : int = 0

        FlatOreAmount : int = 2

        if(MyData := Data[player[Agent]]):
            QuestNum := MyData.QuestNumber
            if(QuestNum < 100):
                if(QuestMod := Mod[QuestNum,3] + 1):
                    if(RoundedReward := (Round[(QuestMod * (FlatOreAmount * 1.0))] * MyData.QuestRewardMultiple) + MyData.QuestRewardLevel){set Reward = RoundedReward}
                    
        return Reward


    GetRandomQuest(Agent:agent,Area1:logic,Area2:logic,Area3:logic,Area4:logic):int=

        var QuestFound : int = 100

        var ArrIndex : []int = array{0,1,2,3,4,5,6,7,8,9,10,11}

        if(Area1 = false):
            Print("Area1 = false")
        if(Area2 = false):
            Print("Area2 = false")
        if(Area3 = false):
            Print("Area3 = false")
        if(Area4 = false):
            Print("Area4= false")

        if(MyData := Data[player[Agent]]):
            AreaArr:=MyData.AreasUnlocked

            for(i:=0..AreaArr.Length-1):
                if(Area := AreaArr[i]):
                    if(Area = false):
                        Print("Area = false")
                        for(i2:=0..ArrIndex.Length-1):
                            if(i = ArrIndex[i2]):
                                if(NewArr := ArrIndex.RemoveElement[i2]):
                                    set ArrIndex = NewArr

            if(Area1 = false):
                for(CheckNum:=0..2):
                    for(i:=0..ArrIndex.Length-1):
                        if(CheckNum = ArrIndex[i]):
                            if(NewArr := ArrIndex.RemoveElement[i]):
                                set ArrIndex = NewArr
            if(Area2 = false):
                for(CheckNum:=3..5):
                    for(i:=0..ArrIndex.Length-1):
                        if(CheckNum = ArrIndex[i]):
                            if(NewArr := ArrIndex.RemoveElement[i]):
                                set ArrIndex = NewArr
            if(Area3 = false):
                for(CheckNum:=6..8):
                    for(i:=0..ArrIndex.Length-1):
                        if(CheckNum = ArrIndex[i]):
                            if(NewArr := ArrIndex.RemoveElement[i]):
                                set ArrIndex = NewArr
            if(Area4 = false):
                for(CheckNum:=9..11):
                    for(i:=0..ArrIndex.Length-1):
                        if(CheckNum = ArrIndex[i]):
                            if(NewArr := ArrIndex.RemoveElement[i]):
                                set ArrIndex = NewArr

            if(MyData.Level < 55):
                for(i:=0..ArrIndex.Length-1):
                    if(ArrIndex[i] = 11):
                        if(NewArr := ArrIndex.RemoveElement[i]):
                            set ArrIndex = NewArr
                    

        Print("ArrIndex.LEgnth = {ArrIndex.Length}")
        for(i:=0..ArrIndex.Length-1):
            if(AR := ArrIndex[i]):
                Print("ArrIndex {i} = {AR}")

        if(ArrIndex.Length < 1):
            return QuestFound
        else:
            RandQuestIndex := GetRandomInt(0,ArrIndex.Length-1)
            if(RandomInd := ArrIndex[RandQuestIndex]):
                set QuestFound = RandomInd
                Print("QuestFound = {QuestFound}")
                return QuestFound
        return QuestFound


    ActuallyAssignQuest(Agent:agent,QuestNumber:int):void=

        if:
            Player := player[Agent]
            MyData := Data[Player]
            Id := CineHudDevice.IdPerAgent[Agent]?
            MyTracker := TrackerDevices[QuestNumber]
            MyTracker1 := TrackerDevices2[QuestNumber]
            MyTracker2 := TrackerDevices3[QuestNumber]
            MyTracker3 := TrackerDevices4[QuestNumber]

            MyTrigger := TriggerDevices[QuestNumber]
            MyTrigger1 := TriggerDevices1[QuestNumber]
            MyTrigger2 := TriggerDevices2[QuestNumber]
            MyTrigger3 := TriggerDevices3[QuestNumber]

        
        then:

            StartMissionTrigger.Trigger(Agent)
            AllQuestData : []quest_data = GetAllQuestData()
            if(RandQuest := AllQuestData[QuestNumber]):

                RandomMultiple := GetRandomFloat(1.0,3.0)
                RandomAddOn := GetRandomInt(3,9)
                if:
                    TargetAlgo := (Round[((MyData.Level - (QuestNumber * 5)) * 3.0) * RandomMultiple]) + RandomAddOn
                    RewardMutiple := Round[RandomMultiple]
                then:
                    #CustomTrackerDevice.InitCustomUI(Agent,RandQuest.Name, TargetAlgo, 0, RandQuest.GoldReward)
                    if:
                        OldData := Data[Player]
                        set Data[Player] = player_data{player_data_update<constructor>(OldData),QuestTarget := TargetAlgo,QuestValue := 0, QuestNumber := QuestNumber, QuestRewardMultiple := RewardMutiple}
                    then:
                        if (Id = 0):

                            for(i:=0..TrackerDevices.Length-1):
                                if(LoopTracker := TrackerDevices[i]):
                                    LoopTracker.Remove(Agent)

                            MyTracker.Reset(Agent)
                            MyTracker.SetTarget(TargetAlgo)

                            MyTrigger.Trigger(Agent)

                        if (Id = 1):

                            for(i:=0..TrackerDevices2.Length-1):
                                if(LoopTracker := TrackerDevices2[i]):
                                    LoopTracker.Remove(Agent)

                            MyTracker1.Reset(Agent)
                            MyTracker1.SetTarget(TargetAlgo)

                            MyTrigger1.Trigger(Agent)


                        if (Id = 2):

                            for(i:=0..TrackerDevices3.Length-1):
                                if(LoopTracker := TrackerDevices3[i]):
                                    LoopTracker.Remove(Agent)

                            MyTracker2.Reset(Agent)
                            MyTracker2.SetTarget(TargetAlgo)
                            
                            MyTrigger2.Trigger(Agent)

                        if (Id = 3):

                            for(i:=0..TrackerDevices4.Length-1):
                                if(LoopTracker := TrackerDevices4[i]):
                                    LoopTracker.Remove(Agent)

                            MyTracker3.Reset(Agent)
                            MyTracker3.SetTarget(TargetAlgo)

                            MyTrigger3.Trigger(Agent)
                        

    
    UpdateTracker(Agent:agent):void= 
        AllQuestData : []quest_data = GetAllQuestData()
        if:
            Player:=player[Agent]
            OldData := Data[Player]
            QuestNumber := OldData.QuestNumber
            Quest := AllQuestData[QuestNumber]
        then:
            #CustomTrackerDevice.UpdateTrackers(Agent,Quest.Name, OldData.QuestTarget, OldData.QuestValue, Quest.GoldReward)

    LoadQuestDelay(Agent:agent):void=
        spawn:
            LoadQuest(Agent)


    LoadQuest(Agent:agent)<suspends>:void=
        Sleep(1.0)
        #AllQuestData := GetAllQuestData()
        if:
            Player:=player[Agent]
            OldData := Data[Player]
            var QuestNumber : int = OldData.QuestNumber
            #Quest := AllQuestData[QuestNumber]
            SavedCurrent : int = OldData.QuestValue
            SavedTarget : int = OldData.QuestTarget
            QuestNumber <= 13

            Id := CineHudDevice.IdPerAgent[Agent]?
            MyTracker := TrackerDevices[QuestNumber]
            MyTracker1 := TrackerDevices2[QuestNumber]
            MyTracker2 := TrackerDevices3[QuestNumber]
            MyTracker3 := TrackerDevices4[QuestNumber]

            MyTrigger := TriggerDevices[QuestNumber]
            MyTrigger1 := TriggerDevices1[QuestNumber]
            MyTrigger2 := TriggerDevices2[QuestNumber]
            MyTrigger3 := TriggerDevices3[QuestNumber]
        then:

            Print("Saved Current {SavedCurrent}")

            if (Id = 0):

                MyTracker.Reset(Agent)
                MyTrigger.Trigger(Agent)
                MyTracker.SetTarget(OldData.QuestTarget)

                Sleep(3.0)
                spawn:
                    SetTrackerCurrent(MyTracker,Agent,SavedCurrent)
                
            if (Id = 1):

                MyTracker1.Reset(Agent)
                MyTrigger1.Trigger(Agent)
                MyTracker1.SetTarget(OldData.QuestTarget)

                Sleep(3.0)
                spawn:
                    SetTrackerCurrent(MyTracker1,Agent,SavedCurrent)
                
            if (Id = 2):

                MyTracker2.Reset(Agent)
                MyTrigger2.Trigger(Agent)
                MyTracker2.SetTarget(OldData.QuestTarget)

                Sleep(3.0)
                spawn:
                    SetTrackerCurrent(MyTracker2,Agent,SavedCurrent)

                
            if (Id = 3):

                MyTracker3.Reset(Agent)
                MyTrigger3.Trigger(Agent)
                MyTracker3.SetTarget(OldData.QuestTarget)

                Sleep(3.0)
                spawn:
                    SetTrackerCurrent(MyTracker3,Agent,SavedCurrent)
                
               
                
                
    SetTrackerCurrent(MyTracker:tracker_device,Agent:agent,Current:int)<suspends>:void=
        for(i:=0..Current-1):
            MyTracker.Increment(Agent)
            Sleep(0.025)
        




    IncrementValueTriggers(MaybeAgent:?agent,MissionNumber:int):void=
        if:
            Agent := MaybeAgent?
            Player:=player[Agent]
            OldData := Data[Player]
            QuestNumber : int = OldData.QuestNumber
            QuestNumber = MissionNumber

            set Data[Player] = player_data{player_data_update<constructor>(OldData),QuestValue := OldData.QuestValue + 1}

            Id := CineHudDevice.IdPerAgent[Agent]?
            MyTracker := TrackerDevices[QuestNumber]
            MyTracker1 := TrackerDevices2[QuestNumber]
            MyTracker2 := TrackerDevices3[QuestNumber]
            MyTracker3 := TrackerDevices4[QuestNumber]

        then:

            Print("Quest Value{OldData.QuestValue }")

            if (Id = 0):
                MyTracker.Increment(Agent)
            if (Id = 1):
                MyTracker1.Increment(Agent)
            if (Id = 2):
                MyTracker2.Increment(Agent)
            if (Id = 3):
                MyTracker3.Increment(Agent)

            UpdateTracker(Agent)
            CheckIfMissionComplete(Agent)

    CheckIfMissionComplete(Agent:agent):void=
        if:
            Player:=player[Agent]
            OldData := Data[Player]
            OldData.QuestTarget <= OldData.QuestValue
        then:
            QuestMissionTrigger.Trigger(Agent)
            CompleteQuest(Agent)


    CompleteQuest(Agent:agent):void=

        AllQuestData := GetAllQuestData()

        QuestReward := CalculateExpReward(Agent)
        OreRewardType := GetOreRewardType(Agent)
        OreRewardAmount := GetOreRewardNumber(Agent)

        if(MyGranter := OreGranters[OreRewardType]):
            for(i:=0..OreRewardAmount-1):
                MyGranter.GrantItem(Agent)

        
        if:
            Player:=player[Agent]
            OldData := Data[Player]
            Quest := AllQuestData[OldData.QuestNumber]
            set Data[Player] = player_data{player_data_update<constructor>(OldData),Xp := OldData.Xp + (QuestReward *1.0), QuestNumber := 100, TotalQuestCompleted := OldData.TotalQuestCompleted+1, QuestRewardMultiple:=0}
        then:

            CompleteQuestHMD.SetText(storedstring("{Quest.Name}"))
            CompleteQuestHMD.Show(Agent)
            CineHudDevice.AddXp(Agent,(Quest.XpReward)*1.0)
            QuestAccolade.Award(Agent)

            #CustomTrackerDevice.RemoveTracker(Agent)

    UnAssignQuest(Agent:agent):void=
        if:
            Player:=player[Agent]
            OldData := Data[Player]
            

            Id := CineHudDevice.IdPerAgent[Agent]?
            MyTracker := TrackerDevices[OldData.QuestNumber]
            MyTracker1 := TrackerDevices2[OldData.QuestNumber]
            MyTracker2 := TrackerDevices3[OldData.QuestNumber]
            MyTracker3 := TrackerDevices4[OldData.QuestNumber]

            set Data[Player] = player_data{player_data_update<constructor>(OldData), QuestNumber := 100, QuestRewardMultiple:=0}
        then:

            if (Id = 0):
                MyTracker.Remove(Agent)
            if (Id = 1):
                MyTracker1.Remove(Agent)
            if (Id = 2):
                MyTracker2.Remove(Agent)
            if (Id = 3):
                MyTracker3.Remove(Agent)
            
            #CustomTrackerDevice.RemoveTracker(Agent)



    GetAllQuestData():[]quest_data=

        AllQuestData : []quest_data = array:

            QData1 := quest_data :
                Value:=0
                Target:=100
                RandomTargets:= array:
                    50
                    100
                    150
                    200
                Name:="Wild Boar Eliminations"
                GoldReward:=9
                XpReward:=5
            QData2 := quest_data :
                Value:=0
                Target:=100
                RandomTargets:= array:
                    50
                    100
                    150
                    200
                Name:="Buffalo Eliminations"
                GoldReward:=9
                XpReward:=5
            QData3 := quest_data :
                Value:=0
                Target:=100
                RandomTargets:= array:
                    50
                    100
                    150
                    200
                Name:="Lizard Archer Eliminations"
                GoldReward:=9
                XpReward:=5
            QData4 := quest_data :
                Value:=0
                Target:=100
                RandomTargets:= array:
                    50
                    100
                    150
                    200
                Name:="Desert Giant Eliminations"
                GoldReward:=9
                XpReward:=5
            QData5 := quest_data :
                Value:=0
                Target:=100
                RandomTargets:= array:
                    50
                    100
                    150
                    200
                Name:="Scorpion Eliminations"
                GoldReward:=9
                XpReward:=5
            QData6 := quest_data :
                Value:=0
                Target:=100
                RandomTargets:= array:
                    50
                    100
                    150
                    200
                Name:="Golem Eliminations"
                GoldReward:=9
                XpReward:=5
            QData7 := quest_data :
                Value:=0
                Target:=100
                RandomTargets:= array:
                    50
                    100
                    150
                    200
                Name:="Leopard Eliminations"
                GoldReward:=9
                XpReward:=5

            QData8 := quest_data :
                Value:=0
                Target:=100
                RandomTargets:= array:
                    50
                    100
                    150
                    200
                Name:="Spider Eliminations"
                GoldReward:=9
                XpReward:=5

            QData9 := quest_data :
                Value:=0
                Target:=100
                RandomTargets:= array:
                    50
                    100
                    150
                    200
                Name:="Yeti Eliminations"
                GoldReward:=9
                XpReward:=5
            QData10 := quest_data :
                Value:=0
                Target:=100
                RandomTargets:= array:
                    50
                    100
                    150
                    200
                Name:="Hound Eliminations"
                GoldReward:=9
                XpReward:=5

            QData11 := quest_data :
                Value:=0
                Target:=100
                RandomTargets:= array:
                    50
                    100
                    150
                    200
                Name:="Lava Jaw Eliminations"
                GoldReward:=9
                XpReward:=5

            QData12 := quest_data :
                Value:=0
                Target:=100
                RandomTargets:= array:
                    50
                    100
                    150
                    200
                Name:="Terravore Eliminations"
                GoldReward:=9
                XpReward:=5


        return AllQuestData




        

