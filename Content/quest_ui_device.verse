using { /Fortnite.com/Devices }
using { /Verse.org/Simulation }
using { /UnrealEngine.com/Temporary/Diagnostics }
using { /UnrealEngine.com/Temporary/UI }
using { /UnrealEngine.com/Temporary/SpatialMath }
using { /Fortnite.com/UI }
using { /Verse.org/Colors }
using { /Fortnite.com/Playspaces }
using { /Fortnite.com/Characters }
using { /Verse.org/Assets }

quest_ui_device := class(creative_device):


    @editable
    OpenShop : button_device = button_device{}

    @editable
    NewQuestAudio : audio_player_device = audio_player_device{}
    @editable
    TickBoxAudio : audio_player_device = audio_player_device{}

    @editable
    OpenAudio : audio_player_device = audio_player_device{}
    @editable
    CloseAudio : audio_player_device = audio_player_device{}

    # Just to pull % multtipliers
    @editable
    NPCManager : npc_manager_device = npc_manager_device{}

    @editable
    QuestDevice : quest_device = quest_device{}

    var MyUI : [agent]?canvas = map{}
    var TickBoxes : [agent][]texture_block = map{}

    storedstring<localizes>(S:string) : message = "{S}"

    OnBegin<override>()<suspends>:void=

        OpenShop.InteractedWithEvent.Subscribe(ShowUi)

    MakeEntireBackground (Agent:agent):canvas=

        
        TitleBarTexture := texture_block :
            DefaultDesiredSize:=vector2{X:=16.0,Y:=90.0}
            DefaultImage:=wooden_pixel_gui.textures.Panels.t_TitleBar

        ScrollTextureBlock := texture_block :
            DefaultDesiredSize:=vector2{X:=200.0,Y:=200.0}
            DefaultImage:=wooden_pixel_gui.textures.Icons2.Size32.t_ScrollRoll_32

        ScrollTShadow := texture_block :
            DefaultDesiredSize:=vector2{X:=200.0,Y:=200.0}
            DefaultImage:=wooden_pixel_gui.textures.Icons2.Size32.t_ScrollRoll_32
            DefaultTint:=MakeColorFromHex("4A331BFF")

        TickBoxTexture1 := texture_block : 
            DefaultDesiredSize:=vector2{X:=75.0,Y:=75.0}
            DefaultImage:=wooden_pixel_gui.textures.Icons2.Size8.t_XYellow_8
        TickBoxTexture2 := texture_block : 
            DefaultDesiredSize:=vector2{X:=75.0,Y:=75.0}
            DefaultImage:=wooden_pixel_gui.textures.Icons2.Size8.t_XYellow_8
        TickBoxTexture3 := texture_block : 
            DefaultDesiredSize:=vector2{X:=75.0,Y:=75.0}
            DefaultImage:=wooden_pixel_gui.textures.Icons2.Size8.t_XYellow_8
        TickBoxTexture4 := texture_block : 
            DefaultDesiredSize:=vector2{X:=75.0,Y:=75.0}
            DefaultImage:=wooden_pixel_gui.textures.Icons2.Size8.t_XYellow_8


        TickBox1 := button_quiet :
            DefaultText:=storedstring("")
        TickBox1.OnClick().SubscribeMessage(AreaTickBox,0)

        TickBox2 := button_quiet :
            DefaultText:=storedstring("")
        TickBox2.OnClick().SubscribeMessage(AreaTickBox,1)

        TickBox3 := button_quiet :
            DefaultText:=storedstring("")
        TickBox3.OnClick().SubscribeMessage(AreaTickBox,2)

        TickBox4:= button_quiet :
            DefaultText:=storedstring("")
        TickBox4.OnClick().SubscribeMessage(AreaTickBox,3)

        NewQuestButton:= button_quiet :
            DefaultText:=storedstring("        New Quest        ")
        NewQuestButton.OnClick().Subscribe(AssignRandomQuest)

        CloseWindowButton := button_quiet :
            DefaultText:=storedstring("               Close               ")
        CloseWindowButton.OnClick().Subscribe(RemoveHUD)

        QuestNameTextBlock := text_block:
            DefaultText:=storedstring("None")
            DefaultTextColor := MakeColorFromHex("FFFFFF")

        if(MyData := Data[player[Agent]]):
            if(MyData.QuestNumber < 100):
                QD : []quest_data = QuestDevice.GetAllQuestData()
                if(CurrentQD := QD[MyData.QuestNumber]):
                    QuestNameTextBlock.SetText(storedstring("{CurrentQD.Name}"))

        QuestValuesTextBlock := text_block:
            DefaultText:=storedstring("{0}/{0}")
            DefaultTextColor := MakeColorFromHex("FFFFFF")

        if(MyData := Data[player[Agent]]):
            if(MyData.QuestNumber < 100):
                QuestValuesTextBlock.SetText(storedstring("{MyData.QuestValue}/{MyData.QuestTarget}"))

        OreTextureBlock := texture_block :
            DefaultDesiredSize:=vector2{X:=32.0,Y:=32.0}
            DefaultImage:=GetOreRewardTexture(Agent)

        ExpRewardTextureBlock := texture_block :
            DefaultDesiredSize:=vector2{X:=32.0,Y:=32.0}
            DefaultImage:=wooden_pixel_gui.textures.Icons2.Size16.t_Star_16
            DefaultTint := MakeColorFromHex("FFEB00FF")

        OreNameTextBlock := text_block :
            DefaultText:=storedstring("--")
            DefaultTextColor:=MakeColorFromHex("FFFFFF")

        OreNameTextBlock.SetText(storedstring("{QuestDevice.GetOreRewardTypeText(Agent, QuestDevice.GetOreRewardType(Agent))}"))

        OreAmountTextBlock := text_block:
            DefaultText:=storedstring("{QuestDevice.GetOreRewardNumber(Agent)}")
            DefaultTextColor:=MakeColorFromHex("FFFFFF")

        ExpAmountTextBlock := text_block : 
            DefaultText:=storedstring("{QuestDevice.CalculateExpReward(Agent)}")
            DefaultTextColor:=MakeColorFromHex("FFEB00FF")

        ExpAmountTextBlock.SetText(storedstring("{formatter.Abbreviate(QuestDevice.CalculateExpReward(Agent)*1.0,false)}"))

        ResetPriceTextBlock := text_block :
            DefaultText:=storedstring("0")
            DefaultTextColor:=MakeColorFromHex("FFFFFF")

        ResetPriceTextBlock.SetText(storedstring("{formatter.Abbreviate(CalculatePrice(Agent)*1.0,false)}"))

        TempTickBoxes : []texture_block = array{TickBoxTexture1,TickBoxTexture2,TickBoxTexture3,TickBoxTexture4}
        if(set TickBoxes[Agent] = TempTickBoxes){}

        LoadTickInfo(Agent)

        Overlay1 := overlay:
            Slots:=array:
                overlay_slot:
                    HorizontalAlignment:=horizontal_alignment.Fill
                    VerticalAlignment:=vertical_alignment.Fill
                    Padding:=margin{Left:=-5.0,Bottom:=-35.0,Top:=-5.0,Right:=-5.0}
                    Widget:=texture_block:
                        DefaultImage:=wooden_pixel_gui.textures.Panels.WhiteSquare100
                        DefaultTint:=MakeColorFromHex("543C22FF")
                overlay_slot:
                    HorizontalAlignment:=horizontal_alignment.Fill
                    VerticalAlignment:=vertical_alignment.Fill
                    Padding:=margin{Left:=0.0,Bottom:=-30.0,Top:=0.0,Right:=0.0}
                    Widget:=texture_block:
                        DefaultImage:=wooden_pixel_gui.textures.Panels.WhiteSquare100
                        DefaultTint:=MakeColorFromHex("6C4B2AE6")
                overlay_slot:
                    HorizontalAlignment:=horizontal_alignment.Fill
                    VerticalAlignment:=vertical_alignment.Fill
                    Padding:=margin{Left:=0.0,Bottom:=0.0,Top:=0.0,Right:=0.0}
                    Widget := stack_box:
                        Orientation:=orientation.Vertical
                        Slots:=array:
                            stack_box_slot:
                                HorizontalAlignment:=horizontal_alignment.Fill
                                VerticalAlignment:=vertical_alignment.Fill
                                Widget := overlay:
                                    Slots:=array:
                                        overlay_slot:
                                            HorizontalAlignment:=horizontal_alignment.Fill
                                            VerticalAlignment:=vertical_alignment.Fill
                                            Padding:=margin{Left:=0.0,Top:=0.0,Right:=0.0,Bottom:=7.0}
                                            Widget:= TitleBarTexture
                                        overlay_slot:
                                            HorizontalAlignment:=horizontal_alignment.Center
                                            VerticalAlignment:=vertical_alignment.Center
                                            Widget := text_block :
                                                DefaultText:=storedstring("Quests")
                                                DefaultTextColor := MakeColorFromHex("543C22FF")
                            stack_box_slot:
                                HorizontalAlignment:=horizontal_alignment.Center
                                VerticalAlignment:=vertical_alignment.Center
                                Padding:=margin{Left:=0.0,Top:=0.0,Right:=0.0,Bottom:=0.0}
                                Widget := stack_box:
                                    Orientation:=orientation.Horizontal
                                    Slots:=array:
                                        stack_box_slot:
                                            HorizontalAlignment:=horizontal_alignment.Fill
                                            VerticalAlignment:=vertical_alignment.Fill
                                            Padding:=margin{Left:=0.0,Top:=0.0,Right:=0.0,Bottom:=0.0}
                                            Widget := stack_box:
                                                Orientation:=orientation.Vertical
                                                Slots:=array:
                                                    stack_box_slot:
                                                        HorizontalAlignment:=horizontal_alignment.Fill
                                                        VerticalAlignment:=vertical_alignment.Fill
                                                        Padding:=margin{Left:=0.0,Top:=30.0,Right:=0.0,Bottom:=0.0}
                                                        Widget := overlay:
                                                            Slots:=array:
                                                                overlay_slot:
                                                                    HorizontalAlignment:=horizontal_alignment.Center
                                                                    VerticalAlignment:=vertical_alignment.Center
                                                                    Padding:=margin{Left:=-5.0,Top:=-5.0,Right:=-5.0,Bottom:=-5.0}
                                                                    Widget:=texture_block :
                                                                        DefaultDesiredSize:=vector2{X:=464.0,Y:=227.0}
                                                                        DefaultImage:=wooden_pixel_gui.textures.Panels.WhiteSquare100
                                                                        DefaultTint:=MakeColorFromHex("3A2713FF")
                                                                overlay_slot:
                                                                    HorizontalAlignment:=horizontal_alignment.Center
                                                                    VerticalAlignment:=vertical_alignment.Center
                                                                    Padding:=margin{Left:=50.0,Top:=10.0,Right:=50.0,Bottom:=10.0}
                                                                    Widget:=texture_block :
                                                                        DefaultDesiredSize:=vector2{X:=453.0,Y:=218.0}
                                                                        DefaultImage:=wooden_pixel_gui.textures.Panels.WhiteSquare100
                                                                        DefaultTint:=MakeColorFromHex("4A331BFF")
                                                                overlay_slot:
                                                                    HorizontalAlignment:=horizontal_alignment.Center
                                                                    VerticalAlignment:=vertical_alignment.Center
                                                                    Padding:=margin{Left:=5.0,Top:=15.0,Right:=0.0,Bottom:=10.0}
                                                                    Widget:=ScrollTShadow
                                                                overlay_slot:
                                                                    HorizontalAlignment:=horizontal_alignment.Center
                                                                    VerticalAlignment:=vertical_alignment.Center
                                                                    Padding:=margin{Left:=0.0,Top:=10.0,Right:=0.0,Bottom:=10.0}
                                                                    Widget:=ScrollTextureBlock
                                                    stack_box_slot:
                                                        HorizontalAlignment:=horizontal_alignment.Left
                                                        VerticalAlignment:=vertical_alignment.Top
                                                        Padding:=margin{Left:=0.0,Top:=0.0,Right:=0.0,Bottom:=0.0}
                                                        Widget := stack_box:
                                                            Orientation:=orientation.Vertical
                                                            Slots:=array:
                                                                stack_box_slot:
                                                                    HorizontalAlignment:=horizontal_alignment.Center
                                                                    VerticalAlignment:=vertical_alignment.Center
                                                                    Padding:=margin{Left:=5.0,Top:=10.0,Right:=0.0,Bottom:=10.0}
                                                                    Widget := text_block :
                                                                        DefaultText:=storedstring("Areas Included:")
                                                                        DefaultTextColor := MakeColorFromHex("FFFFFF")
                                                                stack_box_slot:
                                                                    HorizontalAlignment:=horizontal_alignment.Left
                                                                    VerticalAlignment:=vertical_alignment.Top
                                                                    Padding:=margin{Left:=50.0,Top:=10.0,Right:=20.0,Bottom:=10.0}
                                                                    Widget := stack_box:
                                                                        Orientation:=orientation.Horizontal
                                                                        Slots:=array:
                                                                            stack_box_slot:
                                                                                HorizontalAlignment:=horizontal_alignment.Fill
                                                                                VerticalAlignment:=vertical_alignment.Fill
                                                                                Widget:= overlay:
                                                                                    Slots:=array:
                                                                                        overlay_slot:
                                                                                            Widget:=texture_block :
                                                                                                DefaultDesiredSize:=vector2{X:=75.0,Y:=75.0}
                                                                                                DefaultImage:=wooden_pixel_gui.textures.Panels.WhiteSquare100
                                                                                                DefaultTint:=MakeColorFromHex("4A331BFF")
                                                                                        overlay_slot:
                                                                                            Padding:=margin{Left:=5.0,Top:=5.0,Right:=0.0,Bottom:=0.0}
                                                                                            Widget:=TickBoxTexture1
                                                                                        overlay_slot:
                                                                                            Widget:=TickBox1
                                                                            stack_box_slot:
                                                                                HorizontalAlignment:=horizontal_alignment.Center
                                                                                VerticalAlignment:=vertical_alignment.Center
                                                                                Padding:=margin{Left:=20.0,Top:=0.0,Right:=0.0,Bottom:=0.0}
                                                                                Widget:=text_block:
                                                                                    DefaultText:=storedstring("Grassland")
                                                                                    DefaultTextColor := MakeColorFromHex("FFFFFF")

                                                                stack_box_slot:
                                                                    HorizontalAlignment:=horizontal_alignment.Left
                                                                    VerticalAlignment:=vertical_alignment.Top
                                                                    Padding:=margin{Left:=50.0,Top:=10.0,Right:=20.0,Bottom:=10.0}
                                                                    Widget := stack_box:
                                                                        Orientation:=orientation.Horizontal
                                                                        Slots:=array:
                                                                            stack_box_slot:
                                                                                HorizontalAlignment:=horizontal_alignment.Fill
                                                                                VerticalAlignment:=vertical_alignment.Fill
                                                                                Widget:= overlay:
                                                                                    Slots:=array:
                                                                                        overlay_slot:
                                                                                            Widget:=texture_block :
                                                                                                DefaultDesiredSize:=vector2{X:=75.0,Y:=75.0}
                                                                                                DefaultImage:=wooden_pixel_gui.textures.Panels.WhiteSquare100
                                                                                                DefaultTint:=MakeColorFromHex("4A331BFF")
                                                                                        overlay_slot:
                                                                                            Padding:=margin{Left:=5.0,Top:=5.0,Right:=0.0,Bottom:=0.0}
                                                                                            Widget:=TickBoxTexture2
                                                                                        overlay_slot:
                                                                                            Widget:=TickBox2
                                                                            stack_box_slot:
                                                                                HorizontalAlignment:=horizontal_alignment.Center
                                                                                VerticalAlignment:=vertical_alignment.Center
                                                                                Padding:=margin{Left:=20.0,Top:=0.0,Right:=0.0,Bottom:=0.0}
                                                                                Widget:=text_block:
                                                                                    DefaultText:=storedstring("Desert")
                                                                                    DefaultTextColor := MakeColorFromHex("FFFFFF")

                                                                stack_box_slot:
                                                                    HorizontalAlignment:=horizontal_alignment.Left
                                                                    VerticalAlignment:=vertical_alignment.Top
                                                                    Padding:=margin{Left:=50.0,Top:=10.0,Right:=20.0,Bottom:=10.0}
                                                                    Widget := stack_box:
                                                                        Orientation:=orientation.Horizontal
                                                                        Slots:=array:
                                                                            stack_box_slot:
                                                                                HorizontalAlignment:=horizontal_alignment.Fill
                                                                                VerticalAlignment:=vertical_alignment.Fill
                                                                                Widget:= overlay:
                                                                                    Slots:=array:
                                                                                        overlay_slot:
                                                                                            Widget:=texture_block :
                                                                                                DefaultDesiredSize:=vector2{X:=75.0,Y:=75.0}
                                                                                                DefaultImage:=wooden_pixel_gui.textures.Panels.WhiteSquare100
                                                                                                DefaultTint:=MakeColorFromHex("4A331BFF")
                                                                                        overlay_slot:
                                                                                            Padding:=margin{Left:=5.0,Top:=5.0,Right:=0.0,Bottom:=0.0}
                                                                                            Widget:=TickBoxTexture3
                                                                                        overlay_slot:
                                                                                            Widget:=TickBox3
                                                                            stack_box_slot:
                                                                                HorizontalAlignment:=horizontal_alignment.Center
                                                                                VerticalAlignment:=vertical_alignment.Center
                                                                                Padding:=margin{Left:=20.0,Top:=0.0,Right:=0.0,Bottom:=0.0}
                                                                                Widget:=text_block:
                                                                                    DefaultText:=storedstring("Tundra")
                                                                                    DefaultTextColor := MakeColorFromHex("FFFFFF")

                                                                stack_box_slot:
                                                                    HorizontalAlignment:=horizontal_alignment.Left
                                                                    VerticalAlignment:=vertical_alignment.Top
                                                                    Padding:=margin{Left:=50.0,Top:=10.0,Right:=20.0,Bottom:=10.0}
                                                                    Widget := stack_box:
                                                                        Orientation:=orientation.Horizontal
                                                                        Slots:=array:
                                                                            stack_box_slot:
                                                                                HorizontalAlignment:=horizontal_alignment.Fill
                                                                                VerticalAlignment:=vertical_alignment.Fill
                                                                                Widget:= overlay:
                                                                                    Slots:=array:
                                                                                        overlay_slot:
                                                                                            Widget:=texture_block :
                                                                                                DefaultDesiredSize:=vector2{X:=75.0,Y:=75.0}
                                                                                                DefaultImage:=wooden_pixel_gui.textures.Panels.WhiteSquare100
                                                                                                DefaultTint:=MakeColorFromHex("4A331BFF")
                                                                                        overlay_slot:
                                                                                            Padding:=margin{Left:=5.0,Top:=5.0,Right:=0.0,Bottom:=0.0}
                                                                                            Widget:=TickBoxTexture4
                                                                                        overlay_slot:
                                                                                            Widget:=TickBox4
                                                                            stack_box_slot:
                                                                                HorizontalAlignment:=horizontal_alignment.Center
                                                                                VerticalAlignment:=vertical_alignment.Center
                                                                                Padding:=margin{Left:=20.0,Top:=0.0,Right:=0.0,Bottom:=0.0}
                                                                                Widget:=text_block:
                                                                                    DefaultText:=storedstring("Volcano")
                                                                                    DefaultTextColor := MakeColorFromHex("FFFFFF")
                                        stack_box_slot:
                                            HorizontalAlignment:=horizontal_alignment.Fill
                                            VerticalAlignment:=vertical_alignment.Fill
                                            Padding:=margin{Left:=25.0,Top:=0.0,Right:=25.0,Bottom:=0.0}
                                            Widget:=texture_block:
                                                DefaultImage:=wooden_pixel_gui.textures.Panels.WhiteSquare100
                                                DefaultTint:=MakeColorFromHex("4A331BFF")
                                                DefaultDesiredSize:=vector2{X:=5.0,Y:=75.0}

                                        stack_box_slot:
                                            HorizontalAlignment:=horizontal_alignment.Fill
                                            VerticalAlignment:=vertical_alignment.Fill
                                            Padding:=margin{Left:=0.0,Top:=0.0,Right:=0.0,Bottom:=0.0}
                                            Widget := stack_box:
                                                Orientation:=orientation.Vertical
                                                Slots:=array:
                                                    stack_box_slot:
                                                        HorizontalAlignment:=horizontal_alignment.Fill
                                                        VerticalAlignment:=vertical_alignment.Fill
                                                        Padding:=margin{Left:=5.0,Top:=40.0,Right:=0.0,Bottom:=0.0}
                                                        Widget:=text_block:
                                                            DefaultText:=storedstring("Current Quest:")
                                                            DefaultTextColor:=MakeColorFromHex("FFFFFF")
                                                    stack_box_slot:
                                                        HorizontalAlignment:=horizontal_alignment.Fill
                                                        VerticalAlignment:=vertical_alignment.Fill
                                                        Padding:=margin{Left:=0.0,Top:=30.0,Right:=0.0,Bottom:=0.0}
                                                        Widget:=overlay:
                                                            Slots:=array:
                                                                overlay_slot:
                                                                    HorizontalAlignment:=horizontal_alignment.Fill
                                                                    VerticalAlignment:=vertical_alignment.Fill
                                                                    Widget:=texture_block:
                                                                        DefaultDesiredSize:=vector2{X:=539.0,Y:=53.0}
                                                                        DefaultImage:=wooden_pixel_gui.textures.Panels.WhiteSquare100
                                                                        DefaultTint:=MakeColorFromHex("4A331BFF")
                                                                overlay_slot:
                                                                    Padding:=margin{Left:=5.0,Top:=0.0,Right:=0.0,Bottom:=0.0}
                                                                    HorizontalAlignment:=horizontal_alignment.Left
                                                                    VerticalAlignment:=vertical_alignment.Center
                                                                    Widget:=QuestNameTextBlock
                                                                overlay_slot:
                                                                    Padding:=margin{Left:=0.0,Top:=0.0,Right:=5.0,Bottom:=0.0}
                                                                    HorizontalAlignment:=horizontal_alignment.Right
                                                                    VerticalAlignment:=vertical_alignment.Center
                                                                    Widget:=QuestValuesTextBlock
                                                    stack_box_slot:
                                                        HorizontalAlignment:=horizontal_alignment.Fill
                                                        VerticalAlignment:=vertical_alignment.Fill
                                                        Padding:=margin{Left:=0.0,Top:=30.0,Right:=0.0,Bottom:=0.0}
                                                        Widget:=text_block:
                                                            DefaultText:=storedstring("Reward:")
                                                            DefaultTextColor := MakeColorFromHex("FFFFFF")

                                                    stack_box_slot:
                                                        HorizontalAlignment:=horizontal_alignment.Fill
                                                        VerticalAlignment:=vertical_alignment.Fill
                                                        Padding:=margin{Left:=0.0,Top:=30.0,Right:=0.0,Bottom:=0.0}
                                                        Widget:=overlay:
                                                            Slots:=array:
                                                                overlay_slot:
                                                                    HorizontalAlignment:=horizontal_alignment.Fill
                                                                    VerticalAlignment:=vertical_alignment.Fill
                                                                    Widget:=texture_block:
                                                                        DefaultDesiredSize:=vector2{X:=539.0,Y:=53.0}
                                                                        DefaultImage:=wooden_pixel_gui.textures.Panels.WhiteSquare100
                                                                        DefaultTint:=MakeColorFromHex("4A331BFF")
                                                                overlay_slot:
                                                                    HorizontalAlignment:=horizontal_alignment.Center
                                                                    VerticalAlignment:=vertical_alignment.Center
                                                                    Widget:=stack_box:
                                                                        Orientation:=orientation.Horizontal
                                                                        Slots:=array:
                                                                            stack_box_slot:
                                                                                Widget:=OreAmountTextBlock
                                                                            stack_box_slot:
                                                                                Widget:=OreTextureBlock
                                                                            stack_box_slot:
                                                                                Widget:=OreNameTextBlock
                                                    stack_box_slot:
                                                        HorizontalAlignment:=horizontal_alignment.Fill
                                                        VerticalAlignment:=vertical_alignment.Fill
                                                        Padding:=margin{Left:=0.0,Top:=30.0,Right:=0.0,Bottom:=0.0}
                                                        Widget:=overlay:
                                                            Slots:=array:
                                                                overlay_slot:
                                                                    HorizontalAlignment:=horizontal_alignment.Fill
                                                                    VerticalAlignment:=vertical_alignment.Fill
                                                                    Widget:=texture_block:
                                                                        DefaultDesiredSize:=vector2{X:=539.0,Y:=53.0}
                                                                        DefaultImage:=wooden_pixel_gui.textures.Panels.WhiteSquare100
                                                                        DefaultTint:=MakeColorFromHex("4A331BFF")
                                                                overlay_slot:
                                                                    HorizontalAlignment:=horizontal_alignment.Center
                                                                    VerticalAlignment:=vertical_alignment.Center
                                                                    Widget:=stack_box:
                                                                        Orientation:=orientation.Horizontal
                                                                        Slots:=array:
                                                                            stack_box_slot:
                                                                                Widget:=ExpAmountTextBlock
                                                                            stack_box_slot:
                                                                                Widget:=ExpRewardTextureBlock
                                                                            stack_box_slot:
                                                                                Widget:=text_block:
                                                                                    DefaultText:=storedstring("Experience")
                                                                                    DefaultTextColor:=MakeColorFromHex("FFEB00FF")
                                                    stack_box_slot:
                                                        HorizontalAlignment:=horizontal_alignment.Center
                                                        VerticalAlignment:=vertical_alignment.Fill
                                                        Padding:=margin{Left:=0.0,Top:=100.0,Right:=0.0,Bottom:=0.0}
                                                        Widget:=text_block:
                                                            DefaultText:=storedstring("New Quest")
                                                            DefaultTextColor:=MakeColorFromHex("FFFFFFFF")
                                                    stack_box_slot:
                                                        Padding:=margin{Left:=0.0,Top:=15.0,Right:=0.0,Bottom:=0.0}
                                                        HorizontalAlignment:=horizontal_alignment.Center
                                                        VerticalAlignment:=vertical_alignment.Top
                                                        Widget:=stack_box:
                                                            Orientation:=orientation.Horizontal
                                                            Slots:=array:
                                                                stack_box_slot:
                                                                    Widget:=text_block:
                                                                        DefaultText:=storedstring("Price: ")
                                                                        DefaultTextColor:=MakeColorFromHex("FFFFFFFF")
                                                                stack_box_slot:
                                                                    Widget:=texture_block:
                                                                        DefaultDesiredSize:=vector2{X:=32.0,Y:=32.0}
                                                                        DefaultImage:=wooden_pixel_gui.textures.Icons2.Size8.t_CoinGold_8
                                                                stack_box_slot:
                                                                    Widget:=ResetPriceTextBlock
                                                    stack_box_slot:
                                                        HorizontalAlignment:=horizontal_alignment.Center
                                                        VerticalAlignment:=vertical_alignment.Center
                                                        Padding:=margin{Left:=0.0,Top:=15.0,Right:=0.0,Bottom:=0.0}
                                                        Widget:=NewQuestButton

                                                            
                                    
                            stack_box_slot:
                                HorizontalAlignment:=horizontal_alignment.Center
                                VerticalAlignment:=vertical_alignment.Center
                                Padding:=margin{Left:=0.0,Top:=15.0,Right:=0.0,Bottom:=0.0}
                                Widget := CloseWindowButton
  
        MyCanvas:=canvas:
            Slots:=array:
                canvas_slot:
                    Anchors:=anchors{Minimum:=vector2{X:=0.2,Y:=0.07}, Maximum:=vector2{X:=0.8,Y:=0.93}}
                    Widget:= Overlay1
                                                                                       

        return MyCanvas

    LoadTickInfo(Agent:agent):void=
        if(Player:=player[Agent]):
            if(MyTickBoxes := TickBoxes[Agent]):
                if(MyData := Data[Player]):
                    var SavedTickBoxes : []logic = MyData.QuestAreasSelected
                    for (i:=0..SavedTickBoxes.Length-1):
                        if:
                            TickBox := MyTickBoxes[i]
                            SavedInfo := SavedTickBoxes[i]
                        then:
                            if(SavedInfo = false):
                                TickBox.SetTint(MakeColorFromHex("FF000001"))
                            else:
                                TickBox.SetTint(MakeColorFromHex("FFFFFFFF"))



    AreaTickBox(Wdata:widget_message, tickbox:int):void=
        if(Player:=Wdata.Player,Agent:=agent[Player]):
            TickBoxAudio.Play(Agent)
            if(MyTickBoxes := TickBoxes[Agent]):
                if(TickBox := MyTickBoxes[tickbox]):
                    if(MyData := Data[Player]):
                        var SavedTickBoxes : []logic = MyData.QuestAreasSelected
                        if(SavedTickBoxes[tickbox] = false):
                            TickBox.SetTint(MakeColorFromHex("FFFFFFFF"))
                            if:
                                set SavedTickBoxes[tickbox] = true
                                set Data[Player] = player_data{player_data_update<constructor>(MyData), QuestAreasSelected := SavedTickBoxes}
                        else:
                            TickBox.SetTint(MakeColorFromHex("FF000001"))
                            if:
                                set SavedTickBoxes[tickbox] = false
                                set Data[Player] = player_data{player_data_update<constructor>(MyData), QuestAreasSelected := SavedTickBoxes}



    AssignRandomQuest(Wdata:widget_message):void=
        if(Player:=Wdata.Player, PlayerUI:=GetPlayerUI[Player],SelectedButton:=Wdata.Source,Agent:=agent[Player]):
            if(MyData := Data[Player]):
                Price := CalculatePrice(Agent)
                AreasArr := MyData.QuestAreasSelected
                    if(Area1 := AreasArr[0], Area2 := AreasArr[1], Area3 := AreasArr[2], Area4 := AreasArr[3]):
                        AssignedQuest : int = QuestDevice.GetRandomQuest(Agent,Area1,Area2,Area3,Area4)
                        if(AssignedQuest < 100):
                            if(MyData.Gold >= Price):
                                if:
                                    set Data[Player] = player_data{player_data_update<constructor>(MyData), Gold := MyData.Gold - Price}
                                then:
                                    QuestDevice.ActuallyAssignQuest(Agent,AssignedQuest)

                                    NewQuestAudio.Play(Agent)
                                    RemoveHUD(Wdata)


    CalculatePrice(Agent:agent):float= 
        var Price : float = 0.0
        if(MyData := Data[player[Agent]]):
            PlayerLevel := MyData.Level
            if(MyXpBar := QuestDevice.CMDevice.XpBarPerPlayer[Agent]?):
                PriceFloat : float = MyXpBar.CalculateExpReq(PlayerLevel) / 12.0 # changed from 6
                #if(RoundedPrice := Round[PriceFloat]){set Price = RoundedPrice}
                set Price = PriceFloat
            else:
                Print("MyXpBar not found when calculating price")
                
        return Price



    GetOreRewardTexture(Agent:agent):texture=

        var OreTexture :texture = hud.Copper_icon512

        if(MyData := Data[player[Agent]]):
            QuestNum := MyData.QuestNumber
            if(QuestNum <= 2):
                set OreTexture = hud.Copper_icon512
            else:
                if(QuestNum <= 5):
                    set OreTexture = hud.Silver_icon
                else:
                    if(QuestNum <= 8):
                        set OreTexture = hud.Malachite_icon512
                    else:
                        if(QuestNum <= 11):
                            set OreTexture = hud.Obsidian_icon512
                        else:
                            

        return OreTexture



    ShowUi(Agent:agent):void=

        RemoveFortHud(Agent)
        OpenAudio.Play(Agent)

        MyCanvas: canvas = MakeEntireBackground(Agent)
        
        if(PlayerUI:=GetPlayerUI[player[Agent]]):
            
            #Removing Old Widget to stop duplicates
            if(AddedCanvas := MyUI[Agent]?):
                PlayerUI.RemoveWidget(AddedCanvas)
                AddedCanvas.SetVisibility(widget_visibility.Visible)
                AddedCanvas.SetVisibility(widget_visibility.Hidden)

            PlayerUI.AddWidget(MyCanvas,player_ui_slot{InputMode:=ui_input_mode.All}) # want it to appear on top
            MyCanvas.SetVisibility(widget_visibility.Hidden)
            MyCanvas.SetVisibility(widget_visibility.Visible)

        if(set MyUI[Agent]= option{MyCanvas}){}


    RemoveHUD(Wdata:widget_message):void =

        Print("Removing")
        if(Player:=Wdata.Player, PlayerUI:=GetPlayerUI[Player],SelectedButton:=Wdata.Source,Agent:=agent[Player]):
            if(MyActualUI := MyUI[Agent]?):

                CloseAudio.Play(Agent)

                PlayerUI.RemoveWidget(MyActualUI)
                MyActualUI.SetVisibility(widget_visibility.Visible)
                MyActualUI.SetVisibility(widget_visibility.Hidden)

                ShowFortHud(Agent)
                
        

    
            

                






