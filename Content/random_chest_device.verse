
using { /Fortnite.com/Devices }
using { /Verse.org/Simulation }
using { /UnrealEngine.com/Temporary/Diagnostics }
using { /UnrealEngine.com/Temporary/SpatialMath }

random_chest_device := class(creative_device):

    @editable
    Accolades : accolades_device = accolades_device{}

    @editable
    BuyConditional : conditional_button_device = conditional_button_device{}
    @editable
    ConsumeConditional : conditional_button_device = conditional_button_device{}

    var AllWeaponConditionals : []conditional_button_device = array{}
    @editable
    HealConditional : conditional_button_device = conditional_button_device{}

    @editable
    FullInventoryMessage : hud_message_device = hud_message_device{}


    @editable
    OpenChestCine : cinematic_sequence_device = cinematic_sequence_device{}
    @editable
    OpenChestAudio : audio_player_device = audio_player_device{}
    @editable
    WeaponGranter : item_granter_device = item_granter_device{}

    @editable
    NiagProp : creative_prop_asset = DefaultCreativePropAsset
    @editable
    ChestProp : creative_prop = creative_prop{}
    @editable
    NiagZOffSet : float = 144.0 # Was 92.0



    OnBegin<override>()<suspends>:void=

        BuyConditional.ActivatedEvent.Subscribe(OpenChest)

        TaggedConditionals := FindCreativeObjectsWithTag(InventoryCheckConditionalTag{})
        for(TaggedConditional : TaggedConditionals):
            if(Cond := conditional_button_device[TaggedConditional]):
                set AllWeaponConditionals = AllWeaponConditionals + array{Cond}

    OpenChest(Agent:agent):void=

        LogicCheck := CheckInventoryFull(Agent) 
        if(LogicCheck = false):
            spawn:
                OpenChestWithDelay(Agent)

        

    OpenChestWithDelay(Agent:agent)<suspends>:void=
        ConsumeConditional.Activate(Agent)
        OpenChestCine.Play()
        OpenChestAudio.Play()
        Accolades.Award(Agent)
        Sleep(0.5)
        WeaponGranter.CycleToRandomItem(Agent)
        spawn:
            SpawnBuyNiagara()


    SpawnBuyNiagara()<suspends>:void=

        var Location : vector3 = ChestProp.GetTransform().Translation
        Rotation := ChestProp.GetTransform().Rotation

        set Location.Z = Location.Z + NiagZOffSet

        SpawnedPropResult := SpawnProp(NiagProp,Location,Rotation )
        if(SpawnedProp := SpawnedPropResult(0)?):
            spawn:
                DeletingNiagaraLonger(SpawnedProp)

    DeletingNiagaraLonger(Prop:creative_prop)<suspends>:void=
        Sleep(0.5)
        Prop.Dispose()

    CheckInventoryFull(Agent:agent):logic=

        var ItemCount : int = 0
        #for(Cond : AllWeaponConditionals):

        for(i:=0..AllWeaponConditionals.Length-1):
            if(Cond:=AllWeaponConditionals[i]):
                for(i2:=0..2):
                    CondCount := Cond.GetItemCount(Agent,i2)
                    if (CondCount > 0 and CondCount < 6):
                        Print("Item Count : {CondCount}, i2 = {i2}, Conditional Number = {i}")
                        set ItemCount += CondCount

        for(i:=0..2):
            CondCount := HealConditional.GetItemCount(Agent,i)
            if(i = 0):
                if(Stacks :int= Ceil(CondCount / 15)):
                    set ItemCount += Stacks
            if(i = 1):
                if(Stacks :int= Ceil(CondCount / 6)):
                    set ItemCount += Stacks

        Print("AllWeaponConditionals : {AllWeaponConditionals.Length}")
        Print("Item Count : {ItemCount}")
        if(ItemCount < 5):
            return false
        else:
            FullInventoryMessage.Show(Agent)
            return true

    CheckInventoryFullHeals(Agent:agent,id:int):logic=

        var ItemCount : int = 0
        for(Cond : AllWeaponConditionals):
            for(i:=0..2):
                CondCount := Cond.GetItemCount(Agent,i)
                if (CondCount > 0 and CondCount < 6):
                    set ItemCount += CondCount

        if(id = 0): # trying to buy bandages
            Bandages := HealConditional.GetItemCount(Agent,0)
            Minis := HealConditional.GetItemCount(Agent,1)

            if(Minis > 0):
                set ItemCount += 1

            if(Bandages = 0 and ItemCount >= 5):
                FullInventoryMessage.Show(Agent)
                return true
            else:
                if(Bandages < 15):
                    return false
                else:
                    FullInventoryMessage.Show(Agent)
                    return true
            
                
            
        else: # buying sheild

            Bandages := HealConditional.GetItemCount(Agent,0)
            Minis := HealConditional.GetItemCount(Agent,1)

            Print("Minis Count {Minis}")

            if(Bandages > 0):
                Print("Has Bandages")
                set ItemCount += 1

            if(Minis = 0 and ItemCount >= 5):
                Print("Inv Full no minis")
                FullInventoryMessage.Show(Agent)
                return true
            else:
                if(Minis < 3):
                    Print("Minis less than 6")
                    return false
                else:
                    Print("Full of minis")
                    FullInventoryMessage.Show(Agent)
                    return true


    CheckShieldCount(Agent:agent):void=

        Bandages := HealConditional.GetItemCount(Agent,0)
        Minis := HealConditional.GetItemCount(Agent,1)

        Print("Minis Count {Minis}")
        Print("Bandages {Bandages}")



            









    
