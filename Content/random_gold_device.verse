
using { /Fortnite.com/Devices }
using { /Verse.org/Simulation }
using { /UnrealEngine.com/Temporary/Diagnostics }
using { /Verse.org/Random }

random_gold_device := class(creative_device):

    @editable
    collectibledevices : []collectible_object_device = array{}

    @editable
    CineHudDevice : npc_hp_cinematic_hud_device = npc_hp_cinematic_hud_device{}

    @editable
    ElimGold : float = 113.0

    @editable
    CollectibleGoldMultiple : float = 10.0

    @editable
    Accolade : accolades_device = accolades_device{}

    GoldPerElimInitial : float = 0.125

    @editable
    ElimXp : float = 113.0

    @editable
    CollectibleXpMultiple : float = 10.0


    # Usually 180 - 420
    @editable
    SpawnRateMin : int = 180
    @editable
    SpawnRateMax : int = 420
    @editable
    UpTime : float = 60.0


    OnBegin<override>()<suspends>:void=

    
        for(i:=0..collectibledevices.Length-1):
            if(collectible := collectibledevices[i]):
                collectible.CollectedEvent.Subscribe(PickedUp)

        loop:
            
            maxspawncalculation:= SpawnRateMax
            var rand : int = GetRandomInt(SpawnRateMin,maxspawncalculation)
            floatrand : float = rand * 1.0
            Sleep(floatrand)
            spawn:
                ChooseRandom()

    ChooseRandom()<suspends>:void=

        var rand : int = GetRandomInt(0,collectibledevices.Length-1)

            if(collectible:=collectibledevices[rand]):
                collectible.Show()
                collectible.RespawnForAll()
            Sleep(UpTime)
        
            if(collectible:=collectibledevices[rand]):
                collectible.Hide()

            

    PickedUp(Agent:agent):void=

        if:
            Player := player[Agent]
            Player.IsActive[]
        then:

            Accolade.Award(Agent)

            var PercentAdding : float = 0.0
                if(PlayerData := Data[Agent]):
                    if(PlayerData.GoldPerElim > 0):
                        if(CosmeticGoldPerElim : float = Data[Agent].CosmeticGoldPerElim / 100.0):
                            Print("Cosmetic Gold Per Elim {CosmeticGoldPerElim}")
                            # Changing exponential to Linear
                            #set PercentAdding = (Pow(GoldPerElimInitial + (PlayerData.GoldPL * PlayerData.GoldPLLevel), (PlayerData.GoldPerElim )*1.0) - 1.0)
                            set PercentAdding = (GoldPerElimInitial + (PlayerData.GoldPL * PlayerData.GoldPLLevel)) * PlayerData.GoldPerElim
                            GemGold : float = (PlayerData.PrestigeBonusPerGem/100.0) * (PlayerData.Gems) * (PercentAdding)
                            set PercentAdding = GemGold + PercentAdding
                            set PercentAdding = PercentAdding + (PercentAdding * CosmeticGoldPerElim)


            RoundingGoldElim : float = PercentAdding + 1.0

            GoldMult : float = ElimGold  * RoundingGoldElim * CollectibleGoldMultiple

        
            if:
                OldData := Data[Player]
                set Data[Player] = player_data{player_data_update<constructor>(OldData),Gold := OldData.Gold + GoldMult, PrestigeEarnings:= OldData.PrestigeEarnings + GoldMult, TotalEarnings := OldData.TotalEarnings + GoldMult}

            then:
                CineHudDevice.GoldDrop(Agent,GoldMult)
    