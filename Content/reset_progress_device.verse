
using { /Fortnite.com/Devices }
using { /Verse.org/Simulation }
using { /UnrealEngine.com/Temporary/Diagnostics }


reset_progress_device := class(creative_device):


    @editable
    MakeSurePopup : popup_dialog_device = popup_dialog_device{}

    @editable
    StarterPackGranter : item_granter_device = item_granter_device{}

    @editable
    ResetAudio : audio_player_device = audio_player_device{}

    @editable
    Huds : npc_hp_cinematic_hud_device = npc_hp_cinematic_hud_device{}
    @editable
    AreasDevice : areas_device = areas_device{}
    @editable
    QuestDevice : quest_device = quest_device{}


    @editable
    RemoveWearables : trigger_device = trigger_device{}




    OnBegin<override>()<suspends>:void=

        MakeSurePopup.RespondingButtonEvent.Subscribe(RespondingButton)

    # need to reset 
    RespondingButton(Agent:agent,ButtonPressed:int):void=
        if(ButtonPressed = 0):

            ResetAudio.Play(Agent)
            StarterPackGranter.GrantItem(Agent)
            AreasDevice.LockAreas(Agent)
            QuestDevice.UnAssignQuest(Agent)
            RemoveWearables.Trigger(Agent)

            if(XpBar :=  Huds.CMBDevice.XpBarPerPlayer[Agent]?):
                set XpBar.CurrentLevel = 1
                set XpBar.iTotalXp = 89.0
                set XpBar.iCurrentXp = 0.0

            Huds.CMBDevice.AreasOnLevelUp(Agent)
            Huds.RefreshXpBar(Agent)

            if:
                Player := player[Agent]
                MyData := Data[player[Agent]]
                set Data[Player] = player_data{player_data_update<constructor>(MyData),

                Level := 1,
                Xp  := 0.0,
                Gold := 0.0,

                DamagePL := 0.0125,
                ArmorPL  := 0.1,
                CritPL := 0.0015,
                GoldPL := 0.0125,

                DamagePLLevel := 0,
                ArmorPLLevel := 0,
                CritPLLevel := 0,
                GoldPLLevel := 0,

                Prestige := 0,
                PrestigeBonusPerGem := 1.0,
                PrestigeEarnings := 0.0,
                TotalEarnings := 0.0,
                Gems := 0.0,

                TotalGemsAwarded := 0.0,


                PistolLevel := 1,
                RifleLevel := 1,
                SmgLevel :=  1,
                ShotgunLevel := 1,
                SniperLevel := 1,

                TrueDamage := 0,
                CritLevel := 0,

                GoldPerElim := 0,
                ExpBoost := 0,
                ExpPerElim := 0, # Old Peristence so it doesnt error

                PlayerArmor := 0,
                PlayerArmorCosmetic := 0,
                ShieldLevel := 0,
                MovementSpeed := 0,

                FireChance := 0,
                FireDamage := 1,
                CrouchHeal := 0,

                ChainLightningDamage := 1,
                ChainLightningChance := 0,
                ChainLightningDamagePerUpgrade := 0.45,
                ChainDamageGemLevel := 0,

                FocusDamage := 0,
                AccuracyDamage := 0,

                MarkDamage := 0,
                SuperCritDamage := 0,
                SuperCritChance := 0,
                SuperCritDamagePerUpgrade := 2.0,
                SuperCritGemLevel := 0,

                Dash := 0,

                ScrapLevel:= 1,
                OreDropLevel := 0,
                GemsOreDropLevel := 0,

                AreasUnlocked := array{false,false,false,false,false,false,false,false,false,false,false,false,false,false},

                AchievementsDamageBonus := 0.0,

                TotalBoarEliminations := 0,
                TotalBuffaloEliminations :=0,
                TotalLizardEliminations :=0,
                TotalGiantEliminations := 0,
                TotalScorpionEliminations := 0,

                TotalGolemEliminations := 0,
                TotalLeopardEliminations := 0,
                TotalRamEliminations := 0,
                TotalYetiEliminations := 0,
                TotalHoundEliminations := 0,
                TotalLavaJawEliminations := 0,
                TotalTerravoreEliminations := 0,

                TotalIronTuskEliminations := 0,
                TotalEmberClawEliminations := 0,
                TotalIceheartEliminations := 0,
                TotalBlazebornEliminations := 0,

                TotalQuestCompleted := 0,
                TotalAchievementsCompleted := 0,

                TotalPistolEliminations := 0,
                TotalSmgEliminations := 0,
                TotalAssaultEliminations := 0,
                TotalShotgunEliminations := 0,
                TotalSniperEliminations := 0,

                CosmeticsUnlocked := array{false,false,false,false,false,false,false,false,false,false,false,false,false,false,false,false,false,false,false,false,false,false,false,false},
                CosmeticsLevels := array{1,1,1,1,1,1,1,1,1,1,1,1,1,1,1,1,1,1,1,1,1,1,1,1},
                BackCosmeticEquiped := 13, 
                HatCosmeticEquiped := 13,

                CosmeticDamage := 0.0,
                CosmeticCritChance := 0.0,
                CosmeticCritDamage := 0.0,
                CosmeticBurnDamage := 0.0,
                CosmeticArmor := 0.0,
                CosmeticLightningChance := 0.0,
                CosmeticLightningDamage := 0.0,
                CosmeticGoldPerElim := 0.0,

                QuestAchievementTier := 0,


                CurrentMissionSaved := 0,
                MissionCurrent := 0,

                QuestValue := 0,
                QuestNumber := 100,
                QuestTarget := 0,
                QuestRewardMultiple := 0,
                QuestAreasSelected := array{true,true,true,true},
                
                KeepInventory := 0,

                QuestRewardLevel := 0,
                TowerLevel := 0,
                TutorialsCurrentCount := 0,
                MissionsCompletedPersis := 0,
                MissionsCurrentCount := 0,
                RandomMapLocations := array{0,0,0,0}
                }

