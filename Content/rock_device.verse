using { /Fortnite.com/Devices }
using { /Verse.org/Simulation }
using { /UnrealEngine.com/Temporary/Diagnostics }
using { /UnrealEngine.com/Temporary/SpatialMath }
using { /Verse.org/Random }
using { /Fortnite.com/Characters }
#using { /blindgames@fortnite.com/FarmProjectZ/NumberFormatter }
using { /Verse.org/Assets }

rock__device := class(creative_device):

    @editable
    MissionsSwitch : switch_device = switch_device{}

    @editable
    NiagaraSpawnLoc : creative_prop = creative_prop{}

    @editable 
    manip : prop_manipulator_device = prop_manipulator_device{}
    @editable 
    rockprop : creative_prop = creative_prop{}
    @editable 
    accolade : accolades_device = accolades_device{}
    @editable 
    copperaudio : audio_player_device = audio_player_device{}
    @editable
    rockhitaudio : audio_player_device = audio_player_device{}
    @editable
    coppergranter : item_granter_device = item_granter_device{}
    @editable 
    quartzgranter : item_granter_device = item_granter_device{}
    @editable 
    quartzaudio : audio_player_device = audio_player_device{}

    @editable 
    negativeaudio : audio_player_device = audio_player_device{}


    @editable
    NiagaraRockEffect : creative_prop_asset = DefaultCreativePropAsset
    @editable
    NiagaraBreakEffect : creative_prop_asset = DefaultCreativePropAsset
    @editable
    NiagaraCrystalEffect : creative_prop_asset = DefaultCreativePropAsset

    @editable
    DmgNumberSpawnProp : creative_prop_asset = DefaultCreativePropAsset

    @editable
    SpawnTimerProp : creative_prop_asset = DefaultCreativePropAsset

    @editable
    Spawntime : float = 60.0

    @editable 
    rocktype : int = 0

    @editable
    var damagemax : int = 15
    @editable
    var damagemin : int = 6

    @editable
    RockType : int = 0
    @editable
    SMType : int = 0

    var Depleted : logic = false

    var Rarity : int = 0

    RarityMultiple1 : int = 2
    RarityMultiple2 : int = 5
    RarityMultiple3 : int = 8
    RarityMultiple4 : int = 12

    var RarityNiagaraProp : creative_prop = creative_prop{}
    var RarityNiagaraProp1 : creative_prop = creative_prop{}

    @editable
    RarityShineAsset : creative_prop_asset = DefaultCreativePropAsset
    var MyRarityProp : creative_prop = creative_prop{}
    @editable
    ShineOffset : float = 80.0

    @editable
    RarityNiagaraAssets : []creative_prop_asset = array{}
    
    var Rarity4odds : int = 0
    var Rarity3odds : int = 0
    var Rarity2odds : int = 0
    var Rarity1odds : int = 5

    @editable
    TotalHealth : int = 900

    var CurrentHealth : float = 0.0

    @editable
    GrantedIntervals : int = 5

    var CurrentGrantedInterval : int = 0
    var HealthPerGrantInterval : float = 0.0


    #CircularProgressMat : Mats.CircularProgressFL5.CircularProgress_M_material = Mats.CircularProgressFL5.CircularProgress_M_material{}

        



        



    

