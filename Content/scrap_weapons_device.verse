
using { /Fortnite.com/Devices }
using { /Verse.org/Simulation }
using { /UnrealEngine.com/Temporary/Diagnostics }
using { /UnrealEngine.com/Temporary/UI }
using { /UnrealEngine.com/Temporary/SpatialMath }
using { /Fortnite.com/UI }
using { /Verse.org/Colors }
using { /Fortnite.com/Playspaces }
using { /Fortnite.com/Characters }

scrap_weapons_device := class(creative_device):

    @editable
    ScrapAudio : audio_player_device = audio_player_device{}
    @editable
    OpenAudio : audio_player_device = audio_player_device{}
    @editable
    CloseAudio : audio_player_device = audio_player_device{}


    @editable
    WeaponRarityDev : weapon_rarity_device = weapon_rarity_device{}

    @editable
    TalkAudio : audio_player_device = audio_player_device{}

    @editable
    ItemGranter : item_granter_device = item_granter_device{}
    @editable
    ItemGranters : []item_granter_device = array{}
    @editable
    ScrapButton : button_device = button_device{}

    @editable
    NotHoldingWeaponHMD : hud_message_device = hud_message_device{}

    var MyUI : [agent]?canvas = map{}
    var MyRarity : [agent]int = map{}

    storedstring<localizes>(S:string) : message = "{S}"
    PlusOreString<localizes>(I:int) : message = "(+{I} ore)"
    LevelString<localizes>(I:int) : message = "Scrap Lv {I}"

    OnBegin<override>()<suspends>:void=

        ScrapButton.InteractedWithEvent.Subscribe(DialogButton)

    DialogButton(Agent:agent):void=

        var IsHolding : logic = false

        for(Cond : WeaponRarityDev.CommonConditionalDevices):
            if (Cond.IsHoldingItem[Agent]):
                set IsHolding = true
                ShowUi(Agent,Cond,0)
        for(Cond :  WeaponRarityDev.UncommonConditionalDevices):
            if (Cond.IsHoldingItem[Agent]):
                set IsHolding = true
                ShowUi(Agent,Cond,1)
        for(Cond :  WeaponRarityDev.RareConditionalDevices):
            if (Cond.IsHoldingItem[Agent]):
                set IsHolding = true
                ShowUi(Agent,Cond,2)
        for(Cond :  WeaponRarityDev.EpicConditionalDevices):
            if (Cond.IsHoldingItem[Agent]):
                set IsHolding = true
                ShowUi(Agent,Cond,3)
        for(Cond :  WeaponRarityDev.LegendaryConditionalDevices):
            if (Cond.IsHoldingItem[Agent]):
                set IsHolding = true
                ShowUi(Agent,Cond,4)
        for(Cond :  WeaponRarityDev.ExoticConditionalDevices):
            if (Cond.IsHoldingItem[Agent]):
                set IsHolding = true
                ShowUi(Agent,Cond,5)

        for(Cond : WeaponRarityDev.StartingConditionalDevices):
            if (Cond.IsHoldingItem[Agent]):
                set IsHolding = true
                ShowUi(Agent,Cond,6)

        if(IsHolding = false):
            NotHoldingWeaponHMD.Show(Agent)

    RequiredItemsCheck(Cond:conditional_button_device,Agent:agent):void=
        var ItemScrapped : logic = false
        for(i:=0..2):
            if(ItemScrapped = false):
                if(Cond.IsHoldingItem[Agent,i]):
                    Print("Is Holding Item {i}")
                    Cond.SetItemCountRequired(0,0)
                    Cond.SetItemCountRequired(1,0)
                    Cond.SetItemCountRequired(2,0)
                    Cond.SetItemCountRequired(i,1)
                    Cond.Activate(Agent)
                    set ItemScrapped = true
        Cond.Reset()

    MakeUi(Agent:agent,Cond:conditional_button_device,rarity:int):canvas=

        if(set MyRarity[Agent] = rarity){}

        CopperTexture := hud.Copper_icon512
        SilverTexture := hud.Silver_icon
        MalachiteTexture := hud.Malachite_icon512
        ObisdianTexture := hud.Obsidian_icon512
        BrightcoreTexture := hud.Brightcore_icon512
        SpectroliteTexture := hud.Spectrolite_icon512

        # make it load data here

        var ScrapLevelText: text_block = text_block{DefaultTextColor := MakeColorFromHex("FFFFFFFF"), DefaultText:=LevelString(0)}
        var PlusOreText: text_block = text_block{DefaultTextColor := MakeColorFromHex("00FF00FF"), DefaultText:=PlusOreString(0)}

        var WeaponIconTexture : texture_block = texture_block{DefaultImage:= wooden_pixel_gui.textures.Icons2.Size16.t_Sword_16, DefaultDesiredSize:=vector2{X:=100.0,Y:=100.0}}
        var OreTexture: texture_block = texture_block{DefaultImage:= hud.Copper_icon512, DefaultDesiredSize:=vector2{X:=100.0,Y:=100.0}}

        WeaponShadow : texture_block = texture_block{DefaultImage:= wooden_pixel_gui.textures.Icons2.Size16.t_Sword_16, DefaultDesiredSize:=vector2{X:=100.0,Y:=100.0}, DefaultTint:=MakeColorFromHex("2D1D0DFF")}
        var OreTextureShadow : texture_block = texture_block{DefaultImage:= hud.Copper_icon512, DefaultDesiredSize:=vector2{X:=100.0,Y:=100.0},DefaultTint:=MakeColorFromHex("2D1D0DFF")}

        if:
            Player:= player[Agent]
            OldData := Data[Player]
        then:
            ScrapLv:=OldData.ScrapLevel
            set ScrapLevelText = text_block{DefaultTextColor := MakeColorFromHex("FFFFFFFF"), DefaultText:=LevelString(ScrapLv)}
            set PlusOreText = text_block{DefaultTextColor := MakeColorFromHex("00FF00FF"), DefaultText:=PlusOreString(ScrapLv-1)}

        if(rarity=1):
            set WeaponIconTexture = texture_block{DefaultImage:= wooden_pixel_gui.textures.Icons2.Size16.t_Sword_16, DefaultDesiredSize:=vector2{X:=100.0,Y:=100.0},DefaultTint:=MakeColorFromHex("82bc3b")}
            OreTexture.SetImage(hud.Silver_icon)
            OreTextureShadow.SetImage(hud.Silver_icon)
        if(rarity=2):
            set WeaponIconTexture = texture_block{DefaultImage:= wooden_pixel_gui.textures.Icons2.Size16.t_Sword_16, DefaultDesiredSize:=vector2{X:=100.0,Y:=100.0},DefaultTint:=MakeColorFromHex("15a1cd")}
            OreTexture.SetImage(hud.Malachite_icon512)
            OreTextureShadow.SetImage(hud.Malachite_icon512)
        if(rarity=3):
            set WeaponIconTexture = texture_block{DefaultImage:= wooden_pixel_gui.textures.Icons2.Size16.t_Sword_16, DefaultDesiredSize:=vector2{X:=100.0,Y:=100.0},DefaultTint:=MakeColorFromHex("be54db")}
            OreTexture.SetImage(hud.Obsidian_icon512)
            OreTextureShadow.SetImage(hud.Obsidian_icon512)
        if(rarity=4):
            OreTexture.SetImage(hud.Brightcore_icon512)
            OreTextureShadow.SetImage(hud.Brightcore_icon512)
            set WeaponIconTexture = texture_block{DefaultImage:= wooden_pixel_gui.textures.Icons2.Size16.t_Sword_16, DefaultDesiredSize:=vector2{X:=100.0,Y:=100.0},DefaultTint:=NamedColors.Orange}
        if(rarity=5):
            OreTexture.SetImage(hud.Spectrolite_icon512)
            OreTextureShadow.SetImage(hud.Spectrolite_icon512)
            set WeaponIconTexture = texture_block{DefaultImage:= wooden_pixel_gui.textures.Icons2.Size16.t_Sword_16, DefaultDesiredSize:=vector2{X:=100.0,Y:=100.0},DefaultTint:=NamedColors.Red}
        if(rarity=6):
            set WeaponIconTexture = texture_block{DefaultImage:= wooden_pixel_gui.textures.Icons2.Size16.t_Sword_16, DefaultDesiredSize:=vector2{X:=100.0,Y:=100.0}}
            OreTexture.SetImage(hud.Copper_icon512)
            OreTextureShadow.SetImage(hud.Copper_icon512)


        ScrapQuestionText: text_block = text_block{DefaultTextColor := MakeColorFromHex("FFFFFFFF"), DefaultText:=storedstring("Scrap your weapon?")}
        ArrowTexture: texture_block = texture_block{DefaultImage:= wooden_pixel_gui.textures.Icons2.Size32.t_ArrowRight_32, DefaultDesiredSize:=vector2{X:=100.0,Y:=100.0}}

        NoButton : button_quiet = button_quiet{DefaultText:=storedstring("Cancel")}
        NoButton.OnClick().Subscribe(RemoveHUD) 
        YesButton : button_quiet = button_quiet{DefaultText:=storedstring("Scrap")}
        YesButton.OnClick().SubscribeMessage(ScrapWeapon,Cond) 

        MyCanvas:=canvas:
            Slots:=array:
                canvas_slot:
                    Anchors:=anchors{Minimum:=vector2{X:=0.25,Y:=0.15}, Maximum:=vector2{X:=0.75,Y:=0.6}}
                    Widget:=overlay:
                        Slots:=array:
                            overlay_slot:
                                HorizontalAlignment:=horizontal_alignment.Fill
                                VerticalAlignment:=vertical_alignment.Fill
                                Widget:=texture_block:
                                    DefaultImage:=wooden_pixel_gui.textures.Panels.t_Window
                                    DefaultTint:=MakeColorFromHex("7E5835")
                            overlay_slot:
                                HorizontalAlignment:=horizontal_alignment.Fill
                                VerticalAlignment:=vertical_alignment.Fill
                                Padding:=margin{Left:=10.0,Top:=10.0,Right:=10.0,Bottom:=10.0}
                                Widget:=texture_block:
                                    DefaultImage:=wooden_pixel_gui.textures.Panels.t_Window
                                    DefaultTint:=MakeColorFromHex("634526")
                            overlay_slot:
                                HorizontalAlignment:=horizontal_alignment.Center
                                VerticalAlignment:=vertical_alignment.Center
                                Widget:=stack_box:
                                    Orientation:=orientation.Vertical
                                    Slots:=array:
                                        stack_box_slot:
                                            Padding:=margin{Bottom:=10.0}
                                            HorizontalAlignment:=horizontal_alignment.Center
                                            VerticalAlignment:=vertical_alignment.Fill
                                            Widget:=ScrapQuestionText
                                        stack_box_slot:
                                            HorizontalAlignment:=horizontal_alignment.Fill
                                            VerticalAlignment:=vertical_alignment.Fill
                                            Widget:=overlay:
                                                Slots:=array:
                                                    overlay_slot:
                                                        HorizontalAlignment:=horizontal_alignment.Fill
                                                        VerticalAlignment:=vertical_alignment.Fill
                                                        Widget:=texture_block:
                                                            DefaultImage:=wooden_pixel_gui.textures.Panels.WhiteSquare100
                                                            DefaultTint:=MakeColorFromHex("4A331BFF")
                                                            DefaultDesiredSize:=vector2{X:=450.0,Y:=32.0}
                                                    #overlay_slot:
                                                    #    HorizontalAlignment:=horizontal_alignment.Left
                                                    #    VerticalAlignment:=vertical_alignment.Center
                                                    #    Padding:=margin{Left:=40.0,Top:=0.0,Right:=0.0,Bottom:=0.0}
                                                    #    Widget:=ScrapLevelText
                                                    #overlay_slot:
                                                    #    HorizontalAlignment:=horizontal_alignment.Right
                                                    #    VerticalAlignment:=vertical_alignment.Center
                                                    #    Padding:=margin{Left:=0.0,Top:=0.0,Right:=40.0,Bottom:=0.0}
                                                    #    Widget:=PlusOreText
                                        stack_box_slot:
                                            HorizontalAlignment:=horizontal_alignment.Center
                                            VerticalAlignment:=vertical_alignment.Fill
                                            Padding:=margin{Left:=0.0,Top:=15.0,Right:=00.0,Bottom:=15.0}
                                            Widget:=overlay:
                                                Slots:=array:
                                                    overlay_slot:
                                                        HorizontalAlignment:=horizontal_alignment.Fill
                                                        VerticalAlignment:=vertical_alignment.Fill
                                                        Widget:=texture_block:
                                                            DefaultImage:=wooden_pixel_gui.textures.Panels.WhiteSquare100
                                                            DefaultTint:=MakeColorFromHex("3A2713FF")
                                                            DefaultDesiredSize:=vector2{X:=450.0,Y:=32.0}
                                                    overlay_slot:
                                                        HorizontalAlignment:=horizontal_alignment.Fill
                                                        VerticalAlignment:=vertical_alignment.Fill
                                                        Padding:=margin{Left:=10.0,Top:=10.0,Right:=10.0,Bottom:=10.0}
                                                        Widget:=texture_block:
                                                            DefaultImage:=wooden_pixel_gui.textures.Panels.WhiteSquare100
                                                            DefaultTint:=MakeColorFromHex("4A331BFF")
                                                            DefaultDesiredSize:=vector2{X:=450.0,Y:=32.0}
                                                    overlay_slot:
                                                        Widget:=stack_box:
                                                            Orientation:=orientation.Horizontal
                                                            Slots:=array:
                                                                stack_box_slot:
                                                                    Widget:=overlay:
                                                                        Slots:=array:
                                                                            overlay_slot:
                                                                                Padding:=margin{Left:=0.0,Top:=6.0,Right:=0.0,Bottom:=0.0}
                                                                                Widget:=WeaponShadow
                                                                            overlay_slot:
                                                                                Widget:=WeaponIconTexture
                                                                stack_box_slot:
                                                                    Widget:=ArrowTexture
                                                                stack_box_slot:
                                                                    Widget:=overlay:
                                                                        Slots:=array:
                                                                            overlay_slot:
                                                                                Padding:=margin{Left:=0.0,Top:=6.0,Right:=0.0,Bottom:=0.0}
                                                                                Widget:=OreTextureShadow
                                                                            overlay_slot:
                                                                                Widget:=OreTexture
                                        stack_box_slot:
                                            HorizontalAlignment:=horizontal_alignment.Fill
                                            VerticalAlignment:=vertical_alignment.Fill
                                            Widget:=overlay:
                                                Slots:=array:
                                                    overlay_slot:
                                                        HorizontalAlignment:=horizontal_alignment.Left
                                                        VerticalAlignment:=vertical_alignment.Top
                                                        Padding:=margin{Left:=80.0,Top:=10.0,Right:=0.0,Bottom:=0.0}
                                                        Widget:=YesButton
                                                    overlay_slot:
                                                        HorizontalAlignment:=horizontal_alignment.Right
                                                        VerticalAlignment:=vertical_alignment.Top
                                                        Padding:=margin{Left:=00.0,Top:=10.0,Right:=80.0,Bottom:=0.0}
                                                        Widget:=NoButton

                                    

        return MyCanvas


    ShowUi(Agent:agent,Cond:conditional_button_device,rarity:int):void=

        MyCanvas: canvas = MakeUi(Agent,Cond,rarity)

        if:
            MyActualUI := MyUI[Agent]
            MyActualUI = false
        then:

            if(set MyUI[Agent]= option{MyCanvas}){}

            OpenAudio.Play(Agent)
            TalkAudio.Play(Agent)
            
            if(PlayerUI:=GetPlayerUI[player[Agent]]):
                PlayerUI.AddWidget(MyCanvas,player_ui_slot{InputMode:=ui_input_mode.All}) # want it to appear on top
                MyCanvas.SetVisibility(widget_visibility.Hidden)
                MyCanvas.SetVisibility(widget_visibility.Visible)
        else:
            if(MyUI[Agent]):
            else:
                if(set MyUI[Agent]= option{MyCanvas}){}

                OpenAudio.Play(Agent)
                TalkAudio.Play(Agent)
                
                if(PlayerUI:=GetPlayerUI[player[Agent]]):
                    PlayerUI.AddWidget(MyCanvas,player_ui_slot{InputMode:=ui_input_mode.All}) # want it to appear on top
                    MyCanvas.SetVisibility(widget_visibility.Hidden)
                    MyCanvas.SetVisibility(widget_visibility.Visible)
        

    RemoveHUD(Wdata:widget_message):void =

        if(Player:=Wdata.Player, PlayerUI:=GetPlayerUI[Player],SelectedButton:=Wdata.Source,Agent:=agent[Player]):
            if(MyActualUI := MyUI[Agent]?):

                CloseAudio.Play(Agent)

                PlayerUI.RemoveWidget(MyActualUI)
                MyActualUI.SetVisibility(widget_visibility.Visible)
                MyActualUI.SetVisibility(widget_visibility.Hidden)

                if(set MyUI[Agent] = false){}
                ShowFortHud(Agent)


    ScrapWeapon(Wdata:widget_message,CondIn:conditional_button_device):void=

        Print("Scrap Pressed")

        RemoveHUD(Wdata)

        if(Player:=Wdata.Player, PlayerUI:=GetPlayerUI[Player],SelectedButton:=Wdata.Source,Agent:=agent[Player]):
            RequiredItemsCheck(CondIn,Agent)
            ScrapAudio.Play(Agent)
            if:
                MyGranter : item_granter_device = ItemGranters[MyRarity[Agent]]
            then:
                if:
                    PlayerData := Data[Player]
                    RarityDif := MyRarity[Agent]
                then:
                    var ScrapAmount : int = 0
                    if(RarityDif = 0):
                        set ScrapAmount = 1
                        for(i:=0..ScrapAmount-1):
                            MyGranter.GrantItem(Agent)
                    if(RarityDif = 1):
                        set ScrapAmount = 2
                        for(i:=0..ScrapAmount-1):
                            MyGranter.GrantItem(Agent)
                    if(RarityDif = 2):
                        set ScrapAmount = 2
                        for(i:=0..ScrapAmount-1):
                            MyGranter.GrantItem(Agent)
                    if(RarityDif = 3):
                        set ScrapAmount = 3
                        for(i:=0..ScrapAmount-1):
                            MyGranter.GrantItem(Agent)
                    if(RarityDif = 4):
                        set ScrapAmount = 3
                        for(i:=0..ScrapAmount-1):
                            MyGranter.GrantItem(Agent)
                    if(RarityDif = 5):
                        set ScrapAmount = 3
                        for(i:=0..ScrapAmount-1):
                            MyGranter.GrantItem(Agent)
                    if(RarityDif = 6):
                        set ScrapAmount = 0

            ShowFortHud(Agent)



                




