
using { /Fortnite.com/Devices }
using { /Verse.org/Simulation }
using { /UnrealEngine.com/Temporary/UI }
using { /UnrealEngine.com/Temporary/SpatialMath }
using { /Fortnite.com/UI }
using { /Verse.org/Colors }
using { /Fortnite.com/Playspaces }
using { /Fortnite.com/Characters }
using { /Verse.org/Assets }

settings_ui := class(creative_device):

    @editable
    OpenButton : button_device = button_device{}

    @editable
    OpenAudio : audio_player_device = audio_player_device{}
    @editable
    CloseAudio : audio_player_device = audio_player_device{}

    # Just to pull % multtipliers
    @editable
    NPCManager : npc_manager_device = npc_manager_device{}

    @editable
    QuestDevice : quest_device = quest_device{}

    var MyUI : [agent]?canvas = map{}

    storedstring<localizes>(S:string) : message = "{S}"

    OnBegin<override>()<suspends>:void=

        OpenButton.InteractedWithEvent.Subscribe(ShowUi)

    MakeEntireBackground (Agent:agent):canvas=

        
        TitleBarTexture := texture_block :
            DefaultDesiredSize:=vector2{X:=16.0,Y:=90.0}
            DefaultImage:=wooden_pixel_gui.textures.Panels.t_TitleBar


        CloseWindowButton := button_quiet :
            DefaultText:=storedstring("      Close     ")
        CloseWindowButton.OnClick().Subscribe(RemoveHUD)

        LightningChanceText := text_block:
            DefaultText:=storedstring("Lightning Chance")
            DefaultTextColor := MakeColorFromHex("FFFFFF")

        LightningChanceLevelText := text_block:
            DefaultText:=storedstring("{0}")
            DefaultTextColor := MakeColorFromHex("FFFFFF")

        LightningDamageText := text_block:
            DefaultText:=storedstring("Lightning Damage")
            DefaultTextColor := MakeColorFromHex("FFFFFF")

        LightningDamageLevelText := text_block:
            DefaultText:=storedstring("{0}")
            DefaultTextColor := MakeColorFromHex("FFFFFF")

        ElimText := text_block:
            DefaultText:=storedstring("Elim Mark")
            DefaultTextColor := MakeColorFromHex("FFFFFF")

        ElimPercentText := text_block:
            DefaultText:=storedstring("{0}")
            DefaultTextColor := MakeColorFromHex("FFFFFF")

        SuperCritText := text_block:
            DefaultText:=storedstring("Super-Crit Chance")
            DefaultTextColor := MakeColorFromHex("FFFFFF")

        SuperCritPercentText := text_block:
            DefaultText:=storedstring("{0}")
            DefaultTextColor := MakeColorFromHex("FFFFFF")

        SuperCritDamageText := text_block:
            DefaultText:=storedstring("Super-Crit Damage")
            DefaultTextColor := MakeColorFromHex("FFFFFF")

        SuperCritDamagePercentText := text_block:
            DefaultText:=storedstring("{0}")
            DefaultTextColor := MakeColorFromHex("FFFFFF")

        OreDropRateText := text_block:
            DefaultText:=storedstring("Ore Drop Rate")
            DefaultTextColor := MakeColorFromHex("FFFFFF")

        OreDropRatePercentText := text_block:
            DefaultText:=storedstring("{0}")
            DefaultTextColor := MakeColorFromHex("FFFFFF")

        
        BurnChanceText := text_block:
            DefaultText:=storedstring("Burn Chance")
            DefaultTextColor := MakeColorFromHex("FFFFFF")

        BurnChancePercentText := text_block:
            DefaultText:=storedstring("{0}")
            DefaultTextColor := MakeColorFromHex("FFFFFF")

        BurnDamageText := text_block:
            DefaultText:=storedstring("Burn Damage")
            DefaultTextColor := MakeColorFromHex("FFFFFF")

        BurnDamagePercentText := text_block:
            DefaultText:=storedstring("{0}")
            DefaultTextColor := MakeColorFromHex("FFFFFF")

        CrouchHealText := text_block:
            DefaultText:=storedstring("Crouch Heal")
            DefaultTextColor := MakeColorFromHex("FFFFFF")

        CrouchHealLevelText := text_block:
            DefaultText:=storedstring("{0}")
            DefaultTextColor := MakeColorFromHex("FFFFFF")

        FocusDamageText := text_block:
            DefaultText:=storedstring("Focus Damage")
            DefaultTextColor := MakeColorFromHex("FFFFFF")

        FocusDamagePercentText := text_block:
            DefaultText:=storedstring("{0}")
            DefaultTextColor := MakeColorFromHex("FFFFFF")

        AccuracyDamageText := text_block:
            DefaultText:=storedstring("Accuracy Damage")
            DefaultTextColor := MakeColorFromHex("FFFFFF")

        AccuracyDamagePercentText := text_block:
            DefaultText:=storedstring("{0}")
            DefaultTextColor := MakeColorFromHex("FFFFFF")

        TrueDamageText := text_block:
            DefaultText:=storedstring("Damage")
            DefaultTextColor := MakeColorFromHex("FFFFFF")

        TrueDamagePercentText := text_block:
            DefaultText:=storedstring("{0}")
            DefaultTextColor := MakeColorFromHex("FFFFFF")
        
        CritChanceText := text_block:
            DefaultText:=storedstring("Crit Chance")
            DefaultTextColor := MakeColorFromHex("FFFFFF")

        CritChancePercentText := text_block:
            DefaultText:=storedstring("{0}")
            DefaultTextColor := MakeColorFromHex("FFFFFF")
        
        CritDamageText := text_block:
            DefaultText:=storedstring("Crit Damage")
            DefaultTextColor := MakeColorFromHex("FFFFFF")

        CritDamagePercentText := text_block:
            DefaultText:=storedstring("{0}")
            DefaultTextColor := MakeColorFromHex("FFFFFF")

        GoldText := text_block:
            DefaultText:=storedstring("Gold")
            DefaultTextColor := MakeColorFromHex("FFFFFF")

        GoldLevelText := text_block:
            DefaultText:=storedstring("{0}")
            DefaultTextColor := MakeColorFromHex("FFFFFF")

        ArmorText := text_block:
            DefaultText:=storedstring("Armor")
            DefaultTextColor := MakeColorFromHex("FFFFFF")

        ArmorLevelText := text_block:
            DefaultText:=storedstring("{0}")
            DefaultTextColor := MakeColorFromHex("FFFFFF")

        AgilityText := text_block:
            DefaultText:=storedstring("Agility")
            DefaultTextColor := MakeColorFromHex("FFFFFF")

        AgilityLevelText := text_block:
            DefaultText:=storedstring("{0}")
            DefaultTextColor := MakeColorFromHex("FFFFFF")

        TotalElimText := text_block:
            DefaultText:=storedstring("Total Eliminations")
            DefaultTextColor := MakeColorFromHex("FFFFFF")

        TotalElimNumberText := text_block:
            DefaultText:=storedstring("{0}")
            DefaultTextColor := MakeColorFromHex("FFFFFF")

        PrestigeEarningsText := text_block:
            DefaultText:=storedstring("Prestige Earnings")
            DefaultTextColor := MakeColorFromHex("FFFFFF")

        PrestigeEarningsNumberText := text_block:
            DefaultText:=storedstring("{0}")
            DefaultTextColor := MakeColorFromHex("FFFFFF")

        TotalEarningsText := text_block:
            DefaultText:=storedstring("Total Earnings")
            DefaultTextColor := MakeColorFromHex("FFFFFF")

        TotalEarningsNumberText := text_block:
            DefaultText:=storedstring("{0}")
            DefaultTextColor := MakeColorFromHex("FFFFFF")

        TowerText := text_block:
            DefaultText:=storedstring("Tower Level")
            DefaultTextColor := MakeColorFromHex("FFFFFF")

        TowerLevelText := text_block:
            DefaultText:=storedstring("{0}")
            DefaultTextColor := MakeColorFromHex("FFFFFF")


        if:
            MyData := Data[player[Agent]]
            CosmeticStat := NPCManager.CosmeticUnlockDevice.AgentCosmeticStats[Agent]
            CosmeticLightningChance : float = MyData.CosmeticLightningChance / 100.0
            CosmeticLightningDamage : float = MyData.CosmeticLightningDamage / 100.0
            CosmeticFireChance : float = MyData.CosmeticBurnDamage / 100.0
            CosmeticCritDamage : float = MyData.CosmeticCritDamage / 100.0
            CosmeticCritChance : float = MyData.CosmeticCritChance / 100.0
            CosmeticGoldPerElim : float = MyData.CosmeticGoldPerElim / 100.0
            CosmeticDamage : float = MyData.CosmeticDamage / 100.0

            FireLevel := Data[player[Agent]].FireDamage
            
        then:

            LightningOdds: float = (MyData.ChainLightningChance * NPCManager.ChainLightningOdds) + (MyData.ChainLightningChance * NPCManager.ChainLightningOdds * CosmeticLightningChance)
            LightningChanceLevelText.SetText(storedstring("{formatter.Abbreviate(LightningOdds * 100.0 , false)}%"))

            LightningDamage : float = ( MyData.ChainLightningDamage * MyData.ChainLightningDamagePerUpgrade ) + ( MyData.ChainLightningDamage * MyData.ChainLightningDamagePerUpgrade * CosmeticLightningDamage) 
            LightningDamageLevelText.SetText(storedstring("{formatter.Abbreviate(LightningDamage * 100.0 ,false)}%"))

            ElimPercentText.SetText(storedstring("{formatter.Abbreviate(NPCManager.MarkDamageMultiple * MyData.MarkDamage * 100.0,false)}%"))

            SuperCritPercentText.SetText(storedstring("{ formatter.Abbreviate(MyData.SuperCritChance * NPCManager.SuperCritOddsPerUpgrade * 100.0, false )}%"))
            SuperCritDamagePercentText.SetText(storedstring("{formatter.Abbreviate(MyData.SuperCritDamage * MyData.SuperCritDamagePerUpgrade * 100.0, false)}%"))

            OreDropRatePercentText.SetText(storedstring("Lv. {formatter.Abbreviate((MyData.OreDropLevel + MyData.GemsOreDropLevel) * 1.0, false)}"))

            FireOdds : float = (NPCManager.FireOddsPerUpgrade * MyData.FireChance) + (NPCManager.FireOddsPerUpgrade * MyData.FireChance) * CosmeticFireChance
            BurnChancePercentText.SetText(storedstring("{formatter.Abbreviate(FireOdds * 100.0 , true)}%"))

            BurnDamage : float = (( FireLevel * NPCManager.FireDamagePerUpgrade ) * 1.0)  + (FireLevel * NPCManager.FireDamagePerUpgrade * CosmeticFireChance * 1.0 * MyData.Level) + (FireLevel * NPCManager.FireDamagePerUpgrade * MyData.Level * 1.0)
            BurnDamagePercentText.SetText(storedstring("{formatter.Abbreviate(BurnDamage, true)}"))

            CrouchHealLevelText.SetText(storedstring("{MyData.CrouchHeal}/Sec"))
            FocusDamagePercentText.SetText(storedstring("{formatter.Abbreviate(NPCManager.FocusDamagePerLevel * MyData.FocusDamage * 100.0, false)}%"))
            AccuracyDamagePercentText.SetText(storedstring("{formatter.Abbreviate(NPCManager.AccuracyDamagePerLevel * MyData.AccuracyDamage * 100.0, false)}%"))

            var TDamage : float = ( NPCManager.TrueDamagePercentPerLevel + (MyData.DamagePL * MyData.DamagePLLevel) ) * MyData.TrueDamage 
            #var TDamage : float = (( Pow(NPCManager.TrueDamagePercentPerLevel + (MyData.DamagePL * MyData.DamagePLLevel),(MyData.TrueDamage )*1.0)) )
            GemDamage := ((MyData.PrestigeBonusPerGem*1.0)/100.0) * (MyData.Gems) * TDamage

            set TDamage = TDamage + GemDamage
            AchievmentDamage := (MyData.AchievementsDamageBonus/100.0) * TDamage
            TypeDamage := NPCManager.WeaponTypeMultiple(Agent)
            if(WeaponRarityPercentDmg := NPCManager.WeaponRarityDevice.AgentMap[Agent]):
                TrueDamagePercentText.SetText(storedstring("{formatter.Abbreviate((TDamage   + (TDamage * CosmeticDamage) + (TDamage * TypeDamage) + (TDamage * NPCManager.DamagePerPlayerLevel * MyData.Level) + (TDamage * WeaponRarityPercentDmg)) * 100.0, false)}%"))


            CritChance := (MyData.CritLevel  * (NPCManager.CritChancePerLevel + (MyData.CritPL * MyData.CritPLLevel))) + ((MyData.CritLevel  * (NPCManager.CritChancePerLevel + (MyData.CritPL * MyData.CritPLLevel))) * CosmeticCritChance)
            CritChancePercentText.SetText(storedstring("{formatter.Abbreviate(CritChance * 100.0, false)}%"))

            CritDamage := (MyData.CritLevel  * (NPCManager.CritChancePerLevel + (MyData.CritPL * MyData.CritPLLevel))) + ((MyData.CritLevel  * (NPCManager.CritChancePerLevel + (MyData.CritPL * MyData.CritPLLevel))) * CosmeticCritDamage)
            CritDamagePercentText.SetText(storedstring("{formatter.Abbreviate(CritDamage * 100.0, false)}%"))

            GoldPercent := (NPCManager.GoldPerElimInitial + (MyData.GoldPL * MyData.GoldPLLevel)) * MyData.GoldPerElim * 100.0
            #GoldPercent := (Pow(NPCManager.GoldPerElimInitial + (MyData.GoldPL * MyData.GoldPLLevel), (MyData.GoldPerElim )*1.0) - 1.0) * 100.0
            GoldLevelText.SetText(storedstring("{formatter.Abbreviate(GoldPercent + (GoldPercent * CosmeticGoldPerElim ),false)}%"))

            ArmorLevelText.SetText(storedstring("{(MyData.PlayerArmor * 5) + (MyData.PlayerArmor * 5) * MyData.PlayerArmorCosmetic}"))

            AgilityLevelText.SetText(storedstring("{MyData.MovementSpeed }"))

            TotalElimNumberText.SetText(storedstring("{formatter.Abbreviate((MyData.TotalPistolEliminations + MyData.TotalSmgEliminations + MyData.TotalAssaultEliminations + MyData.TotalShotgunEliminations + MyData.TotalSniperEliminations) *1.0,false)}"))
            PrestigeEarningsNumberText.SetText(storedstring("{formatter.Abbreviate(MyData.PrestigeEarnings * 1.0,false)}"))
            TotalEarningsNumberText.SetText(storedstring("{formatter.Abbreviate(MyData.TotalEarnings*1.0,false)}")) # Needs to be set
            TowerLevelText.SetText(storedstring("{MyData.TowerLevel}"))
                
        Overlay1 := overlay:
            Slots:=array:
                overlay_slot:
                    HorizontalAlignment:=horizontal_alignment.Fill
                    VerticalAlignment:=vertical_alignment.Fill
                    Padding:=margin{Left:=-5.0,Bottom:=-35.0,Top:=-5.0,Right:=-5.0}
                    Widget:=texture_block:
                        DefaultImage:=wooden_pixel_gui.textures.Panels.WhiteSquare100
                        DefaultTint:=MakeColorFromHex("543C22FF")
                overlay_slot:
                    HorizontalAlignment:=horizontal_alignment.Fill
                    VerticalAlignment:=vertical_alignment.Fill
                    Padding:=margin{Left:=0.0,Bottom:=-30.0,Top:=0.0,Right:=0.0}
                    Widget:=texture_block:
                        DefaultImage:=wooden_pixel_gui.textures.Panels.WhiteSquare100
                        DefaultTint:=MakeColorFromHex("6C4B2AE6")
                overlay_slot:
                    HorizontalAlignment:=horizontal_alignment.Fill
                    VerticalAlignment:=vertical_alignment.Fill
                    Padding:=margin{Left:=0.0,Bottom:=0.0,Top:=0.0,Right:=0.0}
                    Widget := stack_box:
                        Orientation:=orientation.Vertical
                        Slots:=array:
                            stack_box_slot:
                                HorizontalAlignment:=horizontal_alignment.Fill
                                VerticalAlignment:=vertical_alignment.Fill
                                Widget := overlay:
                                    Slots:=array:
                                        overlay_slot:
                                            HorizontalAlignment:=horizontal_alignment.Fill
                                            VerticalAlignment:=vertical_alignment.Fill
                                            Padding:=margin{Left:=0.0,Top:=0.0,Right:=0.0,Bottom:=7.0}
                                            Widget:= TitleBarTexture
                                        overlay_slot:
                                            HorizontalAlignment:=horizontal_alignment.Center
                                            VerticalAlignment:=vertical_alignment.Center
                                            Widget := text_block :
                                                DefaultText:=storedstring("Player Stats")
                                                DefaultTextColor := MakeColorFromHex("543C22FF")
                            stack_box_slot:
                                HorizontalAlignment:=horizontal_alignment.Fill
                                VerticalAlignment:=vertical_alignment.Fill
                                Padding:=margin{Left:=0.0,Top:=0.0,Right:=0.0,Bottom:=0.0}
                                Widget := stack_box:
                                    Orientation:=orientation.Horizontal
                                    Slots:=array:
                                        stack_box_slot:
                                            HorizontalAlignment:=horizontal_alignment.Fill
                                            VerticalAlignment:=vertical_alignment.Fill
                                            Padding:=margin{Left:=0.0,Top:=0.0,Right:=0.0,Bottom:=0.0}
                                            Widget := stack_box:
                                                Orientation:=orientation.Vertical
                                                Slots:=array:
                                                    stack_box_slot:
                                                        HorizontalAlignment:=horizontal_alignment.Fill
                                                        VerticalAlignment:=vertical_alignment.Fill
                                                        #Padding:=margin{Left:=0.0,Top:=0.0,Right:=0.0,Bottom:=0.0}
                                                        Widget := stack_box:
                                                            Orientation:=orientation.Vertical
                                                            Slots:=array:
                                                                stack_box_slot:
                                                                    HorizontalAlignment:=horizontal_alignment.Center
                                                                    VerticalAlignment:=vertical_alignment.Center
                                                                    Padding:=margin{Left:=0.0,Top:=15.0,Right:=0.0,Bottom:=0.0}
                                                                    Widget := text_block :
                                                                        DefaultText:=storedstring("Traits")
                                                                        DefaultTextColor := MakeColorFromHex("FFFFFF")
                                                    stack_box_slot:
                                                        HorizontalAlignment:=horizontal_alignment.Fill
                                                        VerticalAlignment:=vertical_alignment.Fill
                                                        Padding:=margin{Left:=0.0,Top:=5.0,Right:=0.0,Bottom:=0.0}
                                                        Widget:=overlay:
                                                            Slots:=array:
                                                                overlay_slot:
                                                                    HorizontalAlignment:=horizontal_alignment.Fill
                                                                    VerticalAlignment:=vertical_alignment.Fill
                                                                    Widget:=texture_block:
                                                                        DefaultDesiredSize:=vector2{X:=539.0,Y:=53.0}
                                                                        DefaultImage:=wooden_pixel_gui.textures.Panels.WhiteSquare100
                                                                        DefaultTint:=MakeColorFromHex("4A331BFF")
                                                                overlay_slot:
                                                                    Padding:=margin{Left:=5.0,Top:=0.0,Right:=0.0,Bottom:=0.0}
                                                                    HorizontalAlignment:=horizontal_alignment.Left
                                                                    VerticalAlignment:=vertical_alignment.Center
                                                                    Widget:=TrueDamageText
                                                                overlay_slot:
                                                                    Padding:=margin{Left:=0.0,Top:=0.0,Right:=5.0,Bottom:=0.0}
                                                                    HorizontalAlignment:=horizontal_alignment.Right
                                                                    VerticalAlignment:=vertical_alignment.Center
                                                                    Widget:=TrueDamagePercentText
                                                    stack_box_slot:
                                                        HorizontalAlignment:=horizontal_alignment.Fill
                                                        VerticalAlignment:=vertical_alignment.Fill
                                                        Padding:=margin{Left:=0.0,Top:=5.0,Right:=0.0,Bottom:=0.0}
                                                        Widget:=overlay:
                                                            Slots:=array:
                                                                overlay_slot:
                                                                    HorizontalAlignment:=horizontal_alignment.Fill
                                                                    VerticalAlignment:=vertical_alignment.Fill
                                                                    Widget:=texture_block:
                                                                        DefaultDesiredSize:=vector2{X:=539.0,Y:=53.0}
                                                                        DefaultImage:=wooden_pixel_gui.textures.Panels.WhiteSquare100
                                                                        DefaultTint:=MakeColorFromHex("4A331BFF")
                                                                overlay_slot:
                                                                    Padding:=margin{Left:=5.0,Top:=0.0,Right:=0.0,Bottom:=0.0}
                                                                    HorizontalAlignment:=horizontal_alignment.Left
                                                                    VerticalAlignment:=vertical_alignment.Center
                                                                    Widget:=CritChanceText
                                                                overlay_slot:
                                                                    Padding:=margin{Left:=0.0,Top:=0.0,Right:=5.0,Bottom:=0.0}
                                                                    HorizontalAlignment:=horizontal_alignment.Right
                                                                    VerticalAlignment:=vertical_alignment.Center
                                                                    Widget:=CritChancePercentText
                                                    stack_box_slot:
                                                        HorizontalAlignment:=horizontal_alignment.Fill
                                                        VerticalAlignment:=vertical_alignment.Fill
                                                        Padding:=margin{Left:=0.0,Top:=5.0,Right:=0.0,Bottom:=0.0}
                                                        Widget:=overlay:
                                                            Slots:=array:
                                                                overlay_slot:
                                                                    HorizontalAlignment:=horizontal_alignment.Fill
                                                                    VerticalAlignment:=vertical_alignment.Fill
                                                                    Widget:=texture_block:
                                                                        DefaultDesiredSize:=vector2{X:=539.0,Y:=53.0}
                                                                        DefaultImage:=wooden_pixel_gui.textures.Panels.WhiteSquare100
                                                                        DefaultTint:=MakeColorFromHex("4A331BFF")
                                                                overlay_slot:
                                                                    Padding:=margin{Left:=5.0,Top:=0.0,Right:=0.0,Bottom:=0.0}
                                                                    HorizontalAlignment:=horizontal_alignment.Left
                                                                    VerticalAlignment:=vertical_alignment.Center
                                                                    Widget:=CritDamageText
                                                                overlay_slot:
                                                                    Padding:=margin{Left:=0.0,Top:=0.0,Right:=5.0,Bottom:=0.0}
                                                                    HorizontalAlignment:=horizontal_alignment.Right
                                                                    VerticalAlignment:=vertical_alignment.Center
                                                                    Widget:=CritDamagePercentText
                                                    stack_box_slot:
                                                        HorizontalAlignment:=horizontal_alignment.Fill
                                                        VerticalAlignment:=vertical_alignment.Fill
                                                        Padding:=margin{Left:=0.0,Top:=5.0,Right:=0.0,Bottom:=0.0}
                                                        Widget:=overlay:
                                                            Slots:=array:
                                                                overlay_slot:
                                                                    HorizontalAlignment:=horizontal_alignment.Fill
                                                                    VerticalAlignment:=vertical_alignment.Fill
                                                                    Widget:=texture_block:
                                                                        DefaultDesiredSize:=vector2{X:=539.0,Y:=53.0}
                                                                        DefaultImage:=wooden_pixel_gui.textures.Panels.WhiteSquare100
                                                                        DefaultTint:=MakeColorFromHex("4A331BFF")
                                                                overlay_slot:
                                                                    Padding:=margin{Left:=5.0,Top:=0.0,Right:=0.0,Bottom:=0.0}
                                                                    HorizontalAlignment:=horizontal_alignment.Left
                                                                    VerticalAlignment:=vertical_alignment.Center
                                                                    Widget:=GoldText
                                                                overlay_slot:
                                                                    Padding:=margin{Left:=0.0,Top:=0.0,Right:=5.0,Bottom:=0.0}
                                                                    HorizontalAlignment:=horizontal_alignment.Right
                                                                    VerticalAlignment:=vertical_alignment.Center
                                                                    Widget:=GoldLevelText
                                                    stack_box_slot:
                                                        HorizontalAlignment:=horizontal_alignment.Fill
                                                        VerticalAlignment:=vertical_alignment.Fill
                                                        Padding:=margin{Left:=0.0,Top:=5.0,Right:=0.0,Bottom:=0.0}
                                                        Widget:=overlay:
                                                            Slots:=array:
                                                                overlay_slot:
                                                                    HorizontalAlignment:=horizontal_alignment.Fill
                                                                    VerticalAlignment:=vertical_alignment.Fill
                                                                    Widget:=texture_block:
                                                                        DefaultDesiredSize:=vector2{X:=539.0,Y:=53.0}
                                                                        DefaultImage:=wooden_pixel_gui.textures.Panels.WhiteSquare100
                                                                        DefaultTint:=MakeColorFromHex("4A331BFF")
                                                                overlay_slot:
                                                                    Padding:=margin{Left:=5.0,Top:=0.0,Right:=0.0,Bottom:=0.0}
                                                                    HorizontalAlignment:=horizontal_alignment.Left
                                                                    VerticalAlignment:=vertical_alignment.Center
                                                                    Widget:=ArmorText
                                                                overlay_slot:
                                                                    Padding:=margin{Left:=0.0,Top:=0.0,Right:=5.0,Bottom:=0.0}
                                                                    HorizontalAlignment:=horizontal_alignment.Right
                                                                    VerticalAlignment:=vertical_alignment.Center
                                                                    Widget:=ArmorLevelText
                                                    stack_box_slot:
                                                        HorizontalAlignment:=horizontal_alignment.Fill
                                                        VerticalAlignment:=vertical_alignment.Fill
                                                        Padding:=margin{Left:=0.0,Top:=5.0,Right:=0.0,Bottom:=0.0}
                                                        Widget:=overlay:
                                                            Slots:=array:
                                                                overlay_slot:
                                                                    HorizontalAlignment:=horizontal_alignment.Fill
                                                                    VerticalAlignment:=vertical_alignment.Fill
                                                                    Widget:=texture_block:
                                                                        DefaultDesiredSize:=vector2{X:=539.0,Y:=53.0}
                                                                        DefaultImage:=wooden_pixel_gui.textures.Panels.WhiteSquare100
                                                                        DefaultTint:=MakeColorFromHex("4A331BFF")
                                                                overlay_slot:
                                                                    Padding:=margin{Left:=5.0,Top:=0.0,Right:=0.0,Bottom:=0.0}
                                                                    HorizontalAlignment:=horizontal_alignment.Left
                                                                    VerticalAlignment:=vertical_alignment.Center
                                                                    Widget:=AgilityText
                                                                overlay_slot:
                                                                    Padding:=margin{Left:=0.0,Top:=0.0,Right:=5.0,Bottom:=0.0}
                                                                    HorizontalAlignment:=horizontal_alignment.Right
                                                                    VerticalAlignment:=vertical_alignment.Center
                                                                    Widget:=AgilityLevelText
                                                    stack_box_slot:
                                                        HorizontalAlignment:=horizontal_alignment.Fill
                                                        VerticalAlignment:=vertical_alignment.Fill
                                                        #Padding:=margin{Left:=0.0,Top:=0.0,Right:=0.0,Bottom:=0.0}
                                                        Widget := stack_box:
                                                            Orientation:=orientation.Vertical
                                                            Slots:=array:
                                                                stack_box_slot:
                                                                    HorizontalAlignment:=horizontal_alignment.Center
                                                                    VerticalAlignment:=vertical_alignment.Center
                                                                    Padding:=margin{Left:=0.0,Top:=15.0,Right:=0.0,Bottom:=0.0}
                                                                    Widget := text_block :
                                                                        DefaultText:=storedstring("Skills")
                                                                        DefaultTextColor := MakeColorFromHex("FFFFFF")
                                                    stack_box_slot:
                                                        HorizontalAlignment:=horizontal_alignment.Fill
                                                        VerticalAlignment:=vertical_alignment.Fill
                                                        Padding:=margin{Left:=0.0,Top:=5.0,Right:=0.0,Bottom:=0.0}
                                                        Widget:=overlay:
                                                            Slots:=array:
                                                                overlay_slot:
                                                                    HorizontalAlignment:=horizontal_alignment.Fill
                                                                    VerticalAlignment:=vertical_alignment.Fill
                                                                    Widget:=texture_block:
                                                                        DefaultDesiredSize:=vector2{X:=539.0,Y:=53.0}
                                                                        DefaultImage:=wooden_pixel_gui.textures.Panels.WhiteSquare100
                                                                        DefaultTint:=MakeColorFromHex("4A331BFF")
                                                                overlay_slot:
                                                                    Padding:=margin{Left:=5.0,Top:=0.0,Right:=0.0,Bottom:=0.0}
                                                                    HorizontalAlignment:=horizontal_alignment.Left
                                                                    VerticalAlignment:=vertical_alignment.Center
                                                                    Widget:=BurnChanceText
                                                                overlay_slot:
                                                                    Padding:=margin{Left:=0.0,Top:=0.0,Right:=5.0,Bottom:=0.0}
                                                                    HorizontalAlignment:=horizontal_alignment.Right
                                                                    VerticalAlignment:=vertical_alignment.Center
                                                                    Widget:=BurnChancePercentText
                                                    stack_box_slot:
                                                        HorizontalAlignment:=horizontal_alignment.Fill
                                                        VerticalAlignment:=vertical_alignment.Fill
                                                        Padding:=margin{Left:=0.0,Top:=5.0,Right:=0.0,Bottom:=0.0}
                                                        Widget:=overlay:
                                                            Slots:=array:
                                                                overlay_slot:
                                                                    HorizontalAlignment:=horizontal_alignment.Fill
                                                                    VerticalAlignment:=vertical_alignment.Fill
                                                                    Widget:=texture_block:
                                                                        DefaultDesiredSize:=vector2{X:=539.0,Y:=53.0}
                                                                        DefaultImage:=wooden_pixel_gui.textures.Panels.WhiteSquare100
                                                                        DefaultTint:=MakeColorFromHex("4A331BFF")
                                                                overlay_slot:
                                                                    Padding:=margin{Left:=5.0,Top:=0.0,Right:=0.0,Bottom:=0.0}
                                                                    HorizontalAlignment:=horizontal_alignment.Left
                                                                    VerticalAlignment:=vertical_alignment.Center
                                                                    Widget:=BurnDamageText
                                                                overlay_slot:
                                                                    Padding:=margin{Left:=0.0,Top:=0.0,Right:=5.0,Bottom:=0.0}
                                                                    HorizontalAlignment:=horizontal_alignment.Right
                                                                    VerticalAlignment:=vertical_alignment.Center
                                                                    Widget:=BurnDamagePercentText
                                                    stack_box_slot:
                                                        HorizontalAlignment:=horizontal_alignment.Fill
                                                        VerticalAlignment:=vertical_alignment.Fill
                                                        Padding:=margin{Left:=0.0,Top:=5.0,Right:=0.0,Bottom:=0.0}
                                                        Widget:=overlay:
                                                            Slots:=array:
                                                                overlay_slot:
                                                                    HorizontalAlignment:=horizontal_alignment.Fill
                                                                    VerticalAlignment:=vertical_alignment.Fill
                                                                    Widget:=texture_block:
                                                                        DefaultDesiredSize:=vector2{X:=539.0,Y:=53.0}
                                                                        DefaultImage:=wooden_pixel_gui.textures.Panels.WhiteSquare100
                                                                        DefaultTint:=MakeColorFromHex("4A331BFF")
                                                                overlay_slot:
                                                                    Padding:=margin{Left:=5.0,Top:=0.0,Right:=0.0,Bottom:=0.0}
                                                                    HorizontalAlignment:=horizontal_alignment.Left
                                                                    VerticalAlignment:=vertical_alignment.Center
                                                                    Widget:=CrouchHealText
                                                                overlay_slot:
                                                                    Padding:=margin{Left:=0.0,Top:=0.0,Right:=5.0,Bottom:=0.0}
                                                                    HorizontalAlignment:=horizontal_alignment.Right
                                                                    VerticalAlignment:=vertical_alignment.Center
                                                                    Widget:=CrouchHealLevelText
                                                    stack_box_slot:
                                                        HorizontalAlignment:=horizontal_alignment.Fill
                                                        VerticalAlignment:=vertical_alignment.Fill
                                                        Padding:=margin{Left:=0.0,Top:=5.0,Right:=0.0,Bottom:=0.0}
                                                        Widget:=overlay:
                                                            Slots:=array:
                                                                overlay_slot:
                                                                    HorizontalAlignment:=horizontal_alignment.Fill
                                                                    VerticalAlignment:=vertical_alignment.Fill
                                                                    Widget:=texture_block:
                                                                        DefaultDesiredSize:=vector2{X:=539.0,Y:=53.0}
                                                                        DefaultImage:=wooden_pixel_gui.textures.Panels.WhiteSquare100
                                                                        DefaultTint:=MakeColorFromHex("4A331BFF")
                                                                overlay_slot:
                                                                    Padding:=margin{Left:=5.0,Top:=0.0,Right:=0.0,Bottom:=0.0}
                                                                    HorizontalAlignment:=horizontal_alignment.Left
                                                                    VerticalAlignment:=vertical_alignment.Center
                                                                    Widget:=FocusDamageText
                                                                overlay_slot:
                                                                    Padding:=margin{Left:=0.0,Top:=0.0,Right:=5.0,Bottom:=0.0}
                                                                    HorizontalAlignment:=horizontal_alignment.Right
                                                                    VerticalAlignment:=vertical_alignment.Center
                                                                    Widget:=FocusDamagePercentText
                                                    stack_box_slot:
                                                        HorizontalAlignment:=horizontal_alignment.Fill
                                                        VerticalAlignment:=vertical_alignment.Fill
                                                        Padding:=margin{Left:=0.0,Top:=5.0,Right:=0.0,Bottom:=0.0}
                                                        Widget:=overlay:
                                                            Slots:=array:
                                                                overlay_slot:
                                                                    HorizontalAlignment:=horizontal_alignment.Fill
                                                                    VerticalAlignment:=vertical_alignment.Fill
                                                                    Widget:=texture_block:
                                                                        DefaultDesiredSize:=vector2{X:=539.0,Y:=53.0}
                                                                        DefaultImage:=wooden_pixel_gui.textures.Panels.WhiteSquare100
                                                                        DefaultTint:=MakeColorFromHex("4A331BFF")
                                                                overlay_slot:
                                                                    Padding:=margin{Left:=5.0,Top:=0.0,Right:=0.0,Bottom:=0.0}
                                                                    HorizontalAlignment:=horizontal_alignment.Left
                                                                    VerticalAlignment:=vertical_alignment.Center
                                                                    Widget:=AccuracyDamageText
                                                                overlay_slot:
                                                                    Padding:=margin{Left:=0.0,Top:=0.0,Right:=5.0,Bottom:=0.0}
                                                                    HorizontalAlignment:=horizontal_alignment.Right
                                                                    VerticalAlignment:=vertical_alignment.Center
                                                                    Widget:=AccuracyDamagePercentText
                                                                
                                        stack_box_slot:
                                            HorizontalAlignment:=horizontal_alignment.Fill
                                            VerticalAlignment:=vertical_alignment.Fill
                                            Padding:=margin{Left:=25.0,Top:=0.0,Right:=25.0,Bottom:=0.0}
                                            Widget:=texture_block:
                                                DefaultImage:=wooden_pixel_gui.textures.Panels.WhiteSquare100
                                                DefaultTint:=MakeColorFromHex("4A331BFF")
                                                DefaultDesiredSize:=vector2{X:=5.0,Y:=75.0}

                                        stack_box_slot:
                                            HorizontalAlignment:=horizontal_alignment.Fill
                                            VerticalAlignment:=vertical_alignment.Fill
                                            Padding:=margin{Left:=0.0,Top:=0.0,Right:=0.0,Bottom:=0.0}
                                            Widget := stack_box:
                                                Orientation:=orientation.Vertical
                                                Slots:=array:
                                                    stack_box_slot:
                                                        HorizontalAlignment:=horizontal_alignment.Fill
                                                        VerticalAlignment:=vertical_alignment.Fill
                                                        Padding:=margin{Left:=5.0,Top:=40.0,Right:=0.0,Bottom:=0.0}
                                                        Widget:=text_block:
                                                            DefaultText:=storedstring("Skills")
                                                            DefaultTextColor:=MakeColorFromHex("FFFFFF")
                                                    stack_box_slot:
                                                        HorizontalAlignment:=horizontal_alignment.Fill
                                                        VerticalAlignment:=vertical_alignment.Fill
                                                        Padding:=margin{Left:=0.0,Top:=5.0,Right:=0.0,Bottom:=0.0}
                                                        Widget:=overlay:
                                                            Slots:=array:
                                                                overlay_slot:
                                                                    HorizontalAlignment:=horizontal_alignment.Fill
                                                                    VerticalAlignment:=vertical_alignment.Fill
                                                                    Widget:=texture_block:
                                                                        DefaultDesiredSize:=vector2{X:=539.0,Y:=53.0}
                                                                        DefaultImage:=wooden_pixel_gui.textures.Panels.WhiteSquare100
                                                                        DefaultTint:=MakeColorFromHex("4A331BFF")
                                                                overlay_slot:
                                                                    Padding:=margin{Left:=5.0,Top:=0.0,Right:=0.0,Bottom:=0.0}
                                                                    HorizontalAlignment:=horizontal_alignment.Left
                                                                    VerticalAlignment:=vertical_alignment.Center
                                                                    Widget:=LightningChanceText
                                                                overlay_slot:
                                                                    Padding:=margin{Left:=0.0,Top:=0.0,Right:=5.0,Bottom:=0.0}
                                                                    HorizontalAlignment:=horizontal_alignment.Right
                                                                    VerticalAlignment:=vertical_alignment.Center
                                                                    Widget:=LightningChanceLevelText
                                                    stack_box_slot:
                                                        HorizontalAlignment:=horizontal_alignment.Fill
                                                        VerticalAlignment:=vertical_alignment.Fill
                                                        Padding:=margin{Left:=0.0,Top:=5.0,Right:=0.0,Bottom:=0.0}
                                                        Widget:=overlay:
                                                            Slots:=array:
                                                                overlay_slot:
                                                                    HorizontalAlignment:=horizontal_alignment.Fill
                                                                    VerticalAlignment:=vertical_alignment.Fill
                                                                    Widget:=texture_block:
                                                                        DefaultDesiredSize:=vector2{X:=539.0,Y:=53.0}
                                                                        DefaultImage:=wooden_pixel_gui.textures.Panels.WhiteSquare100
                                                                        DefaultTint:=MakeColorFromHex("4A331BFF")
                                                                overlay_slot:
                                                                    Padding:=margin{Left:=5.0,Top:=0.0,Right:=0.0,Bottom:=0.0}
                                                                    HorizontalAlignment:=horizontal_alignment.Left
                                                                    VerticalAlignment:=vertical_alignment.Center
                                                                    Widget:=LightningDamageText
                                                                overlay_slot:
                                                                    Padding:=margin{Left:=0.0,Top:=0.0,Right:=5.0,Bottom:=0.0}
                                                                    HorizontalAlignment:=horizontal_alignment.Right
                                                                    VerticalAlignment:=vertical_alignment.Center
                                                                    Widget:=LightningDamageLevelText
                                                    stack_box_slot:
                                                        HorizontalAlignment:=horizontal_alignment.Fill
                                                        VerticalAlignment:=vertical_alignment.Fill
                                                        Padding:=margin{Left:=0.0,Top:=5.0,Right:=0.0,Bottom:=0.0}
                                                        Widget:=overlay:
                                                            Slots:=array:
                                                                overlay_slot:
                                                                    HorizontalAlignment:=horizontal_alignment.Fill
                                                                    VerticalAlignment:=vertical_alignment.Fill
                                                                    Widget:=texture_block:
                                                                        DefaultDesiredSize:=vector2{X:=539.0,Y:=53.0}
                                                                        DefaultImage:=wooden_pixel_gui.textures.Panels.WhiteSquare100
                                                                        DefaultTint:=MakeColorFromHex("4A331BFF")
                                                                overlay_slot:
                                                                    Padding:=margin{Left:=5.0,Top:=0.0,Right:=0.0,Bottom:=0.0}
                                                                    HorizontalAlignment:=horizontal_alignment.Left
                                                                    VerticalAlignment:=vertical_alignment.Center
                                                                    Widget:=ElimText
                                                                overlay_slot:
                                                                    Padding:=margin{Left:=0.0,Top:=0.0,Right:=5.0,Bottom:=0.0}
                                                                    HorizontalAlignment:=horizontal_alignment.Right
                                                                    VerticalAlignment:=vertical_alignment.Center
                                                                    Widget:=ElimPercentText
                                                    stack_box_slot:
                                                        HorizontalAlignment:=horizontal_alignment.Fill
                                                        VerticalAlignment:=vertical_alignment.Fill
                                                        Padding:=margin{Left:=0.0,Top:=5.0,Right:=0.0,Bottom:=0.0}
                                                        Widget:=overlay:
                                                            Slots:=array:
                                                                overlay_slot:
                                                                    HorizontalAlignment:=horizontal_alignment.Fill
                                                                    VerticalAlignment:=vertical_alignment.Fill
                                                                    Widget:=texture_block:
                                                                        DefaultDesiredSize:=vector2{X:=539.0,Y:=53.0}
                                                                        DefaultImage:=wooden_pixel_gui.textures.Panels.WhiteSquare100
                                                                        DefaultTint:=MakeColorFromHex("4A331BFF")
                                                                overlay_slot:
                                                                    Padding:=margin{Left:=5.0,Top:=0.0,Right:=0.0,Bottom:=0.0}
                                                                    HorizontalAlignment:=horizontal_alignment.Left
                                                                    VerticalAlignment:=vertical_alignment.Center
                                                                    Widget:=SuperCritText
                                                                overlay_slot:
                                                                    Padding:=margin{Left:=0.0,Top:=0.0,Right:=5.0,Bottom:=0.0}
                                                                    HorizontalAlignment:=horizontal_alignment.Right
                                                                    VerticalAlignment:=vertical_alignment.Center
                                                                    Widget:=SuperCritPercentText
                                                    stack_box_slot:
                                                        HorizontalAlignment:=horizontal_alignment.Fill
                                                        VerticalAlignment:=vertical_alignment.Fill
                                                        Padding:=margin{Left:=0.0,Top:=5.0,Right:=0.0,Bottom:=0.0}
                                                        Widget:=overlay:
                                                            Slots:=array:
                                                                overlay_slot:
                                                                    HorizontalAlignment:=horizontal_alignment.Fill
                                                                    VerticalAlignment:=vertical_alignment.Fill
                                                                    Widget:=texture_block:
                                                                        DefaultDesiredSize:=vector2{X:=539.0,Y:=53.0}
                                                                        DefaultImage:=wooden_pixel_gui.textures.Panels.WhiteSquare100
                                                                        DefaultTint:=MakeColorFromHex("4A331BFF")
                                                                overlay_slot:
                                                                    Padding:=margin{Left:=5.0,Top:=0.0,Right:=0.0,Bottom:=0.0}
                                                                    HorizontalAlignment:=horizontal_alignment.Left
                                                                    VerticalAlignment:=vertical_alignment.Center
                                                                    Widget:=SuperCritDamageText
                                                                overlay_slot:
                                                                    Padding:=margin{Left:=0.0,Top:=0.0,Right:=5.0,Bottom:=0.0}
                                                                    HorizontalAlignment:=horizontal_alignment.Right
                                                                    VerticalAlignment:=vertical_alignment.Center
                                                                    Widget:=SuperCritDamagePercentText
                                                    stack_box_slot:
                                                        HorizontalAlignment:=horizontal_alignment.Fill
                                                        VerticalAlignment:=vertical_alignment.Fill
                                                        Padding:=margin{Left:=0.0,Top:=5.0,Right:=0.0,Bottom:=0.0}
                                                        Widget:=overlay:
                                                            Slots:=array:
                                                                overlay_slot:
                                                                    HorizontalAlignment:=horizontal_alignment.Fill
                                                                    VerticalAlignment:=vertical_alignment.Fill
                                                                    Widget:=texture_block:
                                                                        DefaultDesiredSize:=vector2{X:=539.0,Y:=53.0}
                                                                        DefaultImage:=wooden_pixel_gui.textures.Panels.WhiteSquare100
                                                                        DefaultTint:=MakeColorFromHex("4A331BFF")
                                                                overlay_slot:
                                                                    Padding:=margin{Left:=5.0,Top:=0.0,Right:=0.0,Bottom:=0.0}
                                                                    HorizontalAlignment:=horizontal_alignment.Left
                                                                    VerticalAlignment:=vertical_alignment.Center
                                                                    Widget:=OreDropRateText
                                                                overlay_slot:
                                                                    Padding:=margin{Left:=0.0,Top:=0.0,Right:=5.0,Bottom:=0.0}
                                                                    HorizontalAlignment:=horizontal_alignment.Right
                                                                    VerticalAlignment:=vertical_alignment.Center
                                                                    Widget:=OreDropRatePercentText
                                                        
                                                    stack_box_slot:
                                                        HorizontalAlignment:=horizontal_alignment.Center
                                                        VerticalAlignment:=vertical_alignment.Center
                                                        Padding:=margin{Left:=0.0,Top:=15.0,Right:=0.0,Bottom:=0.0}
                                                        Widget:=text_block:
                                                            DefaultText:=storedstring("Stats")
                                                            DefaultTextColor := MakeColorFromHex("FFFFFF")

                                                    stack_box_slot:
                                                        HorizontalAlignment:=horizontal_alignment.Fill
                                                        VerticalAlignment:=vertical_alignment.Fill
                                                        Padding:=margin{Left:=0.0,Top:=5.0,Right:=0.0,Bottom:=0.0}
                                                        Widget:=overlay:
                                                            Slots:=array:
                                                                overlay_slot:
                                                                    HorizontalAlignment:=horizontal_alignment.Fill
                                                                    VerticalAlignment:=vertical_alignment.Fill
                                                                    Widget:=texture_block:
                                                                        DefaultDesiredSize:=vector2{X:=539.0,Y:=53.0}
                                                                        DefaultImage:=wooden_pixel_gui.textures.Panels.WhiteSquare100
                                                                        DefaultTint:=MakeColorFromHex("4A331BFF")
                                                                overlay_slot:
                                                                    Padding:=margin{Left:=5.0,Top:=0.0,Right:=0.0,Bottom:=0.0}
                                                                    HorizontalAlignment:=horizontal_alignment.Left
                                                                    VerticalAlignment:=vertical_alignment.Center
                                                                    Widget:=TotalElimText
                                                                overlay_slot:
                                                                    Padding:=margin{Left:=0.0,Top:=0.0,Right:=5.0,Bottom:=0.0}
                                                                    HorizontalAlignment:=horizontal_alignment.Right
                                                                    VerticalAlignment:=vertical_alignment.Center
                                                                    Widget:=TotalElimNumberText
                                                    stack_box_slot:
                                                        HorizontalAlignment:=horizontal_alignment.Fill
                                                        VerticalAlignment:=vertical_alignment.Fill
                                                        Padding:=margin{Left:=0.0,Top:=5.0,Right:=0.0,Bottom:=0.0}
                                                        Widget:=overlay:
                                                            Slots:=array:
                                                                overlay_slot:
                                                                    HorizontalAlignment:=horizontal_alignment.Fill
                                                                    VerticalAlignment:=vertical_alignment.Fill
                                                                    Widget:=texture_block:
                                                                        DefaultDesiredSize:=vector2{X:=539.0,Y:=53.0}
                                                                        DefaultImage:=wooden_pixel_gui.textures.Panels.WhiteSquare100
                                                                        DefaultTint:=MakeColorFromHex("4A331BFF")
                                                                overlay_slot:
                                                                    Padding:=margin{Left:=5.0,Top:=0.0,Right:=0.0,Bottom:=0.0}
                                                                    HorizontalAlignment:=horizontal_alignment.Left
                                                                    VerticalAlignment:=vertical_alignment.Center
                                                                    Widget:=PrestigeEarningsText
                                                                overlay_slot:
                                                                    Padding:=margin{Left:=0.0,Top:=0.0,Right:=5.0,Bottom:=0.0}
                                                                    HorizontalAlignment:=horizontal_alignment.Right
                                                                    VerticalAlignment:=vertical_alignment.Center
                                                                    Widget:=PrestigeEarningsNumberText
                                                    stack_box_slot:
                                                        HorizontalAlignment:=horizontal_alignment.Fill
                                                        VerticalAlignment:=vertical_alignment.Fill
                                                        Padding:=margin{Left:=0.0,Top:=5.0,Right:=0.0,Bottom:=0.0}
                                                        Widget:=overlay:
                                                            Slots:=array:
                                                                overlay_slot:
                                                                    HorizontalAlignment:=horizontal_alignment.Fill
                                                                    VerticalAlignment:=vertical_alignment.Fill
                                                                    Widget:=texture_block:
                                                                        DefaultDesiredSize:=vector2{X:=539.0,Y:=53.0}
                                                                        DefaultImage:=wooden_pixel_gui.textures.Panels.WhiteSquare100
                                                                        DefaultTint:=MakeColorFromHex("4A331BFF")
                                                                overlay_slot:
                                                                    Padding:=margin{Left:=5.0,Top:=0.0,Right:=0.0,Bottom:=0.0}
                                                                    HorizontalAlignment:=horizontal_alignment.Left
                                                                    VerticalAlignment:=vertical_alignment.Center
                                                                    Widget:=TotalEarningsText
                                                                overlay_slot:
                                                                    Padding:=margin{Left:=0.0,Top:=0.0,Right:=5.0,Bottom:=0.0}
                                                                    HorizontalAlignment:=horizontal_alignment.Right
                                                                    VerticalAlignment:=vertical_alignment.Center
                                                                    Widget:=TotalEarningsNumberText
                                                    stack_box_slot:
                                                        HorizontalAlignment:=horizontal_alignment.Fill
                                                        VerticalAlignment:=vertical_alignment.Fill
                                                        Padding:=margin{Left:=0.0,Top:=5.0,Right:=0.0,Bottom:=0.0}
                                                        Widget:=overlay:
                                                            Slots:=array:
                                                                overlay_slot:
                                                                    HorizontalAlignment:=horizontal_alignment.Fill
                                                                    VerticalAlignment:=vertical_alignment.Fill
                                                                    Widget:=texture_block:
                                                                        DefaultDesiredSize:=vector2{X:=539.0,Y:=53.0}
                                                                        DefaultImage:=wooden_pixel_gui.textures.Panels.WhiteSquare100
                                                                        DefaultTint:=MakeColorFromHex("4A331BFF")
                                                                overlay_slot:
                                                                    Padding:=margin{Left:=5.0,Top:=0.0,Right:=0.0,Bottom:=0.0}
                                                                    HorizontalAlignment:=horizontal_alignment.Left
                                                                    VerticalAlignment:=vertical_alignment.Center
                                                                    Widget:=TowerText
                                                                overlay_slot:
                                                                    Padding:=margin{Left:=0.0,Top:=0.0,Right:=5.0,Bottom:=0.0}
                                                                    HorizontalAlignment:=horizontal_alignment.Right
                                                                    VerticalAlignment:=vertical_alignment.Center
                                                                    Widget:=TowerLevelText
                      
                                    
                            stack_box_slot:
                                HorizontalAlignment:=horizontal_alignment.Center
                                VerticalAlignment:=vertical_alignment.Center
                                Padding:=margin{Left:=0.0,Top:=0.0,Right:=0.0,Bottom:=-20.0}
                                Widget := CloseWindowButton

        MyCanvas:=canvas:
            Slots:=array:
                canvas_slot:
                    Anchors:=anchors{Minimum:=vector2{X:=0.2,Y:=0.07}, Maximum:=vector2{X:=0.8,Y:=0.93}}
                    Widget:= Overlay1
                                                                                       

        return MyCanvas


    CalculatePrice(Agent:agent):int= 
        var Price : int = 0
        if(MyData := Data[player[Agent]]):
            PlayerLevel := MyData.Level
            if(MyXpBar := QuestDevice.CMDevice.XpBarPerPlayer[Agent]?):
                PriceFloat : float = MyXpBar.CalculateExpReq(PlayerLevel)
                if(RoundedPrice := Round[PriceFloat]){set Price = RoundedPrice}
            else:
                Print("MyXpBar not found when calculating price")
                
        return Price



    GetOreRewardTexture(Agent:agent):texture=

        var OreTexture :texture = hud.Copper_icon512

        if(MyData := Data[player[Agent]]):
            QuestNum := MyData.QuestNumber
            if(QuestNum <= 2):
                set OreTexture = hud.Copper_icon512
            else:
                if(QuestNum <= 5):
                    set OreTexture = hud.Silver_icon
                else:
                    if(QuestNum <= 8):
                        set OreTexture = hud.Malachite_icon512
                    else:
                        if(QuestNum <= 11):
                            set OreTexture = hud.Obsidian_icon512
                        else:
                            

        return OreTexture



    ShowUi(Agent:agent):void=

        RemoveFortHud(Agent)
        OpenAudio.Play(Agent)

        MyCanvas: canvas = MakeEntireBackground(Agent)
        
        if(PlayerUI:=GetPlayerUI[player[Agent]]):
            PlayerUI.AddWidget(MyCanvas,player_ui_slot{InputMode:=ui_input_mode.All}) # want it to appear on top
            MyCanvas.SetVisibility(widget_visibility.Hidden)
            MyCanvas.SetVisibility(widget_visibility.Visible)
        if(set MyUI[Agent]= option{MyCanvas}){}



    RemoveHUD(Wdata:widget_message):void =

        Print("Removing")
        if(Player:=Wdata.Player, PlayerUI:=GetPlayerUI[Player],SelectedButton:=Wdata.Source,Agent:=agent[Player]):
            if(MyActualUI := MyUI[Agent]?):

                CloseAudio.Play(Agent)

                PlayerUI.RemoveWidget(MyActualUI)
                MyActualUI.SetVisibility(widget_visibility.Visible)
                MyActualUI.SetVisibility(widget_visibility.Hidden)

                spawn:
                    ShowFortHudDelay(Agent)

    ShowFortHudDelay(Agent:agent)<suspends>:void=

        Sleep(0.1)
        ShowFortHud(Agent)
                
        

    
            

                






