
using { /Fortnite.com/Devices }
using { /Verse.org/Simulation }
using { /UnrealEngine.com/Temporary/Diagnostics }
using { /Fortnite.com/Characters }

shield_device := class(creative_device):

    @editable
    PlayerSpawners : []player_spawner_device = array{}
    
    ShieldPerLevel : float = 10.0

    OnBegin<override>()<suspends>:void=

        Players:=Self.GetPlayspace().GetPlayers()
        for(Player:Players):
            if(Agent:=Player.GetFortCharacter[].GetAgent[]):
                SetShield(Agent)

        for(Spawner:PlayerSpawners):
            Spawner.SpawnedEvent.Subscribe(SetShield)

    SetShield(Agent:agent):void=
        if:
            Player:=player[Agent]
            PlayerData:=Data[Player]
            OldData := Data[Player]
            Char := Agent.GetFortCharacter[]
        then:
            Char.SetMaxShield(Char.GetMaxShield() + (OldData.ShieldLevel*ShieldPerLevel))


