
using { /Fortnite.com/Devices }
using { /Verse.org/Simulation }
using { /UnrealEngine.com/Temporary/Diagnostics }
using { /UnrealEngine.com/Temporary/UI }
using { /UnrealEngine.com/Temporary/SpatialMath }
using { /Fortnite.com/UI }
using { /Verse.org/Colors }
using { /Fortnite.com/Playspaces }
using { /Fortnite.com/Characters }
using { /Verse.org/Assets }

shop_ui11_device := class(creative_device):

    @editable
    BuyAccolade : accolades_device = accolades_device{}

    @editable
    OpenShop : button_device = button_device{}
    @editable
    ShopKeeperCine : cinematic_sequence_device = cinematic_sequence_device{}
    @editable
    HeadNodCine : cinematic_sequence_device = cinematic_sequence_device{}

    @editable
    PurchaseAudio : audio_player_device = audio_player_device{}
    @editable
    NegativeAudio : audio_player_device = audio_player_device{}
    @editable
    OpenAudio : audio_player_device = audio_player_device{}
    @editable
    CloseAudio : audio_player_device = audio_player_device{}

    @editable
    SlotAudio1 : audio_player_device = audio_player_device{}
    #@editable
    #SlotAudio2 : audio_player_device = audio_player_device{}
    #@editable
    #SlotAudio3 : audio_player_device = audio_player_device{}

    @editable
    BuyingXpUpgradeTrigger : trigger_device = trigger_device{}


    @editable
    CurrencyUi : currency_ui_device = currency_ui_device{}

    @editable
    var Slot1Multiple : float = 1.14
    @editable
    var Slot1Initial : float = 125.0
    #@editable
    #var Slot2Multiple : float = 7.0
    #@editable
    #var Slot2Initial : float = 50.0
    #@editable
    #var Slot3Multiple : float = 7.0
    #@editable
    #var Slot3Initial : float = 50.0

    ChainGemDamagePerLevel : float = 0.27
    SuperCritGemDamagePerLevel : float = 1.2
    
    @editable
    Slot1Max : int = 1
    #Slot2Max : int = 99
    #Slot3Max : int = 99

    var MyUI : [agent]?canvas = map{}
    var TextBlockPerAgent : [agent]?[]text_block = map{}
    var TextBlockPerAgentShadows : [agent]?[]text_block = map{}

    var ButtonLoudPerAgent : [agent]?[]button_loud = map{} # no longer being used

    storedstring<localizes>(S:string) : message = "{S}"
    JustIntString<localizes>(I:int) : message = "{I}"
    LevelString<localizes>(I:int) : message = "Lv {I}"
    DamageString<localizes>(I:int) : message = " {I}%"
    SecondString<localizes>(I:int) : message = "{I} Sec"
    HpString<localizes>(I:int) : message = "{I} hp"

    var StoredButtonTestingToclose : button_loud = button_loud{}

    OnBegin<override>()<suspends>:void=

        OpenShop.InteractedWithEvent.Subscribe(ShowUi)
    
    MakeEntireBackground (Agent:agent):canvas=

        GemTexture : texture = wooden_pixel_gui.textures.Icons2.Size8.t_GemRed_8

        ChestTexture : texture = RavenIcons.SkillsSpellsandScores.A64x64.S22 # Slot1
        CritTexture : texture = RavenIcons.MapMarkers.S164 # Slot2
        GoldTexture : texture = RavenIcons.MapMarkers.S554 # Slot3

        Slot1Title: text_block = text_block{DefaultTextColor := MakeColorFromHex("529F17FF"), DefaultText:=storedstring("Elim Experience")}
        Slot1TitleShadow: text_block = text_block{DefaultTextColor := MakeColorFromHex("142706FF"), DefaultText:=storedstring("Elim Experience")}
        Slot1Image: texture_block = texture_block{DefaultImage:= ChestTexture, DefaultDesiredSize:=vector2{X:=100.0,Y:=100.0}}
        Slot1ImageShadow: texture_block = texture_block{DefaultImage:= ChestTexture,DefaultTint:=MakeColorFromHex("080402FF"), DefaultDesiredSize:=vector2{X:=100.0,Y:=100.0}}
        Slot1Description: text_block = text_block{DefaultTextColor := MakeColorFromHex("FFFFFF"), DefaultText:=storedstring("Increase experience\ngained from elims")}
        Slot1Price: text_block = text_block{DefaultTextColor := MakeColorFromHex("FFFFFF"), DefaultText:=storedstring("X1.25M")} # Needs to be formated
        Slot1Button : button_quiet = button_quiet{DefaultText:=storedstring("")}
        Slot1Button.OnClick().Subscribe(UpgradeSlotOne) 

        Slot2Title: text_block = text_block{DefaultTextColor := MakeColorFromHex("529F17FF"), DefaultText:=storedstring("Lightning Damage")}
        Slot2TitleShadow: text_block = text_block{DefaultTextColor := MakeColorFromHex("142706FF"), DefaultText:=storedstring("Lightning Damage")}
        Slot2Level: text_block = text_block{DefaultTextColor := MakeColorFromHex("529F17FF"), DefaultText:=LevelString(2)}
        Slot2LevelShadow: text_block = text_block{DefaultTextColor := MakeColorFromHex("142706FF"), DefaultText:=LevelString(2)}
        Slot2Image: texture_block = texture_block{DefaultImage:= CritTexture, DefaultDesiredSize:=vector2{X:=100.0,Y:=100.0}}
        Slot2ImageShadow: texture_block = texture_block{DefaultImage:= CritTexture,DefaultTint:=MakeColorFromHex("080402FF"), DefaultDesiredSize:=vector2{X:=100.0,Y:=100.0}}
        Slot2Description: text_block = text_block{DefaultTextColor := MakeColorFromHex("FFFFFF"), DefaultText:=storedstring("Increase lightning\ndamage scaling")}
        Slot2Price: text_block = text_block{DefaultTextColor := MakeColorFromHex("FFFFFF"), DefaultText:=storedstring("X1.25M")} # Needs to be formated
        Slot2Button : button_quiet = button_quiet{DefaultText:=storedstring("")}
        Slot2Button.OnClick().Subscribe(UpgradeSlotTwo) 

        Slot3Title: text_block = text_block{DefaultTextColor := MakeColorFromHex("529F17FF"), DefaultText:=storedstring("Super Crit Damage")}
        Slot3TitleShadow: text_block = text_block{DefaultTextColor := MakeColorFromHex("142706FF"), DefaultText:=storedstring("Super Crit Damage")}
        Slot3Level: text_block = text_block{DefaultTextColor := MakeColorFromHex("529F17FF"), DefaultText:=LevelString(2)}
        Slot3LevelShadow: text_block = text_block{DefaultTextColor := MakeColorFromHex("142706FF"), DefaultText:=LevelString(2)}
        Slot3Image: texture_block = texture_block{DefaultImage:= GoldTexture, DefaultDesiredSize:=vector2{X:=100.0,Y:=100.0}}
        Slot3ImageShadow: texture_block = texture_block{DefaultImage:= GoldTexture,DefaultTint:=MakeColorFromHex("080402FF"), DefaultDesiredSize:=vector2{X:=100.0,Y:=100.0}}
        Slot3Description: text_block = text_block{DefaultTextColor := MakeColorFromHex("FFFFFF"), DefaultText:=storedstring("Increase super-crit\ndamage scaling")}
        Slot3Price: text_block = text_block{DefaultTextColor := MakeColorFromHex("FFFFFF"), DefaultText:=storedstring("X1.25M")} # Needs to be formated
        Slot3Button : button_quiet = button_quiet{DefaultText:=storedstring("")}
        Slot3Button.OnClick().Subscribe(UpgradeSlotThree) 

        CloseButton : button_quiet = button_quiet{DefaultText:=storedstring("Close")}
        CloseButton.OnClick().Subscribe(RemoveHUD) 

        InventoryGemsText : text_block = text_block{DefaultTextColor := MakeColorFromHex("FFFFFF"), DefaultText:=storedstring("")}

        if(set TextBlockPerAgent[Agent] = option{array{Slot1Price,Slot2Level,Slot2Price,Slot3Level,Slot3Price,InventoryGemsText}}){} 
        if(set TextBlockPerAgentShadows[Agent] = option{array{Slot2LevelShadow,Slot3LevelShadow}}){} 

        UpdateGoldText(Agent)


        MyCanvas:=canvas:
            Slots:=array:
                canvas_slot:
                    Anchors:=anchors{Minimum:=vector2{X:=0.1,Y:=0.05}, Maximum:=vector2{X:=0.75,Y:=0.95}}
                    Widget:=stack_box:
                        Orientation:=orientation.Vertical
                        Slots:=array:
                            stack_box_slot:
                                HorizontalAlignment:=horizontal_alignment.Fill
                                VerticalAlignment:=vertical_alignment.Fill
                                Widget:=overlay:
                                    Slots:=array:
                                        overlay_slot:
                                            HorizontalAlignment:=horizontal_alignment.Fill
                                            VerticalAlignment:=vertical_alignment.Fill
                                            Widget:=texture_block:
                                                DefaultImage:=wooden_pixel_gui.textures.Panels.t_TitleBar
                                                DefaultDesiredSize:=vector2{X:=50.0,Y:=110.0}
                                        overlay_slot:
                                            HorizontalAlignment:=horizontal_alignment.Center
                                            VerticalAlignment:=vertical_alignment.Center
                                            Widget:=text_block:
                                                DefaultText:=storedstring("Gem Shop")
                                                DefaultTextColor := MakeColorFromHex("401805FF")
                            stack_box_slot:
                                HorizontalAlignment:=horizontal_alignment.Fill
                                VerticalAlignment:=vertical_alignment.Fill
                                Widget:=overlay:
                                    Slots:=array:
                                        overlay_slot:
                                            HorizontalAlignment:=horizontal_alignment.Fill
                                            VerticalAlignment:=vertical_alignment.Fill
                                            Widget:=texture_block:
                                                DefaultImage:=wooden_pixel_gui.textures.Panels.WhiteSquare75
                                                DefaultTint:=MakeColorFromHex("251206BF")
                                                DefaultDesiredSize:=vector2{X:=1000.0,Y:=700.0}
                                        overlay_slot: # Main stackbox with panels in it
                                            HorizontalAlignment:=horizontal_alignment.Center
                                            VerticalAlignment:=vertical_alignment.Center
                                            Widget:=stack_box:
                                                Orientation:=orientation.Horizontal
                                                Slots:=array:
                                                    stack_box_slot:
                                                        HorizontalAlignment:=horizontal_alignment.Fill
                                                        VerticalAlignment:=vertical_alignment.Fill
                                                        Padding:=margin{Left:=50.0,Top:=0.0,Right:=50.0,Bottom:=0.0}
                                                        Widget:=overlay:
                                                            Slots:=array:
                                                                overlay_slot:
                                                                    HorizontalAlignment:=horizontal_alignment.Fill
                                                                    VerticalAlignment:=vertical_alignment.Fill
                                                                    Padding:=margin{Left:=-20.0,Top:=-20.0,Right:=-20.0,Bottom:=-20.0}
                                                                    Widget:=texture_block:
                                                                        DefaultImage:=wooden_pixel_gui.textures.Panels.BrownSquare
                                                                        DefaultTint:= MakeColorFromHex("A3A3A3FF")
                                                                overlay_slot:
                                                                    HorizontalAlignment:=horizontal_alignment.Center
                                                                    VerticalAlignment:=vertical_alignment.Center
                                                                    Widget:=stack_box:
                                                                        Orientation:=orientation.Vertical
                                                                        Slots:=array:
                                                                            stack_box_slot:
                                                                                HorizontalAlignment:=horizontal_alignment.Fill
                                                                                VerticalAlignment:=vertical_alignment.Fill
                                                                                Widget:=overlay:
                                                                                    Slots:=array:
                                                                                        overlay_slot:
                                                                                            HorizontalAlignment:=horizontal_alignment.Center
                                                                                            VerticalAlignment:=vertical_alignment.Center
                                                                                            Padding:=margin{Left:=3.0,Top:=3.0,Right:=0.0,Bottom:=0.0}
                                                                                            Widget:=Slot1TitleShadow
                                                                                        overlay_slot:
                                                                                            HorizontalAlignment:=horizontal_alignment.Center
                                                                                            VerticalAlignment:=vertical_alignment.Center
                                                                                            Widget:=Slot1Title
                                                                            stack_box_slot:
                                                                                HorizontalAlignment:=horizontal_alignment.Fill
                                                                                VerticalAlignment:=vertical_alignment.Fill
                                                                                Padding:=margin{Left:=0.0,Top:=45.0,Right:=0.0,Bottom:=0.0}
                                                                                Widget:=overlay:
                                                                                    Slots:=array:
                                                                                        overlay_slot:
                                                                                            HorizontalAlignment:=horizontal_alignment.Center
                                                                                            VerticalAlignment:=vertical_alignment.Center
                                                                                            Padding:=margin{Left:=0.0,Top:=6.0,Right:=0.0,Bottom:=0.0}
                                                                                            Widget:=Slot1ImageShadow
                                                                                        overlay_slot:
                                                                                            HorizontalAlignment:=horizontal_alignment.Center
                                                                                            VerticalAlignment:=vertical_alignment.Center
                                                                                            Widget:=Slot1Image
                                                                            stack_box_slot:
                                                                                HorizontalAlignment:=horizontal_alignment.Fill
                                                                                VerticalAlignment:=vertical_alignment.Fill
                                                                                Padding:=margin{Left:=0.0,Top:=25.0,Right:=0.0,Bottom:=0.0}
                                                                                Widget:=Slot1Description
                                                                            stack_box_slot:
                                                                                HorizontalAlignment:=horizontal_alignment.Fill
                                                                                VerticalAlignment:=vertical_alignment.Fill
                                                                                Padding:=margin{Left:=0.0,Top:=35.0,Right:=0.0,Bottom:=0.0}
                                                                                Widget:=overlay:
                                                                                    Slots:=array:
                                                                                        overlay_slot:
                                                                                            HorizontalAlignment:=horizontal_alignment.Fill
                                                                                            VerticalAlignment:=vertical_alignment.Fill
                                                                                            Widget:=texture_block:
                                                                                                DefaultImage:=wooden_pixel_gui.textures.Panels.BrownSquare
                                                                                                DefaultTint:= MakeColorFromHex("0000004D")
                                                                                        overlay_slot:
                                                                                            HorizontalAlignment:=horizontal_alignment.Center
                                                                                            VerticalAlignment:=vertical_alignment.Center
                                                                                            Widget:=stack_box:
                                                                                                Orientation:=orientation.Horizontal
                                                                                                Slots:=array:
                                                                                                    stack_box_slot:
                                                                                                        HorizontalAlignment:=horizontal_alignment.Center
                                                                                                        VerticalAlignment:=vertical_alignment.Center
                                                                                                        Widget:=texture_block:
                                                                                                            DefaultImage:=GemTexture
                                                                                                    stack_box_slot:
                                                                                                        Widget:=Slot1Price
                                                                                        overlay_slot:
                                                                                            HorizontalAlignment:=horizontal_alignment.Fill
                                                                                            VerticalAlignment:=vertical_alignment.Fill
                                                                                            Widget:=overlay:
                                                                                                Slots:=array:
                                                                                                    overlay_slot:
                                                                                                        HorizontalAlignment:=horizontal_alignment.Fill
                                                                                                        VerticalAlignment:=vertical_alignment.Fill
                                                                                                        Widget:=Slot1Button
                                                    <#stack_box_slot:
                                                        HorizontalAlignment:=horizontal_alignment.Fill
                                                        VerticalAlignment:=vertical_alignment.Fill
                                                        Padding:=margin{Left:=50.0,Top:=0.0,Right:=50.0,Bottom:=0.0}
                                                        Widget:=overlay:
                                                            Slots:=array:
                                                                overlay_slot:
                                                                    HorizontalAlignment:=horizontal_alignment.Fill
                                                                    VerticalAlignment:=vertical_alignment.Fill
                                                                    Padding:=margin{Left:=-20.0,Top:=-20.0,Right:=-20.0,Bottom:=-20.0}
                                                                    Widget:=texture_block:
                                                                        DefaultImage:=wooden_pixel_gui.textures.Panels.BrownSquare
                                                                        DefaultTint:= MakeColorFromHex("A3A3A3FF")
                                                                overlay_slot:
                                                                    HorizontalAlignment:=horizontal_alignment.Center
                                                                    VerticalAlignment:=vertical_alignment.Center
                                                                    Widget:=stack_box:
                                                                        Orientation:=orientation.Vertical
                                                                        Slots:=array:
                                                                            stack_box_slot:
                                                                                HorizontalAlignment:=horizontal_alignment.Fill
                                                                                VerticalAlignment:=vertical_alignment.Fill
                                                                                Widget:=overlay:
                                                                                    Slots:=array:
                                                                                        overlay_slot:
                                                                                            HorizontalAlignment:=horizontal_alignment.Center
                                                                                            VerticalAlignment:=vertical_alignment.Center
                                                                                            Padding:=margin{Left:=3.0,Top:=3.0,Right:=0.0,Bottom:=0.0}
                                                                                            Widget:=Slot2TitleShadow
                                                                                        overlay_slot:
                                                                                            HorizontalAlignment:=horizontal_alignment.Center
                                                                                            VerticalAlignment:=vertical_alignment.Center
                                                                                            Widget:=Slot2Title
                                                                            stack_box_slot:
                                                                                HorizontalAlignment:=horizontal_alignment.Fill
                                                                                VerticalAlignment:=vertical_alignment.Fill
                                                                                Widget:=overlay:
                                                                                    Slots:=array:
                                                                                        overlay_slot:
                                                                                            HorizontalAlignment:=horizontal_alignment.Center
                                                                                            VerticalAlignment:=vertical_alignment.Center
                                                                                            Padding:=margin{Left:=3.0,Top:=3.0,Right:=0.0,Bottom:=0.0}
                                                                                            Widget:=Slot2LevelShadow
                                                                                        overlay_slot:
                                                                                            HorizontalAlignment:=horizontal_alignment.Center
                                                                                            VerticalAlignment:=vertical_alignment.Center
                                                                                            Widget:=Slot2Level
                                                                            stack_box_slot:
                                                                                HorizontalAlignment:=horizontal_alignment.Fill
                                                                                VerticalAlignment:=vertical_alignment.Fill
                                                                                Padding:=margin{Left:=0.0,Top:=45.0,Right:=0.0,Bottom:=0.0}
                                                                                Widget:=overlay:
                                                                                    Slots:=array:
                                                                                        overlay_slot:
                                                                                            HorizontalAlignment:=horizontal_alignment.Center
                                                                                            VerticalAlignment:=vertical_alignment.Center
                                                                                            Padding:=margin{Left:=0.0,Top:=6.0,Right:=0.0,Bottom:=0.0}
                                                                                            Widget:=Slot2ImageShadow
                                                                                        overlay_slot:
                                                                                            HorizontalAlignment:=horizontal_alignment.Center
                                                                                            VerticalAlignment:=vertical_alignment.Center
                                                                                            Widget:=Slot2Image
                                                                            stack_box_slot:
                                                                                HorizontalAlignment:=horizontal_alignment.Fill
                                                                                VerticalAlignment:=vertical_alignment.Fill
                                                                                Padding:=margin{Left:=0.0,Top:=25.0,Right:=0.0,Bottom:=0.0}
                                                                                Widget:=Slot2Description
                                                                            stack_box_slot:
                                                                                HorizontalAlignment:=horizontal_alignment.Fill
                                                                                VerticalAlignment:=vertical_alignment.Fill
                                                                                Padding:=margin{Left:=0.0,Top:=35.0,Right:=0.0,Bottom:=0.0}
                                                                                Widget:=overlay:
                                                                                    Slots:=array:
                                                                                        overlay_slot:
                                                                                            HorizontalAlignment:=horizontal_alignment.Fill
                                                                                            VerticalAlignment:=vertical_alignment.Fill
                                                                                            Widget:=texture_block:
                                                                                                DefaultImage:=wooden_pixel_gui.textures.Panels.BrownSquare
                                                                                                DefaultTint:= MakeColorFromHex("0000004D")           
                                                                                        overlay_slot:
                                                                                            HorizontalAlignment:=horizontal_alignment.Center
                                                                                            VerticalAlignment:=vertical_alignment.Center
                                                                                            Widget:=stack_box:
                                                                                                Orientation:=orientation.Horizontal
                                                                                                Slots:=array:
                                                                                                    stack_box_slot:
                                                                                                        HorizontalAlignment:=horizontal_alignment.Center
                                                                                                        VerticalAlignment:=vertical_alignment.Center
                                                                                                        Widget:=texture_block:
                                                                                                            DefaultImage:=GemTexture
                                                                                                    stack_box_slot:
                                                                                                        Widget:=Slot2Price
                                                                                        overlay_slot:
                                                                                            HorizontalAlignment:=horizontal_alignment.Fill
                                                                                            VerticalAlignment:=vertical_alignment.Fill
                                                                                            Widget:=overlay:
                                                                                                Slots:=array:
                                                                                                    
                                                                                                    overlay_slot:
                                                                                                        HorizontalAlignment:=horizontal_alignment.Fill
                                                                                                        VerticalAlignment:=vertical_alignment.Fill
                                                                                                        Widget:=Slot2Button#>
                                                    <#stack_box_slot:
                                                        HorizontalAlignment:=horizontal_alignment.Fill
                                                        VerticalAlignment:=vertical_alignment.Fill
                                                        Padding:=margin{Left:=50.0,Top:=0.0,Right:=50.0,Bottom:=0.0}
                                                        Widget:=overlay:
                                                            Slots:=array:
                                                                overlay_slot:
                                                                    HorizontalAlignment:=horizontal_alignment.Fill
                                                                    VerticalAlignment:=vertical_alignment.Fill
                                                                    Padding:=margin{Left:=-20.0,Top:=-20.0,Right:=-20.0,Bottom:=-20.0}
                                                                    Widget:=texture_block:
                                                                        DefaultImage:=wooden_pixel_gui.textures.Panels.BrownSquare
                                                                        DefaultTint:= MakeColorFromHex("A3A3A3FF")
                                                                overlay_slot:
                                                                    HorizontalAlignment:=horizontal_alignment.Center
                                                                    VerticalAlignment:=vertical_alignment.Center
                                                                    Widget:=stack_box:
                                                                        Orientation:=orientation.Vertical
                                                                        Slots:=array:
                                                                            stack_box_slot:
                                                                                HorizontalAlignment:=horizontal_alignment.Fill
                                                                                VerticalAlignment:=vertical_alignment.Fill
                                                                                Widget:=overlay:
                                                                                    Slots:=array:
                                                                                        overlay_slot:
                                                                                            HorizontalAlignment:=horizontal_alignment.Center
                                                                                            VerticalAlignment:=vertical_alignment.Center
                                                                                            Padding:=margin{Left:=3.0,Top:=3.0,Right:=0.0,Bottom:=0.0}
                                                                                            Widget:=Slot3TitleShadow
                                                                                        overlay_slot:
                                                                                            HorizontalAlignment:=horizontal_alignment.Center
                                                                                            VerticalAlignment:=vertical_alignment.Center
                                                                                            Widget:=Slot3Title
                                                                            stack_box_slot:
                                                                                HorizontalAlignment:=horizontal_alignment.Fill
                                                                                VerticalAlignment:=vertical_alignment.Fill
                                                                                Widget:=overlay:
                                                                                    Slots:=array:
                                                                                        overlay_slot:
                                                                                            HorizontalAlignment:=horizontal_alignment.Center
                                                                                            VerticalAlignment:=vertical_alignment.Center
                                                                                            Padding:=margin{Left:=3.0,Top:=3.0,Right:=0.0,Bottom:=0.0}
                                                                                            Widget:=Slot3LevelShadow
                                                                                        overlay_slot:
                                                                                            HorizontalAlignment:=horizontal_alignment.Center
                                                                                            VerticalAlignment:=vertical_alignment.Center
                                                                                            Widget:=Slot3Level
                                                                            stack_box_slot:
                                                                                HorizontalAlignment:=horizontal_alignment.Fill
                                                                                VerticalAlignment:=vertical_alignment.Fill
                                                                                Padding:=margin{Left:=0.0,Top:=45.0,Right:=0.0,Bottom:=0.0}
                                                                                Widget:=overlay:
                                                                                    Slots:=array:
                                                                                        overlay_slot:
                                                                                            HorizontalAlignment:=horizontal_alignment.Center
                                                                                            VerticalAlignment:=vertical_alignment.Center
                                                                                            Padding:=margin{Left:=0.0,Top:=6.0,Right:=0.0,Bottom:=0.0}
                                                                                            Widget:=Slot3ImageShadow
                                                                                        overlay_slot:
                                                                                            HorizontalAlignment:=horizontal_alignment.Center
                                                                                            VerticalAlignment:=vertical_alignment.Center
                                                                                            Widget:=Slot3Image
                                                                            stack_box_slot:
                                                                                HorizontalAlignment:=horizontal_alignment.Fill
                                                                                VerticalAlignment:=vertical_alignment.Fill
                                                                                Padding:=margin{Left:=0.0,Top:=25.0,Right:=0.0,Bottom:=0.0}
                                                                                Widget:=Slot3Description
                                                                            stack_box_slot:
                                                                                HorizontalAlignment:=horizontal_alignment.Fill
                                                                                VerticalAlignment:=vertical_alignment.Fill
                                                                                Padding:=margin{Left:=0.0,Top:=35.0,Right:=0.0,Bottom:=0.0}
                                                                                Widget:=overlay:
                                                                                    Slots:=array:
                                                                                        overlay_slot:
                                                                                            HorizontalAlignment:=horizontal_alignment.Fill
                                                                                            VerticalAlignment:=vertical_alignment.Fill
                                                                                            Widget:=texture_block:
                                                                                                DefaultImage:=wooden_pixel_gui.textures.Panels.BrownSquare
                                                                                                DefaultTint:= MakeColorFromHex("0000004D")           
                                                                                        overlay_slot:
                                                                                            HorizontalAlignment:=horizontal_alignment.Center
                                                                                            VerticalAlignment:=vertical_alignment.Center
                                                                                            Widget:=stack_box:
                                                                                                Orientation:=orientation.Horizontal
                                                                                                Slots:=array:
                                                                                                    stack_box_slot:
                                                                                                        HorizontalAlignment:=horizontal_alignment.Center
                                                                                                        VerticalAlignment:=vertical_alignment.Center
                                                                                                        Widget:=texture_block:
                                                                                                            DefaultImage:=GemTexture
                                                                                                    stack_box_slot:
                                                                                                        Widget:=Slot3Price
                                                                                        overlay_slot:
                                                                                            HorizontalAlignment:=horizontal_alignment.Fill
                                                                                            VerticalAlignment:=vertical_alignment.Fill
                                                                                            Widget:=overlay:
                                                                                                Slots:=array:
                                                                                                    
                                                                                                    overlay_slot:
                                                                                                        HorizontalAlignment:=horizontal_alignment.Fill
                                                                                                        VerticalAlignment:=vertical_alignment.Fill
                                                                                                        Widget:=Slot3Button#>
                                                   
                                        overlay_slot:
                                            HorizontalAlignment:=horizontal_alignment.Center
                                            VerticalAlignment:=vertical_alignment.Bottom
                                            Padding:=margin{Left:=0.0,Top:=0.0,Right:=0.0,Bottom:=25.0}
                                            Widget:=CloseButton
                                        overlay_slot:
                                            HorizontalAlignment:=horizontal_alignment.Left
                                            VerticalAlignment:=vertical_alignment.Bottom
                                            Padding:=margin{Left:=20.0,Top:=0.0,Right:=0.0,Bottom:=20.0}
                                            Widget:=overlay:
                                                Slots:=array:
                                                    overlay_slot:
                                                        HorizontalAlignment:=horizontal_alignment.Left
                                                        VerticalAlignment:=vertical_alignment.Top
                                                        Widget:=stack_box:
                                                            Orientation:=orientation.Horizontal
                                                            Slots:=array:
                                                                stack_box_slot:
                                                                    HorizontalAlignment:=horizontal_alignment.Fill
                                                                    VerticalAlignment:=vertical_alignment.Fill
                                                                    Widget:=texture_block:
                                                                        DefaultImage:=GemTexture
                                                                        DefaultDesiredSize:=vector2{X:=41.0,Y:=32.0}
                                                                stack_box_slot:
                                                                    HorizontalAlignment:=horizontal_alignment.Left
                                                                    VerticalAlignment:=vertical_alignment.Center
                                                                    Padding:=margin{Left:=10.0,Top:=0.0,Right:=0.0,Bottom:=0.0}
                                                                    Widget:=InventoryGemsText

                                                                                       

        return MyCanvas

    ShowUi(Agent:agent):void=

        RemoveFortHud(Agent)
        ShopKeeperCine.Play(Agent)
        OpenAudio.Play(Agent)

        MyCanvas: canvas = MakeEntireBackground(Agent)
        
        if(PlayerUI:=GetPlayerUI[player[Agent]]):
            PlayerUI.AddWidget(MyCanvas,player_ui_slot{InputMode:=ui_input_mode.All,ZOrder:=0}) # want it to appear on top
            UpdateUiText(Agent)
            MyCanvas.SetVisibility(widget_visibility.Hidden)
            MyCanvas.SetVisibility(widget_visibility.Visible)
            CurrencyUi.HideCurrencyUI(Agent)
        if(set MyUI[Agent]= option{MyCanvas}){}

    RemoveHUD(Wdata:widget_message):void =

        Print("Removing")
        if(Player:=Wdata.Player, PlayerUI:=GetPlayerUI[Player],SelectedButton:=Wdata.Source,Agent:=agent[Player]):
            if(MyActualUI := MyUI[Agent]?):

                CloseAudio.Play(Agent)

                PlayerUI.RemoveWidget(MyActualUI)
                MyActualUI.SetVisibility(widget_visibility.Visible)
                MyActualUI.SetVisibility(widget_visibility.Hidden)
                CurrencyUi.ShowCurrencyUI(Agent)

                if(set MyUI[Agent] = false){}
                if(set TextBlockPerAgent[Agent] = false){}
                
                spawn:
                    CloseCineDelay(Agent)
                
    CloseCineDelay(Agent:agent)<suspends>:void=

        ShowFortHud(Agent)
        ShopKeeperCine.Stop(Agent)

    UpdateUiText(Agent:agent):void=

        if:
            Slot1Price: text_block = TextBlockPerAgent[Agent]?[0]
            Slot2Level: text_block = TextBlockPerAgent[Agent]?[1]
            Slot2Price: text_block = TextBlockPerAgent[Agent]?[2]
            Slot3Level: text_block = TextBlockPerAgent[Agent]?[3]
            Slot3Price: text_block = TextBlockPerAgent[Agent]?[4]

            Slot2LevelShadow: text_block = TextBlockPerAgentShadows[Agent]?[0]
            Slot3LevelShadow: text_block = TextBlockPerAgentShadows[Agent]?[1]

            Player := player[Agent]
            PlayerData := Data[Agent]
            
        then:


            var Price : float = Slot1Initial
            if(PlayerData.ExpBoost > 0):
                for(i:=0..PlayerData.ExpBoost-1):
                    set Price = Price * Slot1Multiple
            
            if(Slot1Max > PlayerData.ExpBoost or Slot1Max = 0):
                Slot1Price.SetText(storedstring(formatter.Abbreviate((Price * 1.0),false)))
            else:
                Slot1Price.SetText(storedstring("Max"))

            # Burn Chance #############################################################################

            #Slot2Level.SetText(LevelString(PlayerData.ChainDamageGemLevel))
            #Slot2LevelShadow.SetText(LevelString(PlayerData.ChainDamageGemLevel))

            <#var ChancePriceFloat : float = Slot2Initial
            if(PlayerData.ChainDamageGemLevel > 0):
                for(i:=0..PlayerData.ChainDamageGemLevel-1):
                    set ChancePriceFloat = ChancePriceFloat * Slot2Multiple

 
            if(Slot2Max > PlayerData.ChainDamageGemLevel or Slot2Max = 0):
                Slot2Price.SetText(storedstring(formatter.Abbreviate((ChancePriceFloat * 1.0),false)))
            else:
                Slot2Price.SetText(storedstring("Max"))

                #>
            #############################################################################

            <#
            Slot3Level.SetText(LevelString(PlayerData.SuperCritGemLevel))
            Slot3LevelShadow.SetText(LevelString(PlayerData.SuperCritGemLevel))

            var PriceFloat2 : float = Slot3Initial
            if(PlayerData.SuperCritGemLevel > 0):
                for(i:=0..PlayerData.SuperCritGemLevel-1):
                    set PriceFloat2 = PriceFloat2 * Slot3Multiple

            if(Slot3Max > PlayerData.SuperCritGemLevel or Slot3Max = 0):
                Slot3Price.SetText(storedstring(formatter.Abbreviate((PriceFloat2 * 1.0),false)))
            else:
                Slot3Price.SetText(storedstring("Max"))
            #>

 

    UpgradeSlotOne(Wdata:widget_message):void=
        if(Player:=Wdata.Player, PlayerUI:=GetPlayerUI[Player],SelectedButton:=Wdata.Source,Agent:=agent[Player]):
            if (MyData := Data[Player]):

                if(Slot1Max > MyData.ExpBoost):

                    var Price : float = Slot1Initial
                    for(i:=0..MyData.ExpBoost-1):
                        set Price = Price * Slot1Multiple

                    if(MyData.Gems >= Price):
                        if:
                            OldData := Data[Player]
                            set Data[Player] = player_data{player_data_update<constructor>(OldData),ExpBoost := OldData.ExpBoost + 1, Gems := OldData.Gems - Price}
                        then:
                            SlotAudio1.Play(Agent)
                            CompleteBoughtAnims(Agent)
                    else:
                        NegativeAudio.Play(Agent)
                else:
                    NegativeAudio.Play(Agent)

    UpgradeSlotTwo(Wdata:widget_message):void=
        if(Player:=Wdata.Player, PlayerUI:=GetPlayerUI[Player],SelectedButton:=Wdata.Source,Agent:=agent[Player]):   
            if (MyData := Data[Player]):<#
                if(Slot2Max > MyData.ChainDamageGemLevel):
                    var Price : float = Slot2Initial
                    for(i:=0..MyData.ChainDamageGemLevel-1):
                        set Price = Price * Slot2Multiple

                    if(MyData.Gems >= Price):
                        if:
                            OldData := Data[Player]
                            set Data[Player] = player_data{player_data_update<constructor>(OldData),ChainDamageGemLevel := OldData.ChainDamageGemLevel + 1, Gems := OldData.Gems - Price,ChainLightningDamagePerUpgrade := ChainGemDamagePerLevel + OldData.ChainLightningDamagePerUpgrade}
                        then:
                            SlotAudio2.Play(Agent)
                            CompleteBoughtAnims(Agent)
                    else:
                        NegativeAudio.Play(Agent)
                else:
                    NegativeAudio.Play(Agent)#>
                        
    UpgradeSlotThree(Wdata:widget_message):void=
        if(Player:=Wdata.Player, PlayerUI:=GetPlayerUI[Player],SelectedButton:=Wdata.Source,Agent:=agent[Player]):   
            if (MyData := Data[Player]):<#
                if(Slot3Max > MyData.SuperCritGemLevel):
                    var Price : float = Slot3Initial
                    for(i:=0..MyData.SuperCritGemLevel-1):
                        set Price = Price * Slot3Multiple

                    if(MyData.Gems >= Price):
                        if:
                            OldData := Data[Player]
                            set Data[Player] = player_data{player_data_update<constructor>(OldData),SuperCritGemLevel := OldData.SuperCritGemLevel + 1, Gems := OldData.Gems - Price,SuperCritDamagePerUpgrade := SuperCritGemDamagePerLevel + OldData.SuperCritDamagePerUpgrade}
                        then:
                            SlotAudio3.Play(Agent)
                            CompleteBoughtAnims(Agent)
                    else:
                        NegativeAudio.Play(Agent)
                else:
                    NegativeAudio.Play(Agent)#>

    UpdateInvGold(Agent:agent)<suspends>:void=
        loop:
            Sleep(0.5)
            if:
                Player := player[Agent]
                Player.IsActive[]
                PlayerData := Data[player[Agent]]
            then:
                if(TextBlocks  := TextBlockPerAgent[Agent]?):
                    if(GoldTextBlock := TextBlocks[5]):
                        GoldTextBlock.SetText(storedstring("{formatter.Abbreviate((PlayerData.Gems * 1.0),false)}"))
                        if(TextBlockPerAgent[Agent] = false):
                            break
                else:
                    break

    UpdateGoldText(Agent:agent):void=
        if(PlayerData := Data[player[Agent]]):
            if(TextBlocks  := TextBlockPerAgent[Agent]?):
                if(GoldTextBlock := TextBlocks[5]):
                    GoldTextBlock.SetText(storedstring("{formatter.Abbreviate((PlayerData.Gems * 1.0),false)}"))

    CompleteBoughtAnims(Agent:agent):void= 
        #PurchaseAudio.Play(Agent)
        UpdateUiText(Agent)
        HeadNodCine.Play(Agent)
        UpdateGoldText(Agent)
        BuyAccolade.Award(Agent)
        BuyingXpUpgradeTrigger.Trigger(Agent)