using { /Fortnite.com/Devices }
using { /Verse.org/Simulation }
using { /UnrealEngine.com/Temporary/Diagnostics }
using { /UnrealEngine.com/Temporary/UI }
using { /UnrealEngine.com/Temporary/SpatialMath }
using { /Fortnite.com/UI }
using { /Verse.org/Colors }
using { /Fortnite.com/Playspaces }
using { /Fortnite.com/Characters }
using { /Verse.org/Assets }

shop_ui2_device := class(creative_device):

    @editable
    BuyAccolade : accolades_device = accolades_device{}


    @editable
    OpenShop : button_device = button_device{}

    #@editable
    #BackgroundHMD : hud_message_device = hud_message_device{}
    @editable
    ShopKeeperCine : cinematic_sequence_device = cinematic_sequence_device{}
    @editable
    HeadNodCine : cinematic_sequence_device = cinematic_sequence_device{}

    @editable
    PurchaseAudio : audio_player_device = audio_player_device{}
    @editable
    NegativeAudio : audio_player_device = audio_player_device{}

    @editable
    OpenAudio : audio_player_device = audio_player_device{}
    @editable
    TalkAudio : audio_player_device = audio_player_device{}
    @editable
    CloseAudio : audio_player_device = audio_player_device{}

    @editable
    SlotAudio1 : audio_player_device = audio_player_device{}
    @editable
    SlotAudio2 : audio_player_device = audio_player_device{}

    @editable
    RequiredLevel : int = 5
    @editable
    LevelHMD : hud_message_device = hud_message_device{}


    @editable
    var SlotsMultiple : float = 1.4
    @editable
    var SlotInitials : float = 10800.0

    MaxUpgrades : int = 15

    var MyUI : [agent]?canvas = map{}
    var TextBlockPerAgent : [agent]?[]text_block = map{}
    var TextBlockPerAgentShadows : [agent]?[]text_block = map{}

    var ButtonLoudPerAgent : [agent]?[]button_loud = map{} # no longer being used

    storedstring<localizes>(S:string) : message = "{S}"
    JustIntString<localizes>(I:int) : message = "{I}"
    LevelString<localizes>(I:int) : message = "Lv {I}"
    DamageString<localizes>(I:int) : message = " {I}%"
    SecondString<localizes>(I:int) : message = "{I} Sec"
    HpString<localizes>(I:int) : message = "{I} hp"


    var StoredButtonTestingToclose : button_loud = button_loud{}

    @editable
    CurrencyUi : currency_ui_device = currency_ui_device{}

    @editable
    MissionTrigger:trigger_device = trigger_device{}

    OnBegin<override>()<suspends>:void=

        OpenShop.InteractedWithEvent.Subscribe(ShowUi)

    
    MakeEntireBackground (Agent:agent):canvas=

        GoldCoinTexture : texture = wooden_pixel_gui.textures.Icons2.Size8.t_CoinGold_8

        TotalUpgradesText: text_block = text_block{DefaultTextColor := MakeColorFromHex("FFFFFF"), DefaultText:=storedstring("Upgrades")}
        UpgradeNumberText: text_block = text_block{DefaultTextColor := MakeColorFromHex("FFFFFF"), DefaultText:=storedstring("0 / 15")}

        Slot1Title: text_block = text_block{DefaultTextColor := MakeColorFromHex("529F17FF"), DefaultText:=storedstring("Focus Damage")}
        Slot1TitleShadow: text_block = text_block{DefaultTextColor := MakeColorFromHex("142706FF"), DefaultText:=storedstring("Focus Damage")}
        Slot1Level: text_block = text_block{DefaultTextColor := MakeColorFromHex("529F17FF"), DefaultText:=LevelString(2)}
        Slot1LevelShadow: text_block = text_block{DefaultTextColor := MakeColorFromHex("142706FF"), DefaultText:=LevelString(2)}
        Slot1Image: texture_block = texture_block{DefaultImage:= RavenIcons.MapMarkers.S234, DefaultDesiredSize:=vector2{X:=100.0,Y:=100.0}}
        Slot1ImageShadow: texture_block = texture_block{DefaultImage:= RavenIcons.MapMarkers.S234,DefaultTint:=MakeColorFromHex("080402FF"), DefaultDesiredSize:=vector2{X:=100.0,Y:=100.0}}
        Slot1Description: text_block = text_block{DefaultTextColor := MakeColorFromHex("FFFFFF"), DefaultText:=storedstring("Extra damage to\n close enemies")}
        Slot1Price: text_block = text_block{DefaultTextColor := MakeColorFromHex("FFFFFF"), DefaultText:=storedstring("X1.25M")} # Needs to be formated
        Slot1Button : button_quiet = button_quiet{DefaultText:=storedstring("")}
        Slot1Button.OnClick().Subscribe(UpgradeSlotOne) 

        Slot2Title: text_block = text_block{DefaultTextColor := MakeColorFromHex("529F17FF"), DefaultText:=storedstring("Accuracy Damage")}
        Slot2TitleShadow: text_block = text_block{DefaultTextColor := MakeColorFromHex("142706FF"), DefaultText:=storedstring("Accuracy Damage")}
        Slot2Level: text_block = text_block{DefaultTextColor := MakeColorFromHex("529F17FF"), DefaultText:=LevelString(2)}
        Slot2LevelShadow: text_block = text_block{DefaultTextColor := MakeColorFromHex("142706FF"), DefaultText:=LevelString(2)}
        Slot2Image: texture_block = texture_block{DefaultImage:= RavenIcons.MapMarkers.SS147, DefaultDesiredSize:=vector2{X:=100.0,Y:=100.0}}
        Slot2ImageShadow: texture_block = texture_block{DefaultImage:= RavenIcons.MapMarkers.SS147,DefaultTint:=MakeColorFromHex("080402FF"), DefaultDesiredSize:=vector2{X:=100.0,Y:=100.0}}
        Slot2Description: text_block = text_block{DefaultTextColor := MakeColorFromHex("FFFFFF"), DefaultText:=storedstring("Extra damage to\n far enemies")}
        Slot2Price: text_block = text_block{DefaultTextColor := MakeColorFromHex("FFFFFF"), DefaultText:=storedstring("X1.25M")} # Needs to be formated
        Slot2Button : button_quiet = button_quiet{DefaultText:=storedstring("")}
        Slot2Button.OnClick().Subscribe(UpgradeSlotTwo) 

        CloseButton : button_quiet = button_quiet{DefaultText:=storedstring("Close")}
        CloseButton.OnClick().Subscribe(RemoveHUD) 

        InventoryGoldText : text_block = text_block:
            DefaultTextColor := MakeColorFromHex("FFFFFF")
            DefaultText:=storedstring("")

        if(set TextBlockPerAgent[Agent] = option{array{Slot1Level,Slot1Price,Slot2Level,Slot2Price,InventoryGoldText,UpgradeNumberText}}){} 
        if(set TextBlockPerAgentShadows[Agent] = option{array{Slot1LevelShadow,Slot2LevelShadow}}){}

        UpdateGoldText(Agent)
        spawn:
            UpdateInvGold(Agent)


        MyCanvas:=canvas:
            Slots:=array:
                canvas_slot:
                    Anchors:=anchors{Minimum:=vector2{X:=0.1,Y:=0.05}, Maximum:=vector2{X:=0.75,Y:=0.95}}
                    Widget:=stack_box:
                        Orientation:=orientation.Vertical
                        Slots:=array:
                            stack_box_slot:
                                HorizontalAlignment:=horizontal_alignment.Fill
                                VerticalAlignment:=vertical_alignment.Fill
                                Widget:=overlay:
                                    Slots:=array:
                                        overlay_slot:
                                            HorizontalAlignment:=horizontal_alignment.Fill
                                            VerticalAlignment:=vertical_alignment.Fill
                                            Widget:=texture_block:
                                                DefaultImage:=wooden_pixel_gui.textures.Panels.t_TitleBar
                                                DefaultDesiredSize:=vector2{X:=50.0,Y:=110.0}
                                        overlay_slot:
                                            HorizontalAlignment:=horizontal_alignment.Center
                                            VerticalAlignment:=vertical_alignment.Center
                                            Widget:=text_block:
                                                DefaultText:=storedstring("Skill Shop")
                                                DefaultTextColor := MakeColorFromHex("401805FF")
                            stack_box_slot:
                                HorizontalAlignment:=horizontal_alignment.Fill
                                VerticalAlignment:=vertical_alignment.Fill
                                Widget:=overlay:
                                    Slots:=array:
                                        overlay_slot:
                                            HorizontalAlignment:=horizontal_alignment.Fill
                                            VerticalAlignment:=vertical_alignment.Fill
                                            Widget:=texture_block:
                                                DefaultImage:=wooden_pixel_gui.textures.Panels.WhiteSquare75
                                                DefaultTint:=MakeColorFromHex("251206BF")
                                                DefaultDesiredSize:=vector2{X:=1000.0,Y:=700.0}
                                        overlay_slot:
                                            HorizontalAlignment:=horizontal_alignment.Center
                                            VerticalAlignment:=vertical_alignment.Top
                                            Widget:=TotalUpgradesText
                                        overlay_slot:
                                            HorizontalAlignment:=horizontal_alignment.Center
                                            VerticalAlignment:=vertical_alignment.Top
                                            Padding:=margin{Left:=0.0,Top:=40.0,Right:=0.0,Bottom:=0.0}
                                            Widget:=UpgradeNumberText
                                        overlay_slot: # Main stackbox with panels in it
                                            HorizontalAlignment:=horizontal_alignment.Center
                                            VerticalAlignment:=vertical_alignment.Center
                                            Widget:=stack_box:
                                                Orientation:=orientation.Horizontal
                                                Slots:=array:
                                                    stack_box_slot:
                                                        HorizontalAlignment:=horizontal_alignment.Fill
                                                        VerticalAlignment:=vertical_alignment.Fill
                                                        Padding:=margin{Left:=50.0,Top:=0.0,Right:=50.0,Bottom:=0.0}
                                                        Widget:=overlay:
                                                            Slots:=array:
                                                                overlay_slot:
                                                                    HorizontalAlignment:=horizontal_alignment.Fill
                                                                    VerticalAlignment:=vertical_alignment.Fill
                                                                    Padding:=margin{Left:=-20.0,Top:=-20.0,Right:=-20.0,Bottom:=-20.0}
                                                                    Widget:=texture_block:
                                                                        DefaultImage:=wooden_pixel_gui.textures.Panels.BrownSquare
                                                                        DefaultTint:= MakeColorFromHex("A3A3A3FF")
                                                                overlay_slot:
                                                                    HorizontalAlignment:=horizontal_alignment.Center
                                                                    VerticalAlignment:=vertical_alignment.Center
                                                                    Widget:=stack_box:
                                                                        Orientation:=orientation.Vertical
                                                                        Slots:=array:
                                                                            stack_box_slot:
                                                                                HorizontalAlignment:=horizontal_alignment.Fill
                                                                                VerticalAlignment:=vertical_alignment.Fill
                                                                                Widget:=overlay:
                                                                                    Slots:=array:
                                                                                        overlay_slot:
                                                                                            HorizontalAlignment:=horizontal_alignment.Center
                                                                                            VerticalAlignment:=vertical_alignment.Center
                                                                                            Padding:=margin{Left:=3.0,Top:=3.0,Right:=0.0,Bottom:=0.0}
                                                                                            Widget:=Slot1TitleShadow
                                                                                        overlay_slot:
                                                                                            HorizontalAlignment:=horizontal_alignment.Center
                                                                                            VerticalAlignment:=vertical_alignment.Center
                                                                                            Widget:=Slot1Title
                                                                            stack_box_slot:
                                                                                HorizontalAlignment:=horizontal_alignment.Fill
                                                                                VerticalAlignment:=vertical_alignment.Fill
                                                                                Widget:=overlay:
                                                                                    Slots:=array:
                                                                                        overlay_slot:
                                                                                            HorizontalAlignment:=horizontal_alignment.Center
                                                                                            VerticalAlignment:=vertical_alignment.Center
                                                                                            Padding:=margin{Left:=3.0,Top:=3.0,Right:=0.0,Bottom:=0.0}
                                                                                            Widget:=Slot1LevelShadow
                                                                                        overlay_slot:
                                                                                            HorizontalAlignment:=horizontal_alignment.Center
                                                                                            VerticalAlignment:=vertical_alignment.Center
                                                                                            Widget:=Slot1Level
                                                                            stack_box_slot:
                                                                                HorizontalAlignment:=horizontal_alignment.Fill
                                                                                VerticalAlignment:=vertical_alignment.Fill
                                                                                Padding:=margin{Left:=0.0,Top:=45.0,Right:=0.0,Bottom:=0.0}
                                                                                Widget:=overlay:
                                                                                    Slots:=array:
                                                                                        overlay_slot:
                                                                                            HorizontalAlignment:=horizontal_alignment.Center
                                                                                            VerticalAlignment:=vertical_alignment.Center
                                                                                            Padding:=margin{Left:=0.0,Top:=6.0,Right:=0.0,Bottom:=0.0}
                                                                                            Widget:=Slot1ImageShadow
                                                                                        overlay_slot:
                                                                                            HorizontalAlignment:=horizontal_alignment.Center
                                                                                            VerticalAlignment:=vertical_alignment.Center
                                                                                            Widget:=Slot1Image
                                                                            stack_box_slot:
                                                                                HorizontalAlignment:=horizontal_alignment.Fill
                                                                                VerticalAlignment:=vertical_alignment.Fill
                                                                                Padding:=margin{Left:=0.0,Top:=25.0,Right:=0.0,Bottom:=0.0}
                                                                                Widget:=Slot1Description
                                                                            stack_box_slot:
                                                                                HorizontalAlignment:=horizontal_alignment.Fill
                                                                                VerticalAlignment:=vertical_alignment.Fill
                                                                                Padding:=margin{Left:=0.0,Top:=35.0,Right:=0.0,Bottom:=0.0}
                                                                                Widget:=overlay:
                                                                                    Slots:=array:
                                                                                        overlay_slot:
                                                                                            HorizontalAlignment:=horizontal_alignment.Fill
                                                                                            VerticalAlignment:=vertical_alignment.Fill
                                                                                            Widget:=texture_block:
                                                                                                DefaultImage:=wooden_pixel_gui.textures.Panels.BrownSquare
                                                                                                DefaultTint:= MakeColorFromHex("0000004D")
                                                                                        overlay_slot:
                                                                                            HorizontalAlignment:=horizontal_alignment.Center
                                                                                            VerticalAlignment:=vertical_alignment.Center
                                                                                            Widget:=stack_box:
                                                                                                Orientation:=orientation.Horizontal
                                                                                                Slots:=array:
                                                                                                    stack_box_slot:
                                                                                                        HorizontalAlignment:=horizontal_alignment.Center
                                                                                                        VerticalAlignment:=vertical_alignment.Center
                                                                                                        Widget:=texture_block:
                                                                                                            DefaultImage:=wooden_pixel_gui.textures.Icons2.Size8.t_CoinGold_8
                                                                                                    stack_box_slot:
                                                                                                        Widget:=Slot1Price
                                                                                        overlay_slot:
                                                                                            HorizontalAlignment:=horizontal_alignment.Fill
                                                                                            VerticalAlignment:=vertical_alignment.Fill
                                                                                            Widget:=overlay:
                                                                                                Slots:=array:
                                                                                                    overlay_slot:
                                                                                                        HorizontalAlignment:=horizontal_alignment.Fill
                                                                                                        VerticalAlignment:=vertical_alignment.Fill
                                                                                                        Widget:=Slot1Button
                                                    stack_box_slot:
                                                        HorizontalAlignment:=horizontal_alignment.Fill
                                                        VerticalAlignment:=vertical_alignment.Fill
                                                        Padding:=margin{Left:=50.0,Top:=0.0,Right:=50.0,Bottom:=0.0}
                                                        Widget:=overlay:
                                                            Slots:=array:
                                                                overlay_slot:
                                                                    HorizontalAlignment:=horizontal_alignment.Fill
                                                                    VerticalAlignment:=vertical_alignment.Fill
                                                                    Padding:=margin{Left:=-20.0,Top:=-20.0,Right:=-20.0,Bottom:=-20.0}
                                                                    Widget:=texture_block:
                                                                        DefaultImage:=wooden_pixel_gui.textures.Panels.BrownSquare
                                                                        DefaultTint:= MakeColorFromHex("A3A3A3FF")
                                                                overlay_slot:
                                                                    HorizontalAlignment:=horizontal_alignment.Center
                                                                    VerticalAlignment:=vertical_alignment.Center
                                                                    Widget:=stack_box:
                                                                        Orientation:=orientation.Vertical
                                                                        Slots:=array:
                                                                            stack_box_slot:
                                                                                HorizontalAlignment:=horizontal_alignment.Fill
                                                                                VerticalAlignment:=vertical_alignment.Fill
                                                                                Widget:=overlay:
                                                                                    Slots:=array:
                                                                                        overlay_slot:
                                                                                            HorizontalAlignment:=horizontal_alignment.Center
                                                                                            VerticalAlignment:=vertical_alignment.Center
                                                                                            Padding:=margin{Left:=3.0,Top:=3.0,Right:=0.0,Bottom:=0.0}
                                                                                            Widget:=Slot2TitleShadow
                                                                                        overlay_slot:
                                                                                            HorizontalAlignment:=horizontal_alignment.Center
                                                                                            VerticalAlignment:=vertical_alignment.Center
                                                                                            Widget:=Slot2Title
                                                                            stack_box_slot:
                                                                                HorizontalAlignment:=horizontal_alignment.Fill
                                                                                VerticalAlignment:=vertical_alignment.Fill
                                                                                Widget:=overlay:
                                                                                    Slots:=array:
                                                                                        overlay_slot:
                                                                                            HorizontalAlignment:=horizontal_alignment.Center
                                                                                            VerticalAlignment:=vertical_alignment.Center
                                                                                            Padding:=margin{Left:=3.0,Top:=3.0,Right:=0.0,Bottom:=0.0}
                                                                                            Widget:=Slot2LevelShadow
                                                                                        overlay_slot:
                                                                                            HorizontalAlignment:=horizontal_alignment.Center
                                                                                            VerticalAlignment:=vertical_alignment.Center
                                                                                            Widget:=Slot2Level
                                                                            stack_box_slot:
                                                                                HorizontalAlignment:=horizontal_alignment.Fill
                                                                                VerticalAlignment:=vertical_alignment.Fill
                                                                                Padding:=margin{Left:=0.0,Top:=45.0,Right:=0.0,Bottom:=0.0}
                                                                                Widget:=overlay:
                                                                                    Slots:=array:
                                                                                        overlay_slot:
                                                                                            HorizontalAlignment:=horizontal_alignment.Center
                                                                                            VerticalAlignment:=vertical_alignment.Center
                                                                                            Padding:=margin{Left:=0.0,Top:=6.0,Right:=0.0,Bottom:=0.0}
                                                                                            Widget:=Slot2ImageShadow
                                                                                        overlay_slot:
                                                                                            HorizontalAlignment:=horizontal_alignment.Center
                                                                                            VerticalAlignment:=vertical_alignment.Center
                                                                                            Widget:=Slot2Image
                                                                            stack_box_slot:
                                                                                HorizontalAlignment:=horizontal_alignment.Fill
                                                                                VerticalAlignment:=vertical_alignment.Fill
                                                                                Padding:=margin{Left:=0.0,Top:=25.0,Right:=0.0,Bottom:=0.0}
                                                                                Widget:=Slot2Description
                                                                            stack_box_slot:
                                                                                HorizontalAlignment:=horizontal_alignment.Fill
                                                                                VerticalAlignment:=vertical_alignment.Fill
                                                                                Padding:=margin{Left:=0.0,Top:=35.0,Right:=0.0,Bottom:=0.0}
                                                                                Widget:=overlay:
                                                                                    Slots:=array:
                                                                                        overlay_slot:
                                                                                            HorizontalAlignment:=horizontal_alignment.Fill
                                                                                            VerticalAlignment:=vertical_alignment.Fill
                                                                                            Widget:=texture_block:
                                                                                                DefaultImage:=wooden_pixel_gui.textures.Panels.BrownSquare
                                                                                                DefaultTint:= MakeColorFromHex("0000004D")           
                                                                                        overlay_slot:
                                                                                            HorizontalAlignment:=horizontal_alignment.Center
                                                                                            VerticalAlignment:=vertical_alignment.Center
                                                                                            Widget:=stack_box:
                                                                                                Orientation:=orientation.Horizontal
                                                                                                Slots:=array:
                                                                                                    stack_box_slot:
                                                                                                        HorizontalAlignment:=horizontal_alignment.Center
                                                                                                        VerticalAlignment:=vertical_alignment.Center
                                                                                                        Widget:=texture_block:
                                                                                                            DefaultImage:=wooden_pixel_gui.textures.Icons2.Size8.t_CoinGold_8
                                                                                                    stack_box_slot:
                                                                                                        Widget:=Slot2Price
                                                                                        overlay_slot:
                                                                                            HorizontalAlignment:=horizontal_alignment.Fill
                                                                                            VerticalAlignment:=vertical_alignment.Fill
                                                                                            Widget:=overlay:
                                                                                                Slots:=array:
                                                                                                    
                                                                                                    overlay_slot:
                                                                                                        HorizontalAlignment:=horizontal_alignment.Fill
                                                                                                        VerticalAlignment:=vertical_alignment.Fill
                                                                                                        Widget:=Slot2Button
                                                   
                                        overlay_slot:
                                            HorizontalAlignment:=horizontal_alignment.Center
                                            VerticalAlignment:=vertical_alignment.Bottom
                                            Padding:=margin{Left:=0.0,Top:=0.0,Right:=0.0,Bottom:=25.0}
                                            Widget:=CloseButton
                                        overlay_slot:
                                            HorizontalAlignment:=horizontal_alignment.Left
                                            VerticalAlignment:=vertical_alignment.Bottom
                                            Padding:=margin{Left:=20.0,Top:=0.0,Right:=0.0,Bottom:=20.0}
                                            Widget:=overlay:
                                                Slots:=array:
                                                    overlay_slot:
                                                        HorizontalAlignment:=horizontal_alignment.Left
                                                        VerticalAlignment:=vertical_alignment.Top
                                                        Widget:=stack_box:
                                                            Orientation:=orientation.Horizontal
                                                            Slots:=array:
                                                                stack_box_slot:
                                                                    HorizontalAlignment:=horizontal_alignment.Fill
                                                                    VerticalAlignment:=vertical_alignment.Fill
                                                                    Widget:=texture_block:
                                                                        DefaultImage:=GoldCoinTexture
                                                                        DefaultDesiredSize:=vector2{X:=41.0,Y:=32.0}
                                                                stack_box_slot:
                                                                    HorizontalAlignment:=horizontal_alignment.Left
                                                                    VerticalAlignment:=vertical_alignment.Center
                                                                    Padding:=margin{Left:=10.0,Top:=0.0,Right:=0.0,Bottom:=0.0}
                                                                    Widget:=InventoryGoldText

                                                                                       

        return MyCanvas

    ShowUi(Agent:agent):void=

        if:
            MyData := Data[player[Agent]]
            MyData.Level >= RequiredLevel
        then:
            TalkAudio.Play(Agent)
            OpenAudio.Play(Agent)
            RemoveFortHud(Agent)
            ShopKeeperCine.Play(Agent)


            MyCanvas: canvas = MakeEntireBackground(Agent)
            
            if(PlayerUI:=GetPlayerUI[player[Agent]]):
                PlayerUI.AddWidget(MyCanvas,player_ui_slot{InputMode:=ui_input_mode.All,ZOrder:=0}) # want it to appear on top
                UpdateUiText(Agent)
                MyCanvas.SetVisibility(widget_visibility.Hidden)
                MyCanvas.SetVisibility(widget_visibility.Visible)
                CurrencyUi.HideCurrencyUI(Agent)
            if(set MyUI[Agent]= option{MyCanvas}){}
        else:
            LevelHMD.SetText(storedstring("Level {RequiredLevel} Req"))
            LevelHMD.Show(Agent)
            NegativeAudio.Play(Agent)


    RemoveHUD(Wdata:widget_message):void =

        Print("Removing")
        if(Player:=Wdata.Player, PlayerUI:=GetPlayerUI[Player],SelectedButton:=Wdata.Source,Agent:=agent[Player]):
            if(MyActualUI := MyUI[Agent]?):

                
                CloseAudio.Play(Agent)
                PlayerUI.RemoveWidget(MyActualUI)
                MyActualUI.SetVisibility(widget_visibility.Visible)
                MyActualUI.SetVisibility(widget_visibility.Hidden)

                CurrencyUi.ShowCurrencyUI(Agent)

                if(set MyUI[Agent] = false){}
                if(set TextBlockPerAgent[Agent] = false){}
                
                spawn:
                    CloseCineDelay(Agent)
    
    RemoveHUDInputTrigger(Agent:agent):void =

        if:
            MyActualUI := MyUI[Agent]?
            PlayerUI:=GetPlayerUI[player[Agent]]

        then:

            CloseAudio.Play(Agent)
            PlayerUI.RemoveWidget(MyActualUI)
            MyActualUI.SetVisibility(widget_visibility.Visible)
            MyActualUI.SetVisibility(widget_visibility.Hidden)

            CurrencyUi.ShowCurrencyUI(Agent)

            if(set MyUI[Agent] = false){}
            if(set TextBlockPerAgent[Agent] = false){}
                
            spawn:
                CloseCineDelay(Agent)
                
    CloseCineDelay(Agent:agent)<suspends>:void=

        ShowFortHud(Agent)
        ShopKeeperCine.Stop(Agent)

    UpdateUiText(Agent:agent):void=

        if:
            Slot1Level: text_block = TextBlockPerAgent[Agent]?[0]
            Slot1Price: text_block = TextBlockPerAgent[Agent]?[1]
            Slot2Level: text_block = TextBlockPerAgent[Agent]?[2]
            Slot2Price: text_block = TextBlockPerAgent[Agent]?[3]

            Slot1LevelShadow: text_block = TextBlockPerAgentShadows[Agent]?[0]
            Slot2LevelShadow: text_block = TextBlockPerAgentShadows[Agent]?[1]

            Player := player[Agent]
            PlayerData := Data[Agent]

            UpgradeNumberText : text_block = TextBlockPerAgent[Agent]?[5]
            
        then:

            Slot1Level.SetText(LevelString(PlayerData.FocusDamage))
            Slot1LevelShadow.SetText(LevelString(PlayerData.FocusDamage))

            Slot2LevelShadow.SetText(LevelString(PlayerData.AccuracyDamage))
            Slot2Level.SetText(LevelString(PlayerData.AccuracyDamage))

            

            if(PlayerData.AccuracyDamage + PlayerData.FocusDamage < MaxUpgrades):
                
                var SamePrice : float = SlotInitials

                for(i:=0..PlayerData.AccuracyDamage + PlayerData.FocusDamage -1):
                    set SamePrice = SamePrice * SlotsMultiple

                if(PriceR := Round[SamePrice]):
                    Slot2Price.SetText(storedstring(formatter.Abbreviate((PriceR * 1.0),false)))
                    Slot1Price.SetText(storedstring(formatter.Abbreviate((PriceR * 1.0),false)))
            else:
                Slot1Price.SetText(storedstring("Max"))
                Slot2Price.SetText(storedstring("Max"))

            UpgradeNumberText.SetText(storedstring("{PlayerData.AccuracyDamage + PlayerData.FocusDamage} / {MaxUpgrades}"))



 

    UpgradeSlotOne(Wdata:widget_message):void=
        Print("Upgrade Focus button")
        if(Player:=Wdata.Player, PlayerUI:=GetPlayerUI[Player],SelectedButton:=Wdata.Source,Agent:=agent[Player]):
            if (MyData := Data[Player]):
                if (MyData.FocusDamage + MyData.AccuracyDamage < MaxUpgrades):
                    var Price : float = SlotInitials
                    for(i:=0..(MyData.FocusDamage + MyData.AccuracyDamage)-1):
                        set Price = Price * SlotsMultiple

                    if(MyData.Gold >= Price):
                        if:
                            OldData := Data[Player]
                            set Data[Player] = player_data{player_data_update<constructor>(OldData),FocusDamage := OldData.FocusDamage + 1, Gold := OldData.Gold - Price}
                        then:
                            SlotAudio1.Play(Agent)
                            CompleteBoughtAnims(Agent)
                    else:
                        Print("Not Enought Units")
                        NegativeAudio.Play(Agent)
                else:
                    NegativeAudio.Play(Agent)

    UpgradeSlotTwo(Wdata:widget_message):void=
        Print("upgrade Accuracy Button")
        if(Player:=Wdata.Player, PlayerUI:=GetPlayerUI[Player],SelectedButton:=Wdata.Source,Agent:=agent[Player]):   
            if (MyData := Data[Player]):
                if (MyData.FocusDamage + MyData.AccuracyDamage < MaxUpgrades):
                    var Price : float = SlotInitials
                    for(i:=0..MyData.AccuracyDamage + MyData.FocusDamage  -1):
                        set Price = Price * SlotsMultiple
                    if(MyData.Gold >= Price):
                        if:
                            OldData := Data[Player]
                            set Data[Player] = player_data{player_data_update<constructor>(OldData),AccuracyDamage := OldData.AccuracyDamage + 1, Gold := OldData.Gold - Price}
                        then:
                            SlotAudio2.Play(Agent)
                            CompleteBoughtAnims(Agent)
                    else:
                        Print("Not Enought Units")
                        NegativeAudio.Play(Agent)
                else:
                    NegativeAudio.Play(Agent)

    UpdateInvGold(Agent:agent)<suspends>:void=

        loop:
            Sleep(0.5)
            if:
                Player := player[Agent]
                Player.IsActive[]
                PlayerData := Data[player[Agent]]
            then:
                if(TextBlocks  := TextBlockPerAgent[Agent]?):
                    if(GoldTextBlock := TextBlocks[4]):
                        GoldTextBlock.SetText(storedstring("{formatter.Abbreviate((PlayerData.Gold * 1.0),false)}"))
                        if(TextBlockPerAgent[Agent] = false):
                            break
                else:
                    break

    UpdateGoldText(Agent:agent):void=
        if(PlayerData := Data[player[Agent]]):
            if(TextBlocks  := TextBlockPerAgent[Agent]?):
                if(GoldTextBlock := TextBlocks[4]):
                    GoldTextBlock.SetText(storedstring("{formatter.Abbreviate((PlayerData.Gold * 1.0),false)}"))


    CompleteBoughtAnims(Agent:agent):void= 
        #PurchaseAudio.Play(Agent)
        UpdateUiText(Agent)
        HeadNodCine.Play(Agent)
        UpdateGoldText(Agent)
        BuyAccolade.Award(Agent)
        MissionTrigger.Trigger(Agent)
