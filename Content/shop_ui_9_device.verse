
using { /Fortnite.com/Devices }
using { /Verse.org/Simulation }
using { /UnrealEngine.com/Temporary/Diagnostics }
using { /UnrealEngine.com/Temporary/UI }
using { /UnrealEngine.com/Temporary/SpatialMath }
using { /Fortnite.com/UI }
using { /Verse.org/Colors }
using { /Fortnite.com/Playspaces }
using { /Fortnite.com/Characters }
using { /Verse.org/Assets }

shop_ui_9_device := class(creative_device):

    @editable
    BuyAccolade : accolades_device = accolades_device{}


    @editable
    OpenShop : button_device = button_device{}
    @editable
    ShopKeeperCine : cinematic_sequence_device = cinematic_sequence_device{}
    @editable
    HeadNodCine : cinematic_sequence_device = cinematic_sequence_device{}

    @editable
    PurchaseAudio : audio_player_device = audio_player_device{}
    @editable
    NegativeAudio : audio_player_device = audio_player_device{}

    @editable
    OpenAudio : audio_player_device = audio_player_device{}
    @editable
    CloseAudio : audio_player_device = audio_player_device{}

    @editable
    SlotAudio1 : audio_player_device = audio_player_device{}

    # Just to pull % multtipliers
    @editable
    TowerManager : tower_manager_device = tower_manager_device{}

    @editable
    var Slot1Multiple : float = 1.45
    @editable
    var Slot1Initial : float = 1000.0

    Slot1Max : int = 1000

    var MyUI : [agent]?canvas = map{}
    var TextBlockPerAgent : [agent]?[]text_block = map{}
    var TextBlockPerAgentShadows : [agent]?[]text_block = map{}

    var ButtonLoudPerAgent : [agent]?[]button_loud = map{} # no longer being used

    storedstring<localizes>(S:string) : message = "{S}"
    JustIntString<localizes>(I:int) : message = "{I}"
    LevelString<localizes>(I:int) : message = "Lv {I}"
    DamageString<localizes>(I:int) : message = " {I}%"
    SecondString<localizes>(I:int) : message = "{I} Sec"
    HpString<localizes>(I:int) : message = "{I} hp"

    var StoredButtonTestingToclose : button_loud = button_loud{}

    @editable
    CurrencyUi : currency_ui_device = currency_ui_device{}

    OnBegin<override>()<suspends>:void=

        OpenShop.InteractedWithEvent.Subscribe(ShowUi)
    
    MakeEntireBackground (Agent:agent):canvas=

        GemTexture : texture = wooden_pixel_gui.textures.Icons2.Size8.t_GemRed_8

        ChestTexture : texture = wooden_pixel_gui.textures.Icons2.Size32.t_Chest01Open_32
        CritTexture : texture = wooden_pixel_gui.textures.Icons2.Size32.t_Shovel_32
        GoldTexture : texture = wooden_pixel_gui.textures.Icons2.Size32.t_ScrollRoll_32

        Slot1Title: text_block = text_block{DefaultTextColor := MakeColorFromHex("529F17FF"), DefaultText:=storedstring("Upgrade Tower")}
        Slot1TitleShadow: text_block = text_block{DefaultTextColor := MakeColorFromHex("142706FF"), DefaultText:=storedstring("Upgrade Tower")}
        Slot1Level: text_block = text_block{DefaultTextColor := MakeColorFromHex("529F17FF"), DefaultText:=LevelString(2)}
        Slot1LevelShadow: text_block = text_block{DefaultTextColor := MakeColorFromHex("142706FF"), DefaultText:=LevelString(2)}
        Slot1Image: texture_block = texture_block{DefaultImage:= ChestTexture, DefaultDesiredSize:=vector2{X:=100.0,Y:=100.0}}
        Slot1ImageShadow: texture_block = texture_block{DefaultImage:= ChestTexture,DefaultTint:=MakeColorFromHex("080402FF"), DefaultDesiredSize:=vector2{X:=100.0,Y:=100.0}}
        Slot1Description: text_block = text_block{DefaultTextColor := MakeColorFromHex("FFFFFF"), DefaultText:=storedstring(" Tower generates \n gold per second")}
        Slot1Price: text_block = text_block{DefaultTextColor := MakeColorFromHex("FFFFFF"), DefaultText:=storedstring("X1.25M")} # Needs to be formated
        Slot1Button : button_quiet = button_quiet{DefaultText:=storedstring("")}
        Slot1Button.OnClick().Subscribe(UpgradeSlotOne) 

        CloseButton : button_quiet = button_quiet{DefaultText:=storedstring("Close")}
        CloseButton.OnClick().Subscribe(RemoveHUD) 

        if(set TextBlockPerAgent[Agent] = option{array{Slot1Level,Slot1Price}}){} 
        if(set TextBlockPerAgentShadows[Agent] = option{array{Slot1LevelShadow}}){} 

        MyCanvas:=canvas:
            Slots:=array:
                canvas_slot:
                    Anchors:=anchors{Minimum:=vector2{X:=0.1,Y:=0.05}, Maximum:=vector2{X:=0.75,Y:=0.95}}
                    Widget:=stack_box:
                        Orientation:=orientation.Vertical
                        Slots:=array:
                            stack_box_slot:
                                HorizontalAlignment:=horizontal_alignment.Fill
                                VerticalAlignment:=vertical_alignment.Fill
                                Widget:=overlay:
                                    Slots:=array:
                                        overlay_slot:
                                            HorizontalAlignment:=horizontal_alignment.Fill
                                            VerticalAlignment:=vertical_alignment.Fill
                                            Widget:=texture_block:
                                                DefaultImage:=wooden_pixel_gui.textures.Panels.t_TitleBar
                                                DefaultDesiredSize:=vector2{X:=50.0,Y:=110.0}
                                        overlay_slot:
                                            HorizontalAlignment:=horizontal_alignment.Center
                                            VerticalAlignment:=vertical_alignment.Center
                                            Widget:=text_block:
                                                DefaultText:=storedstring("Tower Shop")
                                                DefaultTextColor := MakeColorFromHex("401805FF")
                            stack_box_slot:
                                HorizontalAlignment:=horizontal_alignment.Fill
                                VerticalAlignment:=vertical_alignment.Fill
                                Widget:=overlay:
                                    Slots:=array:
                                        overlay_slot:
                                            HorizontalAlignment:=horizontal_alignment.Fill
                                            VerticalAlignment:=vertical_alignment.Fill
                                            Widget:=texture_block:
                                                DefaultImage:=wooden_pixel_gui.textures.Panels.WhiteSquare75
                                                DefaultTint:=MakeColorFromHex("251206BF")
                                                DefaultDesiredSize:=vector2{X:=1000.0,Y:=700.0}
                                        overlay_slot: # Main stackbox with panels in it
                                            HorizontalAlignment:=horizontal_alignment.Center
                                            VerticalAlignment:=vertical_alignment.Center
                                            Widget:=stack_box:
                                                Orientation:=orientation.Horizontal
                                                Slots:=array:
                                                    stack_box_slot:
                                                        HorizontalAlignment:=horizontal_alignment.Fill
                                                        VerticalAlignment:=vertical_alignment.Fill
                                                        Padding:=margin{Left:=50.0,Top:=0.0,Right:=50.0,Bottom:=0.0}
                                                        Widget:=overlay:
                                                            Slots:=array:
                                                                overlay_slot:
                                                                    HorizontalAlignment:=horizontal_alignment.Fill
                                                                    VerticalAlignment:=vertical_alignment.Fill
                                                                    Padding:=margin{Left:=-20.0,Top:=-20.0,Right:=-20.0,Bottom:=-20.0}
                                                                    Widget:=texture_block:
                                                                        DefaultImage:=wooden_pixel_gui.textures.Panels.BrownSquare
                                                                        DefaultTint:= MakeColorFromHex("A3A3A3FF")
                                                                overlay_slot:
                                                                    HorizontalAlignment:=horizontal_alignment.Center
                                                                    VerticalAlignment:=vertical_alignment.Center
                                                                    Widget:=stack_box:
                                                                        Orientation:=orientation.Vertical
                                                                        Slots:=array:
                                                                            stack_box_slot:
                                                                                HorizontalAlignment:=horizontal_alignment.Fill
                                                                                VerticalAlignment:=vertical_alignment.Fill
                                                                                Widget:=overlay:
                                                                                    Slots:=array:
                                                                                        overlay_slot:
                                                                                            HorizontalAlignment:=horizontal_alignment.Center
                                                                                            VerticalAlignment:=vertical_alignment.Center
                                                                                            Padding:=margin{Left:=3.0,Top:=3.0,Right:=0.0,Bottom:=0.0}
                                                                                            Widget:=Slot1TitleShadow
                                                                                        overlay_slot:
                                                                                            HorizontalAlignment:=horizontal_alignment.Center
                                                                                            VerticalAlignment:=vertical_alignment.Center
                                                                                            Widget:=Slot1Title
                                                                            stack_box_slot:
                                                                                HorizontalAlignment:=horizontal_alignment.Fill
                                                                                VerticalAlignment:=vertical_alignment.Fill
                                                                                Padding:=margin{Left:=0.0,Top:=45.0,Right:=0.0,Bottom:=0.0}
                                                                                Widget:=overlay:
                                                                                    Slots:=array:
                                                                                        overlay_slot:
                                                                                            HorizontalAlignment:=horizontal_alignment.Center
                                                                                            VerticalAlignment:=vertical_alignment.Center
                                                                                            Padding:=margin{Left:=0.0,Top:=6.0,Right:=0.0,Bottom:=0.0}
                                                                                            Widget:=Slot1ImageShadow
                                                                                        overlay_slot:
                                                                                            HorizontalAlignment:=horizontal_alignment.Center
                                                                                            VerticalAlignment:=vertical_alignment.Center
                                                                                            Widget:=Slot1Image
                                                                            stack_box_slot:
                                                                                HorizontalAlignment:=horizontal_alignment.Fill
                                                                                VerticalAlignment:=vertical_alignment.Fill
                                                                                Padding:=margin{Left:=0.0,Top:=25.0,Right:=0.0,Bottom:=0.0}
                                                                                Widget:=Slot1Description
                                                                            stack_box_slot:
                                                                                HorizontalAlignment:=horizontal_alignment.Fill
                                                                                VerticalAlignment:=vertical_alignment.Fill
                                                                                Padding:=margin{Left:=0.0,Top:=35.0,Right:=0.0,Bottom:=0.0}
                                                                                Widget:=overlay:
                                                                                    Slots:=array:
                                                                                        overlay_slot:
                                                                                            HorizontalAlignment:=horizontal_alignment.Fill
                                                                                            VerticalAlignment:=vertical_alignment.Fill
                                                                                            Widget:=texture_block:
                                                                                                DefaultImage:=wooden_pixel_gui.textures.Panels.BrownSquare
                                                                                                DefaultTint:= MakeColorFromHex("0000004D")
                                                                                        overlay_slot:
                                                                                            HorizontalAlignment:=horizontal_alignment.Center
                                                                                            VerticalAlignment:=vertical_alignment.Center
                                                                                            Widget:=stack_box:
                                                                                                Orientation:=orientation.Horizontal
                                                                                                Slots:=array:
                                                                                                    stack_box_slot:
                                                                                                        HorizontalAlignment:=horizontal_alignment.Center
                                                                                                        VerticalAlignment:=vertical_alignment.Center
                                                                                                        Widget:=texture_block:
                                                                                                            DefaultImage:=GemTexture
                                                                                                    stack_box_slot:
                                                                                                        Widget:=Slot1Price
                                                                                        overlay_slot:
                                                                                            HorizontalAlignment:=horizontal_alignment.Fill
                                                                                            VerticalAlignment:=vertical_alignment.Fill
                                                                                            Widget:=overlay:
                                                                                                Slots:=array:
                                                                                                    overlay_slot:
                                                                                                        HorizontalAlignment:=horizontal_alignment.Fill
                                                                                                        VerticalAlignment:=vertical_alignment.Fill
                                                                                                        Widget:=Slot1Button

                                        overlay_slot:
                                            HorizontalAlignment:=horizontal_alignment.Center
                                            VerticalAlignment:=vertical_alignment.Bottom
                                            Padding:=margin{Left:=0.0,Top:=0.0,Right:=0.0,Bottom:=25.0}
                                            Widget:=CloseButton

                                                                                       

        return MyCanvas

    ShowUi(Agent:agent):void=

        RemoveFortHud(Agent)
        ShopKeeperCine.Play(Agent)
        OpenAudio.Play(Agent)

        MyCanvas: canvas = MakeEntireBackground(Agent)
        
        if(PlayerUI:=GetPlayerUI[player[Agent]]):
            PlayerUI.AddWidget(MyCanvas,player_ui_slot{InputMode:=ui_input_mode.All}) # want it to appear on top
            UpdateUiText(Agent)
            MyCanvas.SetVisibility(widget_visibility.Hidden)
            MyCanvas.SetVisibility(widget_visibility.Visible)
            CurrencyUi.HideCurrencyUI(Agent)
        if(set MyUI[Agent]= option{MyCanvas}){}

    RemoveHUD(Wdata:widget_message):void =

        Print("Removing")
        if(Player:=Wdata.Player, PlayerUI:=GetPlayerUI[Player],SelectedButton:=Wdata.Source,Agent:=agent[Player]):
            if(MyActualUI := MyUI[Agent]?):

                CloseAudio.Play(Agent)

                PlayerUI.RemoveWidget(MyActualUI)
                MyActualUI.SetVisibility(widget_visibility.Visible)
                MyActualUI.SetVisibility(widget_visibility.Hidden)
                CurrencyUi.ShowCurrencyUI(Agent)

                if(set MyUI[Agent] = false){}
                if(set TextBlockPerAgent[Agent] = false){}
                
                spawn:
                    CloseCineDelay(Agent)
                
    CloseCineDelay(Agent:agent)<suspends>:void=

        ShowFortHud(Agent)
        ShopKeeperCine.Stop(Agent)

    UpdateUiText(Agent:agent):void=

        if:
            Slot1Level: text_block = TextBlockPerAgent[Agent]?[0]
            Slot1Price: text_block = TextBlockPerAgent[Agent]?[1]

            Slot1LevelShadow: text_block = TextBlockPerAgentShadows[Agent]?[0]

            Player := player[Agent]
            PlayerData := Data[Agent]
            
        then:

            Slot1Level.SetText(LevelString(PlayerData.TowerLevel))
            Slot1LevelShadow.SetText(LevelString(PlayerData.TowerLevel))

            var Price : float = Slot1Initial
            if(PlayerData.TowerLevel > 0):
                for(i:=0..PlayerData.TowerLevel-1):
                    set Price = Price * Slot1Multiple
            
            if(Slot1Max > PlayerData.TowerLevel or Slot1Max = 0):
                Slot1Price.SetText(storedstring(formatter.Abbreviate((Price * 1.0),false)))
            else:
                Slot1Price.SetText(storedstring("Max"))
 

    UpgradeSlotOne(Wdata:widget_message):void=
        if(Player:=Wdata.Player, PlayerUI:=GetPlayerUI[Player],SelectedButton:=Wdata.Source,Agent:=agent[Player]):
            if (MyData := Data[Player]):
                var Price : float = Slot1Initial
                for(i:=0..MyData.TowerLevel-1):
                    set Price = Price * Slot1Multiple

                if(MyData.Gems >= Price):
                    if:
                        OldData := Data[Player]
                        set Data[Player] = player_data{player_data_update<constructor>(OldData),TowerLevel := OldData.TowerLevel + 1, Gems := OldData.Gems - Price}
                            
                    then:
                        TowerManager.LoadTowerLevel(Agent)
                        SlotAudio1.Play(Agent)
                        CompleteBoughtAnims(Agent)
                        BuyAccolade.Award(Agent)
                else:
                    NegativeAudio.Play(Agent)

    CompleteBoughtAnims(Agent:agent):void= 
        #PurchaseAudio.Play(Agent)
        UpdateUiText(Agent)
        HeadNodCine.Play(Agent)

