

using { /Verse.org/Simulation/Tags }


CustomButtonDeviceTag := class(tag){}
CustomButtonKeyDeviceTag := class(tag){}

NotEnoughHudPopupTag := class(tag){}

DiamondDeviceTag := class(tag){}
WildlifeDeviceTag := class(tag){}
CreatureDeviceTag := class(tag){}

ButtonsBoughtTag := class(tag){}

KeyCountingDeviceForButtonsTag := class(tag){}
accoladebutton := class(tag){}
Spawned_Fire_Vfx_Tag := class(tag){}


InventoryCheckConditionalTag := class(tag){} # Used by chests to check if player inventory is full

PistolConditionalTag := class(tag){}
ShotgunConditionalTag := class(tag){}
AssaultConditionalTag := class(tag){}
SmgConditionalTag := class(tag){}
SniperConditionalTag := class(tag){}

StartingWeaponConditionalTag := class(tag){}
CommonConditionalTag := class(tag){}
UncommonConditionalTag := class(tag){}
RareConditionalTag := class(tag){}
EpicConditionalTag := class(tag){}
LegendaryConditionalTag := class(tag){}
ExoticConditionalTag := class(tag){}