
using { /Fortnite.com/Devices }
using { /Verse.org/Simulation }
using { /UnrealEngine.com/Temporary/UI }
using { /UnrealEngine.com/Temporary/SpatialMath }
using { /Fortnite.com/UI }
using { /Verse.org/Colors }
using { /Fortnite.com/Playspaces }
using { /Fortnite.com/Characters }

Icons := module:
    (Icons:)Screenshots<public>:=module:
    (Icons:)CosmeticIcons<public> := module:

teleport_ui_device := class(creative_device):

    @editable
    PlayerSpawners : []player_spawner_device = array{} # used to set teleportin to false

    @editable
    SignalRemote : signal_remote_manager_device = signal_remote_manager_device{}

    @editable
    ButtonClickSound : audio_player_device = audio_player_device{}
    @editable
    TeleportNiagProp : creative_prop_asset = DefaultCreativePropAsset
    @editable
    Teleporters : []teleporter_device = array{}
    @editable
    InteractedButtons : []button_device = array{}

    @editable
    TeleportAudio : audio_player_device = audio_player_device{}

    @editable
    TeleportMissionTrigger : trigger_device = trigger_device{}

    @editable
    OpenAudio : audio_player_device = audio_player_device{}
    @editable
    CloseAudio : audio_player_device = audio_player_device{}
    @editable
    TabAudio : audio_player_device = audio_player_device{}

    var MyUI : [agent]?canvas = map{}

    var TextBlockPerAgent : [agent]?[]text_block = map{}
    var TextBlock2PerAgent : [agent]?[]text_block = map{}
    var TextBlock3PerAgent : [agent]?[]text_block = map{}

    var LevelTextPerAgent : [agent]?[]text_block = map{}

    var TeleportingAlready : [agent]logic = map{}

    var TextureBlockPerAgent : [agent]?[]texture_block = map{}

    var ButtonLoudPerAgent : [agent]?[]button_quiet = map{}
    var TabPerAgent : [agent]?[]texture_block = map{}

    storedstring<localizes>(S:string) : message = "{S}"
    JustIntString<localizes>(I:int) : message = "{I}"
    LevelString<localizes>(I:int) : message = "Lv{I}"
    DamageString<localizes>(I:int) : message = " {I}%"
    SecondString<localizes>(I:int) : message = "{I} Sec"
    HpString<localizes>(I:int) : message = "{I} hp"

    var CurrentTabAgents : [agent]?int = map{}

    OnBegin<override>()<suspends>:void=

        SignalRemote.SecondarySignalEvent.Subscribe(OpenShopHandle)

        for(Button:InteractedButtons):
            Button.InteractedWithEvent.Subscribe(OpenShopHandle)

        AllPlayers := GetPlayspace().GetPlayers()
        for(Player:AllPlayers):
            if(Agent:=agent[Player]):
                InitAlreadyTeleporting(Agent)

        for(Spawners : PlayerSpawners):
            Spawners.SpawnedEvent.Subscribe(InitAlreadyTeleporting)

        GetPlayspace().PlayerAddedEvent().Subscribe(InitAlreadyTeleporting)

        

    InitAlreadyTeleporting(Agent:agent):void=
        if(set TeleportingAlready[Agent] = false){}

    MakeEntireBackground (Agent:agent):canvas=

        BrownBackgroundTexture := texture_block :
            DefaultImage:=wooden_pixel_gui.textures.Panels.BrownSquare
            DefaultDesiredSize:=vector2{X:=32.0,Y:=32.0}
            DefaultTint:=MakeColorFromHex("FFFFFFBF")

        TitleBarTexture := texture_block :
            DefaultDesiredSize:=vector2{X:=16.0,Y:=100.0}
            DefaultImage:=wooden_pixel_gui.textures.Panels.t_TitleBar

        UtilitiesTabButton := button_quiet : 
            DefaultText:=storedstring("")
        UtilitiesTabButton.OnClick().Subscribe(OpenUtilities)

        AttackTabButton := button_quiet : 
            DefaultText:=storedstring("")
        AttackTabButton.OnClick().Subscribe(OpenAttack)

        DefenceTabButton := button_quiet : 
            DefaultText:=storedstring("")
        DefenceTabButton.OnClick().Subscribe(OpenDefence)

        FourthTabButton := button_quiet : 
            DefaultText:=storedstring("")
        FourthTabButton.OnClick().Subscribe(OpenFourthTab)

        BuySlotOneButton := button_quiet :
            DefaultText:=storedstring("   Teleport   ")
        BuySlotOneButton.OnClick().Subscribe(UpgradeSlot1)

        BuySlotTwoButton := button_quiet :
            DefaultText:=storedstring("   Teleport   ")
        BuySlotTwoButton.OnClick().Subscribe(UpgradeSlot2)

        BuySlotThreeButton := button_quiet :
            DefaultText:=storedstring("   Teleport   ")
        BuySlotThreeButton.OnClick().Subscribe(UpgradeSlot3)

        CastleTeleportButton := button_quiet :
            DefaultText:=storedstring("Town")
        CastleTeleportButton.OnClick().Subscribe(TeleportToCastle)

        CloseWindowButton := button_quiet :
            DefaultText:=storedstring("   Close   ")
        CloseWindowButton.OnClick().Subscribe(RemoveHUD)

        SlotOneSkillText : text_block = text_block{DefaultText:=storedstring("Peaceful Ridge"), DefaultTextColor := MakeColorFromHex("142706FF")}
        SlotOneSkillText2 : text_block = text_block{DefaultText:=storedstring("Peaceful Ridge"), DefaultTextColor := MakeColorFromHex("529F17FF")}
        SlotOneDescription : text_block = text_block{DefaultText:=storedstring("Increases the gold earned per enemy\n defeated."), DefaultTextColor := MakeColorFromHex("FFFFFFFF")}

        SlotOneLevel : text_block = text_block{DefaultText:=storedstring("Lv.5"), DefaultTextColor := MakeColorFromHex("142706FF")}
        SlotOneLevel2 : text_block = text_block{DefaultText:=storedstring("Lv.5"), DefaultTextColor := MakeColorFromHex("529F17FF")}
    
        SlotTwoSkillText : text_block = text_block{DefaultText:=storedstring("Pleasant Valley"), DefaultTextColor := MakeColorFromHex("142706FF")}
        SlotTwoSkillText2 : text_block = text_block{DefaultText:=storedstring("Pleasant Valley"), DefaultTextColor := MakeColorFromHex("529F17FF")}
        SlotTwoDescription : text_block = text_block{DefaultText:=storedstring("Increases damage dealt with weapons\n and abilities."), DefaultTextColor := MakeColorFromHex("FFFFFFFF")}
        SlotTwoLevel : text_block = text_block{DefaultText:=storedstring("Lv.10"), DefaultTextColor := MakeColorFromHex("142706FF")}
        SlotTwoLevel2 : text_block = text_block{DefaultText:=storedstring("Lv.10"), DefaultTextColor := MakeColorFromHex("529F17FF")}
        
        SlotThreeSkillText : text_block = text_block{DefaultText:=storedstring("Enchanted Forest"), DefaultTextColor := MakeColorFromHex("142706FF")}
        SlotThreeSkillText2 : text_block = text_block{DefaultText:=storedstring("Enchanted Forest"), DefaultTextColor := MakeColorFromHex("529F17FF")}
        SlotThreeDescription : text_block = text_block{DefaultText:=storedstring("Increases damage dealt with weapons\n and abilities."), DefaultTextColor := MakeColorFromHex("FFFFFFFF")}

        SlotThreeLevel : text_block = text_block{DefaultText:=storedstring("Lv.10"), DefaultTextColor := MakeColorFromHex("142706FF")}
        SlotThreeLevel2 : text_block = text_block{DefaultText:=storedstring("Lv.10"), DefaultTextColor := MakeColorFromHex("529F17FF")}

        SlotOneTexture : texture_block = texture_block{DefaultDesiredSize:=vector2{X:=276.0,Y:=155.0}, DefaultImage:=Icons.Screenshots.TestingScreenshot}
        SlotTwoTexture : texture_block = texture_block{DefaultDesiredSize:=vector2{X:=276.0,Y:=155.0}, DefaultImage:=Icons.Screenshots.TestingScreenshot}
        SlotThreeTexture : texture_block = texture_block{DefaultDesiredSize:=vector2{X:=276.0,Y:=155.0}, DefaultImage:=Icons.Screenshots.TestingScreenshot }

        TabTexture1 : texture_block = texture_block{DefaultDesiredSize:=vector2{X:=100.0,Y:=100.0}, DefaultImage:=wooden_pixel_gui.textures.Panels.t_TagBottom, DefaultTint:=MakeColorFromHex("230D03FF") }
        TabTexture2: texture_block = texture_block{DefaultDesiredSize:=vector2{X:=100.0,Y:=100.0}, DefaultImage:=wooden_pixel_gui.textures.Panels.t_TagBottom, DefaultTint:=MakeColorFromHex("230D03FF") }
        TabTexture3 : texture_block = texture_block{DefaultDesiredSize:=vector2{X:=100.0,Y:=100.0}, DefaultImage:=wooden_pixel_gui.textures.Panels.t_TagBottom, DefaultTint:=MakeColorFromHex("824820FF") }
        TabTexture4 : texture_block = texture_block{DefaultDesiredSize:=vector2{X:=100.0,Y:=100.0}, DefaultImage:=wooden_pixel_gui.textures.Panels.t_TagBottom, DefaultTint:=MakeColorFromHex("230D03FF") }

        if(set TextBlockPerAgent[Agent] = option{array{SlotOneSkillText,SlotOneSkillText2,SlotOneDescription}}){} 
        if(set TextBlock2PerAgent[Agent] = option{array{SlotTwoSkillText,SlotTwoSkillText2,SlotTwoDescription}}){} 
        if(set TextBlock3PerAgent[Agent] = option{array{SlotThreeSkillText,SlotThreeSkillText2,SlotThreeDescription}}){} 

        if(set LevelTextPerAgent[Agent] = option{array{SlotOneLevel,SlotOneLevel2,SlotTwoLevel,SlotTwoLevel2, SlotThreeLevel,SlotThreeLevel2}}){}

        if(set TextureBlockPerAgent[Agent] = option{array{SlotOneTexture,SlotTwoTexture,SlotThreeTexture}}){} 

        if(set ButtonLoudPerAgent[Agent] = option{array{BuySlotOneButton,BuySlotTwoButton,BuySlotThreeButton}}){} 
        if(set TabPerAgent[Agent]= option{array{TabTexture1,TabTexture2,TabTexture3,TabTexture4}}){}

        TeleportStackBox := stack_box:
            Orientation:=orientation.Vertical


        TeleportStackBox.AddWidget(CreateTeleportSlot(SlotOneLevel,SlotOneLevel2,SlotOneSkillText,SlotOneSkillText2,SlotOneTexture,SlotOneDescription,BuySlotOneButton))
        TeleportStackBox.AddWidget(CreateTeleportSlot(SlotTwoLevel,SlotTwoLevel2,SlotTwoSkillText,SlotTwoSkillText2,SlotTwoTexture,SlotTwoDescription,BuySlotTwoButton))
        TeleportStackBox.AddWidget(CreateTeleportSlot(SlotThreeLevel,SlotThreeLevel2,SlotThreeSkillText,SlotThreeSkillText2,SlotThreeTexture,SlotThreeDescription,BuySlotThreeButton))


        Overlay1 := overlay:
            Slots:=array:
                overlay_slot:
                    HorizontalAlignment:=horizontal_alignment.Fill
                    VerticalAlignment:=vertical_alignment.Fill
                    Padding:=margin{Left:=-10.0,Bottom:=-10.0,Top:=-10.0,Right:=-10.0}
                    Widget:=texture_block:
                        DefaultImage:=wooden_pixel_gui.textures.Panels.WhiteSquare90
                        DefaultTint:=MakeColorFromHex("443525FF")
                overlay_slot:
                    HorizontalAlignment:=horizontal_alignment.Fill
                    VerticalAlignment:=vertical_alignment.Fill
                    Widget := stack_box:
                        Orientation:=orientation.Vertical
                        Slots:=array:
                            stack_box_slot:
                                HorizontalAlignment:=horizontal_alignment.Fill
                                VerticalAlignment:=vertical_alignment.Fill
                                Widget := overlay:
                                    Slots:=array:
                                        overlay_slot:
                                            Widget:= TitleBarTexture
                                            HorizontalAlignment:=horizontal_alignment.Fill
                                            VerticalAlignment:=vertical_alignment.Fill
                                        overlay_slot:
                                            Widget := text_block :
                                                DefaultText:=storedstring("Teleporters")
                                                DefaultTextColor := MakeColorFromHex("643D19FF")
                                            HorizontalAlignment:=horizontal_alignment.Center
                                            VerticalAlignment:=vertical_alignment.Center
                                        
                            stack_box_slot: # Tabs
                                HorizontalAlignment:=horizontal_alignment.Fill
                                VerticalAlignment:=vertical_alignment.Fill
                                Widget:=overlay:
                                    Slots:=array:
                                        overlay_slot:
                                            HorizontalAlignment:=horizontal_alignment.Fill
                                            VerticalAlignment:=vertical_alignment.Fill
                                            Widget:=texture_block:
                                                DefaultImage:=wooden_pixel_gui.textures.Panels.WhiteSquare75
                                                DefaultTint:=MakeColorFromHex("413425FF")
                                        overlay_slot:
                                            HorizontalAlignment:=horizontal_alignment.Center
                                            VerticalAlignment:=vertical_alignment.Center                                                    
                                            Widget:=stack_box:
                                                Orientation:=orientation.Horizontal
                                                Slots:=array:     
                                                    stack_box_slot:
                                                        Padding:=margin{Left:=0.0}
                                                        Widget:=overlay:
                                                            Slots:=array:
                                                                overlay_slot:
                                                                    Widget:=texture_block :
                                                                        DefaultDesiredSize:=vector2{X:=100.0,Y:=100.0}
                                                                        DefaultImage:=wooden_pixel_gui.textures.Panels.t_TagBottomTBevel
                                                                        DefaultTint:=MakeColorFromHex("FFFFFF")
                                                                    HorizontalAlignment:=horizontal_alignment.Fill
                                                                    VerticalAlignment:=vertical_alignment.Bottom
                                                                    Padding:=margin{Left:=5.0,Bottom:=5.0}
                                                                overlay_slot:
                                                                    Widget:=TabTexture1
                                                                    HorizontalAlignment:=horizontal_alignment.Fill
                                                                    VerticalAlignment:=vertical_alignment.Bottom
                                                                overlay_slot:
                                                                    Widget:=texture_block :
                                                                        DefaultDesiredSize:=vector2{X:=100.0,Y:=100.0}
                                                                        DefaultImage:=RavenIcons.MapMarkers.S23
                                                                    HorizontalAlignment:=horizontal_alignment.Fill
                                                                    VerticalAlignment:=vertical_alignment.Bottom
                                                                overlay_slot:
                                                                    Widget := UtilitiesTabButton
                                                                    HorizontalAlignment:=horizontal_alignment.Fill
                                                                    VerticalAlignment:=vertical_alignment.Fill
                                                    stack_box_slot:
                                                        Padding:=margin{Left:=30.0}
                                                        Widget:=overlay:
                                                            Slots:=array:
                                                                overlay_slot:
                                                                    Widget:=texture_block :
                                                                        DefaultDesiredSize:=vector2{X:=100.0,Y:=100.0}
                                                                        DefaultImage:=wooden_pixel_gui.textures.Panels.t_TagBottomTBevel
                                                                        DefaultTint:=MakeColorFromHex("FFFFFF")
                                                                    HorizontalAlignment:=horizontal_alignment.Fill
                                                                    VerticalAlignment:=vertical_alignment.Bottom
                                                                    Padding:=margin{Left:=5.0,Bottom:=5.0}
                                                                overlay_slot:
                                                                    Widget:=TabTexture2
                                                                    HorizontalAlignment:=horizontal_alignment.Fill
                                                                    VerticalAlignment:=vertical_alignment.Bottom
                                                                overlay_slot:
                                                                    Widget:=texture_block :
                                                                        DefaultDesiredSize:=vector2{X:=100.0,Y:=100.0}
                                                                        DefaultImage:=RavenIcons.MapMarkers.Stile031
                                                                    HorizontalAlignment:=horizontal_alignment.Fill
                                                                    VerticalAlignment:=vertical_alignment.Bottom 
                                                                overlay_slot:
                                                                    Widget := AttackTabButton
                                                                    HorizontalAlignment:=horizontal_alignment.Fill
                                                                    VerticalAlignment:=vertical_alignment.Fill
                                                    stack_box_slot:
                                                        Padding:=margin{Left:=30.0}
                                                        Widget:=overlay:
                                                            Slots:=array:
                                                                overlay_slot:
                                                                    Widget:=texture_block :
                                                                        DefaultDesiredSize:=vector2{X:=100.0,Y:=100.0}
                                                                        DefaultImage:=wooden_pixel_gui.textures.Panels.t_TagBottomTBevel
                                                                        DefaultTint:=MakeColorFromHex("FFFFFF")
                                                                    HorizontalAlignment:=horizontal_alignment.Fill
                                                                    VerticalAlignment:=vertical_alignment.Bottom
                                                                    Padding:=margin{Left:=5.0,Bottom:=5.0}
                                                                overlay_slot:
                                                                    Widget:=TabTexture3
                                                                    HorizontalAlignment:=horizontal_alignment.Fill
                                                                    VerticalAlignment:=vertical_alignment.Bottom
                                                                overlay_slot:
                                                                    Widget:=texture_block :
                                                                        DefaultDesiredSize:=vector2{X:=100.0,Y:=100.0}
                                                                        DefaultImage:=RavenIcons.MapMarkers.Stile029
                                                                    HorizontalAlignment:=horizontal_alignment.Fill
                                                                    VerticalAlignment:=vertical_alignment.Bottom
                                                                overlay_slot:
                                                                    Widget :=DefenceTabButton
                                                                    HorizontalAlignment:=horizontal_alignment.Fill
                                                                    VerticalAlignment:=vertical_alignment.Fill
                                                    stack_box_slot:
                                                        Padding:=margin{Left:=30.0}
                                                        Widget:=overlay:
                                                            Slots:=array:
                                                                overlay_slot:
                                                                    Widget:=texture_block :
                                                                        DefaultDesiredSize:=vector2{X:=100.0,Y:=100.0}
                                                                        DefaultImage:=wooden_pixel_gui.textures.Panels.t_TagBottomTBevel
                                                                        DefaultTint:=MakeColorFromHex("FFFFFF")
                                                                    HorizontalAlignment:=horizontal_alignment.Fill
                                                                    VerticalAlignment:=vertical_alignment.Bottom
                                                                    Padding:=margin{Left:=5.0,Bottom:=5.0}
                                                                overlay_slot:
                                                                    Widget:=TabTexture4
                                                                    HorizontalAlignment:=horizontal_alignment.Fill
                                                                    VerticalAlignment:=vertical_alignment.Bottom
                                                                overlay_slot:
                                                                    Widget:=texture_block :
                                                                        DefaultDesiredSize:=vector2{X:=100.0,Y:=100.0}
                                                                        DefaultImage:=RavenIcons.MapMarkers.S177
                                                                    HorizontalAlignment:=horizontal_alignment.Fill
                                                                    VerticalAlignment:=vertical_alignment.Bottom
                                                                overlay_slot:
                                                                    Widget :=FourthTabButton
                                                                    HorizontalAlignment:=horizontal_alignment.Fill
                                                                    VerticalAlignment:=vertical_alignment.Fill
                                        overlay_slot:
                                            HorizontalAlignment:=horizontal_alignment.Left
                                            VerticalAlignment:=vertical_alignment.Center
                                            Widget:=CastleTeleportButton
                                            Padding:=margin{Left:=15.0,Bottom:=0.0,Top:=0.0,Right:=25.0}
                            stack_box_slot: # Main Slots
                                HorizontalAlignment:=horizontal_alignment.Center
                                VerticalAlignment:=vertical_alignment.Center
                                Widget:=overlay:
                                    Slots:=array:
                                        overlay_slot:                                               # large background behind all all slots
                                            HorizontalAlignment:=horizontal_alignment.Fill
                                            VerticalAlignment:=vertical_alignment.Fill
                                            Widget:=texture_block : 
                                                DefaultDesiredSize:=vector2{X:=32.0,Y:=32.0}
                                                DefaultImage:=wooden_pixel_gui.textures.Panels.WhiteSquare90
                                                DefaultTint:=MakeColorFromHex("604528FF")
                                        overlay_slot:
                                            HorizontalAlignment:=horizontal_alignment.Center
                                            VerticalAlignment:=vertical_alignment.Center
                                            Padding:=margin{Left:=0.0,Top:=0.0,Right:=0.0,Bottom:=0.0}
                                            Widget:=TeleportStackBox
                            stack_box_slot:
                                HorizontalAlignment:=horizontal_alignment.Center
                                VerticalAlignment:=vertical_alignment.Bottom
                                Widget:=CloseWindowButton         

        # This is the master canvas
        MyCanvas:=canvas:
            Slots:=array:
                canvas_slot:
                    Anchors:=anchors{Minimum:=vector2{X:=0.2,Y:=0.05}, Maximum:=vector2{X:=0.8,Y:=0.95}}
                    Widget:= Overlay1
                    
        return MyCanvas


    CreateTeleportSlot(LevelTextBlock:text_block,LevelTextBlock2:text_block,TitleTextBlock:text_block,TitleTextBlock2:text_block,ImageTexture:texture_block,Description:text_block,TeleportButton:button_quiet):stack_box_slot= 
        
        MainStackBoxSlot := stack_box_slot : # Main Slots
            Padding:=margin{Left:=10.0,Top:=10.0,Right:=10.0,Bottom:=10.0}
            HorizontalAlignment:=horizontal_alignment.Fill
            VerticalAlignment:=vertical_alignment.Fill
            Widget:=overlay:
                Slots:=array:
                    overlay_slot:                                               # large background behind all all slots
                        HorizontalAlignment:=horizontal_alignment.Fill
                        VerticalAlignment:=vertical_alignment.Fill
                        Widget:=texture_block : 
                            DefaultDesiredSize:=vector2{X:=32.0,Y:=32.0}
                            DefaultImage:=wooden_pixel_gui.textures.Panels.WhiteSquare90
                            DefaultTint:=MakeColorFromHex("604528FF")
                    overlay_slot:
                        HorizontalAlignment:=horizontal_alignment.Fill
                        VerticalAlignment:=vertical_alignment.Fill
                        Padding:=margin{Left:=0.0,Top:=0.0,Right:=0.0,Bottom:=0.0}
                        Widget:=stack_box:
                            Orientation:=orientation.Vertical
                            Slots:=array:
                                stack_box_slot:                                            #stackboxslot = overlay
                                    HorizontalAlignment:=horizontal_alignment.Fill
                                    VerticalAlignment:=vertical_alignment.Fill
                                    Padding:=margin{Left:=10.0,Top:=10.0,Right:=10.0,Bottom:=10.0}
                                    Widget:=overlay:
                                        Slots:=array:
                                            overlay_slot:
                                                HorizontalAlignment:=horizontal_alignment.Fill
                                                VerticalAlignment:=vertical_alignment.Fill
                                                Widget:=texture_block : # backgroundfor this slot
                                                    DefaultDesiredSize:=vector2{X:=32.0,Y:=32.0}
                                                    DefaultImage:=wooden_pixel_gui.textures.Panels.WhiteSquare100
                                                    DefaultTint:=MakeColorFromHex("49321CFF")
                                            overlay_slot:
                                                HorizontalAlignment:=horizontal_alignment.Center
                                                VerticalAlignment:=vertical_alignment.Center
                                                Widget:=stack_box:
                                                    Orientation:=orientation.Vertical
                                                    Slots:=array:
                                                        stack_box_slot:
                                                            HorizontalAlignment:=horizontal_alignment.Center
                                                            VerticalAlignment:=vertical_alignment.Center
                                                            Widget:=overlay:
                                                                Slots:=array:
                                                                    overlay_slot:
                                                                        HorizontalAlignment:=horizontal_alignment.Fill
                                                                        VerticalAlignment:=vertical_alignment.Fill
                                                                        Widget:=texture_block : 
                                                                            DefaultDesiredSize:=vector2{X:=1100.0,Y:=32.0}
                                                                            DefaultImage:=wooden_pixel_gui.textures.Panels.WhiteSquare75
                                                                            DefaultTint:=MakeColorFromHex("8B5C2AFF")
                                                                    overlay_slot:
                                                                        HorizontalAlignment:=horizontal_alignment.Center
                                                                        VerticalAlignment:=vertical_alignment.Center
                                                                        Padding:=margin{Left:=3.0,Top:=3.0}
                                                                        Widget := TitleTextBlock
                                                                    overlay_slot:
                                                                        HorizontalAlignment:=horizontal_alignment.Center
                                                                        VerticalAlignment:=vertical_alignment.Center
                                                                        Widget := TitleTextBlock2
                                                                    overlay_slot:
                                                                        HorizontalAlignment:=horizontal_alignment.Left
                                                                        VerticalAlignment:=vertical_alignment.Center
                                                                        Padding:=margin{Left:=8.0,Top:=3.0}
                                                                        Widget := LevelTextBlock
                                                                    overlay_slot:
                                                                        HorizontalAlignment:=horizontal_alignment.Left
                                                                        VerticalAlignment:=vertical_alignment.Center
                                                                        Padding:=margin{Left:=5.0,Top:=3.0}
                                                                        Widget := LevelTextBlock2
                                                                        
                                                        stack_box_slot:
                                                            HorizontalAlignment:=horizontal_alignment.Fill
                                                            VerticalAlignment:=vertical_alignment.Fill
                                                            Widget:=overlay:
                                                                Slots:=array:
                                                                    overlay_slot:
                                                                        HorizontalAlignment:=horizontal_alignment.Left
                                                                        VerticalAlignment:=vertical_alignment.Top
                                                                        Widget:=stack_box:
                                                                            Orientation:=orientation.Horizontal
                                                                            Slots:=array:
                                                                                stack_box_slot:
                                                                                    Widget:=ImageTexture
                                                                                    HorizontalAlignment:=horizontal_alignment.Center
                                                                                    VerticalAlignment:=vertical_alignment.Center
                                                                                stack_box_slot:
                                                                                    Padding:=margin{Left:=10.0,Top:=10.0,Bottom:=10.0}
                                                                                    Widget := Description
                                                                                    HorizontalAlignment:=horizontal_alignment.Left
                                                                                    VerticalAlignment:=vertical_alignment.Top
                                                                                    #Padding:=margin{Top:=0.0}
                                                                    overlay_slot:
                                                                        Widget := TeleportButton
                                                                        HorizontalAlignment:=horizontal_alignment.Center
                                                                        VerticalAlignment:=vertical_alignment.Bottom

        return MainStackBoxSlot

                

    OpenShopHandle(Agent:agent):void=
        RemoveFortHud(Agent)
        spawn:
            ShowUi(Agent)
 
        
    ShowUi(Agent:agent)<suspends>:void=

        if(CurrentTabAgents[Agent]?):
        else:
            if(set CurrentTabAgents[Agent] = option{0}){}

        MyCanvas: canvas = MakeEntireBackground(Agent)
        if(PlayerUI:=GetPlayerUI[player[Agent]]):
            OpenAudio.Play(Agent)
            PlayerUI.AddWidget(MyCanvas,player_ui_slot{InputMode:=ui_input_mode.All}) # want it to appear on top
            UpdateUiText(Agent)
            MyCanvas.SetVisibility(widget_visibility.Hidden)
            MyCanvas.SetVisibility(widget_visibility.Visible)
        if(set MyUI[Agent]= option{MyCanvas}){}

    OpenUtilities(Wdata:widget_message):void=
        if(Player:=Wdata.Player, PlayerUI:=GetPlayerUI[Player],SelectedButton:=Wdata.Source,Agent:=agent[Player]):
            if(set CurrentTabAgents[Agent] = option{1}){}
            TabAudio.Play(Agent)
            UpdateUiText(Agent)

    OpenDefence(Wdata:widget_message):void=
        if(Player:=Wdata.Player, PlayerUI:=GetPlayerUI[Player],SelectedButton:=Wdata.Source,Agent:=agent[Player]):
            if(set CurrentTabAgents[Agent] = option{2}){}
            TabAudio.Play(Agent)
            UpdateUiText(Agent)

    OpenAttack(Wdata:widget_message):void=
        if(Player:=Wdata.Player, PlayerUI:=GetPlayerUI[Player],SelectedButton:=Wdata.Source,Agent:=agent[Player]):
            if(set CurrentTabAgents[Agent] = option{0}){}
            TabAudio.Play(Agent)
            UpdateUiText(Agent)

    OpenFourthTab(Wdata:widget_message):void=

    
        if(Player:=Wdata.Player, PlayerUI:=GetPlayerUI[Player],SelectedButton:=Wdata.Source,Agent:=agent[Player]):
            if(set CurrentTabAgents[Agent] = option{3}){}
            TabAudio.Play(Agent)
            UpdateUiText(Agent)

    RemoveHUD(Wdata:widget_message):void =
        
        Print("Removing")
        if(Player:=Wdata.Player, PlayerUI:=GetPlayerUI[Player],SelectedButton:=Wdata.Source,Agent:=agent[Player]):
            if(MyActualUI := MyUI[Agent]?):

                CloseAudio.Play(Agent)
                PlayerUI.RemoveWidget(MyActualUI)

                MyActualUI.SetVisibility(widget_visibility.Visible)
                MyActualUI.SetVisibility(widget_visibility.Hidden)

                if(set MyUI[Agent] = false){}
                if(set TextBlockPerAgent[Agent] = false){}

                ShowFortHud(Agent)
                

    UpdateUiText(Agent:agent):void=

        if:

            SkillText1: text_block = TextBlockPerAgent[Agent]?[0]
            SkillText11: text_block = TextBlockPerAgent[Agent]?[1]
            DescriptionText1: text_block = TextBlockPerAgent[Agent]?[2]
            
            SkillText2: text_block = TextBlock2PerAgent[Agent]?[0]
            SkillText22: text_block = TextBlock2PerAgent[Agent]?[1]
            DescriptionText2: text_block = TextBlock2PerAgent[Agent]?[2]

            SkillText3: text_block = TextBlock3PerAgent[Agent]?[0]
            SkillText33: text_block = TextBlock3PerAgent[Agent]?[1]
            DescriptionText3: text_block = TextBlock3PerAgent[Agent]?[2]

            Texture1 : texture_block = TextureBlockPerAgent[Agent]?[0]
            Texture2 : texture_block = TextureBlockPerAgent[Agent]?[1]
            Texture3 : texture_block = TextureBlockPerAgent[Agent]?[2]

            TabTexture1 : texture_block = TabPerAgent[Agent]?[0]
            TabTexture2 : texture_block = TabPerAgent[Agent]?[1]
            TabTexture3 : texture_block = TabPerAgent[Agent]?[2]
            TabTexture4 : texture_block = TabPerAgent[Agent]?[3]

            LevelText1 : text_block = LevelTextPerAgent[Agent]?[0]
            LevelText1Shadow : text_block = LevelTextPerAgent[Agent]?[1]
            LevelText2 : text_block = LevelTextPerAgent[Agent]?[2]
            LevelText2Shadow : text_block = LevelTextPerAgent[Agent]?[3]
            LevelText3 : text_block = LevelTextPerAgent[Agent]?[4]
            LevelText3Shadow : text_block = LevelTextPerAgent[Agent]?[5]

            Player := player[Agent]
            MyData := Data[Agent]

            CurrentTab := CurrentTabAgents[Agent]?
            
        then:

            Print("CurrentTab : {CurrentTab}")

            if (CurrentTab = 0):

                TabTexture1.SetTint(MakeColorFromHex("230D03FF"))
                TabTexture2.SetTint(MakeColorFromHex("824820FF"))
                TabTexture3.SetTint(MakeColorFromHex("230D03FF"))
                TabTexture4.SetTint(MakeColorFromHex("230D03FF"))

                LevelText1.SetText(storedstring("Lv.1"))
                LevelText1Shadow.SetText(storedstring("Lv.1"))
                LevelText2.SetText(storedstring("Lv.5"))
                LevelText2Shadow.SetText(storedstring("Lv.5"))
                LevelText3.SetText(storedstring("Lv.10"))
                LevelText3Shadow.SetText(storedstring("Lv.10"))

                if(MyData.AreasUnlocked[0] = true):
                    SkillText1.SetText(storedstring("Peaceful Ridge"))
                    SkillText11.SetText(storedstring("Peaceful Ridge"))
                    DescriptionText1.SetText(storedstring("Peaceful beginnings"))
                    Texture1.SetImage( Icons.Screenshots.Area1 )
                else:
                    Print("AreasUnlocked 0 = false")
                    SetUndiscoveredTexture(SkillText1,SkillText11,DescriptionText1,Texture1)

                if(MyData.AreasUnlocked[1] = true):
                    SkillText2.SetText(storedstring("Pleasant Valley"))
                    SkillText22.SetText(storedstring("Pleasant Valley"))
                    DescriptionText2.SetText(storedstring("A peaceful valley of rolling hills, surrounded by\nnature's calm beauty"))
                    Texture2.SetImage( Icons.Screenshots.Area2 )
                else:
                    Print("AreasUnlocked 1 = false")
                    SetUndiscoveredTexture(SkillText2,SkillText22,DescriptionText2,Texture2)

                if(MyData.AreasUnlocked[2] = true):
                    SkillText3.SetText(storedstring("Enchanted Forest"))
                    SkillText33.SetText(storedstring("Enchanted Forest"))
                    DescriptionText3.SetText(storedstring("Glowing trees where agile lizard archers lurk, ready\nwith their enchanted bows"))
                    Texture3.SetImage( Icons.Screenshots.Area3 )
                else:
                    Print("AreasUnlocked 2 = false")
                    SetUndiscoveredTexture(SkillText3,SkillText33,DescriptionText3,Texture3)

            if(CurrentTab = 1):

                LevelText1.SetText(storedstring("Lv.15"))
                LevelText1Shadow.SetText(storedstring("Lv.15"))
                LevelText2.SetText(storedstring("Lv.20"))
                LevelText2Shadow.SetText(storedstring("Lv.20"))
                LevelText3.SetText(storedstring("Lv.25"))
                LevelText3Shadow.SetText(storedstring("Lv.25"))

                TabTexture1.SetTint(MakeColorFromHex("824820FF"))
                TabTexture2.SetTint(MakeColorFromHex("230D03FF"))
                TabTexture3.SetTint(MakeColorFromHex("230D03FF"))
                TabTexture4.SetTint(MakeColorFromHex("230D03FF"))

                if(MyData.AreasUnlocked[3] = true):
                    SkillText1.SetText(storedstring("Scorched Plains"))
                    SkillText11.SetText(storedstring("Scorched Plains"))
                    DescriptionText1.SetText(storedstring("A scorched plain, cracked and barren, where a\nmassive desert beast prowls beneath a blazing sun"))
                    Texture1.SetImage(Icons.Screenshots.Area4 )
                else:
                    Print("AreasUnlocked 3 = false")
                    SetUndiscoveredTexture(SkillText1,SkillText11,DescriptionText1,Texture1)

                if(MyData.AreasUnlocked[4] = true):
                    SkillText2.SetText(storedstring("Radiant Dunes"))
                    SkillText22.SetText(storedstring("Radiant Dunes"))
                    DescriptionText2.SetText(storedstring("Radiant dunes glow under the sun, stretching\nendlessly in golden waves of sand"))
                    Texture2.SetImage(Icons.Screenshots.Area5 )
                else:
                    Print("AreasUnlocked 4 = false")
                    SetUndiscoveredTexture(SkillText2,SkillText22,DescriptionText2,Texture2)

                if(MyData.AreasUnlocked[5] = true):
                    SkillText3.SetText(storedstring("Dusty Wastes"))
                    SkillText33.SetText(storedstring("Dusty Wastes"))
                    DescriptionText3.SetText(storedstring("A dusty wasteland, barren and bleak, where ancient\ngolems stand silent among the ruins"))
                    Texture3.SetImage(Icons.Screenshots.Area6 )
                else:
                    Print("AreasUnlocked 5 = false")
                    SetUndiscoveredTexture(SkillText3,SkillText33,DescriptionText3,Texture3)

            if(CurrentTab = 2):

                LevelText1.SetText(storedstring("Lv.30"))
                LevelText1Shadow.SetText(storedstring("Lv.30"))
                LevelText2.SetText(storedstring("Lv.35"))
                LevelText2Shadow.SetText(storedstring("Lv.35"))
                LevelText3.SetText(storedstring("Lv.40"))
                LevelText3Shadow.SetText(storedstring("Lv.40"))


                TabTexture1.SetTint(MakeColorFromHex("230D03FF"))
                TabTexture2.SetTint(MakeColorFromHex("230D03FF"))
                TabTexture3.SetTint(MakeColorFromHex("824820FF"))
                TabTexture4.SetTint(MakeColorFromHex("230D03FF"))

                if(MyData.AreasUnlocked[6] = true):
                    SkillText1.SetText(storedstring("Frostwind Pass"))
                    SkillText11.SetText(storedstring("Frostwind Pass"))
                    DescriptionText1.SetText(storedstring("An icy mountain pass, windswept and treacherous,\nwhere leopards silently stalk through the snow"))
                    Texture1.SetImage(Icons.Screenshots.Area7 )
                else:
                    Print("AreasUnlocked 6 = false")
                    SetUndiscoveredTexture(SkillText1,SkillText11,DescriptionText1,Texture1)

                if(MyData.AreasUnlocked[7] = true):
                    SkillText2.SetText(storedstring("Frostbite Peaks"))
                    SkillText22.SetText(storedstring("Frostbite Peaks"))
                    DescriptionText2.SetText(storedstring("A jagged, icy range where mountain goats navigate\nthe steep, snow-covered cliffs with ease"))
                    Texture2.SetImage(Icons.Screenshots.Area8 )
                else:
                    Print("AreasUnlocked 7 = false")
                    SetUndiscoveredTexture(SkillText2,SkillText22,DescriptionText2,Texture2)

                if(MyData.AreasUnlocked[8] = true):
                    SkillText3.SetText(storedstring("Chillfrost Grove"))
                    SkillText33.SetText(storedstring("Chillfrost Grove"))
                    DescriptionText3.SetText(storedstring("A shadowy, frozen forest where yetis roam silently\namong the frosted trees."))
                    Texture3.SetImage(Icons.Screenshots.Area9)
                else:
                    Print("AreasUnlocked 8 = false")
                    SetUndiscoveredTexture(SkillText3,SkillText33,DescriptionText3,Texture3)

            if(CurrentTab = 3):

                LevelText1.SetText(storedstring("Lv.45"))
                LevelText1Shadow.SetText(storedstring("Lv.45"))
                LevelText2.SetText(storedstring("Lv.50"))
                LevelText2Shadow.SetText(storedstring("Lv.50"))
                LevelText3.SetText(storedstring("Lv.55"))
                LevelText3Shadow.SetText(storedstring("Lv.55"))

                TabTexture1.SetTint(MakeColorFromHex("230D03FF"))
                TabTexture2.SetTint(MakeColorFromHex("230D03FF"))
                TabTexture3.SetTint(MakeColorFromHex("230D03FF"))
                TabTexture4.SetTint(MakeColorFromHex("824820FF"))

                if(MyData.AreasUnlocked[9] = true):
                    SkillText1.SetText(storedstring("Scorchpath"))
                    SkillText11.SetText(storedstring("Scorchpath"))
                    DescriptionText1.SetText(storedstring("A fiery trail of molten rock, where wild hounds hunt\namidst the searing heat and ash"))
                    Texture1.SetImage(Icons.Screenshots.Area10)
                else:
                    Print("AreasUnlocked 9 = false")
                    SetUndiscoveredTexture(SkillText1,SkillText11,DescriptionText1,Texture1)

                if(MyData.AreasUnlocked[10] = true):
                    SkillText2.SetText(storedstring("Magma Crossing"))
                    SkillText22.SetText(storedstring("Magma Crossing"))
                    DescriptionText2.SetText(storedstring("A molten path with snapping lava jaws, leading\n toward a towering, active volcano"))
                    Texture2.SetImage(Icons.Screenshots.Area11)
                else:
                    Print("AreasUnlocked 10 = false")
                    SetUndiscoveredTexture(SkillText2,SkillText22,DescriptionText2,Texture2)

                if(MyData.AreasUnlocked[11] = true):
                    SkillText3.SetText(storedstring("Cinder Basin"))
                    SkillText33.SetText(storedstring("Cinder Basin"))
                    DescriptionText3.SetText(storedstring("A Charred valley beneath the volcano, smolders with ash\n and heat, silent and ominous "))
                    Texture3.SetImage(Icons.Screenshots.Area12)
                else:
                    Print("AreasUnlocked 11 = false")
                    SetUndiscoveredTexture(SkillText3,SkillText33,DescriptionText3,Texture3)

    
    SetUndiscoveredTexture(TitleTextblock:text_block,TitleTextblock2:text_block,Descrip:text_block,Image:texture_block):void=
        TitleTextblock.SetText(storedstring("Undiscovered"))
        TitleTextblock2.SetText(storedstring("Undiscovered"))
        Descrip.SetText(storedstring("This area has not been discovered yet"))
        Image.SetImage( Icons.Screenshots.BlackScreen1 )

    TeleportToCastle(Wdata:widget_message):void=
        if(Player:=Wdata.Player, PlayerUI:=GetPlayerUI[Player],SelectedButton:=Wdata.Source,Agent:=agent[Player]):
            ButtonClickSound.Play(Agent)
            if (MyData := Data[Player]):
                spawn:
                    TeleportAnimation(Agent,12,Wdata)

    UpgradeSlot1(Wdata:widget_message):void=
        if(Player:=Wdata.Player, PlayerUI:=GetPlayerUI[Player],SelectedButton:=Wdata.Source,Agent:=agent[Player]):
            ButtonClickSound.Play(Agent)
            if(CurrentTab := CurrentTabAgents[Agent]?):
                if(CurrentTab = 0):
                    if (MyData := Data[Player]):
                        if(MyData.AreasUnlocked[0] = true):
                            spawn:
                                TeleportAnimation(Agent,0,Wdata)
                if(CurrentTab = 1):
                    if (MyData := Data[Player]):
                        if(MyData.AreasUnlocked[3] = true):
                            spawn:
                                TeleportAnimation(Agent,3,Wdata)
                if(CurrentTab = 2):
                    if (MyData := Data[Player]):
                        if(MyData.AreasUnlocked[6] = true):
                            spawn:
                                TeleportAnimation(Agent,6,Wdata)
                if(CurrentTab = 3):
                    if (MyData := Data[Player]):
                        if(MyData.AreasUnlocked[9] = true):
                            spawn:
                                TeleportAnimation(Agent,9,Wdata)


    UpgradeSlot2(Wdata:widget_message):void=
        if(Player:=Wdata.Player, PlayerUI:=GetPlayerUI[Player],SelectedButton:=Wdata.Source,Agent:=agent[Player]): 
            ButtonClickSound.Play(Agent)  
            if(CurrentTab := CurrentTabAgents[Agent]?):
                if(CurrentTab = 0):
                    if (MyData := Data[Player]):
                        if(MyData.AreasUnlocked[1] = true):
                            spawn:
                                TeleportAnimation(Agent,1,Wdata)
                if(CurrentTab = 1):
                    if (MyData := Data[Player]):
                        if(MyData.AreasUnlocked[4] = true):
                            spawn:
                                TeleportAnimation(Agent,4,Wdata)
                if(CurrentTab = 2): 
                    if (MyData := Data[Player]):
                        if(MyData.AreasUnlocked[7] = true):
                            spawn:
                                TeleportAnimation(Agent,7,Wdata)
                if(CurrentTab = 3):
                    if (MyData := Data[Player]):
                        if(MyData.AreasUnlocked[10] = true):
                            spawn:
                                TeleportAnimation(Agent,10,Wdata)

            
    UpgradeSlot3(Wdata:widget_message):void=
        if(Player:=Wdata.Player, PlayerUI:=GetPlayerUI[Player],SelectedButton:=Wdata.Source,Agent:=agent[Player]):
            ButtonClickSound.Play(Agent)
            if(CurrentTab := CurrentTabAgents[Agent]?):
                if(CurrentTab = 0):
                    if (MyData := Data[Player]):
                        if(MyData.AreasUnlocked[2] = true):
                            spawn:
                                TeleportAnimation(Agent,2,Wdata)
                if(CurrentTab = 1):
                    if (MyData := Data[Player]):
                        if(MyData.AreasUnlocked[5] = true):
                            spawn:
                                TeleportAnimation(Agent,5,Wdata)
                if(CurrentTab = 2):
                    if (MyData := Data[Player]):
                        if(MyData.AreasUnlocked[8] = true):
                            spawn:
                                TeleportAnimation(Agent,8,Wdata)
                if(CurrentTab = 3):
                    if (MyData := Data[Player]):
                        if(MyData.AreasUnlocked[11] = true):
                            spawn:
                                TeleportAnimation(Agent,11,Wdata)



    TeleportAnimation(Agent:agent,TeleporterIndex : int,Wdata:widget_message)<suspends>:void=
        if(Player := player[Agent], FortChar:= Player.GetFortCharacter[]):
            if(TeleportingAlready[Agent] = false):
                TeleportMissionTrigger.Trigger(Agent)
                FortChar.PutInStasis(stasis_args{AllowTurning:=true,AllowFalling:=true,AllowEmotes:=true})
                if(set TeleportingAlready[Agent] = true){}
                var Loc : vector3 = FortChar.GetTransform().Translation
                Rot := FortChar.GetTransform().Rotation
                set Loc.Z = Loc.Z - 75.0
                spawn:
                    SpawnEffect(TeleportNiagProp,5.0,Loc,Rot,vector3{X:=1.0,Y:=1.0,Z:=1.0})
                RemoveHUD(Wdata)
                TeleportAudio.Play(Agent)
                Sleep(3.0)
                if(MyTeleporter:=Teleporters[TeleporterIndex]):
                    MyTeleporter.Teleport(Agent)
                FortChar.ReleaseFromStasis()
                if(set TeleportingAlready[Agent] = false){}

    SpawnEffect(Prop:creative_prop_asset, LifeTime:float, Location : vector3 , Rotatoe : rotation, Scale:vector3)<suspends>:void=
        Tran:=transform{
            Translation := Location,
            Rotation := Rotatoe,
            Scale := Scale
        }
        SpawnedPropResult := SpawnProp(Prop,Tran)
        if(SpawnedProp := SpawnedPropResult(0)?):
            spawn:
                DeleteEffect(LifeTime,SpawnedProp)

    DeleteEffect(SleepTime : float, Prop : creative_prop)<suspends>:void=
        Sleep(SleepTime)
        Prop.Dispose()


