
using { /Fortnite.com/Devices }
using { /Verse.org/Simulation }
using { /Fortnite.com/Characters }
using { /UnrealEngine.com/Temporary/SpatialMath }

test_heal_boss_device := class(creative_device):

    @editable
    HealingBoss : npc_manager_device = npc_manager_device{}

    @editable
    Healers : []npc_manager_device = array{}

    var HealPercent : float = 0.011

    @editable
    HealingProp : creative_prop_asset = DefaultCreativePropAsset

    OnBegin<override>()<suspends>:void=

        loop:
            Sleep(1.0)
            if(HealingBoss.TotalHealth > HealingBoss.CurrentHealth):
                var GotHealed : logic = false
                for(HealNPC : Healers):
                    Sleep(0.5)
                    if(HealNPC.NPCSAlive > 0):
                        set GotHealed = true
                        set  HealingBoss.CurrentHealth = HealingBoss.CurrentHealth + (HealingBoss.TotalHealth * HealPercent)
                        if(HealingBoss.CurrentHealth > HealingBoss.TotalHealth):
                            set HealingBoss.CurrentHealth = HealingBoss.TotalHealth

                if:
                    GotHealed = true
                    MyNPCAgent := HealingBoss.NPCAgent?
                    MyChar := MyNPCAgent.GetFortCharacter[]
                    var MYLoc :vector3 = MyChar.GetTransform().Translation
                    MyRot := MyChar.GetTransform().Rotation
                    set MYLoc.Z = MYLoc.Z + 8.0
                then:
                    spawn:
                        SpawnEffect(HealingProp, 0.5, MYLoc, MyRot, vector3{X:=1.0,Y:=1.0,Z:=1.0})




    SpawnEffect(Prop:creative_prop_asset, LifeTime:float, Location : vector3 , Rotatoe : rotation, Scale:vector3)<suspends>:void=

        Tran:=transform{
            Translation := Location,
            Rotation := Rotatoe,
            Scale := Scale
        }

        SpawnedPropResult := SpawnProp(Prop,Tran)
        if(SpawnedProp := SpawnedPropResult(0)?):
            spawn:
                DeleteEffect(LifeTime,SpawnedProp)

    DeleteEffect(SleepTime : float, Prop : creative_prop)<suspends>:void=
        Sleep(SleepTime)
        Prop.Dispose()


