
using { /Fortnite.com/Devices }
using { /Verse.org/Simulation }
using { /UnrealEngine.com/Temporary/Diagnostics }
using { /UnrealEngine.com/Temporary/SpatialMath }

tower_manager_device := class(creative_device):

    @editable 
    buttons : []button_device = array{}
    @editable 
    references : []player_reference_device = array{}
    @editable 
    TowerProps : []creative_prop = array{}
    @editable
    TowerTops : []creative_prop = array{}

    @editable
    IdDevice  : npc_hp_cinematic_hud_device = npc_hp_cinematic_hud_device{}

    var AgentTowerMap : [agent]?int = map{}

    @editable
    GoldPercentPerUpgrade : float = 0.00005

    OnBegin<override>()<suspends>:void=

        Sleep(2.0)

        AllPlayers := GetPlayspace().GetPlayers()
        for(Player:AllPlayers):
            OnPlayerSpawned(Player)

        InitSpawners()
        
        GetPlayspace().PlayerRemovedEvent().Subscribe(OnPlayerRemoved)

        spawn:
            GenerateGold()

    OnPlayerSpawned(Agent:agent):void=

        if(PlayerExists := AgentTowerMap[Agent]):
            Print("Player Data Exists")
        else:
            if(MySpot := IdDevice.IdPerAgent[Agent]?):
                if(set AgentTowerMap[Agent] = option{MySpot}){LoadTowerLevel(Agent)}

        for(Button:buttons):
            Button.InteractedWithEvent.Subscribe(IncreaseTowerLevel)
            

    InitSpawners():void=
        Spawners := FindCreativeObjectsWithTag(spawners{})
        for(Obj : Spawners):
            if(spawner := player_spawner_device[Obj]):
                spawner.SpawnedEvent.Subscribe(OnPlayerSpawned)

    FindEmptyTowerSpot():int=

        var FreeSpot : int = 0
        var TempArr : []int = array{0,1,2,3}

        for(Key -> Value : AgentTowerMap):
            for(i:=0..TempArr.Length-1):
                if(Value = TempArr):
                    if(NewArr := TempArr.RemoveElement[i]):
                        set TempArr = NewArr
            
        for(Ind : TempArr):
            set FreeSpot = Ind
                        
        return FreeSpot

    OnPlayerRemoved(PlayerLeft:agent):void=
        if(Agent:=agent[PlayerLeft]):
            if:
                MyId := AgentTowerMap[PlayerLeft]?
            then:
                ScaleTower(MyId, 0)

            if(ActualPlayer := AgentTowerMap[Agent]):
                var TempPlayerMap : [agent]?int = map{}
                for(Key -> Value : AgentTowerMap,Key <> Agent):
                    set TempPlayerMap = ConcatenateMaps(TempPlayerMap, map{Key =>Value})
                set AgentTowerMap = TempPlayerMap

    LoadTowerLevel(Agent:agent):void=

        if:
            MyData := Data[player[Agent]]
            MyId := AgentTowerMap[Agent]?
            MyRefDevice := references[MyId]
        then:
            MyRefDevice.Register(Agent)
            TowerLevel := MyData.TowerLevel
            ScaleTower(MyId,TowerLevel)

    ScaleTower(TowerId:int,TowerLevel:int):void=

        if(Prop := TowerTops[TowerId]):
            var NewTranslation : vector3 =Prop.GetTransform().Translation
            set NewTranslation.Z = -28.0 + (96.0 * TowerLevel ) 
            Tran:=transform{
                Translation:=NewTranslation,
                Rotation:= Prop.GetTransform().Rotation,
                Scale:=Prop.GetTransform().Scale}
                if(Prop.TeleportTo[Tran]){}

        if(Prop := TowerProps[TowerId]):
            PropScale := Prop.GetTransform().Scale
            NewScale := vector3{X:=PropScale.X, Y:=PropScale.Y, Z:=(TowerLevel+1)*1.0}
            Tran:=transform{
                Translation:=Prop.GetTransform().Translation,
                Rotation:= Prop.GetTransform().Rotation,
                Scale:=NewScale}
                if(Prop.TeleportTo[Tran]){}

    IncreaseTowerLevel(Agent:agent):void=

        if:
            Player:=player[Agent]
            MyData := Data[Player]
        then:
            if(set Data[Player] = player_data{player_data_update<constructor>(MyData),TowerLevel := MyData.TowerLevel + 1}){}
            LoadTowerLevel(Agent)

    GenerateGold()<suspends>:void=

        loop:
            Sleep(3.0)

            AllPlayers := GetPlayspace().GetPlayers()
            for(Player:AllPlayers):
                if:
                    Player.IsActive[]
                    Agent:=agent[Player]
                    MyData := Data[Player]
                    MyTowerLevel := MyData.TowerLevel
                    MyTowerLevel > 0
                    MoneyEarned := MyData.PrestigeEarnings
                    GoldGenerated : float = MoneyEarned * (MyTowerLevel * 0.0005)
                then:
                    if(set Data[Player] = player_data{player_data_update<constructor>(MyData),Gold := MyData.Gold + GoldGenerated, PrestigeEarnings := MyData.PrestigeEarnings + GoldGenerated, TotalEarnings := MyData.TotalEarnings + GoldGenerated }){}
            

        
        


