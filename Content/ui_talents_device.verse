
using { /Fortnite.com/Devices }
using { /Verse.org/Simulation }
using { /UnrealEngine.com/Temporary/Diagnostics }
using { /UnrealEngine.com/Temporary/UI }
using { /UnrealEngine.com/Temporary/SpatialMath }
using { /Fortnite.com/UI }
using { /Verse.org/Colors }
using { /Fortnite.com/Playspaces }
using { /Fortnite.com/Characters }
using { /Verse.org/Assets }

HitsImpactsV2 := module:
    (HitsImpactsV2:)Textures<public>:=module:

wooden_pixel_gui := module:
    (wooden_pixel_gui:)textures<public> := module:
        (textures:)Panels<public> := module:
        (textures:)Icons2<public> := module:
            (Icons2:)Size32<public> := module:
            (Icons2:)Size16<public> := module:
            (Icons2:)Size8<public> := module:


ui_talents_device := class(creative_device):

    @editable
    UpgradeTalentAccolade : accolades_device = accolades_device{}

    @editable
    PlayerSpawners : []player_spawner_device = array{}

    @editable
    OpenAudio : audio_player_device = audio_player_device{}
    @editable
    CloseAudio : audio_player_device = audio_player_device{}
    @editable
    TabAudio : audio_player_device = audio_player_device{}

    @editable
    DamageUpgradeAudio : audio_player_device = audio_player_device{}
    @editable
    CritUpgradeAudio : audio_player_device = audio_player_device{}
    @editable
    CritDamageUpgradeAudio : audio_player_device = audio_player_device{}

    @editable
    ArmorUpgradeAudio : audio_player_device = audio_player_device{}
    @editable
    ShieldUpgradeAudio : audio_player_device = audio_player_device{}
    @editable
    MovementUpgradeAudio : audio_player_device = audio_player_device{}

    @editable
    GoldPerElimUpgradeAudio : audio_player_device = audio_player_device{}
    @editable
    GoldPerLevelUpgradeAudio : audio_player_device = audio_player_device{}
    @editable
    XpUpgradeAudio : audio_player_device = audio_player_device{}

    @editable
    PurchaseAudio : audio_player_device = audio_player_device{}
    @editable
    NegativeAudio : audio_player_device = audio_player_device{}

    # Just to pull % multtipliers
    @editable
    NPCManager : npc_manager_device = npc_manager_device{}
    @editable
    SignalRemote : signal_remote_manager_device = signal_remote_manager_device{}

    @editable
    ButtonClickSound : audio_player_device = audio_player_device{}

    @editable
    ShieldDevice : shield_device = shield_device{}
    @editable
    MovementSpeedDevice: movementspeed_device = movementspeed_device{}

    @editable
    DamageMissionTrigger : trigger_device = trigger_device{}
    @editable
    GoldMissionTrigger : trigger_device = trigger_device{}

    
    var DamagePriceM : float = 1.3
    var DamagePriceI : float = 4.0

    var CritPriceM : float = 1.3
    var CritPriceI : float = 10.0

    var ArmorPriceM : float = 7.5
    var ArmorPriceI : float = 64.0

    var AgilityPriceM : float = 24.2
    var AgilityPriceI : float = 190.0

    var GoldPriceM : float = 1.3
    var GoldPriceI : float = 4.0

    DamageMax : int = 0
    ArmorMax : int = 40
    CritMax : int = 15
    AgilityMax : int = 10
    GoldMax : int = 0

    var MyUI : [agent]?canvas = map{}

    var LevelBlocksPA : [agent]?[]text_block = map{}
    var PriceBlocksPA : [agent]?[]text_block = map{}
    var DescriptionBlocksPA : [agent]?[]text_block = map{}
    var BuyAmountTextBlocks : [agent]?[]text_block = map{}

    var BuyAmount : [agent][]int = map{}

    storedstring<localizes>(S:string) : message = "{S}"
    JustIntString<localizes>(I:int) : message = "{I}"
    LevelString<localizes>(I:int) : message = "Lv. {I}"
    DamageString<localizes>(I:int) : message = " {I}%"
    SecondString<localizes>(I:int) : message = "{I} Sec"
    HpString<localizes>(I:int) : message = "{I} hp"

    OnBegin<override>()<suspends>:void=

        SignalRemote.PrimarySignalEvent.Subscribe(OpenShopHandle)

    MakeEntireBackground (Agent:agent):canvas=

        WhiteSquare100 : texture = wooden_pixel_gui.textures.Panels.WhiteSquare100
        WhiteSquare75 : texture = wooden_pixel_gui.textures.Panels.WhiteSquare75
        GoldCoinTexture : texture = wooden_pixel_gui.textures.Icons2.Size8.t_CoinGold_8

        TitleBarTexture := texture_block :
            DefaultDesiredSize:=vector2{X:=16.0,Y:=120.0}
            DefaultImage:=wooden_pixel_gui.textures.Panels.t_TitleBar

        PurchaseMultiplierButton1 := button_quiet :
            DefaultText:=storedstring("x1")
        PurchaseMultiplierButton1.OnClick().Subscribe(SetBuy1)

        PurchaseMultiplierButton2 := button_quiet :
            DefaultText:=storedstring("x5")
        PurchaseMultiplierButton2.OnClick().Subscribe(SetBuy5)

        PurchaseMultiplierButton3 := button_quiet :
            DefaultText:=storedstring("x10")
        PurchaseMultiplierButton3.OnClick().Subscribe(SetBuy10)

        DamageButton := button_quiet :
            DefaultText:=storedstring("      Level Up      ")
        DamageButton.OnClick().Subscribe(DamageButtonHandle)

        ArmorButton := button_quiet :
            DefaultText:=storedstring("      Level Up      ")
        ArmorButton.OnClick().Subscribe(ArmorButtonHandle)

        CritButton := button_quiet :
            DefaultText:=storedstring("      Level Up      ")
        CritButton.OnClick().Subscribe(CritButtonHandle)

        AgilityButton := button_quiet :
            DefaultText:=storedstring("      Level Up      ")
        AgilityButton.OnClick().Subscribe(AgilityButtonHandle)

        GoldButton := button_quiet :
            DefaultText:=storedstring("      Level Up      ")
        GoldButton.OnClick().Subscribe(GoldButtonHandle)

        CloseWindowButton := button_quiet :
            DefaultText:=storedstring("      Close      ")
        CloseWindowButton.OnClick().Subscribe(RemoveHUD)

        var MyBuyAmountBlock : text_block = text_block{DefaultText:=storedstring(""), DefaultTextColor := MakeColorFromHex("FFFFFF")}

        if(MyBuyAmount := BuyAmount[Agent][0]):
            MyBuyAmountBlock.SetText(storedstring("Buy x{MyBuyAmount}"))
    
        DamageLevelText1 : text_block = text_block{DefaultText:=LevelString(2), DefaultTextColor := MakeColorFromHex("142706FF")}
        DamageLevelText2 : text_block = text_block{DefaultText:=LevelString(2), DefaultTextColor := MakeColorFromHex("529F17FF")}
        DamageTitleText1 : text_block = text_block{DefaultText:=storedstring("Damage"), DefaultTextColor := MakeColorFromHex("142706FF")}
        DamageTitleText2 : text_block = text_block{DefaultText:=storedstring("Damage"), DefaultTextColor := MakeColorFromHex("529F17FF")}
        DamagePriceText : text_block = text_block{DefaultText:=storedstring("1.54M"), DefaultTextColor := MakeColorFromHex("FFFFFFFF")}
        DamageDescription : text_block = text_block{DefaultText:=storedstring("  +12%\n Damage"), DefaultTextColor := MakeColorFromHex("FFFFFFFF")}
        DamageImageShadow : texture_block = texture_block{DefaultDesiredSize:=vector2{X:=100.0,Y:=100.0}, DefaultImage:=RavenIcons.SkillsSpellsandScores.A64x64.S79, DefaultTint:=MakeColorFromHex("110803FF") }
        DamageImage : texture_block = texture_block{DefaultDesiredSize:=vector2{X:=100.0,Y:=100.0}, DefaultImage:=RavenIcons.SkillsSpellsandScores.A64x64.S79}

        ArmorLevelText1 : text_block = text_block{DefaultText:=LevelString(2), DefaultTextColor := MakeColorFromHex("142706FF")}
        ArmorLevelText2 : text_block = text_block{DefaultText:=LevelString(2), DefaultTextColor := MakeColorFromHex("529F17FF")}
        ArmorTitleText1 : text_block = text_block{DefaultText:=storedstring("Armor"), DefaultTextColor := MakeColorFromHex("142706FF")}
        ArmorTitleText2 : text_block = text_block{DefaultText:=storedstring("Armor"), DefaultTextColor := MakeColorFromHex("529F17FF")}
        ArmorPriceText : text_block = text_block{DefaultText:=storedstring("1.540M"), DefaultTextColor := MakeColorFromHex("FFFFFFFF")}
        ArmorDescription : text_block = text_block{DefaultText:=storedstring("Increases the gold earned per enemy\n defeated."), DefaultTextColor := MakeColorFromHex("FFFFFFFF")}
        ArmorImageShadow : texture_block = texture_block{DefaultDesiredSize:=vector2{X:=100.0,Y:=100.0}, DefaultImage:=RavenIcons.MapMarkers.tile078, DefaultTint:=MakeColorFromHex("110803FF") }
        ArmorImage : texture_block = texture_block{DefaultDesiredSize:=vector2{X:=100.0,Y:=100.0}, DefaultImage:=RavenIcons.MapMarkers.tile078 }

        CritLevelText1 : text_block = text_block{DefaultText:=LevelString(2), DefaultTextColor := MakeColorFromHex("142706FF")}
        CritLevelText2 : text_block = text_block{DefaultText:=LevelString(2), DefaultTextColor := MakeColorFromHex("529F17FF")}
        CritTitleText1 : text_block = text_block{DefaultText:=storedstring("     Crit  "), DefaultTextColor := MakeColorFromHex("142706FF")}
        CritTitleText2 : text_block = text_block{DefaultText:=storedstring("     Crit  "), DefaultTextColor := MakeColorFromHex("529F17FF")}
        CritPriceText : text_block = text_block{DefaultText:=storedstring("1.54M"), DefaultTextColor := MakeColorFromHex("FFFFFFFF")}
        CritDescription : text_block = text_block{DefaultText:=storedstring("Increases damage dealt with weapons\n and abilities."), DefaultTextColor := MakeColorFromHex("FFFFFFFF")}
        CritImageShadow : texture_block = texture_block{DefaultDesiredSize:=vector2{X:=100.0,Y:=100.0}, DefaultImage:=RavenIcons.SkillsSpellsandScores.A64x64.S16, DefaultTint:=MakeColorFromHex("110803FF") }
        CritImage : texture_block = texture_block{DefaultDesiredSize:=vector2{X:=100.0,Y:=100.0}, DefaultImage:=RavenIcons.SkillsSpellsandScores.A64x64.S16 }

        AgilityLevelText1 : text_block = text_block{DefaultText:=LevelString(2), DefaultTextColor := MakeColorFromHex("142706FF")}
        AgilityLevelText2 : text_block = text_block{DefaultText:=LevelString(2), DefaultTextColor := MakeColorFromHex("529F17FF")}
        AgilityTitleText1 : text_block = text_block{DefaultText:=storedstring("Agility"), DefaultTextColor := MakeColorFromHex("142706FF")}
        AgilityTitleText2 : text_block = text_block{DefaultText:=storedstring("Agility"), DefaultTextColor := MakeColorFromHex("529F17FF")}
        AgilityPriceText : text_block = text_block{DefaultText:=storedstring("1.54M"), DefaultTextColor := MakeColorFromHex("FFFFFFFF")}
        AgilityDescription : text_block = text_block{DefaultText:=storedstring("Increases damage dealt with weapons\n and abilities."), DefaultTextColor := MakeColorFromHex("FFFFFFFF")}
        AgilityImageShadow : texture_block = texture_block{DefaultDesiredSize:=vector2{X:=100.0,Y:=100.0}, DefaultImage:=RavenIcons.SkillsSpellsandScores.A64x64.S2, DefaultTint:=MakeColorFromHex("110803FF") } 
        AgilityImage : texture_block = texture_block{DefaultDesiredSize:=vector2{X:=100.0,Y:=100.0}, DefaultImage:=RavenIcons.SkillsSpellsandScores.A64x64.S2 } # wooden_pixel_gui.textures.Icons2.Size32.t_Boots_32

        GoldLevelText1 : text_block = text_block{DefaultText:=LevelString(2), DefaultTextColor := MakeColorFromHex("142706FF")}
        GoldLevelText2 : text_block = text_block{DefaultText:=LevelString(2), DefaultTextColor := MakeColorFromHex("529F17FF")}
        GoldTitleText1 : text_block = text_block{DefaultText:=storedstring("  Gold  "), DefaultTextColor := MakeColorFromHex("142706FF")}
        GoldTitleText2 : text_block = text_block{DefaultText:=storedstring("  Gold  "), DefaultTextColor := MakeColorFromHex("529F17FF")}
        GoldPriceText : text_block = text_block{DefaultText:=storedstring("1.54M"), DefaultTextColor := MakeColorFromHex("FFFFFFFF")}
        GoldDescription : text_block = text_block{DefaultText:=storedstring("Increases damage dealt with weapons\n and abilities."), DefaultTextColor := MakeColorFromHex("FFFFFFFF")}
        GoldImageShadow : texture_block = texture_block{DefaultDesiredSize:=vector2{X:=100.0,Y:=100.0}, DefaultImage:=RavenIcons.MapMarkers.S1210, DefaultTint:=MakeColorFromHex("110803FF") }
        AGoldImage : texture_block = texture_block{DefaultDesiredSize:=vector2{X:=100.0,Y:=100.0}, DefaultImage:=RavenIcons.MapMarkers.S1210 }


        if(set LevelBlocksPA[Agent] = option{array{DamageLevelText1,DamageLevelText2,ArmorLevelText1,ArmorLevelText2,CritLevelText1,CritLevelText2,AgilityLevelText1,AgilityLevelText2,GoldLevelText1,GoldLevelText2}}){} 
        if(set PriceBlocksPA[Agent] = option{array{DamagePriceText,ArmorPriceText,CritPriceText,AgilityPriceText,GoldPriceText}}){} 
        if(set DescriptionBlocksPA[Agent] = option{array{DamageDescription,ArmorDescription,CritDescription,AgilityDescription,GoldDescription}}){} 
        if(set BuyAmountTextBlocks[Agent] = option{array{MyBuyAmountBlock}}){} 

        UpdateUiText(Agent)

        Row1 := stack_box:
            Orientation:=orientation.Horizontal
        Row2 := stack_box:
            Orientation:=orientation.Horizontal
        Row3 := stack_box:
            Orientation:=orientation.Horizontal

        Row1.AddWidget(CreateTalentSlot(DamageTitleText1,DamageTitleText2,DamageLevelText1,DamageLevelText2,DamagePriceText,DamageDescription,DamageImageShadow,DamageImage,DamageButton,15.0))
        Row2.AddWidget(CreateTalentSlot(CritTitleText1,CritTitleText2,CritLevelText1,CritLevelText2,CritPriceText,CritDescription,CritImageShadow,CritImage,CritButton,60.0))
        Row2.AddWidget(CreateTalentSlot(GoldTitleText1,GoldTitleText2,GoldLevelText1,GoldLevelText2,GoldPriceText,GoldDescription,GoldImageShadow,AGoldImage,GoldButton,52.0))
        Row3.AddWidget(CreateTalentSlot(ArmorTitleText1,ArmorTitleText2,ArmorLevelText1,ArmorLevelText2,ArmorPriceText,ArmorDescription,ArmorImageShadow,ArmorImage,ArmorButton,30.0))
        Row3.AddWidget(CreateTalentSlot(AgilityTitleText1,AgilityTitleText2,AgilityLevelText1,AgilityLevelText2,AgilityPriceText,AgilityDescription,AgilityImageShadow,AgilityImage,AgilityButton,30.0))

        Overlay1 := overlay:
            Slots:=array:
                overlay_slot:
                    HorizontalAlignment:=horizontal_alignment.Fill
                    VerticalAlignment:=vertical_alignment.Fill
                    Padding:=margin{Left:=-10.0,Bottom:=-10.0,Top:=-10.0,Right:=-10.0}
                    Widget:=texture_block:
                        DefaultImage:=WhiteSquare75
                        DefaultTint:=MakeColorFromHex("443525FF")
                overlay_slot:
                    HorizontalAlignment:=horizontal_alignment.Fill
                    VerticalAlignment:=vertical_alignment.Fill
                    Widget:=texture_block:
                        DefaultImage:=WhiteSquare75
                        DefaultTint:=MakeColorFromHex("6A4B2AFF")
                overlay_slot:
                    HorizontalAlignment:=horizontal_alignment.Fill
                    VerticalAlignment:=vertical_alignment.Fill
                    Widget := stack_box:
                        Orientation:=orientation.Vertical
                        Slots:=array:
                            stack_box_slot:
                                HorizontalAlignment:=horizontal_alignment.Fill
                                VerticalAlignment:=vertical_alignment.Fill
                                Widget := overlay:
                                    Slots:=array:
                                        overlay_slot:
                                            HorizontalAlignment:=horizontal_alignment.Fill
                                            VerticalAlignment:=vertical_alignment.Fill
                                            Padding:=margin{Left:=0.0,Bottom:=10.0,Top:=0.0,Right:=0.0}
                                            Widget:= TitleBarTexture
                                        overlay_slot:
                                            HorizontalAlignment:=horizontal_alignment.Center
                                            VerticalAlignment:=vertical_alignment.Center
                                            Padding:=margin{Left:=0.0,Bottom:=10.0,Top:=0.0,Right:=0.0}
                                            Widget := text_block :
                                                DefaultText:=storedstring("Talents")
                                                DefaultTextColor := MakeColorFromHex("643D19FF")

                            stack_box_slot:
                                HorizontalAlignment:=horizontal_alignment.Center
                                VerticalAlignment:=vertical_alignment.Center
                                Widget := Row1
                            stack_box_slot:
                                HorizontalAlignment:=horizontal_alignment.Center
                                VerticalAlignment:=vertical_alignment.Center
                                Widget := Row2
                            stack_box_slot:
                                HorizontalAlignment:=horizontal_alignment.Center
                                VerticalAlignment:=vertical_alignment.Center
                                Widget := Row3

       

                            stack_box_slot:
                                Padding:=margin{Left:=0.0,Bottom:=5.0,Top:=0.0,Right:=0.0}
                                HorizontalAlignment:=horizontal_alignment.Center
                                VerticalAlignment:=vertical_alignment.Center
                                Widget:=CloseWindowButton

                overlay_slot:
                    Padding:=margin{Left:=0.0,Bottom:=230.0,Top:=0.0,Right:=12.0}
                    HorizontalAlignment:=horizontal_alignment.Right
                    VerticalAlignment:=vertical_alignment.Center
                    Widget:=stack_box:
                        Orientation:=orientation.Vertical
                        Slots:=array:
                            stack_box_slot:
                                Widget:=MyBuyAmountBlock
                            stack_box_slot:
                                Padding:=margin{Left:=0.0,Bottom:=10.0,Top:=10.0,Right:=0.0}
                                HorizontalAlignment:=horizontal_alignment.Fill
                                VerticalAlignment:=vertical_alignment.Fill
                                Widget:=PurchaseMultiplierButton1
                            stack_box_slot:
                                Padding:=margin{Left:=0.0,Bottom:=10.0,Top:=10.0,Right:=0.0}
                                HorizontalAlignment:=horizontal_alignment.Fill
                                VerticalAlignment:=vertical_alignment.Fill
                                Widget:=PurchaseMultiplierButton2
                            stack_box_slot:
                                Padding:=margin{Left:=0.0,Bottom:=10.0,Top:=10.0,Right:=0.0}
                                HorizontalAlignment:=horizontal_alignment.Fill
                                VerticalAlignment:=vertical_alignment.Fill
                                Widget:=PurchaseMultiplierButton3


        # This is the master canvas
        MyCanvas:=canvas:
            Slots:=array:
                canvas_slot:
                    Anchors:=anchors{Minimum:=vector2{X:=0.2,Y:=0.05}, Maximum:=vector2{X:=0.8,Y:=0.95}}
                    Widget:= Overlay1
                    
        return MyCanvas


    CreateTalentSlot(TitleTextBlock1:text_block, TitleTextBlock2:text_block,LevelTextBlock1:text_block, LevelTextBlock2:text_block, PriceTextBlock : text_block, DescriptionTextBlock:text_block, ImageShadow: texture_block, Image:texture_block, UpgradeButton : button_quiet, GoldStackLeftOffset : float):stack_box_slot=

        WhiteSquare100 : texture = wooden_pixel_gui.textures.Panels.WhiteSquare100
        WhiteSquare75 : texture = wooden_pixel_gui.textures.Panels.WhiteSquare75
        GoldCoinTexture : texture = wooden_pixel_gui.textures.Icons2.Size8.t_CoinGold_8

        MainStackBox := stack_box_slot:
            HorizontalAlignment:=horizontal_alignment.Center
            VerticalAlignment:=vertical_alignment.Center
            Padding:=margin{Left:=20.0,Bottom:=20.0,Top:=20.0,Right:=20.0}
            Widget := overlay:
                Slots:=array:
                    overlay_slot: # Background
                        HorizontalAlignment:=horizontal_alignment.Fill
                        VerticalAlignment:=vertical_alignment.Fill
                        Padding:=margin{Left:=-10.0,Bottom:=-10.0,Top:=-10.0,Right:=-10.0}
                        Widget:=texture_block:
                            DefaultImage:=WhiteSquare75
                            DefaultTint:=MakeColorFromHex("49321CFF")
                    overlay_slot: # Containing Stack Box
                        HorizontalAlignment:=horizontal_alignment.Fill
                        VerticalAlignment:=vertical_alignment.Fill
                        Widget := stack_box:
                            Orientation:=orientation.Vertical
                            Slots:=array:
                                stack_box_slot:  # Title Bar
                                    HorizontalAlignment:=horizontal_alignment.Fill
                                    VerticalAlignment:=vertical_alignment.Fill
                                    Padding:=margin{Left:=0.0,Top:=0.0,Right:=0.0,Bottom:=10.0}
                                    Widget := overlay:
                                        Slots:=array:
                                            overlay_slot: # Title Background
                                                HorizontalAlignment:=horizontal_alignment.Fill
                                                VerticalAlignment:=vertical_alignment.Fill
                                                Widget:=texture_block:
                                                    DefaultDesiredSize:=vector2{X:=32.0,Y:=32.0}
                                                    DefaultImage:=WhiteSquare75
                                                    DefaultTint:=MakeColorFromHex("8B5C2AFF")   
                                            overlay_slot: # TitleStackBox
                                                HorizontalAlignment:=horizontal_alignment.Fill
                                                VerticalAlignment:=vertical_alignment.Fill
                                                Padding:=margin{Left:=5.0,Top:=0.0,Right:=0.0,Bottom:=0.0}
                                                Widget := stack_box:
                                                    Orientation:=orientation.Horizontal
                                                    Slots:=array:
                                                        stack_box_slot:
                                                            Widget := overlay: # Lv Text
                                                                Slots:=array:
                                                                    overlay_slot:
                                                                        Padding:=margin{Left:=3.0,Top:=3.0,Right:=0.0,Bottom:=0.0}
                                                                        Widget := LevelTextBlock1
                                                                    overlay_slot:
                                                                        Widget := LevelTextBlock2
                                                        stack_box_slot:
                                                            Padding:=margin{Left:=10.0,Top:=0.0,Right:=0.0,Bottom:=0.0}
                                                            Widget := overlay: # Lv Text
                                                                Slots:=array:
                                                                    overlay_slot:
                                                                        Padding:=margin{Left:=3.0,Top:=3.0,Right:=0.0,Bottom:=0.0}
                                                                        Widget := TitleTextBlock1
                                                                    overlay_slot:
                                                                        Widget := TitleTextBlock2
                                                        stack_box_slot:
                                                            HorizontalAlignment:=horizontal_alignment.Right
                                                            VerticalAlignment:=vertical_alignment.Center
                                                            Padding:=margin{Left:=GoldStackLeftOffset,Top:=0.0,Right:=10.0,Bottom:=0.0}
                                                            Widget := stack_box:
                                                                Orientation:=orientation.Horizontal
                                                                Slots:=array:
                                                                    stack_box_slot:
                                                                        Widget := overlay:
                                                                            Slots:=array:
                                                                                overlay_slot: # Gold Coin Shadow
                                                                                    HorizontalAlignment:=horizontal_alignment.Center
                                                                                    VerticalAlignment:=vertical_alignment.Center
                                                                                    Padding:=margin{Left:=3.0,Top:=3.0,Right:=0.0,Bottom:=0.0}
                                                                                    Widget:=texture_block:
                                                                                        DefaultDesiredSize:=vector2{X:=32.0,Y:=32.0}
                                                                                        DefaultImage:=GoldCoinTexture
                                                                                        DefaultTint:=MakeColorFromHex("828282FF")
                                                                                overlay_slot: # Gold Coin 
                                                                                    HorizontalAlignment:=horizontal_alignment.Center
                                                                                    VerticalAlignment:=vertical_alignment.Center
                                                                                    Padding:=margin{Left:=0.0,Top:=0.0,Right:=0.0,Bottom:=0.0}
                                                                                    Widget:=texture_block:
                                                                                        DefaultImage:=GoldCoinTexture
                                                                    stack_box_slot:
                                                                        Widget := PriceTextBlock
                                stack_box_slot:
                                    HorizontalAlignment:=horizontal_alignment.Center
                                    VerticalAlignment:=vertical_alignment.Center
                                    Widget := stack_box:
                                        Orientation:=orientation.Horizontal
                                        Slots:=array:
                                            stack_box_slot:
                                                Widget := overlay: #ImageOverlay
                                                    Slots:=array:
                                                        overlay_slot:
                                                            Padding:=margin{Left:=20.0,Top:=6.0,Right:=20.0,Bottom:=0.0}
                                                            Widget:=ImageShadow
                                                        overlay_slot:
                                                            Padding:=margin{Left:=20.0,Top:=0.0,Right:=20.0,Bottom:=0.0}
                                                            Widget:=Image
                                            stack_box_slot:
                                                Widget:=DescriptionTextBlock
                                stack_box_slot:
                                    Padding:=margin{Left:=0.0,Top:=10.0,Right:=0.0,Bottom:=0.0}
                                    Widget:=UpgradeButton
            

    OpenShopHandle(Agent:agent):void=

        RemoveFortHud(Agent)
        spawn:
            ShowUi(Agent)
        

    ShowUi(Agent:agent)<suspends>:void=

        #BackgroundHMDAttack.Show(Agent)
        if(set BuyAmount[Agent] = array{1,1,1,1,1,1}){}

        MyCanvas: canvas = MakeEntireBackground(Agent)

        if(PlayerUI:=GetPlayerUI[player[Agent]]):
            
            #Removing Old Widget to stop duplicates
            if(AddedCanvas := MyUI[Agent]?):
                PlayerUI.RemoveWidget(AddedCanvas)
                AddedCanvas.SetVisibility(widget_visibility.Visible)
                AddedCanvas.SetVisibility(widget_visibility.Hidden)

            PlayerUI.AddWidget(MyCanvas,player_ui_slot{InputMode:=ui_input_mode.All}) # want it to appear on top
            UpdateUiText(Agent)
            OpenAudio.Play(Agent)
            MyCanvas.SetVisibility(widget_visibility.Hidden)
            MyCanvas.SetVisibility(widget_visibility.Visible)
        if(set MyUI[Agent]= option{MyCanvas}){}
        

    RemoveHUD(Wdata:widget_message):void =
        
        Print("Removing")
        if(Player:=Wdata.Player, PlayerUI:=GetPlayerUI[Player],SelectedButton:=Wdata.Source,Agent:=agent[Player]):
            if(MyActualUI := MyUI[Agent]?):

                CloseAudio.Play(Agent)

                PlayerUI.RemoveWidget(MyActualUI)
                MyActualUI.SetVisibility(widget_visibility.Visible)
                MyActualUI.SetVisibility(widget_visibility.Hidden)

                if(set MyUI[Agent] = false){}

                if(set LevelBlocksPA[Agent] = false){} 
                if(set PriceBlocksPA[Agent] = false){}
                if(set DescriptionBlocksPA[Agent] = false){}

                ShowFortHud(Agent)


    UpdateUiText(Agent:agent):void=

        if:
            DamageLevel1: text_block = LevelBlocksPA[Agent]?[0]
            DamageLevel2: text_block = LevelBlocksPA[Agent]?[1]
            ArmorLevel1: text_block = LevelBlocksPA[Agent]?[2]
            ArmorLevel2: text_block = LevelBlocksPA[Agent]?[3]
            CritLevel1: text_block = LevelBlocksPA[Agent]?[4]
            CritLevel2: text_block = LevelBlocksPA[Agent]?[5]
            AgilityLevel1: text_block = LevelBlocksPA[Agent]?[6]
            AgilityLevel2: text_block = LevelBlocksPA[Agent]?[7]
            GoldLevel1: text_block = LevelBlocksPA[Agent]?[8]
            GoldLevel2: text_block = LevelBlocksPA[Agent]?[9]

            DamagePrice : text_block = PriceBlocksPA[Agent]?[0]
            ArmorPrice : text_block = PriceBlocksPA[Agent]?[1]
            CritPrice : text_block = PriceBlocksPA[Agent]?[2]
            AgilityPrice : text_block = PriceBlocksPA[Agent]?[3]
            GoldPrice : text_block = PriceBlocksPA[Agent]?[4]

            DamageDescription : text_block = DescriptionBlocksPA[Agent]?[0]
            ArmorDescription : text_block = DescriptionBlocksPA[Agent]?[1]
            CritDescription : text_block = DescriptionBlocksPA[Agent]?[2]
            AgilityDescription : text_block = DescriptionBlocksPA[Agent]?[3]
            GoldDescription : text_block = DescriptionBlocksPA[Agent]?[4]

            Player := player[Agent]
            PlayerData := Data[Agent]
            
        then:

            DamageLevel1.SetText(LevelString(PlayerData.TrueDamage))
            DamageLevel2.SetText(LevelString(PlayerData.TrueDamage))
            RoundedDamagePL := ((NPCManager.TrueDamagePercentPerLevel) + (PlayerData.DamagePL * PlayerData.DamagePLLevel)) * 100.0
            DamageDescription.SetText(storedstring("  +{formatter.Abbreviate(RoundedDamagePL, true)}%\n Damage"))

            var TotalAdded : float = 0.0
            if(BuyAmount1 := BuyAmount[Agent][0]):
                for(i2:=0..BuyAmount1-1):

                    var Price : float = DamagePriceI
                    for(i:=0..PlayerData.TrueDamage+BuyAmount1-2):
                        set Price = Price * DamagePriceM
                    set TotalAdded += Price
                    
            if(DamageMax > PlayerData.TrueDamage or DamageMax = 0):
                DamagePrice.SetText(storedstring(formatter.Abbreviate((TotalAdded * 1.0),false)))
            else:
                DamagePrice.SetText(storedstring("Max"))


            ArmorLevel1.SetText(LevelString(PlayerData.PlayerArmor))
            ArmorLevel2.SetText(LevelString(PlayerData.PlayerArmor))
            if(RoundedArmorPL := Round[PlayerData.ArmorPL * 100.0]){ ArmorDescription.SetText(storedstring("  +{RoundedArmorPL}\n Armor"))}

            var TotalAdded1 : float = 0.0

            if(var BuyAmount1 :int = BuyAmount[Agent][2]):
                if(BuyAmount1 + PlayerData.PlayerArmor > ArmorMax):
                    set BuyAmount1 = ArmorMax - PlayerData.PlayerArmor

                for(i2:=0..BuyAmount1-1):
                    var Price1 : float = ArmorPriceI
                    for(i:=0..PlayerData.PlayerArmor+BuyAmount1-2):
                        set Price1 = Price1 * ArmorPriceM
                    set TotalAdded1 += Price1
                    
            if(ArmorMax > PlayerData.PlayerArmor or ArmorMax = 0):
                ArmorPrice.SetText(storedstring(formatter.Abbreviate((TotalAdded1 * 1.0),false)))
            else:
                ArmorPrice.SetText(storedstring("Max"))


            CritLevel1.SetText(LevelString(PlayerData.CritLevel))
            CritLevel2.SetText(LevelString(PlayerData.CritLevel))
            RoundedCritPL := (NPCManager.CritChancePerLevel + (PlayerData.CritPL * PlayerData.CritPLLevel)) * 100.0
            CritDescription.SetText(storedstring("+ {formatter.Abbreviate(RoundedCritPL,true)}% Chance\n+{formatter.Abbreviate(RoundedCritPL,true)}% Damage"))

            var TotalAdded2 : float = 0.0
            
            if(var BuyAmount1 : int = BuyAmount[Agent][3]):

                if(BuyAmount1 + PlayerData.CritLevel > CritMax):
                    set BuyAmount1 = CritMax - PlayerData.CritLevel

                for(i2:=0..BuyAmount1-1):
                    var Price2 : float = CritPriceI
                    for(i:=0..PlayerData.CritLevel+BuyAmount1-2):
                        set Price2 = Price2 * CritPriceM
                    set TotalAdded2 += Price2

            if(CritMax > PlayerData.CritLevel or CritMax = 0):
                CritPrice.SetText(storedstring(formatter.Abbreviate((TotalAdded2 * 1.0),false)))
            else:
                CritPrice.SetText(storedstring("Max"))

            AgilityLevel1.SetText(LevelString(PlayerData.MovementSpeed))
            AgilityLevel2.SetText(LevelString(PlayerData.MovementSpeed))
            AgilityDescription.SetText(storedstring("+ Movement\n   Speed"))

            var TotalAdded3 : float = 0.0
            if(var BuyAmount1 : int = BuyAmount[Agent][4]):

                if(BuyAmount1 + PlayerData.MovementSpeed > AgilityMax):
                    set BuyAmount1 = AgilityMax - PlayerData.MovementSpeed

                for(i2:=0..BuyAmount1-1):
                    var Price3 : float = AgilityPriceI
                    for(i:=0..PlayerData.MovementSpeed+BuyAmount1-2):
                        set Price3 = Price3 * AgilityPriceM
                    set TotalAdded3 += Price3

            if(AgilityMax > PlayerData.MovementSpeed or AgilityMax = 0):
                AgilityPrice.SetText(storedstring(formatter.Abbreviate((TotalAdded3 * 1.0),false)))
            else:
                AgilityPrice.SetText(storedstring("Max"))
            

            GoldLevel1.SetText(LevelString(PlayerData.GoldPerElim))
            GoldLevel2.SetText(LevelString(PlayerData.GoldPerElim))
            RoundedGoldPL := ((NPCManager.GoldPerElimInitial ) + (PlayerData.GoldPL * PlayerData.GoldPLLevel)) * 100.0
            GoldDescription.SetText(storedstring("+{formatter.Abbreviate(RoundedGoldPL, true)}% Gold"))

            
            var TotalAdded4 : float = 0.0
            if(BuyAmount1 := BuyAmount[Agent][1]):

                for(i2:=0..BuyAmount1-1):
                    var Price4 : float = GoldPriceI
                    for(i:=0..PlayerData.GoldPerElim+BuyAmount1-2):
                        set Price4 = Price4 * GoldPriceM
                    set TotalAdded4 += Price4

            if(GoldMax > PlayerData.GoldPerElim or GoldMax = 0):
                GoldPrice.SetText(storedstring(formatter.Abbreviate((TotalAdded4 * 1.0),false)))
            else:
                GoldPrice.SetText(storedstring("Max"))



    DamageButtonHandle(Wdata:widget_message):void=
        if(Player:=Wdata.Player, PlayerUI:=GetPlayerUI[Player],SelectedButton:=Wdata.Source,Agent:=agent[Player]):
            if (MyData := Data[Player]):
                if(DamageMax > MyData.TrueDamage or DamageMax = 0):
                    
                    if(BuyAmount1 := BuyAmount[Agent][0]):

                        var TotalAdded : float = 0.0

                        for(i2:=0..BuyAmount1-1):
                            var Price : float = DamagePriceI
                            for(i:=0..MyData.TrueDamage+BuyAmount1-2):
                                set Price = Price * DamagePriceM
                            set TotalAdded += Price

                        
                        if(MyData.Gold >= TotalAdded):
                            if:
                                OldData := Data[Player]
                                set Data[Player] = player_data{player_data_update<constructor>(OldData),TrueDamage := OldData.TrueDamage + BuyAmount1, Gold := OldData.Gold - TotalAdded}
                                    
                            then:
                                DamageUpgradeAudio.Play(Agent)
                                CompleteBoughtAnims(Agent)
                                DamageMissionTrigger.Trigger(Agent)
                        else:
                            NegativeAudio.Play(Agent)

    GoldButtonHandle(Wdata:widget_message):void=
        if(Player:=Wdata.Player, PlayerUI:=GetPlayerUI[Player],SelectedButton:=Wdata.Source,Agent:=agent[Player]):
            if (MyData := Data[Player]):
                if(GoldMax > MyData.GoldPerElim or GoldMax = 0):
                    
                    if(BuyAmount1 := BuyAmount[Agent][1]):

                        var TotalAdded : float = 0.0
                        for(i2:=0..BuyAmount1-1):
                            var Price : float = GoldPriceI
                            for(i:=0..MyData.GoldPerElim+BuyAmount1-2):
                                set Price = Price * GoldPriceM
                            set TotalAdded += Price

                        
                        if(MyData.Gold >= TotalAdded):
                            if:
                                OldData := Data[Player]
                                set Data[Player] = player_data{player_data_update<constructor>(OldData),GoldPerElim := OldData.GoldPerElim + BuyAmount1, Gold := OldData.Gold - TotalAdded}
                            then:
                                GoldPerElimUpgradeAudio.Play(Agent)
                                CompleteBoughtAnims(Agent)
                                GoldMissionTrigger.Trigger(Agent)
                        else:
                            NegativeAudio.Play(Agent)
                            Print("Not Enought Units")


    ArmorButtonHandle(Wdata:widget_message):void=
        if(Player:=Wdata.Player, PlayerUI:=GetPlayerUI[Player],SelectedButton:=Wdata.Source,Agent:=agent[Player]):
            if (MyData := Data[Player]):
                if(ArmorMax > MyData.PlayerArmor or ArmorMax = 0):
                    
                    if(var BuyAmount1 : int = BuyAmount[Agent][2]):

                        if(BuyAmount1 + MyData.PlayerArmor > ArmorMax):
                            set BuyAmount1 = ArmorMax - MyData.PlayerArmor

                        var TotalAdded : float = 0.0
                        for(i2:=0..BuyAmount1-1):
                            var Price : float = ArmorPriceI
                            for(i:=0..MyData.PlayerArmor+BuyAmount1-2):
                                set Price = Price * ArmorPriceM
                            set TotalAdded += Price


                        if(MyData.Gold >= TotalAdded):
                            if:
                                OldData := Data[Player]
                                set Data[Player] = player_data{player_data_update<constructor>(OldData),PlayerArmor := OldData.PlayerArmor + BuyAmount1, Gold := OldData.Gold - TotalAdded}
                            then:
                                ArmorUpgradeAudio.Play(Agent)
                                CompleteBoughtAnims(Agent)
                        else:
                            NegativeAudio.Play(Agent)
                            Print("Not Enought Units")


    CritButtonHandle(Wdata:widget_message):void=
        if(Player:=Wdata.Player, PlayerUI:=GetPlayerUI[Player],SelectedButton:=Wdata.Source,Agent:=agent[Player]):
            if (MyData := Data[Player]):
                if(CritMax > MyData.CritLevel or CritMax = 0):
                    
                    if(var BuyAmount1 : int= BuyAmount[Agent][3]):

                        if(BuyAmount1 + MyData.CritLevel > CritMax):
                            set BuyAmount1 = CritMax - MyData.CritLevel

                        var TotalAdded : float = 0.0
                        for(i2:=0..BuyAmount1-1):
                            var Price : float = CritPriceI
                            for(i:=0..MyData.CritLevel+BuyAmount1-2):
                                set Price = Price * CritPriceM
                            set TotalAdded += Price

                        
                        if(MyData.Gold >= TotalAdded):
                            if:
                                OldData := Data[Player]
                                set Data[Player] = player_data{player_data_update<constructor>(OldData),CritLevel := OldData.CritLevel + BuyAmount1, Gold := OldData.Gold - TotalAdded}
                            then:
                                CritDamageUpgradeAudio.Play(Agent)
                                CompleteBoughtAnims(Agent)

                        else:
                            NegativeAudio.Play(Agent)
                            Print("Not Enought Units")

            
    AgilityButtonHandle(Wdata:widget_message):void=

        if(Player:=Wdata.Player, PlayerUI:=GetPlayerUI[Player],SelectedButton:=Wdata.Source,Agent:=agent[Player]):
            if (MyData := Data[Player]):
                if(AgilityMax > MyData.MovementSpeed or AgilityMax = 0):
                    
                    if(var BuyAmount1 : int = BuyAmount[Agent][4]):

                        if(BuyAmount1 + MyData.MovementSpeed > AgilityMax):
                            set BuyAmount1 = AgilityMax - MyData.MovementSpeed

                        var TotalAdded : float = 0.0
                        for(i2:=0..BuyAmount1-1):
                            var Price : float = AgilityPriceI
                            
                            for(i:=0..MyData.MovementSpeed+BuyAmount1-2):
                                set Price = Price * AgilityPriceM
                            set TotalAdded += Price

                        
                        if(MyData.Gold >= TotalAdded ):
                            if:
                                OldData := Data[Player]
                                set Data[Player] = player_data{player_data_update<constructor>(OldData),MovementSpeed := OldData.MovementSpeed + BuyAmount1, Gold := OldData.Gold - TotalAdded}
                            then:
                                MovementUpgradeAudio.Play(Agent)
                                CompleteBoughtAnims(Agent)
                                MovementSpeedDevice.SetClass(Agent)
        
                        else:
                            NegativeAudio.Play(Agent)
                            Print("Not Enought Units")

    CompleteBoughtAnims(Agent:agent):void= 
        UpdateUiText(Agent)
        UpgradeTalentAccolade.Award(Agent)

    SetBuy1(Wdata:widget_message):void=
        if(Player:=Wdata.Player, PlayerUI:=GetPlayerUI[Player],SelectedButton:=Wdata.Source,Agent:=agent[Player]):
            if(set BuyAmount[Agent] = array{1,1,1,1,1}){}
            if(BuyText := BuyAmountTextBlocks[Agent]?[0]):
                BuyText.SetText(storedstring("Buy x1"))
            UpdateUiText(Agent)

    SetBuy5(Wdata:widget_message):void=
        if(Player:=Wdata.Player, PlayerUI:=GetPlayerUI[Player],SelectedButton:=Wdata.Source,Agent:=agent[Player]):
            if(set BuyAmount[Agent] = array{5,5,5,5,5}){}
            if(BuyText := BuyAmountTextBlocks[Agent]?[0]):
                BuyText.SetText(storedstring("Buy x5"))
            UpdateUiText(Agent)

    SetBuy10(Wdata:widget_message):void=
        if(Player:=Wdata.Player, PlayerUI:=GetPlayerUI[Player],SelectedButton:=Wdata.Source,Agent:=agent[Player]):
            if(set BuyAmount[Agent] = array{10,10,10,10,10}){}
            if(BuyText := BuyAmountTextBlocks[Agent]?[0]):
                BuyText.SetText(storedstring("Buy x10"))
            UpdateUiText(Agent)

        



