using { /Verse.org/Simulation }
using { /UnrealEngine.com/Temporary/SpatialMath }
using { /Fortnite.com/Characters }
using { /Fortnite.com/Devices }
using { /UnrealEngine.com/Temporary/UI }
using { /Fortnite.com/UI }

(Listenable : listenable(agent)).SubscribeAgent(OutputFunc : tuple(agent, t)->void, ExtraData : t where t:type) : cancelable =
    Wrapper := wrapper_agent(t){ExtraData := ExtraData, OutputFunc := OutputFunc}
    Listenable.Subscribe(Wrapper.InputFunc)
 
wrapper_agent(t : type) := class():
    ExtraData : t;
    OutputFunc : tuple(agent, t) -> void
    InputFunc(Agent : agent):void = OutputFunc(Agent, ExtraData)

(Listenable : listenable(tuple())).SubscribeTuple(OutputFunc : tuple(tuple(), t)->void, ExtraData : t where t:type) : cancelable =
    Wrapper := wrapper_tuple(t){ExtraData := ExtraData, OutputFunc := OutputFunc}
    Listenable.Subscribe(Wrapper.InputFunc)

wrapper_tuple(t : type) := class():
    ExtraData : t;
    OutputFunc : tuple(tuple(), t) -> void
    InputFunc(Agent : tuple()):void = OutputFunc(Agent, ExtraData)

(Listenable : listenable(device_ai_interaction_result)).SubscribeAi(OutputFunc : tuple(device_ai_interaction_result, t)->void, ExtraData : t where t:type) : cancelable =
    Wrapper := wrapper_ai(t){ExtraData := ExtraData, OutputFunc := OutputFunc}
    Listenable.Subscribe(Wrapper.InputFunc)

wrapper_ai(t : type) := class():
    ExtraData : t;
    OutputFunc : tuple(device_ai_interaction_result, t) -> void
    InputFunc(Agent : device_ai_interaction_result):void = OutputFunc(Agent, ExtraData)

(Listenable : listenable(device_ai_interaction_result)).SubscribeAiDevice(OutputFunc : tuple(device_ai_interaction_result, t)->void, ExtraData : t where t:type) : cancelable =
    Wrapper := wrapper_deviceai(t){ExtraData := ExtraData, OutputFunc := OutputFunc}
    Listenable.Subscribe(Wrapper.InputFunc)
 
wrapper_deviceai(t : type) := class():
    ExtraData : t;
    OutputFunc : tuple(device_ai_interaction_result, t) -> void
    InputFunc(WM : device_ai_interaction_result):void = OutputFunc(WM, ExtraData)

(Listenable : listenable(widget_message)).SubscribeMessage(OutputFunc : tuple(widget_message, t)->void, ExtraData : t where t:type) : cancelable =
    Wrapper := wrapper_message(t){ExtraData := ExtraData, OutputFunc := OutputFunc}
    Listenable.Subscribe(Wrapper.InputFunc)
 
wrapper_message(t : type) := class():
    ExtraData : t;
    OutputFunc : tuple(widget_message, t) -> void
    InputFunc(WM : widget_message):void = OutputFunc(WM, ExtraData)
 
(Listenable : listenable(tuple())).SubscribeEmpty(OutputFunc : t -> void, ExtraData : t where t:type) : cancelable =
    Wrapper := wrapper_empty(t) {ExtraData := ExtraData, OutputFunc := OutputFunc}
    Listenable.Subscribe(Wrapper.InputFunc)
 
wrapper_empty(t : type) := class():
    ExtraData : t;
    OutputFunc : t -> void
    InputFunc():void = OutputFunc(ExtraData)

(Listenable : listenable(?agent)).SubscribeMaybeAgent(OutputFunc : tuple(?agent, t)->void, ExtraData : t where t:type) : cancelable =
    Wrapper := wrapper_maybeagent(t){ExtraData := ExtraData, OutputFunc := OutputFunc}
    Listenable.Subscribe(Wrapper.InputFunc)

wrapper_maybeagent(t : type) := class():
    ExtraData : t;
    OutputFunc : tuple(?agent, t) -> void
    InputFunc(Agent : ?agent):void = OutputFunc(Agent, ExtraData)

RemoveKeyFromMapAgentInt(ExampleMap:[agent]?int, ElementToRemove:agent):[agent]?int=
    var NewMap:[agent]?int = map{}
    # Concatenate Keys from ExampleMap into NewMap, excluding ElementToRemove
    for (Key -> Value : ExampleMap, Key <> ElementToRemove):
        set NewMap = ConcatenateMaps(NewMap, map{Key => Value})
    return NewMap

NormalizeVector(DifferenceVector:vector3):vector3=
    magnitude1 : float = Sqrt(Pow(DifferenceVector.X, 2.0) + Pow(DifferenceVector.Y,2.0) + Pow(DifferenceVector.Z,2.0))
    NormDifVector := vector3 {
        X:= DifferenceVector.X / magnitude1,
        Y:= DifferenceVector.Y / magnitude1,
        Z:= DifferenceVector.Z / magnitude1
        }
    return NormDifVector

DamageCalculation(Fort:fort_character):float= 
    var DamageAfterArmor: float = 0.0
    if(Agent:= Fort.GetAgent[]):
        if (Player := player[Agent]):
            if(ArmorAmount := Data[Player].PlayerArmor * 1.0):
                if(CosArmor := Data[Player].CosmeticArmor / 100.0)  :
                    #Print("ArmorAmount:{ArmorAmount + CosArmor}")
                    set DamageAfterArmor = 100.0 / (100.0 + ((ArmorAmount + ( ArmorAmount * CosArmor))* 5.0))
    #Print("DamageMultuple:{DamageAfterArmor}")
    return DamageAfterArmor

RemoveFortHud(Agent:agent):void=

    #Print("Removing HUD")
    var HudElementsToHide: []hud_element_identifier = array{creative_hud_identifier_all{}, creative_hud_identifier_mimimap{}, creative_hud_identifier_interaction_prompts{}}
 
    if(Player:= player[Agent]):
        if(FortniteCharacter : fort_character = Player.GetFortCharacter[]):
            if(PlayerUI := GetPlayerUI[Player]):
                #PlayerUI.HideHUDElements(HudElementsToHide)

ShowFortHud(Agent:agent):void=

    #Print("Showing HUD")
    var HudElementsToShow: []hud_element_identifier = array{creative_hud_identifier_all{}}
    var HudElimentsToNotShow : []hud_element_identifier = array{hud_identifier_world_resource_wood{}, hud_identifier_world_resource_stone{}, hud_identifier_world_resource_metal{},creative_hud_identifier_storm_timer{},creative_hud_identifier_elimination_counter{}}
 
    if(Player:= player[Agent]):
        if(FortniteCharacter : fort_character = Player.GetFortCharacter[]):
            if(PlayerUI := GetPlayerUI[Player]):
                #PlayerUI.ShowHUDElements(HudElementsToShow)
                #PlayerUI.HideHUDElements(HudElimentsToNotShow)


GetFormattedMoney(Amount:float):tuple(float, string)=
    var ReturnNum : float = 0.0
    var TempAmount : float = Amount * 1.0
    var Suffix : string = "0"
    var Prefix : string = ""
    var SigFig : float = 0.0
    var Format : logic = false
    if(Amount >= 1000000000.0):
        set TempAmount = TempAmount / 1000000000.0
        set SigFig = 1000000000.0
        set Prefix = "B"
        set Format = true
    else if(Amount >= 1000000.0):
        set TempAmount = TempAmount / 1000000.0
        set SigFig = 1000000.0
        set Prefix = "M"
        set Format = true
    else if(Amount >= 1000.0):
        set SigFig = 1000.0
        set TempAmount = TempAmount / 1000.0
        set Prefix = "K"
        set Format = true
        set ReturnNum = Amount
    if(Format = true, MaybeNum := TempAmount, M := Amount - (MaybeNum * SigFig), S := "{M}", set ReturnNum = MaybeNum):
        var Divisor : float = 10.0
        if(M > 100.0):
            set Divisor = 1000.0
        else if(M > 10.0):
            set Divisor = 100.0
        if(ReturnNum > 10.0):
            set Divisor = Divisor * 10.0
        #Print("S is {S}")
        var Remainder : float = ((Amount*1.0)/100.0)
        #Print("Remainder {Remainder}")
        # if(Remainder)
        if(M = 0):
            set Remainder = Remainder / 10.0
        return (Remainder+((M*1.0)/Divisor), "{Prefix}")
    return (ReturnNum*1.0, "")

Abbreviate2<public>(Number : float, Decimals:logic):tuple(float,int)= # decimals represents numbers that you want it to show deciamls before reaching 1000
    
    Abbreviations : []char = array { ' ', 'K', 'M', 'B', 'T','a','b','c','d','e','f','g','h','i','j','k','l','m','n','o','p','q','r','s','t','u','v','w','x','y','z'}

    for (Index : int = 0..Abbreviations.Length, Abbreviation := Abbreviations[Index]):
        if:
            Number < Pow(10.0, ((Index+1)*3.0))
            AbbreviatedNum := Number/((Pow(10.0, (Index)*3.0)))
        then:
            var RoundedAbbreviation : float = 0.0

            if(Number >= 1000.0):
                set RoundedAbbreviation = GetXScale(AbbreviatedNum * 1.0, true)(0)
            else:
                set RoundedAbbreviation = GetXScale(AbbreviatedNum * 1.0, false)(0)


            if(Index = 0):
                return (RoundedAbbreviation,GetXScale(AbbreviatedNum * 1.0, false)(1))
            else:
                return (RoundedAbbreviation,GetXScale(AbbreviatedNum * 1.0, true)(1))

    return (0.0, 0)

GetXScale(Input : float, ShowDecimals : logic): tuple(float,int) =

    var NumDigitsAfterDecimal : int = 0
    if:
        if(ShowDecimals = true):
            BeforeBeforeDecimal := Floor[Input]
            BeforeBeforeString := ToString(BeforeBeforeDecimal)
            if(BeforeBeforeString.Length >=4):
                set NumDigitsAfterDecimal = 0
            else:
                if(BeforeBeforeString.Length =3):
                    set NumDigitsAfterDecimal = 0
                else:
                    if(BeforeBeforeString.Length =2):
                        set NumDigitsAfterDecimal = 1
                    else:
                        if(BeforeBeforeString.Length =1 ):
                            set NumDigitsAfterDecimal = 2
        else:
            set NumDigitsAfterDecimal = 0

        Multiplier := Pow(10.0, NumDigitsAfterDecimal * 1.0)
        #Print("NumDigitsAfterDecimal: {NumDigitsAfterDecimal}")
        #Print("Multiplier: {Multiplier}")
        RoundedValue :=  float[Round[Input*Multiplier]*1.0] / Multiplier
        #Print("RoundedValue: {RoundedValue}")
        BeforeDecimal := Floor[RoundedValue]
        #Print("BeforeDecimal: {BeforeDecimal}")
        AfterDecimal := (RoundedValue - BeforeDecimal*1.0)*Multiplier
        var AfterDecimalString : string = ToString(AfterDecimal)
        #Print("AfterDecimal: {AfterDecimal}")

        #pad with leading zeroes
        for (It := 0..(NumDigitsAfterDecimal-AfterDecimalString.Length-1)):
            set AfterDecimalString = array{'0'} + AfterDecimalString

    then:
        if(NumDigitsAfterDecimal <= 0):
            return ((BeforeDecimal*1.0), NumDigitsAfterDecimal)
            
        else:
            return ((RoundedValue*1.0) * Multiplier, NumDigitsAfterDecimal)
    else:
        (Input,NumDigitsAfterDecimal)





GetNumberAfterDecimal(Input : float, ShowDecimals : logic): int =

    Abbreviations : []char = array { ' ', 'K', 'M', 'B', 'T','a','b','c','d','e','f','g','h','i','j','k','l','m','n','o','p','q','r','s','t','u','v','w','x','y','z'}
    var NumDigitsAfterDecimal : int = 0

    var ShowDecimals2 : logic = true

    for (Index : int = 0..Abbreviations.Length, Abbreviation := Abbreviations[Index]):
        if:
            Input < Pow(10.0, ((Index+1)*3.0))
            AbbreviatedNum := Input/((Pow(10.0, (Index)*3.0)))
        then:


            if(Input >= 1000.0):
                set ShowDecimals2 = true
            else:
                set ShowDecimals2 = false

            if:
                ShowDecimals2 = true
                BeforeBeforeDecimal := Floor[AbbreviatedNum]
                BeforeBeforeString := ToString(BeforeBeforeDecimal)
            then:
                if(BeforeBeforeString.Length >=4):
                    set NumDigitsAfterDecimal = 0
                else:
                    if(BeforeBeforeString.Length =3):
                        set NumDigitsAfterDecimal = 0
                    else:
                        if(BeforeBeforeString.Length =2):
                            set NumDigitsAfterDecimal = 1
                        else:
                            if(BeforeBeforeString.Length =1 ):
                                set NumDigitsAfterDecimal = 2

        return NumDigitsAfterDecimal

    return NumDigitsAfterDecimal


GetSuffixScale(Input : float):float =

    Abbreviations : []char = array { ' ', 'K', 'M', 'B', 'T','a','b','c','d','e','f','g','h','i','j','k','l','m','n','o','p','q','r','s','t','u','v','w','x','y','z'}

    for (Index : int = 0..Abbreviations.Length):

        if:
            Input < Pow(10.0, ((Index+1)*3.0))
        then:
            #Print("Index: {Index}")
            return ((Index+1) * 1.0)

    return 1.0

 

    


MakePriceScale(Price : float, ?X : float = 1.0):vector3=

    FormattedPrice := GetFormattedMoney(Price)

    var Scale : float = Abbreviate2(Price,true)(0) * 1.0
    DecimalPlaces :=  Abbreviate2(Price,true)(1)
    #Print("InScale DecimalPlaces : {DecimalPlaces}")
    var DecimalScale : float = 0.0
    case(DecimalPlaces):
        2=> 
            set DecimalScale =0.5
        1=> 
            set DecimalScale =0.05
        _=> 
            set DecimalScale =0.0
    set Scale = Scale + DecimalScale
    #Print("{Scale}")
    SuffixScale := GetSuffixScale(Price)
    #Print("SuffixScale : {SuffixScale}")
    vector3{X := X, Y := Scale, Z := SuffixScale}
