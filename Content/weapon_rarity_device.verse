
using { /Fortnite.com/Devices }
using { /Verse.org/Simulation }

weapon_rarity_device := class(creative_device):
    

    var StartingConditionalDevices : []conditional_button_device = array{}
    var CommonConditionalDevices : []conditional_button_device = array{}
    var UncommonConditionalDevices : []conditional_button_device = array{}
    var RareConditionalDevices : []conditional_button_device = array{}
    var EpicConditionalDevices : []conditional_button_device = array{}
    var LegendaryConditionalDevices : []conditional_button_device = array{}
    var ExoticConditionalDevices : []conditional_button_device = array{}

    var PistolDevices : []conditional_button_device = array{}
    var SmgDevices : []conditional_button_device = array{}
    var AssaultDevices : []conditional_button_device = array{}
    var ShotgunDevices : []conditional_button_device = array{}
    var SniperDevices : []conditional_button_device = array{}

    @editable
    CommonTrigger : trigger_device = trigger_device {}
    @editable
    UncommonTrigger : trigger_device = trigger_device {}
    @editable
    RareTrigger : trigger_device = trigger_device {}
    @editable
    EpicTrigger : trigger_device = trigger_device {}
    @editable
    LegendaryTrigger : trigger_device = trigger_device {}
    @editable
    ExoticTrigger : trigger_device = trigger_device {}

    @editable
    ClearTrigger : trigger_device = trigger_device{}

    @editable
    ClearCommonTrigger : trigger_device = trigger_device{}
    @editable
    ClearUncommonTrigger : trigger_device = trigger_device{}
    @editable
    ClearRareTrigger : trigger_device = trigger_device{}
    @editable
    ClearEpicTrigger : trigger_device = trigger_device{}
    @editable
    ClearLegendaryTrigger : trigger_device = trigger_device{}
    @editable
    ClearExoticTrigger : trigger_device = trigger_device{}

    @editable
    AchievementTypeTriggers : []trigger_device = array{}

    var AgentMap : [agent]float = map{}

    CommonDamage : float = 0.0
    UncommonDamage : float = 4.0 # = 400 gems
    RareDamage : float = 32.0 # = 3,200 gems
    EpicDamage : float = 256.0 # = 25,600 gems
    LegendaryDamage : float = 2048.0
    ExoticDamage : float = 16384.0

    OnBegin<override>()<suspends>:void=

        StartingConditional := FindCreativeObjectsWithTag(StartingWeaponConditionalTag{})
        for(TaggedConditional : StartingConditional):
            if(Cond := conditional_button_device[TaggedConditional]):
                set StartingConditionalDevices = StartingConditionalDevices + array{Cond}

        CommonConditionals := FindCreativeObjectsWithTag(CommonConditionalTag{})
        for(TaggedConditional : CommonConditionals):
            if(Cond := conditional_button_device[TaggedConditional]):
                set CommonConditionalDevices = CommonConditionalDevices + array{Cond}

        UncommonConditionals := FindCreativeObjectsWithTag(UncommonConditionalTag{})
        for(TaggedConditional : UncommonConditionals):
            if(Cond := conditional_button_device[TaggedConditional]):
                set UncommonConditionalDevices = UncommonConditionalDevices + array{Cond}

        RareConditionals := FindCreativeObjectsWithTag(RareConditionalTag{})
        for(TaggedConditional : RareConditionals):
            if(Cond := conditional_button_device[TaggedConditional]):
                set RareConditionalDevices = RareConditionalDevices + array{Cond}

        EpicConditionals := FindCreativeObjectsWithTag(EpicConditionalTag{})
        for(TaggedConditional : EpicConditionals):
            if(Cond := conditional_button_device[TaggedConditional]):
                set EpicConditionalDevices = EpicConditionalDevices + array{Cond}

        LegendaryConditionals := FindCreativeObjectsWithTag(LegendaryConditionalTag{})
        for(TaggedConditional : LegendaryConditionals):
            if(Cond := conditional_button_device[TaggedConditional]):
                set LegendaryConditionalDevices = LegendaryConditionalDevices + array{Cond}

        ExoticConditionals := FindCreativeObjectsWithTag(ExoticConditionalTag{})
        for(TaggedConditional : ExoticConditionals):
            if(Cond := conditional_button_device[TaggedConditional]):
                set ExoticConditionalDevices = ExoticConditionalDevices + array{Cond}

        PistolConditionals := FindCreativeObjectsWithTag(PistolConditionalTag{})
        for(TaggedConditional : PistolConditionals):
            if(Cond := conditional_button_device[TaggedConditional]):
                set PistolDevices = PistolDevices + array{Cond}

        SmgConditionals := FindCreativeObjectsWithTag(SmgConditionalTag{})
        for(TaggedConditional : SmgConditionals):
            if(Cond := conditional_button_device[TaggedConditional]):
                set SmgDevices = SmgDevices + array{Cond}

        AssaultConditionals := FindCreativeObjectsWithTag(AssaultConditionalTag{})
        for(TaggedConditional : AssaultConditionals):
            if(Cond := conditional_button_device[TaggedConditional]):
                set AssaultDevices = AssaultDevices + array{Cond}

        ShotgunConditionals := FindCreativeObjectsWithTag(ShotgunConditionalTag{})
        for(TaggedConditional : ShotgunConditionals):
            if(Cond := conditional_button_device[TaggedConditional]):
                set ShotgunDevices = ShotgunDevices + array{Cond}

        SniperConditionals := FindCreativeObjectsWithTag(SniperConditionalTag{})
        for(TaggedConditional : SniperConditionals):
            if(Cond := conditional_button_device[TaggedConditional]):
                set SniperDevices = SniperDevices + array{Cond}

        loop:
            Sleep(0.5)
            Players := Self.GetPlayspace().GetPlayers()
            for(Player:Players):
                var IsHoldingRarity : logic = false
                if(Agent:=agent[Player]):

                    for(StartingCond:StartingConditionalDevices):
                        if(StartingCond.IsHoldingItem[Agent]):
                            if(not AgentMap[Agent] = CommonDamage):
                                PlayerHoldingCommonItem(Agent)
                            set IsHoldingRarity = true
                    
                    for(CommonCond:CommonConditionalDevices):
                        if(CommonCond.IsHoldingItem[Agent]):
                            if(not AgentMap[Agent] = CommonDamage):
                                PlayerHoldingCommonItem(Agent)
                            set IsHoldingRarity = true
                        
                    for(UncommonCond:UncommonConditionalDevices):
                        if(UncommonCond.IsHoldingItem[Agent]):
                            if(not AgentMap[Agent] = UncommonDamage):
                                PlayerHoldingUncommonItem(Agent)
                            set IsHoldingRarity = true

                    for(RareCond:RareConditionalDevices):
                        if(RareCond.IsHoldingItem[Agent]):
                            if(not AgentMap[Agent] = RareDamage):
                                PlayerHoldingRareItem(Agent)
                            set IsHoldingRarity = true

                    for(EpicCond:EpicConditionalDevices):
                        if(EpicCond.IsHoldingItem[Agent]):
                            if(not AgentMap[Agent] = EpicDamage):
                                PlayerHoldingEpicItem(Agent)
                            set IsHoldingRarity = true

                    for(LegendaryCond:LegendaryConditionalDevices):
                        if(LegendaryCond.IsHoldingItem[Agent]):
                            if(not AgentMap[Agent] = LegendaryDamage):
                                PlayerHoldingLegendaryItem(Agent)
                            set IsHoldingRarity = true

                    for(ExoticCond:ExoticConditionalDevices):
                        if(ExoticCond.IsHoldingItem[Agent]):
                            if(not AgentMap[Agent] = ExoticDamage):
                                PlayerHoldingExoticItem(Agent)
                            set IsHoldingRarity = true

                    if(IsHoldingRarity = false):
                        ClearTrigger.Trigger(Agent)
                        if(set AgentMap[Agent] = 0.001){}


    CheckWeaponType(Agent:agent):void = 

        var Type : int = 0
        
        for(PistolDevice:PistolDevices):
            if(PistolDevice.IsHoldingItem[Agent]):
                set Type = 0
        for(SmgDevice:SmgDevices):
            if(SmgDevice.IsHoldingItem[Agent]):
                set Type = 1
        for(AssaultDevice:AssaultDevices):
            if(AssaultDevice.IsHoldingItem[Agent]):
                set Type = 2
        for(ShotgunDevice:ShotgunDevices):
            if(ShotgunDevice.IsHoldingItem[Agent]):
                set Type = 3
        for(SniperDevice:SniperDevices):
            if(SniperDevice.IsHoldingItem[Agent]):
                set Type = 4

        #Print("Weapon Type : {Type}")

        if(Trig := AchievementTypeTriggers[Type]):
            Trig.Trigger(Agent)


    GetWeaponType(Agent:agent):int=
        var Type : int = 0
    
        for(PistolDevice:PistolDevices):
            if(PistolDevice.IsHoldingItem[Agent]):
                set Type = 0
        for(SmgDevice:SmgDevices):
            if(SmgDevice.IsHoldingItem[Agent]):
                set Type = 1
        for(AssaultDevice:AssaultDevices):
            if(AssaultDevice.IsHoldingItem[Agent]):
                set Type = 2
        for(ShotgunDevice:ShotgunDevices):
            if(ShotgunDevice.IsHoldingItem[Agent]):
                set Type = 3
        for(SniperDevice:SniperDevices):
            if(SniperDevice.IsHoldingItem[Agent]):
                set Type = 4

        #Print("Weapon Type : {Type}")

        return Type


    PlayerHoldingNothing(Agent:agent):void=

        if(set AgentMap[Agent] = 0.00){}
        ClearTrigger.Trigger(Agent)

        ClearCommonTrigger.Trigger(Agent)
        ClearUncommonTrigger.Trigger(Agent)
        ClearRareTrigger.Trigger(Agent)
        ClearEpicTrigger.Trigger(Agent)
        ClearLegendaryTrigger.Trigger(Agent)
        ClearExoticTrigger.Trigger(Agent)



    PlayerHoldingCommonItem(Agent:agent):void=
        if(set AgentMap[Agent] = CommonDamage){}
        ClearTrigger.Trigger(Agent)
        CommonTrigger.Trigger(Agent)

        spawn{DelayedTrigger(CommonTrigger,Agent)}

        ClearUncommonTrigger.Trigger(Agent)
        ClearRareTrigger.Trigger(Agent)
        ClearEpicTrigger.Trigger(Agent)
        ClearLegendaryTrigger.Trigger(Agent)
        ClearExoticTrigger.Trigger(Agent)

    PlayerHoldingUncommonItem(Agent:agent):void=
        if(set AgentMap[Agent] = UncommonDamage){}
        ClearTrigger.Trigger(Agent)
        UncommonTrigger.Trigger(Agent)

        spawn{DelayedTrigger(UncommonTrigger,Agent)}

        ClearCommonTrigger.Trigger(Agent)
        ClearRareTrigger.Trigger(Agent)
        ClearEpicTrigger.Trigger(Agent)
        ClearLegendaryTrigger.Trigger(Agent)
        ClearExoticTrigger.Trigger(Agent)

    PlayerHoldingRareItem(Agent:agent):void=
        if(set AgentMap[Agent] = RareDamage){}
        ClearTrigger.Trigger(Agent)
        RareTrigger.Trigger(Agent)

        spawn{DelayedTrigger(RareTrigger,Agent)}

        ClearCommonTrigger.Trigger(Agent)
        ClearUncommonTrigger.Trigger(Agent)
        ClearEpicTrigger.Trigger(Agent)
        ClearLegendaryTrigger.Trigger(Agent)
        ClearExoticTrigger.Trigger(Agent)

    PlayerHoldingEpicItem(Agent:agent):void=
        if(set AgentMap[Agent] = EpicDamage){}
        ClearTrigger.Trigger(Agent)
        EpicTrigger.Trigger(Agent)

        spawn{DelayedTrigger(EpicTrigger,Agent)}

        ClearCommonTrigger.Trigger(Agent)
        ClearUncommonTrigger.Trigger(Agent)
        ClearRareTrigger.Trigger(Agent)
        ClearLegendaryTrigger.Trigger(Agent)
        ClearExoticTrigger.Trigger(Agent)

    PlayerHoldingLegendaryItem(Agent:agent):void=
        if(set AgentMap[Agent] = LegendaryDamage){}
        ClearTrigger.Trigger(Agent)
        LegendaryTrigger.Trigger(Agent)

        spawn{DelayedTrigger(LegendaryTrigger,Agent)}

        ClearCommonTrigger.Trigger(Agent)
        ClearUncommonTrigger.Trigger(Agent)
        ClearRareTrigger.Trigger(Agent)
        ClearEpicTrigger.Trigger(Agent)
        ClearExoticTrigger.Trigger(Agent)

    PlayerHoldingExoticItem(Agent:agent):void=
        if(set AgentMap[Agent] = ExoticDamage){}
        ClearTrigger.Trigger(Agent)
        ExoticTrigger.Trigger(Agent)

        spawn{DelayedTrigger(ExoticTrigger,Agent)}

        ClearCommonTrigger.Trigger(Agent)
        ClearUncommonTrigger.Trigger(Agent)
        ClearRareTrigger.Trigger(Agent)
        ClearEpicTrigger.Trigger(Agent)
        ClearLegendaryTrigger.Trigger(Agent)

    DelayedTrigger(MyTrigger:trigger_device,Agent:agent)<suspends>:void=
        Sleep(0.05)
        MyTrigger.Trigger(Agent)

    


        
