
using { /Fortnite.com/Devices }
using { /Verse.org/Simulation }
using { /UnrealEngine.com/Temporary/UI }
using { /UnrealEngine.com/Temporary/SpatialMath }
using { /Fortnite.com/UI }
using { /Verse.org/Colors }

weapon_upgrades_device := class(creative_device):

    @editable
    OpenShop : button_device = button_device{}
    @editable
    PurchaseAudio : audio_player_device = audio_player_device{}
    @editable
    NegativeAudio : audio_player_device = audio_player_device{}

    @editable
    OpenAudio : audio_player_device = audio_player_device{}
    @editable
    CloseAudio : audio_player_device = audio_player_device{}


    @editable
    OreConditionals : []conditional_button_device = array{}


    var MyUI : [agent]?canvas = map{}
    var InventoryTextBlockPerAgent : [agent]?[]text_block = map{}
    var ConversionRateTextBlockPerAgent : [agent]?[]text_block = map{}

    var LevelMap : [agent]?[]text_block = map{}
    var OreCostMap : [agent]?[]text_block = map{}
    var GoldCostMap: [agent]?[]text_block = map{}
    var OreTextureMap: [agent]?[]texture_block = map{}

    @editable
    MissionTrigger: trigger_device = trigger_device{}

    storedstring<localizes>(S:string) : message = "{S}"

    OnBegin<override>()<suspends>:void=
        
        OpenShop.InteractedWithEvent.Subscribe(ShowUi)
     
    MakeOreConverterBackground (Agent:agent) :canvas= 

        GoldTexture := wooden_pixel_gui.textures.Icons2.Size8.t_CoinGold_8

        CopperTexture: texture_block = texture_block{DefaultImage:= hud.Copper_icon512, DefaultDesiredSize:=vector2{X:=100.0,Y:=100.0}}
        CopperTextureShadow: texture_block = texture_block{DefaultImage:= hud.Copper_icon512,DefaultTint:=MakeColorFromHex("080402FF"), DefaultDesiredSize:=vector2{X:=100.0,Y:=100.0}}
        CopperTexture2: texture_block = texture_block{DefaultImage:= hud.Copper_icon512, DefaultDesiredSize:=vector2{X:=100.0,Y:=100.0}}
        CopperTextureShadow2: texture_block = texture_block{DefaultImage:= hud.Copper_icon512,DefaultTint:=MakeColorFromHex("080402FF"), DefaultDesiredSize:=vector2{X:=100.0,Y:=100.0}}
        SilverTexture: texture_block = texture_block{DefaultImage:= hud.Silver_icon, DefaultDesiredSize:=vector2{X:=100.0,Y:=100.0}}
        SilverTextureShadow: texture_block = texture_block{DefaultImage:= hud.Silver_icon,DefaultTint:=MakeColorFromHex("080402FF"), DefaultDesiredSize:=vector2{X:=100.0,Y:=100.0}}
        SilverTexture2: texture_block = texture_block{DefaultImage:= hud.Silver_icon, DefaultDesiredSize:=vector2{X:=100.0,Y:=100.0}}
        SilverTextureShadow2: texture_block = texture_block{DefaultImage:= hud.Silver_icon,DefaultTint:=MakeColorFromHex("080402FF"), DefaultDesiredSize:=vector2{X:=100.0,Y:=100.0}}
        SilverTexture3: texture_block = texture_block{DefaultImage:= hud.Silver_icon, DefaultDesiredSize:=vector2{X:=100.0,Y:=100.0}}
        SilverTextureShadow3: texture_block = texture_block{DefaultImage:= hud.Silver_icon,DefaultTint:=MakeColorFromHex("080402FF"), DefaultDesiredSize:=vector2{X:=100.0,Y:=100.0}}
        MalachiteTexture: texture_block = texture_block{DefaultImage:= hud.Malachite_icon512, DefaultDesiredSize:=vector2{X:=100.0,Y:=100.0}}
        MalachiteTextureShadow: texture_block = texture_block{DefaultImage:= hud.Malachite_icon512,DefaultTint:=MakeColorFromHex("080402FF"), DefaultDesiredSize:=vector2{X:=100.0,Y:=100.0}}
        Malachite2Texture: texture_block = texture_block{DefaultImage:= hud.Malachite_icon512, DefaultDesiredSize:=vector2{X:=100.0,Y:=100.0}}
        Malachite2TextureShadow: texture_block = texture_block{DefaultImage:= hud.Malachite_icon512,DefaultTint:=MakeColorFromHex("080402FF"), DefaultDesiredSize:=vector2{X:=100.0,Y:=100.0}}
        Malachite3Texture: texture_block = texture_block{DefaultImage:= hud.Malachite_icon512, DefaultDesiredSize:=vector2{X:=100.0,Y:=100.0}}
        Malachite3TextureShadow: texture_block = texture_block{DefaultImage:= hud.Malachite_icon512,DefaultTint:=MakeColorFromHex("080402FF"), DefaultDesiredSize:=vector2{X:=100.0,Y:=100.0}}
        ObsidianTexture: texture_block = texture_block{DefaultImage:= hud.Obsidian_icon512, DefaultDesiredSize:=vector2{X:=100.0,Y:=100.0}}
        ObsidianTextureShadow: texture_block = texture_block{DefaultImage:= hud.Obsidian_icon512,DefaultTint:=MakeColorFromHex("080402FF"), DefaultDesiredSize:=vector2{X:=100.0,Y:=100.0}}
        Obsidian2Texture: texture_block = texture_block{DefaultImage:= hud.Obsidian_icon512, DefaultDesiredSize:=vector2{X:=100.0,Y:=100.0}}
        Obsidian2TextureShadow: texture_block = texture_block{DefaultImage:= hud.Obsidian_icon512,DefaultTint:=MakeColorFromHex("080402FF"), DefaultDesiredSize:=vector2{X:=100.0,Y:=100.0}}
        Obsidian3Texture: texture_block = texture_block{DefaultImage:= hud.Obsidian_icon512, DefaultDesiredSize:=vector2{X:=100.0,Y:=100.0}}
        Obsidian3TextureShadow: texture_block = texture_block{DefaultImage:= hud.Obsidian_icon512,DefaultTint:=MakeColorFromHex("080402FF"), DefaultDesiredSize:=vector2{X:=100.0,Y:=100.0}}
        BrightcoreTexture: texture_block = texture_block{DefaultImage:= hud.Brightcore_icon512, DefaultDesiredSize:=vector2{X:=100.0,Y:=100.0}}
        BrightcoreTextureShadow: texture_block = texture_block{DefaultImage:= hud.Brightcore_icon512,DefaultTint:=MakeColorFromHex("080402FF"), DefaultDesiredSize:=vector2{X:=100.0,Y:=100.0}}
        BrightcoreTexture2: texture_block = texture_block{DefaultImage:= hud.Brightcore_icon512, DefaultDesiredSize:=vector2{X:=100.0,Y:=100.0}}
        BrightcoreTextureShadow2: texture_block = texture_block{DefaultImage:= hud.Brightcore_icon512,DefaultTint:=MakeColorFromHex("080402FF"), DefaultDesiredSize:=vector2{X:=100.0,Y:=100.0}}
        BrightcoreTexture3: texture_block = texture_block{DefaultImage:= hud.Brightcore_icon512, DefaultDesiredSize:=vector2{X:=100.0,Y:=100.0}}
        BrightcoreTextureShadow3: texture_block = texture_block{DefaultImage:= hud.Brightcore_icon512,DefaultTint:=MakeColorFromHex("080402FF"), DefaultDesiredSize:=vector2{X:=100.0,Y:=100.0}}
        SpectroliteTexture: texture_block = texture_block{DefaultImage:= hud.Spectrolite_icon512, DefaultDesiredSize:=vector2{X:=100.0,Y:=100.0}}
        SpectroliteTextureShadow: texture_block = texture_block{DefaultImage:= hud.Spectrolite_icon512,DefaultTint:=MakeColorFromHex("080402FF"), DefaultDesiredSize:=vector2{X:=100.0,Y:=100.0}}
        SpectroliteTexture2: texture_block = texture_block{DefaultImage:= hud.Spectrolite_icon512, DefaultDesiredSize:=vector2{X:=100.0,Y:=100.0}}
        SpectroliteTextureShadow2: texture_block = texture_block{DefaultImage:= hud.Spectrolite_icon512,DefaultTint:=MakeColorFromHex("080402FF"), DefaultDesiredSize:=vector2{X:=100.0,Y:=100.0}}
        SpectroliteTexture3: texture_block = texture_block{DefaultImage:= hud.Spectrolite_icon512, DefaultDesiredSize:=vector2{X:=100.0,Y:=100.0}}
        SpectroliteTextureShadow3: texture_block = texture_block{DefaultImage:= hud.Spectrolite_icon512,DefaultTint:=MakeColorFromHex("080402FF"), DefaultDesiredSize:=vector2{X:=100.0,Y:=100.0}}

        Slot1LevelText: text_block = text_block{DefaultTextColor := MakeColorFromHex("FFFFFF"), DefaultText:=storedstring("Lv 0")}
        Slot1TypeText: text_block = text_block{DefaultTextColor := MakeColorFromHex("FFFFFF"), DefaultText:=storedstring("Pistols")}
        Slot1OreIcon : texture_block = texture_block{DefaultImage:= hud.Copper_icon512, DefaultDesiredSize:=vector2{X:=100.0,Y:=100.0}}
        Slot1OreCost: text_block = text_block{DefaultTextColor := MakeColorFromHex("FFFFFF"), DefaultText:=storedstring("1")}
        Slot1GoldCost: text_block = text_block{DefaultTextColor := MakeColorFromHex("FFFFFF"), DefaultText:=storedstring("100")}
        Slot1Icon : texture_block = texture_block{DefaultImage:= hud.PistolIcon,DefaultTint:=MakeColorFromHex("CB9764FF"), DefaultDesiredSize:=vector2{X:=100.0,Y:=100.0}}

        Slot2LevelText: text_block = text_block{DefaultTextColor := MakeColorFromHex("FFFFFF"), DefaultText:=storedstring("Lv 0")}
        Slot2TypeText: text_block = text_block{DefaultTextColor := MakeColorFromHex("FFFFFF"), DefaultText:=storedstring("Rifles")}
        Slot2OreIcon : texture_block = texture_block{DefaultImage:= hud.Copper_icon512, DefaultDesiredSize:=vector2{X:=100.0,Y:=100.0}}
        Slot2OreCost: text_block = text_block{DefaultTextColor := MakeColorFromHex("FFFFFF"), DefaultText:=storedstring("1")}
        Slot2GoldCost: text_block = text_block{DefaultTextColor := MakeColorFromHex("FFFFFF"), DefaultText:=storedstring("100")}
        Slot2Icon : texture_block = texture_block{DefaultImage:= hud.ArIcon,DefaultTint:=MakeColorFromHex("CB9764FF"), DefaultDesiredSize:=vector2{X:=100.0,Y:=100.0}}
        
        Slot3LevelText: text_block = text_block{DefaultTextColor := MakeColorFromHex("FFFFFF"), DefaultText:=storedstring("Lv 0")}
        Slot3TypeText: text_block = text_block{DefaultTextColor := MakeColorFromHex("FFFFFF"), DefaultText:=storedstring("Smgs")}
        Slot3OreIcon : texture_block = texture_block{DefaultImage:= hud.Copper_icon512, DefaultDesiredSize:=vector2{X:=100.0,Y:=100.0}}
        Slot3OreCost: text_block = text_block{DefaultTextColor := MakeColorFromHex("FFFFFF"), DefaultText:=storedstring("1")}
        Slot3GoldCost: text_block = text_block{DefaultTextColor := MakeColorFromHex("FFFFFF"), DefaultText:=storedstring("100")}
        Slot3Icon : texture_block = texture_block{DefaultImage:= hud.SmgIcon,DefaultTint:=MakeColorFromHex("CB9764FF"), DefaultDesiredSize:=vector2{X:=100.0,Y:=100.0}}

        Slot4LevelText: text_block = text_block{DefaultTextColor := MakeColorFromHex("FFFFFF"), DefaultText:=storedstring("Lv 0")}
        Slot4TypeText: text_block = text_block{DefaultTextColor := MakeColorFromHex("FFFFFF"), DefaultText:=storedstring("Shotguns")}
        Slot4OreIcon : texture_block = texture_block{DefaultImage:= hud.Copper_icon512, DefaultDesiredSize:=vector2{X:=100.0,Y:=100.0}}
        Slot4OreCost: text_block = text_block{DefaultTextColor := MakeColorFromHex("FFFFFF"), DefaultText:=storedstring("1")}
        Slot4GoldCost: text_block = text_block{DefaultTextColor := MakeColorFromHex("FFFFFF"), DefaultText:=storedstring("100")}
        Slot4Icon : texture_block = texture_block{DefaultImage:= hud.ShotgunIcon,DefaultTint:=MakeColorFromHex("CB9764FF"), DefaultDesiredSize:=vector2{X:=100.0,Y:=100.0}}

        Slot5LevelText: text_block = text_block{DefaultTextColor := MakeColorFromHex("FFFFFF"), DefaultText:=storedstring("Lv 0")}
        Slot5TypeText: text_block = text_block{DefaultTextColor := MakeColorFromHex("FFFFFF"), DefaultText:=storedstring("Snipers")}
        Slot5OreIcon : texture_block = texture_block{DefaultImage:= hud.Copper_icon512, DefaultDesiredSize:=vector2{X:=100.0,Y:=100.0}}
        Slot5OreCost: text_block = text_block{DefaultTextColor := MakeColorFromHex("FFFFFF"), DefaultText:=storedstring("1")}
        Slot5GoldCost: text_block = text_block{DefaultTextColor := MakeColorFromHex("FFFFFF"), DefaultText:=storedstring("100")}
        Slot5Icon : texture_block = texture_block{DefaultImage:= hud.SniperIcon,DefaultTint:=MakeColorFromHex("CB9764FF"), DefaultDesiredSize:=vector2{X:=100.0,Y:=100.0}}

        
        Slot1Button : button_quiet = button_quiet{DefaultText:=storedstring("Upgrade")}
        Slot1Button.OnClick().SubscribeMessage(UpgradeWeaponTypeLevel,0)
        Slot2Button : button_quiet = button_quiet{DefaultText:=storedstring("Upgrade")}
        Slot2Button.OnClick().SubscribeMessage(UpgradeWeaponTypeLevel,1)
        Slot3Button : button_quiet = button_quiet{DefaultText:=storedstring("Upgrade")}
        Slot3Button.OnClick().SubscribeMessage(UpgradeWeaponTypeLevel,2)
        Slot4Button : button_quiet = button_quiet{DefaultText:=storedstring("Upgrade")}
        Slot4Button.OnClick().SubscribeMessage(UpgradeWeaponTypeLevel,3)
        Slot5Button : button_quiet = button_quiet{DefaultText:=storedstring("Upgrade")}
        Slot5Button.OnClick().SubscribeMessage(UpgradeWeaponTypeLevel,4)

        CloseButton : button_quiet = button_quiet{DefaultText:=storedstring("Close")}
        CloseButton.OnClick().Subscribe(RemoveHUD) 

        if(set LevelMap[Agent] = option{array{Slot1LevelText,Slot2LevelText,Slot3LevelText,Slot4LevelText,Slot5LevelText}}){}
        if(set OreCostMap[Agent] = option{array{Slot1OreCost,Slot2OreCost,Slot3OreCost,Slot4OreCost,Slot5OreCost}}){} 
        if(set GoldCostMap[Agent] = option{array{Slot1GoldCost,Slot2GoldCost,Slot3GoldCost,Slot4GoldCost,Slot5GoldCost}}){} 
        if(set OreTextureMap[Agent] = option{array{Slot1OreIcon,Slot2OreIcon,Slot3OreIcon,Slot4OreIcon,Slot5OreIcon}}){} 

        MyCanvas:=canvas:
            Slots:=array:
                canvas_slot:
                    Anchors:=anchors{Minimum:=vector2{X:=0.2,Y:=0.05}, Maximum:=vector2{X:=0.8,Y:=0.95}}
                    Widget:=overlay:
                        Slots:=array:
                            overlay_slot:
                                HorizontalAlignment:=horizontal_alignment.Fill
                                VerticalAlignment:=vertical_alignment.Fill
                                Padding:=margin{Left:=40.0,Top:=40.0,Right:=40.0,Bottom:=40.0}
                                Widget:=texture_block:
                                    DefaultImage:=wooden_pixel_gui.textures.Panels.BrownSquare
                                    DefaultDesiredSize:=vector2{X:=50.0,Y:=110.0}
                            overlay_slot:
                                HorizontalAlignment:=horizontal_alignment.Fill
                                VerticalAlignment:=vertical_alignment.Fill
                                Widget:=stack_box:
                                    Orientation:=orientation.Vertical
                                    Slots:=array:
                                        stack_box_slot:
                                            HorizontalAlignment:=horizontal_alignment.Fill
                                            VerticalAlignment:=vertical_alignment.Fill
                                            Padding:=margin{Left:=40.0,Top:=40.0,Right:=40.0,Bottom:=0.0}
                                            Widget:=overlay:
                                                Slots:=array:
                                                    overlay_slot:
                                                        HorizontalAlignment:=horizontal_alignment.Fill
                                                        VerticalAlignment:=vertical_alignment.Fill
                                                        Widget:=texture_block:
                                                            DefaultImage:=wooden_pixel_gui.textures.Panels.t_TitleBar
                                                            DefaultDesiredSize:=vector2{X:=15.0,Y:=60.0}
                                                    overlay_slot:
                                                        HorizontalAlignment:=horizontal_alignment.Center
                                                        VerticalAlignment:=vertical_alignment.Center
                                                        Widget:=text_block:
                                                            DefaultText:=storedstring("Weapon Levels")
                                                            DefaultTextColor := MakeColorFromHex("401805FF")
                                        stack_box_slot:
                                            HorizontalAlignment:=horizontal_alignment.Fill
                                            VerticalAlignment:=vertical_alignment.Fill
                                            Padding:=margin{Left:=50.0,Top:=0.0,Right:=50.0,Bottom:=50.0}
                                            Widget:=overlay:
                                                Slots:=array:
                                                    overlay_slot:
                                                        HorizontalAlignment:=horizontal_alignment.Fill
                                                        VerticalAlignment:=vertical_alignment.Fill
                                                        Widget:=texture_block:
                                                            DefaultImage:=wooden_pixel_gui.textures.Panels.BrownSquare
                                                            DefaultDesiredSize:=vector2{X:=32.0,Y:=32.0}
                                                    overlay_slot:
                                                        HorizontalAlignment:=horizontal_alignment.Fill
                                                        VerticalAlignment:=vertical_alignment.Fill
                                                        Widget:=stack_box:
                                                            Orientation:=orientation.Vertical
                                                            Slots:=array:
                                                                stack_box_slot: # Main List
                                                                    HorizontalAlignment:=horizontal_alignment.Fill
                                                                    VerticalAlignment:=vertical_alignment.Fill
                                                                    Widget:=overlay:
                                                                        Slots:=array:
                                                                            overlay_slot:
                                                                                HorizontalAlignment:=horizontal_alignment.Left
                                                                                VerticalAlignment:=vertical_alignment.Center
                                                                                Padding:=margin{Left:=30.0,Top:=0.0,Right:=0.0,Bottom:=0.0}
                                                                                Widget:=overlay:
                                                                                    Slots:=array:
                                                                                        overlay_slot:
                                                                                            Padding:=margin{Left:=3.0,Top:=3.0,Right:=0.0,Bottom:=0.0}
                                                                                            Widget:=text_block:
                                                                                                DefaultText:=storedstring("Weapon Type")
                                                                                                DefaultTextColor := MakeColorFromHex("142706FF")
                                                                                        overlay_slot:
                                                                                            Widget:=text_block:
                                                                                                DefaultText:=storedstring("Weapon Type")
                                                                                                DefaultTextColor := MakeColorFromHex("529F17FF")
                                                                            overlay_slot:
                                                                                HorizontalAlignment:=horizontal_alignment.Center
                                                                                VerticalAlignment:=vertical_alignment.Center
                                                                                Widget:=overlay:
                                                                                    Slots:=array:
                                                                                        overlay_slot:
                                                                                            Padding:=margin{Left:=3.0,Top:=3.0,Right:=0.0,Bottom:=0.0}
                                                                                            Widget:=text_block:
                                                                                                DefaultText:=storedstring("Cost")
                                                                                                DefaultTextColor := MakeColorFromHex("142706FF")
                                                                                        overlay_slot:
                                                                                            Widget:=text_block:
                                                                                                DefaultText:=storedstring("Cost")
                                                                                                DefaultTextColor := MakeColorFromHex("529F17FF")
                                                                ################################################################
                                                                stack_box_slot: # Conversion1
                                                                    HorizontalAlignment:=horizontal_alignment.Fill
                                                                    VerticalAlignment:=vertical_alignment.Fill
                                                                    Padding:=margin{Left:=0.0,Top:=10.0,Right:=0.0,Bottom:=0.0}
                                                                    Widget:=overlay:
                                                                        Slots:=array:
                                                                            overlay_slot:
                                                                                HorizontalAlignment:=horizontal_alignment.Fill
                                                                                VerticalAlignment:=vertical_alignment.Fill
                                                                                Widget:=texture_block:
                                                                                    DefaultImage:=wooden_pixel_gui.textures.Panels.BrownSquare
                                                                                    DefaultTint := MakeColorFromHex("A9A9A9FF")
                                                                            overlay_slot:
                                                                                HorizontalAlignment:=horizontal_alignment.Right
                                                                                VerticalAlignment:=vertical_alignment.Center
                                                                                Padding:=margin{Left:=0.0,Top:=0.0,Right:=20.0,Bottom:=0.0}
                                                                                Widget:=Slot1Button
                                                                            overlay_slot:
                                                                                HorizontalAlignment:=horizontal_alignment.Center
                                                                                VerticalAlignment:=vertical_alignment.Center
                                                                                Widget:=stack_box:
                                                                                    Orientation:=orientation.Horizontal
                                                                                    Slots:=array:
                                                                                        
                                                                                        stack_box_slot:
                                                                                            Widget:=overlay:
                                                                                                Slots:=array:
                                                                                                    overlay_slot:
                                                                                                        HorizontalAlignment:=horizontal_alignment.Fill
                                                                                                        VerticalAlignment:=vertical_alignment.Fill
                                                                                                        Widget:=Slot1OreIcon
                                                                                        stack_box_slot:
                                                                                            HorizontalAlignment:=horizontal_alignment.Center
                                                                                            VerticalAlignment:=vertical_alignment.Center
                                                                                            Widget:=Slot1OreCost
                                                                                        stack_box_slot:
                                                                                            Widget:=overlay:
                                                                                                Slots:=array:
                                                                                                    overlay_slot:
                                                                                                        HorizontalAlignment:=horizontal_alignment.Fill
                                                                                                        VerticalAlignment:=vertical_alignment.Fill
                                                                                                        Widget:= texture_block :
                                                                                                            DefaultImage:= GoldTexture
                                                                                                            DefaultDesiredSize:=vector2{X:=64.0,Y:=64.0}
                                                                                        stack_box_slot:
                                                                                            HorizontalAlignment:=horizontal_alignment.Center
                                                                                            VerticalAlignment:=vertical_alignment.Center
                                                                                            Widget:=Slot1GoldCost
                                                                            overlay_slot:
                                                                                HorizontalAlignment:=horizontal_alignment.Left
                                                                                VerticalAlignment:=vertical_alignment.Center
                                                                                Widget:=stack_box:
                                                                                    Orientation:=orientation.Horizontal
                                                                                    Slots:=array:
                                                                                        stack_box_slot:
                                                                                            HorizontalAlignment:=horizontal_alignment.Center
                                                                                            VerticalAlignment:=vertical_alignment.Center
                                                                                            Widget:=Slot1LevelText
                                                                                        stack_box_slot:
                                                                                            Widget:=overlay:
                                                                                                Slots:=array:
                                                                                                    overlay_slot:
                                                                                                        HorizontalAlignment:=horizontal_alignment.Fill
                                                                                                        VerticalAlignment:=vertical_alignment.Fill
                                                                                                        Widget:=Slot1Icon
                                                                                        stack_box_slot:
                                                                                            HorizontalAlignment:=horizontal_alignment.Fill
                                                                                            VerticalAlignment:=vertical_alignment.Center
                                                                                            Widget:=Slot1TypeText
                                                                #################################################################################################                     
                                                                stack_box_slot: # Conversion2
                                                                    HorizontalAlignment:=horizontal_alignment.Fill
                                                                    VerticalAlignment:=vertical_alignment.Fill
                                                                    Padding:=margin{Left:=0.0,Top:=10.0,Right:=0.0,Bottom:=0.0}
                                                                    Widget:=overlay:
                                                                        Slots:=array:
                                                                            overlay_slot:
                                                                                HorizontalAlignment:=horizontal_alignment.Fill
                                                                                VerticalAlignment:=vertical_alignment.Fill
                                                                                Widget:=texture_block:
                                                                                    DefaultImage:=wooden_pixel_gui.textures.Panels.BrownSquare
                                                                                    DefaultTint := MakeColorFromHex("A9A9A9FF")
                                                                            overlay_slot:
                                                                                HorizontalAlignment:=horizontal_alignment.Right
                                                                                VerticalAlignment:=vertical_alignment.Center
                                                                                Padding:=margin{Left:=0.0,Top:=0.0,Right:=20.0,Bottom:=0.0}
                                                                                Widget:=Slot2Button
                                                                            overlay_slot:
                                                                                HorizontalAlignment:=horizontal_alignment.Center
                                                                                VerticalAlignment:=vertical_alignment.Center
                                                                                Widget:=stack_box:
                                                                                    Orientation:=orientation.Horizontal
                                                                                    Slots:=array:
                                                                                        
                                                                                        stack_box_slot:
                                                                                            Widget:=overlay:
                                                                                                Slots:=array:
                                                                                                    overlay_slot:
                                                                                                        HorizontalAlignment:=horizontal_alignment.Fill
                                                                                                        VerticalAlignment:=vertical_alignment.Fill
                                                                                                        Widget:=Slot2OreIcon
                                                                                        stack_box_slot:
                                                                                            HorizontalAlignment:=horizontal_alignment.Center
                                                                                            VerticalAlignment:=vertical_alignment.Center
                                                                                            Widget:=Slot2OreCost
                                                                                        stack_box_slot:
                                                                                            Widget:=overlay:
                                                                                                Slots:=array:
                                                                                                    overlay_slot:
                                                                                                        HorizontalAlignment:=horizontal_alignment.Fill
                                                                                                        VerticalAlignment:=vertical_alignment.Fill
                                                                                                        Widget:= texture_block :
                                                                                                            DefaultImage:= GoldTexture
                                                                                                            DefaultDesiredSize:=vector2{X:=64.0,Y:=64.0}
                                                                                        stack_box_slot:
                                                                                            HorizontalAlignment:=horizontal_alignment.Center
                                                                                            VerticalAlignment:=vertical_alignment.Center
                                                                                            Widget:=Slot2GoldCost
                                                                            overlay_slot:
                                                                                HorizontalAlignment:=horizontal_alignment.Left
                                                                                VerticalAlignment:=vertical_alignment.Center
                                                                                Widget:=stack_box:
                                                                                    Orientation:=orientation.Horizontal
                                                                                    Slots:=array:
                                                                                        stack_box_slot:
                                                                                            HorizontalAlignment:=horizontal_alignment.Center
                                                                                            VerticalAlignment:=vertical_alignment.Center
                                                                                            Widget:=Slot2LevelText
                                                                                        stack_box_slot:
                                                                                            Widget:=overlay:
                                                                                                Slots:=array:
                                                                                                    overlay_slot:
                                                                                                        HorizontalAlignment:=horizontal_alignment.Fill
                                                                                                        VerticalAlignment:=vertical_alignment.Fill
                                                                                                        Widget:=Slot2Icon
                                                                                        stack_box_slot:
                                                                                            HorizontalAlignment:=horizontal_alignment.Fill
                                                                                            VerticalAlignment:=vertical_alignment.Center
                                                                                            Widget:=Slot2TypeText  
                                                                
                                                                ###################################################################                            
                                                                stack_box_slot: # Conversion3
                                                                    HorizontalAlignment:=horizontal_alignment.Fill
                                                                    VerticalAlignment:=vertical_alignment.Fill
                                                                    Padding:=margin{Left:=0.0,Top:=10.0,Right:=0.0,Bottom:=0.0}
                                                                    Widget:=overlay:
                                                                        Slots:=array:
                                                                            overlay_slot:
                                                                                HorizontalAlignment:=horizontal_alignment.Fill
                                                                                VerticalAlignment:=vertical_alignment.Fill
                                                                                Widget:=texture_block:
                                                                                    DefaultImage:=wooden_pixel_gui.textures.Panels.BrownSquare
                                                                                    DefaultTint := MakeColorFromHex("A9A9A9FF")
                                                                            overlay_slot:
                                                                                HorizontalAlignment:=horizontal_alignment.Right
                                                                                VerticalAlignment:=vertical_alignment.Center
                                                                                Padding:=margin{Left:=0.0,Top:=0.0,Right:=20.0,Bottom:=0.0}
                                                                                Widget:=Slot3Button
                                                                            overlay_slot:
                                                                                HorizontalAlignment:=horizontal_alignment.Center
                                                                                VerticalAlignment:=vertical_alignment.Center
                                                                                Widget:=stack_box:
                                                                                    Orientation:=orientation.Horizontal
                                                                                    Slots:=array:
                                                                                        
                                                                                        stack_box_slot:
                                                                                            Widget:=overlay:
                                                                                                Slots:=array:
                                                                                                    overlay_slot:
                                                                                                        HorizontalAlignment:=horizontal_alignment.Fill
                                                                                                        VerticalAlignment:=vertical_alignment.Fill
                                                                                                        Widget:=Slot3OreIcon
                                                                                        stack_box_slot:
                                                                                            HorizontalAlignment:=horizontal_alignment.Center
                                                                                            VerticalAlignment:=vertical_alignment.Center
                                                                                            Widget:=Slot3OreCost
                                                                                        stack_box_slot:
                                                                                            Widget:=overlay:
                                                                                                Slots:=array:
                                                                                                    overlay_slot:
                                                                                                        HorizontalAlignment:=horizontal_alignment.Fill
                                                                                                        VerticalAlignment:=vertical_alignment.Fill
                                                                                                        Widget:= texture_block :
                                                                                                            DefaultImage:= GoldTexture
                                                                                                            DefaultDesiredSize:=vector2{X:=64.0,Y:=64.0}
                                                                                        stack_box_slot:
                                                                                            HorizontalAlignment:=horizontal_alignment.Center
                                                                                            VerticalAlignment:=vertical_alignment.Center
                                                                                            Widget:=Slot3GoldCost
                                                                            overlay_slot:
                                                                                HorizontalAlignment:=horizontal_alignment.Left
                                                                                VerticalAlignment:=vertical_alignment.Center
                                                                                Widget:=stack_box:
                                                                                    Orientation:=orientation.Horizontal
                                                                                    Slots:=array:
                                                                                        stack_box_slot:
                                                                                            HorizontalAlignment:=horizontal_alignment.Center
                                                                                            VerticalAlignment:=vertical_alignment.Center
                                                                                            Widget:=Slot3LevelText
                                                                                        stack_box_slot:
                                                                                            Widget:=overlay:
                                                                                                Slots:=array:
                                                                                                    overlay_slot:
                                                                                                        HorizontalAlignment:=horizontal_alignment.Fill
                                                                                                        VerticalAlignment:=vertical_alignment.Fill
                                                                                                        Widget:=Slot3Icon
                                                                                        stack_box_slot:
                                                                                            HorizontalAlignment:=horizontal_alignment.Fill
                                                                                            VerticalAlignment:=vertical_alignment.Center
                                                                                            Widget:=Slot3TypeText   
                                                                 ###################################################################                            
                                                                stack_box_slot: # Conversion4
                                                                    HorizontalAlignment:=horizontal_alignment.Fill
                                                                    VerticalAlignment:=vertical_alignment.Fill
                                                                    Padding:=margin{Left:=0.0,Top:=10.0,Right:=0.0,Bottom:=0.0}
                                                                    Widget:=overlay:
                                                                        Slots:=array:
                                                                            overlay_slot:
                                                                                HorizontalAlignment:=horizontal_alignment.Fill
                                                                                VerticalAlignment:=vertical_alignment.Fill
                                                                                Widget:=texture_block:
                                                                                    DefaultImage:=wooden_pixel_gui.textures.Panels.BrownSquare
                                                                                    DefaultTint := MakeColorFromHex("A9A9A9FF")
                                                                            overlay_slot:
                                                                                HorizontalAlignment:=horizontal_alignment.Right
                                                                                VerticalAlignment:=vertical_alignment.Center
                                                                                Padding:=margin{Left:=0.0,Top:=0.0,Right:=20.0,Bottom:=0.0}
                                                                                Widget:=Slot4Button
                                                                            overlay_slot:
                                                                                HorizontalAlignment:=horizontal_alignment.Center
                                                                                VerticalAlignment:=vertical_alignment.Center
                                                                                Widget:=stack_box:
                                                                                    Orientation:=orientation.Horizontal
                                                                                    Slots:=array:
                                                                                        
                                                                                        stack_box_slot:
                                                                                            Widget:=overlay:
                                                                                                Slots:=array:
                                                                                                    overlay_slot:
                                                                                                        HorizontalAlignment:=horizontal_alignment.Fill
                                                                                                        VerticalAlignment:=vertical_alignment.Fill
                                                                                                        Widget:=Slot4OreIcon
                                                                                        stack_box_slot:
                                                                                            HorizontalAlignment:=horizontal_alignment.Center
                                                                                            VerticalAlignment:=vertical_alignment.Center
                                                                                            Widget:=Slot4OreCost
                                                                                        stack_box_slot:
                                                                                            Widget:=overlay:
                                                                                                Slots:=array:
                                                                                                    overlay_slot:
                                                                                                        HorizontalAlignment:=horizontal_alignment.Fill
                                                                                                        VerticalAlignment:=vertical_alignment.Fill
                                                                                                        Widget:= texture_block :
                                                                                                            DefaultImage:= GoldTexture
                                                                                                            DefaultDesiredSize:=vector2{X:=64.0,Y:=64.0}
                                                                                        stack_box_slot:
                                                                                            HorizontalAlignment:=horizontal_alignment.Center
                                                                                            VerticalAlignment:=vertical_alignment.Center
                                                                                            Widget:=Slot4GoldCost
                                                                            overlay_slot:
                                                                                HorizontalAlignment:=horizontal_alignment.Left
                                                                                VerticalAlignment:=vertical_alignment.Center
                                                                                Widget:=stack_box:
                                                                                    Orientation:=orientation.Horizontal
                                                                                    Slots:=array:
                                                                                        stack_box_slot:
                                                                                            HorizontalAlignment:=horizontal_alignment.Center
                                                                                            VerticalAlignment:=vertical_alignment.Center
                                                                                            Widget:=Slot4LevelText
                                                                                        stack_box_slot:
                                                                                            Widget:=overlay:
                                                                                                Slots:=array:
                                                                                                    overlay_slot:
                                                                                                        HorizontalAlignment:=horizontal_alignment.Fill
                                                                                                        VerticalAlignment:=vertical_alignment.Fill
                                                                                                        Widget:=Slot4Icon
                                                                                        stack_box_slot:
                                                                                            HorizontalAlignment:=horizontal_alignment.Fill
                                                                                            VerticalAlignment:=vertical_alignment.Center
                                                                                            Widget:=Slot4TypeText   
                                                                 ###################################################################                            
                                                                stack_box_slot: # Conversion5
                                                                    HorizontalAlignment:=horizontal_alignment.Fill
                                                                    VerticalAlignment:=vertical_alignment.Fill
                                                                    Padding:=margin{Left:=0.0,Top:=10.0,Right:=0.0,Bottom:=0.0}
                                                                    Widget:=overlay:
                                                                        Slots:=array:
                                                                            overlay_slot:
                                                                                HorizontalAlignment:=horizontal_alignment.Fill
                                                                                VerticalAlignment:=vertical_alignment.Fill
                                                                                Widget:=texture_block:
                                                                                    DefaultImage:=wooden_pixel_gui.textures.Panels.BrownSquare
                                                                                    DefaultTint := MakeColorFromHex("A9A9A9FF")
                                                                            overlay_slot:
                                                                                HorizontalAlignment:=horizontal_alignment.Right
                                                                                VerticalAlignment:=vertical_alignment.Center
                                                                                Padding:=margin{Left:=0.0,Top:=0.0,Right:=20.0,Bottom:=0.0}
                                                                                Widget:=Slot5Button
                                                                            overlay_slot:
                                                                                HorizontalAlignment:=horizontal_alignment.Center
                                                                                VerticalAlignment:=vertical_alignment.Center
                                                                                Widget:=stack_box:
                                                                                    Orientation:=orientation.Horizontal
                                                                                    Slots:=array:
                                                                                        
                                                                                        stack_box_slot:
                                                                                            Widget:=overlay:
                                                                                                Slots:=array:
                                                                                                    overlay_slot:
                                                                                                        HorizontalAlignment:=horizontal_alignment.Fill
                                                                                                        VerticalAlignment:=vertical_alignment.Fill
                                                                                                        Widget:=Slot5OreIcon
                                                                                        stack_box_slot:
                                                                                            HorizontalAlignment:=horizontal_alignment.Center
                                                                                            VerticalAlignment:=vertical_alignment.Center
                                                                                            Widget:=Slot5OreCost
                                                                                        stack_box_slot:
                                                                                            Widget:=overlay:
                                                                                                Slots:=array:
                                                                                                    overlay_slot:
                                                                                                        HorizontalAlignment:=horizontal_alignment.Fill
                                                                                                        VerticalAlignment:=vertical_alignment.Fill
                                                                                                        Widget:= texture_block :
                                                                                                            DefaultImage:= GoldTexture
                                                                                                            DefaultDesiredSize:=vector2{X:=64.0,Y:=64.0}
                                                                                        stack_box_slot:
                                                                                            HorizontalAlignment:=horizontal_alignment.Center
                                                                                            VerticalAlignment:=vertical_alignment.Center
                                                                                            Widget:=Slot5GoldCost
                                                                            overlay_slot:
                                                                                HorizontalAlignment:=horizontal_alignment.Left
                                                                                VerticalAlignment:=vertical_alignment.Center
                                                                                Widget:=stack_box:
                                                                                    Orientation:=orientation.Horizontal
                                                                                    Slots:=array:
                                                                                        stack_box_slot:
                                                                                            HorizontalAlignment:=horizontal_alignment.Center
                                                                                            VerticalAlignment:=vertical_alignment.Center
                                                                                            Widget:=Slot5LevelText
                                                                                        stack_box_slot:
                                                                                            Widget:=overlay:
                                                                                                Slots:=array:
                                                                                                    overlay_slot:
                                                                                                        HorizontalAlignment:=horizontal_alignment.Fill
                                                                                                        VerticalAlignment:=vertical_alignment.Fill
                                                                                                        Widget:=Slot5Icon
                                                                                        stack_box_slot:
                                                                                            HorizontalAlignment:=horizontal_alignment.Fill
                                                                                            VerticalAlignment:=vertical_alignment.Center
                                                                                            Widget:=Slot5TypeText   
                                                                stack_box_slot:
                                                                    HorizontalAlignment:=horizontal_alignment.Right
                                                                    VerticalAlignment:=vertical_alignment.Bottom
                                                                    Padding:=margin{Left:=0.0,Top:=20.0,Right:=0.0,Bottom:=0.0}
                                                                    Widget:=CloseButton
        return MyCanvas

    CreateCenteredOverlay(TheWidget:widget):overlay=
        overlay:
            Slots:=array:
                overlay_slot:
                    Widget:= TheWidget
                    HorizontalAlignment:=horizontal_alignment.Center
                    VerticalAlignment:=vertical_alignment.Center

    ChangeImageTest(Wdata:widget_message):void=
        if(Player:=Wdata.Player, PlayerUI:=GetPlayerUI[Player],SelectedButton:=Wdata.Source,Agent:=agent[Player]):
            if(OreIcon := OreTextureMap[Agent]?[0]):
                OreIcon.SetImage(hud.Obsidian_icon512)

    UpgradeWeaponTypeLevel(Wdata:widget_message,id:int):void=

        if(Player:=Wdata.Player, PlayerUI:=GetPlayerUI[Player],SelectedButton:=Wdata.Source,Agent:=agent[Player]):
            if:
                PlayerData := Data[Agent]
            then:

                HasOre := CheckIfPlayerHadOre(Agent,PlayerData.PistolLevel)
                Price := CalculateGoldUpgradePrice(PlayerData.PistolLevel)
                if:
                    id=0
                    HasOre = true
                    PlayerData.Gold > Price * 1.0
                    #PlayerData.Gold >= CalculateGoldUpgradePrice(PlayerData.PistolLevel)
                    PlayerData.PistolLevel < 100
                    set Data[Player] = player_data{player_data_update<constructor>(PlayerData),PistolLevel := PlayerData.PistolLevel + 1, Gold := PlayerData.Gold-Price}
                then:
                    ConsumeOre(Agent,PlayerData.PistolLevel)
                    CompleteConversion(Agent)

                HasOre1 := CheckIfPlayerHadOre(Agent,PlayerData.RifleLevel)
                Price1 := CalculateGoldUpgradePrice(PlayerData.RifleLevel)
                if:
                    id=1
                    HasOre1 = true
                    PlayerData.Gold > Price1 * 1.0
                    PlayerData.RifleLevel < 100
                    set Data[Player] = player_data{player_data_update<constructor>(PlayerData),RifleLevel := PlayerData.RifleLevel + 1, Gold := PlayerData.Gold-Price1}
                then:
                    ConsumeOre(Agent,PlayerData.RifleLevel)
                    CompleteConversion(Agent)

                HasOre2 := CheckIfPlayerHadOre(Agent,PlayerData.SmgLevel)
                Price2 := CalculateGoldUpgradePrice(PlayerData.SmgLevel)
                if:
                    id=2
                    HasOre2 = true
                    PlayerData.Gold > (Price2 * 1.0)
                    PlayerData.SmgLevel < 100
                    set Data[Player] = player_data{player_data_update<constructor>(PlayerData),SmgLevel := PlayerData.SmgLevel + 1, Gold := PlayerData.Gold-Price2}
                then:
                    ConsumeOre(Agent,PlayerData.SmgLevel)
                    CompleteConversion(Agent)

                HasOre3 := CheckIfPlayerHadOre(Agent,PlayerData.ShotgunLevel)
                Price3 := CalculateGoldUpgradePrice(PlayerData.ShotgunLevel)
                if:
                    id=3
                    HasOre3 = true
                    PlayerData.Gold > Price3 * 1.0
                    PlayerData.ShotgunLevel < 100
                    set Data[Player] = player_data{player_data_update<constructor>(PlayerData),ShotgunLevel := PlayerData.ShotgunLevel + 1, Gold := PlayerData.Gold-Price3}
                then:
                    ConsumeOre(Agent,PlayerData.ShotgunLevel)
                    CompleteConversion(Agent)

                HasOre4 := CheckIfPlayerHadOre(Agent,PlayerData.SniperLevel)
                Price4 := CalculateGoldUpgradePrice(PlayerData.SniperLevel)
                if:
                    id=4
                    HasOre4 = true
                    PlayerData.Gold > Price4* 1.0
                    PlayerData.SniperLevel < 100
                    set Data[Player] = player_data{player_data_update<constructor>(PlayerData),SniperLevel := PlayerData.SniperLevel + 1, Gold := PlayerData.Gold-Price4}
                then:
                    ConsumeOre(Agent,PlayerData.SniperLevel)
                    CompleteConversion(Agent)

            UpdateUiText(Agent)


    UpdateUiText(Agent:agent):void=

        if:
            SlotLevel1: text_block = LevelMap[Agent]?[0]
            SlotLevel2: text_block = LevelMap[Agent]?[1]
            SlotLevel3: text_block = LevelMap[Agent]?[2]
            SlotLevel4: text_block = LevelMap[Agent]?[3]
            SlotLevel5: text_block = LevelMap[Agent]?[4]

            Player := player[Agent]
            PlayerData := Data[Agent]

            OreIcon1 : texture_block = OreTextureMap[Agent]?[0]
            OreIcon2 : texture_block = OreTextureMap[Agent]?[1]
            OreIcon3 : texture_block = OreTextureMap[Agent]?[2]
            OreIcon4 : texture_block = OreTextureMap[Agent]?[3]
            OreIcon5 : texture_block = OreTextureMap[Agent]?[4]

            OreAmountText1 : text_block = OreCostMap[Agent]?[0]
            OreAmountText2 : text_block = OreCostMap[Agent]?[1]
            OreAmountText3 : text_block = OreCostMap[Agent]?[2]
            OreAmountText4 : text_block = OreCostMap[Agent]?[3]
            OreAmountText5 : text_block = OreCostMap[Agent]?[4]

            GoldCostText1 : text_block = GoldCostMap[Agent]?[0]
            GoldCostText2 : text_block = GoldCostMap[Agent]?[1]
            GoldCostText3 : text_block = GoldCostMap[Agent]?[2]
            GoldCostText4 : text_block = GoldCostMap[Agent]?[3]
            GoldCostText5 : text_block = GoldCostMap[Agent]?[4]
            
        then:

            SlotLevel1.SetText(storedstring("Lv {PlayerData.PistolLevel}"))
            SlotLevel2.SetText(storedstring("Lv {PlayerData.RifleLevel}"))
            SlotLevel3.SetText(storedstring("Lv {PlayerData.SmgLevel}"))
            SlotLevel4.SetText(storedstring("Lv {PlayerData.ShotgunLevel}"))
            SlotLevel5.SetText(storedstring("Lv {PlayerData.SniperLevel}"))

            SetOreTypeTexture(OreIcon1,PlayerData.PistolLevel)
            SetOreTypeTexture(OreIcon2,PlayerData.RifleLevel)
            SetOreTypeTexture(OreIcon3,PlayerData.SmgLevel)
            SetOreTypeTexture(OreIcon4,PlayerData.ShotgunLevel)
            SetOreTypeTexture(OreIcon5,PlayerData.SniperLevel)

            OreAmountText1.SetText(storedstring("{GetOreAmount(PlayerData.PistolLevel)} "))
            OreAmountText2.SetText(storedstring("{GetOreAmount(PlayerData.RifleLevel)} "))
            OreAmountText3.SetText(storedstring("{GetOreAmount(PlayerData.SmgLevel)} "))
            OreAmountText4.SetText(storedstring("{GetOreAmount(PlayerData.ShotgunLevel)} "))
            OreAmountText5.SetText(storedstring("{GetOreAmount(PlayerData.SniperLevel)} "))

            
            GoldCostText1.SetText(storedstring("{formatter.Abbreviate(CalculateGoldUpgradePrice(PlayerData.PistolLevel)*1.0,false)}"))
            GoldCostText2.SetText(storedstring("{formatter.Abbreviate(CalculateGoldUpgradePrice(PlayerData.RifleLevel)*1.0,false)}"))
            GoldCostText3.SetText(storedstring("{formatter.Abbreviate(CalculateGoldUpgradePrice(PlayerData.SmgLevel)*1.0,false)}"))
            GoldCostText4.SetText(storedstring("{formatter.Abbreviate(CalculateGoldUpgradePrice(PlayerData.ShotgunLevel)*1.0,false)}"))
            GoldCostText5.SetText(storedstring("{formatter.Abbreviate(CalculateGoldUpgradePrice(PlayerData.SniperLevel)*1.0,false)}"))


            if(PlayerData.PistolLevel >= 100):
                GoldCostText1.SetText(storedstring("Max"))
                OreAmountText1.SetText(storedstring("Max"))
            if(PlayerData.RifleLevel >= 100):
                GoldCostText2.SetText(storedstring("Max"))
                OreAmountText2.SetText(storedstring("Max"))
            if(PlayerData.SmgLevel >= 100):
                GoldCostText3.SetText(storedstring("Max"))
                OreAmountText3.SetText(storedstring("Max"))
            if(PlayerData.ShotgunLevel >= 100):
                GoldCostText4.SetText(storedstring("Max"))
                OreAmountText4.SetText(storedstring("Max"))
            if(PlayerData.SniperLevel >= 100):
                GoldCostText5.SetText(storedstring("Max"))
                OreAmountText5.SetText(storedstring("Max"))


    


    SetOreTypeTexture(TextureBlock:texture_block, Level:int):void=

        var OreType : int = 0
        set OreType = GetOreType(Level)

        if(OreType = 0):
            TextureBlock.SetImage(hud.Copper_icon512)
        else:
            if(OreType = 1):
                TextureBlock.SetImage(hud.Silver_icon)
            else:
                if(OreType = 2):
                    TextureBlock.SetImage(hud.Malachite_icon512)
                else:
                    if(OreType = 3):
                        TextureBlock.SetImage(hud.Obsidian_icon512)
                    else:
                        TextureBlock.SetImage(hud.Brightcore_icon512)

    GetOreType(Level:int):int=

        var OreType : int = 0
        if(Level < 20):
            set OreType = 0
        else:
            if(Level < 40):
                set OreType = 1
            else:
                if(Level < 60):
                    set OreType = 2
                else:
                    if(Level < 80):
                        set OreType = 3
                    else:
                        set OreType = 4

        return OreType

    CheckIfPlayerHadOre(Agent:agent,Level:int):logic=
        var OreType : int = 0
        set OreType = GetOreType(Level)
        OreAmount := GetOreAmount(Level)
        if(MyConditional : conditional_button_device = OreConditionals[OreType]):
            CurrentCount := MyConditional.GetItemCount(Agent,0)
            if(CurrentCount >= OreAmount):
                return true
        return false


    ConsumeOre(Agent:agent, Level:int):void=
    
        var OreType : int = 0
        set OreType = GetOreType(Level)

        if(MyConditional : conditional_button_device = OreConditionals[OreType]):

            OreAmount := GetOreAmount(Level)

            for(i:=0..OreAmount-1):
                MyConditional.Activate(Agent)


    GetOreAmount(Level:int):int=

        var OreCount : int = 0

        if(Level < 20):
            set OreCount = Level
        else:
            if(Level < 40):
                set OreCount = Level-19
            else:
                if(Level < 60):
                    set OreCount = Level-39

                else:
                    if(Level < 80):
                        set OreCount = Level-59
                    else:
                        set OreCount = Level-79
        return OreCount

    CalculateGoldUpgradePrice(CosmeticLevel:int):float=

        var Price : float = 0.0
        var InitialPrice : float = 98.0
        set Price += InitialPrice
        for(i:=0..CosmeticLevel-1):
            set Price = Price * 1.40
            
        return Price



                
    
    ShowUi(Agent:agent):void=

        RemoveFortHud(Agent)
        OpenAudio.Play(Agent)

        MyCanvas: canvas = MakeOreConverterBackground(Agent)
        
        if(PlayerUI:=GetPlayerUI[player[Agent]]):

            #Removing Old Widget to stop duplicates
            if(AddedCanvas := MyUI[Agent]?):
                PlayerUI.RemoveWidget(AddedCanvas)
                AddedCanvas.SetVisibility(widget_visibility.Visible)
                AddedCanvas.SetVisibility(widget_visibility.Hidden)

            PlayerUI.AddWidget(MyCanvas,player_ui_slot{InputMode:=ui_input_mode.All}) # want it to appear on top
            UpdateUiText(Agent)
            MyCanvas.SetVisibility(widget_visibility.Hidden)
            MyCanvas.SetVisibility(widget_visibility.Visible)
        if(set MyUI[Agent]= option{MyCanvas}){}

    RemoveHUD(Wdata:widget_message):void =

        Print("Removing")
        if(Player:=Wdata.Player, PlayerUI:=GetPlayerUI[Player],SelectedButton:=Wdata.Source,Agent:=agent[Player]):
            if(MyActualUI := MyUI[Agent]?):

                CloseAudio.Play(Agent)

                PlayerUI.RemoveWidget(MyActualUI)
                MyActualUI.SetVisibility(widget_visibility.Visible)
                MyActualUI.SetVisibility(widget_visibility.Hidden)

                if(set MyUI[Agent] = false){}
                if(set InventoryTextBlockPerAgent[Agent] = false){}
                if(set ConversionRateTextBlockPerAgent[Agent] = false){}
                
                spawn:
                    CloseCineDelay(Agent)


    CloseCineDelay(Agent:agent)<suspends>:void=
        ShowFortHud(Agent)


    CompleteConversion(Agent:agent):void= 
        spawn{DelayedUpdate(Agent)}
        PurchaseAudio.Play(Agent)
        UpdateUiText(Agent)
        MissionTrigger.Trigger(Agent)
        
    DelayedUpdate(Agent:agent)<suspends>:void=
        Sleep(0.5)
        UpdateUiText(Agent)

