
using { /Fortnite.com/Devices }
using { /Verse.org/Simulation }
using { /UnrealEngine.com/Temporary/Diagnostics }

achievements := class(creative_device):

    @editable
    ShowAchievementsButton : button_device = button_device{}
    @editable
    OpenAudio : audio_player_device = audio_player_device{}
    @editable
    CloseAudio : audio_player_device = audio_player_device{}

    var PlayerMap : [agent]custom_player = map{}

    @editable
    BoarTrigger : trigger_device = trigger_device{}
    @editable
    BuffaloTrigger : trigger_device = trigger_device{}
    @editable
    LizardTrigger : trigger_device = trigger_device{}
    @editable
    GiantTrigger : trigger_device = trigger_device{}
    @editable
    ScorpionTrigger : trigger_device = trigger_device{}
    @editable
    GolemTrigger : trigger_device = trigger_device{}
    @editable
    LeopardTrigger : trigger_device = trigger_device{}
    @editable
    RamTrigger : trigger_device = trigger_device{}
    @editable
    YetiTrigger : trigger_device = trigger_device{}
    @editable
    HoundTrigger : trigger_device = trigger_device{}
    @editable
    LavaJawTrigger : trigger_device = trigger_device{}
    @editable
    TerravoreTrigger : trigger_device = trigger_device{}
    @editable
    IronTuskTrigger : trigger_device = trigger_device{}
    @editable
    EmberClawTrigger : trigger_device = trigger_device{}
    @editable
    IceheartTrigger : trigger_device = trigger_device{}
    @editable
    BlazebornTrigger : trigger_device = trigger_device{}

    @editable
    CompletedQuest : trigger_device = trigger_device{}
    @editable
    CompletedAchievement : trigger_device = trigger_device{}

    @editable
    AchievementUnlockAccolade : accolades_device = accolades_device{}
    @editable
    AchievementHudMessage : hud_message_device = hud_message_device{}

    @editable
    PistolTrigger : trigger_device = trigger_device{} 
    @editable
    SmgTrigger : trigger_device = trigger_device{}
    @editable
    AssaultTrigger : trigger_device = trigger_device{}
    @editable
    ShotgunTrigger : trigger_device = trigger_device{}
    @editable
    SniperTrigger : trigger_device = trigger_device{}

    @editable
    TieredUpAchievenmtsCosmeticUnlock : trigger_device = trigger_device{}

    storedstring<localizes>(S:string) : message = "{S}"
    

    OnBegin<override>()<suspends>:void=

        ShowAchievementsButton.InteractedWithEvent.Subscribe(ShowAchievements)

        BoarTrigger.TriggeredEvent.Subscribe(IncrementBoarElims)
        BuffaloTrigger.TriggeredEvent.Subscribe(IncrementBuffaloElims)
        LizardTrigger.TriggeredEvent.Subscribe(IncrementLizardElims)
        GiantTrigger.TriggeredEvent.Subscribe(IncrementGiantElims)
        ScorpionTrigger.TriggeredEvent.Subscribe(IncrementScorpionElims)
        GolemTrigger.TriggeredEvent.Subscribe(IncrementGolemElims)
        LeopardTrigger.TriggeredEvent.Subscribe(IncrementLeopardElims)
        RamTrigger.TriggeredEvent.Subscribe(IncrementRamElims)
        YetiTrigger.TriggeredEvent.Subscribe(IncrementYetiElims)
        HoundTrigger.TriggeredEvent.Subscribe(IncrementHoundElims)
        LavaJawTrigger.TriggeredEvent.Subscribe(IncrementLavaJawElims)
        TerravoreTrigger.TriggeredEvent.Subscribe(IncrementTerravoreElims)
        IronTuskTrigger.TriggeredEvent.Subscribe(IncrementIronTuskElims)
        EmberClawTrigger.TriggeredEvent.Subscribe(IncrementEmberClawElims)
        IceheartTrigger.TriggeredEvent.Subscribe(IncrementIceheartElims)
        BlazebornTrigger.TriggeredEvent.Subscribe(IncrementBlazebornElims)

        CompletedQuest.TriggeredEvent.Subscribe(IncrementQuestsCompleted)
        CompletedAchievement.TriggeredEvent.Subscribe(IncrementAchievementsCompleted)

        PistolTrigger.TriggeredEvent.Subscribe(IncrementPistolEliminations)
        SmgTrigger.TriggeredEvent.Subscribe(IncrementSmgEliminations)
        AssaultTrigger.TriggeredEvent.Subscribe(IncrementAssaultEliminations)
        SniperTrigger.TriggeredEvent.Subscribe(IncrementSniperEliminations)
        ShotgunTrigger.TriggeredEvent.Subscribe(IncrementShotgunEliminations)

        AllPlayers := GetPlayspace().GetPlayers()
        for(Player:AllPlayers):
            OnPlayerSpawned(Player)
        Sleep(0.8)
        InitSpawners()
        #GetPlayspace().PlayerRemovedEvent().Subscribe(OnPlayerRemoved)


    InitSpawners():void=
        Spawners := FindCreativeObjectsWithTag(spawners{})
        for(Obj : Spawners):
            if(spawner := player_spawner_device[Obj]):
                spawner.SpawnedEvent.Subscribe(OnPlayerSpawned)
       

    OnPlayerSpawned(Agent:agent):void=
        if(Player := player[Agent]):
            if(PlayerExists := PlayerMap[Agent]):
                Print("Player Data Exists")
                PlayerExists.InitAchievements()
            else:
                MyCustomPlayer : custom_player = custom_player{Player:=Player}
                if(set PlayerMap[Agent] = MyCustomPlayer){}
                MyCustomPlayer.InitAchievements()


    ShowAchievements(Agent:agent):void=
        if (CustomPlayer := PlayerMap[Agent]):
            OpenAudio.Play(Agent)
            CustomPlayer.InitAchievements()
            CustomPlayer.ViewAchievements()

    IncrementBoarElims(Agent:?agent):void=
        if(MyAgent:=Agent?):
            if(CustomPlayer := PlayerMap[MyAgent]):
                TierUp := CustomPlayer.IncreaseTotalBoarEliminations()
                if (TierUp = true):
                    CompletedAchievement.Trigger(MyAgent)
                    if (ADName := CustomPlayer.AchievementsData["BoarEliminations"].TierDescs[ CustomPlayer.AchievementsData["BoarEliminations"].GetCurrentTier()-2]):
                        AchievemntCompleted(MyAgent,ADName)

    IncrementBuffaloElims(Agent:?agent):void=
        if(MyAgent:=Agent?):
            if(CustomPlayer := PlayerMap[MyAgent]):
                TierUp := CustomPlayer.IncreaseTotalBuffaloEliminations()
                if (TierUp = true):
                    CompletedAchievement.Trigger(MyAgent)
                    if (ADName := CustomPlayer.AchievementsData["BuffaloEliminations"].TierDescs[ CustomPlayer.AchievementsData["BuffaloEliminations"].GetCurrentTier()-2]):
                        AchievemntCompleted(MyAgent,ADName)

    IncrementLizardElims(Agent:?agent):void=
        if(MyAgent:=Agent?):
            if(CustomPlayer := PlayerMap[MyAgent]):
                TierUp := CustomPlayer.IncreaseTotalLizardEliminations()
                if (TierUp = true):
                    CompletedAchievement.Trigger(MyAgent)
                    if (ADName := CustomPlayer.AchievementsData["LizardEliminations"].TierDescs[ CustomPlayer.AchievementsData["LizardEliminations"].GetCurrentTier()-2]):
                        AchievemntCompleted(MyAgent,ADName)
        
    IncrementGiantElims(Agent:?agent):void=
        Print(" Giant Elims in achievements verse")
        if(MyAgent:=Agent?):
            if(CustomPlayer := PlayerMap[MyAgent]):
                TierUp := CustomPlayer.IncreaseTotalGiantEliminations()
                if (TierUp = true):
                    CompletedAchievement.Trigger(MyAgent)
                    if (ADName := CustomPlayer.AchievementsData["GiantEliminations"].TierDescs[ CustomPlayer.AchievementsData["GiantEliminations"].GetCurrentTier()-2]):
                        AchievemntCompleted(MyAgent,ADName)

    IncrementScorpionElims(Agent:?agent):void=
        if(MyAgent:=Agent?):
            if(CustomPlayer := PlayerMap[MyAgent]):
                TierUp := CustomPlayer.IncreaseTotalScorpionEliminations()
                if (TierUp = true):
                    CompletedAchievement.Trigger(MyAgent)
                    if (ADName := CustomPlayer.AchievementsData["ScorpionEliminations"].TierDescs[ CustomPlayer.AchievementsData["ScorpionEliminations"].GetCurrentTier()-2]):
                        AchievemntCompleted(MyAgent,ADName)
    
    IncrementGolemElims(Agent:?agent):void=
        if(MyAgent:=Agent?):
            if(CustomPlayer := PlayerMap[MyAgent]):
                TierUp := CustomPlayer.IncreaseTotalGolemEliminations()
                if (TierUp = true):
                    CompletedAchievement.Trigger(MyAgent)
                    if (ADName := CustomPlayer.AchievementsData["GolemEliminations"].TierDescs[ CustomPlayer.AchievementsData["GolemEliminations"].GetCurrentTier()-2]):
                        AchievemntCompleted(MyAgent,ADName)

    IncrementLeopardElims(Agent:?agent):void=
        if(MyAgent:=Agent?):
            if(CustomPlayer := PlayerMap[MyAgent]):
                TierUp := CustomPlayer.IncreaseTotalLeopardEliminations()
                if (TierUp = true):
                    CompletedAchievement.Trigger(MyAgent)
                    if (ADName := CustomPlayer.AchievementsData["LeopardEliminations"].TierDescs[ CustomPlayer.AchievementsData["LeopardEliminations"].GetCurrentTier()-2]):
                        AchievemntCompleted(MyAgent,ADName)

    IncrementRamElims(Agent:?agent):void=
        if(MyAgent:=Agent?):
            if(CustomPlayer := PlayerMap[MyAgent]):
                TierUp := CustomPlayer.IncreaseTotalRamEliminations()
                if (TierUp = true):
                    CompletedAchievement.Trigger(MyAgent)
                    if (ADName := CustomPlayer.AchievementsData["RamEliminations"].TierDescs[ CustomPlayer.AchievementsData["RamEliminations"].GetCurrentTier()-2]):
                        AchievemntCompleted(MyAgent,ADName)

    IncrementYetiElims(Agent:?agent):void=
        if(MyAgent:=Agent?):
            if(CustomPlayer := PlayerMap[MyAgent]):
                TierUp := CustomPlayer.IncreaseTotalYetiEliminations()
                if (TierUp = true):
                    CompletedAchievement.Trigger(MyAgent)
                    if (ADName := CustomPlayer.AchievementsData["YetiEliminations"].TierDescs[ CustomPlayer.AchievementsData["YetiEliminations"].GetCurrentTier()-2]):
                        AchievemntCompleted(MyAgent,ADName)

    IncrementHoundElims(Agent:?agent):void=
        if(MyAgent:=Agent?):
            if(CustomPlayer := PlayerMap[MyAgent]):
                TierUp := CustomPlayer.IncreaseTotalHoundEliminations()
                if (TierUp = true):
                    CompletedAchievement.Trigger(MyAgent)
                    if (ADName := CustomPlayer.AchievementsData["HoundEliminations"].TierDescs[ CustomPlayer.AchievementsData["HoundEliminations"].GetCurrentTier()-2]):
                        AchievemntCompleted(MyAgent,ADName)

    IncrementLavaJawElims(Agent:?agent):void=
        if(MyAgent:=Agent?):
            if(CustomPlayer := PlayerMap[MyAgent]):
                TierUp := CustomPlayer.IncreaseTotalLavaJawEliminations()
                if (TierUp = true):
                    CompletedAchievement.Trigger(MyAgent)
                    if (ADName := CustomPlayer.AchievementsData["LavaJawEliminations"].TierDescs[ CustomPlayer.AchievementsData["LavaJawEliminations"].GetCurrentTier()-2]):
                        AchievemntCompleted(MyAgent,ADName)

    IncrementTerravoreElims(Agent:?agent):void=
        if(MyAgent:=Agent?):
            if(CustomPlayer := PlayerMap[MyAgent]):
                TierUp := CustomPlayer.IncreaseTotalTerravoreEliminations()
                if (TierUp = true):
                    CompletedAchievement.Trigger(MyAgent)
                    if (ADName := CustomPlayer.AchievementsData["TerravoreEliminations"].TierDescs[ CustomPlayer.AchievementsData["TerravoreEliminations"].GetCurrentTier()-2]):
                        AchievemntCompleted(MyAgent,ADName)

    IncrementIronTuskElims(Agent:?agent):void=
        if(MyAgent:=Agent?):
            if(CustomPlayer := PlayerMap[MyAgent]):
                TierUp := CustomPlayer.IncreaseTotalIronTuskEliminations()
                if (TierUp = true):
                    CompletedAchievement.Trigger(MyAgent)
                    if (ADName := CustomPlayer.AchievementsData["IronTuskEliminations"].TierDescs[ CustomPlayer.AchievementsData["IronTuskEliminations"].GetCurrentTier()-2]):
                        AchievemntCompleted(MyAgent,ADName)

    IncrementEmberClawElims(Agent:?agent):void=
        if(MyAgent:=Agent?):
            if(CustomPlayer := PlayerMap[MyAgent]):
                TierUp := CustomPlayer.IncreaseTotalEmberClawEliminations()
                if (TierUp = true):
                    CompletedAchievement.Trigger(MyAgent)
                    if (ADName := CustomPlayer.AchievementsData["EmberClawEliminations"].TierDescs[ CustomPlayer.AchievementsData["EmberClawEliminations"].GetCurrentTier()-2]):
                        AchievemntCompleted(MyAgent,ADName)

    IncrementIceheartElims(Agent:?agent):void=
        
        if(MyAgent:=Agent?):
            if(CustomPlayer := PlayerMap[MyAgent]):
                TierUp := CustomPlayer.IncreaseTotalIceheartEliminations()
                if (TierUp = true):
                    CompletedAchievement.Trigger(MyAgent)
                    if (ADName := CustomPlayer.AchievementsData["IceheartEliminations"].TierDescs[ CustomPlayer.AchievementsData["IceheartEliminations"].GetCurrentTier()-2]):
                        AchievemntCompleted(MyAgent,ADName)

    IncrementBlazebornElims(Agent:?agent):void=
        if(MyAgent:=Agent?):
            if(CustomPlayer := PlayerMap[MyAgent]):
                TierUp := CustomPlayer.IncreaseTotalBlazebornEliminations()
                if (TierUp = true):
                    CompletedAchievement.Trigger(MyAgent)
                    if (ADName := CustomPlayer.AchievementsData["BlazebornEliminations"].TierDescs[ CustomPlayer.AchievementsData["BlazebornEliminations"].GetCurrentTier()-2]):
                        AchievemntCompleted(MyAgent,ADName)

    IncrementQuestsCompleted(Agent:?agent):void=
        if(MyAgent:=Agent?):
            if(CustomPlayer := PlayerMap[MyAgent]):
                TierUp := CustomPlayer.IncreaseQuestsCompleted()
                if (TierUp = true):
                    CompletedAchievement.Trigger(MyAgent)
                    if (ADName := CustomPlayer.AchievementsData["QuestsCompleted"].TierDescs[ CustomPlayer.AchievementsData["QuestsCompleted"].GetCurrentTier()-2]):
                        AchievemntCompleted(MyAgent,ADName)

    IncrementAchievementsCompleted(Agent:?agent):void=
        if(MyAgent:=Agent?):
            if(CustomPlayer := PlayerMap[MyAgent]):
                TierUp := CustomPlayer.IncreaseAchievementsCompleted()
                if (TierUp = true):
                    CompletedAchievement.Trigger(MyAgent)
                    if (ADName := CustomPlayer.AchievementsData["AchievemntsCompleted"].TierDescs[ CustomPlayer.AchievementsData["AchievemntsCompleted"].GetCurrentTier()-2]):
                        AchievemntCompleted(MyAgent,ADName)
                        TieredUpAchievenmtsCosmeticUnlock.Trigger(MyAgent)

    IncrementPistolEliminations(Agent:?agent):void=
        if(MyAgent:=Agent?):
            if(CustomPlayer := PlayerMap[MyAgent]):
                Print("Pistol Elim ")
                TierUp := CustomPlayer.IncreasePistolEliminations()
                if (TierUp = true):
                    CompletedAchievement.Trigger(MyAgent)
                    if (ADName := CustomPlayer.AchievementsData["PistolEliminations"].TierDescs[ CustomPlayer.AchievementsData["PistolEliminations"].GetCurrentTier()-2]):
                        AchievemntCompleted(MyAgent,ADName)

    IncrementSmgEliminations(Agent:?agent):void=
        if(MyAgent:=Agent?):
            if(CustomPlayer := PlayerMap[MyAgent]):
                TierUp := CustomPlayer.IncreaseSmgEliminations()
                Print("SmgElim")
                if (TierUp = true):
                    CompletedAchievement.Trigger(MyAgent)
                    if (ADName := CustomPlayer.AchievementsData["SmgEliminations"].TierDescs[ CustomPlayer.AchievementsData["SmgEliminations"].GetCurrentTier()-2]):
                        AchievemntCompleted(MyAgent,ADName)
                
    IncrementAssaultEliminations(Agent:?agent):void=
        if(MyAgent:=Agent?):
            if(CustomPlayer := PlayerMap[MyAgent]):
                Print("Assault Elim ")
                TierUp := CustomPlayer.IncreaseAssaultEliminations()
                if (TierUp = true):
                    CompletedAchievement.Trigger(MyAgent)
                    if (ADName := CustomPlayer.AchievementsData["AssaultEliminations"].TierDescs[ CustomPlayer.AchievementsData["AssaultEliminations"].GetCurrentTier()-2]):
                        AchievemntCompleted(MyAgent,ADName)

    IncrementSniperEliminations(Agent:?agent):void=
        if(MyAgent:=Agent?):
            if(CustomPlayer := PlayerMap[MyAgent]):
                Print("Sniper Elim ")
                TierUp := CustomPlayer.IncreaseSniperEliminations()
                if (TierUp = true):
                    CompletedAchievement.Trigger(MyAgent)
                    if (ADName := CustomPlayer.AchievementsData["SniperEliminations"].TierDescs[ CustomPlayer.AchievementsData["SniperEliminations"].GetCurrentTier()-2]):
                        AchievemntCompleted(MyAgent,ADName)

    IncrementShotgunEliminations(Agent:?agent):void=
        if(MyAgent:=Agent?):
            if(CustomPlayer := PlayerMap[MyAgent]):
                Print("Shotgun Elim ")
                TierUp := CustomPlayer.IncreaseShotgunEliminations()
                if (TierUp = true):
                    CompletedAchievement.Trigger(MyAgent)
                    if (ADName := CustomPlayer.AchievementsData["ShotgunEliminations"].TierDescs[ CustomPlayer.AchievementsData["ShotgunEliminations"].GetCurrentTier()-2]):
                        AchievemntCompleted(MyAgent,ADName)
        
        

    AchievemntCompleted(Agent:agent,ADName:string):void=
        AchievementHudMessage.SetText(storedstring(ADName))
        AchievementHudMessage.Show(Agent)
        AchievementUnlockAccolade.Award(Agent)








