using { /Fortnite.com/Characters }
using { /Fortnite.com/Devices }
using { /Fortnite.com/Game }
using { /Fortnite.com/UI }
using { /Verse.org/Simulation }
using { /Verse.org/Random }
using { /Verse.org/Assets }
using { /Verse.org/Colors }
using { /UnrealEngine.com/Temporary/SpatialMath }
using { /UnrealEngine.com/Temporary/Diagnostics }
using { /UnrealEngine.com/Temporary/UI }


achievement_block := class():
    OuterOverlay<private> : overlay = overlay{}
    ColorBlockBG : color_block = color_block:
        DefaultColor := MakeColorFromHex("110803FF")
        #DefaultColor := MakeColorFromHex("0d256c")
        #DefaultColor := color{R:=13.0,G:=37.0,B:=108.0}
        DefaultDesiredSize := vector2:
            X := 900.0
            Y := 125.0

    ColorBlockTrim : color_block = color_block:
        DefaultColor := NamedColors.White
        DefaultDesiredSize := vector2:
            X := 5.0
            Y := 125.0

    TextBlockName : text_block = text_block:
        DefaultTextColor := NamedColors.White
    TextBlockDesc : text_block = text_block:
        DefaultTextColor := NamedColors.Gray
    TextBlockProgressText : text_block = text_block:
        DefaultTextColor := NamedColors.White
        
    var AName : string = "Pleb Master II"
    var ADesc : string = "Eliminate 50 Plebs"
    var AProgressText : string = "9999/10000"
    var ProgressPercent : float = 0.0

    var AProgressBar<private>: custom_progress_bar = custom_progress_bar:
        ForegroundColor := NamedColors.Yellow
        BackgroundColor := NamedColors.Gray
        BorderColor := NamedColors.White
        XSize := 450.0
        YSize := 10.0

    var CustomIcon : custom_icon = custom_icon:
        Scale := 0.25

    GetCustomIcon()<transacts>:custom_icon=CustomIcon

    SetName(V:string):void=set AName=V
    SetDesc(V:string):void=set ADesc=V
    SetProgressText(V:string):void=set AProgressText=V
    SetProgressPercent(V:float):void=set ProgressPercent=V

    SetIcon(V:texture):void=
        CustomIcon.SetMaybeIcon(option{V})

    SetBGTint(V:color):void=
        ColorBlockTrim.SetColor(V)
        CustomIcon.SetMaybeIconColor(option{V})

    GetParentWidget<public>() : overlay = OuterOverlay

    ResetProgressBar():void=
        AProgressBar.SetProgressNow(0.0)

    Update():void=
        TextBlockName.SetText(StringToText(AName))
        TextBlockDesc.SetText(StringToText(ADesc))
        TextBlockProgressText.SetText(StringToText(AProgressText))
        AProgressBar.SetProgress(ProgressPercent)
        CustomIcon.Update()

    UpdateAchievementData(AD : achievement_data):void=
        set AName = AD.GetNameTiered()
        set ADesc = AD.GetDesc()
        set AProgressText = AD.GetProgressText()
        set ProgressPercent = AD.GetProgressPercent()
        Update()

    Init<public>() : overlay  =
        TextBlockName.SetText(StringToText(AName))
        TextBlockDesc.SetText(StringToText(ADesc))
        TextBlockProgressText.SetText(StringToText(AProgressText))
        AProgressBar.SetShowBorder(false)
        AProgressBar.SetShowTextPercent(false)
        AProgressBar.Init()
        AProgressBar.SetProgress(ProgressPercent)
        CustomIcon.Init()

        OuterOverlay.AddWidget(
            overlay_slot:
                Widget := ColorBlockBG
                HorizontalAlignment := horizontal_alignment.Center
                VerticalAlignment := vertical_alignment.Center
        )

        OuterOverlay.AddWidget(
            overlay_slot:
                Widget := ColorBlockTrim
                HorizontalAlignment := horizontal_alignment.Left
                VerticalAlignment := vertical_alignment.Center
        )

        OuterOverlay.AddWidget(
            overlay_slot:
                Widget := TextBlockName
                Padding := margin{Left := 135.0, Top := 10.0}
                HorizontalAlignment := horizontal_alignment.Left
                VerticalAlignment := vertical_alignment.Top
        )
        OuterOverlay.AddWidget(
            overlay_slot:
                Widget := TextBlockDesc
                Padding := margin{Left := 135.0, Top := 40.0}
                HorizontalAlignment := horizontal_alignment.Left
                VerticalAlignment := vertical_alignment.Top
        )
        OuterOverlay.AddWidget(
            overlay_slot:
                Widget := CustomIcon.GetWidget()
                Padding := margin{Left := 5.0}
                HorizontalAlignment := horizontal_alignment.Left
                VerticalAlignment := vertical_alignment.Top
        )
        OuterOverlay.AddWidget(
            overlay_slot:
                Widget := AProgressBar.GetParentWidget()
                Padding := margin{Left := 135.0, Top := 85.0}
                HorizontalAlignment := horizontal_alignment.Left
                VerticalAlignment := vertical_alignment.Top
        )
        OuterOverlay.AddWidget(
            overlay_slot:
                Widget := TextBlockProgressText
                Padding := margin{Left := 600.0, Top := 70.0}
                HorizontalAlignment := horizontal_alignment.Left
                VerticalAlignment := vertical_alignment.Top
        )
        return OuterOverlay


achievements_ui := class():

    var MainTitle : string = "Hero Level 555"
    var SubTitle : string = "Milestones"
    var Page : int = 0
    TextBlockMainTitle : text_block = text_block:
        DefaultTextColor := NamedColors.White
    TextBlockSubTitle : text_block = text_block:
        DefaultTextColor := NamedColors.White
    var ABlocks : []achievement_block = array{}
        
    TextureBlock : texture_block = texture_block:
        DefaultImage := UI.Textures.TextureAchievementBGBrown
        DefaultDesiredSize := vector2:
            X := 1920.0
            Y := 1080.0

    ButtonLoud : button_regular = button_regular{}
    ButtonLeft : button_quiet = button_quiet{}
    ButtonRight : button_quiet = button_quiet{}

    SetMainTitle(V:string):void=set MainTitle=V
    SetSubTitle(V:string):void=set SubTitle=V
    
    CloseMenu<private> : event() = event(){}
    
    SetAchievementBlocks(V:[]achievement_block):void=set ABlocks = V

    InitMenu<public>(Player : player ,?FromState : string = "Lobby")<suspends> : void =
        ButtonLoud.SetText(StringToText("Exit"))
        ButtonLeft.SetText(StringToText("◀️"))
        ButtonRight.SetText(StringToText("▶️"))
        ButtonLeft.OnClick().Subscribe(IncrementPageDown) 
        ButtonRight.OnClick().Subscribe(IncrementPage) 
        TextBlockMainTitle.SetText(StringToText(MainTitle))
        TextBlockSubTitle.SetText(StringToText(SubTitle))
        spawn{ PresentMenu(Player, ?FromState := FromState) }
        CloseMenu.Await()

    PresentMenu<private>(Player : player, ?FromState : string = "Lobby")<suspends> : void =

        if ( PlayerUI := GetPlayerUI[Player]):
            CanvasBlocks := GenerateAchievementBlocks()
            Canvas : canvas = canvas:
                Slots := CanvasBlocks
            PlayerUI.AddWidget(Canvas, player_ui_slot{ InputMode := ui_input_mode.All})
            Canvas.SetVisibility(widget_visibility.Hidden)
            Canvas.SetVisibility(widget_visibility.Visible)
            ButtonLoud.OnClick().Subscribe(CloseMenuSignal)
            CloseMenu.Await()
            Print("CloseMenu")
            PlayerUI.RemoveWidget(Canvas)
            Canvas.SetVisibility(widget_visibility.Visible)
            Canvas.SetVisibility(widget_visibility.Hidden)


    GenerateAchievementBlocks():[]canvas_slot=

        var CanvasSlots :[]canvas_slot = array:
            canvas_slot:
                Anchors := anchors{Minimum := vector2{X := 0.5, Y := 0.5}, Maximum := vector2{X := 0.5, Y := 0.5}}
                Offsets := margin{Left := 0.0, Top := 0.0, Right := 1920.0, Bottom := 1080.0}
                Alignment := vector2{X := 0.5, Y := 0.5}
                Widget := TextureBlock
            canvas_slot:
                Anchors := anchors{Minimum := vector2{X := 0.5, Y := 0.5}, Maximum := vector2{X := 0.5, Y := 0.5}}
                Offsets := margin{Left := 850.0, Top := 500.0, Right := 200.0}
                Alignment := vector2{X := 1.0, Y := 0.5}
                Widget := ButtonLoud
            canvas_slot:
                Anchors := anchors{Minimum := vector2{X := 0.5, Y := 0.5}, Maximum := vector2{X := 0.5, Y := 0.5}}
                Offsets := margin{Left := -850.0, Top := -475.0, Right := 200.0, Bottom := 40.0}
                Alignment := vector2{X := 0.0, Y := 0.5}
                Widget := TextBlockMainTitle
            canvas_slot:
                Anchors := anchors{Minimum := vector2{X := 0.5, Y := 0.5}, Maximum := vector2{X := 0.5, Y := 0.5}}
                Offsets := margin{Left := -850.0, Top := -350.0, Right := 200.0, Bottom := 40.0}
                Alignment := vector2{X := 0.0, Y := 0.5}
                Widget := TextBlockSubTitle
            canvas_slot:
                Anchors := anchors{Minimum := vector2{X := 0.5, Y := 0.5}, Maximum := vector2{X := 0.5, Y := 0.5}}
                Offsets := margin{Left := -50.0, Top := 500.0}
                Alignment := vector2{X := 1.0, Y := 0.5}
                Widget := ButtonLeft
            canvas_slot:
                Anchors := anchors{Minimum := vector2{X := 0.5, Y := 0.5}, Maximum := vector2{X := 0.5, Y := 0.5}}
                Offsets := margin{Left := 50.0, Top := 500.0}
                Alignment := vector2{X := 0.0, Y := 0.5}
                Widget := ButtonRight

        var OffsetX : float = -475.0
        var OffsetY : float = -250.0
        var ACounter : int = 0 

        for(i:=(Page*12)..(Page*12)+11):
            if(ABlock := ABlocks[i]):
                CS := canvas_slot:
                        Anchors := anchors{Minimum := vector2{X := 0.5, Y := 0.5}, Maximum := vector2{X := 0.5, Y := 0.5}}
                        Offsets := margin{Left := OffsetX, Top := OffsetY, Right := 725.0, Bottom := 125.0}
                        Alignment := vector2{X := 0.5, Y := 0.5}
                        Widget := ABlock.GetParentWidget()
                set CanvasSlots += array{CS}
                set OffsetY += 130.0
                set ACounter += 1
                if (ACounter = 6):
                    set OffsetX = 475.0
                    set OffsetY = -250.0

        return CanvasSlots

    CloseMenuSignal(WM : widget_message):void=
        CloseMenu.Signal()

    IncrementPage(WData : widget_message):void=


        Print("AblocksLength = {ABlocks.Length}")
        TestingPage := Page + 1
        Print("testingPage = {TestingPage}")
        
        if(PageLength := ABlocks.Length/12):
            TotalPages := Ceil(PageLength)
            Print("TotalPages = {TotalPages}")
            if(TestingPage <= (TotalPages-1)):
                set Page += 1

                CloseMenu.Signal()
                ?FromState : string = "Lobby"
                Player:=WData.Player   
                spawn{InitMenu(Player, ?FromState := "Lobby")}

    IncrementPageDown(WData : widget_message):void=

        TestingPage := Page - 1
        Print("testingPage = {TestingPage}")

        if(TestingPage >= 0):
            set Page -= 1

            CloseMenu.Signal()
            ?FromState : string = "Lobby"
            Player:=WData.Player   
            spawn{InitMenu(Player, ?FromState := "Lobby")}



ConvertToRomanNumeral(Number : int):string =
    var Symbols : []string = {"M", "CM", "D", "CD", "C", "XC", "L", "XL", "X", "IX", "V", "IV", "I"}
    var Values : []int = {1000, 900, 500, 400, 100, 90, 50, 40, 10, 9, 5, 4, 1}

    var N : int = Number
    var RomanNumeral : string = ""

    loop:
        for (X:=0..N):
            if (N <= 0):
                break
            loop:
                for (SymbolIndex := 0..Symbols.Length):
                    if (SV := Values[SymbolIndex], Symbol := Symbols[SymbolIndex], N >= SV):
                        set RomanNumeral += Symbol
                        set N -= SV
                        break
                break
        break
    return RomanNumeral

achievement_data := class():
    Name : string
    TierDescs : []string
    TierTargets : []int
    var CurrentValue : int = 0
    var AchievementTier : int = 1
    ABlock : achievement_block = achievement_block{}
    GetBlock()<transacts>:achievement_block=ABlock

    GetName():string=Name
    GetNameTiered():string="{Name} {GetRomanNumeral()}"
    GetDesc():string=
        if (T := TierDescs[GetCurrentTier()-1]):
            return T
        return ""

    GetPrevName():string="{Name} {GetPrevRomanNumeral()}"

    GetPrevDesc():string=
        if (T := TierDescs[GetCurrentTier()-2]):
            return T
        return ""

    SetCurrentValue(V:int):logic=
        set CurrentValue=V
        var TierUp : logic = false
        CurrentTier := GetCurrentTier()
        if (CurrentTier > AchievementTier):
            set AchievementTier = CurrentTier
            set TierUp = true
        #Print("AchivementTier: {AchievementTier} {CurrentTier}")
        return TierUp

    Init():void=
        ABlock.Init()
    
    Update():void=
        if (GetCurrentTier() = 2):
            GetBlock().SetBGTint(MakeColorFromHex("82bc3b"))
        else if (GetCurrentTier() = 3):
            GetBlock().SetBGTint(MakeColorFromHex("15a1cd"))
        else if (GetCurrentTier() = 4):
            GetBlock().SetBGTint(MakeColorFromHex("be54db"))
        else if (GetCurrentTier() = 5):
            GetBlock().SetBGTint(NamedColors.Orange)
        else if (GetCurrentTier() >= 6):
            GetBlock().SetBGTint(NamedColors.Gold)
        ABlock.UpdateAchievementData(Self)

    SetIcon(V:texture):void=
        GetBlock().SetIcon(V)

    GetProgressText():string=
        var CurrentTarget : int = 1
        if (T := TierTargets[GetCurrentTier()-1]):
            set CurrentTarget = T
        if (GetCurrentTier() > TierTargets.Length):
            return "{CurrentValue}"
        return "{CurrentValue}/{CurrentTarget}"

    GetProgressPercent():float=
        if (GetCurrentTier() > TierTargets.Length):
            return 100.0
        var PercentProgress : float = 0.0
        if (T := TierTargets[GetCurrentTier()-1], TD := ((CurrentValue*1.0) / (T*1.0))):
            set PercentProgress = TD*100.0

        return PercentProgress

    GetPrevRomanNumeral():string=
        return ConvertToRomanNumeral(GetCurrentTier()-1)

    GetRomanNumeral():string=
        if (GetCurrentTier() > TierTargets.Length):
            return "MAX"
        return ConvertToRomanNumeral(GetCurrentTier())

    GetCurrentTier()<transacts>:int=
        var CurrentTier : int = 1
        loop:
            for(K -> Tier : TierTargets):
                if (CurrentValue >= Tier):
                    set CurrentTier = K+2
                else:
                    break
            break
        return CurrentTier
