using { /UnrealEngine.com/Temporary/UI }
using { /Verse.org/Colors }
using { /Verse.org/Assets }
using { /UnrealEngine.com/Temporary/SpatialMath }

custom_icon := class():
    MainOverlay : overlay = overlay{}
    var MaybeIconColor : ?color = false
    var MaybeIcon : ?texture = false
    var Scale : float = 1.0
    TextureBlockBorder : texture_block = texture_block:
        DefaultImage := wooden_pixel_gui.textures.Panels.t_UniversalPanel3Outline
        DefaultDesiredSize := vector2:
            X := 512.0
            Y := 512.0 
    TextureBlockBG : texture_block = texture_block:
        DefaultImage := wooden_pixel_gui.textures.Panels.t_UniversalPanel3
        DefaultTint :=  DefaultTint:= MakeColorFromHex("261206FF")
        DefaultDesiredSize := vector2:
            X := 512.0
            Y := 512.0 

    TextureBlockIcon : texture_block = texture_block:
        DefaultImage := UI.Textures.TextureIconDefault
        DefaultDesiredSize := vector2:
            X := 256.0
            Y := 256.0 

    SetMaybeIcon(V:?texture):void=set MaybeIcon = V
    SetMaybeIconColor(V:?color):void=set MaybeIconColor = V

    GetIconTexture()<transacts><decides>:texture=MaybeIcon?

    Init():void=
        if (Scale <> 1.0):
            BorderSize := TextureBlockBorder.GetDesiredSize()
            BGSize := TextureBlockBG.GetDesiredSize()
            IconSize := TextureBlockIcon.GetDesiredSize()
            TextureBlockBorder.SetDesiredSize(BorderSize*Scale)
            TextureBlockBG.SetDesiredSize(BGSize*Scale)
            TextureBlockIcon.SetDesiredSize(IconSize*Scale)
        if (IconColor := MaybeIconColor?):
            TextureBlockBorder.SetTint(IconColor)
            TextureBlockIcon.SetTint(IconColor)
        if(Icon := MaybeIcon?):
            TextureBlockIcon.SetImage(Icon)


        MainOverlay.AddWidget(
            overlay_slot:
                Widget := TextureBlockBG
                HorizontalAlignment := horizontal_alignment.Center
                VerticalAlignment := vertical_alignment.Center
        )
        MainOverlay.AddWidget(
            overlay_slot:
                Widget := TextureBlockIcon
                HorizontalAlignment := horizontal_alignment.Center
                VerticalAlignment := vertical_alignment.Center
        )
        MainOverlay.AddWidget(
            overlay_slot:
                Widget := TextureBlockBorder
                HorizontalAlignment := horizontal_alignment.Center
                VerticalAlignment := vertical_alignment.Center
        )

    Update():void=
        if (IconColor := MaybeIconColor?):
            TextureBlockBorder.SetTint(IconColor)
            TextureBlockIcon.SetTint(IconColor)
        if(Icon := MaybeIcon?):
            TextureBlockIcon.SetImage(Icon)

    GetWidget():overlay=MainOverlay