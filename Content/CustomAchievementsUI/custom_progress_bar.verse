

using { /Fortnite.com/Devices }
using { /Fortnite.com/UI }
using { /Verse.org/Concurrency }
using { /Verse.org/Simulation }
using { /Verse.org/Colors }
using { /UnrealEngine.com/Temporary/SpatialMath }
using { /UnrealEngine.com/Temporary/UI }

StringToText<localizes>(String : string) : message = "{String}"

# Custom Progress Bar Class (Horizontal)
custom_progress_bar := class:
    var<private> BackgroundColor : color
    var<private> ForegroundColor : color
    var<private> BorderColor : color
    var<private> XSize : float
    var<private> YSize : float

    var<private> ShowBorder : logic = true
    var<private> ShowTextPercent : logic = true

    OuterOverlay<private> : overlay = overlay{}             # Outer Overlay holding every other widget (Add this to your canvas)
    ColorBlockBorder<private> : color_block = color_block{} # Border Color Block
    InnerOverlay<private> : overlay = overlay{}             # Overlay holding BG and FG Blocks ( So FG is on top of BG color)
    ColorBlockFG<private> : color_block = color_block{}     # Foreground Color
    ColorBlockBG<private> : color_block = color_block{}     # Background Color
    TextBlockProgress<private> : text_block = text_block{ DefaultTextColor := NamedColors.White} # Text Block to show percentage
    # Progress should be beween 0.0 and 100.0
    var<private> ProgressPercent<private> : float = 0.0
    # Signaled every time SetProgress() is called
    NewProgressUpdatePolledEvent<private> : event() = event(){}

    SetShowBorder(V:logic):void=set ShowBorder = V
    SetShowTextPercent(V:logic):void=set ShowTextPercent = V

    SetForegroundColor(V:color):void=
        set ForegroundColor=V
        ColorBlockFG.SetColor(ForegroundColor)

    # IMPORTANT: Call this before using the progress bar, otherwise it won't work lol
    # Block is transacts so that's why a manual Init method is required :/ 
    Init<public>() : overlay  =
        TextBlockProgress.SetTextOpacity(0.5)
        TextBlockProgress.SetText(StringToText("0%"))
        ColorBlockBG.SetDesiredSize(vector2{ X := XSize, Y := YSize})
        ColorBlockFG.SetDesiredSize(vector2{ X := 0.0, Y := YSize})
        ColorBlockBorder.SetDesiredSize(vector2{ X := XSize + (XSize * 0.09), Y := YSize + (YSize * 0.25)})

        ColorBlockBG.SetColor(BackgroundColor)
        ColorBlockFG.SetColor(ForegroundColor)
        ColorBlockBorder.SetColor(BorderColor)

        # ======== Widget Structure ========
        #  * Outer Overlay
        #   -> Border
        #   -> * Inner Overlay
        #        => Background Color Block
        #        => Foreground Color Block
        #   -> Progress Text Block
        InnerOverlay.AddWidget(     
            overlay_slot:
                Widget := ColorBlockBG
                HorizontalAlignment := horizontal_alignment.Left
                VerticalAlignment := vertical_alignment.Center
        )
        InnerOverlay.AddWidget(
            overlay_slot:
                Widget := ColorBlockFG
                HorizontalAlignment := horizontal_alignment.Left
                VerticalAlignment := vertical_alignment.Center
        )
        if (ShowBorder = true):
            OuterOverlay.AddWidget(
                overlay_slot:
                    Widget := ColorBlockBorder
                    HorizontalAlignment := horizontal_alignment.Center
                    VerticalAlignment := vertical_alignment.Center
            )
        OuterOverlay.AddWidget(
            overlay_slot:
                Widget := InnerOverlay
                HorizontalAlignment := horizontal_alignment.Center
                VerticalAlignment := vertical_alignment.Center
        )
        if (ShowTextPercent = true):
            OuterOverlay.AddWidget(
                overlay_slot:
                    Widget := TextBlockProgress
                    HorizontalAlignment := horizontal_alignment.Center
                    VerticalAlignment := vertical_alignment.Center
            )
        return OuterOverlay
        
    # Sets the progress bar to InProgress Percentage
    SetProgress<public>(InProgress : float) : task(void) =
        NewProgressUpdatePolledEvent.Signal()
        set ProgressPercent = InProgress
        # Clamp Progres Percent to 0-100%
        if (ProgressPercent > 100.0) then set ProgressPercent = 100.0
        if (ProgressPercent < 0.000) then set ProgressPercent = 000.0
        # Updates in x amount of intervals
        spawn{ UpdateBarLengthToProgressSmooth(60) }

    SetProgressNow<public>(InProgress : float):void=
        set ProgressPercent = InProgress
        # Clamp Progres Percent to 0-100%
        if (ProgressPercent > 100.0) then set ProgressPercent = 100.0
        if (ProgressPercent < 0.000) then set ProgressPercent = 000.0
        NewBarLength := (0.01 * ProgressPercent) * XSize # New size of Foreground Color based on Percentage
        NewSizeVec2 := vector2{ X := NewBarLength, Y := YSize}
        ColorBlockFG.SetDesiredSize(NewSizeVec2)
        if ( TextWholePercent := Floor[(NewSizeVec2.X / XSize) * 100.0]):
            TextBlockProgress.SetText(StringToText("{TextWholePercent}%"))

    UpdateBarLengthToProgressSmooth<private>(Intervals : type{ X : int where X > 10})<suspends> : void =
        NewBarLength := (0.01 * ProgressPercent) * XSize # New size of Foreground Color based on Percentage
        CurrentX := ColorBlockFG.GetDesiredSize().X      # Current Foreground Color Size
        race:
            NewProgressUpdatePolledEvent.Await() # If a change is made while the bar is being updated, cancel the previous update
            for (I := 1..Intervals):
                LerpParameter := (I * 1.0) / (Intervals * 1.0) # Used to Linearly Interpolate Current Size to New Size, based on current interval
                NewSizeVec2 := Lerp(vector2{ X := CurrentX, Y := YSize}, vector2{ X := NewBarLength, Y := YSize}, LerpParameter)
                ColorBlockFG.SetDesiredSize(NewSizeVec2)
                if ( TextWholePercent := Floor[(NewSizeVec2.X / XSize) * 100.0]):
                    TextBlockProgress.SetText(StringToText("{TextWholePercent}%"))
                # Sleep only on even iterations, thus making the lerp "smoother", but avoiding a Verse Crash
                if (Mod[I, 2] = 0) then Sleep(0.0)
                
    # Add this widget to your canvas
    GetParentWidget<public>() : overlay = OuterOverlay