using { /Verse.org/Simulation }

UI := module:
    (UI:)Textures<public> := module:
    (UI:)Materials<public> := module {}

custom_player := class():

    Player : player
    var AchievementBlocks : []achievement_block = array{}
    var AchievementsData : [string]achievement_data = map{}
    AUI : achievements_ui = achievements_ui{}

    var TotalBoarEliminations : int = 0
    var TotalBuffaloEliminations : int = 0
    var TotalLizardEliminations : int = 0
    var TotalGiantEliminations : int = 0
    var TotalScorpionEliminations : int = 0
    var TotalGolemEliminations : int = 0
    var TotalLeopardEliminations : int = 0
    var TotalRamEliminations : int = 0
    var TotalYetiEliminations : int = 0
    var TotalHoundEliminations : int = 0
    var TotalLavaJawEliminations : int = 0
    var TotalTerravoreEliminations : int = 0

    var TotalIronTuskEliminations : int = 0
    var TotalEmberClawEliminations : int = 0
    var TotalIceheartEliminations : int = 0
    var TotalBlazebornEliminations : int = 0
    
    var TotalPistolEliminations : int = 0
    var TotalSmgEliminations : int = 0
    var TotalAssaultEliminations : int = 0
    var TotalShotgunEliminations : int = 0
    var TotalSniperEliminations : int = 0

    var QuestsCompleted : int = 0
    var AchievementsCompleted : int = 0

    var TotalComingSoon : int = 0

    InitAchievements():void=

        LoadAchievements()

        AData1 : achievement_data = achievement_data:
            Name := "Boar Eliminations"
            TierDescs := array:
                "Eliminate 100 Boars"
                "Eliminate 250 Boars"
                "Eliminate 1000 Boars"
                "Eliminate 10000 Boars"
                "Eliminate 50000 Boars"
            TierTargets := {100,250,1000,10000,50000}
        AData1.Init()
        AData1.SetIcon(MissionsTrackers.BoarMission)
        AData1.SetCurrentValue(GetTotalBoarEliminations())
        AData1.Update()

        AData2 : achievement_data = achievement_data:
            Name := "Buffalo Eliminations"
            TierDescs := array:
                "Eliminate 100 Buffalo"
                "Eliminate 250 Buffalo"
                "Eliminate 1000 Buffalo"
                "Eliminate 10000 Buffalo"
                "Eliminate 50000 Buffalo"
            TierTargets := {100,250,1000,10000,50000}
        AData2.Init()
        AData2.SetIcon(MissionsTrackers.BuffaloMission)
        AData2.SetCurrentValue(GetTotalBuffaloEliminations())
        AData2.Update()

        AData4 : achievement_data = achievement_data:
            Name := "Lizard Eliminations"
            TierDescs := array:
                "Eliminate 100 Lizards"
                "Eliminate 250 Lizards"
                "Eliminate 1000 Lizards"
                "Eliminate 10000 Lizards"
                "Eliminate 50000 Lizards"
            TierTargets := {100,250,1000,10000,50000}
        AData4.Init()
        AData4.SetIcon(MissionsTrackers.LizardMission)
        AData4.SetCurrentValue(GetTotalLizardEliminations())
        AData4.Update()

        AData5 : achievement_data = achievement_data:
            Name := "Giant Beast Eliminations"
            TierDescs := array:
                "Eliminate 100 Giants"
                "Eliminate 250 Giants"
                "Eliminate 1000 Giants"
                "Eliminate 10000 Giants"
                "Eliminate 50000 Giants"
            TierTargets := {100,250,1000,10000,50000}
        AData5.Init()
        AData5.SetIcon(MissionsTrackers.DesertGiantMission)
        AData5.SetCurrentValue(GetTotalGiantEliminations())
        AData5.Update()

        AData6 : achievement_data = achievement_data:
            Name := "Scorpion Eliminations"
            TierDescs := array:
                "Eliminate 100 Scorpion"
                "Eliminate 250 Scorpion"
                "Eliminate 1000 Scorpion"
                "Eliminate 10000 Scorpion"
                "Eliminate 50000 Scorpion"
            TierTargets := {100,250,1000,10000,50000}
        AData6.Init()
        AData6.SetIcon(MissionsTrackers.Scorpion3)
        AData6.SetCurrentValue(GetTotalScorpionEliminations())
        AData6.Update()

        AData7 : achievement_data = achievement_data:
            Name := "Desert Golem Eliminations"
            TierDescs := array:
                "Eliminate 100 Golems"
                "Eliminate 250 Golems"
                "Eliminate 1000 Golems"
                "Eliminate 10000 Golems"
                "Eliminate 50000 Golems"
            TierTargets := {100,250,1000,10000,50000}

        AData7.Init()
        AData7.SetIcon(MissionsTrackers.DesertGolemMission)
        AData7.SetCurrentValue(GetTotalGolemEliminations())
        AData7.Update()

        AData8 : achievement_data = achievement_data:
            Name := "Snow Leopard Eliminations"
            TierDescs := array:
                "Eliminate 100 Snow Leopards"
                "Eliminate 250 Snow Leopards"
                "Eliminate 1000 Snow Leopards"
                "Eliminate 10000 Snow Leopards"
                "Eliminate 50000 Snow Leopards"
            TierTargets := {100,250,1000,10000,50000}
        AData8.Init()
        AData8.SetIcon(MissionsTrackers.LeopardMission)
        AData8.SetCurrentValue(GetTotalLeopardEliminations())
        AData8.Update()

        AData9 : achievement_data = achievement_data:
            Name := "Spider Eliminations"
            TierDescs := array:
                "Eliminate 100 Frozen Spiders"
                "Eliminate 250 Frozen Spiders"
                "Eliminate 1000 Frozen Spiders"
                "Eliminate 10000 Frozen Spiders"
                "Eliminate 50000 Frozen Spiders"
            TierTargets := {100,250,1000,10000,50000}

        AData9.Init()
        AData9.SetIcon(MissionsTrackers.SpiderMission)
        AData9.SetCurrentValue(GetTotalRamEliminations())
        AData9.Update()

        AData10 : achievement_data = achievement_data:
            Name := "Yeti Eliminations"
            TierDescs := array:
                "Eliminate 100 Yetis"
                "Eliminate 250 Yetis"
                "Eliminate 1000 Yetis"
                "Eliminate 10000 Yetis"
                "Eliminate 50000 Yetis"
            TierTargets := {100,250,1000,10000,50000}

        AData10.Init()
        AData10.SetIcon(MissionsTrackers.YetiMission)
        AData10.SetCurrentValue(GetTotalYetiEliminations())
        AData10.Update()


        AData11 : achievement_data = achievement_data:
            Name := "Wild Hound Eliminations"
            TierDescs := array:
                "Eliminate 100 Wild Hounds"
                "Eliminate 250 Wild Hounds"
                "Eliminate 1000 Wild Hounds"
                "Eliminate 10000 Wild Hounds"
                "Eliminate 50000 Wild Hounds"
            TierTargets := {100,250,1000,10000,50000}

        AData11.Init()
        AData11.SetIcon(MissionsTrackers.HoundMission)
        AData11.SetCurrentValue(GetTotalHoundEliminations())
        AData11.Update()

        AData12 : achievement_data = achievement_data:
            Name := "Lava Jaw Eliminations"
            TierDescs := array:
                "Eliminate 100 Lava Jaws"
                "Eliminate 250 Lava Jaws"
                "Eliminate 1000 Lava Jaws"
                "Eliminate 10000 Lava Jaws"
                "Eliminate 50000 Lava Jaws"
            TierTargets := {100,250,1000,10000,50000}
        
        AData12.Init()
        AData12.SetIcon(MissionsTrackers.LavaJawMission)
        AData12.SetCurrentValue(GetTotalLavaJawEliminations())
        AData12.Update()

        AData13 : achievement_data = achievement_data:
            Name := "Terravore Eliminations"
            TierDescs := array:
                "Eliminate 100 Terravores"
                "Eliminate 250 Terravores"
                "Eliminate 1000 Terravores"
                "Eliminate 10000 Terravores"
                "Eliminate 50000 Terravores"
            TierTargets := {100,250,1000,10000,50000}

        AData13.Init()
        AData13.SetIcon(MissionsTrackers.TerravoreMission)
        AData13.SetCurrentValue(GetTotalTerravoreEliminations())
        AData13.Update()

        AData14 : achievement_data = achievement_data:
            Name := "Iron Tusk Eliminations"
            TierDescs := array:
                "10 Iron Tusk Eliminations"
                "25 Iron Tusk Eliminations"
                "50 Iron Tusk Eliminations"
                "100 Iron Tusk Eliminations"
                "1000 Iron Tusk Eliminations"
            TierTargets := {10,25,50,100,1000}

        AData14.Init()
        AData14.SetIcon(RavenIcons.MapMarkers.SS526)
        AData14.SetCurrentValue(GetTotalIronTuskEliminations())
        AData14.Update()

        AData15 : achievement_data = achievement_data:
            Name := "Ember Claw Eliminations"
            TierDescs := array:
                "10 Poison Claw Eliminations"
                "25 Poison Claw Eliminations"
                "50 Poison Claw Eliminations"
                "100 Poison Claw Eliminations"
                "1000 Poison Claw Eliminations"
            TierTargets := {10,25,50,100,1000}

        AData15.Init()
        AData15.SetIcon(RavenIcons.MapMarkers.SS526)
        AData15.SetCurrentValue(GetTotalEmberClawEliminations())
        AData15.Update()

        AData16 : achievement_data = achievement_data:
            Name := "Iceheart Eliminations"
            TierDescs := array:
                "10 Iceheart Eliminations"
                "25 Iceheart Eliminations"
                "50 Iceheart Eliminations"
                "100 Iceheart Eliminations"
                "1000 Iceheart Eliminations"
            TierTargets := {10,25,50,100,1000}

        AData16.Init()
        AData16.SetIcon(RavenIcons.MapMarkers.SS526)
        AData16.SetCurrentValue(GetTotalIceheartEliminations())
        AData16.Update()

        AData17 : achievement_data = achievement_data:
            Name := "Blazeborn Eliminations"
            TierDescs := array:
                "10 Blazeborn Eliminations"
                "25 Blazeborn Eliminations"
                "50 Blazeborn Eliminations"
                "100 Blazeborn Eliminations"
                "1000 Blazeborn Eliminations"
            TierTargets := {10,25,50,100,1000}

        AData17.Init()
        AData17.SetIcon(RavenIcons.MapMarkers.SS528)
        AData17.SetCurrentValue(GetTotalBlazebornEliminations())
        AData17.Update()
    

        AData3 : achievement_data = achievement_data:
            Name := "Quests Completed"
            TierDescs := array:
                "Complete 10 Quests"
                "Complete 25 Quests"
                "Complete 50 Quests"
                "Complete 100 Quests"
                "Complete 1000 Quests"
            TierTargets := {10,25,50,100,1000}

        AData3.Init()
        AData3.SetIcon(RavenIcons.MapMarkers.S207)
        AData3.SetCurrentValue(GetQuestsCompleted())
        AData3.Update()

        AData18 : achievement_data = achievement_data:
            Name := "Achievement Tiers Completed"
            TierDescs := array:
                "Complete 10 Tiers"
                "Complete 25 Tiers"
                "Complete 50 Tiers"
                "Complete 75 Tiers"
                "Complete 90 Tiers"
            TierTargets := {10,25,50,75,90}

        AData18.Init()
        AData18.SetIcon(RavenIcons.MapMarkers.S15)
        AData18.SetCurrentValue(GetAchievementsCompleted())
        AData18.Update()

        AData19 : achievement_data = achievement_data:
            Name := "Pistol Eliminations"
            TierDescs := array:
                "100 Pistol Eliminations"
                "250 Pistol Eliminations"
                "1000 Pistol Eliminations"
                "10000 Pistol Eliminations"
                "50000 Pistol Eliminations"
            TierTargets := {100,250,1000,10000,50000}

        AData19.Init()
        AData19.SetIcon(hud.PistolIcon)
        AData19.SetCurrentValue(GetPistolEliminations())
        AData19.Update()

        AData20 : achievement_data = achievement_data:
            Name := "Smg Eliminations"
            TierDescs := array:
                "100 Smg Eliminations"
                "250 Smg Eliminations"
                "1000 Smg Eliminations"
                "10000 Smg Eliminations"
                "50000 Smg Eliminations"
            TierTargets := {100,250,1000,10000,50000}

        AData20.Init()
        AData20.SetIcon(hud.SmgIcon)
        AData20.SetCurrentValue(GetSmgEliminations())
        AData20.Update()

        AData21 : achievement_data = achievement_data:
            Name := "Assault Eliminations"
            TierDescs := array:
                "100 Assault Eliminations"
                "250 Assault Eliminations"
                "1000 Assault Eliminations"
                "10000 Assault Eliminations"
                "50000 Assault Eliminations"
            TierTargets := {100,250,1000,10000,50000}

        AData21.Init()
        AData21.SetIcon(hud.ArIcon)
        AData21.SetCurrentValue(GetAssaultEliminations())
        AData21.Update()

        AData22 : achievement_data = achievement_data:
            Name := "Shotgun Eliminations"
            TierDescs := array:
                "100 Shotgun Eliminations"
                "250 Shotgun Eliminations"
                "1000 Shotgun Eliminations"
                "10000 Shotgun Eliminations"
                "50000 Shotgun Eliminations"
            TierTargets := {100,250,1000,10000,50000}

        AData22.Init()
        AData22.SetIcon(hud.ShotgunIcon)
        AData22.SetCurrentValue(GetShotgunEliminations())
        AData22.Update()

        AData23 : achievement_data = achievement_data:
            Name := "Sniper Eliminations"
            TierDescs := array:
                "100 Sniper Eliminations"
                "250 Sniper Eliminations"
                "1000 Sniper Eliminations"
                "10000 Sniper Eliminations"
                "50000 Sniper Eliminations"
            TierTargets := {100,250,1000,10000,50000}

        AData23.Init()
        AData23.SetIcon(hud.SniperIcon)
        AData23.SetCurrentValue(GetSniperEliminations())
        AData23.Update()

        AData24 : achievement_data = achievement_data:
            Name := "More Coming Soon!"
            TierDescs := array:
                "Coming Soon"
                "Coming Soon"
                "Coming Soon"
                "Coming Soon"
                "Coming Soon"
            TierTargets := {100,250,1000,10000,50000}

        AData24.Init()
        AData24.SetIcon(Icons.DefaultIcon)
        AData24.SetCurrentValue(GetSniperEliminations())
        AData24.Update()

        if (set AchievementsData["BoarEliminations"] = AData1){}
        if (set AchievementsData["BuffaloEliminations"] = AData2){}
        if (set AchievementsData["LizardEliminations"] = AData4){}
        if (set AchievementsData["GiantEliminations"] = AData5){}
        if (set AchievementsData["ScorpionEliminations"] = AData6){}
        if (set AchievementsData["GolemEliminations"] = AData7){}
        if (set AchievementsData["LeopardEliminations"] = AData8){}
        if (set AchievementsData["RamEliminations"] = AData9){}
        if (set AchievementsData["YetiEliminations"] = AData10){}
        if (set AchievementsData["HoundEliminations"] = AData11){}
        if (set AchievementsData["LavaJawEliminations"] = AData12){}
        if (set AchievementsData["TerravoreEliminations"] = AData13){}
        if (set AchievementsData["IronTuskEliminations"] = AData14){}
        if (set AchievementsData["EmberClawEliminations"] = AData15){}
        if (set AchievementsData["IceheartEliminations"] = AData16){}
        if (set AchievementsData["BlazebornEliminations"] = AData17){}

        if (set AchievementsData["QuestsCompleted"] = AData3){}
        if (set AchievementsData["AchievemntsCompleted"] = AData18){}

        if (set AchievementsData["PistolEliminations"] = AData19){}
        if (set AchievementsData["SmgEliminations"] = AData20){}
        if (set AchievementsData["AssaultEliminations"] = AData21){}
        if (set AchievementsData["ShotgunEliminations"] = AData22){}
        if (set AchievementsData["SniperEliminations"] = AData23){}

        if (set AchievementsData["ComingSoon"] = AData24){}

        var ABlocks : []achievement_block = array{}
        for (AD : AchievementsData):
            set ABlocks += array{AD.GetBlock()}
        AUI.SetAchievementBlocks(ABlocks)

    GetTotalBoarEliminations():int=TotalBoarEliminations
    SetTotalBoarEliminations(V : int):void=set TotalBoarEliminations = V 
    IncreaseTotalBoarEliminations(?V:int=1):logic=
        SetTotalBoarEliminations(GetTotalBoarEliminations()+V)
        Print("Setting boar elims : {GetTotalBoarEliminations()}")
        if (AD := AchievementsData["BoarEliminations"]):
            TierUp := AD.SetCurrentValue(GetTotalBoarEliminations())
            if:
                MyData := Data[Player]
                set Data[Player] = player_data{player_data_update<constructor>(MyData),TotalBoarEliminations := TotalBoarEliminations}
            if (TierUp = true):
                IncreaseAchievementDamage(AD)
                return true
        return false
                # spawn{ADataUnlock("ZombieEliminations")}
                # GameManager.AwardAccolade("AchivementUnlock", Player

    GetTotalBuffaloEliminations():int=TotalBuffaloEliminations
    SetTotalBuffaloEliminations(V : int):void=set TotalBuffaloEliminations = V 
    IncreaseTotalBuffaloEliminations(?V:int=1):logic=
        SetTotalBuffaloEliminations(GetTotalBuffaloEliminations()+V)
        Print("Setting buffalo elims : {GetTotalBuffaloEliminations()}")
        if (AD := AchievementsData["BuffaloEliminations"]):
            TierUp := AD.SetCurrentValue(GetTotalBuffaloEliminations())
            if:
                MyData := Data[Player]
                set Data[Player] = player_data{player_data_update<constructor>(MyData),TotalBuffaloEliminations := TotalBuffaloEliminations}
            if (TierUp = true):
                IncreaseAchievementDamage(AD)
                return true
        return false

    GetTotalLizardEliminations():int=TotalLizardEliminations
    SetTotalLizardEliminations(V : int):void=set TotalLizardEliminations = V 
    IncreaseTotalLizardEliminations(?V:int=1):logic=
        SetTotalLizardEliminations(GetTotalLizardEliminations()+V)
        if (AD := AchievementsData["LizardEliminations"]):
            TierUp := AD.SetCurrentValue(GetTotalLizardEliminations())
            if:
                MyData := Data[Player]
                set Data[Player] = player_data{player_data_update<constructor>(MyData),TotalLizardEliminations := TotalLizardEliminations}
            if (TierUp = true):
                IncreaseAchievementDamage(AD)
                return true
        return false

    GetTotalGiantEliminations():int=TotalGiantEliminations
    SetTotalGiantEliminations(V : int):void=set TotalGiantEliminations = V 
    IncreaseTotalGiantEliminations(?V:int=1):logic=
        SetTotalGiantEliminations(GetTotalGiantEliminations()+V)
        if (AD := AchievementsData["GiantEliminations"]):
            TierUp := AD.SetCurrentValue(GetTotalGiantEliminations())
            if:
                MyData := Data[Player]
                set Data[Player] = player_data{player_data_update<constructor>(MyData),TotalGiantEliminations := TotalGiantEliminations}
            then:
                Print("Giant Elims Set{TotalGiantEliminations}")
            if (TierUp = true):
                IncreaseAchievementDamage(AD)
                return true
        return false

    GetTotalScorpionEliminations():int=TotalScorpionEliminations
    SetTotalScorpionEliminations(V : int):void=set TotalScorpionEliminations = V 
    IncreaseTotalScorpionEliminations(?V:int=1):logic=
        SetTotalScorpionEliminations(GetTotalScorpionEliminations()+V)
        if (AD := AchievementsData["ScorpionEliminations"]):
            TierUp := AD.SetCurrentValue(GetTotalScorpionEliminations())
            if:
                MyData := Data[Player]
                set Data[Player] = player_data{player_data_update<constructor>(MyData),TotalScorpionEliminations := TotalScorpionEliminations}
            if (TierUp = true):
                IncreaseAchievementDamage(AD)
                return true
        return false

    GetTotalGolemEliminations():int=TotalGolemEliminations
    SetTotalGolemEliminations(V : int):void=set TotalGolemEliminations = V 
    IncreaseTotalGolemEliminations(?V:int=1):logic=
        SetTotalGolemEliminations(GetTotalGolemEliminations()+V)
        if (AD := AchievementsData["GolemEliminations"]):
            TierUp := AD.SetCurrentValue(GetTotalGolemEliminations())
            if:
                MyData := Data[Player]
                set Data[Player] = player_data{player_data_update<constructor>(MyData),TotalGolemEliminations := TotalGolemEliminations}
            if (TierUp = true):
                IncreaseAchievementDamage(AD)
                return true
        return false

    GetTotalLeopardEliminations():int=TotalLeopardEliminations
    SetTotalLeopardEliminations(V : int):void=set TotalLeopardEliminations = V 
    IncreaseTotalLeopardEliminations(?V:int=1):logic=
        SetTotalLeopardEliminations(GetTotalLeopardEliminations()+V)
        if (AD := AchievementsData["LeopardEliminations"]):
            TierUp := AD.SetCurrentValue(GetTotalLeopardEliminations())
            if:
                MyData := Data[Player]
                set Data[Player] = player_data{player_data_update<constructor>(MyData),TotalLeopardEliminations := TotalLeopardEliminations}
            if (TierUp = true):
                IncreaseAchievementDamage(AD)
                return true
        return false

    GetTotalRamEliminations():int=TotalRamEliminations
    SetTotalRamEliminations(V : int):void=set TotalRamEliminations = V 
    IncreaseTotalRamEliminations(?V:int=1):logic=
        SetTotalRamEliminations(GetTotalRamEliminations()+V)
        if (AD := AchievementsData["RamEliminations"]):
            TierUp := AD.SetCurrentValue(GetTotalRamEliminations())
            if:
                MyData := Data[Player]
                set Data[Player] = player_data{player_data_update<constructor>(MyData),TotalRamEliminations := TotalRamEliminations}
            if (TierUp = true):
                IncreaseAchievementDamage(AD)
                return true
        return false
    
    GetTotalYetiEliminations():int=TotalYetiEliminations
    SetTotalYetiEliminations(V : int):void=set TotalYetiEliminations = V 
    IncreaseTotalYetiEliminations(?V:int=1):logic=
        SetTotalYetiEliminations(GetTotalYetiEliminations()+V)
        if (AD := AchievementsData["YetiEliminations"]):
            TierUp := AD.SetCurrentValue(GetTotalYetiEliminations())
            if:
                MyData := Data[Player]
                set Data[Player] = player_data{player_data_update<constructor>(MyData),TotalYetiEliminations := TotalYetiEliminations}
            if (TierUp = true):
                IncreaseAchievementDamage(AD)
                return true
        return false

    GetTotalHoundEliminations():int=TotalHoundEliminations
    SetTotalHoundEliminations(V : int):void=set TotalHoundEliminations = V 
    IncreaseTotalHoundEliminations(?V:int=1):logic=
        SetTotalHoundEliminations(GetTotalHoundEliminations()+V)
        if (AD := AchievementsData["HoundEliminations"]):
            TierUp := AD.SetCurrentValue(GetTotalHoundEliminations())
            if:
                MyData := Data[Player]
                set Data[Player] = player_data{player_data_update<constructor>(MyData),TotalHoundEliminations := TotalHoundEliminations}
            if (TierUp = true):
                IncreaseAchievementDamage(AD)
                return true
        return false

    GetTotalLavaJawEliminations():int=TotalLavaJawEliminations
    SetTotalLavaJawEliminations(V : int):void=set TotalLavaJawEliminations = V 
    IncreaseTotalLavaJawEliminations(?V:int=1):logic=
        SetTotalLavaJawEliminations(GetTotalLavaJawEliminations()+V)
        if (AD := AchievementsData["LavaJawEliminations"]):
            TierUp := AD.SetCurrentValue(GetTotalLavaJawEliminations())
            if:
                MyData := Data[Player]
                set Data[Player] = player_data{player_data_update<constructor>(MyData),TotalLavaJawEliminations := TotalLavaJawEliminations}
            if (TierUp = true):
                IncreaseAchievementDamage(AD)
                return true
        return false

    GetTotalTerravoreEliminations():int=TotalTerravoreEliminations
    SetTotalTerravoreEliminations(V : int):void=set TotalTerravoreEliminations = V 
    IncreaseTotalTerravoreEliminations(?V:int=1):logic=
        SetTotalTerravoreEliminations(GetTotalTerravoreEliminations()+V)
        if (AD := AchievementsData["TerravoreEliminations"]):
            TierUp := AD.SetCurrentValue(GetTotalTerravoreEliminations())
            if:
                MyData := Data[Player]
                set Data[Player] = player_data{player_data_update<constructor>(MyData),TotalTerravoreEliminations := TotalTerravoreEliminations}
            if (TierUp = true):
                IncreaseAchievementDamage(AD)
                return true
        return false

    GetTotalIronTuskEliminations():int=TotalIronTuskEliminations
    SetTotalIronTuskEliminations(V : int):void=set TotalIronTuskEliminations = V 
    IncreaseTotalIronTuskEliminations(?V:int=1):logic=
        SetTotalIronTuskEliminations(GetTotalIronTuskEliminations()+V)
        if (AD := AchievementsData["IronTuskEliminations"]):
            TierUp := AD.SetCurrentValue(GetTotalIronTuskEliminations())
            if:
                MyData := Data[Player]
                set Data[Player] = player_data{player_data_update<constructor>(MyData),TotalIronTuskEliminations := TotalIronTuskEliminations}
            if (TierUp = true):
                IncreaseAchievementDamage(AD)
                return true
        return false

    GetTotalEmberClawEliminations():int=TotalEmberClawEliminations
    SetTotalEmberClawEliminations(V : int):void=set TotalEmberClawEliminations = V 
    IncreaseTotalEmberClawEliminations(?V:int=1):logic=
        SetTotalEmberClawEliminations(GetTotalEmberClawEliminations()+V)
        if (AD := AchievementsData["EmberClawEliminations"]):
            TierUp := AD.SetCurrentValue(GetTotalEmberClawEliminations())
            if:
                MyData := Data[Player]
                set Data[Player] = player_data{player_data_update<constructor>(MyData),TotalEmberClawEliminations := TotalEmberClawEliminations}
            if (TierUp = true):
                IncreaseAchievementDamage(AD)
                return true
        return false

    GetTotalIceheartEliminations():int=TotalIceheartEliminations
    SetTotalIceheartEliminations(V : int):void=set TotalIceheartEliminations = V 
    IncreaseTotalIceheartEliminations(?V:int=1):logic=
        SetTotalIceheartEliminations(GetTotalIceheartEliminations()+V)
        if (AD := AchievementsData["IceheartEliminations"]):
            TierUp := AD.SetCurrentValue(GetTotalIceheartEliminations())
            if:
                MyData := Data[Player]
                set Data[Player] = player_data{player_data_update<constructor>(MyData),TotalIceheartEliminations := TotalIceheartEliminations}
            if (TierUp = true):
                IncreaseAchievementDamage(AD)
                return true
        return false

    GetTotalBlazebornEliminations():int=TotalBlazebornEliminations
    SetTotalBlazebornEliminations(V : int):void=set TotalBlazebornEliminations = V 
    IncreaseTotalBlazebornEliminations(?V:int=1):logic=
        SetTotalBlazebornEliminations(GetTotalBlazebornEliminations()+V)
        if (AD := AchievementsData["BlazebornEliminations"]):
            TierUp := AD.SetCurrentValue(GetTotalBlazebornEliminations())
            if:
                MyData := Data[Player]
                set Data[Player] = player_data{player_data_update<constructor>(MyData),TotalBlazebornEliminations := TotalBlazebornEliminations}
            if (TierUp = true):
                IncreaseAchievementDamage(AD)
                return true
        return false

    GetQuestsCompleted():int=QuestsCompleted
    SetQuestsCompleted(V : int):void=set QuestsCompleted = V
    IncreaseQuestsCompleted(?V:int=1):logic=
        SetQuestsCompleted(GetQuestsCompleted()+V)
        if (AD := AchievementsData["QuestsCompleted"]):
            TierUp := AD.SetCurrentValue(GetQuestsCompleted())
            if:
                MyData := Data[Player]
                set Data[Player] = player_data{player_data_update<constructor>(MyData),TotalQuestCompleted := QuestsCompleted}
            if (TierUp = true):
                IncreaseAchievementDamage(AD)
                return true
        return false

    GetAchievementsCompleted():int=AchievementsCompleted
    SetAchievementsCompleted(V : int):void=set AchievementsCompleted = V
    IncreaseAchievementsCompleted(?V:int=1):logic=
        SetAchievementsCompleted(GetAchievementsCompleted()+V)
        if (AD := AchievementsData["AchievementsCompleted"]):
            TierUp := AD.SetCurrentValue(GetAchievementsCompleted())
            if:
                MyData := Data[Player]
                set Data[Player] = player_data{player_data_update<constructor>(MyData),TotalAchievementsCompleted := AchievementsCompleted}
            if (TierUp = true):
                IncreaseAchievementDamage(AD)
                return true
        return false

    GetPistolEliminations():int=TotalPistolEliminations
    SetPistolEliminations(V : int):void=set TotalPistolEliminations = V
    IncreasePistolEliminations(?V:int=1):logic=
        SetPistolEliminations(GetPistolEliminations()+V)
        if (AD := AchievementsData["PistolEliminations"]):
            TierUp := AD.SetCurrentValue(GetPistolEliminations())
            if:
                MyData := Data[Player]
                set Data[Player] = player_data{player_data_update<constructor>(MyData),TotalPistolEliminations := TotalPistolEliminations}
            if (TierUp = true):
                IncreaseAchievementDamage(AD)
                return true
        return false

    GetSmgEliminations():int=TotalSmgEliminations
    SetSmgEliminations(V : int):void=set TotalSmgEliminations = V
    IncreaseSmgEliminations(?V:int=1):logic=
        SetSmgEliminations(GetSmgEliminations()+V)
        if (AD := AchievementsData["SmgEliminations"]):
            TierUp := AD.SetCurrentValue(GetSmgEliminations())
            if:
                MyData := Data[Player]
                set Data[Player] = player_data{player_data_update<constructor>(MyData),TotalSmgEliminations := TotalSmgEliminations}
            if (TierUp = true):
                IncreaseAchievementDamage(AD)
                return true
        return false

    GetAssaultEliminations():int=TotalAssaultEliminations
    SetAssaultEliminations(V : int):void=set TotalAssaultEliminations = V
    IncreaseAssaultEliminations(?V:int=1):logic=
        SetAssaultEliminations(GetAssaultEliminations()+V)
        if (AD := AchievementsData["AssaultEliminations"]):
            TierUp := AD.SetCurrentValue(GetAssaultEliminations())
            if:
                MyData := Data[Player]
                set Data[Player] = player_data{player_data_update<constructor>(MyData),TotalAssaultEliminations := TotalAssaultEliminations}
            if (TierUp = true):
                IncreaseAchievementDamage(AD)
                return true
        return false

    GetShotgunEliminations():int=TotalShotgunEliminations
    SetShotgunEliminations(V : int):void=set TotalShotgunEliminations = V
    IncreaseShotgunEliminations(?V:int=1):logic=
        SetShotgunEliminations(GetShotgunEliminations()+V)
        if (AD := AchievementsData["ShotgunEliminations"]):
            TierUp := AD.SetCurrentValue(GetShotgunEliminations())
            if:
                MyData := Data[Player]
                set Data[Player] = player_data{player_data_update<constructor>(MyData),TotalShotgunEliminations := TotalShotgunEliminations}
            if (TierUp = true):
                IncreaseAchievementDamage(AD)
                return true
        return false

    GetSniperEliminations():int=TotalSniperEliminations
    SetSniperEliminations(V : int):void=set TotalSniperEliminations = V
    IncreaseSniperEliminations(?V:int=1):logic=
        SetSniperEliminations(GetSniperEliminations()+V)
        if (AD := AchievementsData["SniperEliminations"]):
            TierUp := AD.SetCurrentValue(GetSniperEliminations())
            if:
                MyData := Data[Player]
                set Data[Player] = player_data{player_data_update<constructor>(MyData),TotalSniperEliminations := TotalSniperEliminations}
            if (TierUp = true):
                IncreaseAchievementDamage(AD)
                return true
        return false

    GetComingSoon():int=TotalComingSoon
    SetComingSonn(V : int):void=set TotalComingSoon = V


    ViewAchievements(?FromState : string = "Lobby"):void=
        for (AD : AchievementsData):
            AD.GetBlock().ResetProgressBar()
            AD.Update()
        if:
            MyData := Data[Player]
        then:
            AUI.SetMainTitle("Player Level: {MyData.Level}\nTotal Bonus Damage +{formatter.Abbreviate(MyData.AchievementsDamageBonus,true)}%")
        spawn{AUI.InitMenu(Player, ?FromState := FromState)}

    LoadAchievements():void=
        if(MyData := Data[Player]):

            SetTotalBoarEliminations(MyData.TotalBoarEliminations)
            SetTotalBuffaloEliminations(MyData.TotalBuffaloEliminations)
            SetTotalLizardEliminations(MyData.TotalLizardEliminations)
            SetTotalGiantEliminations(MyData.TotalGiantEliminations)
            SetTotalScorpionEliminations(MyData.TotalScorpionEliminations)
            SetTotalGolemEliminations(MyData.TotalGolemEliminations)
            SetTotalLeopardEliminations(MyData.TotalLeopardEliminations)
            SetTotalRamEliminations(MyData.TotalRamEliminations)
            SetTotalYetiEliminations(MyData.TotalYetiEliminations)
            SetTotalHoundEliminations(MyData.TotalHoundEliminations)
            SetTotalLavaJawEliminations(MyData.TotalLavaJawEliminations)
            SetTotalTerravoreEliminations(MyData.TotalTerravoreEliminations)
            SetTotalIronTuskEliminations(MyData.TotalIronTuskEliminations)
            SetTotalEmberClawEliminations(MyData.TotalEmberClawEliminations)
            SetTotalIceheartEliminations(MyData.TotalIceheartEliminations)
            SetTotalBlazebornEliminations(MyData.TotalBlazebornEliminations)
            SetQuestsCompleted(MyData.TotalQuestCompleted)

            SetPistolEliminations(MyData.TotalPistolEliminations)
            SetSmgEliminations(MyData.TotalSmgEliminations)
            SetAssaultEliminations(MyData.TotalAssaultEliminations)
            SetShotgunEliminations(MyData.TotalShotgunEliminations)
            SetSniperEliminations(MyData.TotalSniperEliminations)


    

    IncreaseAchievementDamage(AD:achievement_data):void=

        Tier := AD.GetCurrentTier()*1.0

        if:
            MyData := Data[Player]
            set Data[Player] = player_data{player_data_update<constructor>(MyData),AchievementsDamageBonus := MyData.AchievementsDamageBonus + (Tier * 2.5) }

            #if:
            #    AD.GetCurrentTier() = 2
            #    set Data[Player] = player_data{player_data_update<constructor>(MyData),Gold := MyData.Gold + 50}
            #if:
            #    AD.GetCurrentTier() = 3
            #    set Data[Player] = player_data{player_data_update<constructor>(MyData),Gold := MyData.Gold + 500}
            #if:
            #    AD.GetCurrentTier() = 4
            #    set Data[Player] = player_data{player_data_update<constructor>(MyData),Gold := MyData.Gold + 5000}
            #if:
            #    AD.GetCurrentTier() = 5
            #    set Data[Player] = player_data{player_data_update<constructor>(MyData),Gold := MyData.Gold + 50000}

        
